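# Modular Control Rig module that adds an animator control stack (optional nulls, secondary control) on a connected bone.
# Construction parents it via FK Parent metadata or connector, applies an offset, publishes metadata, adds child-bone sockets, then scales and colors controls.
# Forward solve constrains bones to controls and optionally nulls to bone parents.
# Backwards solve snaps controls to bones, also invoked before forward solve when RequiresBackwardSolve metadata is true.
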ControlRigBlueprint /ControlRigModules/Modules56/AddControl
graphs [
 ConstructionGraph {nodes [
   RigUnit_PrepareForExecution = &RigUnit_PrepareForExecution()
   CONSTR AddControl = RigVMFunctionReferenceNode(
    Add Control Primary Connector=(Type=Connector,Name="Add Control Primary")
    Parent Control Connector=(Type=Connector,Name="Parent Control")
    ExecuteContext=@RigUnit_PrepareForExecution.ExecutePin
    Settings=RigVMVariableNode() {variable=Control Settings}.Value
    Add Null=RigVMVariableNode() {variable=Add Null Above}.Value
    Add Bottom Null=RigVMVariableNode() {variable=Add Null Below}.Value
    Add Secondary Control=RigVMVariableNode() {variable=Add Secondary Control}.Value
    Offset Control=RigVMVariableNode() {variable=Offset Control}.Value
    Name=RigVMVariableNode() {variable=Control Name}.Value
    Orient to World=RigVMVariableNode() {variable=Orient to World}.Value
    Secondary Settings=RigVMVariableNode() {variable=Secondary Control Settings}.Value
    Control Suffix=RigVMVariableNode() {variable=Control Suffix}.Value
    Null Suffix=RigVMVariableNode() {variable=Null Suffix}.Value
   )
  ]}
 ContainedGraph {nodes [
   Entry = RigVMFunctionEntryNode()
   RigUnit_SequenceAggregate = &RigVMFunction_Sequence(ExecuteContext=@Entry.ExecuteContext)
   n1 = &RigVMFunction_ControlFlowBranch(Condition=@Entry.Negative, ExecuteContext=@RigUnit_SequenceAggregate.A)
   RigVMVariableNode(Value=1, ExecuteContext=@n1.False) {variable=isNegative}
   RigVMVariableNode(Value=-1, ExecuteContext=@n1.True) {variable=isNegative}
   n2 = RigVMDispatchNode(
    Array=RigVMDispatchNode(Array=@Entry.controls, Other=@Entry.ignore_ctrls) {template="DISPATCH_RigVMDispatch_ArrayDifference(in Array,in Other,out Result)"}.Result
    ExecuteContext=@RigUnit_SequenceAggregate.B
   ) {template="DISPATCH_RigVMDispatch_ArrayIterator(in Array,out Element,out Index,out Count,out Ratio)"}
   RigUnit_SequenceAggregate_1_1 = &RigVMFunction_Sequence(ExecuteContext=@n2.ExecuteContext)
   RigUnit_SequenceAggregate_1 = &RigVMFunction_Sequence(ExecuteContext=@RigUnit_SequenceAggregate_1_1.A)
   BranchNode = &RigVMFunction_ControlFlowBranch(Condition=@Entry.X Plane, ExecuteContext=@RigUnit_SequenceAggregate_1.A)
   BranchNode_1 = &RigVMFunction_ControlFlowBranch(ExecuteContext=@BranchNode.False, Condition=@Entry.Y Plane)
   BranchNode_2 = &RigVMFunction_ControlFlowBranch(ExecuteContext=@BranchNode_1.False, Condition=@Entry.Z Plane)
   RigVMVariableNode(
    ExecuteContext=@BranchNode_2.True
    Value=&RigUnit_GetTransform(Item=(Type=Control), bInitial=true, Item.Name=@n2.Element.Name).Transform.Translation.Z
   ) {variable=position}
   RigVMVariableNode(
    ExecuteContext=@BranchNode_1.True
    Value=&RigUnit_GetTransform(Item=(Type=Control), bInitial=true, Item.Name=@n2.Element.Name).Transform.Translation.Y
   ) {variable=position}
   RigVMVariableNode(
    ExecuteContext=@BranchNode.True
    Value=&RigUnit_GetTransform(Item=(Type=Control), bInitial=true, Item.Name=@n2.Element.Name).Transform.Translation.X
   ) {variable=position}
   n3 = &RigVMFunction_MathFloatMul(B=RigVMVariableNode() {variable=isNegative}.Value, A=RigVMVariableNode() {variable=position}.Value) {template="Multiply::Execute(in A,in B,out Result)"}
   n4 = &RigVMFunction_ControlFlowBranch(
    Condition=/Script/RigVM.RigVMFunction_MathFloatGreater(B=0, A=@n3.Result) {is=struct RigVMUnitNode, template="Greater::Execute(in A,in B,out Result)"}.Result
    ExecuteContext=@RigUnit_SequenceAggregate_1.B
   )
   /Script/ControlRig.RigUnit_SetControlColor(Control=@n2.Element.Name, ExecutePin=@n4.True, Color=@Entry.LeftColor) {is=struct RigVMUnitNode, template="SetControlColor::Execute(in Control,in Color)"}
   RigUnit_SequenceAggregate_1_2 = &RigVMFunction_Sequence(ExecuteContext=@RigUnit_SequenceAggregate_1_1.B)
   n5 = &RigVMFunction_ControlFlowBranch(
    Condition=/Script/RigVM.RigVMFunction_MathFloatLess(B=0, A=@n3.Result) {is=struct RigVMUnitNode, template="Less::Execute(in A,in B,out Result)"}.Result
    ExecuteContext=@RigUnit_SequenceAggregate_1_2.A
   )
   &RigUnit_SetControlColor(Control=@n2.Element.Name, ExecutePin=@n5.True, Color=@Entry.RightColor) {template="SetControlColor::Execute(in Control,in Color)"}
   MathFloatEquals = RigVMDispatchNode(B=0, A=@n3.Result) {template="DISPATCH_RigVMDispatch_CoreEquals(in A,in B,out Result)"}
   n6 = &RigVMFunction_ControlFlowBranch(
    Condition=&RigVMFunction_MathBoolOr(
     B=@MathFloatEquals.Result
     A=/Script/RigVM.RigVMFunction_MathFloatIsNearlyZero(Tolerance=0.001, Value=@n3.Result) {is=struct RigVMUnitNode, template="IsNearlyZero::Execute(in Value,in Tolerance,out Result)"}.Result
    ).Result
    ExecuteContext=@RigUnit_SequenceAggregate_1_2.B
   )
   &RigUnit_SetControlColor(Control=@n2.Element.Name, ExecutePin=@n6.True, Color=@Entry.CenterColor) {template="SetControlColor::Execute(in Control,in Color)"}
   RigVMFunctionReturnNode(ExecuteContext=@n2.Completed)
  ]}
 New Function_ContainedGraph {nodes [
   Entry = RigVMFunctionEntryNode()
   n1 = &RigVMFunction_Sequence(ExecuteContext=@Entry.ExecuteContext)
   RigVMFunction_Sequence = RigVMAggregateNode(ExecuteContext=@n1.A)
   n2 = RigVMDispatchNode(Default=(Type=Bone,Name="None"), Name=FK Parent, NameSpace=Parent) {template="DISPATCH_RigDispatch_GetModuleMetadata(in Name,in NameSpace,in Default,out Value,out Found)"}
   n3 = &RigVMFunction_ControlFlowBranch(Condition=@n2.Found, ExecuteContext=@RigVMFunction_Sequence.A)
   RigVMVariableNode(
    Value=/Script/ControlRig.RigUnit_ResolveConnector(SkipSocket=true, Connector=RigVMVariableNode() {variable=Parent Control Connector}.Value) {is=struct RigVMUnitNode}.Result
    ExecuteContext=@n3.False
   ) {variable=Parent Control}
   RigVMVariableNode(Value=@n2.Value, ExecuteContext=@n3.True) {variable=Parent Control}
   n4 = &RigUnit_ResolveConnector(SkipSocket=true, Connector=RigVMVariableNode() {variable=Add Control Primary Connector}.Value)
   Control Stack at Item = RigVMFunctionReferenceNode(
    Tag Controls=true
    Zero Controls=true
    Item=@n4.Result
    ExecuteContext=@RigVMFunction_Sequence.B
    Add Null=RigVMVariableNode() {variable=Add Null}.Value
    Name=RigVMVariableNode() {variable=Name}.Value
    Settings=RigVMVariableNode() {variable=Settings}.Value
    Add Bottom Null=RigVMVariableNode() {variable=Add Bottom Null}.Value
    Add Secondary Control=RigVMVariableNode() {variable=Add Secondary Control}.Value
    Parent=RigVMVariableNode() {variable=Parent Control}.Value
    Control Suffix=RigVMVariableNode() {variable=Control Suffix}.Value
    Secondary Settings=RigVMVariableNode() {variable=Secondary Settings}.Value
    Null Suffix=RigVMVariableNode() {variable=Null Suffix}.Value
    Orient to World=RigVMVariableNode() {variable=Orient to World}.Value
   )
   > DISPATCH_RigDispatch_SetMetadata = RigVMDispatchNode(Name=RequiresBackwardSolve, NameSpace=Root, Value=False, Item=@"Control Stack at Item".Control) {template="DISPATCH_RigDispatch_SetMetadata(in Item,in Name,in NameSpace,in Value,out Success)"}
   > RigVMDispatchNode(Name=SourceBone, NameSpace=Root, Item=@"Control Stack at Item".Control, Value=@n4.Result) {template="DISPATCH_RigDispatch_SetMetadata(in Item,in Name,in NameSpace,in Value,out Success)"}
   > DISPATCH_RigDispatch_SetModuleMetadata = RigVMDispatchNode(Name=Parent Control, NameSpace=Self, Value=@"Control Stack at Item".Control) {template="DISPATCH_RigDispatch_SetModuleMetadata(in Name,in NameSpace,in Value,out Success)"}
   > Offset Initial Transform = RigVMFunctionReferenceNode(Transform=RigVMVariableNode() {variable=Offset Control}.Value, Item=@"Control Stack at Item".Control)
   > RigVMDispatchNode(Name=FK Parent, NameSpace=Self, Value=@"Control Stack at Item".Control) {template="DISPATCH_RigDispatch_SetModuleMetadata(in Name,in NameSpace,in Value,out Success)"}
   n5 = RigVMVariableNode() {variable=Add Control Primary Connector}
   > DISPATCH_RigVMDispatch_ArrayIterator = RigVMDispatchNode(Array=/Script/ControlRig.RigUnit_CollectionChildrenArray(
     TypeToSearch=Bone
     bDefaultChildren=True
     bIncludeParent=False
     bRecursive=False
     Parent=&RigUnit_ResolveConnector(SkipSocket=true, Connector=@n5.Value).Result
    ) {is=struct RigVMUnitNode}.Items) {template="DISPATCH_RigVMDispatch_ArrayIterator(in Array,out Element,out Index,out Count,out Ratio)"}
   Concat = /Script/RigVM.RigVMFunction_NameConcat(B=_socket, A=@DISPATCH_RigVMDispatch_ArrayIterator.Element.Name) {is=struct RigVMUnitNode, template="Concat::Execute(in A,in B,out Result)"}
   > /Script/ControlRig.RigUnit_HierarchyAddSocket(
    Color=(R=1,G=1,B=1,A=1)
    Description=Add a module to this socket ...
    Space=LocalSpace
    Transform=(Rotation=(X=0,Y=0,Z=0,W=1),Translation=(X=0,Y=0,Z=0),Scale3D=(X=1,Y=1,Z=1))
    Name=@Concat.Result
    Parent=@DISPATCH_RigVMDispatch_ArrayIterator.Element
   ) {is=struct RigVMUnitNode}
   Finalize Controls = RigVMCollapseNode(ExecuteContext=@RigVMFunction_Sequence.C, Module Settings=RigVMVariableNode() {variable=Module Settings}.Value)
   RigVMFunctionReturnNode(ExecuteContext=@n1.B)
  ]}
 New Function_ContainedGraph {nodes [
   Entry = RigVMFunctionEntryNode()
   n1 = /Script/ControlRig.RigUnit_FindItemsWithMetadataTag(NameSpace=Self, Tag=Controls) {is=struct RigVMUnitNode}
   > Scale Module Controls = RigVMFunctionReferenceNode(
    Global Scale=RigVMDispatchNode(Default=1, Name=Control Size, NameSpace=Root) {template="DISPATCH_RigDispatch_GetModuleMetadata(in Name,in NameSpace,in Default,out Value,out Found)"}.Value
    Local Scale=@Entry.Module Settings.ControlSize_12_#1
    Controls=@n1.Items
   )
   > Color Module Controls = RigVMFunctionReferenceNode(
    CenterColor=@Entry.Module Settings.CenterColor_9_#2
    LeftColor=@Entry.Module Settings.LeftColor_3_#3
    Override=@Entry.Module Settings.OverrideColor_48_#4
    RightColor=@Entry.Module Settings.RightColor_5_#5
    Controls=@n1.Items
   )
   > RigVMFunctionReturnNode()
  ]}
 New Function_ContainedGraph {nodes [
   Entry = RigVMFunctionEntryNode()
   > RigVMFunction_Sequence = RigVMAggregateNode()
   n1 = RigVMDispatchNode(Array=&RigUnit_FindItemsWithMetadataTag(NameSpace=Self, Tag=Controls).Items, ExecuteContext=@RigVMFunction_Sequence.A) {template="DISPATCH_RigVMDispatch_ArrayIterator(in Array,out Element,out Index,out Count,out Ratio)"}
   DISPATCH_RigDispatch_GetMetadata = RigVMDispatchNode(Default=(Type=Bone,Name="None"), Name=Item, NameSpace=Self, Item=@n1.Element) {template="DISPATCH_RigDispatch_GetMetadata(in Item,in Name,in NameSpace,in Default,out Value,out Found)"}
   > n2 = &RigUnit_ParentConstraint(
    AdvancedSettings=(InterpolationType=Average,RotationOrderForFilter=XZY)
    Parents=((Item=(Type=Bone,Name="None"),Weight=1))
    Weight=1
    Child=@DISPATCH_RigDispatch_GetMetadata.Value
    Parents.0.Item=@n1.Element
   )
   RigVMFunction_ControlFlowBranch = &RigVMFunction_ControlFlowBranch(Condition=@Entry.Attach to Skeleton, ExecuteContext=@n2.ExecutePin)
   &RigUnit_ParentConstraint(
    AdvancedSettings=(InterpolationType=Average,RotationOrderForFilter=XZY)
    Parents=((Item=(Type=Bone,Name="None"),Weight=1))
    Weight=1
    Child=RigVMDispatchNode(Default=(Type=Bone,Name="None"), Name=Null, NameSpace=Self, Item=@n1.Element) {template="DISPATCH_RigDispatch_GetMetadata(in Item,in Name,in NameSpace,in Default,out Value,out Found)"}.Value
    Parents.0.Item=/Script/ControlRig.RigUnit_HierarchyGetParent(
     bDefaultParent=True
     Child=RigVMDispatchNode(Default=(Type=Bone,Name="None"), Name=Item, NameSpace=Self, Item=@n1.Element) {template="DISPATCH_RigDispatch_GetMetadata(in Item,in Name,in NameSpace,in Default,out Value,out Found)"}.Value
    ) {is=struct RigVMUnitNode}.Parent
    ExecutePin=@RigVMFunction_ControlFlowBranch.True
   )
   RigVMFunctionReturnNode(ExecuteContext=@RigVMFunction_Sequence.C)
  ]}
 New Function_ContainedGraph {nodes [
   Entry = RigVMFunctionEntryNode()
   Multiply = &RigVMFunction_MathTransformMul(B=@Entry.Transform, A=&RigUnit_GetTransform(Space=LocalSpace, bInitial=false, Item=@Entry.Item).Transform) {template="Multiply::Execute(in A,in B,out Result)"}
   > &RigUnit_SetTransform(Space=LocalSpace, Weight=1, bInitial=true, Item=@Entry.Item, Value=@Multiply.Result)
   > RigVMFunctionReturnNode()
  ]}
 New Function_ContainedGraph {nodes [
   Entry = RigVMFunctionEntryNode()
   Get_Module_Metadata = RigVMDispatchNode(Default=(Type=Bone,Name="None"), Name=Parent Control, NameSpace=Self) {template="DISPATCH_RigDispatch_GetModuleMetadata(in Name,in NameSpace,in Default,out Value,out Found)"}
   DISPATCH_RigDispatch_GetMetadata = RigVMDispatchNode(Default=False, Name=RequiresBackwardSolve, NameSpace=Root, Item=@Get_Module_Metadata.Value) {template="DISPATCH_RigDispatch_GetMetadata(in Item,in Name,in NameSpace,in Default,out Value,out Found)"}
   RigVMFunction_MathBoolAnd = &RigVMFunction_MathBoolAnd(B=@DISPATCH_RigDispatch_GetMetadata.Found, A=@DISPATCH_RigDispatch_GetMetadata.Value)
   RigVMFunction_ControlFlowBranch = &RigVMFunction_ControlFlowBranch(ExecuteContext=@Entry.ExecuteContext, Condition=@RigVMFunction_MathBoolAnd.Result)
   RigVMFunctionReturnNode(ExecuteContext=@RigVMFunction_ControlFlowBranch.Completed)
   RigVMInvokeEntryNode(Entry=Backwards Solve, ExecuteContext=@RigVMFunction_ControlFlowBranch.True)
  ]}
 RigVMFunctionLibrary {nodes [
   CONSTR AddControl = RigVMCollapseNode(
    Add Bottom Null=False
    Add Control Primary Connector=(Type=Connector,Name="Add Control Primary")
    Add Null=true
    Add Secondary Control=False
    Control Suffix=_ctrl
    ExecuteContext=()
    Null Suffix=_null
    Offset Control=(Rotation=(X=0,Y=0,Z=0,W=1),Translation=(X=0,Y=0,Z=0),Scale3D=(X=1,Y=1,Z=1))
    Parent Control Connector=(Type=Connector,Name="Parent Control")
    Secondary Settings="(DisplayName=\"None\",bSelectable=True,InitialSpace=LocalSpace,bUsePreferredRotationOrder=False,PreferredRotationOrder=YZX,Limits=(LimitTranslationX=(bMinimum=False,bMaximum=False),LimitTranslationY=(bMinimum=False,bMaximum=False),LimitTranslationZ=(bMinimum ...(truncated)"
    Settings="(DisplayName=\"None\",bSelectable=True,InitialSpace=LocalSpace,bUsePreferredRotationOrder=False,PreferredRotationOrder=YZX,Limits=(LimitTranslationX=(bMinimum=False,bMaximum=False),LimitTranslationY=(bMinimum=False,bMaximum=False),LimitTranslationZ=(bMinimum ...(truncated)"
   )
   ColorizeControls = RigVMCollapseNode(
    CenterColor=(R=1,G=1,B=0,A=1)
    ExecuteContext=()
    LeftColor=(R=0,G=1,B=0,A=1)
    RightColor=(R=1,G=0,B=0,A=1)
    X Plane=true
    controls=()
    ignore_ctrls=()
   )
   FWD AddControl = RigVMCollapseNode(Attach to Skeleton=False, ExecuteContext=())
   Offset Initial Transform = RigVMCollapseNode(ExecuteContext=(), Item=(Type=Bone,Name="None"), Transform=(Rotation=(X=0,Y=0,Z=0,W=1),Translation=(X=0,Y=0,Z=0),Scale3D=(X=1,Y=1,Z=1)))
   RunBackwardsIfRequired = RigVMCollapseNode(ExecuteContext=())
  ]}
 RigVMFunction_Sequence_1_ContainedGraph {nodes [
   Entry = RigVMFunctionEntryNode()
   RigVMFunction_Sequence = &RigVMFunction_Sequence(ExecuteContext=@Entry.ExecuteContext)
   n1 = &RigVMFunction_Sequence(ExecuteContext=@RigVMFunction_Sequence.B)
   n2 = &RigVMFunction_Sequence(ExecuteContext=@n1.B)
   RigVMFunctionReturnNode(A=@RigVMFunction_Sequence.A, B=@n1.A, C=@n2.A, D=@n2.B)
  ]}
 RigVMFunction_Sequence_1_ContainedGraph {nodes [
   Entry = RigVMFunctionEntryNode()
   RigVMFunction_Sequence = &RigVMFunction_Sequence(ExecuteContext=@Entry.ExecuteContext)
   n1 = &RigVMFunction_Sequence(ExecuteContext=@RigVMFunction_Sequence.B)
   RigVMFunctionReturnNode(A=@RigVMFunction_Sequence.A, B=@n1.A, C=@n1.B)
  ]}
 RigVMModel {nodes [
   n1 = /Script/ControlRig.RigUnit_PreBeginExecution() {is=struct RigVMUnitNode}
   RunBackwardsIfRequired = RigVMFunctionReferenceNode(ExecuteContext=@n1.ExecutePin)
   RigUnit_BeginExecution = &RigUnit_BeginExecution()
   n2 = &RigVMFunction_ControlFlowBranch(ExecuteContext=@RigUnit_BeginExecution.ExecutePin, Condition=RigVMVariableNode() {variable=Drive Skeleton}.Value)
   RigVMFunction_ControlFlowBranch = &RigVMFunction_ControlFlowBranch(ExecuteContext=@n2.Completed, Condition=RigVMVariableNode() {variable=Constrain Ctrl to Skeleton}.Value)
   n3 = RigVMDispatchNode(Array=&RigUnit_FindItemsWithMetadataTag(NameSpace=Self, Tag=Controls).Items, ExecuteContext=@n2.True) {template="DISPATCH_RigVMDispatch_ArrayIterator(in Array,out Element,out Index,out Count,out Ratio)"}
   &RigUnit_ParentConstraint(
    AdvancedSettings=(InterpolationType=Average,RotationOrderForFilter=XZY)
    Parents=((Item=(Type=Bone,Name="None"),Weight=1))
    Weight=1
    Child=RigVMDispatchNode(Default=(Type=Bone,Name="None"), Name=Null, NameSpace=Self, Item=@n3.Element) {template="DISPATCH_RigDispatch_GetMetadata(in Item,in Name,in NameSpace,in Default,out Value,out Found)"}.Value
    Parents.0.Item=&RigUnit_HierarchyGetParent(
     bDefaultParent=True
     Child=RigVMDispatchNode(Default=(Type=Bone,Name="None"), Name=Item, NameSpace=Self, Item=@n3.Element) {template="DISPATCH_RigDispatch_GetMetadata(in Item,in Name,in NameSpace,in Default,out Value,out Found)"}.Value
    ).Parent
    ExecutePin=@RigVMFunction_ControlFlowBranch.True
   )
   DISPATCH_RigDispatch_GetMetadata = RigVMDispatchNode(Default=(Type=Bone,Name="None"), Name=Item, NameSpace=Self, Item=@n3.Element) {template="DISPATCH_RigDispatch_GetMetadata(in Item,in Name,in NameSpace,in Default,out Value,out Found)"}
   &RigUnit_ParentConstraint(
    AdvancedSettings=(InterpolationType=Average,RotationOrderForFilter=XZY)
    Parents=((Item=(Type=Bone,Name="None"),Weight=1))
    Weight=1
    Child=@DISPATCH_RigDispatch_GetMetadata.Value
    Parents.0.Item=@n3.Element
    ExecutePin=@n3.ExecuteContext
   )
  ]}
 RigVMModel Backwards Solve Graph {nodes [
   n1 = /Script/ControlRig.RigUnit_InverseExecution() {is=struct RigVMUnitNode}
   DISPATCH_RigVMDispatch_ArrayIterator = RigVMDispatchNode(Array=&RigUnit_FindItemsWithMetadataTag(NameSpace=Self, Tag=Controls).Items, ExecuteContext=@n1.ExecutePin) {template="DISPATCH_RigVMDispatch_ArrayIterator(in Array,out Element,out Index,out Count,out Ratio)"}
   DISPATCH_RigDispatch_GetMetadata = RigVMDispatchNode(Default=(Type=Bone,Name="None"), Name=Item, NameSpace=Self, Item=@DISPATCH_RigVMDispatch_ArrayIterator.Element) {template="DISPATCH_RigDispatch_GetMetadata(in Item,in Name,in NameSpace,in Default,out Value,out Found)"}
   RigVMFunction_ControlFlowBranch = &RigVMFunction_ControlFlowBranch(Condition=@DISPATCH_RigDispatch_GetMetadata.Found, ExecuteContext=@DISPATCH_RigVMDispatch_ArrayIterator.ExecuteContext)
   &RigUnit_SetTransform(
    Weight=1
    bInitial=False
    Item=@DISPATCH_RigVMDispatch_ArrayIterator.Element
    Value=/Script/ControlRig.RigUnit_ProjectTransformToNewParent(
     bChildInitial=True
     bNewParentInitial=False
     bOldParentInitial=True
     NewParent=@DISPATCH_RigDispatch_GetMetadata.Value
     OldParent=@DISPATCH_RigDispatch_GetMetadata.Value
     Child=@DISPATCH_RigVMDispatch_ArrayIterator.Element
    ) {is=struct RigVMUnitNode}.Transform
    ExecutePin=@RigVMFunction_ControlFlowBranch.True
   )
   > DISPATCH_RigDispatch_SetMetadata = RigVMDispatchNode(Name=RequiresBackwardSolve, NameSpace=Root, Value=False, Item=@DISPATCH_RigVMDispatch_ArrayIterator.Element) {template="DISPATCH_RigDispatch_SetMetadata(in Item,in Name,in NameSpace,in Value,out Success)"}
  ]}
 RigVMModel Connector Event Graph {nodes [
   n1 = /Script/ControlRig.RigUnit_ConnectorExecution() {is=struct RigVMUnitNode}
   Connect to Module Metadata_1 = RigVMFunctionReferenceNode(
    Connector=(Type=Connector,Name="Parent Control")
    Module Metadata Name=Parent Control
    Module Metadata NameSpace=Parent
    ExecuteContext=@n1.ExecuteContext
   )
   > Connect to Module Metadata_1_1 = RigVMFunctionReferenceNode(Connector=(Type=Connector,Name="Parent Control"), Module Metadata Name=FK Parent, Module Metadata NameSpace=Parent)
  ]}
]
hierarchy [
 Add Control Primary {type=Connector, loc=<0 0 0>, rot=<0 0 0 1>}
 Root {type=Socket, loc=<0 0 0>, rot=<0 0 0 1>}
 Parent Control {type=Connector, loc=<0 0 0>, rot=<0 0 0 1>}
]
variables [
 Add Null Above {type=bool, default=True}
 Add Null Below {type=bool, default=False}
 Add Secondary Control {type=bool, default=False}
 Constrain Ctrl to Skeleton {type=bool, default=True}
 Control Name {type=FName, default=None}
 Control Settings {
  type=FRigUnit_HierarchyAddControlTransform_Settings
  default="(InitialSpace=LocalSpace,bUsePreferredRotationOrder=False,PreferredRotationOrder=YZX,Limits=(LimitTranslationX=(bMinimum=False,bMaximum=False),LimitTranslationY=(bMinimum=False,bMaximum=False),LimitTranslationZ=(bMinimum=False,bMaximum=False),LimitPitch=(b ...(truncated)"
 }
 Control Suffix {type=FName, default=_ctrl}
 Drive Skeleton {type=bool, default=True}
 Module Settings {
  type=FUserDefinedStruct_#6
  default="(OverrideColor_48_#4=False,LeftColor_3_#3=(R=0,G=0.0231,B=0.4115,A=1),RightColor_5_#5=(R=1,G=0,B=0,A=1),CenterColor_9_#2=(R=1,G=1,B=0,A= ...(truncated)"
 }
 Null Suffix {type=FName, default=_null}
 Offset Control {type=FTransform, default=(Rotation=(X=0,Y=0,Z=0,W=1),Translation=(X=0,Y=0,Z=0),Scale3D=(X=1,Y=1,Z=1))}
 Orient to World {type=bool, default=False}
 Parent Control {type=FRigElementKey, default=(Type=Connector,Name="")}
 Secondary Control Settings {
  type=FRigUnit_HierarchyAddControlTransform_Settings
  default="(InitialSpace=LocalSpace,bUsePreferredRotationOrder=False,PreferredRotationOrder=YZX,Limits=(LimitTranslationX=(bMinimum=False,bMaximum=False),LimitTranslationY=(bMinimum=False,bMaximum=False),LimitTranslationZ=(bMinimum=False,bMaximum=False),LimitPitch=(b ...(truncated)"
 }
 the_control {type=FRigElementKey, default=(Type=None,Name="")}
]
