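# AnimalSpine is a modular Control Rig module that builds an FK/IK spine rig for animals.
# Construction creates the spine controls and splines, adds main and hips-chest parent controls, and places attachment sockets on neck, thigh, tail, upperarm and shoulder bones.
# Spline functions fit bones to a control-driven curve with optional stretch, aim their up axis at an offset second spline for twist, and constrain the last bone to the end control.
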ControlRigBlueprint /ControlRigModules/Modules56/AnimalSpine
graphs [
 CollapseNode_ContainedGraph {nodes [
   n1 = RigVMFunctionEntryNode()
   > RigVMFunction_Sequence = RigVMAggregateNode()
   DISPATCH_RigVMDispatch_ArrayGetNum = RigVMDispatchNode(Array=@n1.Array) {template="DISPATCH_RigVMDispatch_ArrayGetNum(in Array,out Num)"}
   LessEqual = /Script/RigVM.RigVMFunction_MathIntLessEqual(B=3, A=@DISPATCH_RigVMDispatch_ArrayGetNum.Num) {is=struct RigVMUnitNode, template="LessEqual::Execute(in A,in B,out Result)"}
   DISPATCH_RigVMDispatch_If = RigVMDispatchNode(True=4, False=@DISPATCH_RigVMDispatch_ArrayGetNum.Num, Condition=@LessEqual.Result) {template="DISPATCH_RigVMDispatch_If(in Condition,in True,in False,out Result)"}
   DISPATCH_RigVMDispatch_ArraySetNum = RigVMDispatchNode(Num=@DISPATCH_RigVMDispatch_If.Result, ExecuteContext=@RigVMFunction_Sequence.A, Array=RigVMVariableNode() {variable=Positions}.Value) {template="DISPATCH_RigVMDispatch_ArraySetNum(io Array,in Num)"}
   > DISPATCH_RigVMDispatch_ArrayIterator = RigVMDispatchNode(Array=@n1.Array) {template="DISPATCH_RigVMDispatch_ArrayIterator(in Array,out Element,out Index,out Count,out Ratio)"}
   > RigVMDispatchNode(
    Index=@DISPATCH_RigVMDispatch_ArrayIterator.Index
    Element=&RigUnit_GetTransform(bInitial=False, Item=@DISPATCH_RigVMDispatch_ArrayIterator.Element).Transform.Translation
    Array=RigVMVariableNode() {variable=Positions}.Value
   ) {template="DISPATCH_RigVMDispatch_ArraySetAtIndex(io Array,in Index,in Element)"}
   n2 = RigVMDispatchNode(Array=@n1.Array) {template="DISPATCH_RigVMDispatch_ArrayGetNum(in Array,out Num)"}
   DISPATCH_RigVMDispatch_CoreEquals = RigVMDispatchNode(B=3, A=@n2.Num) {template="DISPATCH_RigVMDispatch_CoreEquals(in A,in B,out Result)"}
   RigVMFunction_ControlFlowBranch = &RigVMFunction_ControlFlowBranch(Condition=@DISPATCH_RigVMDispatch_CoreEquals.Result, ExecuteContext=@RigVMFunction_Sequence.B)
   n3 = &RigVMFunction_ControlFlowBranch(
    Condition=RigVMDispatchNode(B=2, A=@n2.Num) {template="DISPATCH_RigVMDispatch_CoreEquals(in A,in B,out Result)"}.Result
    ExecuteContext=@RigVMFunction_ControlFlowBranch.Completed
   )
   Greater = /Script/RigVM.RigVMFunction_MathIntGreater(B=4, A=@n2.Num) {is=struct RigVMUnitNode, template="Greater::Execute(in A,in B,out Result)"}
   n4 = &RigVMFunction_ControlFlowBranch(Condition=@Greater.Result, ExecuteContext=@n3.Completed)
   RigVMDispatchNode(
    Index=0
    Element=RigVMDispatchNode(Index=0, Array=RigVMVariableNode() {variable=Positions}.Value) {template="DISPATCH_RigVMDispatch_ArrayGetAtIndex(in Array,in Index,out Element)"}.Element
    ExecuteContext=@n4.True
    Array=RigVMVariableNode() {variable=Positions}.Value
   ) {template="DISPATCH_RigVMDispatch_ArraySetAtIndex(io Array,in Index,in Element)"}
   > RigVMDispatchNode(
    Index=1
    Element=RigVMDispatchNode(Index=1, Array=RigVMVariableNode() {variable=Positions}.Value) {template="DISPATCH_RigVMDispatch_ArrayGetAtIndex(in Array,in Index,out Element)"}.Element
    Array=RigVMVariableNode() {variable=Positions}.Value
   ) {template="DISPATCH_RigVMDispatch_ArraySetAtIndex(io Array,in Index,in Element)"}
   > RigVMDispatchNode(
    Index=2
    Element=RigVMDispatchNode(Index=2, Array=RigVMVariableNode() {variable=Positions}.Value) {template="DISPATCH_RigVMDispatch_ArrayGetAtIndex(in Array,in Index,out Element)"}.Element
    Array=RigVMVariableNode() {variable=Positions}.Value
   ) {template="DISPATCH_RigVMDispatch_ArraySetAtIndex(io Array,in Index,in Element)"}
   > RigVMDispatchNode(
    Index=3
    Element=RigVMDispatchNode(Index=-1, Array=RigVMVariableNode() {variable=Positions}.Value) {template="DISPATCH_RigVMDispatch_ArrayGetAtIndex(in Array,in Index,out Element)"}.Element
    Array=RigVMVariableNode() {variable=Positions}.Value
   ) {template="DISPATCH_RigVMDispatch_ArraySetAtIndex(io Array,in Index,in Element)"}
   RigVMDispatchNode(
    Index=0
    Element=RigVMDispatchNode(Index=0, Array=RigVMVariableNode() {variable=Positions}.Value) {template="DISPATCH_RigVMDispatch_ArrayGetAtIndex(in Array,in Index,out Element)"}.Element
    ExecuteContext=@n3.True
    Array=RigVMVariableNode() {variable=Positions}.Value
   ) {template="DISPATCH_RigVMDispatch_ArraySetAtIndex(io Array,in Index,in Element)"}
   > RigVMDispatchNode(
    Index=1
    Element=RigVMDispatchNode(Index=0, Array=RigVMVariableNode() {variable=Positions}.Value) {template="DISPATCH_RigVMDispatch_ArrayGetAtIndex(in Array,in Index,out Element)"}.Element
    Array=RigVMVariableNode() {variable=Positions}.Value
   ) {template="DISPATCH_RigVMDispatch_ArraySetAtIndex(io Array,in Index,in Element)"}
   > RigVMDispatchNode(
    Index=2
    Element=RigVMDispatchNode(Index=0, Array=RigVMVariableNode() {variable=Positions}.Value) {template="DISPATCH_RigVMDispatch_ArrayGetAtIndex(in Array,in Index,out Element)"}.Element
    Array=RigVMVariableNode() {variable=Positions}.Value
   ) {template="DISPATCH_RigVMDispatch_ArraySetAtIndex(io Array,in Index,in Element)"}
   > RigVMDispatchNode(
    Index=3
    Element=RigVMDispatchNode(Index=1, Array=RigVMVariableNode() {variable=Positions}.Value) {template="DISPATCH_RigVMDispatch_ArrayGetAtIndex(in Array,in Index,out Element)"}.Element
    Array=RigVMVariableNode() {variable=Positions}.Value
   ) {template="DISPATCH_RigVMDispatch_ArraySetAtIndex(io Array,in Index,in Element)"}
   DISPATCH_RigVMDispatch_ArrayGetAtIndex = RigVMDispatchNode(Index=0, Array=RigVMVariableNode() {variable=Positions}.Value) {template="DISPATCH_RigVMDispatch_ArrayGetAtIndex(in Array,in Index,out Element)"}
   DISPATCH_RigVMDispatch_ArraySetAtIndex = RigVMDispatchNode(
    Index=0
    Element=@DISPATCH_RigVMDispatch_ArrayGetAtIndex.Element
    ExecuteContext=@RigVMFunction_ControlFlowBranch.True
    Array=RigVMVariableNode() {variable=Positions}.Value
   ) {template="DISPATCH_RigVMDispatch_ArraySetAtIndex(io Array,in Index,in Element)"}
   > RigVMDispatchNode(
    Index=1
    Element=RigVMDispatchNode(Index=1, Array=RigVMVariableNode() {variable=Positions}.Value) {template="DISPATCH_RigVMDispatch_ArrayGetAtIndex(in Array,in Index,out Element)"}.Element
    Array=RigVMVariableNode() {variable=Positions}.Value
   ) {template="DISPATCH_RigVMDispatch_ArraySetAtIndex(io Array,in Index,in Element)"}
   > RigVMDispatchNode(
    Index=2
    Element=RigVMDispatchNode(Index=2, Array=RigVMVariableNode() {variable=Positions}.Value) {template="DISPATCH_RigVMDispatch_ArrayGetAtIndex(in Array,in Index,out Element)"}.Element
    Array=RigVMVariableNode() {variable=Positions}.Value
   ) {template="DISPATCH_RigVMDispatch_ArraySetAtIndex(io Array,in Index,in Element)"}
   > RigVMDispatchNode(
    Index=3
    Element=RigVMDispatchNode(Index=3, Array=RigVMVariableNode() {variable=Positions}.Value) {template="DISPATCH_RigVMDispatch_ArrayGetAtIndex(in Array,in Index,out Element)"}.Element
    Array=RigVMVariableNode() {variable=Positions}.Value
   ) {template="DISPATCH_RigVMDispatch_ArraySetAtIndex(io Array,in Index,in Element)"}
   n5 = &RigVMFunction_ControlFlowBranch(Condition=@n1.Debug, ExecuteContext=@RigVMFunction_Sequence.C)
   n6 = RigVMVariableNode() {variable=Positions}
   RigVMFunctionReturnNode(ExecuteContext=@n5.Completed, Value=@n6.Value)
   n7 = RigVMDispatchNode(ExecuteContext=@n5.True, Array=@n6.Value) {template="DISPATCH_RigVMDispatch_ArrayIterator(in Array,out Element,out Index,out Count,out Ratio)"}
   > RigVMFunction_DebugTransformMutableNoSpace = /Script/RigVM.RigVMFunction_DebugTransformMutableNoSpace(
    Color=(R=1,G=0,B=0,A=1)
    DebugDrawSettings=(DepthPriority=SDPG_Foreground,Lifetime=-1)
    Mode=Axes
    Scale=10
    Thickness=1
    Transform=(Rotation=(X=0,Y=0,Z=0,W=1),Translation=(X=0,Y=0,Z=0),Scale3D=(X=1,Y=1,Z=1))
    WorldOffset=(Rotation=(X=0,Y=0,Z=0,W=1),Translation=(X=0,Y=0,Z=0),Scale3D=(X=1,Y=1,Z=1))
    bEnabled=True
    Transform.Translation=@n7.Element
   ) {is=struct RigVMUnitNode}
  ]}
 CollapseNode_ContainedGraph {nodes [
   n1 = RigVMFunctionEntryNode()
   > RigVMFunction_Sequence = RigVMAggregateNode()
   DISPATCH_RigVMDispatch_ArrayGetNum = RigVMDispatchNode(Array=@n1.Array) {template="DISPATCH_RigVMDispatch_ArrayGetNum(in Array,out Num)"}
   LessEqual = &RigVMFunction_MathIntLessEqual(B=3, A=@DISPATCH_RigVMDispatch_ArrayGetNum.Num) {template="LessEqual::Execute(in A,in B,out Result)"}
   DISPATCH_RigVMDispatch_If = RigVMDispatchNode(True=4, False=@DISPATCH_RigVMDispatch_ArrayGetNum.Num, Condition=@LessEqual.Result) {template="DISPATCH_RigVMDispatch_If(in Condition,in True,in False,out Result)"}
   DISPATCH_RigVMDispatch_ArraySetNum = RigVMDispatchNode(Num=@DISPATCH_RigVMDispatch_If.Result, ExecuteContext=@RigVMFunction_Sequence.A, Array=RigVMVariableNode() {variable=Positions}.Value) {template="DISPATCH_RigVMDispatch_ArraySetNum(io Array,in Num)"}
   > DISPATCH_RigVMDispatch_ArrayIterator = RigVMDispatchNode(Array=@n1.Array) {template="DISPATCH_RigVMDispatch_ArrayIterator(in Array,out Element,out Index,out Count,out Ratio)"}
   > RigVMDispatchNode(
    Index=@DISPATCH_RigVMDispatch_ArrayIterator.Index
    Element=&RigUnit_GetTransform(bInitial=False, Item=@DISPATCH_RigVMDispatch_ArrayIterator.Element).Transform.Translation
    Array=RigVMVariableNode() {variable=Positions}.Value
   ) {template="DISPATCH_RigVMDispatch_ArraySetAtIndex(io Array,in Index,in Element)"}
   n2 = RigVMDispatchNode(Array=@n1.Array) {template="DISPATCH_RigVMDispatch_ArrayGetNum(in Array,out Num)"}
   DISPATCH_RigVMDispatch_CoreEquals = RigVMDispatchNode(B=3, A=@n2.Num) {template="DISPATCH_RigVMDispatch_CoreEquals(in A,in B,out Result)"}
   RigVMFunction_ControlFlowBranch = &RigVMFunction_ControlFlowBranch(Condition=@DISPATCH_RigVMDispatch_CoreEquals.Result, ExecuteContext=@RigVMFunction_Sequence.B)
   n3 = &RigVMFunction_ControlFlowBranch(
    Condition=RigVMDispatchNode(B=2, A=@n2.Num) {template="DISPATCH_RigVMDispatch_CoreEquals(in A,in B,out Result)"}.Result
    ExecuteContext=@RigVMFunction_ControlFlowBranch.Completed
   )
   Greater = &RigVMFunction_MathIntGreater(B=4, A=@n2.Num) {template="Greater::Execute(in A,in B,out Result)"}
   n4 = &RigVMFunction_ControlFlowBranch(Condition=@Greater.Result, ExecuteContext=@n3.Completed)
   RigVMDispatchNode(
    Index=0
    Element=RigVMDispatchNode(Index=0, Array=RigVMVariableNode() {variable=Positions}.Value) {template="DISPATCH_RigVMDispatch_ArrayGetAtIndex(in Array,in Index,out Element)"}.Element
    ExecuteContext=@n4.True
    Array=RigVMVariableNode() {variable=Positions}.Value
   ) {template="DISPATCH_RigVMDispatch_ArraySetAtIndex(io Array,in Index,in Element)"}
   > RigVMDispatchNode(
    Index=1
    Element=RigVMDispatchNode(Index=1, Array=RigVMVariableNode() {variable=Positions}.Value) {template="DISPATCH_RigVMDispatch_ArrayGetAtIndex(in Array,in Index,out Element)"}.Element
    Array=RigVMVariableNode() {variable=Positions}.Value
   ) {template="DISPATCH_RigVMDispatch_ArraySetAtIndex(io Array,in Index,in Element)"}
   > RigVMDispatchNode(
    Index=2
    Element=RigVMDispatchNode(Index=2, Array=RigVMVariableNode() {variable=Positions}.Value) {template="DISPATCH_RigVMDispatch_ArrayGetAtIndex(in Array,in Index,out Element)"}.Element
    Array=RigVMVariableNode() {variable=Positions}.Value
   ) {template="DISPATCH_RigVMDispatch_ArraySetAtIndex(io Array,in Index,in Element)"}
   > RigVMDispatchNode(
    Index=3
    Element=RigVMDispatchNode(Index=-1, Array=RigVMVariableNode() {variable=Positions}.Value) {template="DISPATCH_RigVMDispatch_ArrayGetAtIndex(in Array,in Index,out Element)"}.Element
    Array=RigVMVariableNode() {variable=Positions}.Value
   ) {template="DISPATCH_RigVMDispatch_ArraySetAtIndex(io Array,in Index,in Element)"}
   RigVMDispatchNode(
    Index=0
    Element=RigVMDispatchNode(Index=0, Array=RigVMVariableNode() {variable=Positions}.Value) {template="DISPATCH_RigVMDispatch_ArrayGetAtIndex(in Array,in Index,out Element)"}.Element
    ExecuteContext=@n3.True
    Array=RigVMVariableNode() {variable=Positions}.Value
   ) {template="DISPATCH_RigVMDispatch_ArraySetAtIndex(io Array,in Index,in Element)"}
   > RigVMDispatchNode(
    Index=1
    Element=RigVMDispatchNode(Index=0, Array=RigVMVariableNode() {variable=Positions}.Value) {template="DISPATCH_RigVMDispatch_ArrayGetAtIndex(in Array,in Index,out Element)"}.Element
    Array=RigVMVariableNode() {variable=Positions}.Value
   ) {template="DISPATCH_RigVMDispatch_ArraySetAtIndex(io Array,in Index,in Element)"}
   > RigVMDispatchNode(
    Index=2
    Element=RigVMDispatchNode(Index=0, Array=RigVMVariableNode() {variable=Positions}.Value) {template="DISPATCH_RigVMDispatch_ArrayGetAtIndex(in Array,in Index,out Element)"}.Element
    Array=RigVMVariableNode() {variable=Positions}.Value
   ) {template="DISPATCH_RigVMDispatch_ArraySetAtIndex(io Array,in Index,in Element)"}
   > RigVMDispatchNode(
    Index=3
    Element=RigVMDispatchNode(Index=1, Array=RigVMVariableNode() {variable=Positions}.Value) {template="DISPATCH_RigVMDispatch_ArrayGetAtIndex(in Array,in Index,out Element)"}.Element
    Array=RigVMVariableNode() {variable=Positions}.Value
   ) {template="DISPATCH_RigVMDispatch_ArraySetAtIndex(io Array,in Index,in Element)"}
   DISPATCH_RigVMDispatch_ArrayGetAtIndex = RigVMDispatchNode(Index=0, Array=RigVMVariableNode() {variable=Positions}.Value) {template="DISPATCH_RigVMDispatch_ArrayGetAtIndex(in Array,in Index,out Element)"}
   DISPATCH_RigVMDispatch_ArraySetAtIndex = RigVMDispatchNode(
    Index=0
    Element=@DISPATCH_RigVMDispatch_ArrayGetAtIndex.Element
    ExecuteContext=@RigVMFunction_ControlFlowBranch.True
    Array=RigVMVariableNode() {variable=Positions}.Value
   ) {template="DISPATCH_RigVMDispatch_ArraySetAtIndex(io Array,in Index,in Element)"}
   > RigVMDispatchNode(
    Index=1
    Element=RigVMDispatchNode(Index=1, Array=RigVMVariableNode() {variable=Positions}.Value) {template="DISPATCH_RigVMDispatch_ArrayGetAtIndex(in Array,in Index,out Element)"}.Element
    Array=RigVMVariableNode() {variable=Positions}.Value
   ) {template="DISPATCH_RigVMDispatch_ArraySetAtIndex(io Array,in Index,in Element)"}
   > RigVMDispatchNode(
    Index=2
    Element=RigVMDispatchNode(Index=2, Array=RigVMVariableNode() {variable=Positions}.Value) {template="DISPATCH_RigVMDispatch_ArrayGetAtIndex(in Array,in Index,out Element)"}.Element
    Array=RigVMVariableNode() {variable=Positions}.Value
   ) {template="DISPATCH_RigVMDispatch_ArraySetAtIndex(io Array,in Index,in Element)"}
   > RigVMDispatchNode(
    Index=3
    Element=RigVMDispatchNode(Index=3, Array=RigVMVariableNode() {variable=Positions}.Value) {template="DISPATCH_RigVMDispatch_ArrayGetAtIndex(in Array,in Index,out Element)"}.Element
    Array=RigVMVariableNode() {variable=Positions}.Value
   ) {template="DISPATCH_RigVMDispatch_ArraySetAtIndex(io Array,in Index,in Element)"}
   n5 = &RigVMFunction_ControlFlowBranch(Condition=@n1.Debug, ExecuteContext=@RigVMFunction_Sequence.C)
   n6 = RigVMVariableNode() {variable=Positions}
   RigVMFunctionReturnNode(ExecuteContext=@n5.Completed, Value=@n6.Value)
   n7 = RigVMDispatchNode(ExecuteContext=@n5.True, Array=@n6.Value) {template="DISPATCH_RigVMDispatch_ArrayIterator(in Array,out Element,out Index,out Count,out Ratio)"}
   > RigVMFunction_DebugTransformMutableNoSpace = &RigVMFunction_DebugTransformMutableNoSpace(
    Color=(R=1,G=0,B=0,A=1)
    DebugDrawSettings=(DepthPriority=SDPG_Foreground,Lifetime=-1)
    Mode=Axes
    Scale=10
    Thickness=1
    Transform=(Rotation=(X=0,Y=0,Z=0,W=1),Translation=(X=0,Y=0,Z=0),Scale3D=(X=1,Y=1,Z=1))
    WorldOffset=(Rotation=(X=0,Y=0,Z=0,W=1),Translation=(X=0,Y=0,Z=0),Scale3D=(X=1,Y=1,Z=1))
    bEnabled=True
    Transform.Translation=@n7.Element
   )
  ]}
 Concat_1_ContainedGraph {nodes [
   n1 = RigVMFunctionEntryNode()
   Concat = /Script/RigVM.RigVMFunction_NameConcat(A=@n1.A, B=@n1.B) {is=struct RigVMUnitNode, template="Concat::Execute(in A,in B,out Result)"}
   n2 = &RigVMFunction_NameConcat(A=@Concat.Result, B=@n1.C) {template="Concat::Execute(in A,in B,out Result)"}
   RigVMFunctionReturnNode(Result=&RigVMFunction_NameConcat(A=&RigVMFunction_NameConcat(A=@n2.Result, B=@n1.D) {template="Concat::Execute(in A,in B,out Result)"}.Result, B=@n1.E) {template="Concat::Execute(in A,in B,out Result)"}.Result)
  ]}
 Concat_1_ContainedGraph {nodes [
   n1 = RigVMFunctionEntryNode()
   Concat = &RigVMFunction_NameConcat(A=@n1.A, B=@n1.B) {template="Concat::Execute(in A,in B,out Result)"}
   n2 = &RigVMFunction_NameConcat(A=@Concat.Result, B=@n1.C) {template="Concat::Execute(in A,in B,out Result)"}
   RigVMFunctionReturnNode(Result=&RigVMFunction_NameConcat(A=&RigVMFunction_NameConcat(A=@n2.Result, B=@n1.D) {template="Concat::Execute(in A,in B,out Result)"}.Result, B=@n1.E) {template="Concat::Execute(in A,in B,out Result)"}.Result)
  ]}
 Concat_5_1_1_ContainedGraph {nodes [
   n1 = RigVMFunctionEntryNode()
   RigVMFunctionReturnNode(Result=&RigVMFunction_NameConcat(A=&RigVMFunction_NameConcat(A=@n1.A, B=@n1.B) {template="Concat::Execute(in A,in B,out Result)"}.Result, B=@n1.C) {template="Concat::Execute(in A,in B,out Result)"}.Result)
  ]}
 ConstructionGraph {nodes [
   RigUnit_PrepareForExecution = &RigUnit_PrepareForExecution()
   RigVMFunction_Sequence = &RigVMFunction_Sequence(ExecuteContext=@RigUnit_PrepareForExecution.ExecutePin)
   IK AnimalSpine = RigVMFunctionReferenceNode(
    Even IK Middle Controls=true
    Pelvis Bone Connector=(Type=Connector,Name="Pelvis Bone")
    Spine Bone Start Connector=/Script/ControlRig.RigUnit_ResolveConnector(Connector=(Type=Connector,Name="Spine Start Bone"), SkipSocket=true) {is=struct RigVMUnitNode}.Result
    Spine Bone End Connector=&RigUnit_ResolveConnector(Connector=(Type=Connector,Name="Spine End Bone"), SkipSocket=true).Result
    ExecuteContext=@RigVMFunction_Sequence.A
    Module Settings=RigVMVariableNode() {variable=Module Settings}.Value
    Body Control Settings=RigVMVariableNode() {variable=Body Control Shape Settings}.Value
    Number of Spine Bones=RigVMVariableNode() {variable=Number of Spines Bones}.Value
    Primary Axis=RigVMVariableNode() {variable=Primary Axis}.Value
    Up Axis=RigVMVariableNode() {variable=Secondary Axis}.Value
    Secondary Spline Direction=RigVMVariableNode() {variable=Secondary Spline Direction}.Value
    FK Control Settings=RigVMVariableNode() {variable=FK Control Settings}.Value
    Hips Control Settings=RigVMVariableNode() {variable=Hips Control Settings}.Value
    Debug=RigVMVariableNode() {variable=Debug}.Value
    Body Control Adjustment=RigVMVariableNode() {variable=Body Control Adjustment}.Value
    Chest Control Settings=RigVMVariableNode() {variable=Chest Control Settings}.Value
    Default FK or IK=RigVMVariableNode() {variable=Default FK or IK}.Value
    Mid Control Settings=RigVMVariableNode() {variable=Mid IK Control Settings}.Value
    Hips at Pelvis=RigVMVariableNode() {variable=Hips at Pelvis}.Value
   )
   > RigVMVariableNode(Value=@"IK AnimalSpine".Local Spline Curve) {variable=Local Spline}
   > RigVMVariableNode(Value=@"IK AnimalSpine".Template Spline Curve) {variable=Aim Spline}
   Get_Module_Metadata_2_1_1 = RigVMDispatchNode(Default=(Type=Bone,Name="None"), Name=Body Control, NameSpace=Self) {template="DISPATCH_RigDispatch_GetModuleMetadata(in Name,in NameSpace,in Default,out Value,out Found)"}
   Get_Module_Metadata_1_1 = RigVMDispatchNode(Default=(Type=Bone,Name="None"), Name=Chest Control, NameSpace=Self) {template="DISPATCH_RigDispatch_GetModuleMetadata(in Name,in NameSpace,in Default,out Value,out Found)"}
   Interpolate = /Script/RigVM.RigVMFunction_MathTransformLerp(
    T=0.5
    A=&RigUnit_GetTransform(bInitial=False, Item=@Get_Module_Metadata_2_1_1.Value).Transform
    B=&RigUnit_GetTransform(bInitial=False, Item=@Get_Module_Metadata_1_1.Value).Transform
   ) {is=struct RigVMUnitNode, template="Interpolate::Execute(in A,in B,in T,out Result)"}
   > Control Stack at Position_1 = RigVMFunctionReferenceNode(
    Add Bottom Null=False
    Add Null=False
    Add Secondary Control=False
    Control Suffix=_ctrl
    Name=main
    Null Suffix=_null
    Secondary Settings="(Shape=(bVisible=True,Name=\"Default\",Color=(R=1,G=0,B=0,A=1),Transform=(Rotation=(X=0,Y=0,Z=0,W=1),Translation=(X=0,Y=0,Z=0),Scale3D=(X=1,Y=1,Z=1))),Proxy=(bIsProxy=False,ShapeVisibility=BasedOnSelection),DisplayName=\"None\",InitialSpace=LocalSpace,bUsePref ...(truncated)"
    Tag Controls=true
    Zero Controls=true
    Global Position=@Interpolate.Result
    Parent=&RigUnit_ResolveConnector(Connector=(Type=Connector,Name="Parent Control"), SkipSocket=False).Result
    Settings=RigVMVariableNode() {variable=Main Control Settings}.Value
   )
   Get_Module_Metadata_2_1_1_1 = RigVMDispatchNode(Default=(Type=Bone,Name="None"), Name=Body Control, NameSpace=Self) {template="DISPATCH_RigDispatch_GetModuleMetadata(in Name,in NameSpace,in Default,out Value,out Found)"}
   > /Script/ControlRig.RigUnit_SetDefaultParent(Parent=@"Control Stack at Position_1".Control, Child=@Get_Module_Metadata_2_1_1_1.Value) {is=struct RigVMUnitNode}
   Get_Module_Metadata_1_1_1 = RigVMDispatchNode(Default=(Type=Bone,Name="None"), Name=Chest Control, NameSpace=Self) {template="DISPATCH_RigDispatch_GetModuleMetadata(in Name,in NameSpace,in Default,out Value,out Found)"}
   > &RigUnit_SetDefaultParent(Parent=@"Control Stack at Position_1".Control, Child=@Get_Module_Metadata_1_1_1.Value)
   > Set_Module_Metadata = RigVMDispatchNode(Name=Main Control, NameSpace=Self, Value=@"Control Stack at Position_1".Control) {template="DISPATCH_RigDispatch_SetModuleMetadata(in Name,in NameSpace,in Value,out Success)"}
   Get_Module_Metadata_1 = RigVMDispatchNode(Default=(Type=Bone,Name="None"), Name=Chest Control, NameSpace=Self) {template="DISPATCH_RigDispatch_GetModuleMetadata(in Name,in NameSpace,in Default,out Value,out Found)"}
   n1 = &RigUnit_GetTransform(bInitial=False, Item=@Get_Module_Metadata_1.Value)
   Add = /Script/RigVM.RigVMFunction_MathVectorAdd(A=@n1.Transform.Translation, B=RigVMVariableNode() {variable=Chest Hips Position}.Value) {is=struct RigVMUnitNode, template="Add::Execute(in A,in B,out Result)"}
   > Control Stack at Position = RigVMFunctionReferenceNode(
    Add Bottom Null=False
    Add Null=False
    Add Secondary Control=False
    Control Suffix=_ctrl
    Global Position=(Rotation=(X=0,Y=0,Z=0,W=1),Translation=(X=0,Y=0,Z=0),Scale3D=(X=1,Y=1,Z=1))
    Name=hipschest
    Null Suffix=_null
    Secondary Settings="(Shape=(bVisible=True,Name=\"Default\",Color=(R=1,G=0,B=0,A=1),Transform=(Rotation=(X=0,Y=0,Z=0,W=1),Translation=(X=0,Y=0,Z=0),Scale3D=(X=1,Y=1,Z=1))),Proxy=(bIsProxy=False,ShapeVisibility=BasedOnSelection),DisplayName=\"None\",InitialSpace=LocalSpace,bUsePref ...(truncated)"
    Tag Controls=true
    Zero Controls=true
    Global Position.Translation=@Add.Result
    Parent=@"Control Stack at Position_1".Control
    Global Position.Rotation=@n1.Transform.Rotation
    Global Position.Scale3D=@n1.Transform.Scale3D
    Settings=RigVMVariableNode() {variable=Hips Chest Shape Settings}.Value
   )
   Get_Module_Metadata_2 = RigVMDispatchNode(Default=(Type=Bone,Name="None"), Name=Body Control, NameSpace=Self) {template="DISPATCH_RigDispatch_GetModuleMetadata(in Name,in NameSpace,in Default,out Value,out Found)"}
   > &RigUnit_SetDefaultParent(Parent=@"Control Stack at Position".Control, Child=@Get_Module_Metadata_2.Value)
   > For_Each = RigVMDispatchNode(Array=/Script/ControlRig.RigUnit_CollectionChildrenArray(
     TypeToSearch=Bone
     bDefaultChildren=True
     bIncludeParent=False
     bRecursive=true
     Parent=&RigUnit_ResolveConnector(Connector=(Type=Connector,Name="Pelvis Bone"), SkipSocket=true).Result
    ) {is=struct RigVMUnitNode}.Items) {template="DISPATCH_RigVMDispatch_ArrayIterator(in Array,out Element,out Index,out Count,out Ratio)"}
   Find = RigVMFunctionReferenceNode(Element=neck, Value=@For_Each.Element.Name)
   Find_1 = RigVMFunctionReferenceNode(Element=thigh, Value=@For_Each.Element.Name)
   Find_1_1 = RigVMFunctionReferenceNode(Element=tail, Value=@For_Each.Element.Name)
   Find_1_1_1 = RigVMFunctionReferenceNode(Element=upperarm, Value=@For_Each.Element.Name)
   Find_1_1_1_1 = RigVMFunctionReferenceNode(Element=shoulder, Value=@For_Each.Element.Name)
   RigVMFunction_MathBoolOr = RigVMAggregateNode(A=@Find.Success, B=@Find_1.Success, C=@Find_1_1.Success, D=@Find_1_1_1.Success, E=@Find_1_1_1_1.Success)
   n2 = &RigVMFunction_ControlFlowBranch(ExecuteContext=@For_Each.ExecuteContext, Condition=@RigVMFunction_MathBoolOr.Result)
   Concat = &RigVMFunction_NameConcat(B=_socket, A=@For_Each.Element.Name) {template="Concat::Execute(in A,in B,out Result)"}
   /Script/ControlRig.RigUnit_HierarchyAddSocket(
    Color=(R=1,G=1,B=1,A=1)
    Description=Place a new Module here . . .
    Space=LocalSpace
    Transform=(Rotation=(X=0,Y=0,Z=0,W=1),Translation=(X=0,Y=0,Z=0),Scale3D=(X=1,Y=1,Z=1))
    Name=@Concat.Result
    Parent=@For_Each.Element
    ExecutePin=@n2.True
   ) {is=struct RigVMUnitNode}
   Finalize Controls = RigVMCollapseNode(ExecuteContext=@RigVMFunction_Sequence.B, Module Settings=RigVMVariableNode() {variable=Module Settings}.Value)
   Get_Module_Metadata_1_1_2 = RigVMDispatchNode(Default=(Type=Bone,Name="None"), Name=Chest Control, NameSpace=Self) {template="DISPATCH_RigDispatch_GetModuleMetadata(in Name,in NameSpace,in Default,out Value,out Found)"}
   Get_Module_Metadata_2_1_1_2 = RigVMDispatchNode(Default=(Type=Bone,Name="None"), Name=Body Control, NameSpace=Self) {template="DISPATCH_RigDispatch_GetModuleMetadata(in Name,in NameSpace,in Default,out Value,out Found)"}
   RigVMFunction_MathVectorDistance = &RigVMFunction_MathVectorDistance(
    B=&RigUnit_GetTransform(bInitial=False, Item=@Get_Module_Metadata_1_1_2.Value).Transform.Translation
    A=&RigUnit_GetTransform(bInitial=False, Item=@Get_Module_Metadata_2_1_1_2.Value).Transform.Translation
   )
   Divide = &RigVMFunction_MathDoubleDiv(B=15, A=@RigVMFunction_MathVectorDistance.Result)
   n3 = /Script/ControlRig.RigUnit_HierarchyGetShapeSettings(Item=@"Control Stack at Position_1".Control) {is=struct RigVMUnitNode}
   /Script/ControlRig.RigUnit_HierarchySetShapeSettings(
    ExecutePin=()
    Settings=(bVisible=True,Name="Default",Color=(R=1,G=0,B=0,A=1),Transform=(Rotation=(X=0,Y=0,Z=0,W=1),Translation=(X=0,Y=0,Z=0),Scale3D=(X=0.25,Y=0.25,Z=1)))
    Item=@"Control Stack at Position_1".Control
    Settings.Transform.Scale3D.Z=@Divide.Result
    Settings.Color=@n3.Settings.Color
    Settings.Name=@n3.Settings.Name
    Settings.Transform.Rotation=@n3.Settings.Transform.Rotation
    Settings.Transform.Translation=@n3.Settings.Transform.Translation
    Settings.bVisible=@n3.Settings.bVisible
   ) {is=struct RigVMUnitNode}
  ]}
 ContainedGraph {nodes [
   BranchNode_RigVMUnitNode = &RigVMFunction_ControlFlowBranch(ExecuteContext=(), Condition=RigVMVariableNode() {variable=debug}.Value)
   /Script/ControlRigSpline.RigUnit_DrawControlRigSpline(
    Color=(R=1,G=0,B=0,A=1)
    Detail=16
    Thickness=0.75
    ExecutePin=@BranchNode_RigVMUnitNode.True
    Spline=RigVMVariableNode() {variable=template_spline}.Value
   ) {is=struct RigVMUnitNode}
   n1 = RigVMFunctionEntryNode()
   RigVMFunction_Sequence = &RigVMFunction_Sequence(ExecuteContext=@n1.ExecuteContext)
   n2 = RigVMVariableNode() {variable=Controls}
   BranchNode_4_RigVMUnitNode = &RigVMFunction_ControlFlowBranch(
    Condition=/Script/RigVM.RigVMFunction_MathIntLess(B=4, A=RigVMDispatchNode(Array=@n2.Value) {template="DISPATCH_RigVMDispatch_ArrayGetNum(in Array,out Num)"}.Num) {is=struct RigVMUnitNode, template="Less::Execute(in A,in B,out Result)"}.Result
    ExecuteContext=@RigVMFunction_Sequence.B
   )
   RigVMFunctionReturnNode(ExecuteContext=@BranchNode_4_RigVMUnitNode.Completed, Output Spline=RigVMVariableNode() {variable=local_spline}.Value)
   RigUnit_SequenceExecution_1 = RigVMAggregateNode(ExecuteContext=@BranchNode_4_RigVMUnitNode.False)
   RigVMVariableNode(Value=@n1.Template Spline, ExecuteContext=@RigUnit_SequenceExecution_1.A) {variable=template_spline}
   n3 = /Script/ControlRig.RigUnit_GetTransformItemArray(Space=GlobalSpace, bInitial=False, Items=RigVMVariableNode() {variable=Controls}.Value) {is=struct RigVMUnitNode}
   > n4 = /Script/ControlRigSpline.RigUnit_SetSplineTransforms(Spline=@n1.Template Spline, Transforms=@n3.Transforms) {is=struct RigVMUnitNode}
   > RigVMVariableNode(Value=@n4.Spline) {variable=local_spline}
   > RigVMVariableNode(Value=@n3.Transforms) {variable=template_transforms}
   > RigUnit_SequenceExecution = RigVMAggregateNode()
   n5 = RigVMVariableNode() {variable=template_transforms}
   RigVMDispatchNode(
    Num=RigVMDispatchNode(Array=@n5.Value) {template="DISPATCH_RigVMDispatch_ArrayGetNum(in Array,out Num)"}.Num
    ExecuteContext=@RigUnit_SequenceExecution.A
    Array=RigVMVariableNode() {variable=offset_positions}.Value
   ) {template="DISPATCH_RigVMDispatch_ArraySetNum(io Array,in Num)"}
   > n6 = RigVMDispatchNode(Array=@n5.Value) {template="DISPATCH_RigVMDispatch_ArrayIterator(in Array,out Element,out Index,out Count,out Ratio)"}
   > RigVMDispatchNode(
    Index=@n6.Index
    Element=/Script/RigVM.RigVMFunction_MathTransformTransformVector(Transform=@n6.Element, Location=RigVMVariableNode() {variable=Secondary Spline Direction}.Value) {is=struct RigVMUnitNode}.Result
    Array=RigVMVariableNode() {variable=offset_positions}.Value
   ) {template="DISPATCH_RigVMDispatch_ArraySetAtIndex(io Array,in Index,in Element)"}
   RigUnit_SequenceExecution_4 = RigVMAggregateNode(ExecuteContext=@RigUnit_SequenceExecution.C)
   Branch_RigVMUnitNode = &RigVMFunction_ControlFlowBranch(ExecuteContext=@RigUnit_SequenceExecution_4.A, Condition=RigVMVariableNode() {variable=Stretch}.Value)
   n7 = RigVMVariableNode() {variable=template_spline}
   n8 = RigVMVariableNode() {variable=Primary Axis}
   n9 = RigVMVariableNode() {variable=Bones}
   /Script/ControlRigSpline.RigUnit_FitChainToSplineCurveItemArray(
    Alignment=Front
    DebugSettings=(Scale=1,CurveColor=(R=1,G=1,B=0,A=1),SegmentsColor=(R=1,G=0,B=0,A=1),WorldOffset=(Rotation=(X=0,Y=0,Z=0,W=1),Translation=(X=0,Y=0,Z=0),Scale3D=(X=1,Y=1,Z=1)))
    Maximum=1
    PoleVectorPosition=(X=0,Y=0,Z=0)
    Rotations=()
    SamplingPrecision=12
    SecondaryAxis=(X=0,Y=0,Z=0)
    Weight=1
    bPropagateToChildren=True
    ExecutePin=@Branch_RigVMUnitNode.False
    Spline=@n7.Value
    PrimaryAxis=@n8.Value
    Items=@n9.Value
   ) {is=struct RigVMUnitNode}
   &RigUnit_FitChainToSplineCurveItemArray(
    Alignment=Stretched
    DebugSettings=(Scale=1,CurveColor=(R=1,G=1,B=0,A=1),SegmentsColor=(R=1,G=0,B=0,A=1),WorldOffset=(Rotation=(X=0,Y=0,Z=0,W=1),Translation=(X=0,Y=0,Z=0),Scale3D=(X=1,Y=1,Z=1)))
    Maximum=1
    PoleVectorPosition=(X=0,Y=0,Z=0)
    Rotations=()
    SamplingPrecision=12
    SecondaryAxis=(X=0,Y=0,Z=0)
    Weight=1
    bPropagateToChildren=True
    ExecutePin=@Branch_RigVMUnitNode.True
    Spline=@n7.Value
    PrimaryAxis=@n8.Value
    Items=@n9.Value
   )
   RigUnit_SequenceExecution_2 = RigVMAggregateNode(ExecuteContext=@RigUnit_SequenceExecution_4.B)
   BranchNode_2_RigVMUnitNode = &RigVMFunction_ControlFlowBranch(ExecuteContext=@RigUnit_SequenceExecution_2.A, Condition=RigVMVariableNode() {variable=debug}.Value)
   &RigUnit_DrawControlRigSpline(
    Color=(R=1,G=0,B=0,A=1)
    Detail=16
    Thickness=1
    ExecutePin=@BranchNode_2_RigVMUnitNode.True
    Spline=RigVMVariableNode() {variable=local_spline}.Value
   )
   n10 = /Script/ControlRigSpline.RigUnit_SetSplinePoints(
    ExecutePin=@RigUnit_SequenceExecution_2.B
    Points=RigVMVariableNode() {variable=offset_positions}.Value
    Spline=RigVMVariableNode() {variable=local_spline}.Value
   ) {is=struct RigVMUnitNode}
   > RigVMVariableNode(Value=@n10.Spline) {variable=local_spline}
   n11 = RigVMVariableNode() {variable=Bones}
   DISPATCH_RigVMDispatch_ArrayGetNum = RigVMDispatchNode(Array=@n11.Value) {template="DISPATCH_RigVMDispatch_ArrayGetNum(in Array,out Num)"}
   > DISPATCH_RigVMDispatch_ArraySetNum = RigVMDispatchNode(Num=@DISPATCH_RigVMDispatch_ArrayGetNum.Num, Array=RigVMVariableNode() {variable=aim_transforms}.Value) {template="DISPATCH_RigVMDispatch_ArraySetNum(io Array,in Num)"}
   > n12 = RigVMDispatchNode(Array=@n11.Value) {template="DISPATCH_RigVMDispatch_ArrayIterator(in Array,out Element,out Index,out Count,out Ratio)"}
   n13 = &RigUnit_GetTransform(Item=@n12.Element)
   n14 = RigVMVariableNode() {variable=local_spline}
   > DISPATCH_RigVMDispatch_ArraySetAtIndex = RigVMDispatchNode(
    Index=@n12.Index
    Element=/Script/ControlRigSpline.RigUnit_PositionFromControlRigSpline(
     U=/Script/ControlRigSpline.RigUnit_ClosestParameterFromControlRigSpline(Position=@n13.Transform.Translation, Spline=@n14.Value) {is=struct RigVMUnitNode}.U
     Spline=@n14.Value
    ) {is=struct RigVMUnitNode}.Position
    Array=RigVMVariableNode() {variable=aim_transforms}.Value
   ) {template="DISPATCH_RigVMDispatch_ArraySetAtIndex(io Array,in Index,in Element)"}
   n15 = RigVMDispatchNode(ExecuteContext=@RigUnit_SequenceExecution.D, Array=RigVMVariableNode() {variable=Bones}.Value) {template="DISPATCH_RigVMDispatch_ArrayIterator(in Array,out Element,out Index,out Count,out Ratio)"}
   DISPATCH_RigVMDispatch_ArrayGetAtIndex = RigVMDispatchNode(Index=@n15.Index, Array=RigVMVariableNode() {variable=aim_transforms}.Value) {template="DISPATCH_RigVMDispatch_ArrayGetAtIndex(in Array,in Index,out Element)"}
   n16 = /Script/ControlRig.RigUnit_AimBoneMath(
    DebugSettings=(Scale=10,WorldOffset=(Rotation=(X=0,Y=0,Z=0,W=1),Translation=(X=0,Y=0,Z=0),Scale3D=(X=1,Y=1,Z=1)))
    Primary=(Weight=1,Axis=(X=0,Y=1,Z=0),Target=(X=1,Y=0,Z=0),Kind=Location,Space=(Type=Bone))
    Secondary=(Weight=0,Axis=(X=0,Y=0,Z=0),Target=(X=0,Y=0,Z=0),Kind=Location,Space=(Type=Bone))
    Weight=1
    Primary.Target=@DISPATCH_RigVMDispatch_ArrayGetAtIndex.Element
    InputTransform=&RigUnit_GetTransform(Item=@n15.Element).Transform
    Primary.Axis=RigVMVariableNode() {variable=Up Axis}.Value
   ) {is=struct RigVMUnitNode}
   MathIntEquals = RigVMDispatchNode(
    A=@n15.Index
    B=/Script/RigVM.RigVMFunction_MathIntSub(B=1, A=@n15.Count) {is=struct RigVMUnitNode, template="Subtract::Execute(in A,in B,out Result)"}.Result
   ) {template="DISPATCH_RigVMDispatch_CoreEquals(in A,in B,out Result)"}
   > n17 = &RigUnit_SetTransform(Weight=1, bInitial=False, Value=@n16.Result, Item=@n15.Element, bPropagateToChildren=@MathIntEquals.Result)
   BranchNode_3_RigVMUnitNode = &RigVMFunction_ControlFlowBranch(ExecuteContext=@n17.ExecutePin, Condition=RigVMVariableNode() {variable=debug}.Value)
   /Script/ControlRig.RigUnit_DebugTransformMutableItemSpace(
    Color=(R=1,G=0,B=0,A=1)
    DebugDrawSettings=(DepthPriority=SDPG_Foreground,Lifetime=-1)
    Mode=Axes
    Scale=10
    Space=(Type=Bone)
    Thickness=0.15
    WorldOffset=(Rotation=(X=0,Y=0,Z=0,W=1),Translation=(X=0,Y=0,Z=0),Scale3D=(X=1,Y=1,Z=1))
    bEnabled=True
    Transform=@n16.Result
    ExecutePin=@BranchNode_3_RigVMUnitNode.True
   ) {is=struct RigVMUnitNode}
   Branch_1_RigVMUnitNode = &RigVMFunction_ControlFlowBranch(ExecuteContext=@RigUnit_SequenceExecution_1.B, Condition=RigVMVariableNode() {variable=Stretch}.Value)
   n18 = RigVMDispatchNode(Index=-1, Array=RigVMVariableNode() {variable=Controls}.Value) {template="DISPATCH_RigVMDispatch_ArrayGetAtIndex(in Array,in Index,out Element)"}
   n19 = RigVMDispatchNode(Index=-1, Array=RigVMVariableNode() {variable=Bones}.Value) {template="DISPATCH_RigVMDispatch_ArrayGetAtIndex(in Array,in Index,out Element)"}
   &RigUnit_ParentConstraint(
    AdvancedSettings=(RotationOrderForFilter=XZY)
    Filter=(TranslationFilter=(bX=false,bY=false,bZ=false),RotationFilter=(bX=True,bY=True,bZ=True),ScaleFilter=(bX=True,bY=True,bZ=True))
    Parents=((Item=(Type=Bone),Weight=1))
    Weight=1
    Parents.0.Item=@n18.Element
    Child=@n19.Element
    ExecutePin=@Branch_1_RigVMUnitNode.False
   )
   &RigUnit_ParentConstraint(
    AdvancedSettings=(RotationOrderForFilter=XZY)
    Parents=((Item=(Type=Bone),Weight=1))
    Weight=1
    Parents.0.Item=@n18.Element
    Child=@n19.Element
    ExecutePin=@Branch_1_RigVMUnitNode.True
   )
   RigVMVariableNode() {variable=Spline Mode}
   RigVMVariableNode() {variable=Samples Per Segment}
  ]}
 ContainedGraph {nodes [
   n1 = RigVMFunctionEntryNode()
   RigUnit_SequenceExecution_2 = &RigVMFunction_Sequence(ExecuteContext=@n1.ExecuteContext)
   RigUnit_SequenceExecution_2_1 = &RigVMFunction_Sequence(ExecuteContext=@RigUnit_SequenceExecution_2.B)
   RigUnit_SequenceExecution_2_2 = &RigVMFunction_Sequence(ExecuteContext=@RigUnit_SequenceExecution_2_1.B)
   RigVMFunctionReturnNode(A=@RigUnit_SequenceExecution_2.A, B=@RigUnit_SequenceExecution_2_1.A, C=@RigUnit_SequenceExecution_2_2.A, D=@RigUnit_SequenceExecution_2_2.B)
  ]}
 ContainedGraph {nodes [
   n1 = RigVMFunctionEntryNode()
   RigUnit_SequenceExecution_4 = &RigVMFunction_Sequence(ExecuteContext=@n1.ExecuteContext)
   RigUnit_SequenceExecution_4_1 = &RigVMFunction_Sequence(ExecuteContext=@RigUnit_SequenceExecution_4.B)
   RigUnit_SequenceExecution_4_2 = &RigVMFunction_Sequence(ExecuteContext=@RigUnit_SequenceExecution_4_1.B)
   RigVMFunctionReturnNode(A=@RigUnit_SequenceExecution_4.A, B=@RigUnit_SequenceExecution_4_1.A, C=@RigUnit_SequenceExecution_4_2.A, D=@RigUnit_SequenceExecution_4_2.B)
  ]}
 ContainedGraph {nodes [
   n1 = RigVMFunctionEntryNode()
   RigUnit_SequenceExecution = &RigVMFunction_Sequence(ExecuteContext=@n1.ExecuteContext)
   RigUnit_SequenceExecution_1 = &RigVMFunction_Sequence(ExecuteContext=@RigUnit_SequenceExecution.B)
   RigUnit_SequenceExecution_2 = &RigVMFunction_Sequence(ExecuteContext=@RigUnit_SequenceExecution_1.B)
   RigVMFunctionReturnNode(A=@RigUnit_SequenceExecution.A, B=@RigUnit_SequenceExecution_1.A, C=@RigUnit_SequenceExecution_2.A, D=@RigUnit_SequenceExecution_2.B)
  ]}
 ContainedGraph {nodes [
   n1 = RigVMFunctionEntryNode()
   RigUnit_SequenceExecution_1 = &RigVMFunction_Sequence(ExecuteContext=@n1.ExecuteContext)
   RigUnit_SequenceExecution_1_1 = &RigVMFunction_Sequence(ExecuteContext=@RigUnit_SequenceExecution_1.B)
   RigUnit_SequenceExecution_1_2 = &RigVMFunction_Sequence(ExecuteContext=@RigUnit_SequenceExecution_1_1.B)
   RigVMFunctionReturnNode(A=@RigUnit_SequenceExecution_1.A, B=@RigUnit_SequenceExecution_1_1.A, C=@RigUnit_SequenceExecution_1_2.A, D=@RigUnit_SequenceExecution_1_2.B)
  ]}
 ContainedGraph {nodes [
   n1 = RigVMFunctionEntryNode()
   n2 = RigVMVariableNode() {variable=Controls}
   BranchNode_4_RigVMUnitNode = &RigVMFunction_ControlFlowBranch(
    ExecuteContext=@n1.ExecuteContext
    Condition=&RigVMFunction_MathIntLess(B=4, A=RigVMDispatchNode(Array=@n2.Value) {template="DISPATCH_RigVMDispatch_ArrayGetNum(in Array,out Num)"}.Num) {template="Less::Execute(in A,in B,out Result)"}.Result
   )
   RigVMFunctionReturnNode(
    ExecuteContext=@BranchNode_4_RigVMUnitNode.Completed
    Local Spline=RigVMVariableNode() {variable=local_spline}.Value
    Template Spline=RigVMVariableNode() {variable=template_spline}.Value
   )
   RigUnit_SequenceExecution_1 = RigVMAggregateNode(ExecuteContext=@BranchNode_4_RigVMUnitNode.False)
   n3 = &RigUnit_GetTransformItemArray(Space=GlobalSpace, bInitial=False, Items=RigVMVariableNode() {variable=Controls}.Value)
   RigVMFunction_MathTransformArrayToSRT = /Script/RigVM.RigVMFunction_MathTransformArrayToSRT(Transforms=@n3.Transforms) {is=struct RigVMUnitNode}
   RigVMVariableNode(
    Value=/Script/ControlRigSpline.RigUnit_ControlRigSplineFromPoints(
     bClosed=False
     Points=@RigVMFunction_MathTransformArrayToSRT.Translations
     SplineMode=RigVMVariableNode() {variable=Spline Mode}.Value
     SamplesPerSegment=RigVMVariableNode() {variable=Samples Per Segment}.Value
    ) {is=struct RigVMUnitNode}.Spline
    ExecuteContext=@RigUnit_SequenceExecution_1.A
   ) {variable=template_spline}
   > RigVMVariableNode(Value=@n3.Transforms) {variable=template_transforms}
   > RigUnit_SequenceExecution = RigVMAggregateNode()
   n4 = RigVMVariableNode() {variable=template_transforms}
   RigVMDispatchNode(
    Num=RigVMDispatchNode(Array=@n4.Value) {template="DISPATCH_RigVMDispatch_ArrayGetNum(in Array,out Num)"}.Num
    ExecuteContext=@RigUnit_SequenceExecution.A
    Array=RigVMVariableNode() {variable=offset_positions}.Value
   ) {template="DISPATCH_RigVMDispatch_ArraySetNum(io Array,in Num)"}
   > n5 = RigVMDispatchNode(Array=@n4.Value) {template="DISPATCH_RigVMDispatch_ArrayIterator(in Array,out Element,out Index,out Count,out Ratio)"}
   > RigVMDispatchNode(
    Index=@n5.Index
    Element=&RigVMFunction_MathTransformTransformVector(Transform=@n5.Element, Location=RigVMVariableNode() {variable=Secondary Spline Direction}.Value).Result
    Array=RigVMVariableNode() {variable=offset_positions}.Value
   ) {template="DISPATCH_RigVMDispatch_ArraySetAtIndex(io Array,in Index,in Element)"}
   RigUnit_SequenceExecution_4 = RigVMAggregateNode(ExecuteContext=@RigUnit_SequenceExecution.C)
   Branch_RigVMUnitNode = &RigVMFunction_ControlFlowBranch(ExecuteContext=@RigUnit_SequenceExecution_4.A, Condition=RigVMVariableNode() {variable=Stretch}.Value)
   n6 = RigVMVariableNode() {variable=template_spline}
   n7 = RigVMVariableNode() {variable=Primary Axis}
   n8 = RigVMVariableNode() {variable=Bones}
   &RigUnit_FitChainToSplineCurveItemArray(
    Alignment=Front
    DebugSettings=(Scale=1,CurveColor=(R=1,G=1,B=0,A=1),SegmentsColor=(R=1,G=0,B=0,A=1),WorldOffset=(Rotation=(X=0,Y=0,Z=0,W=1),Translation=(X=0,Y=0,Z=0),Scale3D=(X=1,Y=1,Z=1)))
    Maximum=1
    PoleVectorPosition=(X=0,Y=0,Z=0)
    Rotations=()
    SamplingPrecision=12
    SecondaryAxis=(X=0,Y=0,Z=0)
    Weight=1
    bPropagateToChildren=True
    ExecutePin=@Branch_RigVMUnitNode.False
    Spline=@n6.Value
    PrimaryAxis=@n7.Value
    Items=@n8.Value
   )
   &RigUnit_FitChainToSplineCurveItemArray(
    Alignment=Stretched
    DebugSettings=(Scale=1,CurveColor=(R=1,G=1,B=0,A=1),SegmentsColor=(R=1,G=0,B=0,A=1),WorldOffset=(Rotation=(X=0,Y=0,Z=0,W=1),Translation=(X=0,Y=0,Z=0),Scale3D=(X=1,Y=1,Z=1)))
    Maximum=1
    PoleVectorPosition=(X=0,Y=0,Z=0)
    Rotations=()
    SamplingPrecision=12
    SecondaryAxis=(X=0,Y=0,Z=0)
    Weight=1
    bPropagateToChildren=True
    ExecutePin=@Branch_RigVMUnitNode.True
    Spline=@n6.Value
    PrimaryAxis=@n7.Value
    Items=@n8.Value
   )
   RigUnit_SequenceExecution_2 = RigVMAggregateNode(ExecuteContext=@RigUnit_SequenceExecution_4.B)
   n9 = RigVMVariableNode(
    Value=&RigUnit_ControlRigSplineFromPoints(
     Stretch=0
     bClosed=False
     SplineMode=RigVMVariableNode() {variable=Spline Mode}.Value
     SamplesPerSegment=RigVMVariableNode() {variable=Samples Per Segment}.Value
     Points=RigVMVariableNode() {variable=offset_positions}.Value
    ).Spline
    ExecuteContext=@RigUnit_SequenceExecution_2.A
   ) {variable=local_spline}
   BranchNode_2_RigVMUnitNode = &RigVMFunction_ControlFlowBranch(ExecuteContext=@n9.ExecuteContext, Condition=RigVMVariableNode() {variable=debug}.Value)
   &RigUnit_DrawControlRigSpline(
    Color=(R=1,G=0,B=0,A=1)
    Detail=16
    Thickness=1
    ExecutePin=@BranchNode_2_RigVMUnitNode.True
    Spline=RigVMVariableNode() {variable=local_spline}.Value
   )
   n10 = RigVMVariableNode() {variable=Bones}
   DISPATCH_RigVMDispatch_ArrayGetNum = RigVMDispatchNode(Array=@n10.Value) {template="DISPATCH_RigVMDispatch_ArrayGetNum(in Array,out Num)"}
   DISPATCH_RigVMDispatch_ArraySetNum = RigVMDispatchNode(
    Num=@DISPATCH_RigVMDispatch_ArrayGetNum.Num
    ExecuteContext=@RigUnit_SequenceExecution_2.B
    Array=RigVMVariableNode() {variable=aim_transforms}.Value
   ) {template="DISPATCH_RigVMDispatch_ArraySetNum(io Array,in Num)"}
   > n11 = RigVMDispatchNode(Array=@n10.Value) {template="DISPATCH_RigVMDispatch_ArrayIterator(in Array,out Element,out Index,out Count,out Ratio)"}
   n12 = &RigUnit_GetTransform(bInitial=false, Item=@n11.Element)
   n13 = RigVMVariableNode() {variable=local_spline}
   > DISPATCH_RigVMDispatch_ArraySetAtIndex = RigVMDispatchNode(
    Index=@n11.Index
    Element=&RigUnit_PositionFromControlRigSpline(
     U=&RigUnit_ClosestParameterFromControlRigSpline(Position=@n12.Transform.Translation, Spline=@n13.Value).U
     Spline=@n13.Value
    ).Position
    Array=RigVMVariableNode() {variable=aim_transforms}.Value
   ) {template="DISPATCH_RigVMDispatch_ArraySetAtIndex(io Array,in Index,in Element)"}
   n14 = RigVMDispatchNode(ExecuteContext=@RigUnit_SequenceExecution.D, Array=RigVMVariableNode() {variable=Bones}.Value) {template="DISPATCH_RigVMDispatch_ArrayIterator(in Array,out Element,out Index,out Count,out Ratio)"}
   DISPATCH_RigVMDispatch_ArrayGetAtIndex = RigVMDispatchNode(Index=@n14.Index, Array=RigVMVariableNode() {variable=aim_transforms}.Value) {template="DISPATCH_RigVMDispatch_ArrayGetAtIndex(in Array,in Index,out Element)"}
   n15 = &RigUnit_AimBoneMath(
    DebugSettings=(Scale=10,WorldOffset=(Rotation=(X=0,Y=0,Z=0,W=1),Translation=(X=0,Y=0,Z=0),Scale3D=(X=1,Y=1,Z=1)))
    Primary=(Weight=1,Axis=(X=0,Y=1,Z=0),Target=(X=1,Y=0,Z=0),Kind=Location,Space=(Type=Bone))
    Secondary=(Weight=0,Axis=(X=0,Y=0,Z=0),Target=(X=0,Y=0,Z=0),Kind=Location,Space=(Type=Bone))
    Weight=1
    Primary.Target=@DISPATCH_RigVMDispatch_ArrayGetAtIndex.Element
    InputTransform=&RigUnit_GetTransform(Item=@n14.Element).Transform
    Primary.Axis=RigVMVariableNode() {variable=Up Axis}.Value
   )
   MathIntEquals = RigVMDispatchNode(A=@n14.Index, B=&RigVMFunction_MathIntSub(B=1, A=@n14.Count) {template="Subtract::Execute(in A,in B,out Result)"}.Result) {template="DISPATCH_RigVMDispatch_CoreEquals(in A,in B,out Result)"}
   > n16 = &RigUnit_SetTransform(Weight=1, bInitial=true, Value=@n15.Result, Item=@n14.Element, bPropagateToChildren=@MathIntEquals.Result)
   BranchNode_3_RigVMUnitNode = &RigVMFunction_ControlFlowBranch(ExecuteContext=@n16.ExecutePin, Condition=RigVMVariableNode() {variable=debug}.Value)
   &RigUnit_DebugTransformMutableItemSpace(
    Color=(R=1,G=0,B=0,A=1)
    DebugDrawSettings=(DepthPriority=SDPG_Foreground,Lifetime=-1)
    Mode=Axes
    Scale=10
    Space=(Type=Bone)
    Thickness=0.15
    WorldOffset=(Rotation=(X=0,Y=0,Z=0,W=1),Translation=(X=0,Y=0,Z=0),Scale3D=(X=1,Y=1,Z=1))
    bEnabled=True
    Transform=@n15.Result
    ExecutePin=@BranchNode_3_RigVMUnitNode.True
   )
   Branch_1_RigVMUnitNode = &RigVMFunction_ControlFlowBranch(ExecuteContext=@RigUnit_SequenceExecution_1.B, Condition=RigVMVariableNode() {variable=Stretch}.Value)
   n17 = RigVMDispatchNode(Index=-1, Array=RigVMVariableNode() {variable=Controls}.Value) {template="DISPATCH_RigVMDispatch_ArrayGetAtIndex(in Array,in Index,out Element)"}
   n18 = RigVMDispatchNode(Index=-1, Array=RigVMVariableNode() {variable=Bones}.Value) {template="DISPATCH_RigVMDispatch_ArrayGetAtIndex(in Array,in Index,out Element)"}
   &RigUnit_ParentConstraint(
    AdvancedSettings=(RotationOrderForFilter=XZY)
    Filter=(TranslationFilter=(bX=false,bY=false,bZ=false),RotationFilter=(bX=True,bY=True,bZ=True),ScaleFilter=(bX=True,bY=True,bZ=True))
    Parents=((Item=(Type=Bone),Weight=1))
    Weight=1
    Parents.0.Item=@n17.Element
    Child=@n18.Element
    ExecutePin=@Branch_1_RigVMUnitNode.False
   )
   &RigUnit_ParentConstraint(
    AdvancedSettings=(RotationOrderForFilter=XZY)
    Parents=((Item=(Type=Bone),Weight=1))
    Weight=1
    Parents.0.Item=@n17.Element
    Child=@n18.Element
    ExecutePin=@Branch_1_RigVMUnitNode.True
   )
  ]}
 ContainedGraph {nodes [
   n1 = RigVMFunctionEntryNode()
   RigUnit_SequenceExecution_2 = &RigVMFunction_Sequence(ExecuteContext=@n1.ExecuteContext)
   RigUnit_SequenceExecution_2_1 = &RigVMFunction_Sequence(ExecuteContext=@RigUnit_SequenceExecution_2.B)
   RigUnit_SequenceExecution_2_2 = &RigVMFunction_Sequence(ExecuteContext=@RigUnit_SequenceExecution_2_1.B)
   RigVMFunctionReturnNode(A=@RigUnit_SequenceExecution_2.A, B=@RigUnit_SequenceExecution_2_1.A, C=@RigUnit_SequenceExecution_2_2.A, D=@RigUnit_SequenceExecution_2_2.B)
  ]}
 ContainedGraph {nodes [
   n1 = RigVMFunctionEntryNode()
   RigUnit_SequenceExecution_4 = &RigVMFunction_Sequence(ExecuteContext=@n1.ExecuteContext)
   RigUnit_SequenceExecution_4_1 = &RigVMFunction_Sequence(ExecuteContext=@RigUnit_SequenceExecution_4.B)
   RigUnit_SequenceExecution_4_2 = &RigVMFunction_Sequence(ExecuteContext=@RigUnit_SequenceExecution_4_1.B)
   RigVMFunctionReturnNode(A=@RigUnit_SequenceExecution_4.A, B=@RigUnit_SequenceExecution_4_1.A, C=@RigUnit_SequenceExecution_4_2.A, D=@RigUnit_SequenceExecution_4_2.B)
  ]}
 ContainedGraph {nodes [
   n1 = RigVMFunctionEntryNode()
   RigUnit_SequenceExecution = &RigVMFunction_Sequence(ExecuteContext=@n1.ExecuteContext)
   RigUnit_SequenceExecution_1 = &RigVMFunction_Sequence(ExecuteContext=@RigUnit_SequenceExecution.B)
   RigUnit_SequenceExecution_2 = &RigVMFunction_Sequence(ExecuteContext=@RigUnit_SequenceExecution_1.B)
   RigVMFunctionReturnNode(A=@RigUnit_SequenceExecution.A, B=@RigUnit_SequenceExecution_1.A, C=@RigUnit_SequenceExecution_2.A, D=@RigUnit_SequenceExecution_2.B)
  ]}
 ContainedGraph {nodes [
   n1 = RigVMFunctionEntryNode()
   RigUnit_SequenceExecution_1 = &RigVMFunction_Sequence(ExecuteContext=@n1.ExecuteContext)
   RigUnit_SequenceExecution_1_1 = &RigVMFunction_Sequence(ExecuteContext=@RigUnit_SequenceExecution_1.B)
   RigUnit_SequenceExecution_1_2 = &RigVMFunction_Sequence(ExecuteContext=@RigUnit_SequenceExecution_1_1.B)
   RigVMFunctionReturnNode(A=@RigUnit_SequenceExecution_1.A, B=@RigUnit_SequenceExecution_1_1.A, C=@RigUnit_SequenceExecution_1_2.A, D=@RigUnit_SequenceExecution_1_2.B)
  ]}
 ContainedGraph {nodes [
   n1 = RigVMFunctionEntryNode()
   n2 = RigVMDispatchNode(Num=4, Array=RigVMVariableNode() {variable=Output Points}.Value, ExecuteContext=@n1.Execute) {template="DISPATCH_RigVMDispatch_ArraySetNum(io Array,in Num)"}
   n3 = RigVMVariableNode() {variable=Point 0}
   > n4 = RigVMDispatchNode(Index=0, Array=@n2.Array, Element=@n3.Value) {template="DISPATCH_RigVMDispatch_ArraySetAtIndex(io Array,in Index,in Element)"}
   n5 = RigVMVariableNode() {variable=t}
   n6 = RigVMVariableNode() {variable=t}
   n7 = /Script/RigVM.RigVMFunction_MathFloatSub(A=1, B=@n6.Value) {is=struct RigVMUnitNode, template="Subtract::Execute(in A,in B,out Result)"}
   Get A and C = RigVMCollapseNode(order=3, Start Point=@n1.Point 0, Middle Point=@n1.Point 1, End Point=@n1.Point 2, t=@n1.t)
   n8 = RigVMDispatchNode(
    Condition=true
    False=3
    True=&RigVMFunction_MathVectorDistance(B=@"Get A and C".C, A=RigVMVariableNode() {variable=Point 1}.Value).Result
   ) {template="DISPATCH_RigVMDispatch_If(in Condition,in True,in False,out Result)"}
   n9 = /Script/RigVM.RigVMFunction_MathVectorUnit(Value=/Script/RigVM.RigVMFunction_MathVectorSub(B=RigVMVariableNode() {variable=Point 0}.Value, A=RigVMVariableNode() {variable=Point 2}.Value) {is=struct RigVMUnitNode, template="Subtract::Execute(in A,in B,out Result)"}.Result) {is=struct RigVMUnitNode, template="Unit::Execute(in Value,out Result)"}
   n10 = RigVMVariableNode() {variable=Point 1}
   n11 = &RigVMFunction_MathVectorSub(
    B=/Script/RigVM.RigVMFunction_MathVectorScale(Factor=@n8.Result, Value=@n9.Result) {is=struct RigVMUnitNode, template="Scale::Execute(in Value,in Factor,out Result)"}.Result
    A=@n10.Value
   ) {template="Subtract::Execute(in A,in B,out Result)"}
   n12 = &RigVMFunction_MathVectorAdd(
    B=&RigVMFunction_MathVectorScale(
     Factor=/Script/RigVM.RigVMFunction_MathFloatDiv(A=1, B=@n7.Result) {is=struct RigVMUnitNode, template="Divide::Execute(in A,in B,out Result)"}.Result
     Value=&RigVMFunction_MathVectorSub(A=@n11.Result, B=@"Get A and C".A) {template="Subtract::Execute(in A,in B,out Result)"}.Result
    ) {template="Scale::Execute(in Value,in Factor,out Result)"}.Result
    A=@"Get A and C".A
   ) {template="Add::Execute(in A,in B,out Result)"}
   n13 = RigVMVariableNode() {variable=Point 0}
   n14 = &RigVMFunction_MathVectorAdd(
    B=&RigVMFunction_MathVectorScale(
     Factor=&RigVMFunction_MathFloatDiv(A=1, B=@n5.Value) {template="Divide::Execute(in A,in B,out Result)"}.Result
     Value=&RigVMFunction_MathVectorSub(A=@n12.Result, B=@n13.Value) {template="Subtract::Execute(in A,in B,out Result)"}.Result
    ) {template="Scale::Execute(in Value,in Factor,out Result)"}.Result
    A=@n13.Value
   ) {template="Add::Execute(in A,in B,out Result)"}
   > n15 = RigVMDispatchNode(Index=1, Array=@n4.Array, Element=@n14.Result) {template="DISPATCH_RigVMDispatch_ArraySetAtIndex(io Array,in Index,in Element)"}
   n16 = &RigVMFunction_MathFloatSub(A=1, B=@n5.Value) {template="Subtract::Execute(in A,in B,out Result)"}
   n17 = RigVMVariableNode() {variable=t}
   n18 = &RigVMFunction_MathFloatMul(A=@n8.Result, B=&RigVMFunction_MathFloatSub(A=1, B=@n17.Value) {template="Subtract::Execute(in A,in B,out Result)"}.Result) {template="Multiply::Execute(in A,in B,out Result)"}
   n19 = &RigVMFunction_MathVectorAdd(
    B=&RigVMFunction_MathVectorScale(
     Factor=&RigVMFunction_MathFloatDiv(A=@n18.Result, B=@n17.Value) {template="Divide::Execute(in A,in B,out Result)"}.Result
     Value=@n9.Result
    ) {template="Scale::Execute(in Value,in Factor,out Result)"}.Result
    A=@n10.Value
   ) {template="Add::Execute(in A,in B,out Result)"}
   n20 = &RigVMFunction_MathVectorAdd(
    B=&RigVMFunction_MathVectorScale(
     Factor=&RigVMFunction_MathFloatDiv(A=1, B=@n6.Value) {template="Divide::Execute(in A,in B,out Result)"}.Result
     Value=&RigVMFunction_MathVectorSub(A=@n19.Result, B=@"Get A and C".A) {template="Subtract::Execute(in A,in B,out Result)"}.Result
    ) {template="Scale::Execute(in Value,in Factor,out Result)"}.Result
    A=@"Get A and C".A
   ) {template="Add::Execute(in A,in B,out Result)"}
   n21 = RigVMVariableNode() {variable=Point 2}
   n22 = &RigVMFunction_MathVectorAdd(
    B=&RigVMFunction_MathVectorScale(
     Factor=&RigVMFunction_MathFloatDiv(A=1, B=@n16.Result) {template="Divide::Execute(in A,in B,out Result)"}.Result
     Value=&RigVMFunction_MathVectorSub(A=@n20.Result, B=@n21.Value) {template="Subtract::Execute(in A,in B,out Result)"}.Result
    ) {template="Scale::Execute(in Value,in Factor,out Result)"}.Result
    A=@n21.Value
   ) {template="Add::Execute(in A,in B,out Result)"}
   > n23 = RigVMDispatchNode(Index=2, Array=@n15.Array, Element=@n22.Result) {template="DISPATCH_RigVMDispatch_ArraySetAtIndex(io Array,in Index,in Element)"}
   n24 = RigVMVariableNode() {variable=Point 2}
   > n25 = RigVMDispatchNode(Index=3, Array=@n23.Array, Element=@n24.Value) {template="DISPATCH_RigVMDispatch_ArraySetAtIndex(io Array,in Index,in Element)"}
   RerouteNode = RigVMDispatchNode(
    Values=((X=0,Y=0,Z=0),((X=0,Y=0,Z=0),X=0,Y=0,Z=0),((X=0,Y=0,Z=0),X=0,Y=0,Z=0),((X=0,Y=0,Z=0),X=0,Y=0,Z=0))
    Values.1=@n14.Result
    Values.2=@n22.Result
    Values.3=@n24.Value
    Values.0=@n3.Value
   ) {template="DISPATCH_RigVMDispatch_ArrayMake(in Values,out Array)"}
   RigVMFunctionReturnNode(
    Control Points=@n25.Array
    Execute=@n25.ExecuteContext
    Spline=&RigUnit_ControlRigSplineFromPoints(SamplesPerSegment=16, SplineMode=BSpline, bClosed=False, Points=@RerouteNode.Array).Spline
    Out Point 4=@n24.Value
    Out Point 1=@n3.Value
    Out Point 2=@n22.Result
    Out Point 3=@n14.Result
   )
  ]}
 ContainedGraph {nodes [
   n1 = RigVMFunctionEntryNode()
   Projection Ratio = RigVMCollapseNode(order=@n1.order, t=@n1.t)
   n2 = &RigVMFunction_MathVectorAdd(
    A=&RigVMFunction_MathVectorScale(Factor=@"Projection Ratio".Result, Value=RigVMVariableNode() {variable=Start Point}.Value) {template="Scale::Execute(in Value,in Factor,out Result)"}.Result
    B=&RigVMFunction_MathVectorScale(
     Factor=&RigVMFunction_MathFloatSub(A=1, B=@"Projection Ratio".Result) {template="Subtract::Execute(in A,in B,out Result)"}.Result
     Value=RigVMVariableNode() {variable=End Point}.Value
    ) {template="Scale::Execute(in Value,in Factor,out Result)"}.Result
   ) {template="Add::Execute(in A,in B,out Result)"}
   ABC Ratio = RigVMCollapseNode(order=@n1.order, t=@n1.t)
   n3 = &RigVMFunction_MathFloatDiv(A=1, B=@"ABC Ratio".Result) {template="Divide::Execute(in A,in B,out Result)"}
   n4 = RigVMVariableNode() {variable=Middle Point}
   n5 = &RigVMFunction_MathVectorSub(B=@n2.Result, A=@n4.Value) {template="Subtract::Execute(in A,in B,out Result)"}
   RigVMFunctionReturnNode(
    C=@n2.Result
    A=&RigVMFunction_MathVectorAdd(
     B=&RigVMFunction_MathVectorScale(Factor=@n3.Result, Value=@n5.Result) {template="Scale::Execute(in Value,in Factor,out Result)"}.Result
     A=@n4.Value
    ) {template="Add::Execute(in A,in B,out Result)"}.Result
   )
  ]}
 ContainedGraph {nodes [
   n1 = RigVMFunctionEntryNode()
   n2 = /Script/RigVM.RigVMFunction_MathIntToFloat(Value=@n1.order) {is=struct RigVMUnitNode, template="Cast::Execute(in Value,out Result)"}
   n3 = /Script/RigVM.RigVMFunction_MathFloatPow(A=&RigVMFunction_MathFloatSub(A=1, B=@n1.t) {template="Subtract::Execute(in A,in B,out Result)"}.Result, B=@n2.Result) {is=struct RigVMUnitNode, template="Power::Execute(in A,in B,out Result)"}
   n4 = &RigVMFunction_MathFloatPow(A=@n1.t, B=@n2.Result) {template="Power::Execute(in A,in B,out Result)"}
   RigVMFunctionReturnNode(Result=&RigVMFunction_MathFloatDiv(B=&RigVMFunction_MathFloatAdd(A=@n3.Result, B=@n4.Result) {template="Add::Execute(in A,in B,out Result)"}.Result, A=@n3.Result) {template="Divide::Execute(in A,in B,out Result)"}.Result)
  ]}
 ContainedGraph {nodes [
   n1 = RigVMFunctionEntryNode()
   n2 = &RigVMFunction_MathIntToFloat(Value=@n1.order) {template="Cast::Execute(in Value,out Result)"}
   n3 = &RigVMFunction_MathFloatAdd(
    B=&RigVMFunction_MathFloatPow(A=@n1.t, B=@n2.Result) {template="Power::Execute(in A,in B,out Result)"}.Result
    A=&RigVMFunction_MathFloatPow(A=&RigVMFunction_MathFloatSub(A=1, B=@n1.t) {template="Subtract::Execute(in A,in B,out Result)"}.Result, B=@n2.Result) {template="Power::Execute(in A,in B,out Result)"}.Result
   ) {template="Add::Execute(in A,in B,out Result)"}
   n4 = &RigVMFunction_MathFloatSub(B=1, A=@n3.Result) {template="Subtract::Execute(in A,in B,out Result)"}
   RigVMFunctionReturnNode(Result=/Script/RigVM.RigVMFunction_MathFloatAbs(Value=&RigVMFunction_MathFloatDiv(B=@n3.Result, A=@n4.Result) {template="Divide::Execute(in A,in B,out Result)"}.Result) {is=struct RigVMUnitNode, template="Absolute::Execute(in Value,out Result)"}.Result)
  ]}
 ContainedGraph {nodes [
   n1 = RigVMFunctionEntryNode()
   RigUnit_SequenceAggregate = &RigVMFunction_Sequence(ExecuteContext=@n1.ExecuteContext)
   RigVMVariableNode(Value=1000000000, ExecuteContext=@RigUnit_SequenceAggregate.A) {variable=closestDistanceSoFar}
   > n2 = RigVMDispatchNode(Array=@n1.Bones) {template="DISPATCH_RigVMDispatch_ArrayIterator(in Array,out Element,out Index,out Count,out Ratio)"}
   Item = /Script/ControlRig.RigUnit_Item(Item=@n1.Item) {is=struct RigVMUnitNode}
   n3 = &RigVMFunction_MathVectorDistance(
    A=&RigUnit_GetTransform(bInitial=@n1.Initial, Item=@n2.Element).Transform.Translation
    B=&RigUnit_GetTransform(bInitial=@n1.Initial, Item=@Item.Item).Transform.Translation
   )
   n4 = &RigVMFunction_ControlFlowBranch(
    ExecuteContext=@n2.ExecuteContext
    Condition=/Script/RigVM.RigVMFunction_MathFloatLessEqual(A=@n3.Result, B=RigVMVariableNode() {variable=closestDistanceSoFar}.Value) {is=struct RigVMUnitNode, template="LessEqual::Execute(in A,in B,out Result)"}.Result
   )
   RigVMVariableNode(ExecuteContext=@n4.True, Value=@n3.Result) {variable=closestDistanceSoFar}
   > RigVMVariableNode(Value=@n2.Element) {variable=closestBone}
   RigVMFunctionReturnNode(ExecuteContext=@RigUnit_SequenceAggregate.B, Value=RigVMVariableNode() {variable=closestBone}.Value)
  ]}
 ContainedGraph {nodes [
   n1 = RigVMVariableNode() {variable=Point 3}
   n2 = RigVMFunctionEntryNode()
   n3 = &RigVMFunction_MathVectorDistance(B=@n2.Point 0, A=@n2.Point 1)
   n4 = &RigVMFunction_MathFloatAdd(A=@n3.Result, B=&RigVMFunction_MathVectorDistance(A=@n2.Point 1, B=@n2.Point 2).Result) {template="Add::Execute(in A,in B,out Result)"}
   n5 = &RigVMFunction_MathFloatAdd(A=@n4.Result, B=&RigVMFunction_MathVectorDistance(A=@n2.Point 2, B=@n2.Point 3).Result) {template="Add::Execute(in A,in B,out Result)"}
   n6 = &RigVMFunction_MathFloatDiv(B=@n5.Result, A=@n3.Result) {template="Divide::Execute(in A,in B,out Result)"}
   b3t1 = RigVMCollapseNode(t=@n6.Result)
   n7 = RigVMVariableNode() {variable=Point 0}
   b0t1 = RigVMCollapseNode(t=@n6.Result)
   n8 = &RigVMFunction_MathVectorScale(Value=@n7.Value, Factor=@b0t1.Result) {template="Scale::Execute(in Value,in Factor,out Result)"}
   n9 = RigVMVariableNode() {variable=Point 1}
   n10 = &RigVMFunction_MathFloatDiv(A=@n4.Result, B=@n5.Result) {template="Divide::Execute(in A,in B,out Result)"}
   b3t2 = RigVMCollapseNode(t=@n10.Result)
   b0t2 = RigVMCollapseNode(t=@n10.Result)
   n11 = &RigVMFunction_MathVectorScale(Value=@n7.Value, Factor=@b0t2.Result) {template="Scale::Execute(in Value,in Factor,out Result)"}
   n12 = RigVMVariableNode() {variable=Point 2}
   b1t1 = RigVMCollapseNode(t=@n6.Result)
   b1t2 = RigVMCollapseNode(t=@n10.Result)
   b2t1 = RigVMCollapseNode(t=@n6.Result)
   b2t2 = RigVMCollapseNode(t=@n10.Result)
   Solve = RigVMCollapseNode(
    c=&RigVMFunction_MathVectorSub(
     B=&RigVMFunction_MathVectorScale(Value=@n1.Value, Factor=@b3t1.Result) {template="Scale::Execute(in Value,in Factor,out Result)"}.Result
     A=&RigVMFunction_MathVectorSub(B=@n8.Result, A=@n9.Value) {template="Subtract::Execute(in A,in B,out Result)"}.Result
    ) {template="Subtract::Execute(in A,in B,out Result)"}.Result
    f=&RigVMFunction_MathVectorSub(
     B=&RigVMFunction_MathVectorScale(Value=@n1.Value, Factor=@b3t2.Result) {template="Scale::Execute(in Value,in Factor,out Result)"}.Result
     A=&RigVMFunction_MathVectorSub(B=@n11.Result, A=@n12.Value) {template="Subtract::Execute(in A,in B,out Result)"}.Result
    ) {template="Subtract::Execute(in A,in B,out Result)"}.Result
    a=@b1t1.Result
    d=@b1t2.Result
    b=@b2t1.Result
    e=@b2t2.Result
   )
   n13 = RigVMVariableNode() {variable=Point 0}
   n14 = RigVMVariableNode() {variable=Point 3}
   RigVMFunctionReturnNode(
    Spline=&RigUnit_ControlRigSplineFromPoints(
     Points=((X=0,Y=0,Z=0),(X=0,Y=0,Z=0),(X=0,Y=0,Z=0),(X=0,Y=0,Z=0))
     SamplesPerSegment=16
     SplineMode=BSpline
     bClosed=False
     Points.1=@Solve.i
     Points.2=@Solve.j
     Points.0=@n13.Value
     Points.3=@n14.Value
    ).Spline
    Out Point 2=@Solve.i
    Out Point 3=@Solve.j
    Out Point 1=@n13.Value
    Out Point 4=@n14.Value
    Execute=@n2.Execute
   )
  ]}
 ContainedGraph {nodes [
   RigVMFunctionEntryNode()
   n1 = &RigVMFunction_MathFloatDiv(
    A=&RigVMFunction_MathFloatMul(A=RigVMVariableNode() {variable=a}.Value, B=RigVMVariableNode() {variable=e}.Value) {template="Multiply::Execute(in A,in B,out Result)"}.Result
    B=RigVMVariableNode() {variable=d}.Value
   ) {template="Divide::Execute(in A,in B,out Result)"}
   n2 = RigVMVariableNode() {variable=b}
   n3 = &RigVMFunction_MathFloatDiv(A=RigVMVariableNode() {variable=a}.Value, B=RigVMVariableNode() {variable=d}.Value) {template="Divide::Execute(in A,in B,out Result)"}
   n4 = RigVMVariableNode() {variable=f}
   n5 = &RigVMFunction_MathVectorScale(
    Factor=&RigVMFunction_MathFloatDiv(A=1, B=&RigVMFunction_MathFloatSub(B=@n1.Result, A=@n2.Value) {template="Subtract::Execute(in A,in B,out Result)"}.Result) {template="Divide::Execute(in A,in B,out Result)"}.Result
    Value=&RigVMFunction_MathVectorSub(
     B=&RigVMFunction_MathVectorScale(Factor=@n3.Result, Value=@n4.Value) {template="Scale::Execute(in Value,in Factor,out Result)"}.Result
     A=RigVMVariableNode() {variable=c}.Value
    ) {template="Subtract::Execute(in A,in B,out Result)"}.Result
   ) {template="Scale::Execute(in Value,in Factor,out Result)"}
   n6 = RigVMVariableNode() {variable=a}
   n7 = &RigVMFunction_MathVectorScale(Value=@n5.Result, Factor=RigVMVariableNode() {variable=b}.Value) {template="Scale::Execute(in Value,in Factor,out Result)"}
   n8 = RigVMVariableNode() {variable=c}
   RigVMFunctionReturnNode(
    j=@n5.Result
    i=&RigVMFunction_MathVectorScale(
     Factor=&RigVMFunction_MathFloatDiv(A=1, B=@n6.Value) {template="Divide::Execute(in A,in B,out Result)"}.Result
     Value=&RigVMFunction_MathVectorSub(B=@n7.Result, A=@n8.Value) {template="Subtract::Execute(in A,in B,out Result)"}.Result
    ) {template="Scale::Execute(in Value,in Factor,out Result)"}.Result
   )
  ]}
 ContainedGraph {nodes [
   n1 = RigVMFunctionEntryNode()
   n2 = &RigVMFunction_MathFloatSub(A=1, B=@n1.t) {template="Subtract::Execute(in A,in B,out Result)"}
   n3 = &RigVMFunction_MathFloatMul(A=@n2.Result, B=@n2.Result) {template="Multiply::Execute(in A,in B,out Result)"}
   RigVMFunctionReturnNode(Result=&RigVMFunction_MathFloatMul(B=3, A=&RigVMFunction_MathFloatMul(A=@n1.t, B=@n3.Result) {template="Multiply::Execute(in A,in B,out Result)"}.Result) {template="Multiply::Execute(in A,in B,out Result)"}.Result)
  ]}
 ContainedGraph {nodes [
   n1 = RigVMFunctionEntryNode()
   n2 = &RigVMFunction_MathFloatMul(A=@n1.t, B=@n1.t) {template="Multiply::Execute(in A,in B,out Result)"}
   n3 = &RigVMFunction_MathFloatSub(A=1, B=@n1.t) {template="Subtract::Execute(in A,in B,out Result)"}
   RigVMFunctionReturnNode(Result=&RigVMFunction_MathFloatMul(B=3, A=&RigVMFunction_MathFloatMul(B=@n2.Result, A=@n3.Result) {template="Multiply::Execute(in A,in B,out Result)"}.Result) {template="Multiply::Execute(in A,in B,out Result)"}.Result)
  ]}
 ContainedGraph {nodes [
   n1 = RigVMFunctionEntryNode()
   n2 = &RigVMFunction_MathFloatSub(A=1, B=@n1.t) {template="Subtract::Execute(in A,in B,out Result)"}
   n3 = &RigVMFunction_MathFloatMul(A=@n2.Result, B=@n2.Result) {template="Multiply::Execute(in A,in B,out Result)"}
   RigVMFunctionReturnNode(Result=&RigVMFunction_MathFloatMul(B=3, A=&RigVMFunction_MathFloatMul(A=@n1.t, B=@n3.Result) {template="Multiply::Execute(in A,in B,out Result)"}.Result) {template="Multiply::Execute(in A,in B,out Result)"}.Result)
  ]}
 ContainedGraph {nodes [
   n1 = RigVMFunctionEntryNode()
   n2 = &RigVMFunction_MathFloatMul(A=@n1.t, B=@n1.t) {template="Multiply::Execute(in A,in B,out Result)"}
   n3 = &RigVMFunction_MathFloatSub(A=1, B=@n1.t) {template="Subtract::Execute(in A,in B,out Result)"}
   RigVMFunctionReturnNode(Result=&RigVMFunction_MathFloatMul(B=3, A=&RigVMFunction_MathFloatMul(B=@n2.Result, A=@n3.Result) {template="Multiply::Execute(in A,in B,out Result)"}.Result) {template="Multiply::Execute(in A,in B,out Result)"}.Result)
  ]}
 ContainedGraph {nodes [
   n1 = RigVMFunctionEntryNode()
   RigVMFunctionReturnNode(Result=&RigVMFunction_MathFloatPow(B=3, A=&RigVMFunction_MathFloatSub(A=1, B=@n1.t) {template="Subtract::Execute(in A,in B,out Result)"}.Result) {template="Power::Execute(in A,in B,out Result)"}.Result)
  ]}
 ContainedGraph {nodes [RigVMFunctionReturnNode(Result=&RigVMFunction_MathFloatPow(B=3, A=RigVMFunctionEntryNode().t) {template="Power::Execute(in A,in B,out Result)"}.Result)]}
 ContainedGraph {nodes [
   n1 = RigVMFunctionEntryNode()
   RigVMFunctionReturnNode(Result=&RigVMFunction_MathFloatPow(B=3, A=&RigVMFunction_MathFloatSub(A=1, B=@n1.t) {template="Subtract::Execute(in A,in B,out Result)"}.Result) {template="Power::Execute(in A,in B,out Result)"}.Result)
  ]}
 ContainedGraph {nodes [RigVMFunctionReturnNode(Result=&RigVMFunction_MathFloatPow(B=3, A=RigVMFunctionEntryNode().t) {template="Power::Execute(in A,in B,out Result)"}.Result)]}
 ContainedGraph {nodes [
   n1 = RigVMFunctionEntryNode()
   SequenceExecution = RigVMAggregateNode(ExecuteContext=@n1.Execute)
   n2 = RigVMVariableNode() {variable=Bones}
   LessEqual = &RigVMFunction_MathIntLessEqual(B=3, A=RigVMDispatchNode(Array=@n2.Value) {template="DISPATCH_RigVMDispatch_ArrayGetNum(in Array,out Num)"}.Num) {template="LessEqual::Execute(in A,in B,out Result)"}
   n3 = &RigVMFunction_ControlFlowBranch(Condition=@LessEqual.Result, ExecuteContext=@SequenceExecution.D)
   DISPATCH_RigVMDispatch_ArraySetNum = RigVMDispatchNode(Num=4, ExecuteContext=@n3.True, Array=@n2.Value) {template="DISPATCH_RigVMDispatch_ArraySetNum(io Array,in Num)"}
   DISPATCH_RigVMDispatch_ArrayGetAtIndex = RigVMDispatchNode(Index=3, Array=RigVMVariableNode() {variable=Bones}.Value) {template="DISPATCH_RigVMDispatch_ArrayGetAtIndex(in Array,in Index,out Element)"}
   > DISPATCH_RigVMDispatch_ArraySetAtIndex = RigVMDispatchNode(Index=3, Element=@DISPATCH_RigVMDispatch_ArrayGetAtIndex.Element, Array=@DISPATCH_RigVMDispatch_ArraySetNum.Array) {template="DISPATCH_RigVMDispatch_ArraySetAtIndex(io Array,in Index,in Element)"}
   Generate Spline Points = RigVMFunctionReferenceNode(Items=@n1.Bones, ExecuteContext=@SequenceExecution.C)
   > RigVMVariableNode(Value=@"Generate Spline Points".Value) {variable=Points}
   n4 = RigVMDispatchNode(Array=RigVMVariableNode() {variable=IK Controls}.Value, ExecuteContext=@SequenceExecution.E) {template="DISPATCH_RigVMDispatch_ArrayIterator(in Array,out Element,out Index,out Count,out Ratio)"}
   MathIntEquals_1 = RigVMDispatchNode(B=0, A=@n4.Index) {template="DISPATCH_RigVMDispatch_CoreEquals(in A,in B,out Result)"}
   n5 = &RigVMFunction_ControlFlowBranch(ExecuteContext=@n4.ExecuteContext, Condition=@MathIntEquals_1.Result)
   n6 = RigVMDispatchNode(Index=@n4.Index, Array=RigVMVariableNode() {variable=Bones}.Value) {template="DISPATCH_RigVMDispatch_ArrayGetAtIndex(in Array,in Index,out Element)"}
   n7 = /Script/ControlRig.RigUnit_ProjectTransformToNewParent(bChildInitial=True, bOldParentInitial=True, NewParent=@n6.Element, OldParent=@n6.Element, Child=@n4.Element) {is=struct RigVMUnitNode}
   n8 = /Script/RigVM.RigVMFunction_MathTransformMakeAbsolute(
    Local=(Rotation=(X=0,Y=0,Z=0,W=1),Translation=(X=0,Y=0,Z=0),Scale3D=(X=1,Y=1,Z=1))
    Parent=(Rotation=(X=0,Y=0,Z=0,W=1),Translation=(X=0,Y=0,Z=0),Scale3D=(X=1,Y=1,Z=1))
    Parent.Translation=RigVMDispatchNode(Index=@n4.Index, Array=RigVMVariableNode() {variable=Points}.Value) {template="DISPATCH_RigVMDispatch_ArrayGetAtIndex(in Array,in Index,out Element)"}.Element
    Parent.Rotation=@n7.Transform.Rotation
    Parent.Scale3D=@n7.Transform.Scale3D
   ) {is=struct RigVMUnitNode, template="Make Absolute::Execute(in Local,in Parent,out Global)"}
   &RigUnit_SetTransform(Weight=1, bPropagateToChildren=false, Item=@n4.Element, ExecutePin=@n5.False, Value=@n8.Global)
   RigVMFunctionReturnNode(Execute=@SequenceExecution.F)
  ]}
 ContainedGraph {nodes [
   n1 = RigVMFunctionEntryNode()
   SequenceExecution = &RigVMFunction_Sequence(ExecuteContext=@n1.ExecuteContext)
   SequenceExecution_1 = &RigVMFunction_Sequence(ExecuteContext=@SequenceExecution.B)
   SequenceExecution_2 = &RigVMFunction_Sequence(ExecuteContext=@SequenceExecution_1.B)
   SequenceExecution_3 = &RigVMFunction_Sequence(ExecuteContext=@SequenceExecution_2.B)
   SequenceExecution_4 = &RigVMFunction_Sequence(ExecuteContext=@SequenceExecution_3.B)
   RigVMFunctionReturnNode(
    A=@SequenceExecution.A
    B=@SequenceExecution_1.A
    C=@SequenceExecution_2.A
    D=@SequenceExecution_3.A
    E=@SequenceExecution_4.A
    F=@SequenceExecution_4.B
   )
  ]}
 ContainedGraph {nodes [
   n1 = RigVMFunctionEntryNode()
   > n2 = RigVMDispatchNode(Array=@n1.Children) {template="DISPATCH_RigVMDispatch_ArrayIterator(in Array,out Element,out Index,out Count,out Ratio)"}
   n3 = RigVMDispatchNode(Index=@n2.Index, Array=@n1.Parents) {template="DISPATCH_RigVMDispatch_ArrayGetAtIndex(in Array,in Index,out Element)"}
   > &RigUnit_SetTransform(
    Weight=1
    Item=@n2.Element
    bPropagateToChildren=@n1.Propagate to Children
    Value=&RigUnit_ProjectTransformToNewParent(bChildInitial=True, bOldParentInitial=True, NewParent=@n3.Element, OldParent=@n3.Element, Child=@n2.Element).Transform
   )
   RigVMFunctionReturnNode(ExecuteContext=@n2.Completed)
  ]}
 ContainedGraph {nodes [
   n1 = RigVMFunctionEntryNode()
   > SequenceExecution = RigVMAggregateNode()
   n2 = RigVMDispatchNode(ExecuteContext=@SequenceExecution.A, Array=RigVMVariableNode() {variable=FK Controls}.Value) {template="DISPATCH_RigVMDispatch_ArrayIterator(in Array,out Element,out Index,out Count,out Ratio)"}
   > /Script/ControlRig.RigUnit_SetControlVisibility(
    Item=@n2.Element
    bVisible=RigVMDispatchNode(False=False, Condition=@n1.Show Controls, True=&RigVMFunction_MathBoolNot(Value=@n1.IKFK Switch).Result) {template="DISPATCH_RigVMDispatch_If(in Condition,in True,in False,out Result)"}.Result
   ) {is=struct RigVMUnitNode, template="SetControlVisibility::Execute(in Item,in Pattern,in bVisible)"}
   n3 = RigVMDispatchNode(ExecuteContext=@SequenceExecution.B, Array=RigVMVariableNode() {variable=IK Controls}.Value) {template="DISPATCH_RigVMDispatch_ArrayIterator(in Array,out Element,out Index,out Count,out Ratio)"}
   > &RigUnit_SetControlVisibility(
    Item=@n3.Element
    bVisible=RigVMDispatchNode(False=False, True=@n1.IKFK Switch, Condition=@n1.Show Controls) {template="DISPATCH_RigVMDispatch_If(in Condition,in True,in False,out Result)"}.Result
   ) {template="SetControlVisibility::Execute(in Item,in Pattern,in bVisible)"}
   RigVMFunctionReturnNode(ExecuteContext=@SequenceExecution.D)
  ]}
 ContainedGraph {nodes [
   n1 = RigVMFunctionEntryNode()
   SequenceExecution = &RigVMFunction_Sequence(ExecuteContext=@n1.ExecuteContext)
   SequenceExecution_1 = &RigVMFunction_Sequence(ExecuteContext=@SequenceExecution.B)
   SequenceExecution_2 = &RigVMFunction_Sequence(ExecuteContext=@SequenceExecution_1.B)
   RigVMFunctionReturnNode(A=@SequenceExecution.A, B=@SequenceExecution_1.A, C=@SequenceExecution_2.A, D=@SequenceExecution_2.B)
  ]}
 ContainedGraph {nodes [
   BranchNode_RigVMUnitNode = &RigVMFunction_ControlFlowBranch(ExecuteContext=(), Condition=RigVMVariableNode() {variable=debug}.Value)
   &RigUnit_DrawControlRigSpline(
    Color=(R=1,G=0,B=0,A=1)
    Detail=16
    Thickness=0.75
    ExecutePin=@BranchNode_RigVMUnitNode.True
    Spline=RigVMVariableNode() {variable=template_spline}.Value
   )
   n1 = RigVMFunctionEntryNode()
   > RigVMFunction_Sequence = RigVMAggregateNode()
   n2 = RigVMVariableNode() {variable=Controls}
   BranchNode_4_RigVMUnitNode = &RigVMFunction_ControlFlowBranch(
    Condition=&RigVMFunction_MathIntLess(B=4, A=RigVMDispatchNode(Array=@n2.Value) {template="DISPATCH_RigVMDispatch_ArrayGetNum(in Array,out Num)"}.Num) {template="Less::Execute(in A,in B,out Result)"}.Result
    ExecuteContext=@RigVMFunction_Sequence.B
   )
   RigUnit_SequenceExecution_1 = RigVMAggregateNode(ExecuteContext=@BranchNode_4_RigVMUnitNode.False)
   RigVMVariableNode(Value=@n1.Template Spline, ExecuteContext=@RigUnit_SequenceExecution_1.A) {variable=template_spline}
   n3 = &RigUnit_GetTransformItemArray(Space=GlobalSpace, bInitial=False, Items=RigVMVariableNode() {variable=Controls}.Value)
   > n4 = &RigUnit_SetSplineTransforms(Transforms=@n3.Transforms, Spline=RigVMVariableNode() {variable=template_spline}.Value)
   > RigVMVariableNode(Value=@n4.Spline) {variable=local_spline}
   > RigVMVariableNode(Value=@n3.Transforms) {variable=template_transforms}
   > RigUnit_SequenceExecution = RigVMAggregateNode()
   n5 = RigVMVariableNode() {variable=template_transforms}
   RigVMDispatchNode(
    Num=RigVMDispatchNode(Array=@n5.Value) {template="DISPATCH_RigVMDispatch_ArrayGetNum(in Array,out Num)"}.Num
    ExecuteContext=@RigUnit_SequenceExecution.A
    Array=RigVMVariableNode() {variable=offset_positions}.Value
   ) {template="DISPATCH_RigVMDispatch_ArraySetNum(io Array,in Num)"}
   > n6 = RigVMDispatchNode(Array=@n5.Value) {template="DISPATCH_RigVMDispatch_ArrayIterator(in Array,out Element,out Index,out Count,out Ratio)"}
   > RigVMDispatchNode(
    Index=@n6.Index
    Element=&RigVMFunction_MathTransformTransformVector(Transform=@n6.Element, Location=RigVMVariableNode() {variable=Secondary Spline Direction}.Value).Result
    Array=RigVMVariableNode() {variable=offset_positions}.Value
   ) {template="DISPATCH_RigVMDispatch_ArraySetAtIndex(io Array,in Index,in Element)"}
   RigUnit_SequenceExecution_4 = RigVMAggregateNode(ExecuteContext=@RigUnit_SequenceExecution.C)
   Branch_RigVMUnitNode = &RigVMFunction_ControlFlowBranch(ExecuteContext=@RigUnit_SequenceExecution_4.A, Condition=RigVMVariableNode() {variable=Stretch}.Value)
   n7 = RigVMVariableNode() {variable=template_spline}
   n8 = RigVMVariableNode() {variable=Primary Axis}
   n9 = RigVMVariableNode() {variable=Bones}
   &RigUnit_FitChainToSplineCurveItemArray(
    Alignment=Front
    DebugSettings=(Scale=1,CurveColor=(R=1,G=1,B=0,A=1),SegmentsColor=(R=1,G=0,B=0,A=1),WorldOffset=(Rotation=(X=0,Y=0,Z=0,W=1),Translation=(X=0,Y=0,Z=0),Scale3D=(X=1,Y=1,Z=1)))
    Maximum=1
    PoleVectorPosition=(X=0,Y=0,Z=0)
    Rotations=()
    SamplingPrecision=12
    SecondaryAxis=(X=0,Y=0,Z=0)
    Weight=1
    bPropagateToChildren=True
    ExecutePin=@Branch_RigVMUnitNode.False
    Spline=@n7.Value
    PrimaryAxis=@n8.Value
    Items=@n9.Value
   )
   n10 = RigVMVariableNode() {variable=template_spline}
   &RigUnit_FitChainToSplineCurveItemArray(
    Alignment=Stretched
    DebugSettings=(Scale=1,CurveColor=(R=1,G=1,B=0,A=1),SegmentsColor=(R=1,G=0,B=0,A=1),WorldOffset=(Rotation=(X=0,Y=0,Z=0,W=1),Translation=(X=0,Y=0,Z=0),Scale3D=(X=1,Y=1,Z=1)))
    Maximum=1
    PoleVectorPosition=(X=0,Y=0,Z=0)
    Rotations=()
    SamplingPrecision=12
    SecondaryAxis=(X=0,Y=0,Z=0)
    Weight=1
    bPropagateToChildren=True
    ExecutePin=@Branch_RigVMUnitNode.True
    Spline=@n10.Value
    PrimaryAxis=@n8.Value
    Items=RigVMVariableNode() {variable=Bones}.Value
   )
   RigUnit_SequenceExecution_2 = RigVMAggregateNode(ExecuteContext=@RigUnit_SequenceExecution_4.B)
   n11 = &RigUnit_SetSplinePoints(
    ExecutePin=@RigUnit_SequenceExecution_2.B
    Points=RigVMVariableNode() {variable=offset_positions}.Value
    Spline=RigVMVariableNode() {variable=local_spline}.Value
   )
   > RigVMVariableNode(Value=@n11.Spline) {variable=local_spline}
   n12 = RigVMVariableNode() {variable=Bones}
   DISPATCH_RigVMDispatch_ArrayGetNum = RigVMDispatchNode(Array=@n12.Value) {template="DISPATCH_RigVMDispatch_ArrayGetNum(in Array,out Num)"}
   > DISPATCH_RigVMDispatch_ArraySetNum = RigVMDispatchNode(Num=@DISPATCH_RigVMDispatch_ArrayGetNum.Num, Array=RigVMVariableNode() {variable=aim_transforms}.Value) {template="DISPATCH_RigVMDispatch_ArraySetNum(io Array,in Num)"}
   > n13 = RigVMDispatchNode(Array=@n12.Value) {template="DISPATCH_RigVMDispatch_ArrayIterator(in Array,out Element,out Index,out Count,out Ratio)"}
   n14 = RigVMVariableNode() {variable=local_spline}
   > DISPATCH_RigVMDispatch_ArraySetAtIndex = RigVMDispatchNode(
    Index=@n13.Index
    Element=&RigUnit_PositionFromControlRigSpline(
     U=/Script/ControlRigSpline.RigUnit_ParameterAtPercentage(Percentage=@n13.Ratio, Spline=@n14.Value) {is=struct RigVMUnitNode}.U
     Spline=@n14.Value
    ).Position
    Array=RigVMVariableNode() {variable=aim_transforms}.Value
   ) {template="DISPATCH_RigVMDispatch_ArraySetAtIndex(io Array,in Index,in Element)"}
   BranchNode_2_RigVMUnitNode = &RigVMFunction_ControlFlowBranch(ExecuteContext=@RigUnit_SequenceExecution_2.D, Condition=RigVMVariableNode() {variable=debug}.Value)
   &RigUnit_DrawControlRigSpline(
    Color=(R=1,G=0,B=0,A=1)
    Detail=16
    Thickness=1
    ExecutePin=@BranchNode_2_RigVMUnitNode.True
    Spline=RigVMVariableNode() {variable=local_spline}.Value
   )
   n15 = RigVMDispatchNode(ExecuteContext=@RigUnit_SequenceExecution.D, Array=RigVMVariableNode() {variable=Bones}.Value) {template="DISPATCH_RigVMDispatch_ArrayIterator(in Array,out Element,out Index,out Count,out Ratio)"}
   DISPATCH_RigVMDispatch_ArrayGetAtIndex = RigVMDispatchNode(Index=@n15.Index, Array=RigVMVariableNode() {variable=aim_transforms}.Value) {template="DISPATCH_RigVMDispatch_ArrayGetAtIndex(in Array,in Index,out Element)"}
   n16 = &RigUnit_AimBoneMath(
    DebugSettings=(Scale=10,WorldOffset=(Rotation=(X=0,Y=0,Z=0,W=1),Translation=(X=0,Y=0,Z=0),Scale3D=(X=1,Y=1,Z=1)))
    Primary=(Weight=1,Axis=(X=0,Y=1,Z=0),Target=(X=1,Y=0,Z=0),Kind=Location,Space=(Type=Bone))
    Secondary=(Weight=0,Axis=(X=0,Y=0,Z=0),Target=(X=0,Y=0,Z=0),Kind=Location,Space=(Type=Bone))
    Weight=1
    Primary.Target=@DISPATCH_RigVMDispatch_ArrayGetAtIndex.Element
    InputTransform=&RigUnit_GetTransform(Item=@n15.Element).Transform
    Primary.Axis=RigVMVariableNode() {variable=Up Axis}.Value
   )
   MathIntEquals = RigVMDispatchNode(A=@n15.Index, B=&RigVMFunction_MathIntSub(B=1, A=@n15.Count) {template="Subtract::Execute(in A,in B,out Result)"}.Result) {template="DISPATCH_RigVMDispatch_CoreEquals(in A,in B,out Result)"}
   > n17 = &RigUnit_SetTransform(Weight=1, bInitial=False, Value=@n16.Result, Item=@n15.Element, bPropagateToChildren=@MathIntEquals.Result)
   BranchNode_3_RigVMUnitNode = &RigVMFunction_ControlFlowBranch(ExecuteContext=@n17.ExecutePin, Condition=RigVMVariableNode() {variable=debug}.Value)
   &RigUnit_DebugTransformMutableItemSpace(
    Color=(R=1,G=0,B=0,A=1)
    DebugDrawSettings=(DepthPriority=SDPG_Foreground,Lifetime=-1)
    Mode=Axes
    Scale=10
    Space=(Type=Bone)
    Thickness=0.15
    WorldOffset=(Rotation=(X=0,Y=0,Z=0,W=1),Translation=(X=0,Y=0,Z=0),Scale3D=(X=1,Y=1,Z=1))
    bEnabled=True
    Transform=@n16.Result
    ExecutePin=@BranchNode_3_RigVMUnitNode.True
   )
   Branch_1_RigVMUnitNode = &RigVMFunction_ControlFlowBranch(ExecuteContext=@RigUnit_SequenceExecution_1.B, Condition=RigVMVariableNode() {variable=Stretch}.Value)
   n18 = RigVMDispatchNode(Index=-1, Array=RigVMVariableNode() {variable=Controls}.Value) {template="DISPATCH_RigVMDispatch_ArrayGetAtIndex(in Array,in Index,out Element)"}
   n19 = RigVMDispatchNode(Index=-1, Array=RigVMVariableNode() {variable=Bones}.Value) {template="DISPATCH_RigVMDispatch_ArrayGetAtIndex(in Array,in Index,out Element)"}
   &RigUnit_ParentConstraint(
    AdvancedSettings=(RotationOrderForFilter=XZY)
    Filter=(TranslationFilter=(bX=false,bY=false,bZ=false),RotationFilter=(bX=True,bY=True,bZ=True),ScaleFilter=(bX=True,bY=True,bZ=True))
    Parents=((Item=(Type=Bone),Weight=1))
    Weight=1
    Parents.0.Item=@n18.Element
    Child=@n19.Element
    ExecutePin=@Branch_1_RigVMUnitNode.False
   )
   &RigUnit_ParentConstraint(
    AdvancedSettings=(RotationOrderForFilter=XZY)
    Parents=((Item=(Type=Bone),Weight=1))
    Weight=1
    Parents.0.Item=@n18.Element
    Child=@n19.Element
    ExecutePin=@Branch_1_RigVMUnitNode.True
   )
   RigVMFunctionReturnNode(ExecuteContext=@RigVMFunction_Sequence.D, Output Spline=RigVMVariableNode() {variable=local_spline}.Value)
   For_Each_1 = RigVMDispatchNode(ExecuteContext=(), Array=RigVMVariableNode() {variable=Bones}.Value) {template="DISPATCH_RigVMDispatch_ArrayIterator(in Array,out Element,out Index,out Count,out Ratio)"}
   n20 = &RigUnit_GetTransform(bInitial=False, Item=@For_Each_1.Element)
   n21 = RigVMVariableNode() {variable=template_spline}
   > &RigUnit_SetTransform(
    Weight=1
    bInitial=false
    Item=@For_Each_1.Element
    Value=/Script/ControlRigSpline.RigUnit_TransformFromControlRigSpline2(
     PrimaryAxis=(X=0,Y=0,Z=0)
     SecondaryAxis=(X=0,Y=0,Z=0)
     U=&RigUnit_ClosestParameterFromControlRigSpline(Position=@n20.Transform.Translation, Spline=@n21.Value).U
     Spline=@n21.Value
    ) {is=struct RigVMUnitNode}.Transform
   )
   For_Each_1_1 = RigVMDispatchNode(ExecuteContext=(), Array=RigVMVariableNode() {variable=Bones}.Value) {template="DISPATCH_RigVMDispatch_ArrayIterator(in Array,out Element,out Index,out Count,out Ratio)"}
   n22 = RigVMVariableNode() {variable=template_spline}
   n23 = &RigUnit_ParameterAtPercentage(Percentage=@For_Each_1_1.Ratio, Spline=@n22.Value)
   > &RigUnit_SetTransform(
    Weight=1
    bInitial=false
    bPropagateToChildren=false
    Item=@For_Each_1_1.Element
    Value=/Script/ControlRig.RigUnit_ParentConstraintMath(
     AdvancedSettings=(InterpolationType=Average)
     Parents=((Item=(Type=Bone,Name="None"),Weight=1))
     Parents.0.Item=@For_Each_1_1.Element
     Input=&RigUnit_TransformFromControlRigSpline2(PrimaryAxis=(X=0,Y=0,Z=0), SecondaryAxis=(X=0,Y=0,Z=0), U=@n23.U, Spline=@n22.Value).Transform
    ) {is=struct RigVMUnitNode}.Output
   )
   &RigUnit_ClosestParameterFromControlRigSpline(Position=&RigUnit_GetTransform(Item=@n13.Element).Transform.Translation, Spline=@n14.Value)
   /Script/ControlRigSpline.RigUnit_SplineConstraint(
    Alignment=Stretched
    ExecutePin=()
    Maximum=1
    Minimum=0
    SecondaryAxis=(X=0,Y=0,Z=0)
    bPropagateToChildren=true
    Spline=@n10.Value
    PrimaryAxis=@n8.Value
    Items=RigVMVariableNode() {variable=Bones}.Value
   ) {is=struct RigVMUnitNode}
   &RigUnit_SplineConstraint(
    Alignment=Front
    ExecutePin=()
    Maximum=1
    Minimum=0
    SecondaryAxis=(X=0,Y=0,Z=0)
    bPropagateToChildren=true
    Spline=@n7.Value
    PrimaryAxis=@n8.Value
    Items=@n9.Value
   )
   RigVMVariableNode() {variable=template_spline}
   RigVMVariableNode() {variable=Spline Mode}
   RigVMVariableNode() {variable=Samples Per Segment}
  ]}
 ContainedGraph {nodes [
   n1 = RigVMFunctionEntryNode()
   RigUnit_SequenceExecution_2 = &RigVMFunction_Sequence(ExecuteContext=@n1.ExecuteContext)
   RigUnit_SequenceExecution_2_1 = &RigVMFunction_Sequence(ExecuteContext=@RigUnit_SequenceExecution_2.B)
   RigUnit_SequenceExecution_2_2 = &RigVMFunction_Sequence(ExecuteContext=@RigUnit_SequenceExecution_2_1.B)
   RigVMFunctionReturnNode(A=@RigUnit_SequenceExecution_2.A, B=@RigUnit_SequenceExecution_2_1.A, C=@RigUnit_SequenceExecution_2_2.A, D=@RigUnit_SequenceExecution_2_2.B)
  ]}
 ContainedGraph {nodes [
   n1 = RigVMFunctionEntryNode()
   RigUnit_SequenceExecution_4 = &RigVMFunction_Sequence(ExecuteContext=@n1.ExecuteContext)
   RigUnit_SequenceExecution_4_1 = &RigVMFunction_Sequence(ExecuteContext=@RigUnit_SequenceExecution_4.B)
   RigUnit_SequenceExecution_4_2 = &RigVMFunction_Sequence(ExecuteContext=@RigUnit_SequenceExecution_4_1.B)
   RigVMFunctionReturnNode(A=@RigUnit_SequenceExecution_4.A, B=@RigUnit_SequenceExecution_4_1.A, C=@RigUnit_SequenceExecution_4_2.A, D=@RigUnit_SequenceExecution_4_2.B)
  ]}
 ContainedGraph {nodes [
   n1 = RigVMFunctionEntryNode()
   RigUnit_SequenceExecution = &RigVMFunction_Sequence(ExecuteContext=@n1.ExecuteContext)
   RigUnit_SequenceExecution_1 = &RigVMFunction_Sequence(ExecuteContext=@RigUnit_SequenceExecution.B)
   RigUnit_SequenceExecution_2 = &RigVMFunction_Sequence(ExecuteContext=@RigUnit_SequenceExecution_1.B)
   RigVMFunctionReturnNode(A=@RigUnit_SequenceExecution.A, B=@RigUnit_SequenceExecution_1.A, C=@RigUnit_SequenceExecution_2.A, D=@RigUnit_SequenceExecution_2.B)
  ]}
 ContainedGraph {nodes [
   n1 = RigVMFunctionEntryNode()
   RigUnit_SequenceExecution_1 = &RigVMFunction_Sequence(ExecuteContext=@n1.ExecuteContext)
   RigUnit_SequenceExecution_1_1 = &RigVMFunction_Sequence(ExecuteContext=@RigUnit_SequenceExecution_1.B)
   RigUnit_SequenceExecution_1_2 = &RigVMFunction_Sequence(ExecuteContext=@RigUnit_SequenceExecution_1_1.B)
   RigVMFunctionReturnNode(A=@RigUnit_SequenceExecution_1.A, B=@RigUnit_SequenceExecution_1_1.A, C=@RigUnit_SequenceExecution_1_2.A, D=@RigUnit_SequenceExecution_1_2.B)
  ]}
 ContainedGraph {nodes [
   n1 = RigVMFunctionEntryNode()
   Replace = /Script/RigVM.RigVMFunction_NameReplace(New=None, Name=@n1.Item.Name, Old=@n1.Replace Suffix) {is=struct RigVMUnitNode, template="Replace::Execute(in Name,in Old,in New,out Result)"}
   Concat = &RigVMFunction_NameConcat(B=@n1.New Suffix, A=@Replace.Result) {template="Concat::Execute(in A,in B,out Result)"}
   > n2 = &RigUnit_HierarchyAddNull(
    Space=GlobalSpace
    Name=@Concat.Result
    Transform=&RigUnit_GetTransform(bInitial=true, Item=@n1.Item).Transform
    Parent=/Script/ControlRig.RigUnit_HierarchyGetParent(bDefaultParent=True, Child=@n1.Item) {is=struct RigVMUnitNode}.Parent
   )
   > &RigUnit_SetDefaultParent(Child=@n1.Item, Parent=@n2.Item)
   > RigVMFunctionReturnNode(Null=@n2.Item)
  ]}
 ContainedGraph {nodes [
   n1 = RigVMFunctionEntryNode()
   > SequenceAggregate = RigVMAggregateNode()
   RigVMFunction_Sequence = RigVMAggregateNode(ExecuteContext=@SequenceAggregate.A)
   Concat = &RigVMFunction_NameConcat(B=@n1.Control Suffix, A=@n1.Name) {template="Concat::Execute(in A,in B,out Result)"}
   SpawnControl = /Script/ControlRig.RigUnit_HierarchyAddControlTransform(
    OffsetSpace=LocalSpace
    Parent=(Type=Bone,Name="None")
    Name=@Concat.Result
    OffsetTransform=@n1.Global Position
    Settings=@n1.Settings
    InitialValue=&RigVMFunction_MathTransformMakeRelative(Global=@n1.Global Position, Parent=@n1.Global Position).Local
    ExecutePin=@RigVMFunction_Sequence.A
   ) {
    is=struct RigVMUnitNode
    template="SpawnControl::Execute(in InitialValue,in Settings,in OffsetTransform,in OffsetSpace,in Parent,in Name,out Item)"
   }
   > &RigUnit_SetDefaultParent(Parent=@n1.Parent, Child=@SpawnControl.Item)
   > RigVMVariableNode(Value=@SpawnControl.Item) {variable=the control}
   n2 = &RigVMFunction_ControlFlowBranch(Condition=@n1.Orient to World, ExecuteContext=@RigVMFunction_Sequence.C)
   n3 = RigVMVariableNode() {variable=the control}
   n4 = &RigUnit_GetTransform(bInitial=true, Item=@n3.Value)
   &RigUnit_SetTransform(
    Value=(Rotation=(X=0,Y=0,Z=0,W=1),Translation=(X=0,Y=0,Z=0),Scale3D=(X=1,Y=1,Z=1))
    Weight=1
    bInitial=true
    Value.Scale3D=@n4.Transform.Scale3D
    Value.Translation=@n4.Transform.Translation
    ExecutePin=@n2.True
    Item=@n3.Value
   )
   n5 = &RigVMFunction_ControlFlowBranch(Condition=@n1.Zero Controls, ExecuteContext=@RigVMFunction_Sequence.B)
   &RigUnit_SetTransform(
    Weight=1
    bInitial=true
    Value=&RigUnit_GetTransform(bInitial=true, Item=RigVMVariableNode() {variable=the control}.Value).Transform
    ExecutePin=@n5.True
    Item=RigVMVariableNode() {variable=the control}.Value
   )
   BranchNode_2 = &RigVMFunction_ControlFlowBranch(Condition=@n1.Add Secondary Control, ExecuteContext=@SequenceAggregate.B)
   n6 = &RigUnit_HierarchyAddControlTransform(
    InitialValue=(Rotation=(X=0,Y=0,Z=0,W=1),Translation=(X=0,Y=0,Z=0),Scale3D=(X=1,Y=1,Z=1))
    OffsetSpace=LocalSpace
    OffsetTransform=(Rotation=(X=0,Y=0,Z=0,W=1),Translation=(X=0,Y=0,Z=0),Scale3D=(X=1,Y=1,Z=1))
    ExecutePin=@BranchNode_2.True
    Name=&RigVMFunction_NameConcat(A=&RigVMFunction_NameConcat(B=_sec, A=@n1.Name) {template="Concat::Execute(in A,in B,out Result)"}.Result, B=@n1.Control Suffix) {template="Concat::Execute(in A,in B,out Result)"}.Result
    Settings=@n1.Secondary Settings
    Parent=RigVMVariableNode() {variable=the control}.Value
   ) {template="SpawnControl::Execute(in InitialValue,in Settings,in OffsetTransform,in OffsetSpace,in Parent,in Name,out Item)"}
   > RigVMVariableNode(Value=@n6.Item) {variable=the secondary ctrl}
   n7 = &RigVMFunction_ControlFlowBranch(Condition=@n1.Add Null, ExecuteContext=@SequenceAggregate.C)
   n8 = RigVMVariableNode() {variable=the control}
   Add Null Above = RigVMFunctionReferenceNode(ExecuteContext=@n7.True, Replace Suffix=@n1.Control Suffix, New Suffix=@n1.Null Suffix, Item=@n8.Value)
   > DISPATCH_RigDispatch_SetMetadata = RigVMDispatchNode(Name=Null, NameSpace=Self, Value=@"Add Null Above".Null, Item=@n8.Value) {template="DISPATCH_RigDispatch_SetMetadata(in Item,in Name,in NameSpace,in Value,out Success)"}
   > n9 = RigVMVariableNode(Value=@"Add Null Above".Null) {variable=the null}
   RigVMFunction_ControlFlowBranch = &RigVMFunction_ControlFlowBranch(Condition=@n1.Zero Controls, ExecuteContext=@n9.ExecuteContext)
   &RigUnit_SetTransform(
    Weight=1
    bInitial=true
    Value=&RigUnit_GetTransform(bInitial=true, Item=RigVMVariableNode() {variable=the null}.Value).Transform
    ExecutePin=@RigVMFunction_ControlFlowBranch.True
    Item=RigVMVariableNode() {variable=the control}.Value
   )
   BranchNode_1 = &RigVMFunction_ControlFlowBranch(Condition=@n1.Add Bottom Null, ExecuteContext=@SequenceAggregate.D)
   n10 = &RigUnit_HierarchyAddNull(
    Space=LocalSpace
    Transform=(Rotation=(X=0,Y=0,Z=0,W=1),Translation=(X=0,Y=0,Z=0),Scale3D=(X=1,Y=1,Z=1))
    ExecutePin=@BranchNode_1.True
    Name=&RigVMFunction_NameConcat(B=_ctrl_null, A=@n1.Name) {template="Concat::Execute(in A,in B,out Result)"}.Result
    Parent=RigVMDispatchNode(
     Condition=@n1.Add Secondary Control
     False=RigVMVariableNode() {variable=the control}.Value
     True=RigVMVariableNode() {variable=the secondary ctrl}.Value
    ) {template="DISPATCH_RigVMDispatch_If(in Condition,in True,in False,out Result)"}.Result
   )
   > RigVMDispatchNode(Name=Bottom Null, NameSpace=Self, Value=@n10.Item, Item=RigVMVariableNode() {variable=the control}.Value) {template="DISPATCH_RigDispatch_SetMetadata(in Item,in Name,in NameSpace,in Value,out Success)"}
   > RigVMVariableNode(Value=@n10.Item) {variable=the bottom null}
   RigVMDispatchNode(
    Name=Secondary
    NameSpace=Self
    ExecuteContext=@SequenceAggregate.E
    Item=RigVMVariableNode() {variable=the control}.Value
    Value=RigVMVariableNode() {variable=the secondary ctrl}.Value
   ) {template="DISPATCH_RigDispatch_SetMetadata(in Item,in Name,in NameSpace,in Value,out Success)"}
   > RigVMFunctionReturnNode(
    Control=RigVMVariableNode() {variable=the control}.Value
    Secondary=RigVMVariableNode() {variable=the secondary ctrl}.Value
    Null=RigVMVariableNode() {variable=the null}.Value
    Bottom Null=RigVMVariableNode() {variable=the bottom null}.Value
   )
   n11 = &RigVMFunction_ControlFlowBranch(Condition=@n1.Tag Controls, ExecuteContext=@SequenceAggregate.F)
   /Script/ControlRig.RigUnit_SetMetadataTag(NameSpace=Self, Tag=Controls, ExecutePin=@n11.True, Item=RigVMVariableNode() {variable=the control}.Value) {is=struct RigVMUnitNode}
  ]}
 ContainedGraph {nodes [
   n1 = RigVMFunctionEntryNode()
   SequenceAggregate = &RigVMFunction_Sequence(ExecuteContext=@n1.ExecuteContext)
   SequenceAggregate_1 = &RigVMFunction_Sequence(ExecuteContext=@SequenceAggregate.B)
   SequenceAggregate_2 = &RigVMFunction_Sequence(ExecuteContext=@SequenceAggregate_1.B)
   SequenceAggregate_3 = &RigVMFunction_Sequence(ExecuteContext=@SequenceAggregate_2.B)
   SequenceAggregate_4 = &RigVMFunction_Sequence(ExecuteContext=@SequenceAggregate_3.B)
   RigVMFunctionReturnNode(
    A=@SequenceAggregate.A
    B=@SequenceAggregate_1.A
    C=@SequenceAggregate_2.A
    D=@SequenceAggregate_3.A
    E=@SequenceAggregate_4.A
    F=@SequenceAggregate_4.B
   )
  ]}
 ContainedGraph {nodes [
   n1 = RigVMFunctionEntryNode()
   RigUnit_SequenceAggregate = &RigVMFunction_Sequence(ExecuteContext=@n1.ExecuteContext)
   DISPATCH_RigDispatch_CoreNotEquals = RigVMDispatchNode(B=None, A=@n1.Name) {template="DISPATCH_RigVMDispatch_CoreNotEquals(in A,in B,out Result)"}
   Control Stack at Position = RigVMFunctionReferenceNode(
    Add Bottom Null=@n1.Add Bottom Null
    Add Null=@n1.Add Null
    Add Secondary Control=@n1.Add Secondary Control
    Control Suffix=@n1.Control Suffix
    Null Suffix=@n1.Null Suffix
    Orient to World=@n1.Orient to World
    Parent=@n1.Parent
    Secondary Settings=@n1.Secondary Settings
    Settings=@n1.Settings
    Tag Controls=@n1.Tag Controls
    Zero Controls=@n1.Zero Controls
    Global Position=&RigUnit_GetTransform(bInitial=true, Item=@n1.Item).Transform
    Name=RigVMDispatchNode(Condition=@DISPATCH_RigDispatch_CoreNotEquals.Result, False=@n1.Item.Name, True=@n1.Name) {template="DISPATCH_RigVMDispatch_If(in Condition,in True,in False,out Result)"}.Result
    ExecuteContext=@RigUnit_SequenceAggregate.A
   )
   > RigVMDispatchNode(Name=Item, NameSpace=Self, Item=@"Control Stack at Position".Control, Value=@n1.Item) {template="DISPATCH_RigDispatch_SetMetadata(in Item,in Name,in NameSpace,in Value,out Success)"}
   RigVMFunctionReturnNode(
    Bottom Null=@"Control Stack at Position".Bottom Null
    Control=@"Control Stack at Position".Control
    Null=@"Control Stack at Position".Null
    Secondary=@"Control Stack at Position".Secondary
    ExecuteContext=@RigUnit_SequenceAggregate.B
   )
  ]}
 ContainedGraph {nodes [
   n1 = RigVMFunctionEntryNode()
   RigUnit_SequenceAggregate = &RigVMFunction_Sequence(ExecuteContext=@n1.ExecuteContext)
   n2 = &RigVMFunction_ControlFlowBranch(Condition=@n1.Active, ExecuteContext=@RigUnit_SequenceAggregate.A)
   n3 = RigVMDispatchNode(ExecuteContext=@n2.True, Array=@n1.Transforms) {template="DISPATCH_RigVMDispatch_ArrayIterator(in Array,out Element,out Index,out Count,out Ratio)"}
   > DebugTransformMutableItemSpace = &RigVMFunction_DebugTransformMutableNoSpace(
    DebugDrawSettings=(DepthPriority=SDPG_Foreground,Lifetime=-1)
    WorldOffset=(Rotation=(X=0,Y=0,Z=0,W=1),Translation=(X=0,Y=0,Z=0),Scale3D=(X=1,Y=1,Z=1))
    bEnabled=True
    Transform=@n3.Element
    Color=@n1.Color
    Mode=@n1.Mode
    Thickness=@n1.Thickness
    Scale=@n1.Scale
   )
   RigVMFunctionReturnNode(ExecuteContext=@RigUnit_SequenceAggregate.B)
  ]}
 ContainedGraph {nodes [
   n1 = RigVMVariableNode() {variable=Point 3}
   n2 = RigVMFunctionEntryNode()
   n3 = &RigVMFunction_MathVectorDistance(B=@n2.Point 0, A=@n2.Point 1)
   n4 = &RigVMFunction_MathFloatAdd(A=@n3.Result, B=&RigVMFunction_MathVectorDistance(A=@n2.Point 1, B=@n2.Point 2).Result) {template="Add::Execute(in A,in B,out Result)"}
   n5 = &RigVMFunction_MathFloatAdd(A=@n4.Result, B=&RigVMFunction_MathVectorDistance(A=@n2.Point 2, B=@n2.Point 3).Result) {template="Add::Execute(in A,in B,out Result)"}
   n6 = &RigVMFunction_MathFloatDiv(B=@n5.Result, A=@n3.Result) {template="Divide::Execute(in A,in B,out Result)"}
   b3t1 = RigVMCollapseNode(t=@n6.Result)
   n7 = RigVMVariableNode() {variable=Point 0}
   b0t1 = RigVMCollapseNode(t=@n6.Result)
   n8 = &RigVMFunction_MathVectorScale(Value=@n7.Value, Factor=@b0t1.Result) {template="Scale::Execute(in Value,in Factor,out Result)"}
   n9 = RigVMVariableNode() {variable=Point 1}
   n10 = &RigVMFunction_MathFloatDiv(A=@n4.Result, B=@n5.Result) {template="Divide::Execute(in A,in B,out Result)"}
   b3t2 = RigVMCollapseNode(t=@n10.Result)
   b0t2 = RigVMCollapseNode(t=@n10.Result)
   n11 = &RigVMFunction_MathVectorScale(Value=@n7.Value, Factor=@b0t2.Result) {template="Scale::Execute(in Value,in Factor,out Result)"}
   n12 = RigVMVariableNode() {variable=Point 2}
   b1t1 = RigVMCollapseNode(t=@n6.Result)
   b1t2 = RigVMCollapseNode(t=@n10.Result)
   b2t1 = RigVMCollapseNode(t=@n6.Result)
   b2t2 = RigVMCollapseNode(t=@n10.Result)
   Solve = RigVMCollapseNode(
    c=&RigVMFunction_MathVectorSub(
     B=&RigVMFunction_MathVectorScale(Value=@n1.Value, Factor=@b3t1.Result) {template="Scale::Execute(in Value,in Factor,out Result)"}.Result
     A=&RigVMFunction_MathVectorSub(B=@n8.Result, A=@n9.Value) {template="Subtract::Execute(in A,in B,out Result)"}.Result
    ) {template="Subtract::Execute(in A,in B,out Result)"}.Result
    f=&RigVMFunction_MathVectorSub(
     B=&RigVMFunction_MathVectorScale(Value=@n1.Value, Factor=@b3t2.Result) {template="Scale::Execute(in Value,in Factor,out Result)"}.Result
     A=&RigVMFunction_MathVectorSub(B=@n11.Result, A=@n12.Value) {template="Subtract::Execute(in A,in B,out Result)"}.Result
    ) {template="Subtract::Execute(in A,in B,out Result)"}.Result
    a=@b1t1.Result
    d=@b1t2.Result
    b=@b2t1.Result
    e=@b2t2.Result
   )
   n13 = RigVMVariableNode() {variable=Point 0}
   n14 = RigVMVariableNode() {variable=Point 3}
   RigVMFunctionReturnNode(
    Spline=&RigUnit_ControlRigSplineFromPoints(
     Points=((X=0,Y=0,Z=0),(X=0,Y=0,Z=0),(X=0,Y=0,Z=0),(X=0,Y=0,Z=0))
     SamplesPerSegment=16
     SplineMode=BSpline
     bClosed=False
     Points.1=@Solve.i
     Points.2=@Solve.j
     Points.0=@n13.Value
     Points.3=@n14.Value
    ).Spline
    Out Point 2=@Solve.i
    Out Point 3=@Solve.j
    Out Point 1=@n13.Value
    Out Point 4=@n14.Value
    Execute=@n2.Execute
   )
  ]}
 ContainedGraph {nodes [
   RigVMFunctionEntryNode()
   n1 = &RigVMFunction_MathFloatDiv(
    A=&RigVMFunction_MathFloatMul(A=RigVMVariableNode() {variable=a}.Value, B=RigVMVariableNode() {variable=e}.Value) {template="Multiply::Execute(in A,in B,out Result)"}.Result
    B=RigVMVariableNode() {variable=d}.Value
   ) {template="Divide::Execute(in A,in B,out Result)"}
   n2 = RigVMVariableNode() {variable=b}
   n3 = &RigVMFunction_MathFloatDiv(A=RigVMVariableNode() {variable=a}.Value, B=RigVMVariableNode() {variable=d}.Value) {template="Divide::Execute(in A,in B,out Result)"}
   n4 = RigVMVariableNode() {variable=f}
   n5 = &RigVMFunction_MathVectorScale(
    Factor=&RigVMFunction_MathFloatDiv(A=1, B=&RigVMFunction_MathFloatSub(B=@n1.Result, A=@n2.Value) {template="Subtract::Execute(in A,in B,out Result)"}.Result) {template="Divide::Execute(in A,in B,out Result)"}.Result
    Value=&RigVMFunction_MathVectorSub(
     B=&RigVMFunction_MathVectorScale(Factor=@n3.Result, Value=@n4.Value) {template="Scale::Execute(in Value,in Factor,out Result)"}.Result
     A=RigVMVariableNode() {variable=c}.Value
    ) {template="Subtract::Execute(in A,in B,out Result)"}.Result
   ) {template="Scale::Execute(in Value,in Factor,out Result)"}
   n6 = RigVMVariableNode() {variable=a}
   n7 = &RigVMFunction_MathVectorScale(Value=@n5.Result, Factor=RigVMVariableNode() {variable=b}.Value) {template="Scale::Execute(in Value,in Factor,out Result)"}
   n8 = RigVMVariableNode() {variable=c}
   RigVMFunctionReturnNode(
    j=@n5.Result
    i=&RigVMFunction_MathVectorScale(
     Factor=&RigVMFunction_MathFloatDiv(A=1, B=@n6.Value) {template="Divide::Execute(in A,in B,out Result)"}.Result
     Value=&RigVMFunction_MathVectorSub(B=@n7.Result, A=@n8.Value) {template="Subtract::Execute(in A,in B,out Result)"}.Result
    ) {template="Scale::Execute(in Value,in Factor,out Result)"}.Result
   )
  ]}
 ContainedGraph {nodes [
   n1 = RigVMFunctionEntryNode()
   n2 = &RigVMFunction_MathFloatSub(A=1, B=@n1.t) {template="Subtract::Execute(in A,in B,out Result)"}
   n3 = &RigVMFunction_MathFloatMul(A=@n2.Result, B=@n2.Result) {template="Multiply::Execute(in A,in B,out Result)"}
   RigVMFunctionReturnNode(Result=&RigVMFunction_MathFloatMul(B=3, A=&RigVMFunction_MathFloatMul(A=@n1.t, B=@n3.Result) {template="Multiply::Execute(in A,in B,out Result)"}.Result) {template="Multiply::Execute(in A,in B,out Result)"}.Result)
  ]}
 ContainedGraph {nodes [
   n1 = RigVMFunctionEntryNode()
   n2 = &RigVMFunction_MathFloatMul(A=@n1.t, B=@n1.t) {template="Multiply::Execute(in A,in B,out Result)"}
   n3 = &RigVMFunction_MathFloatSub(A=1, B=@n1.t) {template="Subtract::Execute(in A,in B,out Result)"}
   RigVMFunctionReturnNode(Result=&RigVMFunction_MathFloatMul(B=3, A=&RigVMFunction_MathFloatMul(B=@n2.Result, A=@n3.Result) {template="Multiply::Execute(in A,in B,out Result)"}.Result) {template="Multiply::Execute(in A,in B,out Result)"}.Result)
  ]}
 ContainedGraph {nodes [
   n1 = RigVMFunctionEntryNode()
   n2 = &RigVMFunction_MathFloatSub(A=1, B=@n1.t) {template="Subtract::Execute(in A,in B,out Result)"}
   n3 = &RigVMFunction_MathFloatMul(A=@n2.Result, B=@n2.Result) {template="Multiply::Execute(in A,in B,out Result)"}
   RigVMFunctionReturnNode(Result=&RigVMFunction_MathFloatMul(B=3, A=&RigVMFunction_MathFloatMul(A=@n1.t, B=@n3.Result) {template="Multiply::Execute(in A,in B,out Result)"}.Result) {template="Multiply::Execute(in A,in B,out Result)"}.Result)
  ]}
 ContainedGraph {nodes [
   n1 = RigVMFunctionEntryNode()
   n2 = &RigVMFunction_MathFloatMul(A=@n1.t, B=@n1.t) {template="Multiply::Execute(in A,in B,out Result)"}
   n3 = &RigVMFunction_MathFloatSub(A=1, B=@n1.t) {template="Subtract::Execute(in A,in B,out Result)"}
   RigVMFunctionReturnNode(Result=&RigVMFunction_MathFloatMul(B=3, A=&RigVMFunction_MathFloatMul(B=@n2.Result, A=@n3.Result) {template="Multiply::Execute(in A,in B,out Result)"}.Result) {template="Multiply::Execute(in A,in B,out Result)"}.Result)
  ]}
 ContainedGraph {nodes [
   n1 = RigVMFunctionEntryNode()
   RigVMFunctionReturnNode(Result=&RigVMFunction_MathFloatPow(B=3, A=&RigVMFunction_MathFloatSub(A=1, B=@n1.t) {template="Subtract::Execute(in A,in B,out Result)"}.Result) {template="Power::Execute(in A,in B,out Result)"}.Result)
  ]}
 ContainedGraph {nodes [RigVMFunctionReturnNode(Result=&RigVMFunction_MathFloatPow(B=3, A=RigVMFunctionEntryNode().t) {template="Power::Execute(in A,in B,out Result)"}.Result)]}
 ContainedGraph {nodes [
   n1 = RigVMFunctionEntryNode()
   RigVMFunctionReturnNode(Result=&RigVMFunction_MathFloatPow(B=3, A=&RigVMFunction_MathFloatSub(A=1, B=@n1.t) {template="Subtract::Execute(in A,in B,out Result)"}.Result) {template="Power::Execute(in A,in B,out Result)"}.Result)
  ]}
 ContainedGraph {nodes [RigVMFunctionReturnNode(Result=&RigVMFunction_MathFloatPow(B=3, A=RigVMFunctionEntryNode().t) {template="Power::Execute(in A,in B,out Result)"}.Result)]}
 ContainedGraph {nodes [
   n1 = RigVMFunctionEntryNode()
   RigUnit_SequenceAggregate = &RigVMFunction_Sequence(ExecuteContext=@n1.ExecuteContext)
   n2 = &RigVMFunction_ControlFlowBranch(Condition=@n1.Negative, ExecuteContext=@RigUnit_SequenceAggregate.A)
   RigVMVariableNode(Value=1, ExecuteContext=@n2.False) {variable=isNegative}
   RigVMVariableNode(Value=-1, ExecuteContext=@n2.True) {variable=isNegative}
   n3 = RigVMDispatchNode(
    Array=RigVMDispatchNode(Array=@n1.controls, Other=@n1.ignore_ctrls) {template="DISPATCH_RigVMDispatch_ArrayDifference(in Array,in Other,out Result)"}.Result
    ExecuteContext=@RigUnit_SequenceAggregate.B
   ) {template="DISPATCH_RigVMDispatch_ArrayIterator(in Array,out Element,out Index,out Count,out Ratio)"}
   RigUnit_SequenceAggregate_1_1 = &RigVMFunction_Sequence(ExecuteContext=@n3.ExecuteContext)
   RigUnit_SequenceAggregate_1 = &RigVMFunction_Sequence(ExecuteContext=@RigUnit_SequenceAggregate_1_1.A)
   BranchNode = &RigVMFunction_ControlFlowBranch(Condition=@n1.X Plane, ExecuteContext=@RigUnit_SequenceAggregate_1.A)
   BranchNode_1 = &RigVMFunction_ControlFlowBranch(ExecuteContext=@BranchNode.False, Condition=@n1.Y Plane)
   BranchNode_2 = &RigVMFunction_ControlFlowBranch(ExecuteContext=@BranchNode_1.False, Condition=@n1.Z Plane)
   RigVMVariableNode(
    ExecuteContext=@BranchNode_2.True
    Value=&RigUnit_GetTransform(Item=(Type=Control), bInitial=true, Item.Name=@n3.Element.Name).Transform.Translation.Z
   ) {variable=position}
   RigVMVariableNode(
    ExecuteContext=@BranchNode_1.True
    Value=&RigUnit_GetTransform(Item=(Type=Control), bInitial=true, Item.Name=@n3.Element.Name).Transform.Translation.Y
   ) {variable=position}
   RigVMVariableNode(
    ExecuteContext=@BranchNode.True
    Value=&RigUnit_GetTransform(Item=(Type=Control), bInitial=true, Item.Name=@n3.Element.Name).Transform.Translation.X
   ) {variable=position}
   n4 = &RigVMFunction_MathFloatMul(B=RigVMVariableNode() {variable=isNegative}.Value, A=RigVMVariableNode() {variable=position}.Value) {template="Multiply::Execute(in A,in B,out Result)"}
   n5 = &RigVMFunction_ControlFlowBranch(
    Condition=/Script/RigVM.RigVMFunction_MathFloatGreater(B=0, A=@n4.Result) {is=struct RigVMUnitNode, template="Greater::Execute(in A,in B,out Result)"}.Result
    ExecuteContext=@RigUnit_SequenceAggregate_1.B
   )
   /Script/ControlRig.RigUnit_SetControlColor(Control=@n3.Element.Name, ExecutePin=@n5.True, Color=@n1.LeftColor) {is=struct RigVMUnitNode, template="SetControlColor::Execute(in Control,in Color)"}
   RigUnit_SequenceAggregate_1_2 = &RigVMFunction_Sequence(ExecuteContext=@RigUnit_SequenceAggregate_1_1.B)
   n6 = &RigVMFunction_ControlFlowBranch(
    Condition=/Script/RigVM.RigVMFunction_MathFloatLess(B=0, A=@n4.Result) {is=struct RigVMUnitNode, template="Less::Execute(in A,in B,out Result)"}.Result
    ExecuteContext=@RigUnit_SequenceAggregate_1_2.A
   )
   &RigUnit_SetControlColor(Control=@n3.Element.Name, ExecutePin=@n6.True, Color=@n1.RightColor) {template="SetControlColor::Execute(in Control,in Color)"}
   MathFloatEquals = RigVMDispatchNode(B=0, A=@n4.Result) {template="DISPATCH_RigVMDispatch_CoreEquals(in A,in B,out Result)"}
   n7 = &RigVMFunction_ControlFlowBranch(
    Condition=&RigVMFunction_MathBoolOr(
     B=@MathFloatEquals.Result
     A=/Script/RigVM.RigVMFunction_MathFloatIsNearlyZero(Tolerance=0.001, Value=@n4.Result) {is=struct RigVMUnitNode, template="IsNearlyZero::Execute(in Value,in Tolerance,out Result)"}.Result
    ).Result
    ExecuteContext=@RigUnit_SequenceAggregate_1_2.B
   )
   &RigUnit_SetControlColor(Control=@n3.Element.Name, ExecutePin=@n7.True, Color=@n1.CenterColor) {template="SetControlColor::Execute(in Control,in Color)"}
   RigVMFunctionReturnNode(ExecuteContext=@n3.Completed)
  ]}
 ContainedGraph {nodes [
   n1 = RigVMFunctionEntryNode()
   n2 = RigVMDispatchNode(Num=4, Array=RigVMVariableNode() {variable=Output Points}.Value, ExecuteContext=@n1.Execute) {template="DISPATCH_RigVMDispatch_ArraySetNum(io Array,in Num)"}
   n3 = RigVMVariableNode() {variable=Point 0}
   > n4 = RigVMDispatchNode(Index=0, Array=@n2.Array, Element=@n3.Value) {template="DISPATCH_RigVMDispatch_ArraySetAtIndex(io Array,in Index,in Element)"}
   n5 = RigVMVariableNode() {variable=t}
   n6 = RigVMVariableNode() {variable=t}
   n7 = &RigVMFunction_MathFloatSub(A=1, B=@n6.Value) {template="Subtract::Execute(in A,in B,out Result)"}
   Get A and C = RigVMCollapseNode(order=3, Start Point=@n1.Point 0, Middle Point=@n1.Point 1, End Point=@n1.Point 2, t=@n1.t)
   n8 = RigVMDispatchNode(
    Condition=true
    False=3
    True=&RigVMFunction_MathVectorDistance(B=@"Get A and C".C, A=RigVMVariableNode() {variable=Point 1}.Value).Result
   ) {template="DISPATCH_RigVMDispatch_If(in Condition,in True,in False,out Result)"}
   n9 = &RigVMFunction_MathVectorUnit(Value=&RigVMFunction_MathVectorSub(B=RigVMVariableNode() {variable=Point 0}.Value, A=RigVMVariableNode() {variable=Point 2}.Value) {template="Subtract::Execute(in A,in B,out Result)"}.Result) {template="Unit::Execute(in Value,out Result)"}
   n10 = RigVMVariableNode() {variable=Point 1}
   n11 = &RigVMFunction_MathVectorSub(
    B=&RigVMFunction_MathVectorScale(Factor=@n8.Result, Value=@n9.Result) {template="Scale::Execute(in Value,in Factor,out Result)"}.Result
    A=@n10.Value
   ) {template="Subtract::Execute(in A,in B,out Result)"}
   n12 = &RigVMFunction_MathVectorAdd(
    B=&RigVMFunction_MathVectorScale(
     Factor=&RigVMFunction_MathFloatDiv(A=1, B=@n7.Result) {template="Divide::Execute(in A,in B,out Result)"}.Result
     Value=&RigVMFunction_MathVectorSub(A=@n11.Result, B=@"Get A and C".A) {template="Subtract::Execute(in A,in B,out Result)"}.Result
    ) {template="Scale::Execute(in Value,in Factor,out Result)"}.Result
    A=@"Get A and C".A
   ) {template="Add::Execute(in A,in B,out Result)"}
   n13 = RigVMVariableNode() {variable=Point 0}
   n14 = &RigVMFunction_MathVectorAdd(
    B=&RigVMFunction_MathVectorScale(
     Factor=&RigVMFunction_MathFloatDiv(A=1, B=@n5.Value) {template="Divide::Execute(in A,in B,out Result)"}.Result
     Value=&RigVMFunction_MathVectorSub(A=@n12.Result, B=@n13.Value) {template="Subtract::Execute(in A,in B,out Result)"}.Result
    ) {template="Scale::Execute(in Value,in Factor,out Result)"}.Result
    A=@n13.Value
   ) {template="Add::Execute(in A,in B,out Result)"}
   > n15 = RigVMDispatchNode(Index=1, Array=@n4.Array, Element=@n14.Result) {template="DISPATCH_RigVMDispatch_ArraySetAtIndex(io Array,in Index,in Element)"}
   n16 = &RigVMFunction_MathFloatSub(A=1, B=@n5.Value) {template="Subtract::Execute(in A,in B,out Result)"}
   n17 = RigVMVariableNode() {variable=t}
   n18 = &RigVMFunction_MathFloatMul(A=@n8.Result, B=&RigVMFunction_MathFloatSub(A=1, B=@n17.Value) {template="Subtract::Execute(in A,in B,out Result)"}.Result) {template="Multiply::Execute(in A,in B,out Result)"}
   n19 = &RigVMFunction_MathVectorAdd(
    B=&RigVMFunction_MathVectorScale(
     Factor=&RigVMFunction_MathFloatDiv(A=@n18.Result, B=@n17.Value) {template="Divide::Execute(in A,in B,out Result)"}.Result
     Value=@n9.Result
    ) {template="Scale::Execute(in Value,in Factor,out Result)"}.Result
    A=@n10.Value
   ) {template="Add::Execute(in A,in B,out Result)"}
   n20 = &RigVMFunction_MathVectorAdd(
    B=&RigVMFunction_MathVectorScale(
     Factor=&RigVMFunction_MathFloatDiv(A=1, B=@n6.Value) {template="Divide::Execute(in A,in B,out Result)"}.Result
     Value=&RigVMFunction_MathVectorSub(A=@n19.Result, B=@"Get A and C".A) {template="Subtract::Execute(in A,in B,out Result)"}.Result
    ) {template="Scale::Execute(in Value,in Factor,out Result)"}.Result
    A=@"Get A and C".A
   ) {template="Add::Execute(in A,in B,out Result)"}
   n21 = RigVMVariableNode() {variable=Point 2}
   n22 = &RigVMFunction_MathVectorAdd(
    B=&RigVMFunction_MathVectorScale(
     Factor=&RigVMFunction_MathFloatDiv(A=1, B=@n16.Result) {template="Divide::Execute(in A,in B,out Result)"}.Result
     Value=&RigVMFunction_MathVectorSub(A=@n20.Result, B=@n21.Value) {template="Subtract::Execute(in A,in B,out Result)"}.Result
    ) {template="Scale::Execute(in Value,in Factor,out Result)"}.Result
    A=@n21.Value
   ) {template="Add::Execute(in A,in B,out Result)"}
   > n23 = RigVMDispatchNode(Index=2, Array=@n15.Array, Element=@n22.Result) {template="DISPATCH_RigVMDispatch_ArraySetAtIndex(io Array,in Index,in Element)"}
   n24 = RigVMVariableNode() {variable=Point 2}
   > n25 = RigVMDispatchNode(Index=3, Array=@n23.Array, Element=@n24.Value) {template="DISPATCH_RigVMDispatch_ArraySetAtIndex(io Array,in Index,in Element)"}
   RerouteNode = RigVMDispatchNode(
    Values=((X=0,Y=0,Z=0),((X=0,Y=0,Z=0),X=0,Y=0,Z=0),((X=0,Y=0,Z=0),X=0,Y=0,Z=0),((X=0,Y=0,Z=0),X=0,Y=0,Z=0))
    Values.1=@n14.Result
    Values.2=@n22.Result
    Values.3=@n24.Value
    Values.0=@n3.Value
   ) {template="DISPATCH_RigVMDispatch_ArrayMake(in Values,out Array)"}
   RigVMFunctionReturnNode(
    Control Points=@n25.Array
    Execute=@n25.ExecuteContext
    Spline=&RigUnit_ControlRigSplineFromPoints(SamplesPerSegment=16, SplineMode=BSpline, bClosed=False, Points=@RerouteNode.Array).Spline
    Out Point 4=@n24.Value
    Out Point 1=@n3.Value
    Out Point 2=@n22.Result
    Out Point 3=@n14.Result
   )
  ]}
 ContainedGraph {nodes [
   n1 = RigVMFunctionEntryNode()
   Projection Ratio = RigVMCollapseNode(order=@n1.order, t=@n1.t)
   n2 = &RigVMFunction_MathVectorAdd(
    A=&RigVMFunction_MathVectorScale(Factor=@"Projection Ratio".Result, Value=RigVMVariableNode() {variable=Start Point}.Value) {template="Scale::Execute(in Value,in Factor,out Result)"}.Result
    B=&RigVMFunction_MathVectorScale(
     Factor=&RigVMFunction_MathFloatSub(A=1, B=@"Projection Ratio".Result) {template="Subtract::Execute(in A,in B,out Result)"}.Result
     Value=RigVMVariableNode() {variable=End Point}.Value
    ) {template="Scale::Execute(in Value,in Factor,out Result)"}.Result
   ) {template="Add::Execute(in A,in B,out Result)"}
   ABC Ratio = RigVMCollapseNode(order=@n1.order, t=@n1.t)
   n3 = &RigVMFunction_MathFloatDiv(A=1, B=@"ABC Ratio".Result) {template="Divide::Execute(in A,in B,out Result)"}
   n4 = RigVMVariableNode() {variable=Middle Point}
   n5 = &RigVMFunction_MathVectorSub(B=@n2.Result, A=@n4.Value) {template="Subtract::Execute(in A,in B,out Result)"}
   RigVMFunctionReturnNode(
    C=@n2.Result
    A=&RigVMFunction_MathVectorAdd(
     B=&RigVMFunction_MathVectorScale(Factor=@n3.Result, Value=@n5.Result) {template="Scale::Execute(in Value,in Factor,out Result)"}.Result
     A=@n4.Value
    ) {template="Add::Execute(in A,in B,out Result)"}.Result
   )
  ]}
 ContainedGraph {nodes [
   n1 = RigVMFunctionEntryNode()
   n2 = &RigVMFunction_MathIntToFloat(Value=@n1.order) {template="Cast::Execute(in Value,out Result)"}
   n3 = &RigVMFunction_MathFloatPow(A=&RigVMFunction_MathFloatSub(A=1, B=@n1.t) {template="Subtract::Execute(in A,in B,out Result)"}.Result, B=@n2.Result) {template="Power::Execute(in A,in B,out Result)"}
   n4 = &RigVMFunction_MathFloatPow(A=@n1.t, B=@n2.Result) {template="Power::Execute(in A,in B,out Result)"}
   RigVMFunctionReturnNode(Result=&RigVMFunction_MathFloatDiv(B=&RigVMFunction_MathFloatAdd(A=@n3.Result, B=@n4.Result) {template="Add::Execute(in A,in B,out Result)"}.Result, A=@n3.Result) {template="Divide::Execute(in A,in B,out Result)"}.Result)
  ]}
 ContainedGraph {nodes [
   n1 = RigVMFunctionEntryNode()
   n2 = &RigVMFunction_MathIntToFloat(Value=@n1.order) {template="Cast::Execute(in Value,out Result)"}
   n3 = &RigVMFunction_MathFloatAdd(
    B=&RigVMFunction_MathFloatPow(A=@n1.t, B=@n2.Result) {template="Power::Execute(in A,in B,out Result)"}.Result
    A=&RigVMFunction_MathFloatPow(A=&RigVMFunction_MathFloatSub(A=1, B=@n1.t) {template="Subtract::Execute(in A,in B,out Result)"}.Result, B=@n2.Result) {template="Power::Execute(in A,in B,out Result)"}.Result
   ) {template="Add::Execute(in A,in B,out Result)"}
   n4 = &RigVMFunction_MathFloatSub(B=1, A=@n3.Result) {template="Subtract::Execute(in A,in B,out Result)"}
   RigVMFunctionReturnNode(Result=&RigVMFunction_MathFloatAbs(Value=&RigVMFunction_MathFloatDiv(B=@n3.Result, A=@n4.Result) {template="Divide::Execute(in A,in B,out Result)"}.Result) {template="Absolute::Execute(in Value,out Result)"}.Result)
  ]}
 ContainedGraph {nodes [
   n1 = RigVMFunctionEntryNode()
   n2 = RigVMVariableNode() {variable=Controls}
   BranchNode_4_RigVMUnitNode = &RigVMFunction_ControlFlowBranch(
    ExecuteContext=@n1.ExecuteContext
    Condition=&RigVMFunction_MathIntLess(B=4, A=RigVMDispatchNode(Array=@n2.Value) {template="DISPATCH_RigVMDispatch_ArrayGetNum(in Array,out Num)"}.Num) {template="Less::Execute(in A,in B,out Result)"}.Result
   )
   RigVMFunctionReturnNode(
    ExecuteContext=@BranchNode_4_RigVMUnitNode.Completed
    Local Spline=RigVMVariableNode() {variable=local_spline}.Value
    Orig Positions=RigVMVariableNode() {variable=Original Positions}.Value
    Positions=RigVMVariableNode() {variable=Final Positions}.Value
    Template Spline=RigVMVariableNode() {variable=template_spline}.Value
   )
   RigUnit_SequenceExecution_1 = RigVMAggregateNode(ExecuteContext=@BranchNode_4_RigVMUnitNode.False)
   RigVMVariableNode(
    Value=&RigUnit_GetTransformItemArray(Space=GlobalSpace, bInitial=true, Items=@n1.Bones).Transforms
    ExecuteContext=@RigUnit_SequenceExecution_1.A
   ) {variable=Original Positions}
   n3 = &RigUnit_GetTransformItemArray(Space=GlobalSpace, bInitial=False, Items=RigVMVariableNode() {variable=Controls}.Value)
   RigVMFunction_MathTransformArrayToSRT = &RigVMFunction_MathTransformArrayToSRT(Transforms=@n3.Transforms)
   RigVMVariableNode(
    Value=&RigUnit_ControlRigSplineFromPoints(
     Stretch=0
     bClosed=False
     Points=@RigVMFunction_MathTransformArrayToSRT.Translations
     SplineMode=RigVMVariableNode() {variable=Spline Mode}.Value
     SamplesPerSegment=RigVMVariableNode() {variable=Samples Per Segment}.Value
    ).Spline
    ExecuteContext=@RigUnit_SequenceExecution_1.B
   ) {variable=template_spline}
   > RigVMVariableNode(Value=@n3.Transforms) {variable=template_transforms}
   > RigUnit_SequenceExecution = RigVMAggregateNode()
   n4 = RigVMVariableNode() {variable=template_transforms}
   RigVMDispatchNode(
    Num=RigVMDispatchNode(Array=@n4.Value) {template="DISPATCH_RigVMDispatch_ArrayGetNum(in Array,out Num)"}.Num
    ExecuteContext=@RigUnit_SequenceExecution.A
    Array=RigVMVariableNode() {variable=offset_positions}.Value
   ) {template="DISPATCH_RigVMDispatch_ArraySetNum(io Array,in Num)"}
   > n5 = RigVMDispatchNode(Array=@n4.Value) {template="DISPATCH_RigVMDispatch_ArrayIterator(in Array,out Element,out Index,out Count,out Ratio)"}
   > RigVMDispatchNode(
    Index=@n5.Index
    Element=&RigVMFunction_MathTransformTransformVector(Transform=@n5.Element, Location=RigVMVariableNode() {variable=Secondary Spline Direction}.Value).Result
    Array=RigVMVariableNode() {variable=offset_positions}.Value
   ) {template="DISPATCH_RigVMDispatch_ArraySetAtIndex(io Array,in Index,in Element)"}
   RigUnit_SequenceExecution_4 = RigVMAggregateNode(ExecuteContext=@RigUnit_SequenceExecution.C)
   Branch_RigVMUnitNode = &RigVMFunction_ControlFlowBranch(ExecuteContext=@RigUnit_SequenceExecution_4.A, Condition=RigVMVariableNode() {variable=Stretch}.Value)
   n6 = RigVMVariableNode() {variable=template_spline}
   n7 = RigVMVariableNode() {variable=Primary Axis}
   n8 = RigVMVariableNode() {variable=Bones}
   &RigUnit_FitChainToSplineCurveItemArray(
    Alignment=Front
    DebugSettings=(Scale=1,CurveColor=(R=1,G=1,B=0,A=1),SegmentsColor=(R=1,G=0,B=0,A=1),WorldOffset=(Rotation=(X=0,Y=0,Z=0,W=1),Translation=(X=0,Y=0,Z=0),Scale3D=(X=1,Y=1,Z=1)))
    Maximum=1
    PoleVectorPosition=(X=0,Y=0,Z=0)
    Rotations=()
    SamplingPrecision=12
    SecondaryAxis=(X=0,Y=0,Z=0)
    Weight=1
    bPropagateToChildren=True
    ExecutePin=@Branch_RigVMUnitNode.False
    Spline=@n6.Value
    PrimaryAxis=@n7.Value
    Items=@n8.Value
   )
   &RigUnit_FitChainToSplineCurveItemArray(
    Alignment=Stretched
    DebugSettings=(Scale=1,CurveColor=(R=1,G=1,B=0,A=1),SegmentsColor=(R=1,G=0,B=0,A=1),WorldOffset=(Rotation=(X=0,Y=0,Z=0,W=1),Translation=(X=0,Y=0,Z=0),Scale3D=(X=1,Y=1,Z=1)))
    Maximum=1
    PoleVectorPosition=(X=0,Y=0,Z=0)
    Rotations=()
    SamplingPrecision=12
    SecondaryAxis=(X=0,Y=0,Z=0)
    Weight=1
    bPropagateToChildren=True
    ExecutePin=@Branch_RigVMUnitNode.True
    Spline=@n6.Value
    PrimaryAxis=@n7.Value
    Items=@n8.Value
   )
   RigUnit_SequenceExecution_2 = RigVMAggregateNode(ExecuteContext=@RigUnit_SequenceExecution_4.B)
   n9 = RigVMVariableNode(
    Value=&RigUnit_ControlRigSplineFromPoints(
     Stretch=0
     bClosed=False
     SplineMode=RigVMVariableNode() {variable=Spline Mode}.Value
     SamplesPerSegment=RigVMVariableNode() {variable=Samples Per Segment}.Value
     Points=RigVMVariableNode() {variable=offset_positions}.Value
    ).Spline
    ExecuteContext=@RigUnit_SequenceExecution_2.A
   ) {variable=local_spline}
   BranchNode_2_RigVMUnitNode = &RigVMFunction_ControlFlowBranch(ExecuteContext=@n9.ExecuteContext, Condition=RigVMVariableNode() {variable=debug}.Value)
   &RigUnit_DrawControlRigSpline(
    Color=(R=1,G=0,B=0,A=1)
    Detail=16
    Thickness=1
    ExecutePin=@BranchNode_2_RigVMUnitNode.True
    Spline=RigVMVariableNode() {variable=local_spline}.Value
   )
   n10 = RigVMVariableNode() {variable=Bones}
   DISPATCH_RigVMDispatch_ArrayGetNum = RigVMDispatchNode(Array=@n10.Value) {template="DISPATCH_RigVMDispatch_ArrayGetNum(in Array,out Num)"}
   DISPATCH_RigVMDispatch_ArraySetNum = RigVMDispatchNode(
    Num=@DISPATCH_RigVMDispatch_ArrayGetNum.Num
    ExecuteContext=@RigUnit_SequenceExecution_2.B
    Array=RigVMVariableNode() {variable=aim_transforms}.Value
   ) {template="DISPATCH_RigVMDispatch_ArraySetNum(io Array,in Num)"}
   > n11 = RigVMDispatchNode(Array=@n10.Value) {template="DISPATCH_RigVMDispatch_ArrayIterator(in Array,out Element,out Index,out Count,out Ratio)"}
   n12 = &RigUnit_GetTransform(bInitial=false, Item=@n11.Element)
   n13 = RigVMVariableNode() {variable=local_spline}
   > DISPATCH_RigVMDispatch_ArraySetAtIndex = RigVMDispatchNode(
    Index=@n11.Index
    Element=&RigUnit_PositionFromControlRigSpline(
     U=&RigUnit_ClosestParameterFromControlRigSpline(Position=@n12.Transform.Translation, Spline=@n13.Value).U
     Spline=@n13.Value
    ).Position
    Array=RigVMVariableNode() {variable=aim_transforms}.Value
   ) {template="DISPATCH_RigVMDispatch_ArraySetAtIndex(io Array,in Index,in Element)"}
   n14 = RigVMDispatchNode(ExecuteContext=@RigUnit_SequenceExecution.D, Array=RigVMVariableNode() {variable=Bones}.Value) {template="DISPATCH_RigVMDispatch_ArrayIterator(in Array,out Element,out Index,out Count,out Ratio)"}
   DISPATCH_RigVMDispatch_ArrayGetAtIndex = RigVMDispatchNode(Index=@n14.Index, Array=RigVMVariableNode() {variable=aim_transforms}.Value) {template="DISPATCH_RigVMDispatch_ArrayGetAtIndex(in Array,in Index,out Element)"}
   n15 = &RigUnit_AimBoneMath(
    DebugSettings=(Scale=10,WorldOffset=(Rotation=(X=0,Y=0,Z=0,W=1),Translation=(X=0,Y=0,Z=0),Scale3D=(X=1,Y=1,Z=1)))
    Primary=(Weight=1,Axis=(X=0,Y=1,Z=0),Target=(X=1,Y=0,Z=0),Kind=Location,Space=(Type=Bone))
    Secondary=(Weight=0,Axis=(X=0,Y=0,Z=0),Target=(X=0,Y=0,Z=0),Kind=Location,Space=(Type=Bone))
    Weight=1
    Primary.Target=@DISPATCH_RigVMDispatch_ArrayGetAtIndex.Element
    InputTransform=&RigUnit_GetTransform(Item=@n14.Element).Transform
    Primary.Axis=RigVMVariableNode() {variable=Up Axis}.Value
   )
   MathIntEquals = RigVMDispatchNode(A=@n14.Index, B=&RigVMFunction_MathIntSub(B=1, A=@n14.Count) {template="Subtract::Execute(in A,in B,out Result)"}.Result) {template="DISPATCH_RigVMDispatch_CoreEquals(in A,in B,out Result)"}
   > n16 = &RigUnit_SetTransform(Weight=1, bInitial=true, Value=@n15.Result, Item=@n14.Element, bPropagateToChildren=@MathIntEquals.Result)
   BranchNode_3_RigVMUnitNode = &RigVMFunction_ControlFlowBranch(ExecuteContext=@n16.ExecutePin, Condition=RigVMVariableNode() {variable=debug}.Value)
   &RigUnit_DebugTransformMutableItemSpace(
    Color=(R=1,G=0,B=0,A=1)
    DebugDrawSettings=(DepthPriority=SDPG_Foreground,Lifetime=-1)
    Mode=Axes
    Scale=10
    Space=(Type=Bone)
    Thickness=0.15
    WorldOffset=(Rotation=(X=0,Y=0,Z=0,W=1),Translation=(X=0,Y=0,Z=0),Scale3D=(X=1,Y=1,Z=1))
    bEnabled=True
    Transform=@n15.Result
    ExecutePin=@BranchNode_3_RigVMUnitNode.True
   )
   Branch_1_RigVMUnitNode = &RigVMFunction_ControlFlowBranch(ExecuteContext=@RigUnit_SequenceExecution_1.C, Condition=RigVMVariableNode() {variable=Stretch}.Value)
   n17 = RigVMDispatchNode(Index=-1, Array=RigVMVariableNode() {variable=Controls}.Value) {template="DISPATCH_RigVMDispatch_ArrayGetAtIndex(in Array,in Index,out Element)"}
   n18 = RigVMDispatchNode(Index=-1, Array=RigVMVariableNode() {variable=Bones}.Value) {template="DISPATCH_RigVMDispatch_ArrayGetAtIndex(in Array,in Index,out Element)"}
   &RigUnit_ParentConstraint(
    AdvancedSettings=(RotationOrderForFilter=XZY)
    Filter=(TranslationFilter=(bX=false,bY=false,bZ=false),RotationFilter=(bX=True,bY=True,bZ=True),ScaleFilter=(bX=True,bY=True,bZ=True))
    Parents=((Item=(Type=Bone),Weight=1))
    Weight=1
    Parents.0.Item=@n17.Element
    Child=@n18.Element
    ExecutePin=@Branch_1_RigVMUnitNode.False
   )
   &RigUnit_ParentConstraint(
    AdvancedSettings=(RotationOrderForFilter=XZY)
    Parents=((Item=(Type=Bone),Weight=1))
    Weight=1
    Parents.0.Item=@n17.Element
    Child=@n18.Element
    ExecutePin=@Branch_1_RigVMUnitNode.True
   )
   RigVMVariableNode(
    Value=&RigUnit_GetTransformItemArray(Space=GlobalSpace, bInitial=False, Items=@n1.Bones).Transforms
    ExecuteContext=@RigUnit_SequenceExecution_1.D
   ) {variable=Final Positions}
   For_Each_1 = RigVMDispatchNode(ExecuteContext=(), Array=RigVMVariableNode() {variable=Bones}.Value) {template="DISPATCH_RigVMDispatch_ArrayIterator(in Array,out Element,out Index,out Count,out Ratio)"}
   n19 = &RigUnit_GetTransform(bInitial=true, Item=@For_Each_1.Element)
   n20 = RigVMVariableNode() {variable=template_spline}
   > &RigUnit_SetTransform(
    Weight=1
    bInitial=true
    Item=@For_Each_1.Element
    Value=&RigUnit_TransformFromControlRigSpline2(
     PrimaryAxis=(X=0,Y=0,Z=0)
     SecondaryAxis=(X=0,Y=0,Z=0)
     U=&RigUnit_ClosestParameterFromControlRigSpline(Position=@n19.Transform.Translation, Spline=@n20.Value).U
     Spline=@n20.Value
    ).Transform
   )
   RigVMDispatchNode(ExecuteContext=(), Array=RigVMVariableNode() {variable=Dummy Bones}.Value) {template="DISPATCH_RigVMDispatch_ArrayReset(io Array)"}
   > For_Each = RigVMDispatchNode(Array=RigVMVariableNode() {variable=Bones}.Value) {template="DISPATCH_RigVMDispatch_ArrayIterator(in Array,out Element,out Index,out Count,out Ratio)"}
   Concat = &RigVMFunction_NameConcat(B=_dummy, A=@For_Each.Element.Name) {template="Concat::Execute(in A,in B,out Result)"}
   > n21 = /Script/ControlRig.RigUnit_HierarchyAddBone(
    Parent=(Type=Bone,Name="None")
    Space=GlobalSpace
    Name=@Concat.Result
    Parent.Name=&RigVMFunction_NameConcat(B=_dummy, A=&RigUnit_HierarchyGetParent(bDefaultParent=True, Child=@For_Each.Element).Parent.Name) {template="Concat::Execute(in A,in B,out Result)"}.Result
    Transform=&RigUnit_GetTransform(bInitial=False, Item=@For_Each.Element).Transform
   ) {is=struct RigVMUnitNode}
   > Add = RigVMDispatchNode(Element=@n21.Item, Array=RigVMVariableNode() {variable=Dummy Bones}.Value) {template="DISPATCH_RigVMDispatch_ArrayAdd(io Array,in Element,out Index)"}
   n22 = RigVMVariableNode() {variable=Bones}
   &RigUnit_SetDefaultParent(
    Child=RigVMDispatchNode(Index=0, Array=RigVMVariableNode() {variable=Dummy Bones}.Value) {template="DISPATCH_RigVMDispatch_ArrayGetAtIndex(in Array,in Index,out Element)"}.Element
    ExecutePin=@For_Each.Completed
    Parent=&RigUnit_HierarchyGetParent(
     bDefaultParent=True
     Child=RigVMDispatchNode(Index=0, Array=@n22.Value) {template="DISPATCH_RigVMDispatch_ArrayGetAtIndex(in Array,in Index,out Element)"}.Element
    ).Parent
   )
   RigVMFunction_ForLoopCount = /Script/RigVM.RigVMFunction_ForLoopCount(Count=4, ExecutePin=()) {is=struct RigVMUnitNode}
   n23 = RigVMVariableNode() {variable=Controls}
   n24 = RigVMDispatchNode(Array=@n23.Value) {template="DISPATCH_RigVMDispatch_ArrayGetNum(in Array,out Num)"}
   Less = &RigVMFunction_MathIntLess(B=@n24.Num, A=@RigVMFunction_ForLoopCount.Index) {template="Less::Execute(in A,in B,out Result)"}
   RigVMFunction_ControlFlowBranch = &RigVMFunction_ControlFlowBranch(Condition=@Less.Result, ExecuteContext=@RigVMFunction_ForLoopCount.ExecutePin)
   RigVMDispatchNode(
    Element=RigVMDispatchNode(Index=@n24.Num, Array=@n23.Value) {template="DISPATCH_RigVMDispatch_ArrayGetAtIndex(in Array,in Index,out Element)"}.Element
    ExecuteContext=@RigVMFunction_ControlFlowBranch.True
    Index=@RigVMFunction_ForLoopCount.Index
    Array=RigVMVariableNode() {variable=Controls}.Value
   ) {template="DISPATCH_RigVMDispatch_ArrayInsert(io Array,in Index,in Element)"}
  ]}
 ContainedGraph {nodes [
   n1 = RigVMFunctionEntryNode()
   RigUnit_SequenceExecution_2 = &RigVMFunction_Sequence(ExecuteContext=@n1.ExecuteContext)
   RigUnit_SequenceExecution_2_1 = &RigVMFunction_Sequence(ExecuteContext=@RigUnit_SequenceExecution_2.B)
   RigUnit_SequenceExecution_2_2 = &RigVMFunction_Sequence(ExecuteContext=@RigUnit_SequenceExecution_2_1.B)
   RigVMFunctionReturnNode(A=@RigUnit_SequenceExecution_2.A, B=@RigUnit_SequenceExecution_2_1.A, C=@RigUnit_SequenceExecution_2_2.A, D=@RigUnit_SequenceExecution_2_2.B)
  ]}
 ContainedGraph {nodes [
   n1 = RigVMFunctionEntryNode()
   RigUnit_SequenceExecution_4 = &RigVMFunction_Sequence(ExecuteContext=@n1.ExecuteContext)
   RigUnit_SequenceExecution_4_1 = &RigVMFunction_Sequence(ExecuteContext=@RigUnit_SequenceExecution_4.B)
   RigUnit_SequenceExecution_4_2 = &RigVMFunction_Sequence(ExecuteContext=@RigUnit_SequenceExecution_4_1.B)
   RigVMFunctionReturnNode(A=@RigUnit_SequenceExecution_4.A, B=@RigUnit_SequenceExecution_4_1.A, C=@RigUnit_SequenceExecution_4_2.A, D=@RigUnit_SequenceExecution_4_2.B)
  ]}
 ContainedGraph {nodes [
   n1 = RigVMFunctionEntryNode()
   RigUnit_SequenceExecution = &RigVMFunction_Sequence(ExecuteContext=@n1.ExecuteContext)
   RigUnit_SequenceExecution_1 = &RigVMFunction_Sequence(ExecuteContext=@RigUnit_SequenceExecution.B)
   RigUnit_SequenceExecution_2 = &RigVMFunction_Sequence(ExecuteContext=@RigUnit_SequenceExecution_1.B)
   RigVMFunctionReturnNode(A=@RigUnit_SequenceExecution.A, B=@RigUnit_SequenceExecution_1.A, C=@RigUnit_SequenceExecution_2.A, D=@RigUnit_SequenceExecution_2.B)
  ]}
 ContainedGraph {nodes [
   n1 = RigVMFunctionEntryNode()
   RigUnit_SequenceExecution_1 = &RigVMFunction_Sequence(ExecuteContext=@n1.ExecuteContext)
   RigUnit_SequenceExecution_1_1 = &RigVMFunction_Sequence(ExecuteContext=@RigUnit_SequenceExecution_1.B)
   RigUnit_SequenceExecution_1_2 = &RigVMFunction_Sequence(ExecuteContext=@RigUnit_SequenceExecution_1_1.B)
   RigVMFunctionReturnNode(A=@RigUnit_SequenceExecution_1.A, B=@RigUnit_SequenceExecution_1_1.A, C=@RigUnit_SequenceExecution_1_2.A, D=@RigUnit_SequenceExecution_1_2.B)
  ]}
 Find_ContainedGraph {nodes [
   n1 = RigVMFunctionEntryNode()
   To_String = RigVMDispatchNode(Value=@n1.Value) {template="DISPATCH_RigDispatch_ToString(in Value,out Result)"}
   RigVMFunction_StringSplit = /Script/RigVM.RigVMFunction_StringSplit(Separator=-, Value=@To_String.Result) {is=struct RigVMUnitNode}
   > RigVMFunctionReturnNode(Success=RigVMDispatchNode(Element=@n1.Element, Array=@RigVMFunction_StringSplit.Result) {template="DISPATCH_RigVMDispatch_ArrayFind(in Array,in Element,out Index,out Success)"}.Success)
  ]}
 Multiply_1_1_ContainedGraph {nodes [
   n1 = RigVMFunctionEntryNode()
   RigVMFunctionReturnNode(Result=/Script/RigVM.RigVMFunction_MathVectorMul(B=@n1.C, A=&RigVMFunction_MathVectorMul(A=@n1.A, B=@n1.B) {template="Multiply::Execute(in A,in B,out Result)"}.Result) {is=struct RigVMUnitNode, template="Multiply::Execute(in A,in B,out Result)"}.Result)
  ]}
 New Function_ContainedGraph {nodes [
   n1 = RigVMFunctionEntryNode()
   n2 = /Script/ControlRig.RigUnit_FindItemsWithMetadataTag(NameSpace=Self, Tag=Controls) {is=struct RigVMUnitNode}
   > Scale Module Controls = RigVMFunctionReferenceNode(
    Global Scale=RigVMDispatchNode(Default=1, Name=Control Size, NameSpace=Root) {template="DISPATCH_RigDispatch_GetModuleMetadata(in Name,in NameSpace,in Default,out Value,out Found)"}.Value
    Local Scale=@n1.Module Settings.ControlSize_12_#1
    Controls=@n2.Items
   )
   > Color Module Controls = RigVMFunctionReferenceNode(
    CenterColor=@n1.Module Settings.CenterColor_9_#2
    LeftColor=@n1.Module Settings.LeftColor_3_#3
    Override=@n1.Module Settings.OverrideColor_48_#4
    RightColor=@n1.Module Settings.RightColor_5_#5
    Controls=@n2.Items
   )
   > RigVMFunctionReturnNode()
  ]}
 New Function_ContainedGraph {nodes [
   n1 = RigVMFunctionEntryNode()
   > &RigUnit_SetTransform(
    Weight=1
    bInitial=False
    Item=@n1.Child
    Value=&RigUnit_ProjectTransformToNewParent(bChildInitial=True, bNewParentInitial=False, bOldParentInitial=True, Child=@n1.Child, NewParent=@n1.New Parent, OldParent=@n1.New Parent).Transform
   )
   > RigVMFunctionReturnNode()
  ]}
 New Function_ContainedGraph {nodes [
   n1 = RigVMFunctionEntryNode()
   RigVMFunction_Sequence = &RigVMFunction_Sequence(ExecuteContext=@n1.ExecuteContext)
   Compute Spline Positions = RigVMFunctionReferenceNode(Debug=false, Array=@n1.Items, ExecuteContext=@RigVMFunction_Sequence.A)
   n2 = RigVMDispatchNode(Array=@n1.Items) {template="DISPATCH_RigVMDispatch_ArrayGetNum(in Array,out Num)"}
   n3 = &RigVMFunction_ControlFlowBranch(
    ExecuteContext=@"Compute Spline Positions".ExecuteContext
    Condition=/Script/RigVM.RigVMFunction_MathIntGreaterEqual(B=4, A=@n2.Num) {is=struct RigVMUnitNode, template="GreaterEqual::Execute(in A,in B,out Result)"}.Result
   )
   n4 = &RigVMFunction_ControlFlowBranch(
    Condition=RigVMDispatchNode(B=3, A=@n2.Num) {template="DISPATCH_RigVMDispatch_CoreEquals(in A,in B,out Result)"}.Result
    ExecuteContext=@n3.Completed
   )
   RigVMFunctionReturnNode(ExecuteContext=@RigVMFunction_Sequence.B, Value=RigVMVariableNode() {variable=Final Points}.Value)
   n5 = RigVMDispatchNode(Index=0, Array=@"Compute Spline Positions".Value) {template="DISPATCH_RigVMDispatch_ArrayGetAtIndex(in Array,in Index,out Element)"}
   n6 = RigVMDispatchNode(Index=1, Array=@"Compute Spline Positions".Value) {template="DISPATCH_RigVMDispatch_ArrayGetAtIndex(in Array,in Index,out Element)"}
   n7 = RigVMDispatchNode(Index=2, Array=@"Compute Spline Positions".Value) {template="DISPATCH_RigVMDispatch_ArrayGetAtIndex(in Array,in Index,out Element)"}
   SplineThrough3Points = RigVMFunctionReferenceNode(t=0.5, Point 0=@n5.Element, Point 1=@n6.Element, Point 2=@n7.Element, Execute=@n4.True)
   RigVMVariableNode(
    Value=((X=0,Y=0,Z=0),(X=0,Y=0,Z=0),(X=0,Y=0,Z=0),(X=0,Y=0,Z=0))
    ExecuteContext=@SplineThrough3Points.Execute
    Value.0=@SplineThrough3Points.Out Point 1
    Value.1=@SplineThrough3Points.Out Point 2
    Value.2=@SplineThrough3Points.Out Point 3
    Value.3=@SplineThrough3Points.Out Point 4
   ) {variable=Final Points}
   SplineThrough4Points = RigVMFunctionReferenceNode(
    Point 0=@n5.Element
    Point 1=@n6.Element
    Point 2=@n7.Element
    Point 3=RigVMDispatchNode(Index=3, Array=@"Compute Spline Positions".Value) {template="DISPATCH_RigVMDispatch_ArrayGetAtIndex(in Array,in Index,out Element)"}.Element
    Execute=@n3.True
   )
   RigVMVariableNode(
    Value=((X=0,Y=0,Z=0),(X=0,Y=0,Z=0),(X=0,Y=0,Z=0),(X=0,Y=0,Z=0))
    ExecuteContext=@SplineThrough4Points.Execute
    Value.0=@SplineThrough4Points.Out Point 1
    Value.1=@SplineThrough4Points.Out Point 2
    Value.2=@SplineThrough4Points.Out Point 3
    Value.3=@SplineThrough4Points.Out Point 4
   ) {variable=Final Points}
  ]}
 New Function_ContainedGraph {nodes [
   n1 = RigVMFunctionEntryNode()
   > RigVMFunction_Sequence_3 = RigVMAggregateNode()
   RigVMVariableNode(Value=@n1.FK IK Switch, ExecuteContext=@RigVMFunction_Sequence_3.B) {variable=FK or IK}
   > RigVMFunction_Sequence_1 = RigVMAggregateNode()
   Parent = RigVMFunctionReferenceNode(
    New Parent=RigVMDispatchNode(Default=(Type=Bone,Name="None"), Name=Hips Control, NameSpace=Self) {template="DISPATCH_RigDispatch_GetModuleMetadata(in Name,in NameSpace,in Default,out Value,out Found)"}.Value
    Child=&RigUnit_ResolveConnector(Connector=(Type=Connector,Name="Pelvis Bone"), SkipSocket=true).Result
    ExecuteContext=@RigVMFunction_Sequence_1.A
   )
   Get_Module_Metadata_2 = RigVMDispatchNode(Default=False, Name=Even IK Middle Controls, NameSpace=Self) {template="DISPATCH_RigDispatch_GetModuleMetadata(in Name,in NameSpace,in Default,out Value,out Found)"}
   RigVMFunction_ControlFlowBranch = &RigVMFunction_ControlFlowBranch(Condition=@Get_Module_Metadata_2.Value, ExecuteContext=@Parent.ExecuteContext)
   Get_Module_Metadata_2_1 = RigVMDispatchNode(Default=(Type=Bone,Name="None"), Name=Hips Tan Control, NameSpace=Self) {template="DISPATCH_RigDispatch_GetModuleMetadata(in Name,in NameSpace,in Default,out Value,out Found)"}
   &RigUnit_ParentConstraint(
    AdvancedSettings=(InterpolationType=Average,RotationOrderForFilter=XZY)
    Parents=((Item=(Type=Bone,Name="None"),Weight=0.7),(Item=(Type=Bone,Name="None"),Weight=0.3))
    Weight=1
    Parents.0.Item=RigVMDispatchNode(Default=(Type=Bone,Name="None"), Name=Hips Control, NameSpace=Self) {template="DISPATCH_RigDispatch_GetModuleMetadata(in Name,in NameSpace,in Default,out Value,out Found)"}.Value
    Parents.1.Item=RigVMDispatchNode(Default=(Type=Bone,Name="None"), Name=Chest Control, NameSpace=Self) {template="DISPATCH_RigDispatch_GetModuleMetadata(in Name,in NameSpace,in Default,out Value,out Found)"}.Value
    Child=&RigUnit_HierarchyGetParent(bDefaultParent=True, Child=@Get_Module_Metadata_2_1.Value).Parent
    ExecutePin=@RigVMFunction_ControlFlowBranch.True
   )
   Get_Module_Metadata_2_1_1 = RigVMDispatchNode(Default=(Type=Bone,Name="None"), Name=Chest Tan Control, NameSpace=Self) {template="DISPATCH_RigDispatch_GetModuleMetadata(in Name,in NameSpace,in Default,out Value,out Found)"}
   > &RigUnit_ParentConstraint(
    AdvancedSettings=(InterpolationType=Average,RotationOrderForFilter=XZY)
    Parents=((Item=(Type=Bone,Name="None"),Weight=0.3),(Item=(Type=Bone,Name="None"),Weight=0.7))
    Weight=1
    Parents.0.Item=RigVMDispatchNode(Default=(Type=Bone,Name="None"), Name=Hips Control, NameSpace=Self) {template="DISPATCH_RigDispatch_GetModuleMetadata(in Name,in NameSpace,in Default,out Value,out Found)"}.Value
    Parents.1.Item=RigVMDispatchNode(Default=(Type=Bone,Name="None"), Name=Chest Control, NameSpace=Self) {template="DISPATCH_RigDispatch_GetModuleMetadata(in Name,in NameSpace,in Default,out Value,out Found)"}.Value
    Child=&RigUnit_HierarchyGetParent(bDefaultParent=True, Child=@Get_Module_Metadata_2_1_1.Value).Parent
   )
   n2 = &RigVMFunction_ControlFlowBranch(ExecuteContext=@RigVMFunction_Sequence_1.B, Condition=RigVMVariableNode() {variable=FK or IK}.Value)
   Get_Module_Metadata_3_1 = RigVMDispatchNode(Default=(), Name=New Positions, NameSpace=Self) {template="DISPATCH_RigDispatch_GetModuleMetadata(in Name,in NameSpace,in Default,out Value,out Found)"}
   Get_Module_Metadata_4_1 = RigVMDispatchNode(Default=(), Name=Spine Rig Bones, NameSpace=Self) {template="DISPATCH_RigDispatch_GetModuleMetadata(in Name,in NameSpace,in Default,out Value,out Found)"}
   /Script/ControlRig.RigUnit_SetTransformItemArray(
    Space=GlobalSpace
    Weight=1
    bInitial=false
    bPropagateToChildren=True
    Transforms=@Get_Module_Metadata_3_1.Value
    Items=@Get_Module_Metadata_4_1.Value
    ExecutePin=@n2.False
   ) {is=struct RigVMUnitNode}
   > Dual SplineIK for Spine = RigVMFunctionReferenceNode(
    Controls=((Type=None,Name="None"),(Type=None,Name="None"),(Type=None,Name="None"),(Type=None,Name="None"))
    Samples Per Segment=16
    Spline Mode=BSpline
    Controls.0=RigVMDispatchNode(Default=(Type=Bone,Name="None"), Name=Hips Control, NameSpace=Self) {template="DISPATCH_RigDispatch_GetModuleMetadata(in Name,in NameSpace,in Default,out Value,out Found)"}.Value
    Bones=RigVMDispatchNode(Default=(), Name=Spine Rig Bones, NameSpace=Self) {template="DISPATCH_RigDispatch_GetModuleMetadata(in Name,in NameSpace,in Default,out Value,out Found)"}.Value
    Controls.3=RigVMDispatchNode(Default=(Type=Bone,Name="None"), Name=Chest Control, NameSpace=Self) {template="DISPATCH_RigDispatch_GetModuleMetadata(in Name,in NameSpace,in Default,out Value,out Found)"}.Value
    Controls.2=RigVMDispatchNode(Default=(Type=Bone,Name="None"), Name=Chest Tan Control, NameSpace=Self) {template="DISPATCH_RigDispatch_GetModuleMetadata(in Name,in NameSpace,in Default,out Value,out Found)"}.Value
    Controls.1=RigVMDispatchNode(Default=(Type=Bone,Name="None"), Name=Hips Tan Control, NameSpace=Self) {template="DISPATCH_RigDispatch_GetModuleMetadata(in Name,in NameSpace,in Default,out Value,out Found)"}.Value
    Aim Spline=@n1.Aim Spline
    Template Spline=@n1.Template Spline
    debug=RigVMVariableNode() {variable=debug}.Value
    Stretch=RigVMVariableNode() {variable=Stretch}.Value
    Primary Axis=RigVMVariableNode() {variable=Primary Axis}.Value
    Up Axis=RigVMVariableNode() {variable=Up Axis}.Value
    Secondary Spline Direction=RigVMVariableNode() {variable=Secondary Spline Direction}.Value
   )
   > Project to New Parent = RigVMFunctionReferenceNode(
    Propagate to Children=True
    Parents=RigVMDispatchNode(Default=(), Name=Spine Rig Bones, NameSpace=Self) {template="DISPATCH_RigDispatch_GetModuleMetadata(in Name,in NameSpace,in Default,out Value,out Found)"}.Value
    Children=&RigUnit_FindItemsWithMetadataTag(NameSpace=Self, Tag=FK Spine Controls).Items
   )
   RigVMFunction_Sequence = RigVMAggregateNode(ExecuteContext=@n2.True)
   DISPATCH_RigVMDispatch_ArrayIterator = RigVMDispatchNode(
    Array=&RigUnit_FindItemsWithMetadataTag(NameSpace=Self, Tag=FK Spine Controls).Items
    ExecuteContext=@RigVMFunction_Sequence.B
   ) {template="DISPATCH_RigVMDispatch_ArrayIterator(in Array,out Element,out Index,out Count,out Ratio)"}
   > &RigUnit_ParentConstraint(
    AdvancedSettings=(InterpolationType=Average,RotationOrderForFilter=XZY)
    Parents=((Item=(Type=Bone,Name="None"),Weight=1))
    Weight=1
    Child=RigVMDispatchNode(Default=(Type=Bone,Name="None"), Name=Item, NameSpace=Self, Item=@DISPATCH_RigVMDispatch_ArrayIterator.Element) {template="DISPATCH_RigDispatch_GetMetadata(in Item,in Name,in NameSpace,in Default,out Value,out Found)"}.Value
    Parents.0.Item=@DISPATCH_RigVMDispatch_ArrayIterator.Element
   )
   Parent_1 = RigVMFunctionReferenceNode(
    New Parent=RigVMDispatchNode(
     Index=-1
     Array=RigVMDispatchNode(Default=(), Name=Spine Rig Bones, NameSpace=Self) {template="DISPATCH_RigDispatch_GetModuleMetadata(in Name,in NameSpace,in Default,out Value,out Found)"}.Value
    ) {template="DISPATCH_RigVMDispatch_ArrayGetAtIndex(in Array,in Index,out Element)"}.Element
    Child=RigVMDispatchNode(Default=(Type=Bone,Name="None"), Name=Chest Space Null, NameSpace=Root) {template="DISPATCH_RigDispatch_GetModuleMetadata(in Name,in NameSpace,in Default,out Value,out Found)"}.Value
    ExecuteContext=@RigVMFunction_Sequence_1.C
   )
   Set IKFK Visibility_2 = RigVMFunctionReferenceNode(
    Show Controls=true
    IK Controls=&RigUnit_FindItemsWithMetadataTag(NameSpace=Self, Tag=IK Spine Controls).Items
    FK Controls=&RigUnit_FindItemsWithMetadataTag(NameSpace=Self, Tag=FK Spine Controls).Items
    IKFK Switch=&RigVMFunction_MathBoolNot(Value=RigVMVariableNode() {variable=FK or IK}.Value).Result
    ExecuteContext=@RigVMFunction_Sequence_1.D
   )
   n3 = &RigVMFunction_ControlFlowBranch(Condition=@n1.Attach Skeleton, ExecuteContext=@RigVMFunction_Sequence_3.C)
   Get_Module_Metadata_1 = RigVMDispatchNode(Default=(), Name=Spine Bones, NameSpace=Self) {template="DISPATCH_RigDispatch_GetModuleMetadata(in Name,in NameSpace,in Default,out Value,out Found)"}
   n4 = RigVMDispatchNode(Array=@Get_Module_Metadata_1.Value, ExecuteContext=@n3.True) {template="DISPATCH_RigVMDispatch_ArrayIterator(in Array,out Element,out Index,out Count,out Ratio)"}
   Get_Module_Metadata = RigVMDispatchNode(Default=(), Name=Spine Rig Bones, NameSpace=Self) {template="DISPATCH_RigDispatch_GetModuleMetadata(in Name,in NameSpace,in Default,out Value,out Found)"}
   > FindClosestBone_0 = RigVMFunctionReferenceNode(Initial=true, Item=@n4.Element, Bones=@Get_Module_Metadata.Value)
   > Parent_2 = RigVMFunctionReferenceNode(Child=@n4.Element, New Parent=@FindClosestBone_0.Value)
   RigVMFunctionReturnNode(ExecuteContext=@RigVMFunction_Sequence_3.D)
   RigVMFunction_Sequence_2 = RigVMAggregateNode(ExecuteContext=())
   n5 = RigVMDispatchNode(
    Array=&RigUnit_FindItemsWithMetadataTag(NameSpace=Self, Tag=FK Spine Controls).Items
    ExecuteContext=@RigVMFunction_Sequence_2.A
   ) {template="DISPATCH_RigVMDispatch_ArrayIterator(in Array,out Element,out Index,out Count,out Ratio)"}
   > /Script/ControlRig.RigUnit_SendEvent(Event=RequestAutoKey, bEnable=True, bOnlyDuringInteraction=True, Item=@n5.Element) {is=struct RigVMUnitNode, template="SendEvent::Execute(in Event,in Item,in OffsetInSeconds,in bEnable,in bOnlyDuringInteraction)"}
   n6 = RigVMDispatchNode(
    Array=&RigUnit_FindItemsWithMetadataTag(NameSpace=Self, Tag=IK Spine Controls).Items
    ExecuteContext=@RigVMFunction_Sequence_2.B
   ) {template="DISPATCH_RigVMDispatch_ArrayIterator(in Array,out Element,out Index,out Count,out Ratio)"}
   > &RigUnit_SendEvent(Event=RequestAutoKey, bEnable=True, bOnlyDuringInteraction=True, Item=@n6.Element) {template="SendEvent::Execute(in Event,in Item,in OffsetInSeconds,in bEnable,in bOnlyDuringInteraction)"}
   &RigUnit_SendEvent(
    Event=RequestAutoKey
    bEnable=True
    bOnlyDuringInteraction=True
    Item=RigVMDispatchNode(Default=(Type=Bone,Name="None"), Name=IK FK Switch, NameSpace=Self) {template="DISPATCH_RigDispatch_GetModuleMetadata(in Name,in NameSpace,in Default,out Value,out Found)"}.Value
    ExecutePin=@RigVMFunction_Sequence_2.C
   ) {template="SendEvent::Execute(in Event,in Item,in OffsetInSeconds,in bEnable,in bOnlyDuringInteraction)"}
   DISPATCH_RigDispatch_GetModuleMetadata = RigVMDispatchNode(Default=(), Name=Spine Offsets, NameSpace=Self) {template="DISPATCH_RigDispatch_GetModuleMetadata(in Name,in NameSpace,in Default,out Value,out Found)"}
   Spine Match IK onto FK = RigVMFunctionReferenceNode(
    IK Controls=((Type=None,Name="None"),(Type=None,Name="None"),(Type=None,Name="None"),(Type=None,Name="None"))
    Point 0=(X=0,Y=0,Z=0)
    Point 1=(X=0,Y=0,Z=0)
    Point 2=(X=0,Y=0,Z=0)
    Point 3=(X=0,Y=0,Z=0)
    Offsets=@DISPATCH_RigDispatch_GetModuleMetadata.Value
    IK Controls.0=RigVMDispatchNode(Default=(Type=Bone,Name="None"), Name=Hips Control, NameSpace=Self) {template="DISPATCH_RigDispatch_GetModuleMetadata(in Name,in NameSpace,in Default,out Value,out Found)"}.Value
    Bones=RigVMDispatchNode(Default=(), Name=Spine Rig Bones, NameSpace=Self) {template="DISPATCH_RigDispatch_GetModuleMetadata(in Name,in NameSpace,in Default,out Value,out Found)"}.Value
    IK Controls.3=RigVMDispatchNode(Default=(Type=Bone,Name="None"), Name=Chest Control, NameSpace=Self) {template="DISPATCH_RigDispatch_GetModuleMetadata(in Name,in NameSpace,in Default,out Value,out Found)"}.Value
    IK Controls.1=RigVMDispatchNode(Default=(Type=Bone,Name="None"), Name=Hips Tan Control, NameSpace=Self) {template="DISPATCH_RigDispatch_GetModuleMetadata(in Name,in NameSpace,in Default,out Value,out Found)"}.Value
    IK Controls.2=RigVMDispatchNode(Default=(Type=Bone,Name="None"), Name=Chest Tan Control, NameSpace=Self) {template="DISPATCH_RigDispatch_GetModuleMetadata(in Name,in NameSpace,in Default,out Value,out Found)"}.Value
    Execute=@RigVMFunction_Sequence.C
   )
  ]}
 New Function_ContainedGraph {nodes [
   n1 = RigVMFunctionEntryNode()
   RigVMFunction_Sequence = &RigVMFunction_Sequence(ExecuteContext=@n1.ExecuteContext)
   Get_Module_Metadata = RigVMDispatchNode(Default=(Type=Bone,Name="None"), NameSpace=Root, Name=@n1.Metadata Name) {template="DISPATCH_RigDispatch_GetModuleMetadata(in Name,in NameSpace,in Default,out Value,out Found)"}
   n2 = RigVMVariableNode(Value=@Get_Module_Metadata.Found, ExecuteContext=@RigVMFunction_Sequence.A) {variable=found_metadata}
   RigVMFunction_ControlFlowBranch = &RigVMFunction_ControlFlowBranch(Condition=@Get_Module_Metadata.Found, ExecuteContext=@n2.ExecuteContext)
   n3 = &RigUnit_HierarchyAddSocket(
    Space=LocalSpace
    Color=@n1.Color
    Description=@n1.Description
    Transform=@n1.Extra Transform
    Name=@n1.Name
    Parent=@Get_Module_Metadata.Value
    ExecutePin=@RigVMFunction_ControlFlowBranch.True
   )
   > RigVMVariableNode(Value=@n3.Item) {variable=the_socket}
   RigVMFunctionReturnNode(
    Bone=@Get_Module_Metadata.Value
    ExecuteContext=@RigVMFunction_Sequence.B
    New Socket=RigVMVariableNode() {variable=the_socket}.Value
    Found=RigVMVariableNode() {variable=found_metadata}.Value
   )
  ]}
 New Function_ContainedGraph {nodes [
   n1 = RigVMFunctionEntryNode()
   > &RigUnit_SetTransform(Weight=1, bInitial=true, Item=@n1.Control, Value=&RigUnit_GetTransform(bInitial=true, Item=@n1.Control).Transform)
   > RigVMFunctionReturnNode()
  ]}
 New Function_ContainedGraph {nodes [
   n1 = RigVMFunctionEntryNode()
   > n2 = RigVMVariableNode(Value=false) {variable=found}
   RigVMFunction_ControlFlowBranch = &RigVMFunction_ControlFlowBranch(Condition=@n1.Children, ExecuteContext=@n2.ExecuteContext)
   RigVMFunctionReturnNode(ExecuteContext=@RigVMFunction_ControlFlowBranch.Completed, Item Found=RigVMVariableNode() {variable=found item}.Value)
   DISPATCH_RigVMDispatch_ArrayIterator = RigVMDispatchNode(
    Array=/Script/ControlRig.RigUnit_HierarchyGetParentsItemArray(bDefaultParent=True, bIncludeChild=true, bReverse=false, Child=@n1.Item) {is=struct RigVMUnitNode}.Parents
    ExecuteContext=@RigVMFunction_ControlFlowBranch.False
   ) {template="DISPATCH_RigVMDispatch_ArrayIterator(in Array,out Element,out Index,out Count,out Ratio)"}
   > n3 = RigVMDispatchNode(Array=@n1.Search for) {template="DISPATCH_RigVMDispatch_ArrayIterator(in Array,out Element,out Index,out Count,out Ratio)"}
   Contains = /Script/RigVM.RigVMFunction_Contains(Name=@DISPATCH_RigVMDispatch_ArrayIterator.Element.Name, Search=@n3.Element) {is=struct RigVMUnitNode, template="Contains::Execute(in Name,in Search,out Result)"}
   RigVMFunction_MathBoolNot = &RigVMFunction_MathBoolNot(Value=RigVMVariableNode() {variable=found}.Value)
   RigVMFunction_MathBoolAnd = &RigVMFunction_MathBoolAnd(A=@Contains.Result, B=@RigVMFunction_MathBoolNot.Result)
   n4 = &RigVMFunction_ControlFlowBranch(ExecuteContext=@n3.ExecuteContext, Condition=@RigVMFunction_MathBoolAnd.Result)
   RigVMVariableNode(Value=true, ExecuteContext=@n4.True) {variable=found}
   > RigVMVariableNode(Value=@DISPATCH_RigVMDispatch_ArrayIterator.Element) {variable=found item}
   RigVMVariableNode(Value=False, ExecuteContext=@RigVMFunction_ControlFlowBranch.True) {variable=bad name}
   > n5 = RigVMDispatchNode(Array=&RigUnit_CollectionChildrenArray(bDefaultChildren=True, bIncludeParent=true, bRecursive=true, Parent=@n1.Item, TypeToSearch=@n1.Type to Search).Items) {template="DISPATCH_RigVMDispatch_ArrayIterator(in Array,out Element,out Index,out Count,out Ratio)"}
   > RigVMVariableNode(Value=false) {variable=bad name}
   > n6 = RigVMDispatchNode(Array=@n1.Ignore) {template="DISPATCH_RigVMDispatch_ArrayIterator(in Array,out Element,out Index,out Count,out Ratio)"}
   n7 = &RigVMFunction_ControlFlowBranch(
    Condition=&RigVMFunction_Contains(Name=@n5.Element.Name, Search=@n6.Element) {template="Contains::Execute(in Name,in Search,out Result)"}.Result
    ExecuteContext=@n6.ExecuteContext
   )
   RigVMVariableNode(Value=true, ExecuteContext=@n7.True) {variable=bad name}
   n8 = RigVMDispatchNode(ExecuteContext=@n6.Completed, Array=@n1.Search for) {template="DISPATCH_RigVMDispatch_ArrayIterator(in Array,out Element,out Index,out Count,out Ratio)"}
   RigVMFunction_MathBoolAnd_1 = RigVMAggregateNode(
    A=&RigVMFunction_Contains(Name=@n5.Element.Name, Search=@n8.Element) {template="Contains::Execute(in Name,in Search,out Result)"}.Result
    B=&RigVMFunction_MathBoolNot(Value=RigVMVariableNode() {variable=found}.Value).Result
    C=&RigVMFunction_MathBoolNot(Value=RigVMVariableNode() {variable=bad name}.Value).Result
   )
   n9 = &RigVMFunction_ControlFlowBranch(ExecuteContext=@n8.ExecuteContext, Condition=@RigVMFunction_MathBoolAnd_1.Result)
   RigVMVariableNode(Value=true, ExecuteContext=@n9.True) {variable=found}
   > RigVMVariableNode(Value=@n5.Element) {variable=found item}
  ]}
 New Function_ContainedGraph {nodes [
   n1 = RigVMFunctionEntryNode()
   > RigVMFunction_Sequence = RigVMAggregateNode()
   RigVMFunction_Sequence_1 = RigVMAggregateNode(ExecuteContext=@RigVMFunction_Sequence.A)
   DISPATCH_RigVMDispatch_ArrayIterator = RigVMDispatchNode(Array=@n1.Bones, ExecuteContext=@RigVMFunction_Sequence_1.A) {template="DISPATCH_RigVMDispatch_ArrayIterator(in Array,out Element,out Index,out Count,out Ratio)"}
   > DISPATCH_RigDispatch_SetMetadata = RigVMDispatchNode(
    Name=Orig Parent
    NameSpace=Self
    Item=@DISPATCH_RigVMDispatch_ArrayIterator.Element
    Value=&RigUnit_HierarchyGetParent(bDefaultParent=True, Child=@DISPATCH_RigVMDispatch_ArrayIterator.Element).Parent
   ) {template="DISPATCH_RigDispatch_SetMetadata(in Item,in Name,in NameSpace,in Value,out Success)"}
   n2 = RigVMVariableNode() {variable=Bones}
   > &RigUnit_SetDefaultParent(
    Child=@DISPATCH_RigVMDispatch_ArrayIterator.Element
    Parent=&RigUnit_HierarchyGetParent(
     bDefaultParent=True
     Child=RigVMDispatchNode(Index=0, Array=@n2.Value) {template="DISPATCH_RigVMDispatch_ArrayGetAtIndex(in Array,in Index,out Element)"}.Element
    ).Parent
   )
   DISPATCH_RigVMDispatch_ArrayGetNum = RigVMDispatchNode(Array=@n1.Bones) {template="DISPATCH_RigVMDispatch_ArrayGetNum(in Array,out Num)"}
   Interpolate = &RigVMFunction_MathTransformLerp(
    T=0.5
    A=&RigUnit_GetTransform(
     bInitial=False
     Item=RigVMDispatchNode(Index=0, Array=@n1.Bones) {template="DISPATCH_RigVMDispatch_ArrayGetAtIndex(in Array,in Index,out Element)"}.Element
    ).Transform
    B=&RigUnit_GetTransform(
     bInitial=False
     Item=RigVMDispatchNode(Index=-1, Array=@n1.Bones) {template="DISPATCH_RigVMDispatch_ArrayGetAtIndex(in Array,in Index,out Element)"}.Element
    ).Transform
   ) {template="Interpolate::Execute(in A,in B,in T,out Result)"}
   DISPATCH_RigVMDispatch_If = RigVMDispatchNode(
    False=(X=0,Y=0,Z=0)
    Condition=RigVMDispatchNode(B=3, A=@DISPATCH_RigVMDispatch_ArrayGetNum.Num) {template="DISPATCH_RigVMDispatch_CoreEquals(in A,in B,out Result)"}.Result
    True=@Interpolate.Result.Translation
   ) {template="DISPATCH_RigVMDispatch_If(in Condition,in True,in False,out Result)"}
   RigVMFunction_DebugTransformMutableNoSpace = &RigVMFunction_DebugTransformMutableNoSpace(
    Color=(R=1,G=0,B=0,A=1)
    DebugDrawSettings=(DepthPriority=SDPG_Foreground,Lifetime=-1)
    Mode=Point
    Scale=10
    Thickness=1
    Transform=(Rotation=(X=0,Y=0,Z=0,W=1),Translation=(X=0,Y=0,Z=0),Scale3D=(X=1,Y=1,Z=1))
    WorldOffset=(Rotation=(X=0,Y=0,Z=0,W=1),Translation=(X=0,Y=0,Z=0),Scale3D=(X=1,Y=1,Z=1))
    Transform.Translation=@DISPATCH_RigVMDispatch_If.Result
    Transform.Rotation=@Interpolate.Result.Rotation
    ExecutePin=@RigVMFunction_Sequence_1.B
    bEnabled=RigVMVariableNode() {variable=Debug}.Value
   )
   > RigVMVariableNode(Value=@DISPATCH_RigVMDispatch_If.Result) {variable=Secondary Target}
   n3 = RigVMDispatchNode(Array=@n1.Bones, ExecuteContext=@RigVMFunction_Sequence_1.C) {template="DISPATCH_RigVMDispatch_ArrayIterator(in Array,out Element,out Index,out Count,out Ratio)"}
   Subtract = &RigVMFunction_MathIntSub(B=1, A=@n3.Count) {template="Subtract::Execute(in A,in B,out Result)"}
   DISPATCH_RigVMDispatch_CoreEquals = RigVMDispatchNode(A=@n3.Index, B=@Subtract.Result) {template="DISPATCH_RigVMDispatch_CoreEquals(in A,in B,out Result)"}
   n4 = &RigVMFunction_ControlFlowBranch(ExecuteContext=@n3.ExecuteContext, Condition=@DISPATCH_RigVMDispatch_CoreEquals.Result)
   Add = /Script/RigVM.RigVMFunction_MathIntAdd(B=1, A=@n3.Index) {is=struct RigVMUnitNode, template="Add::Execute(in A,in B,out Result)"}
   n5 = &RigUnit_AimBoneMath(
    DebugSettings=(bEnabled=False,Scale=10,WorldOffset=(Rotation=(X=0,Y=0,Z=0,W=1),Translation=(X=0,Y=0,Z=0),Scale3D=(X=1,Y=1,Z=1)))
    Primary=(Weight=1,Axis=(X=1,Y=0,Z=0),Target=(X=1,Y=0,Z=0),Kind=Location,Space=(Type=Bone,Name="None"))
    Secondary=(Weight=1,Axis=(X=0,Y=0,Z=1),Target=(X=1,Y=0,Z=0),Kind=Direction,Space=(Type=Bone,Name="None"))
    Weight=1
    Secondary.Kind=@n1.Kind
    Primary.Axis=@n1.Primary Axis
    Secondary.Axis=@n1.Secondary Axis
    InputTransform=&RigUnit_GetTransform(bInitial=False, Item=@n3.Element).Transform
    Primary.Target=&RigUnit_GetTransform(
     bInitial=False
     Item=RigVMDispatchNode(Index=@Add.Result, Array=@n1.Bones) {template="DISPATCH_RigVMDispatch_ArrayGetAtIndex(in Array,in Index,out Element)"}.Element
    ).Transform.Translation
    Secondary.Target=RigVMVariableNode() {variable=Secondary Target}.Value
   )
   &RigUnit_SetTransform(Weight=1, bInitial=true, Value=@n5.Result, Item=@n3.Element, ExecutePin=@n4.False)
   n6 = RigVMDispatchNode(ExecuteContext=@n3.Completed, Array=RigVMVariableNode() {variable=Bones}.Value) {template="DISPATCH_RigVMDispatch_ArrayIterator(in Array,out Element,out Index,out Count,out Ratio)"}
   DISPATCH_RigDispatch_GetMetadata = RigVMDispatchNode(Default=(Type=Bone,Name="None"), Name=Orig Parent, NameSpace=Self, Item=@n6.Element) {template="DISPATCH_RigDispatch_GetMetadata(in Item,in Name,in NameSpace,in Default,out Value,out Found)"}
   > &RigUnit_SetDefaultParent(Parent=@DISPATCH_RigDispatch_GetMetadata.Value, Child=@n6.Element)
   DISPATCH_RigVMDispatch_ArrayGetAtIndex = RigVMDispatchNode(Index=-1, Array=RigVMVariableNode() {variable=Bones}.Value) {template="DISPATCH_RigVMDispatch_ArrayGetAtIndex(in Array,in Index,out Element)"}
   Set Transform = &RigUnit_SetRotation(Value=(X=0,Y=0,Z=0,W=1), Weight=1, bInitial=true, Item=@DISPATCH_RigVMDispatch_ArrayGetAtIndex.Element, ExecutePin=@n6.Completed)
   RigVMFunction_DebugTransformArrayMutableNoSpace = /Script/RigVM.RigVMFunction_DebugTransformArrayMutableNoSpace(
    Color=(R=1,G=1,B=1,A=1)
    DebugDrawSettings=(DepthPriority=SDPG_Foreground,Lifetime=-1)
    Mode=Axes
    ParentIndices=()
    Scale=15
    Thickness=1
    WorldOffset=(Rotation=(X=0,Y=0,Z=0,W=1),Translation=(X=0,Y=0,Z=0),Scale3D=(X=1,Y=1,Z=1))
    bEnabled=@n1.Debug
    Transforms=&RigUnit_GetTransformItemArray(Space=GlobalSpace, bInitial=False, Items=@n1.Bones).Transforms
    ExecutePin=@RigVMFunction_Sequence.B
   ) {is=struct RigVMUnitNode}
   RigVMFunctionReturnNode(ExecuteContext=@RigVMFunction_Sequence.C)
   &RigUnit_FindItemsWithMetadataTag(NameSpace=Self, Tag=New Bones)
  ]}
 New Function_ContainedGraph {nodes [
   n1 = RigVMFunctionEntryNode()
   RigVMFunction_Sequence = &RigVMFunction_Sequence(ExecuteContext=@n1.ExecuteContext)
   n2 = &RigVMFunction_Sequence(ExecuteContext=@RigVMFunction_Sequence.A)
   RigVMDispatchNode(ExecuteContext=@n2.A, Array=RigVMVariableNode() {variable=New Bones}.Value) {template="DISPATCH_RigVMDispatch_ArrayReset(io Array)"}
   > DISPATCH_RigVMDispatch_ArrayIterator = RigVMDispatchNode(Array=@n1.Transforms) {template="DISPATCH_RigVMDispatch_ArrayIterator(in Array,out Element,out Index,out Count,out Ratio)"}
   RigVMFunction_StringPadInteger = /Script/RigVM.RigVMFunction_StringPadInteger(
    Digits=2
    Value=&RigVMFunction_MathIntAdd(B=1, A=@DISPATCH_RigVMDispatch_ArrayIterator.Index) {template="Add::Execute(in A,in B,out Result)"}.Result
   ) {is=struct RigVMUnitNode}
   DISPATCH_RigDispatch_FromString = RigVMDispatchNode(String=@RigVMFunction_StringPadInteger.Result) {template="DISPATCH_RigDispatch_FromString(in String,out Result)"}
   Concat = RigVMAggregateNode(B=_, D=_, C=@DISPATCH_RigDispatch_FromString.Result, A=@n1.Name, E=@n1.Optional Suffix)
   Concat_1 = RigVMAggregateNode(B=_, D=None, C=@DISPATCH_RigDispatch_FromString.Result, A=RigVMVariableNode() {variable=Name}.Value)
   EndsWith = /Script/RigVM.RigVMFunction_EndsWith(Ending=_, Name=@Concat.Result) {is=struct RigVMUnitNode, template="EndsWith::Execute(in Name,in Ending,out Result)"}
   DISPATCH_RigVMDispatch_If = RigVMDispatchNode(False=@Concat.Result, True=@Concat_1.Result, Condition=@EndsWith.Result) {template="DISPATCH_RigVMDispatch_If(in Condition,in True,in False,out Result)"}
   > n3 = &RigUnit_HierarchyAddBone(Space=GlobalSpace, Transform=@DISPATCH_RigVMDispatch_ArrayIterator.Element, Name=@DISPATCH_RigVMDispatch_If.Result, Parent=@n1.Parent)
   > Add = RigVMDispatchNode(Element=@n3.Item, Array=RigVMVariableNode() {variable=New Bones}.Value) {template="DISPATCH_RigVMDispatch_ArrayAdd(io Array,in Element,out Index)"}
   n4 = RigVMVariableNode() {variable=New Bones}
   n5 = RigVMDispatchNode(ExecuteContext=@n2.B, Array=@n4.Value) {template="DISPATCH_RigVMDispatch_ArrayIterator(in Array,out Element,out Index,out Count,out Ratio)"}
   n6 = &RigVMFunction_MathIntSub(B=1, A=@n5.Index) {template="Subtract::Execute(in A,in B,out Result)"}
   > &RigUnit_SetDefaultParent(
    Child=@n5.Element
    Parent=RigVMDispatchNode(
     False=RigVMDispatchNode(Index=@n6.Result, Array=@n4.Value) {template="DISPATCH_RigVMDispatch_ArrayGetAtIndex(in Array,in Index,out Element)"}.Element
     Condition=RigVMDispatchNode(B=0, A=@n5.Index) {template="DISPATCH_RigVMDispatch_CoreEquals(in A,in B,out Result)"}.Result
     True=RigVMVariableNode() {variable=Parent}.Value
    ) {template="DISPATCH_RigVMDispatch_If(in Condition,in True,in False,out Result)"}.Result
   )
   RigVMFunction_ControlFlowBranch = &RigVMFunction_ControlFlowBranch(ExecuteContext=@n5.Completed, Condition=RigVMVariableNode() {variable=Enable Orient Bones}.Value)
   Orient Bones = RigVMFunctionReferenceNode(
    ExecuteContext=@RigVMFunction_ControlFlowBranch.True
    Secondary Axis=RigVMVariableNode() {variable=Secondary Axis}.Value
    Primary Axis=RigVMVariableNode() {variable=Primary Axis}.Value
    Bones=RigVMVariableNode() {variable=New Bones}.Value
    Debug=RigVMVariableNode() {variable=Debug}.Value
    Kind=RigVMVariableNode() {variable=Kind}.Value
   )
   RigVMFunctionReturnNode(ExecuteContext=@RigVMFunction_Sequence.B, Bones=RigVMVariableNode() {variable=New Bones}.Value)
   &RigUnit_FindItemsWithMetadataTag(NameSpace=Self, Tag=New Bones)
  ]}
 New Function_ContainedGraph {nodes [
   n1 = RigVMFunctionEntryNode()
   > DISPATCH_RigVMDispatch_ArrayIterator = RigVMDispatchNode(Array=@n1.Controls) {template="DISPATCH_RigVMDispatch_ArrayIterator(in Array,out Element,out Index,out Count,out Ratio)"}
   n2 = /Script/ControlRig.RigUnit_GetShapeTransform(Control=@DISPATCH_RigVMDispatch_ArrayIterator.Element.Name) {is=struct RigVMUnitNode}
   Multiply_1 = RigVMAggregateNode(
    B=(X=1,Y=1,Z=1)
    C=(X=0,Y=0,Z=0)
    B.X=@n1.Global Scale
    B.Y=@n1.Global Scale
    B.Z=@n1.Global Scale
    C.X=@n1.Local Scale
    C.Y=@n1.Local Scale
    C.Z=@n1.Local Scale
    A=@n2.Transform.Scale3D
   )
   > /Script/ControlRig.RigUnit_SetShapeTransform(
    Transform=(Rotation=(X=0,Y=0,Z=0,W=1),Translation=(X=0,Y=0,Z=0),Scale3D=(X=1,Y=1,Z=1))
    Control=@DISPATCH_RigVMDispatch_ArrayIterator.Element.Name
    Transform.Rotation=@n2.Transform.Rotation
    Transform.Translation=@n2.Transform.Translation
    Transform.Scale3D=@Multiply_1.Result
   ) {is=struct RigVMUnitNode}
   RigVMFunctionReturnNode(ExecuteContext=@DISPATCH_RigVMDispatch_ArrayIterator.Completed)
  ]}
 New Function_ContainedGraph {nodes [
   n1 = RigVMFunctionEntryNode()
   DISPATCH_RigDispatch_GetModuleMetadata = RigVMDispatchNode(Default=(R=1,G=1,B=1,A=1), Name=Color Center, NameSpace=Root) {template="DISPATCH_RigDispatch_GetModuleMetadata(in Name,in NameSpace,in Default,out Value,out Found)"}
   RigVMFunction_MathBoolNot = &RigVMFunction_MathBoolNot(Value=@n1.Override)
   DISPATCH_RigVMDispatch_If = RigVMDispatchNode(True=@DISPATCH_RigDispatch_GetModuleMetadata.Value, False=@n1.CenterColor, Condition=@RigVMFunction_MathBoolNot.Result) {template="DISPATCH_RigVMDispatch_If(in Condition,in True,in False,out Result)"}
   > ColorizeControls = RigVMFunctionReferenceNode(
    X Plane=true
    ignore_ctrls=()
    CenterColor=@DISPATCH_RigVMDispatch_If.Result
    LeftColor=RigVMDispatchNode(
     True=RigVMDispatchNode(Default=(R=1,G=1,B=1,A=1), Name=Color left, NameSpace=Root) {template="DISPATCH_RigDispatch_GetModuleMetadata(in Name,in NameSpace,in Default,out Value,out Found)"}.Value
     False=@n1.LeftColor
     Condition=@RigVMFunction_MathBoolNot.Result
    ) {template="DISPATCH_RigVMDispatch_If(in Condition,in True,in False,out Result)"}.Result
    RightColor=RigVMDispatchNode(
     True=RigVMDispatchNode(Default=(R=1,G=1,B=1,A=1), Name=Color Right, NameSpace=Root) {template="DISPATCH_RigDispatch_GetModuleMetadata(in Name,in NameSpace,in Default,out Value,out Found)"}.Value
     False=@n1.RightColor
     Condition=@RigVMFunction_MathBoolNot.Result
    ) {template="DISPATCH_RigVMDispatch_If(in Condition,in True,in False,out Result)"}.Result
    controls=@n1.Controls
   )
   > RigVMFunctionReturnNode()
  ]}
 New Function_ContainedGraph {nodes [
   n1 = RigVMFunctionEntryNode()
   > DISPATCH_RigVMDispatch_ArrayIterator = RigVMDispatchNode(Array=@n1.Item) {template="DISPATCH_RigVMDispatch_ArrayIterator(in Array,out Element,out Index,out Count,out Ratio)"}
   > &RigUnit_SetMetadataTag(NameSpace=Self, Tag=IK Spine Controls, Item=@DISPATCH_RigVMDispatch_ArrayIterator.Element)
   RigVMFunctionReturnNode(ExecuteContext=@DISPATCH_RigVMDispatch_ArrayIterator.Completed)
  ]}
 New Function_ContainedGraph {nodes [
   n1 = RigVMFunctionEntryNode()
   RigVMFunction_Sequence = &RigVMFunction_Sequence(ExecuteContext=@n1.ExecuteContext)
   Compute Spline Positions = RigVMFunctionReferenceNode(Debug=false, Array=@n1.Items, ExecuteContext=@RigVMFunction_Sequence.A)
   n2 = RigVMDispatchNode(Array=@n1.Items) {template="DISPATCH_RigVMDispatch_ArrayGetNum(in Array,out Num)"}
   n3 = &RigVMFunction_ControlFlowBranch(
    ExecuteContext=@"Compute Spline Positions".ExecuteContext
    Condition=&RigVMFunction_MathIntGreaterEqual(B=4, A=@n2.Num) {template="GreaterEqual::Execute(in A,in B,out Result)"}.Result
   )
   n4 = &RigVMFunction_ControlFlowBranch(
    Condition=RigVMDispatchNode(B=3, A=@n2.Num) {template="DISPATCH_RigVMDispatch_CoreEquals(in A,in B,out Result)"}.Result
    ExecuteContext=@n3.Completed
   )
   RigVMFunctionReturnNode(ExecuteContext=@RigVMFunction_Sequence.B, Value=RigVMVariableNode() {variable=Final Points}.Value)
   n5 = RigVMDispatchNode(Index=0, Array=@"Compute Spline Positions".Value) {template="DISPATCH_RigVMDispatch_ArrayGetAtIndex(in Array,in Index,out Element)"}
   n6 = RigVMDispatchNode(Index=1, Array=@"Compute Spline Positions".Value) {template="DISPATCH_RigVMDispatch_ArrayGetAtIndex(in Array,in Index,out Element)"}
   n7 = RigVMDispatchNode(Index=2, Array=@"Compute Spline Positions".Value) {template="DISPATCH_RigVMDispatch_ArrayGetAtIndex(in Array,in Index,out Element)"}
   SplineThrough3Points = RigVMFunctionReferenceNode(t=0.5, Point 0=@n5.Element, Point 1=@n6.Element, Point 2=@n7.Element, Execute=@n4.True)
   RigVMVariableNode(
    Value=((X=0,Y=0,Z=0),(X=0,Y=0,Z=0),(X=0,Y=0,Z=0),(X=0,Y=0,Z=0))
    ExecuteContext=@SplineThrough3Points.Execute
    Value.0=@SplineThrough3Points.Out Point 1
    Value.1=@SplineThrough3Points.Out Point 2
    Value.2=@SplineThrough3Points.Out Point 3
    Value.3=@SplineThrough3Points.Out Point 4
   ) {variable=Final Points}
   SplineThrough4Points = RigVMFunctionReferenceNode(
    Point 0=@n5.Element
    Point 1=@n6.Element
    Point 2=@n7.Element
    Point 3=RigVMDispatchNode(Index=3, Array=@"Compute Spline Positions".Value) {template="DISPATCH_RigVMDispatch_ArrayGetAtIndex(in Array,in Index,out Element)"}.Element
    Execute=@n3.True
   )
   RigVMVariableNode(
    Value=((X=0,Y=0,Z=0),(X=0,Y=0,Z=0),(X=0,Y=0,Z=0),(X=0,Y=0,Z=0))
    ExecuteContext=@SplineThrough4Points.Execute
    Value.0=@SplineThrough4Points.Out Point 1
    Value.1=@SplineThrough4Points.Out Point 2
    Value.2=@SplineThrough4Points.Out Point 3
    Value.3=@SplineThrough4Points.Out Point 4
   ) {variable=Final Points}
  ]}
 New Function_ContainedGraph {nodes [
   n1 = RigVMFunctionEntryNode()
   > RigVMFunction_Sequence_3 = RigVMAggregateNode()
   Set_Module_Metadata_1 = RigVMDispatchNode(Name=Even IK Middle Controls, NameSpace=Self, Value=@n1.Even IK Middle Controls, ExecuteContext=@RigVMFunction_Sequence_3.A) {template="DISPATCH_RigDispatch_SetModuleMetadata(in Name,in NameSpace,in Value,out Success)"}
   RigVMFunction_Sequence = RigVMAggregateNode(ExecuteContext=@RigVMFunction_Sequence_3.B)
   RigVMDispatchNode(Name=Hips Bone, NameSpace=Self, Value=@n1.Spine Bone Start Connector, ExecuteContext=@RigVMFunction_Sequence.B) {template="DISPATCH_RigDispatch_SetModuleMetadata(in Name,in NameSpace,in Value,out Success)"}
   > RigVMDispatchNode(Name=Chest Bone, NameSpace=Root, Value=@n1.Spine Bone End Connector) {template="DISPATCH_RigDispatch_SetModuleMetadata(in Name,in NameSpace,in Value,out Success)"}
   > RigVMDispatchNode(Name=Chest Bone, NameSpace=Self, Value=@n1.Spine Bone End Connector) {template="DISPATCH_RigDispatch_SetModuleMetadata(in Name,in NameSpace,in Value,out Success)"}
   > &RigUnit_SetMetadataTag(NameSpace=Self, Tag=Chest Bone, Item=@n1.Spine Bone End Connector)
   > RigVMDispatchNode(
    Name=Parent Control
    NameSpace=Self
    Value=&RigUnit_ResolveConnector(Connector=(Type=Connector,Name="Parent Control"), SkipSocket=true).Result
   ) {template="DISPATCH_RigDispatch_SetModuleMetadata(in Name,in NameSpace,in Value,out Success)"}
   DISPATCH_RigVMDispatch_ArrayGetAtIndex = RigVMDispatchNode(
    Index=0
    Array=&RigUnit_CollectionChildrenArray(
     TypeToSearch=Bone
     bDefaultChildren=True
     bIncludeParent=true
     bRecursive=False
     Parent=RigVMDispatchNode(Default=(Type=Bone,Name="None"), Name=Hips Bone, NameSpace=Self) {template="DISPATCH_RigDispatch_GetModuleMetadata(in Name,in NameSpace,in Default,out Value,out Found)"}.Value
    ).Items
   ) {template="DISPATCH_RigVMDispatch_ArrayGetAtIndex(in Array,in Index,out Element)"}
   > RigVMDispatchNode(
    Name=Spine Bones
    NameSpace=Self
    Value=/Script/ControlRig.RigUnit_HierarchyGetChainItemArray(
     bIncludeEnd=True
     bIncludeStart=True
     bReverse=False
     End=RigVMDispatchNode(Default=(Type=Bone,Name="None"), Name=Chest Bone, NameSpace=Self) {template="DISPATCH_RigDispatch_GetModuleMetadata(in Name,in NameSpace,in Default,out Value,out Found)"}.Value
     Start=@DISPATCH_RigVMDispatch_ArrayGetAtIndex.Element
    ) {is=struct RigVMUnitNode}.Chain
   ) {template="DISPATCH_RigDispatch_SetModuleMetadata(in Name,in NameSpace,in Value,out Success)"}
   > RigVMDispatchNode(
    Name=Bones
    NameSpace=Self
    Value=&RigUnit_HierarchyGetChainItemArray(
     bIncludeEnd=True
     bIncludeStart=True
     bReverse=False
     End=RigVMDispatchNode(Default=(Type=Bone,Name="None"), Name=Chest Bone, NameSpace=Self) {template="DISPATCH_RigDispatch_GetModuleMetadata(in Name,in NameSpace,in Default,out Value,out Found)"}.Value
     Start=RigVMDispatchNode(Default=(Type=Bone,Name="None"), Name=Hips Bone, NameSpace=Self) {template="DISPATCH_RigDispatch_GetModuleMetadata(in Name,in NameSpace,in Default,out Value,out Found)"}.Value
    ).Chain
   ) {template="DISPATCH_RigDispatch_SetModuleMetadata(in Name,in NameSpace,in Value,out Success)"}
   > RigVMDispatchNode(Name=Pelvis Bone, NameSpace=Self, Value=@n1.Pelvis Bone Connector) {template="DISPATCH_RigDispatch_SetModuleMetadata(in Name,in NameSpace,in Value,out Success)"}
   n2 = RigVMVariableNode() {variable=Rig Bone Positions}
   > RigVMDispatchNode(Array=@n2.Value) {template="DISPATCH_RigVMDispatch_ArrayReset(io Array)"}
   > Set_Num = RigVMDispatchNode(Array=@n2.Value, Num=RigVMVariableNode() {variable=Number of Spine Bones}.Value) {template="DISPATCH_RigVMDispatch_ArraySetNum(io Array,in Num)"}
   > RigVMFunction_ForLoopCount = &RigVMFunction_ForLoopCount(Count=RigVMVariableNode() {variable=Number of Spine Bones}.Value)
   n3 = RigVMDispatchNode(Default=(Type=Bone,Name="None"), Name=Chest Bone, NameSpace=Self) {template="DISPATCH_RigDispatch_GetModuleMetadata(in Name,in NameSpace,in Default,out Value,out Found)"}
   n4 = RigVMDispatchNode(Default=(Type=Bone,Name="None"), Name=Hips Bone, NameSpace=Self) {template="DISPATCH_RigDispatch_GetModuleMetadata(in Name,in NameSpace,in Default,out Value,out Found)"}
   > Set_At = RigVMDispatchNode(
    Element=&RigVMFunction_MathTransformLerp(
     B=&RigUnit_GetTransform(bInitial=true, Item=@n3.Value).Transform
     A=&RigUnit_GetTransform(bInitial=true, Item=@n4.Value).Transform
     T=@RigVMFunction_ForLoopCount.Ratio
    ) {template="Interpolate::Execute(in A,in B,in T,out Result)"}.Result
    Index=@RigVMFunction_ForLoopCount.Index
    Array=RigVMVariableNode() {variable=Rig Bone Positions}.Value
   ) {template="DISPATCH_RigVMDispatch_ArraySetAtIndex(io Array,in Index,in Element)"}
   Get_Module_Metadata_2 = RigVMDispatchNode(Default=(), Name=Bones, NameSpace=Self) {template="DISPATCH_RigDispatch_GetModuleMetadata(in Name,in NameSpace,in Default,out Value,out Found)"}
   n5 = RigVMDispatchNode(Index=0, Array=@Get_Module_Metadata_2.Value) {template="DISPATCH_RigVMDispatch_ArrayGetAtIndex(in Array,in Index,out Element)"}
   Create Bone Chain = RigVMFunctionReferenceNode(
    Debug=true
    Debug Scale=15
    Debug Thickness=1
    Enable Orient Bones=true
    Kind=Direction
    Optional Suffix=rig
    Parent=&RigUnit_HierarchyGetParent(bDefaultParent=True, Child=@n5.Element).Parent
    ExecuteContext=@RigVMFunction_ForLoopCount.Completed
    Transforms=RigVMVariableNode() {variable=Rig Bone Positions}.Value
    Primary Axis=RigVMVariableNode() {variable=Primary Axis}.Value
    Secondary Axis=RigVMVariableNode() {variable=Up Axis}.Value
    Name=@n5.Element.Name
   )
   > Set_Module_Metadata = RigVMDispatchNode(Name=Spine Rig Bones, NameSpace=Self, Value=@"Create Bone Chain".Bones) {template="DISPATCH_RigDispatch_SetModuleMetadata(in Name,in NameSpace,in Value,out Success)"}
   n6 = &RigVMFunction_Sequence(ExecuteContext=@RigVMFunction_Sequence.C)
   n7 = RigVMDispatchNode(Default=(Type=Bone,Name="None"), Name=Hips Bone, NameSpace=Self) {template="DISPATCH_RigDispatch_GetModuleMetadata(in Name,in NameSpace,in Default,out Value,out Found)"}
   n8 = RigVMDispatchNode(
    Index=0
    Array=RigVMDispatchNode(Default=(), Name=Spine Rig Bones, NameSpace=Self) {template="DISPATCH_RigDispatch_GetModuleMetadata(in Name,in NameSpace,in Default,out Value,out Found)"}.Value
   ) {template="DISPATCH_RigVMDispatch_ArrayGetAtIndex(in Array,in Index,out Element)"}
   Divide = /Script/RigVM.RigVMFunction_MathVectorDiv(
    B=(X=2,Y=2,Z=2)
    A=&RigVMFunction_MathVectorAdd(
     A=&RigUnit_GetTransform(bInitial=False, Item=@n7.Value).Transform.Translation
     B=&RigUnit_GetTransform(bInitial=False, Item=@n8.Element).Transform.Translation
    ) {template="Add::Execute(in A,in B,out Result)"}.Result
   ) {is=struct RigVMUnitNode, template="Divide::Execute(in A,in B,out Result)"}
   Control Stack at Position = RigVMFunctionReferenceNode(
    Add Bottom Null=False
    Add Null=False
    Add Secondary Control=False
    Control Suffix=_ctrl
    Global Position=(Rotation=(X=0,Y=0,Z=0,W=1),Translation=(X=0,Y=0,Z=0),Scale3D=(X=1,Y=1,Z=1))
    Name=body
    Null Suffix=_null
    Secondary Settings="(DisplayName=\"None\",Shape=(bVisible=True,Name=\"Default\",Color=(R=1,G=0,B=0,A=1),Transform=(Rotation=(X=0,Y=0,Z=0,W=1),Translation=(X=0,Y=0,Z=0),Scale3D=(X=1,Y=1,Z=1))),Proxy=(bIsProxy=False,ShapeVisibility=BasedOnSelection),InitialSpace=LocalSpace,bSelecta ...(truncated)"
    Tag Controls=true
    Zero Controls=true
    Parent=RigVMDispatchNode(Default=(Type=Bone,Name="None"), Name=Parent Control, NameSpace=Self) {template="DISPATCH_RigDispatch_GetModuleMetadata(in Name,in NameSpace,in Default,out Value,out Found)"}.Value
    Global Position.Translation=@Divide.Result
    ExecuteContext=@n6.A
    Settings=RigVMVariableNode() {variable=Body Control Settings}.Value
   )
   > DISPATCH_RigDispatch_SetModuleMetadata = RigVMDispatchNode(Name=Body Control, NameSpace=Self, Value=@"Control Stack at Position".Control) {template="DISPATCH_RigDispatch_SetModuleMetadata(in Name,in NameSpace,in Value,out Success)"}
   > RigVMDispatchNode(Name=Body Control, NameSpace=Root, Value=@"Control Stack at Position".Control) {template="DISPATCH_RigDispatch_SetModuleMetadata(in Name,in NameSpace,in Value,out Success)"}
   > Set Offset Transform from Current_0_4 = RigVMFunctionReferenceNode(Control=@"Control Stack at Position".Control)
   n9 = &RigUnit_GetTransform(bInitial=true, Item=RigVMVariableNode() {variable=Pelvis Bone Connector}.Value)
   Interpolate = /Script/RigVM.RigVMFunction_MathVectorLerp(A=@Divide.Result, B=@n9.Transform.Translation, T=RigVMVariableNode() {variable=Body Control Adjustment}.Value) {is=struct RigVMUnitNode, template="Interpolate::Execute(in A,in B,in T,out Result)"}
   n10 = RigVMVariableNode() {variable=Body Control Adjustment}
   > Set Transform_1 = /Script/ControlRig.RigUnit_SetTranslation(
    Space=GlobalSpace
    Weight=1
    bInitial=true
    bPropagateToChildren=True
    Item=@"Control Stack at Position".Control
    Value=RigVMDispatchNode(
     False=@Interpolate.Result
     True=&RigVMFunction_MathVectorLerp(B=@Divide.Result, A=@n9.Transform.Translation, T=@n10.Value) {template="Interpolate::Execute(in A,in B,in T,out Result)"}.Result
     Condition=RigVMVariableNode() {variable=Hips at Pelvis}.Value
    ) {template="DISPATCH_RigVMDispatch_If(in Condition,in True,in False,out Result)"}.Result
   ) {is=struct RigVMUnitNode, template="Set Transform::Execute(in Item,in Space,in bInitial,in Value,in Weight,in bPropagateToChildren)"}
   RigVMFunction_Sequence_1 = RigVMAggregateNode(ExecuteContext=@n6.B)
   n11 = RigVMDispatchNode(Default=(), Name=Spine Rig Bones, NameSpace=Self) {template="DISPATCH_RigDispatch_GetModuleMetadata(in Name,in NameSpace,in Default,out Value,out Found)"}
   Control Stack at Item_1_1_1 = RigVMFunctionReferenceNode(
    Add Bottom Null=False
    Add Null=False
    Add Secondary Control=False
    Control Suffix=_ctrl
    Name=hips
    Null Suffix=_null
    Orient to World=true
    Secondary Settings="(DisplayName=\"None\",Shape=(bVisible=True,Name=\"Default\",Color=(R=1,G=0,B=0,A=1),Transform=(Rotation=(X=0,Y=0,Z=0,W=1),Translation=(X=0,Y=0,Z=0),Scale3D=(X=1,Y=1,Z=1))),Proxy=(bIsProxy=False,ShapeVisibility=BasedOnSelection),InitialSpace=LocalSpace,bSelecta ...(truncated)"
    Tag Controls=true
    Zero Controls=true
    Parent=RigVMDispatchNode(Default=(Type=Bone,Name="None"), Name=Body Control, NameSpace=Self) {template="DISPATCH_RigDispatch_GetModuleMetadata(in Name,in NameSpace,in Default,out Value,out Found)"}.Value
    Item=RigVMDispatchNode(
     False=RigVMDispatchNode(Index=0, Array=@n11.Value) {template="DISPATCH_RigVMDispatch_ArrayGetAtIndex(in Array,in Index,out Element)"}.Element
     True=RigVMDispatchNode(Default=(Type=Bone,Name="None"), Name=Pelvis Bone, NameSpace=Self) {template="DISPATCH_RigDispatch_GetModuleMetadata(in Name,in NameSpace,in Default,out Value,out Found)"}.Value
     Condition=@n1.Hips at Pelvis
    ) {template="DISPATCH_RigVMDispatch_If(in Condition,in True,in False,out Result)"}.Result
    ExecuteContext=@RigVMFunction_Sequence_1.B
    Settings=RigVMVariableNode() {variable=Hips Control Settings}.Value
   )
   > Set Offset Transform from Current_0_2 = RigVMFunctionReferenceNode(Control=@"Control Stack at Item_1_1_1".Control)
   > RigVMDispatchNode(Name=Hips Control, NameSpace=Self, Value=@"Control Stack at Item_1_1_1".Control) {template="DISPATCH_RigDispatch_SetModuleMetadata(in Name,in NameSpace,in Value,out Success)"}
   Control Stack at Item_1_1_1_1_1 = RigVMFunctionReferenceNode(
    Add Bottom Null=false
    Add Null=False
    Add Secondary Control=False
    Control Suffix=_ctrl
    Name=chest
    Null Suffix=_null
    Orient to World=true
    Secondary Settings="(DisplayName=\"None\",Shape=(bVisible=True,Name=\"Default\",Color=(R=1,G=0,B=0,A=1),Transform=(Rotation=(X=0,Y=0,Z=0,W=1),Translation=(X=0,Y=0,Z=0),Scale3D=(X=1,Y=1,Z=1))),Proxy=(bIsProxy=False,ShapeVisibility=BasedOnSelection),InitialSpace=LocalSpace,bSelecta ...(truncated)"
    Tag Controls=true
    Zero Controls=true
    Item=RigVMDispatchNode(
     Index=-1
     Array=RigVMDispatchNode(Default=(), Name=Spine Rig Bones, NameSpace=Self) {template="DISPATCH_RigDispatch_GetModuleMetadata(in Name,in NameSpace,in Default,out Value,out Found)"}.Value
    ) {template="DISPATCH_RigVMDispatch_ArrayGetAtIndex(in Array,in Index,out Element)"}.Element
    Parent=&RigUnit_ResolveConnector(Connector=(Type=Connector,Name="Parent Control"), SkipSocket=False).Result
    ExecuteContext=@RigVMFunction_Sequence_1.C
    Settings=RigVMVariableNode() {variable=Chest Control Settings}.Value
   )
   > Set Offset Transform from Current_0_3 = RigVMFunctionReferenceNode(Control=@"Control Stack at Item_1_1_1_1_1".Control)
   > RigVMDispatchNode(Name=Chest Control, NameSpace=Self, Value=@"Control Stack at Item_1_1_1_1_1".Control) {template="DISPATCH_RigDispatch_SetModuleMetadata(in Name,in NameSpace,in Value,out Success)"}
   n12 = RigVMDispatchNode(Default=(), Name=Spine Rig Bones, NameSpace=Self) {template="DISPATCH_RigDispatch_GetModuleMetadata(in Name,in NameSpace,in Default,out Value,out Found)"}
   > n13 = &RigUnit_HierarchyAddNull(
    Name=chest_space_null
    Space=LocalSpace
    Transform=&RigUnit_GetTransform(
     bInitial=False
     Item=RigVMDispatchNode(Index=-1, Array=@n12.Value) {template="DISPATCH_RigVMDispatch_ArrayGetAtIndex(in Array,in Index,out Element)"}.Element
    ).Transform
    Parent=&RigUnit_ResolveConnector(Connector=(Type=Connector,Name="Parent Control"), SkipSocket=False).Result
   )
   > RigVMDispatchNode(Name=Chest Space Null, NameSpace=Root, Value=@n13.Item) {template="DISPATCH_RigDispatch_SetModuleMetadata(in Name,in NameSpace,in Value,out Success)"}
   n14 = &RigVMFunction_Sequence(ExecuteContext=@RigVMFunction_Sequence_1.D)
   Generate Spline Points = RigVMFunctionReferenceNode(
    Items=RigVMDispatchNode(Default=(), Name=Spine Rig Bones, NameSpace=Self) {template="DISPATCH_RigDispatch_GetModuleMetadata(in Name,in NameSpace,in Default,out Value,out Found)"}.Value
    ExecuteContext=@n14.A
   )
   > RigVMVariableNode(Value=@"Generate Spline Points".Value) {variable=Curve Points}
   n15 = &RigVMFunction_Sequence(ExecuteContext=@n14.B)
   DISPATCH_RigVMDispatch_ArrayReset = RigVMDispatchNode(ExecuteContext=@n15.A, Array=RigVMVariableNode() {variable=Spine Offsets Local}.Value) {template="DISPATCH_RigVMDispatch_ArrayReset(io Array)"}
   > n16 = RigVMDispatchNode(Array=RigVMDispatchNode(Default=(), Name=Spine Rig Bones, NameSpace=Self) {template="DISPATCH_RigDispatch_GetModuleMetadata(in Name,in NameSpace,in Default,out Value,out Found)"}.Value) {template="DISPATCH_RigVMDispatch_ArrayIterator(in Array,out Element,out Index,out Count,out Ratio)"}
   n17 = &RigUnit_ControlRigSplineFromPoints(SamplesPerSegment=16, SplineMode=BSpline, Points=RigVMVariableNode() {variable=Curve Points}.Value)
   > DISPATCH_RigVMDispatch_ArrayAdd = RigVMDispatchNode(
    Element=&RigVMFunction_MathTransformMakeRelative(
     Global=&RigUnit_GetTransform(bInitial=False, Item=@n16.Element).Transform
     Parent=/Script/ControlRigSpline.RigUnit_TransformFromControlRigSpline(Roll=0, UpVector=(X=0,Y=1,Z=0), Spline=@n17.Spline, U=@n16.Ratio) {is=struct RigVMUnitNode}.Transform
    ).Local
    Array=RigVMVariableNode() {variable=Spine Offsets Local}.Value
   ) {template="DISPATCH_RigVMDispatch_ArrayAdd(io Array,in Element,out Index)"}
   RigVMDispatchNode(Name=Spine Offsets, NameSpace=Self, ExecuteContext=@n16.Completed, Value=RigVMVariableNode() {variable=Spine Offsets Local}.Value) {template="DISPATCH_RigDispatch_SetModuleMetadata(in Name,in NameSpace,in Value,out Success)"}
   n18 = RigVMVariableNode() {variable=Curve Points}
   n19 = RigVMDispatchNode(Index=1, Array=@n18.Value) {template="DISPATCH_RigVMDispatch_ArrayGetAtIndex(in Array,in Index,out Element)"}
   n20 = RigVMDispatchNode(Index=2, Array=@n18.Value) {template="DISPATCH_RigVMDispatch_ArrayGetAtIndex(in Array,in Index,out Element)"}
   DrawTransforms = RigVMFunctionReferenceNode(
    Color=(R=1,G=0,B=0,A=1)
    Mode=Axes
    Scale=10
    Thickness=1
    Transforms="((Rotation=(X=0,Y=0,Z=0,W=1),Translation=(X=0,Y=0,Z=0),Scale3D=(X=1,Y=1,Z=1)),(Rotation=(X=0,Y=0,Z=0,W=1),Translation=(X=0,Y=0,Z=0),Scale3D=(X=1,Y=1,Z=1)),(Rotation=(X=0,Y=0,Z=0,W=1),Translation=(X=0,Y=0,Z=0),Scale3D=(X=1,Y=1,Z=1)),(Rotation=(X=0,Y=0,Z=0,W ...(truncated)"
    Transforms.0.Translation=RigVMDispatchNode(Index=0, Array=@n18.Value) {template="DISPATCH_RigVMDispatch_ArrayGetAtIndex(in Array,in Index,out Element)"}.Element
    Transforms.1.Translation=@n19.Element
    Transforms.2.Translation=@n20.Element
    Transforms.3.Translation=RigVMDispatchNode(Index=3, Array=@n18.Value) {template="DISPATCH_RigVMDispatch_ArrayGetAtIndex(in Array,in Index,out Element)"}.Element
    ExecuteContext=@n15.B
    Active=RigVMVariableNode() {variable=Debug}.Value
   )
   > Control Stack at Position_1 = RigVMFunctionReferenceNode(
    Add Bottom Null=False
    Add Secondary Control=False
    Control Suffix=_ctrl
    Global Position=(Rotation=(X=0,Y=0,Z=0,W=1),Translation=(X=0,Y=0,Z=0),Scale3D=(X=1,Y=1,Z=1))
    Name=hips_tan
    Null Suffix=_null
    Secondary Settings="(DisplayName=\"None\",Shape=(bVisible=True,Name=\"Default\",Color=(R=1,G=0,B=0,A=1),Transform=(Rotation=(X=0,Y=0,Z=0,W=1),Translation=(X=0,Y=0,Z=0),Scale3D=(X=1,Y=1,Z=1))),Proxy=(bIsProxy=False,ShapeVisibility=BasedOnSelection),InitialSpace=LocalSpace,bUsePref ...(truncated)"
    Tag Controls=true
    Zero Controls=true
    Settings=@n1.Mid Control Settings
    Parent=RigVMDispatchNode(
     False=RigVMDispatchNode(Default=(Type=Bone,Name="None"), Name=Hips Control, NameSpace=Self) {template="DISPATCH_RigDispatch_GetModuleMetadata(in Name,in NameSpace,in Default,out Value,out Found)"}.Value
     True=RigVMDispatchNode(Default=(Type=Bone,Name="None"), Name=Body Control, NameSpace=Self) {template="DISPATCH_RigDispatch_GetModuleMetadata(in Name,in NameSpace,in Default,out Value,out Found)"}.Value
     Condition=RigVMVariableNode() {variable=Even IK Middle Controls}.Value
    ) {template="DISPATCH_RigVMDispatch_If(in Condition,in True,in False,out Result)"}.Result
    Add Null=RigVMDispatchNode(False=False, True=true, Condition=RigVMVariableNode() {variable=Even IK Middle Controls}.Value) {template="DISPATCH_RigVMDispatch_If(in Condition,in True,in False,out Result)"}.Result
    Global Position.Translation=@n19.Element
   )
   > RigVMDispatchNode(Name=Hips Tan Control, NameSpace=Self, Value=@"Control Stack at Position_1".Control) {template="DISPATCH_RigDispatch_SetModuleMetadata(in Name,in NameSpace,in Value,out Success)"}
   > Set Offset Transform from Current_0_1 = RigVMFunctionReferenceNode(Control=@"Control Stack at Position_1".Control)
   > Control Stack at Position_1_1 = RigVMFunctionReferenceNode(
    Add Bottom Null=False
    Add Secondary Control=False
    Control Suffix=_ctrl
    Global Position=(Rotation=(X=0,Y=0,Z=0,W=1),Translation=(X=0,Y=0,Z=0),Scale3D=(X=1,Y=1,Z=1))
    Name=chest_tan
    Null Suffix=_null
    Secondary Settings="(DisplayName=\"None\",Shape=(bVisible=True,Name=\"Default\",Color=(R=1,G=0,B=0,A=1),Transform=(Rotation=(X=0,Y=0,Z=0,W=1),Translation=(X=0,Y=0,Z=0),Scale3D=(X=1,Y=1,Z=1))),Proxy=(bIsProxy=False,ShapeVisibility=BasedOnSelection),InitialSpace=LocalSpace,bUsePref ...(truncated)"
    Tag Controls=true
    Zero Controls=true
    Parent=RigVMDispatchNode(
     True=RigVMDispatchNode(Default=(Type=Bone,Name="None"), Name=Body Control, NameSpace=Self) {template="DISPATCH_RigDispatch_GetModuleMetadata(in Name,in NameSpace,in Default,out Value,out Found)"}.Value
     False=RigVMDispatchNode(Default=(Type=Bone,Name="None"), Name=Chest Control, NameSpace=Self) {template="DISPATCH_RigDispatch_GetModuleMetadata(in Name,in NameSpace,in Default,out Value,out Found)"}.Value
     Condition=RigVMVariableNode() {variable=Even IK Middle Controls}.Value
    ) {template="DISPATCH_RigVMDispatch_If(in Condition,in True,in False,out Result)"}.Result
    Add Null=RigVMDispatchNode(False=False, True=true, Condition=RigVMVariableNode() {variable=Even IK Middle Controls}.Value) {template="DISPATCH_RigVMDispatch_If(in Condition,in True,in False,out Result)"}.Result
    Settings=RigVMVariableNode() {variable=Mid Control Settings}.Value
    Global Position.Translation=@n20.Element
   )
   > RigVMDispatchNode(Name=Chest Tan Control, NameSpace=Self, Value=@"Control Stack at Position_1_1".Control) {template="DISPATCH_RigDispatch_SetModuleMetadata(in Name,in NameSpace,in Value,out Success)"}
   > Set Offset Transform from Current_0 = RigVMFunctionReferenceNode(Control=@"Control Stack at Position_1_1".Control)
   > Tag IK Spine Controls = RigVMFunctionReferenceNode(
    Item=((Type=None,Name="None"),(Type=None,Name="None"),(Type=None,Name="None"),(Type=None,Name="None"))
    Item.0=RigVMDispatchNode(Default=(Type=Bone,Name="None"), Name=Hips Control, NameSpace=Self) {template="DISPATCH_RigDispatch_GetModuleMetadata(in Name,in NameSpace,in Default,out Value,out Found)"}.Value
    Item.3=RigVMDispatchNode(Default=(Type=Bone,Name="None"), Name=Chest Control, NameSpace=Self) {template="DISPATCH_RigDispatch_GetModuleMetadata(in Name,in NameSpace,in Default,out Value,out Found)"}.Value
    Item.2=RigVMDispatchNode(Default=(Type=Bone,Name="None"), Name=Chest Tan Control, NameSpace=Self) {template="DISPATCH_RigDispatch_GetModuleMetadata(in Name,in NameSpace,in Default,out Value,out Found)"}.Value
    Item.1=RigVMDispatchNode(Default=(Type=Bone,Name="None"), Name=Hips Tan Control, NameSpace=Self) {template="DISPATCH_RigDispatch_GetModuleMetadata(in Name,in NameSpace,in Default,out Value,out Found)"}.Value
   )
   n21 = RigVMDispatchNode(
    Array=RigVMDispatchNode(Default=(), Name=Spine Rig Bones, NameSpace=Self) {template="DISPATCH_RigDispatch_GetModuleMetadata(in Name,in NameSpace,in Default,out Value,out Found)"}.Value
    ExecuteContext=@RigVMFunction_Sequence_1.E
   ) {template="DISPATCH_RigVMDispatch_ArrayIterator(in Array,out Element,out Index,out Count,out Ratio)"}
   Add = &RigVMFunction_MathIntAdd(B=1, A=@n21.Index) {template="Add::Execute(in A,in B,out Result)"}
   Int to Name = /Script/RigVM.RigVMFunction_MathIntToName(PaddedSize=2, Number=@Add.Result) {is=struct RigVMUnitNode, template="Int to Name::Execute(in Number,in PaddedSize,out Result)"}
   Int to Name_1 = &RigVMFunction_MathIntToName(PaddedSize=2, Number=@n21.Index) {template="Int to Name::Execute(in Number,in PaddedSize,out Result)"}
   Concat_5_1 = RigVMAggregateNode(A=spine_, C=_ctrl, B=@"Int to Name_1".Result)
   > Control Stack at Item = RigVMFunctionReferenceNode(
    Add Bottom Null=False
    Add Null=False
    Add Secondary Control=False
    Control Suffix=_ctrl
    Null Suffix=_null
    Orient to World=true
    Secondary Settings="(DisplayName=\"None\",Shape=(bVisible=True,Name=\"Default\",Color=(R=1,G=0,B=0,A=1),Transform=(Rotation=(X=0,Y=0,Z=0,W=1),Translation=(X=0,Y=0,Z=0),Scale3D=(X=1,Y=1,Z=1))),Proxy=(bIsProxy=False,ShapeVisibility=BasedOnSelection),InitialSpace=LocalSpace,bUsePref ...(truncated)"
    Tag Controls=true
    Zero Controls=false
    Name=&RigVMFunction_NameConcat(A=spine_, B=@"Int to Name".Result) {template="Concat::Execute(in A,in B,out Result)"}.Result
    Item=@n21.Element
    Parent=RigVMDispatchNode(
     False=(Type=Control,Name="None")
     False.Name=@Concat_5_1.Result
     True=RigVMDispatchNode(Default=(Type=Bone,Name="None"), Name=Body Control, NameSpace=Self) {template="DISPATCH_RigDispatch_GetModuleMetadata(in Name,in NameSpace,in Default,out Value,out Found)"}.Value
     Condition=RigVMDispatchNode(B=0, A=@n21.Index) {template="DISPATCH_RigVMDispatch_CoreEquals(in A,in B,out Result)"}.Result
    ) {template="DISPATCH_RigVMDispatch_If(in Condition,in True,in False,out Result)"}.Result
    Settings=RigVMVariableNode() {variable=FK Control Settings}.Value
   )
   > &RigUnit_SetMetadataTag(NameSpace=Self, Tag=FK Spine Controls, Item=@"Control Stack at Item".Control)
   > RigVMDispatchNode(Name=FK Parent, NameSpace=Self, Value=@"Control Stack at Item".Control) {template="DISPATCH_RigDispatch_SetModuleMetadata(in Name,in NameSpace,in Value,out Success)"}
   SpawnAnimationChannel = /Script/ControlRig.RigUnit_HierarchyAddAnimationChannelBool(
    LimitsEnabled=()
    MaximumValue=True
    MinimumValue=false
    Name=spine_fk_ik_switch
    Parent=&RigUnit_ResolveConnector(Connector=(Type=Connector,Name="Settings Control"), SkipSocket=true).Result
    ExecutePin=@RigVMFunction_Sequence.H
    InitialValue=RigVMVariableNode() {variable=Default FK or IK}.Value
   ) {
    is=struct RigVMUnitNode
    template="SpawnAnimationChannel::Execute(in InitialValue,in MinimumValue,in MaximumValue,in LimitsEnabled,in Parent,in Name,out Item)"
   }
   > RigVMDispatchNode(Name=IK FK Switch, NameSpace=Self, Value=@SpawnAnimationChannel.Item) {template="DISPATCH_RigDispatch_SetModuleMetadata(in Name,in NameSpace,in Value,out Success)"}
   RigVMFunction_Sequence_6 = RigVMAggregateNode(ExecuteContext=@RigVMFunction_Sequence.I)
   Create Socket at Metadata = RigVMFunctionReferenceNode(
    Color=(R=1,G=1,B=1,A=1)
    Description=Place the shoulder module here
    Extra Transform=(Rotation=(X=0,Y=0,Z=0,W=1),Translation=(X=0,Y=0,Z=0),Scale3D=(X=1,Y=1,Z=1))
    Metadata Name=shoulder_l
    Name=shoulder_l_socket
    ExecuteContext=@RigVMFunction_Sequence_6.A
   )
   n22 = &RigVMFunction_ControlFlowBranch(ExecuteContext=@"Create Socket at Metadata".ExecuteContext, Condition=@"Create Socket at Metadata".Found)
   n23 = RigVMDispatchNode(Default=_l, Name=Left Side Suffix, NameSpace=Root) {template="DISPATCH_RigDispatch_GetModuleMetadata(in Name,in NameSpace,in Default,out Value,out Found)"}
   Concat = &RigVMFunction_NameConcat(A=upperarm, B=@n23.Value) {template="Concat::Execute(in A,in B,out Result)"}
   n24 = RigVMDispatchNode(
    Index=-1
    Array=&RigUnit_HierarchyGetParentsItemArray(
     bDefaultParent=True
     bIncludeChild=true
     bReverse=False
     Child=&RigUnit_ResolveConnector(Connector=(Type=Connector,Name="Spine Start Bone"), SkipSocket=true).Result
    ).Parents
   ) {template="DISPATCH_RigVMDispatch_ArrayGetAtIndex(in Array,in Index,out Element)"}
   Find first in Hierarchy_2_1 = RigVMFunctionReferenceNode(
    Children=true
    Ignore=("corrective")
    Search for=("None")
    Type to Search=Bone
    Search for.0=@Concat.Result
    Item=@n24.Element
    ExecuteContext=@n22.False
   )
   > &RigUnit_HierarchyAddSocket(
    Color=(R=1,G=1,B=1,A=1)
    Description=Place the shoulder module here
    Name=shoulder_l
    Space=LocalSpace
    Transform=(Rotation=(X=0,Y=0,Z=0,W=1),Translation=(X=0,Y=0,Z=0),Scale3D=(X=1,Y=1,Z=1))
    Parent=@"Find first in Hierarchy_2_1".Item Found
   )
   Create Socket at Metadata_1 = RigVMFunctionReferenceNode(
    Color=(R=1,G=1,B=1,A=1)
    Description=Place the shoulder module here
    Extra Transform=(Rotation=(X=0,Y=0,Z=0,W=1),Translation=(X=0,Y=0,Z=0),Scale3D=(X=1,Y=1,Z=1))
    Metadata Name=shoulder_r
    Name=shoulder_r_socket
    ExecuteContext=@RigVMFunction_Sequence_6.B
   )
   n25 = &RigVMFunction_ControlFlowBranch(ExecuteContext=@"Create Socket at Metadata_1".ExecuteContext, Condition=@"Create Socket at Metadata_1".Found)
   n26 = RigVMDispatchNode(Default=_r, Name=Right Side Suffix, NameSpace=Root) {template="DISPATCH_RigDispatch_GetModuleMetadata(in Name,in NameSpace,in Default,out Value,out Found)"}
   Find first in Hierarchy_2 = RigVMFunctionReferenceNode(
    Children=true
    Ignore=("corrective")
    Search for=("None")
    Type to Search=Bone
    Search for.0=&RigVMFunction_NameConcat(A=upperarm, B=@n26.Value) {template="Concat::Execute(in A,in B,out Result)"}.Result
    Item=@n24.Element
    ExecuteContext=@n25.False
   )
   > &RigUnit_HierarchyAddSocket(
    Color=(R=1,G=1,B=1,A=1)
    Description=Place the shoulder module here
    Name=shoulder_r
    Space=LocalSpace
    Transform=(Rotation=(X=0,Y=0,Z=0,W=1),Translation=(X=0,Y=0,Z=0),Scale3D=(X=1,Y=1,Z=1))
    Parent=@"Find first in Hierarchy_2".Item Found
   )
   Create Socket at Metadata_2 = RigVMFunctionReferenceNode(
    Color=(R=1,G=1,B=1,A=1)
    Description=Place the shoulder module here
    Extra Transform=(Rotation=(X=0,Y=0,Z=0,W=1),Translation=(X=0,Y=0,Z=0),Scale3D=(X=1,Y=1,Z=1))
    Metadata Name=thigh_l
    Name=thigh_l_socket
    ExecuteContext=@RigVMFunction_Sequence_6.C
   )
   n27 = &RigVMFunction_ControlFlowBranch(ExecuteContext=@"Create Socket at Metadata_2".ExecuteContext, Condition=@"Create Socket at Metadata_2".Found)
   Find first in Hierarchy_5_1_1 = RigVMFunctionReferenceNode(
    Children=true
    Ignore=("corrective","ik","twist")
    Search for=("neck")
    Type to Search=Bone
    Search for.0=&RigVMFunction_NameConcat(A=thigh, B=@n23.Value) {template="Concat::Execute(in A,in B,out Result)"}.Result
    Item=@n24.Element
    ExecuteContext=@n27.False
   )
   > &RigUnit_HierarchyAddSocket(
    Color=(R=1,G=1,B=1,A=1)
    Description=Place the leg module here
    Name=thigh_l_socket
    Space=LocalSpace
    Transform=(Rotation=(X=0,Y=0,Z=0,W=1),Translation=(X=0,Y=0,Z=0),Scale3D=(X=1,Y=1,Z=1))
    Parent=@"Find first in Hierarchy_5_1_1".Item Found
   )
   Create Socket at Metadata_1_2 = RigVMFunctionReferenceNode(
    Color=(R=1,G=1,B=1,A=1)
    Description=Place the shoulder module here
    Extra Transform=(Rotation=(X=0,Y=0,Z=0,W=1),Translation=(X=0,Y=0,Z=0),Scale3D=(X=1,Y=1,Z=1))
    Metadata Name=thigh_r
    Name=thigh_r_socket
    ExecuteContext=@RigVMFunction_Sequence_6.D
   )
   n28 = &RigVMFunction_ControlFlowBranch(ExecuteContext=@"Create Socket at Metadata_1_2".ExecuteContext, Condition=@"Create Socket at Metadata_1_2".Found)
   Find first in Hierarchy_5_1 = RigVMFunctionReferenceNode(
    Children=true
    Ignore=("corrective","ik","twist")
    Search for=("neck")
    Type to Search=Bone
    Search for.0=&RigVMFunction_NameConcat(A=thigh, B=@n26.Value) {template="Concat::Execute(in A,in B,out Result)"}.Result
    Item=@n24.Element
    ExecuteContext=@n28.False
   )
   > &RigUnit_HierarchyAddSocket(
    Color=(R=1,G=1,B=1,A=1)
    Description=Place the leg module here
    Name=thigh_r_socket
    Space=LocalSpace
    Transform=(Rotation=(X=0,Y=0,Z=0,W=1),Translation=(X=0,Y=0,Z=0),Scale3D=(X=1,Y=1,Z=1))
    Parent=@"Find first in Hierarchy_5_1".Item Found
   )
   Create Socket at Metadata_1_1 = RigVMFunctionReferenceNode(
    Color=(R=1,G=1,B=1,A=1)
    Description=Place the neck module here
    Extra Transform=(Rotation=(X=0,Y=0,Z=0,W=1),Translation=(X=0,Y=0,Z=0),Scale3D=(X=1,Y=1,Z=1))
    Metadata Name=neck
    Name=neck_socket
    ExecuteContext=@RigVMFunction_Sequence_6.E
   )
   n29 = &RigVMFunction_ControlFlowBranch(ExecuteContext=@"Create Socket at Metadata_1_1".ExecuteContext, Condition=@"Create Socket at Metadata_1_1".Found)
   Find first in Hierarchy_5 = RigVMFunctionReferenceNode(Children=true, Ignore=("corrective","ik","twist"), Search for=("neck"), Type to Search=Bone, Item=@n24.Element, ExecuteContext=@n29.False)
   > &RigUnit_HierarchyAddSocket(
    Color=(R=1,G=1,B=1,A=1)
    Description=Place the neck module here
    Name=neck_socket
    Space=LocalSpace
    Transform=(Rotation=(X=0,Y=0,Z=0,W=1),Translation=(X=0,Y=0,Z=0),Scale3D=(X=1,Y=1,Z=1))
    Parent=@"Find first in Hierarchy_5".Item Found
   )
   Get_Module_Metadata = RigVMDispatchNode(Default=(), Name=Spine Rig Bones, NameSpace=Self) {template="DISPATCH_RigDispatch_GetModuleMetadata(in Name,in NameSpace,in Default,out Value,out Found)"}
   Create Dual Spline for Spine = RigVMFunctionReferenceNode(
    Controls=((Type=None,Name="None"),(Type=None,Name="None"),(Type=None,Name="None"),(Type=None,Name="None"))
    Samples Per Segment=16
    Spline Mode=BSpline
    Stretch=true
    debug=false
    Controls.0=RigVMDispatchNode(Default=(Type=Bone,Name="None"), Name=Hips Control, NameSpace=Self) {template="DISPATCH_RigDispatch_GetModuleMetadata(in Name,in NameSpace,in Default,out Value,out Found)"}.Value
    Controls.2=RigVMDispatchNode(Default=(Type=Bone,Name="None"), Name=Chest Tan Control, NameSpace=Self) {template="DISPATCH_RigDispatch_GetModuleMetadata(in Name,in NameSpace,in Default,out Value,out Found)"}.Value
    Controls.1=RigVMDispatchNode(Default=(Type=Bone,Name="None"), Name=Hips Tan Control, NameSpace=Self) {template="DISPATCH_RigDispatch_GetModuleMetadata(in Name,in NameSpace,in Default,out Value,out Found)"}.Value
    Controls.3=RigVMDispatchNode(Default=(Type=Bone,Name="None"), Name=Chest Control, NameSpace=Self) {template="DISPATCH_RigDispatch_GetModuleMetadata(in Name,in NameSpace,in Default,out Value,out Found)"}.Value
    Primary Axis=@n1.Primary Axis
    Secondary Spline Direction=@n1.Secondary Spline Direction
    Up Axis=@n1.Up Axis
    Bones=@Get_Module_Metadata.Value
    ExecuteContext=@RigVMFunction_Sequence.K
   )
   > RigVMVariableNode(Value=@"Create Dual Spline for Spine".Local Spline) {variable=Local Spline}
   > RigVMVariableNode(Value=@"Create Dual Spline for Spine".Template Spline) {variable=Template Spline}
   > Set_Module_Metadata_2 = RigVMDispatchNode(Name=Original Positions, NameSpace=Self, Value=@"Create Dual Spline for Spine".Orig Positions) {template="DISPATCH_RigDispatch_SetModuleMetadata(in Name,in NameSpace,in Value,out Success)"}
   > Set_Module_Metadata_2_1 = RigVMDispatchNode(Name=New Positions, NameSpace=Self, Value=@"Create Dual Spline for Spine".Positions) {template="DISPATCH_RigDispatch_SetModuleMetadata(in Name,in NameSpace,in Value,out Success)"}
   RigVMFunctionReturnNode(
    ExecuteContext=@RigVMFunction_Sequence_3.C
    Local Spline Curve=RigVMVariableNode() {variable=Local Spline}.Value
    Template Spline Curve=RigVMVariableNode() {variable=Template Spline}.Value
    Dummy Bones=RigVMVariableNode() {variable=Phantom Bones}.Value
   )
   RigVMDispatchNode(Default=(Type=Bone,Name="None"), Name=Chest Bone, NameSpace=Self) {template="DISPATCH_RigDispatch_GetModuleMetadata(in Name,in NameSpace,in Default,out Value,out Found)"}
   RigVMDispatchNode(Default=(), Name=Spine Rig Bones, NameSpace=Self) {template="DISPATCH_RigDispatch_GetModuleMetadata(in Name,in NameSpace,in Default,out Value,out Found)"}
   RigVMDispatchNode(Default=(), Name=Spine Rig Bones, NameSpace=Self) {template="DISPATCH_RigDispatch_GetModuleMetadata(in Name,in NameSpace,in Default,out Value,out Found)"}
   RigVMDispatchNode(Default=(), Name=Spine Bones, NameSpace=Self) {template="DISPATCH_RigDispatch_GetModuleMetadata(in Name,in NameSpace,in Default,out Value,out Found)"}
   n30 = RigVMDispatchNode(Default=(), Name=Spine Rig Bones, NameSpace=Self) {template="DISPATCH_RigDispatch_GetModuleMetadata(in Name,in NameSpace,in Default,out Value,out Found)"}
   Control Stack at Position_2 = RigVMFunctionReferenceNode(
    Add Bottom Null=False
    Add Null=False
    Add Secondary Control=False
    Control Suffix=_ctrl
    ExecuteContext=()
    Name=chestend
    Null Suffix=_null
    Secondary Settings="(Shape=(bVisible=True,Name=\"Default\",Color=(R=1,G=0,B=0,A=1),Transform=(Rotation=(X=0,Y=0,Z=0,W=1),Translation=(X=0,Y=0,Z=0),Scale3D=(X=1,Y=1,Z=1))),Proxy=(bIsProxy=False,ShapeVisibility=BasedOnSelection),DisplayName=\"None\",InitialSpace=LocalSpace,bUsePref ...(truncated)"
    Settings="(Shape=(bVisible=True,Name=\"Box_Solid\",Color=(R=1,G=0,B=0,A=1),Transform=(Rotation=(X=0,Y=0,Z=0,W=1),Translation=(X=0,Y=0,Z=0),Scale3D=(X=0.2,Y=0.2,Z=0.2))),Proxy=(bIsProxy=False,ShapeVisibility=BasedOnSelection),DisplayName=\"None\",InitialSpace=LocalSpace, ...(truncated)"
    Tag Controls=true
    Zero Controls=true
    Parent=@"Control Stack at Item_1_1_1_1_1".Control
    Global Position=&RigUnit_GetTransform(
     bInitial=False
     Item=RigVMDispatchNode(Index=-1, Array=@n30.Value) {template="DISPATCH_RigVMDispatch_ArrayGetAtIndex(in Array,in Index,out Element)"}.Element
    ).Transform
    Settings.Shape.bVisible=RigVMVariableNode() {variable=Spine End Control Vis Debug}.Value
   )
   RigVMDispatchNode(ExecuteContext=(), Name=Chest End Control, NameSpace=Self, Value=@"Control Stack at Position_2".Control) {template="DISPATCH_RigDispatch_SetModuleMetadata(in Name,in NameSpace,in Value,out Success)"}
   Finalize Controls = RigVMCollapseNode(ExecuteContext=(), Module Settings=@n1.Module Settings)
   &RigUnit_GetTransformItemArray(Space=GlobalSpace, bInitial=true, Items=@Get_Module_Metadata_2.Value)
   &RigUnit_GetTransform(
    bInitial=False
    Item=RigVMDispatchNode(Default=(Type=Bone,Name="None"), Name=Pelvis Bone, NameSpace=Self) {template="DISPATCH_RigDispatch_GetModuleMetadata(in Name,in NameSpace,in Default,out Value,out Found)"}.Value
   )
   RigVMVariableNode() {variable=Chest Control Settings}
  ]}
 New Function_ContainedGraph {nodes [
   n1 = RigVMFunctionEntryNode()
   n2 = &RigUnit_FindItemsWithMetadataTag(NameSpace=Self, Tag=Controls)
   > Scale Module Controls = RigVMFunctionReferenceNode(
    Global Scale=RigVMDispatchNode(Default=1, Name=Control Size, NameSpace=Root) {template="DISPATCH_RigDispatch_GetModuleMetadata(in Name,in NameSpace,in Default,out Value,out Found)"}.Value
    Local Scale=@n1.Module Settings.ControlSize_12_#1
    Controls=@n2.Items
   )
   > Color Module Controls = RigVMFunctionReferenceNode(
    CenterColor=@n1.Module Settings.CenterColor_9_#2
    LeftColor=@n1.Module Settings.LeftColor_3_#3
    Override=@n1.Module Settings.OverrideColor_48_#4
    RightColor=@n1.Module Settings.RightColor_5_#5
    Controls=@n2.Items
   )
   > RigVMFunctionReturnNode()
  ]}
 New Function_ContainedGraph {nodes [
   n1 = RigVMFunctionEntryNode()
   > For_Each_1 = RigVMDispatchNode(Array=@n1.Items) {template="DISPATCH_RigVMDispatch_ArrayIterator(in Array,out Element,out Index,out Count,out Ratio)"}
   n2 = &RigVMFunction_ControlFlowBranch(
    Condition=RigVMDispatchNode(
     A=@For_Each_1.Index
     B=&RigVMFunction_MathIntSub(B=1, A=@For_Each_1.Count) {template="Subtract::Execute(in A,in B,out Result)"}.Result
    ) {template="DISPATCH_RigVMDispatch_CoreEquals(in A,in B,out Result)"}.Result
    ExecuteContext=@For_Each_1.ExecuteContext
   )
   n3 = &RigUnit_GetTransform(bInitial=False, Item=@For_Each_1.Element)
   n4 = &RigUnit_GetTransform(
    bInitial=False
    Item=RigVMDispatchNode(
     Index=&RigVMFunction_MathIntAdd(B=1, A=@For_Each_1.Index) {template="Add::Execute(in A,in B,out Result)"}.Result
     Array=@n1.Items
    ) {template="DISPATCH_RigVMDispatch_ArrayGetAtIndex(in Array,in Index,out Element)"}.Element
   )
   RigVMFunction_DebugLineNoSpace = /Script/RigVM.RigVMFunction_DebugLineNoSpace(
    DebugDrawSettings=(DepthPriority=SDPG_Foreground,Lifetime=-1)
    Color=@n1.Color
    bEnabled=@n1.Enabled
    Thickness=@n1.Thickness
    WorldOffset=@n1.World Offset
    A=@n3.Transform.Translation
    B=@n4.Transform.Translation
    ExecutePin=@n2.False
   ) {is=struct RigVMUnitNode}
   RigVMFunctionReturnNode(ExecuteContext=@For_Each_1.Completed)
   Add = &RigVMFunction_MathVectorAdd(B=(X=0,Y=-30,Z=0), A=@n3.Transform.Translation) {template="Add::Execute(in A,in B,out Result)"}
   &RigVMFunction_MathVectorAdd(B=(X=0,Y=-30,Z=0), A=@n4.Transform.Translation) {template="Add::Execute(in A,in B,out Result)"}
  ]}
 RigVMFunctionLibrary {nodes [
   Add Null Above = RigVMCollapseNode(ExecuteContext=(), Item=(Name="None"))
   Color Module Controls = RigVMCollapseNode(CenterColor=(R=0,G=0,B=0,A=0), Controls=(), ExecuteContext=(), LeftColor=(R=0,G=0,B=0,A=0), Override=False, RightColor=(R=0,G=0,B=0,A=0))
   ColorizeControls = RigVMCollapseNode(
    CenterColor=(R=1,G=1,B=0,A=1)
    ExecuteContext=()
    LeftColor=(R=0,G=1,B=0,A=1)
    RightColor=(R=1,G=0,B=0,A=1)
    X Plane=true
    controls=()
    ignore_ctrls=()
   )
   Compute Spline Positions = RigVMCollapseNode(Array=(), ExecuteContext=())
   Compute Spline Positions_1 = RigVMCollapseNode(Array=(), ExecuteContext=())
   Control Stack at Item = RigVMCollapseNode(
    Add Bottom Null=False
    Add Null=False
    Add Secondary Control=False
    Control Suffix=_ctrl
    ExecuteContext=()
    Item=(Name="None")
    Null Suffix=_null
    Parent=(Name="None")
    Secondary Settings="(Shape=(bVisible=True,Name=\"Default\",Color=(R=1,G=0,B=0,A=1),Transform=(Rotation=(X=0,Y=0,Z=0,W=1),Translation=(X=0,Y=0,Z=0),Scale3D=(X=1,Y=1,Z=1))),Proxy=(bIsProxy=False,ShapeVisibility=BasedOnSelection),DisplayName=\"None\",bSelectable=True,InitialSpace=Lo ...(truncated)"
    Settings="(Shape=(bVisible=True,Name=\"Default\",Color=(R=1,G=0,B=0,A=1),Transform=(Rotation=(X=0,Y=0,Z=0,W=1),Translation=(X=0,Y=0,Z=0),Scale3D=(X=1,Y=1,Z=1))),Proxy=(bIsProxy=False,ShapeVisibility=BasedOnSelection),DisplayName=\"None\",bSelectable=True,InitialSpace=Lo ...(truncated)"
    Tag Controls=true
    Zero Controls=true
   )
   Control Stack at Position = RigVMCollapseNode(
    Add Bottom Null=False
    Add Null=False
    Add Secondary Control=False
    Control Suffix=_ctrl
    ExecuteContext=()
    Global Position=(Rotation=(X=0,Y=0,Z=0,W=1),Translation=(X=0,Y=0,Z=0),Scale3D=(X=1,Y=1,Z=1))
    Null Suffix=_null
    Parent=(Type=None,Name="None")
    Secondary Settings="(Shape=(bVisible=True,Name=\"Default\",Color=(R=1,G=0,B=0,A=1),Transform=(Rotation=(X=0,Y=0,Z=0,W=1),Translation=(X=0,Y=0,Z=0),Scale3D=(X=1,Y=1,Z=1))),Proxy=(bIsProxy=False,ShapeVisibility=BasedOnSelection),DisplayName=\"None\",bSelectable=True,InitialSpace=Lo ...(truncated)"
    Settings="(Shape=(bVisible=True,Name=\"Default\",Color=(R=1,G=0,B=0,A=1),Transform=(Rotation=(X=0,Y=0,Z=0,W=1),Translation=(X=0,Y=0,Z=0),Scale3D=(X=1,Y=1,Z=1))),Proxy=(bIsProxy=False,ShapeVisibility=BasedOnSelection),DisplayName=\"None\",bSelectable=True,InitialSpace=Lo ...(truncated)"
    Tag Controls=true
    Zero Controls=true
   )
   Create Bone Chain = RigVMCollapseNode(
    Debug=false
    Debug Scale=15
    Debug Thickness=0.5
    Enable Orient Bones=true
    ExecuteContext=()
    Kind=Direction
    Name=NewBone
    Parent=(Type=Bone,Name="None")
    Primary Axis=(X=1,Y=0,Z=0)
    Secondary Axis=(X=0,Y=1,Z=0)
    Transforms=()
   )
   Create Dual Spline = RigVMCollapseNode(
    Bones=()
    Controls=()
    ExecuteContext=()
    Primary Axis=(X=1,Y=0,Z=0)
    Samples Per Segment=16
    Secondary Spline Direction=(X=0,Y=5,Z=0)
    Spline Mode=BSpline
    Up Axis=(X=0,Y=1,Z=0)
   )
   Create Dual Spline for Spine = RigVMCollapseNode(
    Bones=()
    Controls=()
    ExecuteContext=()
    Primary Axis=(X=1,Y=0,Z=0)
    Samples Per Segment=16
    Secondary Spline Direction=(X=0,Y=5,Z=0)
    Spline Mode=BSpline
    Up Axis=(X=0,Y=1,Z=0)
   )
   Create Socket at Metadata = RigVMCollapseNode(
    Color=(R=1,G=1,B=1,A=1)
    Description=Place the shoulder module here
    ExecuteContext=()
    Extra Transform=(Rotation=(X=0,Y=0,Z=0,W=1),Translation=(X=0,Y=0,Z=0),Scale3D=(X=1,Y=1,Z=1))
    Metadata Name=None
    Name=shoulder_l
   )
   Draw Line Between Items = RigVMCollapseNode(
    Color=(R=1,G=1,B=0,A=1)
    Enabled=True
    ExecuteContext=()
    Items=()
    Thickness=0.15
    World Offset=(Rotation=(X=0,Y=0,Z=0,W=1),Translation=(X=0,Y=0,Z=0),Scale3D=(X=1,Y=1,Z=1))
   )
   DrawTransforms = RigVMCollapseNode(Active=true, Color=(R=1,G=0,B=0,A=1), ExecuteContext=(), Mode=Axes, Scale=10, Transforms=())
   Dual SplineIK = RigVMCollapseNode(
    Aim Spline=()
    Bones=()
    Controls=()
    ExecuteContext=()
    Primary Axis=(X=1,Y=0,Z=0)
    Samples Per Segment=16
    Secondary Spline Direction=(X=0,Y=5,Z=0)
    Spline Mode=BSpline
    Template Spline=()
    Up Axis=(X=0,Y=1,Z=0)
   )
   Dual SplineIK for Spine = RigVMCollapseNode(
    Aim Spline=()
    Bones=()
    Controls=()
    ExecuteContext=()
    Primary Axis=(X=1,Y=0,Z=0)
    Samples Per Segment=16
    Secondary Spline Direction=(X=0,Y=5,Z=0)
    Spline Mode=BSpline
    Template Spline=()
    Up Axis=(X=0,Y=1,Z=0)
   )
   FWD AnimalSpineLimb = RigVMCollapseNode(
    Aim Spline=()
    Attach Skeleton=False
    ExecuteContext=()
    FK IK Switch=False
    Input Bones=()
    Primary Axis=(X=1,Y=0,Z=0)
    Secondary Spline Direction=(X=0,Y=5,Z=0)
    Stretch=false
    Template Spline=()
    Up Axis=(X=0,Y=1,Z=0)
    debug=false
   )
   Find = RigVMCollapseNode(Element=foot, Value=None)
   Find first in Hierarchy = RigVMCollapseNode(Children=true, ExecuteContext=(), Ignore=(), Item=(Type=Bone,Name="None"), Search for=(), Type to Search=Bone)
   FindClosestBone = RigVMCollapseNode(Bones=(), ExecuteContext=(), Initial=true, Item=())
   Generate Spline Points = RigVMCollapseNode(ExecuteContext=(), Items=())
   Generate Spline Points_1 = RigVMCollapseNode(ExecuteContext=(), Items=())
   IK AnimalSpine = RigVMCollapseNode(
    Body Control Adjustment=1
    Body Control Settings="(DisplayName=\"None\",bSelectable=True,InitialSpace=LocalSpace,bUsePreferredRotationOrder=False,PreferredRotationOrder=YZX,Limits=(LimitTranslationX=(bMinimum=False,bMaximum=False),LimitTranslationY=(bMinimum=False,bMaximum=False),LimitTranslationZ=(bMinimum ...(truncated)"
    Chest Control Settings="(DisplayName=\"None\",bSelectable=True,InitialSpace=LocalSpace,bUsePreferredRotationOrder=False,PreferredRotationOrder=YZX,Limits=(LimitTranslationX=(bMinimum=False,bMaximum=False),LimitTranslationY=(bMinimum=False,bMaximum=False),LimitTranslationZ=(bMinimum ...(truncated)"
    Debug=false
    Default FK or IK=false
    ExecuteContext=()
    FK Control Settings="(DisplayName=\"None\",bSelectable=True,InitialSpace=LocalSpace,bUsePreferredRotationOrder=False,PreferredRotationOrder=YZX,Limits=(LimitTranslationX=(bMinimum=False,bMaximum=False),LimitTranslationY=(bMinimum=False,bMaximum=False),LimitTranslationZ=(bMinimum ...(truncated)"
    Hips Control Settings="(DisplayName=\"None\",bSelectable=True,InitialSpace=LocalSpace,bUsePreferredRotationOrder=False,PreferredRotationOrder=YZX,Limits=(LimitTranslationX=(bMinimum=False,bMaximum=False),LimitTranslationY=(bMinimum=False,bMaximum=False),LimitTranslationZ=(bMinimum ...(truncated)"
    Hips at Pelvis=False
    Mid Control Settings="(DisplayName=\"None\",bSelectable=True,InitialSpace=LocalSpace,bUsePreferredRotationOrder=False,PreferredRotationOrder=YZX,Limits=(LimitTranslationX=(bMinimum=False,bMaximum=False),LimitTranslationY=(bMinimum=False,bMaximum=False),LimitTranslationZ=(bMinimum ...(truncated)"
    Module Settings="(OverrideColor_48_#4=False,LeftColor_3_#3=(R=0,G=0.0231,B=0.4115,A=1),RightColor_5_#5=(R=1,G=0,B=0,A=1),CenterColor_9_#2=(R=1,G=1,B=0,A= ...(truncated)"
    Number of Spine Bones=8
    Pelvis Bone Connector=(Type=Bone,Name="None")
    Primary Axis=(X=1,Y=0,Z=0)
    Secondary Spline Direction=(X=0,Y=5,Z=0)
    Spine Bone End Connector=(Type=Bone,Name="None")
    Spine Bone Start Connector=(Type=Bone,Name="None")
    Up Axis=(X=0,Y=1,Z=0)
   )
   Orient Bones = RigVMCollapseNode(Bones=(), Debug=false, ExecuteContext=(), Kind=Direction, Primary Axis=(X=1,Y=0,Z=0), Secondary Axis=(X=0,Y=1,Z=0))
   Parent = RigVMCollapseNode(Child=(Type=Bone,Name="None"), ExecuteContext=(), New Parent=(Type=Bone,Name="None"))
   Project to New Parents = RigVMCollapseNode(Children=(), ExecuteContext=(), Parents=(), Propagate to Children=True)
   Scale Module Controls = RigVMCollapseNode(Controls=(), ExecuteContext=(), Global Scale=1, Local Scale=0)
   Set IKFK Visibility = RigVMCollapseNode(ExecuteContext=(), FK Controls=(), IK Controls=(), IKFK Switch=True)
   Set Offset Transform from Current = RigVMCollapseNode(Control=(Type=Bone,Name="None"), ExecuteContext=())
   Spine Match IK onto FK = RigVMCollapseNode(Bones=(), Execute=(), IK Controls=(), Offsets=(), Point 0=(X=0,Y=0,Z=0), Point 1=(X=0,Y=0,Z=0), Point 2=(X=0,Y=0,Z=0), Point 3=(X=0,Y=0,Z=0))
   SplineThrough3Points = RigVMCollapseNode(Execute=(), Point 0=(X=0,Y=0,Z=0), Point 1=(X=0,Y=0,Z=0), Point 2=(X=0,Y=0,Z=0), t=0.5)
   SplineThrough3Points_1 = RigVMCollapseNode(Execute=(), Point 0=(X=0,Y=0,Z=0), Point 1=(X=0,Y=0,Z=0), Point 2=(X=0,Y=0,Z=0), t=0.5)
   SplineThrough4Points = RigVMCollapseNode(Execute=(), Point 0=(X=0,Y=0,Z=0), Point 1=(X=0,Y=0,Z=0), Point 2=(X=0,Y=0,Z=0), Point 3=(X=0,Y=0,Z=0))
   SplineThrough4Points_1 = RigVMCollapseNode(Execute=(), Point 0=(X=0,Y=0,Z=0), Point 1=(X=0,Y=0,Z=0), Point 2=(X=0,Y=0,Z=0), Point 3=(X=0,Y=0,Z=0))
   Tag IK Spine Controls = RigVMCollapseNode(ExecuteContext=(), Item=())
  ]}
 RigVMFunction_MathBoolAnd_1_1_ContainedGraph {nodes [
   n1 = RigVMFunctionEntryNode()
   RigVMFunctionReturnNode(Result=&RigVMFunction_MathBoolAnd(B=@n1.C, A=&RigVMFunction_MathBoolAnd(A=@n1.A, B=@n1.B).Result).Result)
  ]}
 RigVMFunction_MathBoolOr_1_ContainedGraph {nodes [
   n1 = RigVMFunctionEntryNode()
   RigVMFunction_MathBoolOr = &RigVMFunction_MathBoolOr(A=@n1.A, B=@n1.B)
   n2 = &RigVMFunction_MathBoolOr(B=@n1.C, A=@RigVMFunction_MathBoolOr.Result)
   RigVMFunctionReturnNode(Result=&RigVMFunction_MathBoolOr(B=@n1.E, A=&RigVMFunction_MathBoolOr(B=@n1.D, A=@n2.Result).Result).Result)
  ]}
 RigVMFunction_Sequence_1_1_ContainedGraph {nodes [
   n1 = RigVMFunctionEntryNode()
   n2 = &RigVMFunction_Sequence(ExecuteContext=@n1.ExecuteContext)
   n3 = &RigVMFunction_Sequence(ExecuteContext=@n2.B)
   n4 = &RigVMFunction_Sequence(ExecuteContext=@n3.B)
   n5 = &RigVMFunction_Sequence(ExecuteContext=@n4.B)
   RigVMFunctionReturnNode(A=@n2.A, B=@n3.A, C=@n4.A, D=@n5.A, E=@n5.B)
  ]}
 RigVMFunction_Sequence_1_1_ContainedGraph {nodes [
   n1 = RigVMFunctionEntryNode()
   n2 = &RigVMFunction_Sequence(ExecuteContext=@n1.ExecuteContext)
   n3 = &RigVMFunction_Sequence(ExecuteContext=@n2.B)
   n4 = &RigVMFunction_Sequence(ExecuteContext=@n3.B)
   RigVMFunctionReturnNode(A=@n2.A, B=@n3.A, C=@n4.A, D=@n4.B)
  ]}
 RigVMFunction_Sequence_1_1_ContainedGraph {nodes [
   n1 = RigVMFunctionEntryNode()
   n2 = &RigVMFunction_Sequence(ExecuteContext=@n1.ExecuteContext)
   n3 = &RigVMFunction_Sequence(ExecuteContext=@n2.B)
   n4 = &RigVMFunction_Sequence(ExecuteContext=@n3.B)
   RigVMFunctionReturnNode(A=@n2.A, B=@n3.A, C=@n4.A, D=@n4.B)
  ]}
 RigVMFunction_Sequence_1_1_ContainedGraph {nodes [
   n1 = RigVMFunctionEntryNode()
   n2 = &RigVMFunction_Sequence(ExecuteContext=@n1.ExecuteContext)
   n3 = &RigVMFunction_Sequence(ExecuteContext=@n2.B)
   n4 = &RigVMFunction_Sequence(ExecuteContext=@n3.B)
   n5 = &RigVMFunction_Sequence(ExecuteContext=@n4.B)
   RigVMFunctionReturnNode(A=@n2.A, B=@n3.A, C=@n4.A, D=@n5.A, E=@n5.B)
  ]}
 RigVMFunction_Sequence_1_ContainedGraph {nodes [
   n1 = RigVMFunctionEntryNode()
   RigVMFunction_Sequence = &RigVMFunction_Sequence(ExecuteContext=@n1.ExecuteContext)
   n2 = &RigVMFunction_Sequence(ExecuteContext=@RigVMFunction_Sequence.B)
   RigVMFunctionReturnNode(A=@RigVMFunction_Sequence.A, B=@n2.A, C=@n2.B)
  ]}
 RigVMFunction_Sequence_1_ContainedGraph {nodes [
   n1 = RigVMFunctionEntryNode()
   RigVMFunction_Sequence = &RigVMFunction_Sequence(ExecuteContext=@n1.ExecuteContext)
   n2 = &RigVMFunction_Sequence(ExecuteContext=@RigVMFunction_Sequence.B)
   n3 = &RigVMFunction_Sequence(ExecuteContext=@n2.B)
   RigVMFunctionReturnNode(A=@RigVMFunction_Sequence.A, B=@n2.A, C=@n3.A, D=@n3.B)
  ]}
 RigVMFunction_Sequence_1_ContainedGraph {nodes [
   n1 = RigVMFunctionEntryNode()
   RigVMFunction_Sequence = &RigVMFunction_Sequence(ExecuteContext=@n1.ExecuteContext)
   n2 = &RigVMFunction_Sequence(ExecuteContext=@RigVMFunction_Sequence.B)
   RigVMFunctionReturnNode(A=@RigVMFunction_Sequence.A, B=@n2.A, C=@n2.B)
  ]}
 RigVMFunction_Sequence_1_ContainedGraph {nodes [
   n1 = RigVMFunctionEntryNode()
   RigVMFunction_Sequence = &RigVMFunction_Sequence(ExecuteContext=@n1.ExecuteContext)
   n2 = &RigVMFunction_Sequence(ExecuteContext=@RigVMFunction_Sequence.B)
   RigVMFunctionReturnNode(A=@RigVMFunction_Sequence.A, B=@n2.A, C=@n2.B)
  ]}
 RigVMFunction_Sequence_1_ContainedGraph {nodes [
   n1 = RigVMFunctionEntryNode()
   RigVMFunction_Sequence = &RigVMFunction_Sequence(ExecuteContext=@n1.ExecuteContext)
   n2 = &RigVMFunction_Sequence(ExecuteContext=@RigVMFunction_Sequence.B)
   n3 = &RigVMFunction_Sequence(ExecuteContext=@n2.B)
   n4 = &RigVMFunction_Sequence(ExecuteContext=@n3.B)
   n5 = &RigVMFunction_Sequence(ExecuteContext=@n4.B)
   n6 = &RigVMFunction_Sequence(ExecuteContext=@n5.B)
   n7 = &RigVMFunction_Sequence(ExecuteContext=@n6.B)
   n8 = &RigVMFunction_Sequence(ExecuteContext=@n7.B)
   n9 = &RigVMFunction_Sequence(ExecuteContext=@n8.B)
   n10 = &RigVMFunction_Sequence(ExecuteContext=@n9.B)
   RigVMFunctionReturnNode(A=@RigVMFunction_Sequence.A, B=@n2.A, C=@n3.A, D=@n4.A, E=@n5.A, F=@n6.A, G=@n7.A, H=@n8.A, I=@n9.A, J=@n10.A, K=@n10.B)
  ]}
 RigVMFunction_Sequence_2_1_ContainedGraph {nodes [
   n1 = RigVMFunctionEntryNode()
   n2 = &RigVMFunction_Sequence(ExecuteContext=@n1.ExecuteContext)
   n3 = &RigVMFunction_Sequence(ExecuteContext=@n2.B)
   RigVMFunctionReturnNode(A=@n2.A, B=@n3.A, C=@n3.B)
  ]}
 RigVMFunction_Sequence_2_ContainedGraph {nodes [
   n1 = RigVMFunctionEntryNode()
   RigVMFunction_Sequence = &RigVMFunction_Sequence(ExecuteContext=@n1.ExecuteContext)
   n2 = &RigVMFunction_Sequence(ExecuteContext=@RigVMFunction_Sequence.B)
   RigVMFunctionReturnNode(A=@RigVMFunction_Sequence.A, B=@n2.A, C=@n2.B)
  ]}
 RigVMFunction_Sequence_3_1_ContainedGraph {nodes [
   n1 = RigVMFunctionEntryNode()
   n2 = &RigVMFunction_Sequence(ExecuteContext=@n1.ExecuteContext)
   n3 = &RigVMFunction_Sequence(ExecuteContext=@n2.B)
   n4 = &RigVMFunction_Sequence(ExecuteContext=@n3.B)
   RigVMFunctionReturnNode(A=@n2.A, B=@n3.A, C=@n4.A, D=@n4.B)
  ]}
 RigVMFunction_Sequence_3_1_ContainedGraph {nodes [
   n1 = RigVMFunctionEntryNode()
   n2 = &RigVMFunction_Sequence(ExecuteContext=@n1.ExecuteContext)
   n3 = &RigVMFunction_Sequence(ExecuteContext=@n2.B)
   RigVMFunctionReturnNode(A=@n2.A, B=@n3.A, C=@n3.B)
  ]}
 RigVMFunction_Sequence_4_ContainedGraph {nodes [
   n1 = RigVMFunctionEntryNode()
   RigVMFunction_Sequence = &RigVMFunction_Sequence(ExecuteContext=@n1.ExecuteContext)
   n2 = &RigVMFunction_Sequence(ExecuteContext=@RigVMFunction_Sequence.B)
   RigVMFunctionReturnNode(A=@RigVMFunction_Sequence.A, B=@n2.A, C=@n2.B)
  ]}
 RigVMFunction_Sequence_6_1_ContainedGraph {nodes [
   n1 = RigVMFunctionEntryNode()
   n2 = &RigVMFunction_Sequence(ExecuteContext=@n1.ExecuteContext)
   n3 = &RigVMFunction_Sequence(ExecuteContext=@n2.B)
   n4 = &RigVMFunction_Sequence(ExecuteContext=@n3.B)
   n5 = &RigVMFunction_Sequence(ExecuteContext=@n4.B)
   RigVMFunctionReturnNode(A=@n2.A, B=@n3.A, C=@n4.A, D=@n5.A, E=@n5.B)
  ]}
 RigVMModel {nodes [
   RigUnit_BeginExecution = &RigUnit_BeginExecution()
   GetAnimationChannel = /Script/ControlRig.RigUnit_GetBoolAnimationChannel(
    Channel=spine_fk_ik_switch
    bInitial=False
    Control=&RigUnit_ResolveConnector(Connector=(Type=Connector,Name="Settings Control"), SkipSocket=true).Result.Name
   ) {is=struct RigVMUnitNode, template="GetAnimationChannel::Execute(out Value,in Control,in Channel,in bInitial)"}
   FWD AnimalSpineLimb = RigVMFunctionReferenceNode(
    Attach Skeleton=true
    Input Bones=()
    FK IK Switch=@GetAnimationChannel.Value
    ExecuteContext=@RigUnit_BeginExecution.ExecutePin
    Secondary Spline Direction=RigVMVariableNode() {variable=Secondary Spline Direction}.Value
    Primary Axis=RigVMVariableNode() {variable=Primary Axis}.Value
    Stretch=RigVMVariableNode() {variable=Enable Stretch}.Value
    debug=RigVMVariableNode() {variable=Debug}.Value
    Aim Spline=RigVMVariableNode() {variable=Aim Spline}.Value
    Template Spline=RigVMVariableNode() {variable=Local Spline}.Value
    Up Axis=RigVMVariableNode() {variable=Secondary Axis}.Value
   )
  ]}
 RigVMModel Backwards Solve Graph {nodes [
   n1 = /Script/ControlRig.RigUnit_InverseExecution() {is=struct RigVMUnitNode}
   Get_Module_Metadata = RigVMDispatchNode(Default=(Type=Bone,Name="None"), Name=Body Control, NameSpace=Self) {template="DISPATCH_RigDispatch_GetModuleMetadata(in Name,in NameSpace,in Default,out Value,out Found)"}
   Get_Module_Metadata_1 = RigVMDispatchNode(Default=(Type=Bone,Name="None"), Name=Chest Control, NameSpace=Self) {template="DISPATCH_RigDispatch_GetModuleMetadata(in Name,in NameSpace,in Default,out Value,out Found)"}
   Get_Module_Metadata_1_1 = RigVMDispatchNode(Default=(Type=Bone,Name="None"), Name=Main Control, NameSpace=Self) {template="DISPATCH_RigDispatch_GetModuleMetadata(in Name,in NameSpace,in Default,out Value,out Found)"}
   For_Each = RigVMDispatchNode(
    Array=((Type=None,Name="None"),(Type=None,Name="None"),(Type=None,Name="None"))
    Array.1=@Get_Module_Metadata.Value
    Array.2=@Get_Module_Metadata_1.Value
    Array.0=@Get_Module_Metadata_1_1.Value
    ExecuteContext=@n1.ExecutePin
   ) {template="DISPATCH_RigVMDispatch_ArrayIterator(in Array,out Element,out Index,out Count,out Ratio)"}
   > FindClosestBone_0 = RigVMFunctionReferenceNode(
    Initial=true
    Bones=RigVMDispatchNode(Default=(), Name=Bones, NameSpace=Self) {template="DISPATCH_RigDispatch_GetModuleMetadata(in Name,in NameSpace,in Default,out Value,out Found)"}.Value
    Item=@For_Each.Element
   )
   > &RigUnit_ParentConstraint(
    AdvancedSettings=(InterpolationType=Average,RotationOrderForFilter=XZY)
    Parents=((Item=(Type=Bone,Name="None"),Weight=1))
    Weight=1
    Parents.0.Item=@FindClosestBone_0.Value
    Child=@For_Each.Element
   )
  ]}
 RigVMModel Connector Event Graph {nodes [
   n1 = /Script/ControlRig.RigUnit_ConnectorExecution() {is=struct RigVMUnitNode}
   RigVMFunction_Sequence_1 = RigVMAggregateNode(ExecuteContext=@n1.ExecuteContext)
   n2 = /Script/ControlRig.RigUnit_GetCandidates() {is=struct RigVMUnitNode}
   n3 = &RigVMFunction_ControlFlowBranch(
    Condition=RigVMDispatchNode(B=(Type=Connector,Name="Pelvis Bone"), A=@n2.Connector) {template="DISPATCH_RigVMDispatch_CoreEquals(in A,in B,out Result)"}.Result
    ExecuteContext=@RigVMFunction_Sequence_1.A
   )
   RigVMVariableNode(ExecuteContext=@n3.True) {variable=Found Pelvis}
   > For_Each_2 = RigVMDispatchNode(Array=@n2.Candidates) {template="DISPATCH_RigVMDispatch_ArrayIterator(in Array,out Element,out Index,out Count,out Ratio)"}
   Find = RigVMFunctionReferenceNode(Element=pelvis, Value=@For_Each_2.Element.Name)
   n4 = RigVMVariableNode() {variable=Found Pelvis}
   n5 = &RigVMFunction_ControlFlowBranch(
    ExecuteContext=@For_Each_2.ExecuteContext
    Condition=&RigVMFunction_MathBoolAnd(A=@Find.Success, B=&RigVMFunction_MathBoolNot(Value=@n4.Value).Result).Result
   )
   /Script/ControlRig.RigUnit_SetDefaultMatch(Default=@For_Each_2.Element, ExecutePin=@n5.True) {is=struct RigVMUnitNode, template="Set Default Match::Execute(in Default)"}
   > RigVMVariableNode(Value=true) {variable=Found Pelvis}
   n6 = &RigUnit_GetCandidates()
   n7 = &RigVMFunction_ControlFlowBranch(
    Condition=RigVMDispatchNode(B=(Type=Connector,Name="Spine Start Bone"), A=@n6.Connector) {template="DISPATCH_RigVMDispatch_CoreEquals(in A,in B,out Result)"}.Result
    ExecuteContext=@RigVMFunction_Sequence_1.B
   )
   n8 = RigVMDispatchNode(ExecuteContext=@n7.True, Array=@n6.Candidates) {template="DISPATCH_RigVMDispatch_ArrayIterator(in Array,out Element,out Index,out Count,out Ratio)"}
   n9 = &RigUnit_ResolveConnector(Connector=(Type=Connector,Name="AnimalSpine Primary"), SkipSocket=true)
   RigVMFunction_ControlFlowBranch = &RigVMFunction_ControlFlowBranch(
    ExecuteContext=@n8.ExecuteContext
    Condition=RigVMDispatchNode(A=@n8.Element, B=@n9.Result) {template="DISPATCH_RigVMDispatch_CoreEquals(in A,in B,out Result)"}.Result
   )
   &RigUnit_SetDefaultMatch(Default=@n8.Element, ExecutePin=@RigVMFunction_ControlFlowBranch.True) {template="Set Default Match::Execute(in Default)"}
   > RigVMVariableNode(Value=true) {variable=Found First Spine}
   DISPATCH_RigVMDispatch_CoreEquals = RigVMDispatchNode(B=(Type=Connector,Name="Spine End Bone"), A=&RigUnit_GetCandidates().Connector) {template="DISPATCH_RigVMDispatch_CoreEquals(in A,in B,out Result)"}
   n10 = &RigVMFunction_ControlFlowBranch(Condition=@DISPATCH_RigVMDispatch_CoreEquals.Result, ExecuteContext=@RigVMFunction_Sequence_1.C)
   For_Each_1 = RigVMDispatchNode(
    Array=&RigUnit_CollectionChildrenArray(
     TypeToSearch=Bone
     bDefaultChildren=True
     bIncludeParent=False
     bRecursive=true
     Parent=&RigUnit_ResolveConnector(Connector=(Type=Connector,Name="AnimalSpine Primary"), SkipSocket=true).Result
    ).Items
    ExecuteContext=@n10.True
   ) {template="DISPATCH_RigVMDispatch_ArrayIterator(in Array,out Element,out Index,out Count,out Ratio)"}
   Contains = &RigVMFunction_Contains(Search=neck, Name=@For_Each_1.Element.Name) {template="Contains::Execute(in Name,in Search,out Result)"}
   Find_1_1 = RigVMFunctionReferenceNode(Element=neck, Value=@For_Each_1.Element.Name)
   RigVMFunction_MathBoolOr = &RigVMFunction_MathBoolOr(B=@Contains.Result, A=@Find_1_1.Success)
   n11 = &RigVMFunction_ControlFlowBranch(ExecuteContext=@For_Each_1.ExecuteContext, Condition=@RigVMFunction_MathBoolOr.Result)
   &RigUnit_SetDefaultMatch(Default=&RigUnit_HierarchyGetParent(bDefaultParent=True, Child=@For_Each_1.Element).Parent, ExecutePin=@n11.True) {template="Set Default Match::Execute(in Default)"}
   > RigVMDispatchNode(Enabled=True, ScreenColor=(R=1,G=1,B=1,A=1), ScreenDuration=0.05, Value=@For_Each_1.Element) {template="DISPATCH_RigVMDispatch_Print(in Prefix,in Value,in Enabled,in ScreenDuration,in ScreenColor)"}
   Connect to Module Metadata = RigVMFunctionReferenceNode(
    Connector=(Type=Connector,Name="Parent Control")
    Module Metadata Name=Parent Control
    Module Metadata NameSpace=Root
    ExecuteContext=@RigVMFunction_Sequence_1.E
   )
   > Connect to Module Metadata_1 = RigVMFunctionReferenceNode(Connector=(Type=Connector,Name="Settings Control"), Module Metadata Name=Settings Control, Module Metadata NameSpace=Root)
   RigVMVariableNode(ExecuteContext=(), Value=false) {variable=Found First Spine}
   > For_Each = RigVMDispatchNode(Array=&RigUnit_CollectionChildrenArray(TypeToSearch=Bone, bDefaultChildren=True, bIncludeParent=False, bRecursive=true, Parent=@n9.Result).Items) {template="DISPATCH_RigVMDispatch_ArrayIterator(in Array,out Element,out Index,out Count,out Ratio)"}
   Find_1 = RigVMFunctionReferenceNode(Element=spine, Value=@For_Each.Element.Name)
   RigVMFunction_MathBoolNot = &RigVMFunction_MathBoolNot(Value=RigVMVariableNode() {variable=Found First Spine}.Value)
   RigVMFunction_MathBoolAnd = &RigVMFunction_MathBoolAnd(A=@Find_1.Success, B=@RigVMFunction_MathBoolNot.Result)
   n12 = &RigVMFunction_ControlFlowBranch(ExecuteContext=@For_Each.ExecuteContext, Condition=@RigVMFunction_MathBoolAnd.Result)
   RigVMDispatchNode(Enabled=True, Prefix="Value - ", ScreenColor=(R=1,G=1,B=1,A=1), ScreenDuration=0.7, Value=@Find_1.Success, ExecuteContext=@n12.Completed) {template="DISPATCH_RigVMDispatch_Print(in Prefix,in Value,in Enabled,in ScreenDuration,in ScreenColor)"}
   &RigUnit_SetDefaultMatch(Default=@For_Each.Element, ExecutePin=@n12.True) {template="Set Default Match::Execute(in Default)"}
   > RigVMVariableNode(Value=true) {variable=Found First Spine}
  ]}
]
hierarchy [
 AnimalSpine Primary {type=Connector, loc=<0 0 0>, rot=<0 0 0 1>}
 Root {type=Socket, loc=<0 0 0>, rot=<0 0 0 1>}
 Spine Start Bone {type=Connector, loc=<0 0 0>, rot=<0 0 0 1>}
 Spine End Bone {type=Connector, loc=<0 0 0>, rot=<0 0 0 1>}
 Parent Control {type=Connector, loc=<0 0 0>, rot=<0 0 0 1>}
 Settings Control {type=Connector, loc=<0 0 0>, rot=<0 0 0 1>}
 Pelvis Bone {type=Connector, loc=<0 0 0>, rot=<0 0 0 1>}
]
variables [
 Aim Spline {type=FControlRigSpline, default=()}
 Body Control Adjustment {type=float, default=0}
 Body Control Shape Settings {
  type=FRigUnit_HierarchyAddControlTransform_Settings
  default="(InitialSpace=LocalSpace,bUsePreferredRotationOrder=False,PreferredRotationOrder=YZX,Limits=(LimitTranslationX=(bMinimum=False,bMaximum=False),LimitTranslationY=(bMinimum=False,bMaximum=False),LimitTranslationZ=(bMinimum=False,bMaximum=False),LimitPitch=(b ...(truncated)"
 }
 Chest Control Settings {
  type=FRigUnit_HierarchyAddControlTransform_Settings
  default="(InitialSpace=LocalSpace,bUsePreferredRotationOrder=False,PreferredRotationOrder=YZX,Limits=(LimitTranslationX=(bMinimum=False,bMaximum=False),LimitTranslationY=(bMinimum=False,bMaximum=False),LimitTranslationZ=(bMinimum=False,bMaximum=False),LimitPitch=(b ...(truncated)"
 }
 Chest Hips Position {type=FVector, default=(X=0,Y=0,Z=25)}
 Debug {type=bool, default=False}
 Default FK or IK {type=bool, default=False}
 Enable Stretch {type=bool, default=True}
 FK Control Settings {
  type=FRigUnit_HierarchyAddControlTransform_Settings
  default="(InitialSpace=LocalSpace,bUsePreferredRotationOrder=False,PreferredRotationOrder=YZX,Limits=(LimitTranslationX=(bMinimum=False,bMaximum=False),LimitTranslationY=(bMinimum=False,bMaximum=False),LimitTranslationZ=(bMinimum=False,bMaximum=False),LimitPitch=(b ...(truncated)"
 }
 Found First Spine {type=bool, default=False}
 Found Pelvis {type=bool, default=False}
 Hips Chest Shape Settings {
  type=FRigUnit_HierarchyAddControlTransform_Settings
  default="(InitialSpace=LocalSpace,bUsePreferredRotationOrder=False,PreferredRotationOrder=YZX,Limits=(LimitTranslationX=(bMinimum=False,bMaximum=False),LimitTranslationY=(bMinimum=False,bMaximum=False),LimitTranslationZ=(bMinimum=False,bMaximum=False),LimitPitch=(b ...(truncated)"
 }
 Hips Control Settings {
  type=FRigUnit_HierarchyAddControlTransform_Settings
  default="(InitialSpace=LocalSpace,bUsePreferredRotationOrder=False,PreferredRotationOrder=YZX,Limits=(LimitTranslationX=(bMinimum=False,bMaximum=False),LimitTranslationY=(bMinimum=False,bMaximum=False),LimitTranslationZ=(bMinimum=False,bMaximum=False),LimitPitch=(b ...(truncated)"
 }
 Hips at Pelvis {type=bool, default=True}
 Local Spline {type=FControlRigSpline, default=()}
 Main Control Settings {
  type=FRigUnit_HierarchyAddControlTransform_Settings
  default="(InitialSpace=LocalSpace,bUsePreferredRotationOrder=False,PreferredRotationOrder=YZX,Limits=(LimitTranslationX=(bMinimum=False,bMaximum=False),LimitTranslationY=(bMinimum=False,bMaximum=False),LimitTranslationZ=(bMinimum=False,bMaximum=False),LimitPitch=(b ...(truncated)"
 }
 Mid IK Control Settings {
  type=FRigUnit_HierarchyAddControlTransform_Settings
  default="(InitialSpace=LocalSpace,bUsePreferredRotationOrder=False,PreferredRotationOrder=YZX,Limits=(LimitTranslationX=(bMinimum=False,bMaximum=False),LimitTranslationY=(bMinimum=False,bMaximum=False),LimitTranslationZ=(bMinimum=False,bMaximum=False),LimitPitch=(b ...(truncated)"
 }
 Module Settings {
  type=FUserDefinedStruct_#6
  default="(OverrideColor_48_#4=False,LeftColor_3_#3=(R=0,G=0.0231,B=0.4115,A=1),RightColor_5_#5=(R=1,G=0,B=0,A=1),CenterColor_9_#2=(R=1,G=1,B=0,A= ...(truncated)"
 }
 Number of Spines Bones {type=int32, default=5}
 Prefix {type=FName, default=spinespine}
 Primary Axis {type=FVector, default=(X=1,Y=0,Z=0)}
 Secondary Axis {type=FVector, default=(X=0,Y=0,Z=1)}
 Secondary Spline Direction {type=FVector, default=(X=0,Y=0,Z=5)}
 Suffix {type=FName, default=_ctrl_ctrl}
]
