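# Modular Control Rig arm module that builds an FK/IK two-bone limb from shoulder, elbow and hand connectors.
# Construction calls TwoBoneLimb with module variables and root/parent metadata for spaces and offsets, then adds a twist rig.
# It publishes the hand bone as metadata and creates finger sockets from metadata or detected leaf bones.
# Helper graphs include exponential soft IK, pole-vector matching and optional IK hand/gun bone constraints to the hand.
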ControlRigBlueprint /ControlRigModules/Modules56/Arm
graphs [
 CollapseNode_ContainedGraph {nodes [
   n1 = RigVMFunctionEntryNode()
   RigVMFunction_Sequence = &RigVMFunction_Sequence(ExecuteContext=@n1.ExecuteContext)
   Get_Module_Metadata_3_1 = RigVMDispatchNode(Default=(), Name=Finger Sockets, NameSpace=Root) {template="DISPATCH_RigDispatch_GetModuleMetadata(in Name,in NameSpace,in Default,out Value,out Found)"}
   n2 = &RigVMFunction_ControlFlowBranch(Condition=@Get_Module_Metadata_3_1.Found, ExecuteContext=@RigVMFunction_Sequence.A)
   n3 = /Script/ControlRig.RigUnit_ResolveConnector(Connector=(Type=Connector,Name="Hand Bone"), SkipSocket=False) {is=struct RigVMUnitNode}
   n4 = RigVMDispatchNode(
    Array=/Script/ControlRig.RigUnit_CollectionChildrenArray(TypeToSearch=Bone, bDefaultChildren=True, bIncludeParent=False, bRecursive=true, Parent=@n3.Result) {is=struct RigVMUnitNode}.Items
    ExecuteContext=@n2.False
   ) {template="DISPATCH_RigVMDispatch_ArrayIterator(in Array,out Element,out Index,out Count,out Ratio)"}
   n5 = RigVMDispatchNode(Array=&RigUnit_CollectionChildrenArray(TypeToSearch=Bone, bDefaultChildren=True, bIncludeParent=False, bRecursive=False, Parent=@n4.Element).Items) {template="DISPATCH_RigVMDispatch_ArrayGetNum(in Array,out Num)"}
   n6 = RigVMDispatchNode(A=/Script/ControlRig.RigUnit_HierarchyGetParent(bDefaultParent=True, Child=@n4.Element) {is=struct RigVMUnitNode}.Parent, B=@n3.Result) {template="DISPATCH_RigVMDispatch_CoreEquals(in A,in B,out Result)"}
   n7 = &RigVMFunction_ControlFlowBranch(
    ExecuteContext=@n4.ExecuteContext
    Condition=&RigVMFunction_MathBoolAnd(
     A=RigVMDispatchNode(B=0, A=@n5.Num) {template="DISPATCH_RigVMDispatch_CoreEquals(in A,in B,out Result)"}.Result
     B=&RigVMFunction_MathBoolNot(Value=@n6.Result).Result
    ).Result
   )
   /Script/ControlRig.RigUnit_SetMetadataTag(NameSpace=Self, Tag=Finger Leaf Bones, Item=@n4.Element, ExecutePin=@n7.True) {is=struct RigVMUnitNode}
   n8 = RigVMDispatchNode(
    ExecuteContext=@n4.Completed
    Array=/Script/ControlRig.RigUnit_FindItemsWithMetadataTag(NameSpace=Self, Tag=Finger Leaf Bones) {is=struct RigVMUnitNode}.Items
   ) {template="DISPATCH_RigVMDispatch_ArrayIterator(in Array,out Element,out Index,out Count,out Ratio)"}
   > Has Socket_1 = RigVMFunctionReferenceNode(Item=@n8.Element)
   n9 = &RigVMFunction_ControlFlowBranch(ExecuteContext=@"Has Socket_1".ExecuteContext, Condition=@"Has Socket_1".Result)
   /Script/ControlRig.RigUnit_HierarchyAddSocket(
    Color=(R=1,G=1,B=1,A=1)
    Space=LocalSpace
    Transform=(Rotation=(X=0,Y=0,Z=0,W=1),Translation=(X=0,Y=0,Z=0),Scale3D=(X=1,Y=1,Z=1))
    Name=/Script/RigVM.RigVMFunction_NameConcat(B=_socket, A=@n8.Element.Name) {is=struct RigVMUnitNode, template="Concat::Execute(in A,in B,out Result)"}.Result
    Parent=@n8.Element
    ExecutePin=@n9.False
   ) {is=struct RigVMUnitNode}
   For_Each_1 = RigVMDispatchNode(Array=@Get_Module_Metadata_3_1.Value, ExecuteContext=@n2.True) {template="DISPATCH_RigVMDispatch_ArrayIterator(in Array,out Element,out Index,out Count,out Ratio)"}
   n10 = &RigUnit_ResolveConnector(Connector=(Type=Connector,Name="Hand Bone"), SkipSocket=true)
   n11 = &RigVMFunction_ControlFlowBranch(
    Condition=RigVMDispatchNode(
     Array=&RigUnit_CollectionChildrenArray(TypeToSearch=Bone, bDefaultChildren=True, bIncludeParent=False, bRecursive=true, Parent=@n10.Result).Items
     Element=@For_Each_1.Element
    ) {template="DISPATCH_RigVMDispatch_ArrayFind(in Array,in Element,out Index,out Success)"}.Success
    ExecuteContext=@For_Each_1.ExecuteContext
   )
   &RigUnit_HierarchyAddSocket(
    Color=(R=1,G=1,B=1,A=1)
    Space=LocalSpace
    Transform=(Rotation=(X=0,Y=0,Z=0,W=1),Translation=(X=0,Y=0,Z=0),Scale3D=(X=1,Y=1,Z=1))
    Name=&RigVMFunction_NameConcat(B=_socket, A=@For_Each_1.Element.Name) {template="Concat::Execute(in A,in B,out Result)"}.Result
    Parent=@For_Each_1.Element
    ExecutePin=@n11.True
   )
   RigVMFunctionReturnNode(ExecuteContext=@RigVMFunction_Sequence.B)
  ]}
 CollapseNode_ContainedGraph {nodes [
   RigVMFunctionEntryNode()
   RigVMFunction_MathVectorDistance = &RigVMFunction_MathVectorDistance(
    A=&RigUnit_GetTransform(
     Space=LocalSpace
     bInitial=true
     Item=&RigUnit_ResolveConnector(Connector=(Type=Connector,Name="Shoulder Bone"), SkipSocket=true).Result
    ).Transform.Translation
    B=&RigUnit_GetTransform(
     Space=LocalSpace
     bInitial=true
     Item=&RigUnit_ResolveConnector(Connector=(Type=Connector,Name="Hand Bone"), SkipSocket=true).Result
    ).Transform.Translation
   )
   Divide = &RigVMFunction_MathDoubleDiv(B=3, A=@RigVMFunction_MathVectorDistance.Result)
   Get_Module_Metadata_2 = RigVMDispatchNode(Default=50, Name=Arm PV Position, NameSpace=Root) {template="DISPATCH_RigDispatch_GetModuleMetadata(in Name,in NameSpace,in Default,out Value,out Found)"}
   Multiply = &RigVMFunction_MathDoubleMul(
    B=RigVMDispatchNode(False=@Divide.Result, Condition=@Get_Module_Metadata_2.Found, True=@Get_Module_Metadata_2.Value) {template="DISPATCH_RigVMDispatch_If(in Condition,in True,in False,out Result)"}.Result
    A=RigVMVariableNode() {variable=Pole Vector Position}.Value
   )
   RigVMFunctionReturnNode(Result=@Multiply.Result)
  ]}
 CollapseNode_ContainedGraph {nodes [
   RigVMFunctionEntryNode()
   > Set Closest Bone = RigVMFunctionReferenceNode(
    Bones=&RigUnit_FindItemsWithMetadataTag(NameSpace=Self, Tag=Bones).Items
    Items=&RigUnit_FindItemsWithMetadataTag(NameSpace=Self, Tag=PV Controls).Items
   )
   > RigVMFunctionReturnNode()
  ]}
 CollapseNode_ContainedGraph {nodes [
   RigVMFunctionEntryNode()
   > RigVMFunction_Sequence_1 = RigVMAggregateNode()
   n1 = &RigUnit_ResolveConnector(Connector=(Type=Connector,Name="IK Hand Gun Bone"), SkipSocket=true)
   n2 = &RigVMFunction_ControlFlowBranch(Condition=@n1.bIsConnected, ExecuteContext=@RigVMFunction_Sequence_1.A)
   &RigUnit_ParentConstraint(
    AdvancedSettings=(InterpolationType=Average,RotationOrderForFilter=XZY)
    Parents=((Item=(Type=Bone,Name="None"),Weight=1))
    Weight=1
    Child=@n1.Result
    Parents.0.Item=&RigUnit_ResolveConnector(Connector=(Type=Connector,Name="Hand Bone"), SkipSocket=true).Result
    ExecutePin=@n2.True
   )
   n3 = &RigUnit_ResolveConnector(Connector=(Type=Connector,Name="IK Hand Bone"), SkipSocket=true)
   n4 = &RigVMFunction_ControlFlowBranch(Condition=@n3.bIsConnected, ExecuteContext=@RigVMFunction_Sequence_1.B)
   &RigUnit_ParentConstraint(
    AdvancedSettings=(InterpolationType=Average,RotationOrderForFilter=XZY)
    Parents=((Item=(Type=Bone,Name="None"),Weight=1))
    Weight=1
    Child=@n3.Result
    Parents.0.Item=&RigUnit_ResolveConnector(Connector=(Type=Connector,Name="Hand Bone"), SkipSocket=true).Result
    ExecutePin=@n4.True
   )
   RigVMFunctionReturnNode(ExecuteContext=@RigVMFunction_Sequence_1.C)
  ]}
 Concat_1_ContainedGraph {nodes [
   n1 = RigVMFunctionEntryNode()
   Concat = &RigVMFunction_NameConcat(A=@n1.A, B=@n1.B) {template="Concat::Execute(in A,in B,out Result)"}
   n2 = &RigVMFunction_NameConcat(A=@Concat.Result, B=@n1.C) {template="Concat::Execute(in A,in B,out Result)"}
   RigVMFunctionReturnNode(Result=&RigVMFunction_NameConcat(A=&RigVMFunction_NameConcat(A=@n2.Result, B=@n1.D) {template="Concat::Execute(in A,in B,out Result)"}.Result, B=@n1.E) {template="Concat::Execute(in A,in B,out Result)"}.Result)
  ]}
 Concat_1_ContainedGraph {nodes [
   n1 = RigVMFunctionEntryNode()
   Concat = &RigVMFunction_NameConcat(A=@n1.A, B=@n1.B) {template="Concat::Execute(in A,in B,out Result)"}
   n2 = &RigVMFunction_NameConcat(A=@Concat.Result, B=@n1.C) {template="Concat::Execute(in A,in B,out Result)"}
   RigVMFunctionReturnNode(Result=&RigVMFunction_NameConcat(A=&RigVMFunction_NameConcat(A=@n2.Result, B=@n1.D) {template="Concat::Execute(in A,in B,out Result)"}.Result, B=@n1.E) {template="Concat::Execute(in A,in B,out Result)"}.Result)
  ]}
 Concat_3_1_ContainedGraph {nodes [
   n1 = RigVMFunctionEntryNode()
   RigVMFunctionReturnNode(Result=&RigVMFunction_NameConcat(A=&RigVMFunction_NameConcat(A=@n1.A, B=@n1.B) {template="Concat::Execute(in A,in B,out Result)"}.Result, B=@n1.C) {template="Concat::Execute(in A,in B,out Result)"}.Result)
  ]}
 Concat_4_1_ContainedGraph {nodes [
   n1 = RigVMFunctionEntryNode()
   n2 = &RigVMFunction_NameConcat(A=@n1.A, B=@n1.B) {template="Concat::Execute(in A,in B,out Result)"}
   RigVMFunctionReturnNode(Result=&RigVMFunction_NameConcat(A=&RigVMFunction_NameConcat(A=@n2.Result, B=@n1.C) {template="Concat::Execute(in A,in B,out Result)"}.Result, B=@n1.D) {template="Concat::Execute(in A,in B,out Result)"}.Result)
  ]}
 ConstructionGraph {nodes [
   n1 = &RigUnit_PrepareForExecution()
   Get_Module_Metadata_3 = RigVMDispatchNode(Default=(X=0,Y=0,Z=0), Name=elbow_bone_offset, NameSpace=Root) {template="DISPATCH_RigDispatch_GetModuleMetadata(in Name,in NameSpace,in Default,out Value,out Found)"}
   Add = /Script/RigVM.RigVMFunction_MathVectorAdd(A=@Get_Module_Metadata_3.Value, B=RigVMVariableNode() {variable=Elbow Bone Offset}.Value) {is=struct RigVMUnitNode, template="Add::Execute(in A,in B,out Result)"}
   Get_Module_Metadata = RigVMDispatchNode(Default=(Type=Control,Name="None"), Name=Body Control, NameSpace=Root) {template="DISPATCH_RigDispatch_GetModuleMetadata(in Name,in NameSpace,in Default,out Value,out Found)"}
   Get_Module_Metadata_1 = RigVMDispatchNode(Default=(Type=Control,Name="None"), Name=FK Parent, NameSpace=Parent) {template="DISPATCH_RigDispatch_GetModuleMetadata(in Name,in NameSpace,in Default,out Value,out Found)"}
   Get_Module_Metadata_1_1 = RigVMDispatchNode(Default=(Type=Control,Name="None"), Name=Chest Space Null, NameSpace=Root) {template="DISPATCH_RigDispatch_GetModuleMetadata(in Name,in NameSpace,in Default,out Value,out Found)"}
   Get_Module_Metadata_4_1 = RigVMDispatchNode(Default=False, Name=elbow_bone_offset_globalSpace, NameSpace=Root) {template="DISPATCH_RigDispatch_GetModuleMetadata(in Name,in NameSpace,in Default,out Value,out Found)"}
   RigVMFunction_MathBoolOr = &RigVMFunction_MathBoolOr(A=@Get_Module_Metadata_4_1.Value, B=RigVMVariableNode() {variable=Elbow Bone Offset Global Space}.Value)
   CONSTR TwoBoneLimb = RigVMFunctionReferenceNode(
    Basename=arm
    End Bone Connector=(Type=Connector,Name="Hand Bone")
    IK Control Basename=hand
    IK Control Spaces=((Type=None,Name="None"),(Type=None,Name="None"),(Type=None,Name="None"))
    Middle Bone Connector=(Type=Connector,Name="Elbow Bone")
    PV Control Spaces=((Type=None,Name="None"),(Type=None,Name="None"),(Type=None,Name="None"))
    Parent Control Connector=(Type=Connector,Name="Parent Control")
    Pole Vector Parent Connector=(Type=Connector,Name="Pole Vector Parent")
    Primary Axis=(X=1,Y=0,Z=0)
    Secondary Axis=(X=0,Y=1,Z=0)
    Settings Control Connector=(Type=Connector,Name="Settings Control")
    Start Bone Connector=(Type=Connector,Name="Shoulder Bone")
    Wrist Control Settings="(DisplayName=\"None\",InitialSpace=LocalSpace,bUsePreferredRotationOrder=False,PreferredRotationOrder=YZX,Limits=(LimitTranslationX=(bMinimum=False,bMaximum=False),LimitTranslationY=(bMinimum=False,bMaximum=False),LimitTranslationZ=(bMinimum=False,bMaximum=F ...(truncated)"
    Offset Elbow Bone=@Add.Result
    PV Offset Factor=RigVMCollapseNode().Result
    IK Control Spaces.0=@Get_Module_Metadata.Value
    PV Control Spaces.0=@Get_Module_Metadata.Value
    IK Control Spaces.2=@Get_Module_Metadata_1.Value
    PV Control Spaces.1=@Get_Module_Metadata_1.Value
    IK Control Spaces.1=@Get_Module_Metadata_1_1.Value
    PV Control Spaces.2=@Get_Module_Metadata_1_1.Value
    Offset Elbow Bone Space=RigVMDispatchNode(True=GlobalSpace, Condition=@RigVMFunction_MathBoolOr.Result) {template="DISPATCH_RigVMDispatch_If(in Condition,in True,in False,out Result)"}.Result
    ExecuteContext=@n1.ExecutePin
    FK IK Default=RigVMVariableNode() {variable=FK IK Default}.Value
    Orient to World=RigVMVariableNode() {variable=Orient IK to World}.Value
    Offset PV Control=RigVMVariableNode() {variable=Offset PV Control}.Value
    Enable PV Follow=RigVMVariableNode() {variable=Enable PV Follow}.Value
    Enable Keep Aligned=RigVMVariableNode() {variable=Enable Keep Aligned}.Value
    PV Control Settings=RigVMVariableNode() {variable=PV Shape Settings}.Value
    IK Control Settings=RigVMVariableNode() {variable=IK Shape Settings}.Value
    FK Control Settings=RigVMVariableNode() {variable=FK Shape Settings}.Value
    Module Settings=RigVMVariableNode() {variable=Module Settings}.Value
    Offset IK Ctrl Transform=RigVMVariableNode() {variable=Offset IK Ctrl}.Value
    Local Default=RigVMVariableNode() {variable=Local Default}.Value
    Enable Softness=RigVMVariableNode() {variable=Enable Softness}.Value
   )
   > CONSTR Basic Twist Rig = RigVMFunctionReferenceNode(Template Bones=@"CONSTR TwoBoneLimb".Template Bones)
   > DISPATCH_RigDispatch_SetModuleMetadata = RigVMDispatchNode(
    Name=Hand Bone
    NameSpace=Self
    Value=&RigUnit_ResolveConnector(Connector=(Type=Connector,Name="Hand Bone"), SkipSocket=true).Result
   ) {template="DISPATCH_RigDispatch_SetModuleMetadata(in Name,in NameSpace,in Value,out Success)"}
   > Create Sockets = RigVMCollapseNode()
  ]}
 ContainedGraph {nodes [
   n1 = RigVMFunctionEntryNode()
   RigUnit_SequenceAggregate_1 = &RigVMFunction_Sequence(ExecuteContext=@n1.ExecuteContext)
   RigUnit_SequenceAggregate = RigVMAggregateNode(ExecuteContext=@RigUnit_SequenceAggregate_1.A)
   BranchNode_4 = &RigVMFunction_ControlFlowBranch(Condition=@n1.X Plane, ExecuteContext=@RigUnit_SequenceAggregate.A)
   n2 = RigVMVariableNode() {variable=vector pos}
   RigVMVariableNode(
    Value=(X=0,Y=0,Z=0)
    ExecuteContext=@BranchNode_4.True
    Value.X=&RigUnit_GetTransform(bInitial=true, Item=@n1.Item).Transform.Translation.X
    Value.Y=@n2.Value.Y
    Value.Z=@n2.Value.Z
   ) {variable=vector pos}
   BranchNode_1_1 = &RigVMFunction_ControlFlowBranch(Condition=@n1.Y Plane, ExecuteContext=@RigUnit_SequenceAggregate.B)
   n3 = RigVMVariableNode() {variable=vector pos}
   RigVMVariableNode(
    Value=(X=0,Y=0,Z=0)
    ExecuteContext=@BranchNode_1_1.True
    Value.Y=&RigUnit_GetTransform(bInitial=true, Item=@n1.Item).Transform.Translation.Y
    Value.X=@n3.Value.X
    Value.Z=@n3.Value.Z
   ) {variable=vector pos}
   BranchNode_2_1 = &RigVMFunction_ControlFlowBranch(Condition=@n1.Z Plane, ExecuteContext=@RigUnit_SequenceAggregate.C)
   n4 = RigVMVariableNode() {variable=vector pos}
   RigVMVariableNode(
    Value=(X=0,Y=0,Z=0)
    ExecuteContext=@BranchNode_2_1.True
    Value.Z=&RigUnit_GetTransform(bInitial=true, Item=@n1.Item).Transform.Translation.Z
    Value.X=@n4.Value.X
    Value.Y=@n4.Value.Y
   ) {variable=vector pos}
   Cast = &RigVMFunction_MathBoolToFloat(Value=@n1.X Plane)
   Dot__ = /Script/RigVM.RigVMFunction_MathVectorDot(
    B=(X=1,Y=1,Z=0)
    B.X=@Cast.Result
    B.Y=&RigVMFunction_MathBoolToFloat(Value=@n1.Y Plane).Result
    B.Z=&RigVMFunction_MathBoolToFloat(Value=@n1.Z Plane).Result
    A=RigVMVariableNode() {variable=vector pos}.Value
   ) {is=struct RigVMUnitNode, template="Dot,|::Execute(in A,in B,out Result)"}
   Less = /Script/RigVM.RigVMFunction_MathDoubleLess(B=0, A=@Dot__.Result) {is=struct RigVMUnitNode, template="Less::Execute(in A,in B,out Result)"}
   RigVMVariableNode(Value=@Less.Result, ExecuteContext=@RigUnit_SequenceAggregate.D) {variable=is on right}
   RigVMFunctionReturnNode(ExecuteContext=@RigUnit_SequenceAggregate_1.B, Right Side=RigVMVariableNode() {variable=is on right}.Value)
  ]}
 ContainedGraph {nodes [
   n1 = RigVMFunctionEntryNode()
   RigUnit_SequenceAggregate = &RigVMFunction_Sequence(ExecuteContext=@n1.ExecuteContext)
   RigUnit_SequenceAggregate_1 = &RigVMFunction_Sequence(ExecuteContext=@RigUnit_SequenceAggregate.B)
   RigUnit_SequenceAggregate_2 = &RigVMFunction_Sequence(ExecuteContext=@RigUnit_SequenceAggregate_1.B)
   RigVMFunctionReturnNode(A=@RigUnit_SequenceAggregate.A, B=@RigUnit_SequenceAggregate_1.A, C=@RigUnit_SequenceAggregate_2.A, D=@RigUnit_SequenceAggregate_2.B)
  ]}
 ContainedGraph {nodes [
   n1 = RigVMFunctionEntryNode()
   n2 = RigVMVariableNode() {variable=Softness}
   n3 = /Script/RigVM.RigVMFunction_MathFloatAbs(Value=&RigUnit_GetTransform(
     Space=LocalSpace
     bInitial=true
     Item=RigVMDispatchNode(Index=1, Array=@n1.Bones) {template="DISPATCH_RigVMDispatch_ArrayGetAtIndex(in Array,in Index,out Element)"}.Element
    ).Transform.Translation.X) {is=struct RigVMUnitNode, template="Absolute::Execute(in Value,out Result)"}
   n4 = &RigVMFunction_MathFloatAbs(Value=&RigUnit_GetTransform(
     Space=LocalSpace
     bInitial=true
     Item=RigVMDispatchNode(Index=2, Array=@n1.Bones) {template="DISPATCH_RigVMDispatch_ArrayGetAtIndex(in Array,in Index,out Element)"}.Element
    ).Transform.Translation.X) {template="Absolute::Execute(in Value,out Result)"}
   n5 = &RigVMFunction_MathFloatAdd(A=@n3.Result, B=@n4.Result) {template="Add::Execute(in A,in B,out Result)"}
   n6 = &RigVMFunction_MathFloatMul(
    B=0.02
    A=&RigVMFunction_MathFloatMul(A=@n5.Result, B=RigVMVariableNode() {variable=Softness}.Value) {template="Multiply::Execute(in A,in B,out Result)"}.Result
   ) {template="Multiply::Execute(in A,in B,out Result)"}
   n7 = /Script/RigVM.RigVMFunction_MathFloatSub(A=@n5.Result, B=@n6.Result) {is=struct RigVMUnitNode, template="Subtract::Execute(in A,in B,out Result)"}
   n8 = &RigVMFunction_MathVectorDistance(
    B=&RigUnit_GetTransform(Item=@n1.Effector).Transform.Translation
    A=&RigUnit_GetTransform(Item=RigVMDispatchNode(Index=0, Array=@n1.Bones) {template="DISPATCH_RigVMDispatch_ArrayGetAtIndex(in Array,in Index,out Element)"}.Element).Transform.Translation
   )
   Exponential Softness = RigVMCollapseNode(Softness=@n6.Result, DA=@n7.Result, Distance=@n8.Result, Stretch Distance=@n8.Result)
   n9 = RigVMDispatchNode(
    False=@n1.Maximum
    Condition=&RigVMFunction_MathBoolAnd(
     A=/Script/RigVM.RigVMFunction_MathFloatGreater(B=0, A=@n2.Value) {is=struct RigVMUnitNode, template="Greater::Execute(in A,in B,out Result)"}.Result
     B=&RigVMFunction_MathFloatGreater(B=@n7.Result, A=@n8.Result) {template="Greater::Execute(in A,in B,out Result)"}.Result
    ).Result
    True=/Script/RigVM.RigVMFunction_MathFloatMax(B=1, A=@"Exponential Softness".Result) {is=struct RigVMUnitNode, template="Maximum::Execute(in A,in B,out Result)"}.Result
   ) {template="DISPATCH_RigVMDispatch_If(in Condition,in True,in False,out Result)"}
   RigVMFunctionReturnNode(
    Result=0
    Length A=&RigVMFunction_MathFloatMul(B=@n9.Result, A=@n3.Result) {template="Multiply::Execute(in A,in B,out Result)"}.Result
    Length B=&RigVMFunction_MathFloatMul(B=@n9.Result, A=@n4.Result) {template="Multiply::Execute(in A,in B,out Result)"}.Result
   )
  ]}
 ContainedGraph {nodes [
   n1 = RigVMFunctionEntryNode()
   n2 = /Script/RigVM.RigVMFunction_MathFloatNegate(Value=/Script/RigVM.RigVMFunction_MathFloatDiv(
     B=@n1.Softness
     A=&RigVMFunction_MathFloatSub(B=@n1.DA, A=@n1.Distance) {template="Subtract::Execute(in A,in B,out Result)"}.Result
    ) {is=struct RigVMUnitNode, template="Divide::Execute(in A,in B,out Result)"}.Result) {is=struct RigVMUnitNode, template="Negate::Execute(in Value,out Result)"}
   n3 = &RigVMFunction_MathFloatMul(
    B=@n1.Softness
    A=&RigVMFunction_MathFloatSub(
     A=1
     B=/Script/RigVM.RigVMFunction_MathFloatExponential(Value=@n2.Result) {is=struct RigVMUnitNode, template="Exponential::Execute(in Value,out Result)"}.Result
    ) {template="Subtract::Execute(in A,in B,out Result)"}.Result
   ) {template="Multiply::Execute(in A,in B,out Result)"}
   RigVMFunctionReturnNode(Result=&RigVMFunction_MathFloatDiv(
     A=@n1.Stretch Distance
     B=&RigVMFunction_MathFloatAdd(B=@n1.DA, A=@n3.Result) {template="Add::Execute(in A,in B,out Result)"}.Result
    ) {template="Divide::Execute(in A,in B,out Result)"}.Result)
  ]}
 ContainedGraph {nodes [
   n1 = RigVMFunctionEntryNode()
   > /Script/ControlRig.RigUnit_TwoBoneIKSimplePerItem(
    DebugSettings=(Scale=10,WorldOffset=(Rotation=(X=0,Y=0,Z=0,W=1),Translation=(X=0,Y=0,Z=0),Scale3D=(X=1,Y=1,Z=1)))
    PoleVectorKind=Location
    PoleVectorSpace=(Type=Bone)
    SecondaryAxisWeight=1
    Weight=1
    bPropagateToChildren=True
    ItemA=@n1.Bone A
    ItemB=@n1.Bone B
    EffectorItem=@n1.Bone C
    ItemALength=@n1.Item ALength
    ItemBLength=@n1.Item BLength
    PrimaryAxis=@n1.Primary Axis
    SecondaryAxis=@n1.Secondary Axis
    Effector=/Script/ControlRig.RigUnit_ProjectTransformToNewParent(bChildInitial=True, bOldParentInitial=True, Child=@n1.Bone C, NewParent=@n1.End Ctrl, OldParent=@n1.End Ctrl) {is=struct RigVMUnitNode}.Transform
    PoleVector=&RigUnit_GetTransform(bInitial=false, Item=@n1.Pole Vector Node).Transform.Translation
    StretchMaximumRatio=@n1.Stretch Maximum Ratio
    StretchStartRatio=@n1.Stretch Start Ratio
    bEnableStretch=@n1.Enable Stretch
   ) {is=struct RigVMUnitNode}
   > DebugLineItemSpace = /Script/RigVM.RigVMFunction_DebugLineNoSpace(
    Color=(R=0.2,G=0.2,B=0.2,A=1)
    DebugDrawSettings=(DepthPriority=SDPG_Foreground,Lifetime=-1)
    WorldOffset=(Rotation=(X=0,Y=0,Z=0,W=1),Translation=(X=0,Y=0,Z=0),Scale3D=(X=1,Y=1,Z=1))
    bEnabled=True
    Thickness=@n1.Line Thickness
    B=&RigUnit_GetTransform(Item=@n1.Bone B).Transform.Translation
    A=&RigUnit_GetTransform(bInitial=false, Item=@n1.Pole Vector Node).Transform.Translation
   ) {is=struct RigVMUnitNode}
   > RigVMFunctionReturnNode()
   n2 = &RigUnit_GetTransform(bInitial=true, Item=@n1.Bone B)
   /Script/RigVM.RigVMFunction_MathVectorScale(
    Factor=&RigVMFunction_MathVectorDistance(A=&RigUnit_GetTransform(bInitial=true, Item=@n1.Bone A).Transform.Translation, B=@n2.Transform.Translation).Result
    Value=@n1.scale
   ) {is=struct RigVMUnitNode, template="Scale::Execute(in Value,in Factor,out Result)"}
   &RigVMFunction_MathVectorScale(
    Factor=&RigVMFunction_MathVectorDistance(A=@n2.Transform.Translation, B=&RigUnit_GetTransform(bInitial=true, Item=@n1.Bone C).Transform.Translation).Result
    Value=@n1.scale
   ) {template="Scale::Execute(in Value,in Factor,out Result)"}
  ]}
 ContainedGraph {nodes [
   n1 = RigVMFunctionEntryNode()
   > n2 = RigVMDispatchNode(Array=@n1.Children) {template="DISPATCH_RigVMDispatch_ArrayIterator(in Array,out Element,out Index,out Count,out Ratio)"}
   n3 = RigVMDispatchNode(Index=@n2.Index, Array=@n1.Parents) {template="DISPATCH_RigVMDispatch_ArrayGetAtIndex(in Array,in Index,out Element)"}
   > &RigUnit_SetTransform(
    Weight=1
    Item=@n2.Element
    bPropagateToChildren=@n1.Propagate to Children
    Value=&RigUnit_ProjectTransformToNewParent(bChildInitial=True, bOldParentInitial=True, NewParent=@n3.Element, OldParent=@n3.Element, Child=@n2.Element).Transform
   )
   RigVMFunctionReturnNode(ExecuteContext=@n2.Completed)
  ]}
 ContainedGraph {nodes [
   RigVMFunctionEntryNode()
   n1 = RigVMVariableNode() {variable=Bones}
   DISPATCH_RigVMDispatch_ArrayGetAtIndex = RigVMDispatchNode(Index=0, Array=@n1.Value) {template="DISPATCH_RigVMDispatch_ArrayGetAtIndex(in Array,in Index,out Element)"}
   > ComputePoleVector = RigVMFunctionReferenceNode(
    Draw Transform=false
    Initial=false
    OffsetFactor=2
    Origin Scale=(X=0,Y=0,Z=0)
    Bone A=@DISPATCH_RigVMDispatch_ArrayGetAtIndex.Element
    Bone B=RigVMDispatchNode(Index=1, Array=@n1.Value) {template="DISPATCH_RigVMDispatch_ArrayGetAtIndex(in Array,in Index,out Element)"}.Element
    Bone C=RigVMDispatchNode(Index=2, Array=@n1.Value) {template="DISPATCH_RigVMDispatch_ArrayGetAtIndex(in Array,in Index,out Element)"}.Element
   )
   n2 = RigVMVariableNode() {variable=Pole Vector Control}
   > &RigUnit_SetTransform(Weight=1, Value=@ComputePoleVector.Transform, Item=@n2.Value)
   n3 = RigVMDispatchNode(Index=2, Array=RigVMVariableNode() {variable=Bones}.Value) {template="DISPATCH_RigVMDispatch_ArrayGetAtIndex(in Array,in Index,out Element)"}
   n4 = RigVMVariableNode() {variable=IK Control}
   > &RigUnit_SetTransform(
    Weight=1
    Value=&RigUnit_ProjectTransformToNewParent(bChildInitial=True, bOldParentInitial=True, NewParent=@n3.Element, OldParent=@n3.Element, Child=@n4.Value).Transform
    Item=@n4.Value
   )
   > RigVMFunctionReturnNode()
   n5 = RigVMDispatchNode(Index=0, Array=RigVMVariableNode() {variable=Bones}.Value) {template="DISPATCH_RigVMDispatch_ArrayGetAtIndex(in Array,in Index,out Element)"}
   &RigUnit_ProjectTransformToNewParent(bChildInitial=True, bOldParentInitial=True, NewParent=@n5.Element, OldParent=@n5.Element, Child=@n2.Value)
  ]}
 ContainedGraph {nodes [
   n1 = RigVMFunctionEntryNode()
   RigUnit_SequenceAggregate = &RigVMFunction_Sequence(ExecuteContext=@n1.ExecuteContext)
   RigVMVariableNode(Value=1000000000, ExecuteContext=@RigUnit_SequenceAggregate.A) {variable=closestDistanceSoFar}
   > n2 = RigVMDispatchNode(Array=@n1.Bones) {template="DISPATCH_RigVMDispatch_ArrayIterator(in Array,out Element,out Index,out Count,out Ratio)"}
   Item = /Script/ControlRig.RigUnit_Item(Item=@n1.Item) {is=struct RigVMUnitNode}
   n3 = &RigVMFunction_MathVectorDistance(
    A=&RigUnit_GetTransform(bInitial=@n1.Initial, Item=@n2.Element).Transform.Translation
    B=&RigUnit_GetTransform(bInitial=@n1.Initial, Item=@Item.Item).Transform.Translation
   )
   n4 = &RigVMFunction_ControlFlowBranch(
    ExecuteContext=@n2.ExecuteContext
    Condition=/Script/RigVM.RigVMFunction_MathFloatLessEqual(A=@n3.Result, B=RigVMVariableNode() {variable=closestDistanceSoFar}.Value) {is=struct RigVMUnitNode, template="LessEqual::Execute(in A,in B,out Result)"}.Result
   )
   RigVMVariableNode(ExecuteContext=@n4.True, Value=@n3.Result) {variable=closestDistanceSoFar}
   > RigVMVariableNode(Value=@n2.Element) {variable=closestBone}
   RigVMFunctionReturnNode(ExecuteContext=@RigUnit_SequenceAggregate.B, Value=RigVMVariableNode() {variable=closestBone}.Value)
  ]}
 ContainedGraph {nodes [
   n1 = RigVMFunctionEntryNode()
   RigUnit_SequenceAggregate_1 = &RigVMFunction_Sequence(ExecuteContext=@n1.ExecuteContext)
   RigUnit_SequenceAggregate = RigVMAggregateNode(ExecuteContext=@RigUnit_SequenceAggregate_1.A)
   Greater = /Script/RigVM.RigVMFunction_MathDoubleGreater(B=0, A=@n1.Mirror Plane.X) {is=struct RigVMUnitNode, template="Greater::Execute(in A,in B,out Result)"}
   BranchNode_4 = &RigVMFunction_ControlFlowBranch(Condition=@Greater.Result, ExecuteContext=@RigUnit_SequenceAggregate.A)
   n2 = RigVMVariableNode() {variable=vector pos}
   RigVMVariableNode(
    Value=(X=0,Y=0,Z=0)
    ExecuteContext=@BranchNode_4.True
    Value.X=&RigUnit_GetTransform(bInitial=true, Item=@n1.Item).Transform.Translation.X
    Value.Y=@n2.Value.Y
    Value.Z=@n2.Value.Z
   ) {variable=vector pos}
   n3 = &RigVMFunction_MathDoubleGreater(B=0, A=@n1.Mirror Plane.Y) {template="Greater::Execute(in A,in B,out Result)"}
   BranchNode_1_1 = &RigVMFunction_ControlFlowBranch(Condition=@n3.Result, ExecuteContext=@RigUnit_SequenceAggregate.B)
   n4 = RigVMVariableNode() {variable=vector pos}
   RigVMVariableNode(
    Value=(X=0,Y=0,Z=0)
    ExecuteContext=@BranchNode_1_1.True
    Value.Y=&RigUnit_GetTransform(bInitial=true, Item=@n1.Item).Transform.Translation.Y
    Value.X=@n4.Value.X
    Value.Z=@n4.Value.Z
   ) {variable=vector pos}
   n5 = &RigVMFunction_MathDoubleGreater(B=0, A=@n1.Mirror Plane.Z) {template="Greater::Execute(in A,in B,out Result)"}
   BranchNode_2_1 = &RigVMFunction_ControlFlowBranch(Condition=@n5.Result, ExecuteContext=@RigUnit_SequenceAggregate.C)
   n6 = RigVMVariableNode() {variable=vector pos}
   RigVMVariableNode(
    Value=(X=0,Y=0,Z=0)
    ExecuteContext=@BranchNode_2_1.True
    Value.Z=&RigUnit_GetTransform(bInitial=true, Item=@n1.Item).Transform.Translation.Z
    Value.X=@n6.Value.X
    Value.Y=@n6.Value.Y
   ) {variable=vector pos}
   Cast = &RigVMFunction_MathBoolToFloat(Value=@Greater.Result)
   Dot__ = &RigVMFunction_MathVectorDot(
    B=(X=1,Y=1,Z=0)
    B.X=@Cast.Result
    B.Y=&RigVMFunction_MathBoolToFloat(Value=@n3.Result).Result
    B.Z=&RigVMFunction_MathBoolToFloat(Value=@n5.Result).Result
    A=RigVMVariableNode() {variable=vector pos}.Value
   ) {template="Dot,|::Execute(in A,in B,out Result)"}
   Less = &RigVMFunction_MathDoubleLess(B=0, A=@Dot__.Result) {template="Less::Execute(in A,in B,out Result)"}
   RigVMVariableNode(Value=@Less.Result, ExecuteContext=@RigUnit_SequenceAggregate.D) {variable=is on right}
   RigVMFunctionReturnNode(ExecuteContext=@RigUnit_SequenceAggregate_1.B, Right Side=RigVMVariableNode() {variable=is on right}.Value)
  ]}
 ContainedGraph {nodes [
   n1 = RigVMFunctionEntryNode()
   RigUnit_SequenceAggregate = &RigVMFunction_Sequence(ExecuteContext=@n1.ExecuteContext)
   RigUnit_SequenceAggregate_1 = &RigVMFunction_Sequence(ExecuteContext=@RigUnit_SequenceAggregate.B)
   RigUnit_SequenceAggregate_2 = &RigVMFunction_Sequence(ExecuteContext=@RigUnit_SequenceAggregate_1.B)
   RigVMFunctionReturnNode(A=@RigUnit_SequenceAggregate.A, B=@RigUnit_SequenceAggregate_1.A, C=@RigUnit_SequenceAggregate_2.A, D=@RigUnit_SequenceAggregate_2.B)
  ]}
 ContainedGraph {nodes [
   n1 = RigVMFunctionEntryNode()
   RigUnit_SequenceAggregate_1 = &RigVMFunction_Sequence(ExecuteContext=@n1.ExecuteContext)
   RigUnit_SequenceAggregate = RigVMAggregateNode(ExecuteContext=@RigUnit_SequenceAggregate_1.A)
   Greater = &RigVMFunction_MathDoubleGreater(B=0, A=@n1.Mirror Plane.X) {template="Greater::Execute(in A,in B,out Result)"}
   BranchNode_4 = &RigVMFunction_ControlFlowBranch(Condition=@Greater.Result, ExecuteContext=@RigUnit_SequenceAggregate.A)
   n2 = RigVMVariableNode() {variable=vector pos}
   RigVMVariableNode(
    Value=(X=0,Y=0,Z=0)
    ExecuteContext=@BranchNode_4.True
    Value.X=&RigUnit_GetTransform(bInitial=true, Item=@n1.Item).Transform.Translation.X
    Value.Y=@n2.Value.Y
    Value.Z=@n2.Value.Z
   ) {variable=vector pos}
   n3 = &RigVMFunction_MathDoubleGreater(B=0, A=@n1.Mirror Plane.Y) {template="Greater::Execute(in A,in B,out Result)"}
   BranchNode_1_1 = &RigVMFunction_ControlFlowBranch(Condition=@n3.Result, ExecuteContext=@RigUnit_SequenceAggregate.B)
   n4 = RigVMVariableNode() {variable=vector pos}
   RigVMVariableNode(
    Value=(X=0,Y=0,Z=0)
    ExecuteContext=@BranchNode_1_1.True
    Value.Y=&RigUnit_GetTransform(bInitial=true, Item=@n1.Item).Transform.Translation.Y
    Value.X=@n4.Value.X
    Value.Z=@n4.Value.Z
   ) {variable=vector pos}
   n5 = &RigVMFunction_MathDoubleGreater(B=0, A=@n1.Mirror Plane.Z) {template="Greater::Execute(in A,in B,out Result)"}
   BranchNode_2_1 = &RigVMFunction_ControlFlowBranch(Condition=@n5.Result, ExecuteContext=@RigUnit_SequenceAggregate.C)
   n6 = RigVMVariableNode() {variable=vector pos}
   RigVMVariableNode(
    Value=(X=0,Y=0,Z=0)
    ExecuteContext=@BranchNode_2_1.True
    Value.Z=&RigUnit_GetTransform(bInitial=true, Item=@n1.Item).Transform.Translation.Z
    Value.X=@n6.Value.X
    Value.Y=@n6.Value.Y
   ) {variable=vector pos}
   Cast = &RigVMFunction_MathBoolToFloat(Value=@Greater.Result)
   Dot__ = &RigVMFunction_MathVectorDot(
    B=(X=1,Y=1,Z=0)
    B.X=@Cast.Result
    B.Y=&RigVMFunction_MathBoolToFloat(Value=@n3.Result).Result
    B.Z=&RigVMFunction_MathBoolToFloat(Value=@n5.Result).Result
    A=RigVMVariableNode() {variable=vector pos}.Value
   ) {template="Dot,|::Execute(in A,in B,out Result)"}
   Less = &RigVMFunction_MathDoubleLess(B=0, A=@Dot__.Result) {template="Less::Execute(in A,in B,out Result)"}
   RigVMVariableNode(Value=@Less.Result, ExecuteContext=@RigUnit_SequenceAggregate.D) {variable=is on right}
   RigVMFunctionReturnNode(ExecuteContext=@RigUnit_SequenceAggregate_1.B, Right Side=RigVMVariableNode() {variable=is on right}.Value)
  ]}
 ContainedGraph {nodes [
   n1 = RigVMFunctionEntryNode()
   RigUnit_SequenceAggregate = &RigVMFunction_Sequence(ExecuteContext=@n1.ExecuteContext)
   RigUnit_SequenceAggregate_1 = &RigVMFunction_Sequence(ExecuteContext=@RigUnit_SequenceAggregate.B)
   RigUnit_SequenceAggregate_2 = &RigVMFunction_Sequence(ExecuteContext=@RigUnit_SequenceAggregate_1.B)
   RigVMFunctionReturnNode(A=@RigUnit_SequenceAggregate.A, B=@RigUnit_SequenceAggregate_1.A, C=@RigUnit_SequenceAggregate_2.A, D=@RigUnit_SequenceAggregate_2.B)
  ]}
 ContainedGraph {nodes [
   n1 = RigVMFunctionEntryNode()
   RigUnit_SequenceAggregate_1 = &RigVMFunction_Sequence(ExecuteContext=@n1.ExecuteContext)
   RigUnit_SequenceAggregate = RigVMAggregateNode(ExecuteContext=@RigUnit_SequenceAggregate_1.A)
   Greater = &RigVMFunction_MathDoubleGreater(B=0, A=@n1.Mirror Plane.X) {template="Greater::Execute(in A,in B,out Result)"}
   BranchNode_4 = &RigVMFunction_ControlFlowBranch(Condition=@Greater.Result, ExecuteContext=@RigUnit_SequenceAggregate.A)
   n2 = RigVMVariableNode() {variable=vector pos}
   RigVMVariableNode(
    Value=(X=0,Y=0,Z=0)
    ExecuteContext=@BranchNode_4.True
    Value.X=&RigUnit_GetTransform(bInitial=true, Item=@n1.Item).Transform.Translation.X
    Value.Y=@n2.Value.Y
    Value.Z=@n2.Value.Z
   ) {variable=vector pos}
   n3 = &RigVMFunction_MathDoubleGreater(B=0, A=@n1.Mirror Plane.Y) {template="Greater::Execute(in A,in B,out Result)"}
   BranchNode_1_1 = &RigVMFunction_ControlFlowBranch(Condition=@n3.Result, ExecuteContext=@RigUnit_SequenceAggregate.B)
   n4 = RigVMVariableNode() {variable=vector pos}
   RigVMVariableNode(
    Value=(X=0,Y=0,Z=0)
    ExecuteContext=@BranchNode_1_1.True
    Value.Y=&RigUnit_GetTransform(bInitial=true, Item=@n1.Item).Transform.Translation.Y
    Value.X=@n4.Value.X
    Value.Z=@n4.Value.Z
   ) {variable=vector pos}
   n5 = &RigVMFunction_MathDoubleGreater(B=0, A=@n1.Mirror Plane.Z) {template="Greater::Execute(in A,in B,out Result)"}
   BranchNode_2_1 = &RigVMFunction_ControlFlowBranch(Condition=@n5.Result, ExecuteContext=@RigUnit_SequenceAggregate.C)
   n6 = RigVMVariableNode() {variable=vector pos}
   RigVMVariableNode(
    Value=(X=0,Y=0,Z=0)
    ExecuteContext=@BranchNode_2_1.True
    Value.Z=&RigUnit_GetTransform(bInitial=true, Item=@n1.Item).Transform.Translation.Z
    Value.X=@n6.Value.X
    Value.Y=@n6.Value.Y
   ) {variable=vector pos}
   Cast = &RigVMFunction_MathBoolToFloat(Value=@Greater.Result)
   Dot__ = &RigVMFunction_MathVectorDot(
    B=(X=1,Y=1,Z=0)
    B.X=@Cast.Result
    B.Y=&RigVMFunction_MathBoolToFloat(Value=@n3.Result).Result
    B.Z=&RigVMFunction_MathBoolToFloat(Value=@n5.Result).Result
    A=RigVMVariableNode() {variable=vector pos}.Value
   ) {template="Dot,|::Execute(in A,in B,out Result)"}
   Less = &RigVMFunction_MathDoubleLess(B=0, A=@Dot__.Result) {template="Less::Execute(in A,in B,out Result)"}
   RigVMVariableNode(Value=@Less.Result, ExecuteContext=@RigUnit_SequenceAggregate.D) {variable=is on right}
   RigVMFunctionReturnNode(ExecuteContext=@RigUnit_SequenceAggregate_1.B, Right Side=RigVMVariableNode() {variable=is on right}.Value)
  ]}
 ContainedGraph {nodes [
   n1 = RigVMFunctionEntryNode()
   RigUnit_SequenceAggregate = &RigVMFunction_Sequence(ExecuteContext=@n1.ExecuteContext)
   RigUnit_SequenceAggregate_1 = &RigVMFunction_Sequence(ExecuteContext=@RigUnit_SequenceAggregate.B)
   RigUnit_SequenceAggregate_2 = &RigVMFunction_Sequence(ExecuteContext=@RigUnit_SequenceAggregate_1.B)
   RigVMFunctionReturnNode(A=@RigUnit_SequenceAggregate.A, B=@RigUnit_SequenceAggregate_1.A, C=@RigUnit_SequenceAggregate_2.A, D=@RigUnit_SequenceAggregate_2.B)
  ]}
 ContainedGraph {nodes [
   n1 = RigVMFunctionEntryNode()
   RigUnit_SequenceAggregate_1 = &RigVMFunction_Sequence(ExecuteContext=@n1.ExecuteContext)
   RigUnit_SequenceAggregate = RigVMAggregateNode(ExecuteContext=@RigUnit_SequenceAggregate_1.A)
   Greater = &RigVMFunction_MathDoubleGreater(B=0, A=@n1.Mirror Plane.X) {template="Greater::Execute(in A,in B,out Result)"}
   BranchNode_4 = &RigVMFunction_ControlFlowBranch(Condition=@Greater.Result, ExecuteContext=@RigUnit_SequenceAggregate.A)
   n2 = &RigVMFunction_MathDoubleGreater(B=0, A=@n1.Mirror Plane.Y) {template="Greater::Execute(in A,in B,out Result)"}
   BranchNode_1_1 = &RigVMFunction_ControlFlowBranch(ExecuteContext=@BranchNode_4.Completed, Condition=@n2.Result)
   n3 = &RigVMFunction_MathDoubleGreater(B=0, A=@n1.Mirror Plane.Z) {template="Greater::Execute(in A,in B,out Result)"}
   BranchNode_2_1 = &RigVMFunction_ControlFlowBranch(ExecuteContext=@BranchNode_1_1.Completed, Condition=@n3.Result)
   Cast = &RigVMFunction_MathBoolToFloat(Value=@Greater.Result)
   Dot__ = &RigVMFunction_MathVectorDot(
    B=(X=1,Y=1,Z=0)
    B.X=@Cast.Result
    B.Y=&RigVMFunction_MathBoolToFloat(Value=@n2.Result).Result
    B.Z=&RigVMFunction_MathBoolToFloat(Value=@n3.Result).Result
    A=RigVMVariableNode() {variable=vector pos}.Value
   ) {template="Dot,|::Execute(in A,in B,out Result)"}
   Less = &RigVMFunction_MathDoubleLess(B=0, A=@Dot__.Result) {template="Less::Execute(in A,in B,out Result)"}
   RigVMVariableNode(ExecuteContext=@BranchNode_2_1.Completed, Value=@Less.Result) {variable=is on right}
   n4 = RigVMVariableNode() {variable=vector pos}
   RigVMVariableNode(
    Value=(X=0,Y=0,Z=0)
    ExecuteContext=@BranchNode_2_1.True
    Value.Z=&RigUnit_GetTransform(bInitial=true, Item=@n1.Item).Transform.Translation.Z
    Value.X=@n4.Value.X
    Value.Y=@n4.Value.Y
   ) {variable=vector pos}
   Dot___1_1_1_1_1_2 = &RigVMFunction_MathVectorDot(
    B=(X=1,Y=1,Z=0)
    B.Y=&RigVMFunction_MathBoolToFloat(Value=@n2.Result).Result
    B.Z=&RigVMFunction_MathBoolToFloat(Value=@n3.Result).Result
    B.X=&RigVMFunction_MathBoolToFloat(Value=@Greater.Result).Result
    A=RigVMVariableNode() {variable=vector pos}.Value
   ) {template="Dot,|::Execute(in A,in B,out Result)"}
   > RigVMVariableNode(Value=&RigVMFunction_MathDoubleLess(B=0, A=@Dot___1_1_1_1_1_2.Result) {template="Less::Execute(in A,in B,out Result)"}.Result) {variable=is on right}
   n5 = RigVMVariableNode() {variable=vector pos}
   RigVMVariableNode(
    Value=(X=0,Y=0,Z=0)
    ExecuteContext=@BranchNode_1_1.True
    Value.Y=&RigUnit_GetTransform(bInitial=true, Item=@n1.Item).Transform.Translation.Y
    Value.X=@n5.Value.X
    Value.Z=@n5.Value.Z
   ) {variable=vector pos}
   Dot___1_1_1_1_1_1 = &RigVMFunction_MathVectorDot(
    B=(X=1,Y=1,Z=0)
    B.Y=&RigVMFunction_MathBoolToFloat(Value=@n2.Result).Result
    B.Z=&RigVMFunction_MathBoolToFloat(Value=@n3.Result).Result
    B.X=&RigVMFunction_MathBoolToFloat(Value=@Greater.Result).Result
    A=RigVMVariableNode() {variable=vector pos}.Value
   ) {template="Dot,|::Execute(in A,in B,out Result)"}
   > RigVMVariableNode(Value=&RigVMFunction_MathDoubleLess(B=0, A=@Dot___1_1_1_1_1_1.Result) {template="Less::Execute(in A,in B,out Result)"}.Result) {variable=is on right}
   n6 = RigVMVariableNode() {variable=vector pos}
   RigVMVariableNode(
    Value=(X=0,Y=0,Z=0)
    ExecuteContext=@BranchNode_4.True
    Value.X=&RigUnit_GetTransform(bInitial=true, Item=@n1.Item).Transform.Translation.X
    Value.Y=@n6.Value.Y
    Value.Z=@n6.Value.Z
   ) {variable=vector pos}
   Dot___1_1_1_1_1 = &RigVMFunction_MathVectorDot(
    B=(X=1,Y=1,Z=0)
    B.Y=&RigVMFunction_MathBoolToFloat(Value=@n2.Result).Result
    B.Z=&RigVMFunction_MathBoolToFloat(Value=@n3.Result).Result
    B.X=&RigVMFunction_MathBoolToFloat(Value=@Greater.Result).Result
    A=RigVMVariableNode() {variable=vector pos}.Value
   ) {template="Dot,|::Execute(in A,in B,out Result)"}
   > RigVMVariableNode(Value=&RigVMFunction_MathDoubleLess(B=0, A=@Dot___1_1_1_1_1.Result) {template="Less::Execute(in A,in B,out Result)"}.Result) {variable=is on right}
   n7 = &RigVMFunction_MathDoubleLess(B=0, A=@n1.Mirror Plane.X) {template="Less::Execute(in A,in B,out Result)"}
   BranchNode_4_1 = &RigVMFunction_ControlFlowBranch(Condition=@n7.Result, ExecuteContext=@RigUnit_SequenceAggregate.D)
   n8 = RigVMVariableNode() {variable=vector pos}
   RigVMVariableNode(
    Value=(X=0,Y=0,Z=0)
    ExecuteContext=@BranchNode_4_1.True
    Value.X=&RigUnit_GetTransform(bInitial=true, Item=@n1.Item).Transform.Translation.X
    Value.Y=@n8.Value.Y
    Value.Z=@n8.Value.Z
   ) {variable=vector pos}
   n9 = &RigVMFunction_MathDoubleLess(B=0, A=@n1.Mirror Plane.Y) {template="Less::Execute(in A,in B,out Result)"}
   n10 = &RigVMFunction_MathDoubleLess(B=0, A=@n1.Mirror Plane.Z) {template="Less::Execute(in A,in B,out Result)"}
   Dot___1_1_1_1 = &RigVMFunction_MathVectorDot(
    B=(X=1,Y=1,Z=0)
    B.Y=&RigVMFunction_MathBoolToFloat(Value=@n9.Result).Result
    B.Z=&RigVMFunction_MathBoolToFloat(Value=@n10.Result).Result
    B.X=&RigVMFunction_MathBoolToFloat(Value=@n7.Result).Result
    A=RigVMVariableNode() {variable=vector pos}.Value
   ) {template="Dot,|::Execute(in A,in B,out Result)"}
   > RigVMVariableNode(Value=&RigVMFunction_MathDoubleGreater(B=0, A=@Dot___1_1_1_1.Result) {template="Greater::Execute(in A,in B,out Result)"}.Result) {variable=is on right}
   BranchNode_1_1_1 = &RigVMFunction_ControlFlowBranch(Condition=@n9.Result, ExecuteContext=@RigUnit_SequenceAggregate.E)
   n11 = RigVMVariableNode() {variable=vector pos}
   RigVMVariableNode(
    Value=(X=0,Y=0,Z=0)
    ExecuteContext=@BranchNode_1_1_1.True
    Value.Y=&RigUnit_GetTransform(bInitial=true, Item=@n1.Item).Transform.Translation.Y
    Value.X=@n11.Value.X
    Value.Z=@n11.Value.Z
   ) {variable=vector pos}
   Dot___1_1 = &RigVMFunction_MathVectorDot(
    B=(X=1,Y=1,Z=0)
    B.Y=&RigVMFunction_MathBoolToFloat(Value=@n9.Result).Result
    B.Z=&RigVMFunction_MathBoolToFloat(Value=@n10.Result).Result
    B.X=&RigVMFunction_MathBoolToFloat(Value=@n7.Result).Result
    A=RigVMVariableNode() {variable=vector pos}.Value
   ) {template="Dot,|::Execute(in A,in B,out Result)"}
   > RigVMVariableNode(Value=&RigVMFunction_MathDoubleGreater(B=0, A=@Dot___1_1.Result) {template="Greater::Execute(in A,in B,out Result)"}.Result) {variable=is on right}
   BranchNode_2_1_1 = &RigVMFunction_ControlFlowBranch(Condition=@n10.Result, ExecuteContext=@RigUnit_SequenceAggregate.F)
   n12 = RigVMVariableNode() {variable=vector pos}
   RigVMVariableNode(
    Value=(X=0,Y=0,Z=0)
    ExecuteContext=@BranchNode_2_1_1.True
    Value.Z=&RigUnit_GetTransform(bInitial=true, Item=@n1.Item).Transform.Translation.Z
    Value.X=@n12.Value.X
    Value.Y=@n12.Value.Y
   ) {variable=vector pos}
   Dot___1_1_1 = &RigVMFunction_MathVectorDot(
    B=(X=1,Y=1,Z=0)
    B.Y=&RigVMFunction_MathBoolToFloat(Value=@n9.Result).Result
    B.Z=&RigVMFunction_MathBoolToFloat(Value=@n10.Result).Result
    B.X=&RigVMFunction_MathBoolToFloat(Value=@n7.Result).Result
    A=RigVMVariableNode() {variable=vector pos}.Value
   ) {template="Dot,|::Execute(in A,in B,out Result)"}
   > RigVMVariableNode(Value=&RigVMFunction_MathDoubleGreater(B=0, A=@Dot___1_1_1.Result) {template="Greater::Execute(in A,in B,out Result)"}.Result) {variable=is on right}
   RigVMFunctionReturnNode(ExecuteContext=@RigUnit_SequenceAggregate_1.B, Right Side=RigVMVariableNode() {variable=is on right}.Value)
   &RigVMFunction_MathDoubleLess(B=0, A=@Dot___1_1.Result) {template="Less::Execute(in A,in B,out Result)"}
   &RigVMFunction_MathDoubleLess(B=0, A=@Dot___1_1_1.Result) {template="Less::Execute(in A,in B,out Result)"}
   &RigVMFunction_MathDoubleLess(B=0, A=@Dot___1_1_1_1.Result) {template="Less::Execute(in A,in B,out Result)"}
  ]}
 ContainedGraph {nodes [
   n1 = RigVMFunctionEntryNode()
   RigUnit_SequenceAggregate = &RigVMFunction_Sequence(ExecuteContext=@n1.ExecuteContext)
   RigUnit_SequenceAggregate_1 = &RigVMFunction_Sequence(ExecuteContext=@RigUnit_SequenceAggregate.B)
   RigUnit_SequenceAggregate_2 = &RigVMFunction_Sequence(ExecuteContext=@RigUnit_SequenceAggregate_1.B)
   RigUnit_SequenceAggregate_3 = &RigVMFunction_Sequence(ExecuteContext=@RigUnit_SequenceAggregate_2.B)
   RigUnit_SequenceAggregate_4 = &RigVMFunction_Sequence(ExecuteContext=@RigUnit_SequenceAggregate_3.B)
   RigUnit_SequenceAggregate_5 = &RigVMFunction_Sequence(ExecuteContext=@RigUnit_SequenceAggregate_4.B)
   RigVMFunctionReturnNode(
    A=@RigUnit_SequenceAggregate.A
    B=@RigUnit_SequenceAggregate_1.A
    C=@RigUnit_SequenceAggregate_2.A
    D=@RigUnit_SequenceAggregate_3.A
    E=@RigUnit_SequenceAggregate_4.A
    F=@RigUnit_SequenceAggregate_5.A
    G=@RigUnit_SequenceAggregate_5.B
   )
  ]}
 Create Twist _ContainedGraph {nodes [
   n1 = RigVMFunctionEntryNode()
   > For_Each = RigVMDispatchNode(Array=/Script/ControlRig.RigUnit_ResolveArrayConnector(Connector=(Type=Connector,Name="Twist Bones"), SkipSocket=true) {is=struct RigVMUnitNode}.Result) {template="DISPATCH_RigVMDispatch_ArrayIterator(in Array,out Element,out Index,out Count,out Ratio)"}
   > FindClosestBone_1 = RigVMFunctionReferenceNode(Initial=true, Bones=@n1.Template Bones, Item=@For_Each.Element)
   Clone = RigVMDispatchNode(Array=@n1.Template Bones) {template="DISPATCH_RigVMDispatch_ArrayClone(in Array,out Clone)"}
   > n2 = RigVMDispatchNode(
    Array=@Clone.Clone
    Index=RigVMDispatchNode(Array=@Clone.Clone, Element=@FindClosestBone_1.Value) {template="DISPATCH_RigVMDispatch_ArrayFind(in Array,in Element,out Index,out Success)"}.Index
   ) {template="DISPATCH_RigVMDispatch_ArrayRemove(io Array,in Index)"}
   > FindClosestBone_0 = RigVMFunctionReferenceNode(Initial=true, Item=@For_Each.Element, Bones=@n2.Array)
   RigVMFunction_MathVectorDistance = &RigVMFunction_MathVectorDistance(
    A=&RigUnit_GetTransform(bInitial=False, Item=@FindClosestBone_1.Value).Transform.Translation
    B=&RigUnit_GetTransform(bInitial=False, Item=@For_Each.Element).Transform.Translation
   )
   n3 = &RigVMFunction_MathVectorDistance(
    A=&RigUnit_GetTransform(bInitial=False, Item=@FindClosestBone_1.Value).Transform.Translation
    B=&RigUnit_GetTransform(bInitial=False, Item=@FindClosestBone_0.Value).Transform.Translation
   )
   Divide = &RigVMFunction_MathDoubleDiv(A=@RigVMFunction_MathVectorDistance.Result, B=@n3.Result)
   > Set_Metadata = RigVMDispatchNode(Name=Twist Amount A, NameSpace=Self, Value=@Divide.Result, Item=@For_Each.Element) {template="DISPATCH_RigDispatch_SetMetadata(in Item,in Name,in NameSpace,in Value,out Success)"}
   n4 = &RigUnit_GetTransform(bInitial=False, Item=@For_Each.Element)
   n5 = &RigUnit_GetTransform(bInitial=False, Item=@FindClosestBone_0.Value)
   > Set_Metadata_1 = RigVMDispatchNode(
    Name=Twist Amount B
    NameSpace=Self
    Value=&RigVMFunction_MathDoubleDiv(A=&RigVMFunction_MathVectorDistance(A=@n4.Transform.Translation, B=@n5.Transform.Translation).Result, B=@n3.Result).Result
    Item=@For_Each.Element
   ) {template="DISPATCH_RigDispatch_SetMetadata(in Item,in Name,in NameSpace,in Value,out Success)"}
   > Set_Metadata_2 = RigVMDispatchNode(Name=Twist Item A, NameSpace=Self, Value=@FindClosestBone_1.Value, Item=@For_Each.Element) {template="DISPATCH_RigDispatch_SetMetadata(in Item,in Name,in NameSpace,in Value,out Success)"}
   > Set_Metadata_3 = RigVMDispatchNode(Name=Twist Item B, NameSpace=Self, Value=@FindClosestBone_0.Value, Item=@For_Each.Element) {template="DISPATCH_RigDispatch_SetMetadata(in Item,in Name,in NameSpace,in Value,out Success)"}
   RigVMFunctionReturnNode(ExecuteContext=@For_Each.Completed)
  ]}
 New Function_ContainedGraph {nodes [
   n1 = RigVMFunctionEntryNode()
   n2 = /Script/ControlRig.RigUnit_AimBoneMath(
    DebugSettings=(bEnabled=False,Scale=10,WorldOffset=(Rotation=(X=0,Y=0,Z=0,W=1),Translation=(X=0,Y=0,Z=0),Scale3D=(X=1,Y=1,Z=1)))
    Primary=(Weight=1,Axis=(X=1,Y=0,Z=0),Target=(X=0,Y=0,Z=0),Kind=Location,Space=(Type=Control,Name="None"))
    Secondary=(Weight=1,Axis=(X=0,Y=-1,Z=0),Target=(X=0,Y=0,Z=1),Kind=Direction,Space=(Type=Control,Name="None"))
    Weight=1
    Primary.Axis=@n1.Axis
    Secondary.Axis=@n1.Secondary Axis
    Secondary.Target=@n1.Secondary Direction
    Secondary.Space=@n1.Target item
    InputTransform=&RigUnit_GetTransform(bInitial=false, Item=@n1.Item).Transform
    Primary.Target=&RigUnit_GetTransform(bInitial=False, Item=@n1.Target item).Transform.Translation
   ) {is=struct RigVMUnitNode}
   > &RigUnit_SetTransform(Weight=1, bInitial=False, Value=@n2.Result, Item=@n1.Item)
   > RigVMFunction_DebugTransformMutableNoSpace = /Script/RigVM.RigVMFunction_DebugTransformMutableNoSpace(
    Color=(R=1,G=0,B=0,A=1)
    DebugDrawSettings=(DepthPriority=SDPG_Foreground,Lifetime=-1)
    Mode=Axes
    Scale=50
    Thickness=3
    WorldOffset=(Rotation=(X=0,Y=0,Z=0,W=1),Translation=(X=0,Y=0,Z=0),Scale3D=(X=1,Y=1,Z=1))
    Transform=@n2.Result
    bEnabled=@n1.Debug
   ) {is=struct RigVMUnitNode}
   > RigVMFunctionReturnNode(Result=@n2.Result)
  ]}
 New Function_ContainedGraph {nodes [
   n1 = RigVMFunctionEntryNode()
   DISPATCH_RigVMDispatch_ArrayGetNum = RigVMDispatchNode(Array=&RigUnit_CollectionChildrenArray(TypeToSearch=Socket, bDefaultChildren=True, bIncludeParent=False, bRecursive=False, Parent=@n1.Item).Items) {template="DISPATCH_RigVMDispatch_ArrayGetNum(in Array,out Num)"}
   Greater = /Script/RigVM.RigVMFunction_MathIntGreater(B=0, A=@DISPATCH_RigVMDispatch_ArrayGetNum.Num) {is=struct RigVMUnitNode, template="Greater::Execute(in A,in B,out Result)"}
   > RigVMFunctionReturnNode(Result=@Greater.Result)
  ]}
 New Function_ContainedGraph {nodes [
   n1 = RigVMFunctionEntryNode()
   > RigVMFunction_Sequence_1 = RigVMAggregateNode()
   Get_Module_Metadata_4 = RigVMDispatchNode(Default=(X=0,Y=0,Z=0), Name=Mirror Axis, NameSpace=Root) {template="DISPATCH_RigDispatch_GetModuleMetadata(in Name,in NameSpace,in Default,out Value,out Found)"}
   RigVMVariableNode(Value=@Get_Module_Metadata_4.Value, ExecuteContext=@RigVMFunction_Sequence_1.A) {variable=Mirror Axis}
   > RigVMVariableNode(Value=RigVMDispatchNode(Default=(Type=Bone,Name="None"), Name=Effector Null, NameSpace=Self) {template="DISPATCH_RigDispatch_GetModuleMetadata(in Name,in NameSpace,in Default,out Value,out Found)"}.Value) {variable=Effector Null}
   Get_Module_Metadata_3 = RigVMDispatchNode(Default=(Type=Bone,Name="None"), Name=Aligned Null, NameSpace=Self) {template="DISPATCH_RigDispatch_GetModuleMetadata(in Name,in NameSpace,in Default,out Value,out Found)"}
   > RigVMVariableNode(Value=@Get_Module_Metadata_3.Value) {variable=Aligned Null}
   Get_Module_Metadata_5 = RigVMDispatchNode(Default=(), Name=Rig Bones, NameSpace=Self) {template="DISPATCH_RigDispatch_GetModuleMetadata(in Name,in NameSpace,in Default,out Value,out Found)"}
   > RigVMVariableNode(Value=@Get_Module_Metadata_5.Value) {variable=Rig Bones}
   > RigVMVariableNode(Value=RigVMDispatchNode(Default=(), Name=Aim Follow Bones, NameSpace=Self) {template="DISPATCH_RigDispatch_GetModuleMetadata(in Name,in NameSpace,in Default,out Value,out Found)"}.Value) {variable=Aim Follow Bones}
   > RigVMVariableNode(Value=RigVMDispatchNode(Default=(Type=Bone,Name="None"), Name=Effector Ori, NameSpace=Self) {template="DISPATCH_RigDispatch_GetModuleMetadata(in Name,in NameSpace,in Default,out Value,out Found)"}.Value) {variable=Effector Ori}
   n2 = &RigVMFunction_ControlFlowBranch(ExecuteContext=@RigVMFunction_Sequence_1.B, Condition=RigVMVariableNode() {variable=FK_or_IK}.Value)
   RigVMFunction_Sequence = RigVMAggregateNode(ExecuteContext=@n2.False)
   n3 = &RigVMFunction_ControlFlowBranch(ExecuteContext=@RigVMFunction_Sequence.A, Condition=RigVMVariableNode() {variable=FK Space}.Value)
   RigVMFunction_ControlFlowBranch = &RigVMFunction_ControlFlowBranch(ExecuteContext=@n3.Completed, Condition=RigVMVariableNode() {variable=local}.Value)
   n4 = RigVMDispatchNode(Index=0, Array=RigVMVariableNode() {variable=FK Controls}.Value) {template="DISPATCH_RigVMDispatch_ArrayGetAtIndex(in Array,in Index,out Element)"}
   n5 = &RigUnit_HierarchyGetParent(bDefaultParent=True, Child=@n4.Element)
   n6 = RigVMVariableNode() {variable=FK Arm Rotation Space}
   Set Transform_1 = &RigUnit_SetRotation(
    Space=GlobalSpace
    Weight=1
    bInitial=False
    Item=@n5.Parent
    Value=&RigUnit_ProjectTransformToNewParent(bChildInitial=True, bNewParentInitial=False, bOldParentInitial=True, Child=@n5.Parent, NewParent=@n6.Value, OldParent=@n6.Value).Transform.Rotation
    ExecutePin=@RigVMFunction_ControlFlowBranch.False
   )
   &RigUnit_ParentConstraint(
    AdvancedSettings=(InterpolationType=Average,RotationOrderForFilter=XZY)
    Parents=((Item=(Type=Bone,Name="None"),Weight=1))
    Weight=1
    Parents.0.Item=&RigUnit_HierarchyGetParent(
     bDefaultParent=True
     Child=RigVMDispatchNode(Default=(Type=Bone,Name="None"), Name=Item, NameSpace=Self, Item=@n4.Element) {template="DISPATCH_RigDispatch_GetMetadata(in Item,in Name,in NameSpace,in Default,out Value,out Found)"}.Value
    ).Parent
    Child=@n5.Parent
    ExecutePin=@n3.True
   )
   n7 = RigVMDispatchNode(ExecuteContext=@RigVMFunction_Sequence.B, Array=RigVMVariableNode() {variable=FK Controls}.Value) {template="DISPATCH_RigVMDispatch_ArrayIterator(in Array,out Element,out Index,out Count,out Ratio)"}
   DISPATCH_RigDispatch_GetMetadata = RigVMDispatchNode(Default=(Type=Bone,Name="None"), Name=Item, NameSpace=Self, Item=@n7.Element) {template="DISPATCH_RigDispatch_GetMetadata(in Item,in Name,in NameSpace,in Default,out Value,out Found)"}
   > &RigUnit_ParentConstraint(
    AdvancedSettings=(InterpolationType=Average,RotationOrderForFilter=XZY)
    Parents=((Item=(Type=Bone,Name="None"),Weight=1))
    Weight=1
    Child=@DISPATCH_RigDispatch_GetMetadata.Value
    Parents.0.Item=@n7.Element
   )
   > /Script/ControlRig.RigUnit_SetControlVisibility(bVisible=true, Item=@n7.Element) {is=struct RigVMUnitNode, template="SetControlVisibility::Execute(in Item,in Pattern,in bVisible)"}
   n8 = RigVMVariableNode() {variable=IK Controls}
   Match IK onto FK = RigVMFunctionReferenceNode(
    IK Control=RigVMDispatchNode(Index=0, Array=@n8.Value) {template="DISPATCH_RigVMDispatch_ArrayGetAtIndex(in Array,in Index,out Element)"}.Element
    Pole Vector Control=RigVMDispatchNode(Index=1, Array=@n8.Value) {template="DISPATCH_RigVMDispatch_ArrayGetAtIndex(in Array,in Index,out Element)"}.Element
    ExecuteContext=@RigVMFunction_Sequence.C
    Bones=RigVMVariableNode() {variable=Bones}.Value
   )
   n9 = RigVMDispatchNode(ExecuteContext=@RigVMFunction_Sequence.D, Array=@n8.Value) {template="DISPATCH_RigVMDispatch_ArrayIterator(in Array,out Element,out Index,out Count,out Ratio)"}
   > SetControlVisibility = &RigUnit_SetControlVisibility(bVisible=false, Item=@n9.Element) {template="SetControlVisibility::Execute(in Item,in Pattern,in bVisible)"}
   Get Side Metadata = RigVMFunctionReferenceNode(Item=RigVMDispatchNode(Index=0, Array=RigVMVariableNode() {variable=Bones}.Value) {template="DISPATCH_RigVMDispatch_ArrayGetAtIndex(in Array,in Index,out Element)"}.Element)
   RigVMVariableNode(Value=@"Get Side Metadata".Result, ExecuteContext=@n2.True) {variable=Is on Right Side}
   n10 = RigVMVariableNode() {variable=Bones}
   n11 = RigVMVariableNode() {variable=Bones}
   n12 = RigVMDispatchNode(Index=1, Array=@n11.Value) {template="DISPATCH_RigVMDispatch_ArrayGetAtIndex(in Array,in Index,out Element)"}
   n13 = RigVMDispatchNode(Index=0, Array=@n11.Value) {template="DISPATCH_RigVMDispatch_ArrayGetAtIndex(in Array,in Index,out Element)"}
   n14 = /Script/RigVM.RigVMFunction_MathVectorUnit(Value=/Script/RigVM.RigVMFunction_MathVectorSub(
     A=&RigUnit_GetTransform(Space=LocalSpace, bInitial=true, Item=@n12.Element).Transform.Translation
     B=&RigUnit_GetTransform(Space=LocalSpace, bInitial=true, Item=@n13.Element).Transform.Translation
    ) {is=struct RigVMUnitNode, template="Subtract::Execute(in A,in B,out Result)"}.Result) {is=struct RigVMUnitNode, template="Unit::Execute(in Value,out Result)"}
   n15 = RigVMVariableNode() {variable=Effector Null}
   Softness = RigVMCollapseNode(Maximum=1, Softness=RigVMVariableNode() {variable=Softness}.Value, Bones=RigVMVariableNode() {variable=Bones}.Value, Effector=@n15.Value)
   n16 = RigVMVariableNode() {variable=Enable Softness}
   > IKTwoBone = RigVMFunctionReferenceNode(
    Line Thickness=0.25
    Primary Axis=(X=1,Y=0,Z=0)
    Secondary Axis=(X=0,Y=1,Z=0)
    Stretch Maximum Ratio=2
    Stretch Start Ratio=1
    debug=false
    scale=(X=0,Y=0,Z=0)
    Bone B=RigVMDispatchNode(Index=1, Array=@n10.Value) {template="DISPATCH_RigVMDispatch_ArrayGetAtIndex(in Array,in Index,out Element)"}.Element
    Bone C=RigVMDispatchNode(Index=2, Array=@n10.Value) {template="DISPATCH_RigVMDispatch_ArrayGetAtIndex(in Array,in Index,out Element)"}.Element
    Bone A=RigVMDispatchNode(Index=0, Array=@n10.Value) {template="DISPATCH_RigVMDispatch_ArrayGetAtIndex(in Array,in Index,out Element)"}.Element
    Pole Vector Node=RigVMDispatchNode(Index=1, Array=RigVMVariableNode() {variable=IK Controls}.Value) {template="DISPATCH_RigVMDispatch_ArrayGetAtIndex(in Array,in Index,out Element)"}.Element
    Primary Axis.X=RigVMDispatchNode(
     True=(X=0,Y=0,Z=0)
     True.X=@n14.Result.X
     False=RigVMVariableNode() {variable=Primary Axis}.Value
     Condition=RigVMVariableNode() {variable=Auto Determine Axis}.Value
    ) {template="DISPATCH_RigVMDispatch_If(in Condition,in True,in False,out Result)"}.Result.X
    Item ALength=RigVMDispatchNode(False=0, True=@Softness.Length A, Condition=@n16.Value) {template="DISPATCH_RigVMDispatch_If(in Condition,in True,in False,out Result)"}.Result
    Item BLength=RigVMDispatchNode(False=0, True=@Softness.Length B, Condition=@n16.Value) {template="DISPATCH_RigVMDispatch_If(in Condition,in True,in False,out Result)"}.Result
    Secondary Axis.Y=&RigVMFunction_MathDoubleMul(A=@n14.Result.X, B=RigVMVariableNode() {variable=Mirror Threshold}.Value).Result
    End Ctrl=@n15.Value
    Enable Stretch=RigVMVariableNode() {variable=Enable Stretch}.Value
   )
   > Project to New Parent = RigVMFunctionReferenceNode(Propagate to Children=True, Parents=RigVMVariableNode() {variable=Bones}.Value, Children=RigVMVariableNode() {variable=FK Controls}.Value)
   > n17 = RigVMDispatchNode(Array=RigVMVariableNode() {variable=IK Controls}.Value) {template="DISPATCH_RigVMDispatch_ArrayIterator(in Array,out Element,out Index,out Count,out Ratio)"}
   > &RigUnit_SetControlVisibility(bVisible=true, Item=@n17.Element) {template="SetControlVisibility::Execute(in Item,in Pattern,in bVisible)"}
   n18 = RigVMDispatchNode(ExecuteContext=@n17.Completed, Array=RigVMVariableNode() {variable=FK Controls}.Value) {template="DISPATCH_RigVMDispatch_ArrayIterator(in Array,out Element,out Index,out Count,out Ratio)"}
   > &RigUnit_SetControlVisibility(bVisible=false, Item=@n18.Element) {template="SetControlVisibility::Execute(in Item,in Pattern,in bVisible)"}
   RigVMFunction_Sequence_2 = RigVMAggregateNode(ExecuteContext=@RigVMFunction_Sequence_1.C)
   n19 = &RigVMFunction_ControlFlowBranch(ExecuteContext=@RigVMFunction_Sequence_2.A, Condition=RigVMVariableNode() {variable=Attach Skeleton}.Value)
   DISPATCH_RigVMDispatch_ArrayIterator = RigVMDispatchNode(Array=@n1.Bones, ExecuteContext=@n19.True) {template="DISPATCH_RigVMDispatch_ArrayIterator(in Array,out Element,out Index,out Count,out Ratio)"}
   > FindClosestBone_0 = RigVMFunctionReferenceNode(Initial=true, Item=@DISPATCH_RigVMDispatch_ArrayIterator.Element, Bones=@n1.Driven Bones)
   > &RigUnit_SetTransform(
    Weight=1
    bInitial=False
    Item=@FindClosestBone_0.Value
    Value=&RigUnit_ProjectTransformToNewParent(
     bChildInitial=True
     bNewParentInitial=False
     bOldParentInitial=True
     NewParent=@DISPATCH_RigVMDispatch_ArrayIterator.Element
     OldParent=@DISPATCH_RigVMDispatch_ArrayIterator.Element
     Child=@FindClosestBone_0.Value
    ).Transform
   )
   n20 = &RigVMFunction_ControlFlowBranch(ExecuteContext=@RigVMFunction_Sequence_2.B, Condition=RigVMVariableNode() {variable=Enable PV Follow}.Value)
   /Script/ControlRig.RigUnit_PositionConstraintLocalSpaceOffset(
    Filter=(bX=True,bY=True,bZ=True)
    Parents=((Item=(Type=Bone,Name="None"),Weight=1))
    Weight=1
    bMaintainOffset=True
    ExecutePin=@n20.True
    Parents.0.Item=RigVMVariableNode() {variable=Effector Null}.Value
    Child=RigVMVariableNode() {variable=Effector Ori}.Value
   ) {is=struct RigVMUnitNode}
   > &RigUnit_RotationConstraintLocalSpaceOffset(
    AdvancedSettings=(InterpolationType=Average,RotationOrderForFilter=XZY)
    Parents=((Item=(Type=Bone,Name="None"),Weight=1))
    Weight=1
    bMaintainOffset=false
    Filter=@n1.PV Follow Twist Lock
    Parents.0.Item=RigVMVariableNode() {variable=Effector Null}.Value
    Child=RigVMVariableNode() {variable=Effector Ori}.Value
   )
   n21 = RigVMDispatchNode(Index=0, Array=RigVMVariableNode() {variable=Aim Follow Bones}.Value) {template="DISPATCH_RigVMDispatch_ArrayGetAtIndex(in Array,in Index,out Element)"}
   n22 = RigVMDispatchNode(True=-1, False=RigVMVariableNode() {variable=Mirror Threshold}.Value, Condition=RigVMVariableNode() {variable=Is on Right Side}.Value) {template="DISPATCH_RigVMDispatch_If(in Condition,in True,in False,out Result)"}
   n23 = RigVMVariableNode() {variable=Effector Ori}
   > &RigUnit_AimConstraintLocalSpaceOffset(
    AdvancedSettings=(DebugSettings=(bEnabled=false,Scale=10,WorldOffset=(Rotation=(X=0,Y=0,Z=0,W=1),Translation=(X=0,Y=0,Z=0),Scale3D=(X=1,Y=1,Z=1))),RotationOrderForFilter=XZY)
    Filter=(bX=True,bY=True,bZ=True)
    Parents=((Item=(Type=Bone,Name="None"),Weight=1))
    Weight=1
    WorldUp=(Target=(X=0,Y=0,Z=1),Kind=Location,Space=(Type=None,Name="None"))
    bMaintainOffset=true
    Child=@n21.Element
    AimAxis=/Script/RigVM.RigVMFunction_MathVectorMul(B=(X=1,Y=1,Z=1), B.X=@n22.Result, B.Y=@n22.Result, B.Z=@n22.Result, A=RigVMVariableNode() {variable=PV Follow Aim Axis}.Value) {is=struct RigVMUnitNode, template="Multiply::Execute(in A,in B,out Result)"}.Result
    UpAxis=&RigVMFunction_MathVectorMul(B=(X=1,Y=1,Z=1), B.X=@n22.Result, B.Y=@n22.Result, B.Z=@n22.Result, A=RigVMVariableNode() {variable=PV Follow Up Axis}.Value) {template="Multiply::Execute(in A,in B,out Result)"}.Result
    WorldUp.Target=&RigVMFunction_MathVectorMul(B=(X=1,Y=1,Z=1), B.X=@n22.Result, B.Y=@n22.Result, B.Z=@n22.Result, A=RigVMVariableNode() {variable=PV Follow Aim Direction}.Value) {template="Multiply::Execute(in A,in B,out Result)"}.Result
    Parents.0.Item=@n23.Value
    WorldUp.Space=@n23.Value
   )
   n24 = RigVMVariableNode() {variable=IK Controls}
   > Parent_2 = RigVMFunctionReferenceNode(
    New Parent=@n21.Element
    Child=&RigUnit_HierarchyGetParent(
     bDefaultParent=True
     Child=RigVMDispatchNode(Index=1, Array=@n24.Value) {template="DISPATCH_RigVMDispatch_ArrayGetAtIndex(in Array,in Index,out Element)"}.Element
    ).Parent
   )
   n25 = &RigVMFunction_ControlFlowBranch(ExecuteContext=@RigVMFunction_Sequence_2.C, Condition=RigVMVariableNode() {variable=Keep Aligned}.Value)
   &RigUnit_SetControlVisibility(bVisible=false, ExecutePin=@n25.False, Item=RigVMVariableNode() {variable=Aligned Null}.Value) {template="SetControlVisibility::Execute(in Item,in Pattern,in bVisible)"}
   &RigUnit_ParentConstraint(
    AdvancedSettings=(InterpolationType=Average,RotationOrderForFilter=XZY)
    Parents=((Item=(Type=Bone,Name="None"),Weight=1))
    Weight=1
    Parents.0.Item=RigVMDispatchNode(Index=1, Array=RigVMVariableNode() {variable=Rig Bones}.Value) {template="DISPATCH_RigVMDispatch_ArrayGetAtIndex(in Array,in Index,out Element)"}.Element
    Child=&RigUnit_HierarchyGetParent(bDefaultParent=True, Child=RigVMVariableNode() {variable=Aligned Null}.Value).Parent
    ExecutePin=@n25.True
   )
   > &RigUnit_RotationConstraintLocalSpaceOffset(
    AdvancedSettings=(InterpolationType=Average,RotationOrderForFilter=XZY)
    Filter=(bX=True,bY=True,bZ=True)
    Parents=((Item=(Type=Bone,Name="None"),Weight=1))
    Weight=1
    Child=RigVMDispatchNode(Index=2, Array=&RigUnit_FindItemsWithMetadataTag(NameSpace=Self, Tag=Bones).Items) {template="DISPATCH_RigVMDispatch_ArrayGetAtIndex(in Array,in Index,out Element)"}.Element
    Parents.0.Item=RigVMVariableNode() {variable=Aligned Null}.Value
   )
   > &RigUnit_SetControlVisibility(bVisible=True, Item=RigVMVariableNode() {variable=Aligned Null}.Value) {template="SetControlVisibility::Execute(in Item,in Pattern,in bVisible)"}
   RigVMFunctionReturnNode(ExecuteContext=@RigVMFunction_Sequence_1.E)
  ]}
 New Function_ContainedGraph {nodes [
   n1 = RigVMFunctionEntryNode()
   n2 = &RigVMFunction_MathVectorSub(
    A=&RigUnit_GetTransform(Space=LocalSpace, bInitial=true, Item=@n1.End Bone).Transform.Translation
    B=&RigUnit_GetTransform(Space=LocalSpace, bInitial=true, Item=@n1.Start Bone).Transform.Translation
   ) {template="Subtract::Execute(in A,in B,out Result)"}
   DISPATCH_RigVMDispatch_If = RigVMDispatchNode(
    True=@n1.Left Side Suffix
    False=@n1.Right Side Suffix
    Condition=&RigVMFunction_MathDoubleGreater(B=0, A=&RigVMFunction_MathVectorUnit(Value=@n2.Result) {template="Unit::Execute(in Value,out Result)"}.Result.X) {template="Greater::Execute(in A,in B,out Result)"}.Result
   ) {template="DISPATCH_RigVMDispatch_If(in Condition,in True,in False,out Result)"}
   > RigVMVariableNode(Value=@DISPATCH_RigVMDispatch_If.Result) {variable=side}
   > RigVMFunctionReturnNode(Result=@DISPATCH_RigVMDispatch_If.Result)
  ]}
 New Function_ContainedGraph {nodes [
   n1 = RigVMFunctionEntryNode()
   Concat_4 = RigVMAggregateNode(B=@n1.name, A=@n1.prefix, C=@n1.side, D=@n1.suffix)
   > RigVMFunctionReturnNode(Result=@Concat_4.Result)
  ]}
 New Function_ContainedGraph {nodes [
   RigVMFunctionEntryNode()
   > RigVMFunction_Sequence = RigVMAggregateNode()
   Get_Module_Metadata = RigVMDispatchNode(Default=(Type=Bone,Name="None"), Name=IK Control, NameSpace=Self) {template="DISPATCH_RigDispatch_GetModuleMetadata(in Name,in NameSpace,in Default,out Value,out Found)"}
   SetAnimationChannel = /Script/ControlRig.RigUnit_SetBoolAnimationChannel(Channel=PV Follow, Value=False, bInitial=False, Control=@Get_Module_Metadata.Value.Name, ExecutePin=@RigVMFunction_Sequence.A) {is=struct RigVMUnitNode, template="SetAnimationChannel::Execute(in Value,in Control,in Channel,in bInitial)"}
   DISPATCH_RigVMDispatch_ArrayIterator = RigVMDispatchNode(Array=&RigUnit_FindItemsWithMetadataTag(NameSpace=Self, Tag=Controls).Items, ExecuteContext=@RigVMFunction_Sequence.B) {template="DISPATCH_RigVMDispatch_ArrayIterator(in Array,out Element,out Index,out Count,out Ratio)"}
   DISPATCH_RigDispatch_GetMetadata = RigVMDispatchNode(Default=(Type=Bone,Name="None"), Name=Item, NameSpace=Self, Item=@DISPATCH_RigVMDispatch_ArrayIterator.Element) {template="DISPATCH_RigDispatch_GetMetadata(in Item,in Name,in NameSpace,in Default,out Value,out Found)"}
   RigVMFunction_ControlFlowBranch = &RigVMFunction_ControlFlowBranch(Condition=@DISPATCH_RigDispatch_GetMetadata.Found, ExecuteContext=@DISPATCH_RigVMDispatch_ArrayIterator.ExecuteContext)
   &RigUnit_SetTransform(
    Weight=1
    bInitial=False
    Item=@DISPATCH_RigVMDispatch_ArrayIterator.Element
    Value=&RigUnit_ProjectTransformToNewParent(
     bChildInitial=True
     bNewParentInitial=False
     bOldParentInitial=True
     NewParent=@DISPATCH_RigDispatch_GetMetadata.Value
     OldParent=@DISPATCH_RigDispatch_GetMetadata.Value
     Child=@DISPATCH_RigVMDispatch_ArrayIterator.Element
    ).Transform
    ExecutePin=@RigVMFunction_ControlFlowBranch.True
   )
   n1 = RigVMDispatchNode(Array=&RigUnit_FindItemsWithMetadataTag(NameSpace=Self, Tag=PV Controls).Items, ExecuteContext=@RigVMFunction_Sequence.C) {template="DISPATCH_RigVMDispatch_ArrayIterator(in Array,out Element,out Index,out Count,out Ratio)"}
   Get_Metadata = RigVMDispatchNode(Default=(Type=Bone,Name="None"), Name=Closest Bone, NameSpace=Self, Item=@n1.Element) {template="DISPATCH_RigDispatch_GetMetadata(in Item,in Name,in NameSpace,in Default,out Value,out Found)"}
   > Parent_3 = RigVMFunctionReferenceNode(Child=@n1.Element, New Parent=@Get_Metadata.Value)
   RigVMFunctionReturnNode(ExecuteContext=@RigVMFunction_Sequence.D)
   n2 = &RigUnit_FindItemsWithMetadataTag(NameSpace=Self, Tag=Controls FK)
   Parent = RigVMFunctionReferenceNode(
    Child=RigVMDispatchNode(Index=0, Array=@n2.Items) {template="DISPATCH_RigVMDispatch_ArrayGetAtIndex(in Array,in Index,out Element)"}.Element
    New Parent=&RigUnit_ResolveConnector(Connector=(Type=Connector,Name="Shoulder Bone"), SkipSocket=False).Result
    ExecuteContext=@RigVMFunction_Sequence.E
   )
   > Parent_1 = RigVMFunctionReferenceNode(
    Child=RigVMDispatchNode(Index=1, Array=@n2.Items) {template="DISPATCH_RigVMDispatch_ArrayGetAtIndex(in Array,in Index,out Element)"}.Element
    New Parent=&RigUnit_ResolveConnector(Connector=(Type=Connector,Name="Elbow Bone"), SkipSocket=False).Result
   )
   > Parent_2 = RigVMFunctionReferenceNode(
    Child=RigVMDispatchNode(Index=2, Array=@n2.Items) {template="DISPATCH_RigVMDispatch_ArrayGetAtIndex(in Array,in Index,out Element)"}.Element
    New Parent=&RigUnit_ResolveConnector(Connector=(Type=Connector,Name="Hand Bone"), SkipSocket=False).Result
   )
   > For_Each = RigVMDispatchNode(Array=&RigUnit_FindItemsWithMetadataTag(NameSpace=Self, Tag=PV Controls).Items) {template="DISPATCH_RigVMDispatch_ArrayIterator(in Array,out Element,out Index,out Count,out Ratio)"}
   Get_Metadata_1 = RigVMDispatchNode(Default=(Type=Bone,Name="None"), Name=Closest Bone, NameSpace=Self, Item=@For_Each.Element) {template="DISPATCH_RigDispatch_GetMetadata(in Item,in Name,in NameSpace,in Default,out Value,out Found)"}
   > &RigUnit_SetTransform(
    Weight=1
    bInitial=False
    Item=@For_Each.Element
    Value=&RigUnit_GetTransform(bInitial=False, Item=@Get_Metadata_1.Value).Transform
   )
  ]}
 New Function_ContainedGraph {nodes [
   RigVMFunctionEntryNode()
   > RigVMFunction_Sequence_3 = RigVMAggregateNode()
   Get_Module_Metadata_1 = RigVMDispatchNode(Default=(X=0,Y=0,Z=0), Name=Mirror Axis, NameSpace=Root) {template="DISPATCH_RigDispatch_GetModuleMetadata(in Name,in NameSpace,in Default,out Value,out Found)"}
   RigVMVariableNode(Value=@Get_Module_Metadata_1.Value, ExecuteContext=@RigVMFunction_Sequence_3.A) {variable=Mirror Axis}
   > RigVMVariableNode(Value=template) {variable=Template Bone Chain Name}
   > RigVMVariableNode(Value=RigVMVariableNode() {variable=Module Settings}.Value) {variable=Module Settings Constr}
   > RigVMVariableNode(Value=RigVMVariableNode() {variable=Module Settings}.Value.Prefix_20_#1) {variable=name prefix}
   &RigUnit_SetMetadataTag(
    NameSpace=Self
    Tag=Bones
    Item=&RigUnit_ResolveConnector(SkipSocket=true, Connector=RigVMVariableNode() {variable=End Bone Connector}.Value).Result
    ExecutePin=@RigVMFunction_Sequence_3.B
   )
   > &RigUnit_SetMetadataTag(
    NameSpace=Self
    Tag=Bones
    Item=&RigUnit_ResolveConnector(SkipSocket=true, Connector=RigVMVariableNode() {variable=Middle Bone Connector}.Value).Result
   )
   > &RigUnit_SetMetadataTag(
    NameSpace=Self
    Tag=Bones
    Item=&RigUnit_ResolveConnector(SkipSocket=true, Connector=RigVMVariableNode() {variable=Start Bone Connector}.Value).Result
   )
   > Is Right Side v2 = RigVMFunctionReferenceNode(
    Item=&RigUnit_ResolveConnector(SkipSocket=true, Connector=RigVMVariableNode() {variable=End Bone Connector}.Value).Result
    Mirror Plane=RigVMVariableNode() {variable=Mirror Axis}.Value
   )
   n1 = RigVMVariableNode() {variable=Module Settings}
   > RigVMVariableNode(Value=RigVMDispatchNode(
     Condition=@"Is Right Side v2".Right Side
     False=@n1.Value.LeftSideSuffix_23_#2
     True=@n1.Value.RightSideSuffix_24_#3
    ) {template="DISPATCH_RigVMDispatch_If(in Condition,in True,in False,out Result)"}.Result) {variable=Current Side}
   > RigVMVariableNode(Value=@"Is Right Side v2".Right Side) {variable=Is on Right Side}
   For_Each_1_1 = RigVMDispatchNode(ExecuteContext=@RigVMFunction_Sequence_3.H, Array=RigVMVariableNode() {variable=IK Control Spaces}.Value) {template="DISPATCH_RigVMDispatch_ArrayIterator(in Array,out Element,out Index,out Count,out Ratio)"}
   > /Script/ControlRig.RigUnit_AddParent(DisplayLabel=None, Parent=@For_Each_1_1.Element, Child=RigVMVariableNode() {variable=IK Ctrl}.Value) {is=struct RigVMUnitNode}
   For_Each_1 = RigVMDispatchNode(ExecuteContext=@For_Each_1_1.Completed, Array=RigVMVariableNode() {variable=PV Control Spaces}.Value) {template="DISPATCH_RigVMDispatch_ArrayIterator(in Array,out Element,out Index,out Count,out Ratio)"}
   > &RigUnit_AddParent(DisplayLabel=None, Parent=@For_Each_1.Element, Child=RigVMVariableNode() {variable=PV Ctrl}.Value)
   RigVMDispatchNode(Name=Aligned Null, NameSpace=Self, ExecuteContext=@RigVMFunction_Sequence_3.J, Value=RigVMVariableNode() {variable=Aligned Ctrl}.Value) {template="DISPATCH_RigDispatch_SetModuleMetadata(in Name,in NameSpace,in Value,out Success)"}
   > RigVMDispatchNode(Name=Effector Ori, NameSpace=Self, Value=RigVMVariableNode() {variable=Effector Ori}.Value) {template="DISPATCH_RigDispatch_SetModuleMetadata(in Name,in NameSpace,in Value,out Success)"}
   > RigVMDispatchNode(Name=IK Control, NameSpace=Self, Value=RigVMVariableNode() {variable=IK Ctrl}.Value) {template="DISPATCH_RigDispatch_SetModuleMetadata(in Name,in NameSpace,in Value,out Success)"}
   > RigVMDispatchNode(Name=Effector Null, NameSpace=Self, Value=RigVMVariableNode() {variable=Effector Null}.Value) {template="DISPATCH_RigDispatch_SetModuleMetadata(in Name,in NameSpace,in Value,out Success)"}
   > RigVMDispatchNode(Name=fk_ik_switch, NameSpace=Self, Value=RigVMVariableNode() {variable=fk_ik_Switch}.Value) {template="DISPATCH_RigDispatch_SetModuleMetadata(in Name,in NameSpace,in Value,out Success)"}
   > RigVMDispatchNode(Name=PV Control, NameSpace=Self, Value=RigVMVariableNode() {variable=PV Ctrl}.Value) {template="DISPATCH_RigDispatch_SetModuleMetadata(in Name,in NameSpace,in Value,out Success)"}
   > Set_Module_Metadata = RigVMDispatchNode(Name=Is Right Side, NameSpace=Self, Value=RigVMVariableNode() {variable=Is on Right Side}.Value) {template="DISPATCH_RigDispatch_SetModuleMetadata(in Name,in NameSpace,in Value,out Success)"}
   > RigVMDispatchNode(Name=Rig Bones, NameSpace=Self, Value=RigVMVariableNode() {variable=Rig Bones}.Value) {template="DISPATCH_RigDispatch_SetModuleMetadata(in Name,in NameSpace,in Value,out Success)"}
   > RigVMDispatchNode(Name=Aim Follow Bones, NameSpace=Self, Value=RigVMVariableNode() {variable=Aim Follow Bones}.Value) {template="DISPATCH_RigDispatch_SetModuleMetadata(in Name,in NameSpace,in Value,out Success)"}
   Set Closest Bones and Controls = RigVMCollapseNode(ExecuteContext=@RigVMFunction_Sequence_3.N)
   > Finalize Controls = RigVMCollapseNode(Module Settings=RigVMVariableNode() {variable=Module Settings}.Value)
   > RigVMFunctionReturnNode(Current Side Name=RigVMVariableNode() {variable=Current Side}.Value, Template Bones=RigVMVariableNode() {variable=Rig Bones}.Value)
   n2 = RigVMDispatchNode(Array=RigVMVariableNode() {variable=Rig Bones}.Value, ExecuteContext=@RigVMFunction_Sequence_3.F) {template="DISPATCH_RigVMDispatch_ArrayIterator(in Array,out Element,out Index,out Count,out Ratio)"}
   n3 = RigVMDispatchNode(Index=@n2.Index, Array=&RigUnit_FindItemsWithMetadataTag(NameSpace=Self, Tag=Bones).Items) {template="DISPATCH_RigVMDispatch_ArrayGetAtIndex(in Array,in Index,out Element)"}
   DISPATCH_RigVMDispatch_If = RigVMDispatchNode(
    False=False
    True=true
    Condition=RigVMDispatchNode(B=0, A=@n2.Index) {template="DISPATCH_RigVMDispatch_CoreEquals(in A,in B,out Result)"}.Result
   ) {template="DISPATCH_RigVMDispatch_If(in Condition,in True,in False,out Result)"}
   Concat_3 = RigVMAggregateNode(B=_fk, C=_ctrl, A=&RigUnit_HierarchyGetParent(bDefaultParent=True, Child=@n3.Element).Parent.Name)
   n4 = RigVMVariableNode() {variable=Parent Control Connector}
   > Control Stack at Item_1 = RigVMFunctionReferenceNode(
    Add Bottom Null=False
    Add Secondary Control=False
    Control Suffix=_ctrl
    Secondary Settings="(DisplayName=\"None\",Shape=(bVisible=True,Name=\"Default\",Color=(R=1,G=0,B=0,A=1),Transform=(Rotation=(X=0,Y=0,Z=0,W=1),Translation=(X=0,Y=0,Z=0),Scale3D=(X=1,Y=1,Z=1))),Proxy=(bIsProxy=False,ShapeVisibility=BasedOnSelection),InitialSpace=LocalSpace,bUsePref ...(truncated)"
    Tag Controls=true
    Zero Controls=true
    Name=&RigVMFunction_NameConcat(B=_fk, A=@n3.Element.Name) {template="Concat::Execute(in A,in B,out Result)"}.Result
    Item=@n2.Element
    Add Null=@DISPATCH_RigVMDispatch_If.Result
    Parent=RigVMDispatchNode(
     False=(Type=Control,Name="None")
     False.Name=@Concat_3.Result
     Condition=RigVMDispatchNode(B=0, A=@n2.Index) {template="DISPATCH_RigVMDispatch_CoreEquals(in A,in B,out Result)"}.Result
     True=&RigUnit_ResolveConnector(SkipSocket=true, Connector=@n4.Value).Result
    ) {template="DISPATCH_RigVMDispatch_If(in Condition,in True,in False,out Result)"}.Result
    Settings=RigVMVariableNode() {variable=FK Control Settings}.Value
   )
   > &RigUnit_SetTransform(
    Weight=1
    bInitial=true
    Item=@"Control Stack at Item_1".Control
    Value=&RigUnit_GetTransform(bInitial=true, Item=@"Control Stack at Item_1".Control).Transform
   )
   > &RigUnit_SetMetadataTag(NameSpace=Self, Tag=Controls FK, Item=@"Control Stack at Item_1".Control)
   > RigVMDispatchNode(Name=FK Parent, NameSpace=Self, Value=@"Control Stack at Item_1".Control) {template="DISPATCH_RigDispatch_SetModuleMetadata(in Name,in NameSpace,in Value,out Success)"}
   /Script/ControlRig.RigUnit_HierarchyAddAnimationChannelBool(
    LimitsEnabled=()
    MaximumValue=True
    MinimumValue=False
    Name=local
    Parent=RigVMDispatchNode(Index=0, Array=&RigUnit_FindItemsWithMetadataTag(NameSpace=Self, Tag=Controls FK).Items) {template="DISPATCH_RigVMDispatch_ArrayGetAtIndex(in Array,in Index,out Element)"}.Element
    InitialValue=RigVMVariableNode() {variable=Local Default}.Value
    ExecutePin=@n2.Completed
   ) {
    is=struct RigVMUnitNode
    template="SpawnAnimationChannel::Execute(in InitialValue,in MinimumValue,in MaximumValue,in LimitsEnabled,in Parent,in Name,out Item)"
   }
   Concat = &RigVMFunction_NameConcat(B=RigVMVariableNode() {variable=Current Side}.Value, A=RigVMVariableNode() {variable=Basename}.Value) {template="Concat::Execute(in A,in B,out Result)"}
   n5 = &RigUnit_FindItemsWithMetadataTag(NameSpace=Self, Tag=Bones)
   n6 = RigVMDispatchNode(False=(X=1,Y=1,Z=1), True=(X=-1,Y=-1,Z=-1), Condition=RigVMVariableNode() {variable=Is on Right Side}.Value) {template="DISPATCH_RigVMDispatch_If(in Condition,in True,in False,out Result)"}
   n7 = &RigVMFunction_MathVectorMul(B=@n6.Result, A=RigVMVariableNode() {variable=Secondary Axis}.Value) {template="Multiply::Execute(in A,in B,out Result)"}
   Multiply = &RigVMFunction_MathVectorMul(B=@n6.Result, A=RigVMVariableNode() {variable=Primary Axis}.Value) {template="Multiply::Execute(in A,in B,out Result)"}
   Create Bone Chain = RigVMFunctionReferenceNode(
    Debug=false
    Debug Scale=15
    Debug Thickness=0.5
    Enable Orient Bones=true
    Kind=Location
    Name=@Concat.Result
    Transforms=/Script/ControlRig.RigUnit_GetTransformItemArray(Space=GlobalSpace, bInitial=False, Items=@n5.Items) {is=struct RigVMUnitNode}.Transforms
    Parent=&RigUnit_HierarchyGetParent(
     bDefaultParent=True
     Child=RigVMDispatchNode(Index=0, Array=@n5.Items) {template="DISPATCH_RigVMDispatch_ArrayGetAtIndex(in Array,in Index,out Element)"}.Element
    ).Parent
    Optional Suffix=RigVMVariableNode() {variable=Template Bone Chain Name}.Value
    Secondary Axis=@n7.Result
    Primary Axis=@Multiply.Result
    ExecuteContext=@RigVMFunction_Sequence_3.C
   )
   > RigVMVariableNode(Value=@"Create Bone Chain".Bones) {variable=Rig Bones}
   n8 = RigVMDispatchNode(Index=1, Array=@"Create Bone Chain".Bones) {template="DISPATCH_RigVMDispatch_ArrayGetAtIndex(in Array,in Index,out Element)"}
   Add = &RigVMFunction_MathVectorAdd(
    A=&RigUnit_GetTransform(bInitial=true, Item=@n8.Element, Space=RigVMVariableNode() {variable=Offset Elbow Bone Space}.Value).Transform.Translation
    B=RigVMVariableNode() {variable=Offset Elbow Bone}.Value
   ) {template="Add::Execute(in A,in B,out Result)"}
   > Set Transform_2 = /Script/ControlRig.RigUnit_SetTranslation(
    Weight=1
    bInitial=true
    bPropagateToChildren=false
    Value=@Add.Result
    Item=@n8.Element
    Space=RigVMVariableNode() {variable=Offset Elbow Bone Space}.Value
   ) {is=struct RigVMUnitNode, template="Set Transform::Execute(in Item,in Space,in bInitial,in Value,in Weight,in bPropagateToChildren)"}
   > Orient Bones_0 = RigVMFunctionReferenceNode(Debug=false, Kind=Location, Secondary Axis=@n7.Result, Primary Axis=@Multiply.Result, Bones=@"Create Bone Chain".Bones)
   > /Script/ControlRig.RigUnit_SetTransformItemArray(
    Space=GlobalSpace
    Weight=1
    bInitial=true
    bPropagateToChildren=True
    Transforms=&RigUnit_GetTransformItemArray(Space=GlobalSpace, bInitial=true, Items=@"Create Bone Chain".Bones).Transforms
    Items=@"Create Bone Chain".Bones
   ) {is=struct RigVMUnitNode}
   n9 = &RigUnit_ResolveConnector(SkipSocket=true, Connector=RigVMVariableNode() {variable=Start Bone Connector}.Value)
   > Create Bone Chain_1 = RigVMFunctionReferenceNode(
    Debug=true
    Debug Scale=15
    Debug Thickness=0.5
    Enable Orient Bones=true
    Kind=Location
    Optional Suffix=aimFollow
    Transforms=((Rotation=(X=0,Y=0,Z=0,W=1),Translation=(X=0,Y=0,Z=0),Scale3D=(X=1,Y=1,Z=1)))
    Transforms.0=&RigUnit_GetTransform(bInitial=true, Item=@n9.Result).Transform
    Parent=&RigUnit_HierarchyGetParent(bDefaultParent=True, Child=@n9.Result).Parent
    Name=RigVMVariableNode() {variable=Basename}.Value
    Secondary Axis=RigVMVariableNode() {variable=Secondary Axis}.Value
    Primary Axis=RigVMVariableNode() {variable=Primary Axis}.Value
   )
   > RigVMVariableNode(Value=@"Create Bone Chain_1".Bones) {variable=Aim Follow Bones}
   Format Name_1 = RigVMFunctionReferenceNode(
    suffix=None
    side=RigVMVariableNode() {variable=Current Side}.Value
    prefix=RigVMVariableNode() {variable=name prefix}.Value
    name=RigVMVariableNode() {variable=IK Control Basename}.Value
   )
   Control Stack at Item = RigVMFunctionReferenceNode(
    Add Bottom Null=False
    Add Null=true
    Add Secondary Control=False
    Control Suffix=_ctrl
    Null Suffix=_null
    Secondary Settings="(DisplayName=\"None\",Shape=(bVisible=True,Name=\"Default\",Color=(R=1,G=0,B=0,A=1),Transform=(Rotation=(X=0,Y=0,Z=0,W=1),Translation=(X=0,Y=0,Z=0),Scale3D=(X=1,Y=1,Z=1))),Proxy=(bIsProxy=False,ShapeVisibility=BasedOnSelection),InitialSpace=LocalSpace,bUsePref ...(truncated)"
    Tag Controls=true
    Zero Controls=true
    Name=@"Format Name_1".Result
    Parent=&RigUnit_ResolveConnector(SkipSocket=true, Connector=RigVMVariableNode() {variable=Parent Control Connector}.Value).Result
    Item=&RigUnit_ResolveConnector(SkipSocket=true, Connector=RigVMVariableNode() {variable=End Bone Connector}.Value).Result
    Settings=RigVMVariableNode() {variable=IK Control Settings}.Value
    Orient to World=RigVMVariableNode() {variable=Orient to World}.Value
    ExecuteContext=@RigVMFunction_Sequence_3.D
   )
   > RigVMVariableNode(Value=@"Control Stack at Item".Control) {variable=IK Ctrl}
   > Add Null Below = RigVMFunctionReferenceNode(
    New Suffix=_effector
    Replace Suffix=_null
    Transform=(Rotation=(X=0,Y=0,Z=0,W=1),Translation=(X=0,Y=0,Z=0),Scale3D=(X=1,Y=1,Z=1))
    Use Transform=False
    Item=@"Control Stack at Item".Control
   )
   > RigVMVariableNode(Value=@"Add Null Below".Null) {variable=Effector Null}
   > Add Null Below_1 = RigVMFunctionReferenceNode(
    New Suffix=_oriNull
    Replace Suffix=_oriNull
    Transform=(Rotation=(X=0,Y=0,Z=0,W=1),Translation=(X=0,Y=0,Z=0),Scale3D=(X=1,Y=1,Z=1))
    Use Transform=false
    Item=@"Control Stack at Item".Control
   )
   > n10 = RigVMVariableNode(Value=@"Add Null Below_1".Null) {variable=Effector Ori}
   n11 = &RigVMFunction_ControlFlowBranch(Condition=RigVMVariableNode() {variable=Enable Keep Aligned}.Value, ExecuteContext=@n10.ExecuteContext)
   Format Name_1_1 = RigVMFunctionReferenceNode(
    suffix=_wrist
    side=RigVMVariableNode() {variable=Current Side}.Value
    prefix=RigVMVariableNode() {variable=name prefix}.Value
    name=RigVMVariableNode() {variable=IK Control Basename}.Value
   )
   n12 = RigVMVariableNode() {variable=IK Ctrl}
   Control Stack at Item_2 = RigVMFunctionReferenceNode(
    Add Bottom Null=False
    Add Null=true
    Add Secondary Control=False
    Control Suffix=_ctrl
    Null Suffix=_null
    Secondary Settings="(DisplayName=\"None\",Shape=(bVisible=True,Name=\"Default\",Color=(R=1,G=0,B=0,A=1),Transform=(Rotation=(X=0,Y=0,Z=0,W=1),Translation=(X=0,Y=0,Z=0),Scale3D=(X=1,Y=1,Z=1))),Proxy=(bIsProxy=False,ShapeVisibility=BasedOnSelection),InitialSpace=LocalSpace,bUsePref ...(truncated)"
    Tag Controls=true
    Zero Controls=true
    Name=@"Format Name_1_1".Result
    Parent=&RigUnit_HierarchyGetParent(bDefaultParent=True, Child=@n12.Value).Parent
    ExecuteContext=@n11.True
    Settings=RigVMVariableNode() {variable=Wrist Control Settings}.Value
    Item=@n12.Value
   )
   > RigVMVariableNode(Value=@"Control Stack at Item_2".Control) {variable=Aligned Ctrl}
   /Script/ControlRig.RigUnit_SetDefaultParent(Child=@"Add Null Below_1".Null, ExecutePin=@n11.Completed, Parent=@"Control Stack at Item".Null) {is=struct RigVMUnitNode}
   > &RigUnit_SetMetadataTag(NameSpace=Self, Tag=Controls IK, Item=@"Control Stack at Item".Control)
   > &RigUnit_SetTransform(
    Weight=1
    bInitial=true
    Value=&RigUnit_GetTransform(bInitial=true, Item=@"Control Stack at Item".Control).Transform
    Item=@"Control Stack at Item".Control
   )
   > Offset Initial Transform = RigVMFunctionReferenceNode(Space=LocalSpace, Transform=RigVMVariableNode() {variable=Offset IK Ctrl Transform}.Value, Item=@"Control Stack at Item".Control)
   Format Name = RigVMFunctionReferenceNode(
    suffix=_fk_ik_switch
    name=RigVMVariableNode() {variable=Basename}.Value
    prefix=RigVMVariableNode() {variable=name prefix}.Value
    side=RigVMVariableNode() {variable=Current Side}.Value
   )
   > SpawnAnimationChannel = &RigUnit_HierarchyAddAnimationChannelBool(
    LimitsEnabled=()
    MaximumValue=True
    MinimumValue=False
    Name=@"Format Name".Result
    Parent=&RigUnit_ResolveConnector(SkipSocket=true, Connector=RigVMVariableNode() {variable=Settings Control Connector}.Value).Result
    InitialValue=RigVMVariableNode() {variable=FK IK Default}.Value
   ) {template="SpawnAnimationChannel::Execute(in InitialValue,in MinimumValue,in MaximumValue,in LimitsEnabled,in Parent,in Name,out Item)"}
   > n13 = RigVMVariableNode(Value=@SpawnAnimationChannel.Item.Name) {variable=fk_ik_Switch}
   n14 = &RigVMFunction_ControlFlowBranch(ExecuteContext=@n13.ExecuteContext, Condition=RigVMVariableNode() {variable=Enable Softness}.Value)
   n15 = &RigUnit_HierarchyAddAnimationChannelBool(
    InitialValue=false
    LimitsEnabled=()
    MaximumValue=True
    MinimumValue=False
    Name=Stretch
    ExecutePin=@n14.Completed
    Parent=RigVMVariableNode() {variable=IK Ctrl}.Value
   ) {template="SpawnAnimationChannel::Execute(in InitialValue,in MinimumValue,in MaximumValue,in LimitsEnabled,in Parent,in Name,out Item)"}
   n16 = &RigVMFunction_ControlFlowBranch(ExecuteContext=@n15.ExecutePin, Condition=RigVMVariableNode() {variable=Enable PV Follow}.Value)
   RigVMFunction_ControlFlowBranch = &RigVMFunction_ControlFlowBranch(ExecuteContext=@n16.Completed, Condition=RigVMVariableNode() {variable=Enable Keep Aligned}.Value)
   &RigUnit_HierarchyAddAnimationChannelBool(
    InitialValue=false
    LimitsEnabled=()
    MaximumValue=True
    MinimumValue=False
    Name=Keep Aligned
    ExecutePin=@RigVMFunction_ControlFlowBranch.True
    Parent=RigVMVariableNode() {variable=IK Ctrl}.Value
   ) {template="SpawnAnimationChannel::Execute(in InitialValue,in MinimumValue,in MaximumValue,in LimitsEnabled,in Parent,in Name,out Item)"}
   &RigUnit_HierarchyAddAnimationChannelBool(
    InitialValue=true
    LimitsEnabled=()
    MaximumValue=True
    MinimumValue=False
    Name=PV Follow
    ExecutePin=@n16.True
    Parent=RigVMVariableNode() {variable=IK Ctrl}.Value
   ) {template="SpawnAnimationChannel::Execute(in InitialValue,in MinimumValue,in MaximumValue,in LimitsEnabled,in Parent,in Name,out Item)"}
   /Script/ControlRig.RigUnit_HierarchyAddAnimationChannelFloat(
    InitialValue=1
    LimitsEnabled=(Enabled=(bMinimum=True,bMaximum=True))
    MaximumValue=2
    MinimumValue=0
    Name=Softness
    ExecutePin=@n14.True
    Parent=RigVMVariableNode() {variable=IK Ctrl}.Value
   ) {
    is=struct RigVMUnitNode
    template="SpawnAnimationChannel::Execute(in InitialValue,in MinimumValue,in MaximumValue,in LimitsEnabled,in Parent,in Name,out Item)"
   }
   n17 = RigVMVariableNode() {variable=Rig Bones}
   DISPATCH_RigVMDispatch_ArrayGetAtIndex = RigVMDispatchNode(Index=0, Array=@n17.Value) {template="DISPATCH_RigVMDispatch_ArrayGetAtIndex(in Array,in Index,out Element)"}
   ComputePoleVector_1 = RigVMFunctionReferenceNode(
    Draw Transform=false
    Initial=true
    Origin Scale=(X=0,Y=0,Z=0)
    Bone A=@DISPATCH_RigVMDispatch_ArrayGetAtIndex.Element
    Bone B=RigVMDispatchNode(Index=1, Array=@n17.Value) {template="DISPATCH_RigVMDispatch_ArrayGetAtIndex(in Array,in Index,out Element)"}.Element
    Bone C=RigVMDispatchNode(Index=2, Array=@n17.Value) {template="DISPATCH_RigVMDispatch_ArrayGetAtIndex(in Array,in Index,out Element)"}.Element
    OffsetFactor=RigVMVariableNode() {variable=PV Offset Factor}.Value
    ExecuteContext=@RigVMFunction_Sequence_3.E
   )
   Format Name_2 = RigVMFunctionReferenceNode(
    suffix=_pv_ik
    side=RigVMVariableNode() {variable=Current Side}.Value
    name=RigVMVariableNode() {variable=Basename}.Value
    prefix=RigVMVariableNode() {variable=name prefix}.Value
   )
   > Control Stack at Position = RigVMFunctionReferenceNode(
    Add Bottom Null=False
    Add Null=true
    Add Secondary Control=False
    Control Suffix=_ctrl
    Null Suffix=_null
    Secondary Settings="(DisplayName=\"None\",Shape=(bVisible=True,Name=\"Default\",Color=(R=1,G=0,B=0,A=1),Transform=(Rotation=(X=0,Y=0,Z=0,W=1),Translation=(X=0,Y=0,Z=0),Scale3D=(X=1,Y=1,Z=1))),Proxy=(bIsProxy=False,ShapeVisibility=BasedOnSelection),InitialSpace=LocalSpace,bUsePref ...(truncated)"
    Tag Controls=true
    Zero Controls=true
    Global Position=@ComputePoleVector_1.Transform
    Name=@"Format Name_2".Result
    Settings=RigVMVariableNode() {variable=PV Control Settings}.Value
    Parent=RigVMVariableNode() {variable=Pole Vector Parent Connector}.Value
   )
   > &RigUnit_SetMetadataTag(NameSpace=Self, Tag=Controls IK, Item=@"Control Stack at Position".Control)
   > &RigUnit_SetMetadataTag(NameSpace=Self, Tag=PV Controls, Item=@"Control Stack at Position".Control)
   > RigVMVariableNode(Value=@"Control Stack at Position".Control) {variable=PV Ctrl}
   > Offset Initial Transform_1 = RigVMFunctionReferenceNode(Space=LocalSpace, Transform=RigVMVariableNode() {variable=Offset PV Control}.Value, Item=@"Control Stack at Position".Control)
   RigVMDispatchNode(ExecuteContext=(), Name=Effector Null, NameSpace=Self, Value=(Type=Bone,Name="None")) {template="DISPATCH_RigDispatch_SetModuleMetadata(in Name,in NameSpace,in Value,out Success)"}
   RigVMDispatchNode(ExecuteContext=(), Name=IK Control, NameSpace=Self, Value=@"Control Stack at Item".Control) {template="DISPATCH_RigDispatch_SetModuleMetadata(in Name,in NameSpace,in Value,out Success)"}
  ]}
 New Function_ContainedGraph {nodes [
   n1 = RigVMFunctionEntryNode()
   n2 = &RigUnit_FindItemsWithMetadataTag(NameSpace=Self, Tag=Controls)
   > Scale Module Controls = RigVMFunctionReferenceNode(
    Global Scale=RigVMDispatchNode(Default=1, Name=Control Size, NameSpace=Root) {template="DISPATCH_RigDispatch_GetModuleMetadata(in Name,in NameSpace,in Default,out Value,out Found)"}.Value
    Local Scale=@n1.Module Settings.ControlSize_12_#4
    Controls=@n2.Items
   )
   > Color Module Controls = RigVMFunctionReferenceNode(
    CenterColor=@n1.Module Settings.CenterColor_9_#5
    LeftColor=@n1.Module Settings.LeftColor_3_#6
    Override=@n1.Module Settings.OverrideColor_48_#7
    RightColor=@n1.Module Settings.RightColor_5_#8
    Controls=@n2.Items
   )
   > RigVMFunctionReturnNode()
  ]}
 New Function_ContainedGraph {nodes [
   n1 = RigVMFunctionEntryNode()
   Multiply = &RigVMFunction_MathTransformMul(B=@n1.Transform, A=&RigUnit_GetTransform(bInitial=false, Item=@n1.Item, Space=@n1.Space).Transform) {template="Multiply::Execute(in A,in B,out Result)"}
   > &RigUnit_SetTransform(Weight=1, bInitial=true, Item=@n1.Item, Space=@n1.Space, Value=@Multiply.Result)
   > RigVMFunctionReturnNode()
  ]}
 New Function_ContainedGraph {nodes [
   n1 = &RigUnit_FindItemsWithMetadataTag(NameSpace=Self, Tag=New Bones)
   n2 = RigVMDispatchNode(ExecuteContext=(), Array=@n1.Items) {template="DISPATCH_RigVMDispatch_ArrayIterator(in Array,out Element,out Index,out Count,out Ratio)"}
   Subtract = /Script/RigVM.RigVMFunction_MathIntSub(B=1, A=@n2.Count) {is=struct RigVMUnitNode, template="Subtract::Execute(in A,in B,out Result)"}
   DISPATCH_RigVMDispatch_CoreEquals = RigVMDispatchNode(A=@n2.Index, B=@Subtract.Result) {template="DISPATCH_RigVMDispatch_CoreEquals(in A,in B,out Result)"}
   n3 = &RigVMFunction_ControlFlowBranch(ExecuteContext=@n2.ExecuteContext, Condition=@DISPATCH_RigVMDispatch_CoreEquals.Result)
   Add = /Script/RigVM.RigVMFunction_MathIntAdd(B=1, A=@n2.Index) {is=struct RigVMUnitNode, template="Add::Execute(in A,in B,out Result)"}
   n4 = &RigUnit_AimBoneMath(
    DebugSettings=(bEnabled=False,Scale=10,WorldOffset=(Rotation=(X=0,Y=0,Z=0,W=1),Translation=(X=0,Y=0,Z=0),Scale3D=(X=1,Y=1,Z=1)))
    Primary=(Weight=1,Axis=(X=1,Y=0,Z=0),Target=(X=1,Y=0,Z=0),Kind=Location,Space=(Type=Bone,Name="None"))
    Secondary=(Weight=1,Axis=(X=0,Y=0,Z=1),Target=(X=1,Y=0,Z=0),Kind=Location,Space=(Type=Bone,Name="None"))
    Weight=1
    InputTransform=&RigUnit_GetTransform(bInitial=False, Item=@n2.Element).Transform
    Primary.Target=&RigUnit_GetTransform(
     bInitial=False
     Item=RigVMDispatchNode(Index=@Add.Result, Array=@n1.Items) {template="DISPATCH_RigVMDispatch_ArrayGetAtIndex(in Array,in Index,out Element)"}.Element
    ).Transform.Translation
   )
   &RigUnit_SetTransform(Weight=1, bInitial=true, Value=@n4.Result, Item=@n2.Element, ExecutePin=@n3.False)
   n5 = RigVMFunctionEntryNode()
   RigVMFunction_Sequence = &RigVMFunction_Sequence(ExecuteContext=@n5.ExecuteContext)
   n6 = &RigVMFunction_Sequence(ExecuteContext=@RigVMFunction_Sequence.A)
   DISPATCH_RigVMDispatch_ArrayIterator = RigVMDispatchNode(Array=@n5.Transforms, ExecuteContext=@n6.A) {template="DISPATCH_RigVMDispatch_ArrayIterator(in Array,out Element,out Index,out Count,out Ratio)"}
   RigVMFunction_StringPadInteger = /Script/RigVM.RigVMFunction_StringPadInteger(
    Digits=2
    Value=&RigVMFunction_MathIntAdd(B=1, A=@DISPATCH_RigVMDispatch_ArrayIterator.Index) {template="Add::Execute(in A,in B,out Result)"}.Result
   ) {is=struct RigVMUnitNode}
   DISPATCH_RigDispatch_FromString = RigVMDispatchNode(String=@RigVMFunction_StringPadInteger.Result) {template="DISPATCH_RigDispatch_FromString(in String,out Result)"}
   Concat = RigVMAggregateNode(B=_, D=_, C=@DISPATCH_RigDispatch_FromString.Result, A=@n5.Name, E=@n5.Optional Suffix)
   Concat_1 = RigVMAggregateNode(B=_, D=None, C=@DISPATCH_RigDispatch_FromString.Result, A=RigVMVariableNode() {variable=Name}.Value)
   EndsWith = /Script/RigVM.RigVMFunction_EndsWith(Ending=_, Name=@Concat.Result) {is=struct RigVMUnitNode, template="EndsWith::Execute(in Name,in Ending,out Result)"}
   DISPATCH_RigVMDispatch_If = RigVMDispatchNode(False=@Concat.Result, True=@Concat_1.Result, Condition=@EndsWith.Result) {template="DISPATCH_RigVMDispatch_If(in Condition,in True,in False,out Result)"}
   > n7 = /Script/ControlRig.RigUnit_HierarchyAddBone(Space=GlobalSpace, Transform=@DISPATCH_RigVMDispatch_ArrayIterator.Element, Name=@DISPATCH_RigVMDispatch_If.Result, Parent=@n5.Parent) {is=struct RigVMUnitNode}
   > &RigUnit_SetMetadataTag(NameSpace=Self, Tag=New Bones, Item=@n7.Item)
   n8 = &RigUnit_FindItemsWithMetadataTag(NameSpace=Self, Tag=New Bones)
   n9 = RigVMDispatchNode(Array=@n8.Items, ExecuteContext=@n6.B) {template="DISPATCH_RigVMDispatch_ArrayIterator(in Array,out Element,out Index,out Count,out Ratio)"}
   n10 = &RigVMFunction_MathIntSub(B=1, A=@n9.Index) {template="Subtract::Execute(in A,in B,out Result)"}
   > &RigUnit_SetDefaultParent(
    Child=@n9.Element
    Parent=RigVMDispatchNode(
     False=RigVMDispatchNode(Array=@n8.Items, Index=@n10.Result) {template="DISPATCH_RigVMDispatch_ArrayGetAtIndex(in Array,in Index,out Element)"}.Element
     Condition=RigVMDispatchNode(B=0, A=@n9.Index) {template="DISPATCH_RigVMDispatch_CoreEquals(in A,in B,out Result)"}.Result
     True=RigVMVariableNode() {variable=Parent}.Value
    ) {template="DISPATCH_RigVMDispatch_If(in Condition,in True,in False,out Result)"}.Result
   )
   Orient Bones = RigVMFunctionReferenceNode(
    Debug=True
    ExecuteContext=@n9.Completed
    Bones=&RigUnit_FindItemsWithMetadataTag(NameSpace=Self, Tag=New Bones).Items
    Primary Axis=RigVMVariableNode() {variable=Primary Axis}.Value
    Secondary Axis=RigVMVariableNode() {variable=Secondary Axis}.Value
   )
   RigVMFunctionReturnNode(Bones=&RigUnit_FindItemsWithMetadataTag(NameSpace=Self, Tag=New Bones).Items, ExecuteContext=@RigVMFunction_Sequence.B)
  ]}
 New Function_ContainedGraph {nodes [
   n1 = RigVMFunctionEntryNode()
   > RigVMFunction_Sequence = RigVMAggregateNode()
   RigVMFunction_Sequence_1 = RigVMAggregateNode(ExecuteContext=@RigVMFunction_Sequence.A)
   DISPATCH_RigVMDispatch_ArrayIterator = RigVMDispatchNode(Array=@n1.Bones, ExecuteContext=@RigVMFunction_Sequence_1.A) {template="DISPATCH_RigVMDispatch_ArrayIterator(in Array,out Element,out Index,out Count,out Ratio)"}
   > DISPATCH_RigDispatch_SetMetadata = RigVMDispatchNode(
    Name=Orig Parent
    NameSpace=Self
    Item=@DISPATCH_RigVMDispatch_ArrayIterator.Element
    Value=&RigUnit_HierarchyGetParent(bDefaultParent=True, Child=@DISPATCH_RigVMDispatch_ArrayIterator.Element).Parent
   ) {template="DISPATCH_RigDispatch_SetMetadata(in Item,in Name,in NameSpace,in Value,out Success)"}
   n2 = RigVMVariableNode() {variable=Bones}
   > &RigUnit_SetDefaultParent(
    Child=@DISPATCH_RigVMDispatch_ArrayIterator.Element
    Parent=&RigUnit_HierarchyGetParent(
     bDefaultParent=True
     Child=RigVMDispatchNode(Index=0, Array=@n2.Value) {template="DISPATCH_RigVMDispatch_ArrayGetAtIndex(in Array,in Index,out Element)"}.Element
    ).Parent
   )
   DISPATCH_RigVMDispatch_ArrayGetNum = RigVMDispatchNode(Array=@n1.Bones) {template="DISPATCH_RigVMDispatch_ArrayGetNum(in Array,out Num)"}
   Interpolate = /Script/RigVM.RigVMFunction_MathTransformLerp(
    T=0.5
    A=&RigUnit_GetTransform(
     bInitial=False
     Item=RigVMDispatchNode(Index=0, Array=@n1.Bones) {template="DISPATCH_RigVMDispatch_ArrayGetAtIndex(in Array,in Index,out Element)"}.Element
    ).Transform
    B=&RigUnit_GetTransform(
     bInitial=False
     Item=RigVMDispatchNode(Index=-1, Array=@n1.Bones) {template="DISPATCH_RigVMDispatch_ArrayGetAtIndex(in Array,in Index,out Element)"}.Element
    ).Transform
   ) {is=struct RigVMUnitNode, template="Interpolate::Execute(in A,in B,in T,out Result)"}
   DISPATCH_RigVMDispatch_If = RigVMDispatchNode(
    False=(X=0,Y=0,Z=0)
    Condition=RigVMDispatchNode(B=3, A=@DISPATCH_RigVMDispatch_ArrayGetNum.Num) {template="DISPATCH_RigVMDispatch_CoreEquals(in A,in B,out Result)"}.Result
    True=@Interpolate.Result.Translation
   ) {template="DISPATCH_RigVMDispatch_If(in Condition,in True,in False,out Result)"}
   RigVMFunction_DebugTransformMutableNoSpace = &RigVMFunction_DebugTransformMutableNoSpace(
    Color=(R=1,G=0,B=0,A=1)
    DebugDrawSettings=(DepthPriority=SDPG_Foreground,Lifetime=-1)
    Mode=Axes
    Scale=10
    Thickness=0
    Transform=(Rotation=(X=0,Y=0,Z=0,W=1),Translation=(X=0,Y=0,Z=0),Scale3D=(X=1,Y=1,Z=1))
    WorldOffset=(Rotation=(X=0,Y=0,Z=0,W=1),Translation=(X=0,Y=0,Z=0),Scale3D=(X=1,Y=1,Z=1))
    Transform.Translation=@DISPATCH_RigVMDispatch_If.Result
    Transform.Rotation=@Interpolate.Result.Rotation
    ExecutePin=@RigVMFunction_Sequence_1.B
    bEnabled=RigVMVariableNode() {variable=Debug}.Value
   )
   > RigVMVariableNode(Value=@DISPATCH_RigVMDispatch_If.Result) {variable=Secondary Target}
   n3 = RigVMDispatchNode(Array=@n1.Bones, ExecuteContext=@RigVMFunction_Sequence_1.C) {template="DISPATCH_RigVMDispatch_ArrayIterator(in Array,out Element,out Index,out Count,out Ratio)"}
   Subtract = &RigVMFunction_MathIntSub(B=1, A=@n3.Count) {template="Subtract::Execute(in A,in B,out Result)"}
   DISPATCH_RigVMDispatch_CoreEquals = RigVMDispatchNode(A=@n3.Index, B=@Subtract.Result) {template="DISPATCH_RigVMDispatch_CoreEquals(in A,in B,out Result)"}
   n4 = &RigVMFunction_ControlFlowBranch(ExecuteContext=@n3.ExecuteContext, Condition=@DISPATCH_RigVMDispatch_CoreEquals.Result)
   Add = &RigVMFunction_MathIntAdd(B=1, A=@n3.Index) {template="Add::Execute(in A,in B,out Result)"}
   n5 = &RigUnit_AimBoneMath(
    DebugSettings=(bEnabled=False,Scale=10,WorldOffset=(Rotation=(X=0,Y=0,Z=0,W=1),Translation=(X=0,Y=0,Z=0),Scale3D=(X=1,Y=1,Z=1)))
    Primary=(Weight=1,Axis=(X=1,Y=0,Z=0),Target=(X=1,Y=0,Z=0),Kind=Location,Space=(Type=Bone,Name="None"))
    Secondary=(Weight=1,Axis=(X=0,Y=0,Z=1),Target=(X=1,Y=0,Z=0),Kind=Location,Space=(Type=Bone,Name="None"))
    Weight=1
    Primary.Axis=@n1.Primary Axis
    Secondary.Axis=@n1.Secondary Axis
    InputTransform=&RigUnit_GetTransform(bInitial=False, Item=@n3.Element).Transform
    Primary.Target=&RigUnit_GetTransform(
     bInitial=False
     Item=RigVMDispatchNode(Index=@Add.Result, Array=@n1.Bones) {template="DISPATCH_RigVMDispatch_ArrayGetAtIndex(in Array,in Index,out Element)"}.Element
    ).Transform.Translation
    Secondary.Target=RigVMVariableNode() {variable=Secondary Target}.Value
   )
   &RigUnit_SetTransform(Weight=1, bInitial=true, Value=@n5.Result, Item=@n3.Element, ExecutePin=@n4.False)
   n6 = RigVMDispatchNode(ExecuteContext=@n3.Completed, Array=RigVMVariableNode() {variable=Bones}.Value) {template="DISPATCH_RigVMDispatch_ArrayIterator(in Array,out Element,out Index,out Count,out Ratio)"}
   DISPATCH_RigDispatch_GetMetadata = RigVMDispatchNode(Default=(Type=Bone,Name="None"), Name=Orig Parent, NameSpace=Self, Item=@n6.Element) {template="DISPATCH_RigDispatch_GetMetadata(in Item,in Name,in NameSpace,in Default,out Value,out Found)"}
   > &RigUnit_SetDefaultParent(Parent=@DISPATCH_RigDispatch_GetMetadata.Value, Child=@n6.Element)
   DISPATCH_RigVMDispatch_ArrayGetAtIndex = RigVMDispatchNode(Index=-1, Array=RigVMVariableNode() {variable=Bones}.Value) {template="DISPATCH_RigVMDispatch_ArrayGetAtIndex(in Array,in Index,out Element)"}
   Set Transform = &RigUnit_SetRotation(Value=(X=0,Y=0,Z=0,W=1), Weight=1, bInitial=true, Item=@DISPATCH_RigVMDispatch_ArrayGetAtIndex.Element, ExecutePin=@n6.Completed)
   RigVMFunction_DebugTransformArrayMutableNoSpace = /Script/RigVM.RigVMFunction_DebugTransformArrayMutableNoSpace(
    Color=(R=1,G=1,B=1,A=1)
    DebugDrawSettings=(DepthPriority=SDPG_Foreground,Lifetime=-1)
    Mode=Axes
    ParentIndices=()
    Scale=15
    Thickness=1
    WorldOffset=(Rotation=(X=0,Y=0,Z=0,W=1),Translation=(X=0,Y=0,Z=0),Scale3D=(X=1,Y=1,Z=1))
    bEnabled=@n1.Debug
    Transforms=&RigUnit_GetTransformItemArray(Space=GlobalSpace, bInitial=False, Items=@n1.Bones).Transforms
    ExecutePin=@RigVMFunction_Sequence.B
   ) {is=struct RigVMUnitNode}
   RigVMFunctionReturnNode(ExecuteContext=@RigVMFunction_Sequence.C)
   &RigUnit_FindItemsWithMetadataTag(NameSpace=Self, Tag=New Bones)
  ]}
 New Function_ContainedGraph {nodes [
   RigVMFunctionEntryNode()
   > RigVMFunction_Sequence_1 = RigVMAggregateNode()
   Get_Module_Metadata_3 = RigVMDispatchNode(Default=(X=0,Y=0,Z=0), Name=Mirror Axis, NameSpace=Root) {template="DISPATCH_RigDispatch_GetModuleMetadata(in Name,in NameSpace,in Default,out Value,out Found)"}
   Is Right Side v2 = RigVMFunctionReferenceNode(
    Mirror Plane=@Get_Module_Metadata_3.Value
    Item=&RigUnit_ResolveConnector(Connector=(Type=Connector,Name="Arm Primary"), SkipSocket=true).Result
    ExecuteContext=@RigVMFunction_Sequence_1.A
   )
   > RigVMVariableNode(Value=@"Is Right Side v2".Right Side) {variable=is on right side}
   Connect to Module Metadata = RigVMFunctionReferenceNode(
    Connector=(Type=Connector,Name="Parent Control")
    Module Metadata Name=Parent Control
    Module Metadata NameSpace=Root
    ExecuteContext=@RigVMFunction_Sequence_1.B
   )
   > Connect to Module Metadata_1 = RigVMFunctionReferenceNode(Connector=(Type=Connector,Name="Settings Control"), Module Metadata Name=Settings Control, Module Metadata NameSpace=Root)
   > Connect to Module Metadata_2 = RigVMFunctionReferenceNode(Connector=(Type=Connector,Name="Pole Vector Parent"), Module Metadata Name=Parent Control, Module Metadata NameSpace=Root)
   > Connect to Module Metadata_2_1 = RigVMFunctionReferenceNode(Connector=(Type=Connector,Name="FK Arm Space"), Module Metadata Name=Parent Control, Module Metadata NameSpace=Root)
   n1 = &RigUnit_Item(Item=(Type=Connector,Name="Hand Bone"))
   Connect to Module Metadata_1_1_1 = RigVMFunctionReferenceNode(
    Module Metadata NameSpace=Root
    Module Metadata Name=RigVMDispatchNode(False=hand_l, True=hand_r, Condition=RigVMVariableNode() {variable=is on right side}.Value) {template="DISPATCH_RigVMDispatch_If(in Condition,in True,in False,out Result)"}.Result
    Connector=@n1.Item
    ExecuteContext=@RigVMFunction_Sequence_1.C
   )
   n2 = &RigVMFunction_ControlFlowBranch(ExecuteContext=@"Connect to Module Metadata_1_1_1".ExecuteContext, Condition=@"Connect to Module Metadata_1_1_1".Found)
   n3 = /Script/ControlRig.RigUnit_GetCandidates() {is=struct RigVMUnitNode}
   n4 = &RigVMFunction_ControlFlowBranch(
    Condition=RigVMDispatchNode(B=@n3.Connector, A=@n1.Item) {template="DISPATCH_RigVMDispatch_CoreEquals(in A,in B,out Result)"}.Result
    ExecuteContext=@n2.False
   )
   n5 = RigVMDispatchNode(Array=@n3.Candidates, ExecuteContext=@n4.True) {template="DISPATCH_RigVMDispatch_ArrayIterator(in Array,out Element,out Index,out Count,out Ratio)"}
   n6 = &RigUnit_Item(Item=(Type=Connector,Name="Arm Primary"))
   n7 = &RigVMFunction_ControlFlowBranch(
    ExecuteContext=@n5.ExecuteContext
    Condition=RigVMDispatchNode(B=@n5.Element, A=&RigUnit_ResolveConnector(SkipSocket=true, Connector=@n6.Item).Result) {template="DISPATCH_RigVMDispatch_CoreEquals(in A,in B,out Result)"}.Result
   )
   /Script/ControlRig.RigUnit_SetDefaultMatch(Default=@n5.Element, ExecutePin=@n7.True) {is=struct RigVMUnitNode, template="Set Default Match::Execute(in Default)"}
   n8 = &RigUnit_Item(Item=(Type=Connector,Name="Elbow Bone"))
   Connect to Module Metadata_1_1_1_2 = RigVMFunctionReferenceNode(
    Module Metadata NameSpace=Root
    Module Metadata Name=RigVMDispatchNode(False=elbow_l, True=elbow_r, Condition=RigVMVariableNode() {variable=is on right side}.Value) {template="DISPATCH_RigVMDispatch_If(in Condition,in True,in False,out Result)"}.Result
    Connector=@n8.Item
    ExecuteContext=@RigVMFunction_Sequence_1.D
   )
   n9 = &RigVMFunction_ControlFlowBranch(ExecuteContext=@"Connect to Module Metadata_1_1_1_2".ExecuteContext, Condition=@"Connect to Module Metadata_1_1_1_2".Found)
   n10 = &RigUnit_GetCandidates()
   n11 = &RigVMFunction_ControlFlowBranch(
    Condition=RigVMDispatchNode(B=@n10.Connector, A=@n8.Item) {template="DISPATCH_RigVMDispatch_CoreEquals(in A,in B,out Result)"}.Result
    ExecuteContext=@n9.False
   )
   n12 = RigVMDispatchNode(Array=@n10.Candidates, ExecuteContext=@n11.True) {template="DISPATCH_RigVMDispatch_ArrayIterator(in Array,out Element,out Index,out Count,out Ratio)"}
   n13 = &RigUnit_ResolveConnector(SkipSocket=true, Connector=&RigUnit_Item(Item=(Type=Connector,Name="Arm Primary")).Item)
   n14 = &RigVMFunction_ControlFlowBranch(
    ExecuteContext=@n12.ExecuteContext
    Condition=RigVMDispatchNode(B=@n12.Element, A=&RigUnit_HierarchyGetParent(bDefaultParent=True, Child=@n13.Result).Parent) {template="DISPATCH_RigVMDispatch_CoreEquals(in A,in B,out Result)"}.Result
   )
   &RigUnit_SetDefaultMatch(Default=@n12.Element, ExecutePin=@n14.True) {template="Set Default Match::Execute(in Default)"}
   n15 = &RigUnit_Item(Item=(Type=Connector,Name="Shoulder Bone"))
   Connect to Module Metadata_1_1_1_1 = RigVMFunctionReferenceNode(
    Module Metadata NameSpace=Root
    Module Metadata Name=RigVMDispatchNode(False=Shoulder_L, True=Shoulder_R, Condition=RigVMVariableNode() {variable=is on right side}.Value) {template="DISPATCH_RigVMDispatch_If(in Condition,in True,in False,out Result)"}.Result
    Connector=@n15.Item
    ExecuteContext=@RigVMFunction_Sequence_1.E
   )
   n16 = &RigVMFunction_ControlFlowBranch(ExecuteContext=@"Connect to Module Metadata_1_1_1_1".ExecuteContext, Condition=@"Connect to Module Metadata_1_1_1_1".Found)
   n17 = &RigUnit_GetCandidates()
   n18 = &RigVMFunction_ControlFlowBranch(
    Condition=RigVMDispatchNode(B=@n17.Connector, A=@n15.Item) {template="DISPATCH_RigVMDispatch_CoreEquals(in A,in B,out Result)"}.Result
    ExecuteContext=@n16.False
   )
   n19 = RigVMDispatchNode(Array=@n17.Candidates, ExecuteContext=@n18.True) {template="DISPATCH_RigVMDispatch_ArrayIterator(in Array,out Element,out Index,out Count,out Ratio)"}
   n20 = &RigUnit_HierarchyGetParent(bDefaultParent=True, Child=&RigUnit_ResolveConnector(Connector=(Type=Connector,Name="Hand Bone"), SkipSocket=true).Result)
   n21 = &RigVMFunction_ControlFlowBranch(
    ExecuteContext=@n19.ExecuteContext
    Condition=RigVMDispatchNode(B=@n19.Element, A=&RigUnit_HierarchyGetParent(bDefaultParent=True, Child=@n20.Parent).Parent) {template="DISPATCH_RigVMDispatch_CoreEquals(in A,in B,out Result)"}.Result
   )
   &RigUnit_SetDefaultMatch(Default=@n19.Element, ExecutePin=@n21.True) {template="Set Default Match::Execute(in Default)"}
   RigVMFunctionReturnNode(ExecuteContext=@RigVMFunction_Sequence_1.G)
   n22 = &RigUnit_GetCandidates()
   n23 = &RigVMFunction_ControlFlowBranch(
    ExecuteContext=()
    Condition=RigVMDispatchNode(B=@n22.Connector, A=&RigUnit_Item(Item=(Type=Connector,Name="Twist Bones")).Item) {template="DISPATCH_RigVMDispatch_CoreEquals(in A,in B,out Result)"}.Result
   )
   RigVMDispatchNode(Array=@n22.Candidates, ExecuteContext=@n23.True) {template="DISPATCH_RigVMDispatch_ArrayIterator(in Array,out Element,out Index,out Count,out Ratio)"}
   > For_Each = RigVMDispatchNode(Array=&RigUnit_CollectionChildrenArray(
     TypeToSearch=Bone
     bDefaultChildren=True
     bIncludeParent=False
     bRecursive=true
     Parent=&RigUnit_ResolveConnector(Connector=(Type=Connector,Name="Shoulder Bone"), SkipSocket=true).Result
    ).Items) {template="DISPATCH_RigVMDispatch_ArrayIterator(in Array,out Element,out Index,out Count,out Ratio)"}
   Num = RigVMDispatchNode(Array=&RigUnit_CollectionChildrenArray(TypeToSearch=Bone, bDefaultChildren=True, bIncludeParent=False, bRecursive=False, Parent=@For_Each.Element).Items) {template="DISPATCH_RigVMDispatch_ArrayGetNum(in Array,out Num)"}
   Contains = /Script/RigVM.RigVMFunction_Contains(Search=twist, Name=@For_Each.Element.Name) {is=struct RigVMUnitNode, template="Contains::Execute(in Name,in Search,out Result)"}
   RigVMFunction_MathBoolAnd = &RigVMFunction_MathBoolAnd(
    B=RigVMDispatchNode(B=0, A=@Num.Num) {template="DISPATCH_RigVMDispatch_CoreEquals(in A,in B,out Result)"}.Result
    A=&RigVMFunction_MathBoolOr(
     A=@Contains.Result
     B=&RigVMFunction_Contains(Search=twist, Name=@For_Each.Element.Name) {template="Contains::Execute(in Name,in Search,out Result)"}.Result
    ).Result
   )
   RigVMFunction_ControlFlowBranch = &RigVMFunction_ControlFlowBranch(ExecuteContext=@For_Each.ExecuteContext, Condition=@RigVMFunction_MathBoolAnd.Result)
   RigVMDispatchNode(
    Enabled=True
    ScreenColor=(R=1,G=1,B=1,A=1)
    ScreenDuration=2
    Value=@For_Each.Element.Name
    ExecuteContext=@RigVMFunction_ControlFlowBranch.True
   ) {template="DISPATCH_RigVMDispatch_Print(in Prefix,in Value,in Enabled,in ScreenDuration,in ScreenColor)"}
   > &RigUnit_SetDefaultMatch(Default=@For_Each.Element) {template="Set Default Match::Execute(in Default)"}
   &RigUnit_ResolveArrayConnector(Connector=(Type=Connector,Name="Twist Bones"), SkipSocket=false)
  ]}
 New Function_ContainedGraph {nodes [
   n1 = RigVMFunctionEntryNode()
   n2 = RigVMDispatchNode(ExecuteContext=(), Array=@n1.Items) {template="DISPATCH_RigVMDispatch_ArrayIterator(in Array,out Element,out Index,out Count,out Ratio)"}
   > FindClosestBone = RigVMFunctionReferenceNode(Initial=true, Item=@n2.Element, Bones=@n1.Bones)
   > Set_Metadata = RigVMDispatchNode(Name=Closest Bone, NameSpace=Self, Item=@n2.Element, Value=@FindClosestBone.Value) {template="DISPATCH_RigDispatch_SetMetadata(in Item,in Name,in NameSpace,in Value,out Success)"}
   RigVMFunctionReturnNode(ExecuteContext=@n3.Completed)
   n3 = RigVMDispatchNode(ExecuteContext=@n1.ExecuteContext, Array=@n1.Items) {template="DISPATCH_RigVMDispatch_ArrayIterator(in Array,out Element,out Index,out Count,out Ratio)"}
   > FindClosestBone_1 = RigVMFunctionReferenceNode(Initial=true, Item=@n3.Element, Bones=@n1.Bones)
   > Set_Metadata_1 = RigVMDispatchNode(Name=Closest Bone, NameSpace=Self, Item=@n3.Element, Value=@FindClosestBone_1.Value) {template="DISPATCH_RigDispatch_SetMetadata(in Item,in Name,in NameSpace,in Value,out Success)"}
  ]}
 New Function_ContainedGraph {nodes [
   n1 = RigVMFunctionEntryNode()
   Get_Module_Metadata_4 = RigVMDispatchNode(Default=(X=0,Y=0,Z=0), Name=Mirror Axis, NameSpace=Root) {template="DISPATCH_RigDispatch_GetModuleMetadata(in Name,in NameSpace,in Default,out Value,out Found)"}
   n2 = &RigVMFunction_MathVectorMul(A=&RigUnit_GetTransform(bInitial=False, Item=@n1.Item).Transform.Translation, B=@Get_Module_Metadata_4.Value) {template="Multiply::Execute(in A,in B,out Result)"}
   RigVMFunction_MathBoolOr = RigVMAggregateNode(
    B=&RigVMFunction_MathDoubleGreater(B=0, A=@n2.Result.Y) {template="Greater::Execute(in A,in B,out Result)"}.Result
    C=&RigVMFunction_MathDoubleGreater(B=0, A=@n2.Result.Z) {template="Greater::Execute(in A,in B,out Result)"}.Result
    A=&RigVMFunction_MathDoubleGreater(B=0, A=@n2.Result.X) {template="Greater::Execute(in A,in B,out Result)"}.Result
   )
   > RigVMFunctionReturnNode(Result=@RigVMFunction_MathBoolOr.Result)
   RigVMFunction_MathBoolNot = &RigVMFunction_MathBoolNot(Value=@RigVMFunction_MathBoolOr.Result)
   Less = &RigVMFunction_MathDoubleLess(B=0, A=@Get_Module_Metadata_4.Value.X) {template="Less::Execute(in A,in B,out Result)"}
   RigVMFunction_MathBoolOr_1 = RigVMAggregateNode(
    A=@Less.Result
    B=&RigVMFunction_MathDoubleLess(B=0, A=@Get_Module_Metadata_4.Value.Y) {template="Less::Execute(in A,in B,out Result)"}.Result
    C=&RigVMFunction_MathDoubleLess(B=0, A=@Get_Module_Metadata_4.Value.Z) {template="Less::Execute(in A,in B,out Result)"}.Result
   )
   RigVMDispatchNode(True=@RigVMFunction_MathBoolNot.Result, False=@RigVMFunction_MathBoolOr.Result, Condition=@RigVMFunction_MathBoolOr_1.Result) {template="DISPATCH_RigVMDispatch_If(in Condition,in True,in False,out Result)"}
  ]}
 New Function_ContainedGraph {nodes [
   n1 = RigVMFunctionEntryNode()
   Get_Module_Metadata_4 = RigVMDispatchNode(Default=(X=1,Y=0,Z=0), Name=Mirror Axis, NameSpace=Root) {template="DISPATCH_RigDispatch_GetModuleMetadata(in Name,in NameSpace,in Default,out Value,out Found)"}
   Not_Equals_1 = RigVMDispatchNode(B=0, A=@Get_Module_Metadata_4.Value.Y) {template="DISPATCH_RigVMDispatch_CoreNotEquals(in A,in B,out Result)"}
   Not_Equals_2 = RigVMDispatchNode(B=0, A=@Get_Module_Metadata_4.Value.Z) {template="DISPATCH_RigVMDispatch_CoreNotEquals(in A,in B,out Result)"}
   Not_Equals = RigVMDispatchNode(B=0, A=@Get_Module_Metadata_4.Value.X) {template="DISPATCH_RigVMDispatch_CoreNotEquals(in A,in B,out Result)"}
   Dot___1 = &RigVMFunction_MathVectorDot(
    A=(X=0,Y=0,Z=0)
    B=(X=1,Y=1,Z=0)
    B.Y=&RigVMFunction_MathBoolToFloat(Value=@Not_Equals_1.Result).Result
    B.Z=&RigVMFunction_MathBoolToFloat(Value=@Not_Equals_2.Result).Result
    B.X=&RigVMFunction_MathBoolToFloat(Value=@Not_Equals.Result).Result
    A.X=RigVMDispatchNode(False=0, True=&RigUnit_GetTransform(bInitial=true, Item=@n1.Item).Transform.Translation.X, Condition=@Not_Equals.Result) {template="DISPATCH_RigVMDispatch_If(in Condition,in True,in False,out Result)"}.Result
    A.Y=RigVMDispatchNode(False=0, True=&RigUnit_GetTransform(bInitial=true, Item=@n1.Item).Transform.Translation.Y, Condition=@Not_Equals_1.Result) {template="DISPATCH_RigVMDispatch_If(in Condition,in True,in False,out Result)"}.Result
    A.Z=RigVMDispatchNode(False=0, True=&RigUnit_GetTransform(bInitial=true, Item=@n1.Item).Transform.Translation.Z, Condition=@Not_Equals_2.Result) {template="DISPATCH_RigVMDispatch_If(in Condition,in True,in False,out Result)"}.Result
   ) {template="Dot,|::Execute(in A,in B,out Result)"}
   Greater = &RigVMFunction_MathDoubleGreater(B=0, A=@Get_Module_Metadata_4.Value.X) {template="Greater::Execute(in A,in B,out Result)"}
   RigVMFunction_MathBoolOr = RigVMAggregateNode(
    A=@Greater.Result
    B=&RigVMFunction_MathDoubleGreater(B=0, A=@Get_Module_Metadata_4.Value.Y) {template="Greater::Execute(in A,in B,out Result)"}.Result
    C=&RigVMFunction_MathDoubleGreater(B=0, A=@Get_Module_Metadata_4.Value.Z) {template="Greater::Execute(in A,in B,out Result)"}.Result
   )
   RigVMFunction_MathBoolNot = &RigVMFunction_MathBoolNot(Value=RigVMDispatchNode(
     True=&RigVMFunction_MathDoubleGreater(B=0, A=@Dot___1.Result) {template="Greater::Execute(in A,in B,out Result)"}.Result
     False=&RigVMFunction_MathDoubleLess(B=0, A=@Dot___1.Result) {template="Less::Execute(in A,in B,out Result)"}.Result
     Condition=@RigVMFunction_MathBoolOr.Result
    ) {template="DISPATCH_RigVMDispatch_If(in Condition,in True,in False,out Result)"}.Result)
   > RigVMFunctionReturnNode(Result=@RigVMFunction_MathBoolNot.Result)
  ]}
 New Function_ContainedGraph {nodes [
   n1 = RigVMFunctionEntryNode()
   Less = &RigVMFunction_MathDoubleLess(B=0, A=@n1.Value.X) {template="Less::Execute(in A,in B,out Result)"}
   n2 = &RigVMFunction_MathDoubleLess(B=0, A=@n1.Value.Y) {template="Less::Execute(in A,in B,out Result)"}
   n3 = &RigVMFunction_MathDoubleLess(B=0, A=@n1.Value.Z) {template="Less::Execute(in A,in B,out Result)"}
   RigVMFunction_MathBoolOr_3 = RigVMAggregateNode(A=@Less.Result, B=@n2.Result, C=@n3.Result)
   Greater = &RigVMFunction_MathDoubleGreater(B=0, A=@n1.Value.X) {template="Greater::Execute(in A,in B,out Result)"}
   RigVMFunction_MathBoolOr = &RigVMFunction_MathBoolOr(A=@Greater.Result, B=@Less.Result)
   > RigVMFunctionReturnNode(
    Is Negative=RigVMDispatchNode(False=false, True=true, Condition=@RigVMFunction_MathBoolOr_3.Result) {template="DISPATCH_RigVMDispatch_If(in Condition,in True,in False,out Result)"}.Result
    X=@RigVMFunction_MathBoolOr.Result
    Y=&RigVMFunction_MathBoolOr(A=&RigVMFunction_MathDoubleGreater(B=0, A=@n1.Value.Y) {template="Greater::Execute(in A,in B,out Result)"}.Result, B=@n2.Result).Result
    Z=&RigVMFunction_MathBoolOr(A=&RigVMFunction_MathDoubleGreater(B=0, A=@n1.Value.Z) {template="Greater::Execute(in A,in B,out Result)"}.Result, B=@n3.Result).Result
   )
  ]}
 New Function_ContainedGraph {nodes [
   n1 = RigVMFunctionEntryNode()
   > RigVMFunction_Sequence = RigVMAggregateNode()
   n2 = &RigUnit_ResolveArrayConnector(Connector=(Type=Connector,Name="Twist Bones"), SkipSocket=true)
   Num = RigVMDispatchNode(Array=@n2.Result) {template="DISPATCH_RigVMDispatch_ArrayGetNum(in Array,out Num)"}
   Greater = &RigVMFunction_MathIntGreater(B=0, A=@Num.Num) {template="Greater::Execute(in A,in B,out Result)"}
   n3 = &RigVMFunction_ControlFlowBranch(Condition=@Greater.Result, ExecuteContext=@RigVMFunction_Sequence.A)
   For_Each = RigVMDispatchNode(Array=@n2.Result, ExecuteContext=@n3.True) {template="DISPATCH_RigVMDispatch_ArrayIterator(in Array,out Element,out Index,out Count,out Ratio)"}
   Get_Metadata = RigVMDispatchNode(Default=(Type=Bone,Name="None"), Name=Twist Item A, NameSpace=Self, Item=@For_Each.Element) {template="DISPATCH_RigDispatch_GetMetadata(in Item,in Name,in NameSpace,in Default,out Value,out Found)"}
   Get_Metadata_1 = RigVMDispatchNode(Default=(Type=Bone,Name="None"), Name=Twist Item B, NameSpace=Self, Item=@For_Each.Element) {template="DISPATCH_RigDispatch_GetMetadata(in Item,in Name,in NameSpace,in Default,out Value,out Found)"}
   Get_Metadata_2 = RigVMDispatchNode(Default=0, Name=Twist Amount A, NameSpace=Self, Item=@For_Each.Element) {template="DISPATCH_RigDispatch_GetMetadata(in Item,in Name,in NameSpace,in Default,out Value,out Found)"}
   Get_Metadata_3 = RigVMDispatchNode(Default=0, Name=Twist Amount B, NameSpace=Self, Item=@For_Each.Element) {template="DISPATCH_RigDispatch_GetMetadata(in Item,in Name,in NameSpace,in Default,out Value,out Found)"}
   > &RigUnit_RotationConstraintLocalSpaceOffset(
    AdvancedSettings=(InterpolationType=Shortest,RotationOrderForFilter=ZXY)
    Parents=((Item=(Type=Bone,Name="None"),Weight=1),(Item=(Type=Bone,Name="None"),Weight=1))
    Weight=1
    bMaintainOffset=true
    Filter=@n1.Filter
    Child=@For_Each.Element
    Parents.0.Item=@Get_Metadata.Value
    Parents.1.Item=@Get_Metadata_1.Value
    Parents.1.Weight=@Get_Metadata_2.Value
    Parents.0.Weight=@Get_Metadata_3.Value
   )
   > &RigUnit_PositionConstraintLocalSpaceOffset(
    Filter=(bX=True,bY=True,bZ=True)
    Parents=((Item=(Type=Bone,Name="None"),Weight=1),(Item=(Type=Bone,Name="None"),Weight=1))
    Weight=1
    bMaintainOffset=True
    Child=@For_Each.Element
    Parents.0.Item=@Get_Metadata.Value
    Parents.1.Item=@Get_Metadata_1.Value
    Parents.1.Weight=@Get_Metadata_2.Value
    Parents.0.Weight=@Get_Metadata_3.Value
   )
   > Debug Item = RigVMFunctionReferenceNode(Scale=15, Thickness=2, Active=@n1.Active, Item=@For_Each.Element)
   RigVMFunctionReturnNode(ExecuteContext=@RigVMFunction_Sequence.C)
  ]}
 New Function_ContainedGraph {nodes [
   n1 = RigVMFunctionEntryNode()
   RigVMFunction_Sequence = &RigVMFunction_Sequence(ExecuteContext=@n1.ExecuteContext)
   RigVMFunction_Sequence_1 = RigVMAggregateNode(ExecuteContext=@RigVMFunction_Sequence.A)
   n2 = &RigUnit_ResolveConnector(Connector=(Type=Connector,Name="IK Hand Gun Bone"), SkipSocket=true)
   n3 = &RigVMFunction_ControlFlowBranch(Condition=@n2.bIsConnected, ExecuteContext=@RigVMFunction_Sequence_1.A)
   &RigUnit_ParentConstraint(
    AdvancedSettings=(InterpolationType=Average,RotationOrderForFilter=XZY)
    Parents=((Item=(Type=Bone,Name="None"),Weight=1))
    Weight=1
    Child=@n2.Result
    Parents.0.Item=&RigUnit_ResolveConnector(Connector=(Type=Connector,Name="Hand Bone"), SkipSocket=true).Result
    ExecutePin=@n3.True
   )
   n4 = &RigUnit_ResolveConnector(Connector=(Type=Connector,Name="IK Hand Bone"), SkipSocket=true)
   n5 = &RigVMFunction_ControlFlowBranch(Condition=@n4.bIsConnected, ExecuteContext=@RigVMFunction_Sequence_1.B)
   &RigUnit_ParentConstraint(
    AdvancedSettings=(InterpolationType=Average,RotationOrderForFilter=XZY)
    Parents=((Item=(Type=Bone,Name="None"),Weight=1))
    Weight=1
    Child=@n4.Result
    Parents.0.Item=&RigUnit_ResolveConnector(Connector=(Type=Connector,Name="Hand Bone"), SkipSocket=true).Result
    ExecutePin=@n5.True
   )
   RigVMFunctionReturnNode(ExecuteContext=@RigVMFunction_Sequence.B)
  ]}
 RigVMFunctionLibrary {nodes [
   Attach IK Bones = RigVMCollapseNode(ExecuteContext=())
   Axis to Bool = RigVMCollapseNode(Value=(X=0,Y=0,Z=0))
   CONNECT Arm = RigVMCollapseNode(ExecuteContext=())
   CONSTR Basic Twist Rig = RigVMCollapseNode(ExecuteContext=(), Template Bones=())
   CONSTR TwoBoneLimb = RigVMCollapseNode(
    Basename=arm
    Enable Keep Aligned=False
    Enable PV Follow=False
    Enable Softness=False
    End Bone Connector=(Type=Connector,Name="Hand")
    ExecuteContext=()
    FK Control Settings="(DisplayName=\"None\",bSelectable=True,InitialSpace=LocalSpace,bUsePreferredRotationOrder=False,PreferredRotationOrder=YZX,Limits=(LimitTranslationX=(bMinimum=False,bMaximum=False),LimitTranslationY=(bMinimum=False,bMaximum=False),LimitTranslationZ=(bMinimum ...(truncated)"
    FK IK Default=false
    IK Control Basename=hand
    IK Control Settings="(DisplayName=\"None\",bSelectable=True,InitialSpace=LocalSpace,bUsePreferredRotationOrder=False,PreferredRotationOrder=YZX,Limits=(LimitTranslationX=(bMinimum=False,bMaximum=False),LimitTranslationY=(bMinimum=False,bMaximum=False),LimitTranslationZ=(bMinimum ...(truncated)"
    IK Control Spaces=()
    Local Default=False
    Middle Bone Connector=(Type=None,Name="None")
    Module Settings="(OverrideColor_48_#7=False,LeftColor_3_#6=(R=0,G=0.0231,B=0.4115,A=1),RightColor_5_#8=(R=1,G=0,B=0,A=1),CenterColor_9_#5=(R=1,G=1,B=0,A= ...(truncated)"
    Offset Elbow Bone=(X=0,Y=0,Z=0)
    Offset Elbow Bone Space=LocalSpace
    Offset IK Ctrl Transform=(Rotation=(X=0,Y=0,Z=0,W=1),Translation=(X=0,Y=0,Z=0),Scale3D=(X=1,Y=1,Z=1))
    Offset PV Control=(Rotation=(X=0,Y=0,Z=0,W=1),Translation=(X=0,Y=0,Z=0),Scale3D=(X=1,Y=1,Z=1))
    PV Control Settings="(DisplayName=\"None\",bSelectable=True,InitialSpace=LocalSpace,bUsePreferredRotationOrder=False,PreferredRotationOrder=YZX,Limits=(LimitTranslationX=(bMinimum=False,bMaximum=False),LimitTranslationY=(bMinimum=False,bMaximum=False),LimitTranslationZ=(bMinimum ...(truncated)"
    PV Control Spaces=()
    PV Offset Factor=3
    Parent Control Connector=(Type=Connector,Name="Parent Control")
    Pole Vector Parent Connector=(Type=Connector,Name="None")
    Primary Axis=(X=1,Y=0,Z=0)
    Secondary Axis=(X=0,Y=0,Z=-1)
    Settings Control Connector=(Type=Connector,Name="Settings Control")
    Start Bone Connector=(Type=Connector,Name="Shoulder")
    Wrist Control Settings="(DisplayName=\"None\",bSelectable=True,InitialSpace=LocalSpace,bUsePreferredRotationOrder=False,PreferredRotationOrder=YZX,Limits=(LimitTranslationX=(bMinimum=False,bMaximum=False),LimitTranslationY=(bMinimum=False,bMaximum=False),LimitTranslationZ=(bMinimum ...(truncated)"
   )
   Compute Pole Vector Offset for Twist = RigVMCollapseNode(
    Axis=(X=1,Y=0,Z=0)
    Debug=True
    ExecuteContext=()
    Initial=true
    Item=(Type=Null,Name="None")
    Secondary Axis=(X=0,Y=-1,Z=0)
    Secondary Direction=(X=0,Y=0,Z=1)
    Target item=(Type=Control,Name="None")
   )
   Create Bone Chain = RigVMCollapseNode(
    Debug=false
    Debug Scale=15
    Debug Thickness=0.5
    ExecuteContext=()
    Name=NewBone
    Parent=(Type=Bone,Name="None")
    Primary Axis=(X=1,Y=0,Z=0)
    Secondary Axis=(X=0,Y=1,Z=0)
    Transforms=()
   )
   FWD Basic Twist Rig = RigVMCollapseNode(Active=true, ExecuteContext=(), Filter=(bX=True,bY=True,bZ=True))
   FWD TwoBoneLimb = RigVMCollapseNode(
    Attach Skeleton=true
    Auto Determine Axis=False
    Bones=()
    Debug=True
    Driven Bones=()
    Enable PV Follow=true
    Enable Softness=False
    Enable Stretch=false
    End Ctrl=(Type=Control,Name="None")
    ExecuteContext=()
    FK Arm Rotation Space=(Type=Bone,Name="root")
    FK Controls=()
    FK Space=true
    FK_or_IK=False
    IK Controls=()
    Keep Aligned=False
    Mirror Threshold=1
    PV Follow Aim Axis=(X=1,Y=0,Z=0)
    PV Follow Aim Direction=(X=0,Y=0,Z=1)
    PV Follow Twist Lock=(bX=True,bY=False,bZ=False)
    PV Follow Up Axis=(X=0,Y=0,Z=1)
    Primary Axis=(X=1,Y=0,Z=0)
    Secondary Axis=(X=0,Y=-1,Z=0)
    Softness=1
   )
   FindClosestBone = RigVMCollapseNode(Bones=(), ExecuteContext=(), Initial=true, Item=())
   Format Name = RigVMCollapseNode(name=arm, prefix=None, side=_fk_ik_switch, suffix=_fk_ik_switch)
   Get Right Side = RigVMCollapseNode(Item=(Type=None,Name="None"))
   Get Right Side DEPRECATED = RigVMCollapseNode(ExecuteContext=(), Item=(Type=Bone))
   Get Side DEPRECATED = RigVMCollapseNode(End Bone=(Type=Bone,Name="None"), ExecuteContext=(), Left Side Suffix=None, Right Side Suffix=None, Start Bone=(Type=Bone,Name="None"))
   Get Side Metadata = RigVMCollapseNode(Item=(Type=Bone,Name="None"))
   Has Socket = RigVMCollapseNode(ExecuteContext=(), Item=(Type=Bone,Name="None"))
   IKTwoBone = RigVMCollapseNode(
    Bone A=(Type=Bone,Name="None")
    Bone B=(Type=Bone,Name="None")
    Bone C=(Type=Bone,Name="None")
    Enable Stretch=false
    End Ctrl=(Type=Control,Name="None")
    ExecuteContext=()
    Line Thickness=0.25
    Pole Vector Node=(Type=Control,Name="None")
    Primary Axis=(X=1,Y=0,Z=0)
    Secondary Axis=(X=0,Y=-1,Z=0)
    Stretch Maximum Ratio=1.25
    Stretch Start Ratio=0.75
    debug=true
    scale=(X=0,Y=0,Z=0)
   )
   INV TwoBoneLimb = RigVMCollapseNode(ExecuteContext=())
   Is Right Side v2 = RigVMCollapseNode(ExecuteContext=(), Item=(Type=Bone), Mirror Plane=(X=1,Y=1,Z=0))
   Is Right Side v2 DEPRECATED = RigVMCollapseNode(ExecuteContext=(), Item=(Type=Bone), Mirror Plane=(X=1,Y=1,Z=0))
   Is Right Side wVector = RigVMCollapseNode(ExecuteContext=(), Item=(Type=Bone), Mirror Plane=(X=1,Y=1,Z=0))
   Is Right Side wVector DEPRECATED = RigVMCollapseNode(ExecuteContext=(), Item=(Type=Bone), Mirror Plane=(X=1,Y=1,Z=0))
   Match 2 Bones IK onto FK = RigVMCollapseNode(Bones=(), ExecuteContext=(), IK Control=(Type=Control,Name="hand_l_ik_ctrl"), Pole Vector Control=(Type=Control,Name="arm_l_pv_ik_ctrl"))
   New Function = RigVMCollapseNode(ExecuteContext=())
   Offset Initial Transform = RigVMCollapseNode(
    ExecuteContext=()
    Item=(Type=Bone,Name="None")
    Space=LocalSpace
    Transform=(Rotation=(X=0,Y=0,Z=0,W=1),Translation=(X=0,Y=0,Z=0),Scale3D=(X=1,Y=1,Z=1))
   )
   Orient Bones = RigVMCollapseNode(Bones=(), Debug=false, ExecuteContext=(), Primary Axis=(X=1,Y=0,Z=0), Secondary Axis=(X=0,Y=1,Z=0))
   Project to New Parents = RigVMCollapseNode(Children=(), ExecuteContext=(), Parents=(), Propagate to Children=True)
   Set Closest Bone = RigVMCollapseNode(Bones=(), ExecuteContext=(), Items=())
  ]}
 RigVMFunction_MathBoolOr_1_ContainedGraph {nodes [
   n1 = RigVMFunctionEntryNode()
   RigVMFunction_MathBoolOr = &RigVMFunction_MathBoolOr(A=@n1.A, B=@n1.B)
   RigVMFunctionReturnNode(Result=&RigVMFunction_MathBoolOr(B=@n1.C, A=@RigVMFunction_MathBoolOr.Result).Result)
  ]}
 RigVMFunction_MathBoolOr_1_ContainedGraph {nodes [
   n1 = RigVMFunctionEntryNode()
   RigVMFunction_MathBoolOr = &RigVMFunction_MathBoolOr(A=@n1.A, B=@n1.B)
   RigVMFunctionReturnNode(Result=&RigVMFunction_MathBoolOr(B=@n1.C, A=@RigVMFunction_MathBoolOr.Result).Result)
  ]}
 RigVMFunction_MathBoolOr_1_ContainedGraph {nodes [
   n1 = RigVMFunctionEntryNode()
   RigVMFunction_MathBoolOr = &RigVMFunction_MathBoolOr(A=@n1.A, B=@n1.B)
   RigVMFunctionReturnNode(Result=&RigVMFunction_MathBoolOr(B=@n1.C, A=@RigVMFunction_MathBoolOr.Result).Result)
  ]}
 RigVMFunction_MathBoolOr_3_1_ContainedGraph {nodes [
   n1 = RigVMFunctionEntryNode()
   RigVMFunctionReturnNode(Result=&RigVMFunction_MathBoolOr(B=@n1.C, A=&RigVMFunction_MathBoolOr(A=@n1.A, B=@n1.B).Result).Result)
  ]}
 RigVMFunction_Sequence_1_1_ContainedGraph {nodes [
   n1 = RigVMFunctionEntryNode()
   n2 = &RigVMFunction_Sequence(ExecuteContext=@n1.ExecuteContext)
   n3 = &RigVMFunction_Sequence(ExecuteContext=@n2.B)
   n4 = &RigVMFunction_Sequence(ExecuteContext=@n3.B)
   n5 = &RigVMFunction_Sequence(ExecuteContext=@n4.B)
   RigVMFunctionReturnNode(A=@n2.A, B=@n3.A, C=@n4.A, D=@n5.A, E=@n5.B)
  ]}
 RigVMFunction_Sequence_1_1_ContainedGraph {nodes [
   n1 = RigVMFunctionEntryNode()
   n2 = &RigVMFunction_Sequence(ExecuteContext=@n1.ExecuteContext)
   n3 = &RigVMFunction_Sequence(ExecuteContext=@n2.B)
   n4 = &RigVMFunction_Sequence(ExecuteContext=@n3.B)
   n5 = &RigVMFunction_Sequence(ExecuteContext=@n4.B)
   n6 = &RigVMFunction_Sequence(ExecuteContext=@n5.B)
   n7 = &RigVMFunction_Sequence(ExecuteContext=@n6.B)
   n8 = &RigVMFunction_Sequence(ExecuteContext=@n7.B)
   n9 = &RigVMFunction_Sequence(ExecuteContext=@n8.B)
   n10 = &RigVMFunction_Sequence(ExecuteContext=@n9.B)
   n11 = &RigVMFunction_Sequence(ExecuteContext=@n10.B)
   n12 = &RigVMFunction_Sequence(ExecuteContext=@n11.B)
   n13 = &RigVMFunction_Sequence(ExecuteContext=@n12.B)
   n14 = &RigVMFunction_Sequence(ExecuteContext=@n13.B)
   RigVMFunctionReturnNode(A=@n2.A, K=@n12.A, L=@n13.A, M=@n14.A, N=@n14.B, B=@n3.A, C=@n4.A, D=@n5.A, E=@n6.A, F=@n7.A, G=@n8.A, H=@n9.A, I=@n10.A, J=@n11.A)
  ]}
 RigVMFunction_Sequence_1_1_ContainedGraph {nodes [
   n1 = RigVMFunctionEntryNode()
   n2 = &RigVMFunction_Sequence(ExecuteContext=@n1.ExecuteContext)
   n3 = &RigVMFunction_Sequence(ExecuteContext=@n2.B)
   n4 = &RigVMFunction_Sequence(ExecuteContext=@n3.B)
   RigVMFunctionReturnNode(A=@n2.A, B=@n3.A, C=@n4.A, D=@n4.B)
  ]}
 RigVMFunction_Sequence_1_1_ContainedGraph {nodes [
   n1 = RigVMFunctionEntryNode()
   n2 = &RigVMFunction_Sequence(ExecuteContext=@n1.ExecuteContext)
   n3 = &RigVMFunction_Sequence(ExecuteContext=@n2.B)
   n4 = &RigVMFunction_Sequence(ExecuteContext=@n3.B)
   n5 = &RigVMFunction_Sequence(ExecuteContext=@n4.B)
   n6 = &RigVMFunction_Sequence(ExecuteContext=@n5.B)
   n7 = &RigVMFunction_Sequence(ExecuteContext=@n6.B)
   RigVMFunctionReturnNode(A=@n2.A, B=@n3.A, C=@n4.A, D=@n5.A, E=@n6.A, F=@n7.A, G=@n7.B)
  ]}
 RigVMFunction_Sequence_1_ContainedGraph {nodes [
   n1 = RigVMFunctionEntryNode()
   RigVMFunction_Sequence = &RigVMFunction_Sequence(ExecuteContext=@n1.ExecuteContext)
   n2 = &RigVMFunction_Sequence(ExecuteContext=@RigVMFunction_Sequence.B)
   n3 = &RigVMFunction_Sequence(ExecuteContext=@n2.B)
   n4 = &RigVMFunction_Sequence(ExecuteContext=@n3.B)
   RigVMFunctionReturnNode(A=@RigVMFunction_Sequence.A, B=@n2.A, C=@n3.A, D=@n4.A, E=@n4.B)
  ]}
 RigVMFunction_Sequence_1_ContainedGraph {nodes [
   n1 = RigVMFunctionEntryNode()
   RigVMFunction_Sequence = &RigVMFunction_Sequence(ExecuteContext=@n1.ExecuteContext)
   n2 = &RigVMFunction_Sequence(ExecuteContext=@RigVMFunction_Sequence.B)
   RigVMFunctionReturnNode(A=@RigVMFunction_Sequence.A, B=@n2.A, C=@n2.B)
  ]}
 RigVMFunction_Sequence_2_1_ContainedGraph {nodes [
   n1 = RigVMFunctionEntryNode()
   n2 = &RigVMFunction_Sequence(ExecuteContext=@n1.ExecuteContext)
   n3 = &RigVMFunction_Sequence(ExecuteContext=@n2.B)
   n4 = &RigVMFunction_Sequence(ExecuteContext=@n3.B)
   RigVMFunctionReturnNode(A=@n2.A, B=@n3.A, C=@n4.A, D=@n4.B)
  ]}
 RigVMFunction_Sequence_2_ContainedGraph {nodes [
   n1 = RigVMFunctionEntryNode()
   RigVMFunction_Sequence = &RigVMFunction_Sequence(ExecuteContext=@n1.ExecuteContext)
   n2 = &RigVMFunction_Sequence(ExecuteContext=@RigVMFunction_Sequence.B)
   n3 = &RigVMFunction_Sequence(ExecuteContext=@n2.B)
   n4 = &RigVMFunction_Sequence(ExecuteContext=@n3.B)
   RigVMFunctionReturnNode(A=@RigVMFunction_Sequence.A, B=@n2.A, C=@n3.A, D=@n4.A, E=@n4.B)
  ]}
 RigVMFunction_Sequence_2_ContainedGraph {nodes [
   n1 = RigVMFunctionEntryNode()
   RigVMFunction_Sequence = &RigVMFunction_Sequence(ExecuteContext=@n1.ExecuteContext)
   n2 = &RigVMFunction_Sequence(ExecuteContext=@RigVMFunction_Sequence.B)
   RigVMFunctionReturnNode(A=@RigVMFunction_Sequence.A, B=@n2.A, C=@n2.B)
  ]}
 RigVMFunction_Sequence_2_ContainedGraph {nodes [
   n1 = RigVMFunctionEntryNode()
   n2 = &RigVMFunction_Sequence(ExecuteContext=@n1.ExecuteContext)
   n3 = &RigVMFunction_Sequence(ExecuteContext=@n2.B)
   RigVMFunctionReturnNode(A=@n2.A, B=@n3.A, C=@n3.B)
  ]}
 RigVMFunction_Sequence_2_ContainedGraph {nodes [
   n1 = RigVMFunctionEntryNode()
   n2 = &RigVMFunction_Sequence(ExecuteContext=@n1.ExecuteContext)
   n3 = &RigVMFunction_Sequence(ExecuteContext=@n2.B)
   RigVMFunctionReturnNode(A=@n2.A, B=@n3.A, C=@n3.B)
  ]}
 RigVMModel {nodes [
   n1 = &RigUnit_BeginExecution()
   RigVMFunction_Sequence = &RigVMFunction_Sequence(ExecuteContext=@n1.ExecutePin)
   Get_Module_Metadata_3_1 = RigVMDispatchNode(Default=true, Name=Enable Softness, NameSpace=Root) {template="DISPATCH_RigDispatch_GetModuleMetadata(in Name,in NameSpace,in Default,out Value,out Found)"}
   RigVMFunction_ControlFlowBranch = &RigVMFunction_ControlFlowBranch(Condition=@Get_Module_Metadata_3_1.Found, ExecuteContext=@RigVMFunction_Sequence.A)
   RigVMVariableNode(Value=@Get_Module_Metadata_3_1.Value, ExecuteContext=@RigVMFunction_ControlFlowBranch.True) {variable=Enable Softness}
   DISPATCH_RigDispatch_GetModuleMetadata = RigVMDispatchNode(Default=None, Name=fk_ik_switch, NameSpace=Self) {template="DISPATCH_RigDispatch_GetModuleMetadata(in Name,in NameSpace,in Default,out Value,out Found)"}
   GetAnimationChannel = /Script/ControlRig.RigUnit_GetBoolAnimationChannel(
    bInitial=False
    Channel=@DISPATCH_RigDispatch_GetModuleMetadata.Value
    Control=&RigUnit_ResolveConnector(Connector=(Type=Connector,Name="Settings Control"), SkipSocket=true).Result.Name
   ) {is=struct RigVMUnitNode, template="GetAnimationChannel::Execute(out Value,in Control,in Channel,in bInitial)"}
   n2 = &RigUnit_FindItemsWithMetadataTag(NameSpace=Self, Tag=Controls IK)
   n3 = &RigUnit_FindItemsWithMetadataTag(NameSpace=Self, Tag=Controls IK)
   n4 = &RigUnit_FindItemsWithMetadataTag(NameSpace=Self, Tag=Controls IK)
   DISPATCH_RigVMDispatch_ArrayGetAtIndex = RigVMDispatchNode(Index=0, Array=&RigUnit_FindItemsWithMetadataTag(NameSpace=Self, Tag=Controls FK).Items) {template="DISPATCH_RigVMDispatch_ArrayGetAtIndex(in Array,in Index,out Element)"}
   Get_Module_Metadata_2 = RigVMDispatchNode(Default=(X=0,Y=0,Z=0), Name=Primary Axis, NameSpace=Self) {template="DISPATCH_RigDispatch_GetModuleMetadata(in Name,in NameSpace,in Default,out Value,out Found)"}
   n5 = RigVMVariableNode() {variable=Module Settings}
   Get_Module_Metadata_2_1 = RigVMDispatchNode(Default=(X=0,Y=0,Z=0), Name=Secondary Axis, NameSpace=Self) {template="DISPATCH_RigDispatch_GetModuleMetadata(in Name,in NameSpace,in Default,out Value,out Found)"}
   n6 = RigVMDispatchNode(Index=0, Array=&RigUnit_FindItemsWithMetadataTag(NameSpace=Self, Tag=Controls IK).Items) {template="DISPATCH_RigVMDispatch_ArrayGetAtIndex(in Array,in Index,out Element)"}
   FWD TwoBoneLimb = RigVMFunctionReferenceNode(
    Attach Skeleton=true
    Debug=true
    End Ctrl=(Type=Control,Name="None")
    FK Space=true
    Mirror Threshold=-1
    PV Follow Aim Axis=(X=1,Y=0,Z=0)
    PV Follow Aim Direction=(X=0,Y=0,Z=1)
    PV Follow Twist Lock=(bX=True,bY=False,bZ=False)
    PV Follow Up Axis=(X=0,Y=0,Z=1)
    Bones=RigVMDispatchNode(Default=(), Name=Rig Bones, NameSpace=Self) {template="DISPATCH_RigDispatch_GetModuleMetadata(in Name,in NameSpace,in Default,out Value,out Found)"}.Value
    IK Controls=&RigUnit_FindItemsWithMetadataTag(NameSpace=Self, Tag=Controls IK).Items
    Driven Bones=&RigUnit_FindItemsWithMetadataTag(NameSpace=Self, Tag=Bones).Items
    FK Controls=&RigUnit_FindItemsWithMetadataTag(NameSpace=Self, Tag=Controls FK).Items
    FK_or_IK=@GetAnimationChannel.Value
    Softness=/Script/ControlRig.RigUnit_GetFloatAnimationChannel(
     Channel=Softness
     bInitial=False
     Control=RigVMDispatchNode(Index=0, Array=@n2.Items) {template="DISPATCH_RigVMDispatch_ArrayGetAtIndex(in Array,in Index,out Element)"}.Element.Name
    ) {is=struct RigVMUnitNode, template="GetAnimationChannel::Execute(out Value,in Control,in Channel,in bInitial)"}.Value
    Enable Stretch=&RigUnit_GetBoolAnimationChannel(
     Channel=Stretch
     bInitial=False
     Control=RigVMDispatchNode(Index=0, Array=@n3.Items) {template="DISPATCH_RigVMDispatch_ArrayGetAtIndex(in Array,in Index,out Element)"}.Element.Name
    ) {template="GetAnimationChannel::Execute(out Value,in Control,in Channel,in bInitial)"}.Value
    Keep Aligned=&RigUnit_GetBoolAnimationChannel(
     Channel=Keep Aligned
     bInitial=False
     Control=RigVMDispatchNode(Index=0, Array=@n4.Items) {template="DISPATCH_RigVMDispatch_ArrayGetAtIndex(in Array,in Index,out Element)"}.Element.Name
    ) {template="GetAnimationChannel::Execute(out Value,in Control,in Channel,in bInitial)"}.Value
    local=&RigUnit_GetBoolAnimationChannel(Channel=local, bInitial=False, Control=@DISPATCH_RigVMDispatch_ArrayGetAtIndex.Element.Name) {template="GetAnimationChannel::Execute(out Value,in Control,in Channel,in bInitial)"}.Value
    Primary Axis=RigVMDispatchNode(
     Condition=@Get_Module_Metadata_2.Found
     True=@Get_Module_Metadata_2.Value
     False=@n5.Value.PrimaryBoneAxis_32_#9
    ) {template="DISPATCH_RigVMDispatch_If(in Condition,in True,in False,out Result)"}.Result
    Secondary Axis=RigVMDispatchNode(
     Condition=@Get_Module_Metadata_2_1.Found
     True=@Get_Module_Metadata_2_1.Value
     False=@n5.Value.SecondaryAxis_51_#10
    ) {template="DISPATCH_RigVMDispatch_If(in Condition,in True,in False,out Result)"}.Result
    Auto Determine Axis=RigVMDispatchNode(True=False, Condition=@Get_Module_Metadata_2.Found, False=RigVMVariableNode() {variable=Auto Determine Axis}.Value) {template="DISPATCH_RigVMDispatch_If(in Condition,in True,in False,out Result)"}.Result
    Enable PV Follow=RigVMDispatchNode(
     False=False
     True=&RigUnit_GetBoolAnimationChannel(Channel=PV Follow, bInitial=False, Control=@n6.Element.Name) {template="GetAnimationChannel::Execute(out Value,in Control,in Channel,in bInitial)"}.Value
     Condition=RigVMVariableNode() {variable=Enable PV Follow}.Value
    ) {template="DISPATCH_RigVMDispatch_If(in Condition,in True,in False,out Result)"}.Result
    FK Arm Rotation Space=&RigUnit_ResolveConnector(Connector=(Type=Connector,Name="FK Arm Space"), SkipSocket=true).Result
    ExecuteContext=@RigVMFunction_Sequence.B
    Enable Softness=RigVMVariableNode() {variable=Enable Softness}.Value
   )
   > FWD Basic Twist Rig = RigVMFunctionReferenceNode(Filter=RigVMVariableNode() {variable=Twist Bone Axis}.Value, Active=RigVMVariableNode() {variable=Debug Twist Rigs}.Value)
  ]}
 RigVMModel Backwards Solve Graph {nodes [
   n1 = /Script/ControlRig.RigUnit_InverseExecution() {is=struct RigVMUnitNode}
   INV TwoBoneLimb_1 = RigVMFunctionReferenceNode(ExecuteContext=@n1.ExecutePin)
  ]}
 RigVMModel Connector Event Graph {nodes [
   n1 = /Script/ControlRig.RigUnit_ConnectorExecution() {is=struct RigVMUnitNode}
   CONNECT Arm = RigVMFunctionReferenceNode(ExecuteContext=@n1.ExecuteContext)
  ]}
]
hierarchy [
 Arm Primary {type=Connector, loc=<0 0 0>, rot=<0 0 0 1>}
 Root {type=Socket, loc=<0 0 0>, rot=<0 0 0 1>}
 Parent Control {type=Connector, loc=<0 0 0>, rot=<0 0 0 1>}
 Hand Bone {type=Connector, loc=<0 0 0>, rot=<0 0 0 1>}
 Shoulder Bone {type=Connector, loc=<0 0 0>, rot=<0 0 0 1>}
 Settings Control {type=Connector, loc=<0 0 0>, rot=<0 0 0 1>}
 Pole Vector Parent {type=Connector, loc=<0 0 0>, rot=<0 0 0 1>}
 FK Arm Space {type=Connector, loc=<0 0 0>, rot=<0 0 0 1>}
 Elbow Bone {type=Connector, loc=<0 0 0>, rot=<0 0 0 1>}
 Twist Bones {type=Connector, loc=<0 0 0>, rot=<0 0 0 1>}
]
variables [
 Array {type="TArray<FRigElementKey>"}
 Auto Determine Axis {type=bool, default=True}
 Auto Shoulder - Experiment {type=bool, default=True}
 Debug Twist Rigs {type=bool, default=False}
 Elbow Bone {type=FRigElementKey, default=(Type=Connector,Name="")}
 Elbow Bone Offset {type=FVector, default=(X=0,Y=0,Z=0)}
 Elbow Bone Offset Global Space {type=bool, default=False}
 Enable Keep Aligned {type=bool, default=True}
 Enable PV Follow {type=bool, default=True}
 Enable Softness {type=bool, default=False}
 FK IK Default {type=bool, default=True}
 FK Shape Settings {
  type=FRigUnit_HierarchyAddControlTransform_Settings
  default="(InitialSpace=LocalSpace,bUsePreferredRotationOrder=False,PreferredRotationOrder=YZX,Limits=(LimitTranslationX=(bMinimum=False,bMaximum=False),LimitTranslationY=(bMinimum=False,bMaximum=False),LimitTranslationZ=(bMinimum=False,bMaximum=False),LimitPitch=(b ...(truncated)"
 }
 Finger Leaf Bones {type="TArray<FRigElementKey>"}
 Hand Bone {type=FRigElementKey, default=(Type=Connector,Name="")}
 IK Shape Settings {
  type=FRigUnit_HierarchyAddControlTransform_Settings
  default="(InitialSpace=LocalSpace,bUsePreferredRotationOrder=False,PreferredRotationOrder=YZX,Limits=(LimitTranslationX=(bMinimum=False,bMaximum=False),LimitTranslationY=(bMinimum=False,bMaximum=False),LimitTranslationZ=(bMinimum=False,bMaximum=False),LimitPitch=(b ...(truncated)"
 }
 Is Right Side {type=bool, default=False}
 Local Default {type=bool, default=True}
 Module Settings {
  type=FUserDefinedStruct_#11
  default="(OverrideColor_48_#7=False,LeftColor_3_#6=(R=0,G=0.0231,B=0.4115,A=1),RightColor_5_#8=(R=1,G=0,B=0,A=1),CenterColor_9_#5=(R=1,G=1,B=0,A= ...(truncated)"
 }
 Offset IK Ctrl {type=FTransform, default=(Rotation=(X=0,Y=0,Z=0,W=1),Translation=(X=0,Y=0,Z=0),Scale3D=(X=1,Y=1,Z=1))}
 Offset PV Control {type=FTransform, default=(Rotation=(X=0,Y=0,Z=0,W=1),Translation=(X=0,Y=0,Z=0),Scale3D=(X=1,Y=1,Z=1))}
 Orient IK to World {type=bool, default=False}
 PV Shape Settings {
  type=FRigUnit_HierarchyAddControlTransform_Settings
  default="(InitialSpace=LocalSpace,bUsePreferredRotationOrder=False,PreferredRotationOrder=YZX,Limits=(LimitTranslationX=(bMinimum=False,bMaximum=False),LimitTranslationY=(bMinimum=False,bMaximum=False),LimitTranslationZ=(bMinimum=False,bMaximum=False),LimitPitch=(b ...(truncated)"
 }
 Pole Vector Position {type=float, default=1}
 Shoulder Bone {type=FRigElementKey, default=(Type=Connector,Name="")}
 Side {type=FName, default=None}
 Template Bones {type="TArray<FRigElementKey>"}
 Twist Bone Axis {type=FFilterOptionPerAxis, default=(bX=True,bY=False,bZ=False)}
 follow_null {type=FRigElementKey, default=(Type=Bone,Name="")}
 foot_control {type=FRigElementKey, default=(Type=None,Name="")}
 prefix {type=FName, default=None}
 pv_control {type=FRigElementKey, default=(Type=None,Name="")}
]
