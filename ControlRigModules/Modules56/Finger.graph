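# Modular Control Rig finger module that builds an FK control chain between Start and End bone connectors.
# Construction creates side-oriented rig bones and parented controls, stores them as module metadata and tags, then finalizes the controls.
# Forward solve drives rig bones from controls and optionally attaches skeleton bones; backward solve matches controls and rig bones to cached closest bones.
# Connector events suggest Start and End bones from the primary connector's hierarchy.
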ControlRigBlueprint /ControlRigModules/Modules56/Finger
graphs [
 ConstructionGraph {nodes [
   n1 = &RigUnit_PrepareForExecution()
   CONSTR Finger = RigVMFunctionReferenceNode(
    End Connector=(Type=Connector,Name="End Bone")
    Parent Control Connector=(Type=Connector,Name="Parent Control")
    Proxy Ctrl Connector=(Type=Connector,Name="Proxy Ctrl Location")
    Start Connector=(Type=Connector,Name="Start Bone")
    ExecuteContext=@n1.ExecutePin
    Module Settings=RigVMVariableNode() {variable=Module Settings}.Value
    Override Shape Settings=RigVMVariableNode() {variable=Override Control Settings}.Value
    Shape Settings=RigVMVariableNode() {variable=Control Settings}.Value
   )
  ]}
 ContainedGraph {nodes [
   n1 = RigVMFunctionEntryNode()
   RigUnit_SequenceAggregate_1 = &RigVMFunction_Sequence(ExecuteContext=@n1.ExecuteContext)
   RigUnit_SequenceAggregate = RigVMAggregateNode(ExecuteContext=@RigUnit_SequenceAggregate_1.A)
   Greater = /Script/RigVM.RigVMFunction_MathDoubleGreater(B=0, A=@n1.Mirror Plane.X) {is=struct RigVMUnitNode, template="Greater::Execute(in A,in B,out Result)"}
   BranchNode_4 = &RigVMFunction_ControlFlowBranch(Condition=@Greater.Result, ExecuteContext=@RigUnit_SequenceAggregate.A)
   n2 = RigVMVariableNode() {variable=vector pos}
   RigVMVariableNode(
    Value=(X=0,Y=0,Z=0)
    ExecuteContext=@BranchNode_4.True
    Value.X=&RigUnit_GetTransform(bInitial=true, Item=@n1.Item).Transform.Translation.X
    Value.Y=@n2.Value.Y
    Value.Z=@n2.Value.Z
   ) {variable=vector pos}
   n3 = &RigVMFunction_MathDoubleGreater(B=0, A=@n1.Mirror Plane.Y) {template="Greater::Execute(in A,in B,out Result)"}
   BranchNode_1_1 = &RigVMFunction_ControlFlowBranch(Condition=@n3.Result, ExecuteContext=@RigUnit_SequenceAggregate.B)
   n4 = RigVMVariableNode() {variable=vector pos}
   RigVMVariableNode(
    Value=(X=0,Y=0,Z=0)
    ExecuteContext=@BranchNode_1_1.True
    Value.Y=&RigUnit_GetTransform(bInitial=true, Item=@n1.Item).Transform.Translation.Y
    Value.X=@n4.Value.X
    Value.Z=@n4.Value.Z
   ) {variable=vector pos}
   n5 = &RigVMFunction_MathDoubleGreater(B=0, A=@n1.Mirror Plane.Z) {template="Greater::Execute(in A,in B,out Result)"}
   BranchNode_2_1 = &RigVMFunction_ControlFlowBranch(Condition=@n5.Result, ExecuteContext=@RigUnit_SequenceAggregate.C)
   n6 = RigVMVariableNode() {variable=vector pos}
   RigVMVariableNode(
    Value=(X=0,Y=0,Z=0)
    ExecuteContext=@BranchNode_2_1.True
    Value.Z=&RigUnit_GetTransform(bInitial=true, Item=@n1.Item).Transform.Translation.Z
    Value.X=@n6.Value.X
    Value.Y=@n6.Value.Y
   ) {variable=vector pos}
   Cast = &RigVMFunction_MathBoolToFloat(Value=@Greater.Result)
   Dot__ = /Script/RigVM.RigVMFunction_MathVectorDot(
    B=(X=1,Y=1,Z=0)
    B.X=@Cast.Result
    B.Y=&RigVMFunction_MathBoolToFloat(Value=@n3.Result).Result
    B.Z=&RigVMFunction_MathBoolToFloat(Value=@n5.Result).Result
    A=RigVMVariableNode() {variable=vector pos}.Value
   ) {is=struct RigVMUnitNode, template="Dot,|::Execute(in A,in B,out Result)"}
   Less = /Script/RigVM.RigVMFunction_MathDoubleLess(B=0, A=@Dot__.Result) {is=struct RigVMUnitNode, template="Less::Execute(in A,in B,out Result)"}
   RigVMVariableNode(Value=@Less.Result, ExecuteContext=@RigUnit_SequenceAggregate.D) {variable=is on right}
   RigVMFunctionReturnNode(ExecuteContext=@RigUnit_SequenceAggregate_1.B, Right Side=RigVMVariableNode() {variable=is on right}.Value)
  ]}
 ContainedGraph {nodes [
   n1 = RigVMFunctionEntryNode()
   RigUnit_SequenceAggregate = &RigVMFunction_Sequence(ExecuteContext=@n1.ExecuteContext)
   RigUnit_SequenceAggregate_1 = &RigVMFunction_Sequence(ExecuteContext=@RigUnit_SequenceAggregate.B)
   RigUnit_SequenceAggregate_2 = &RigVMFunction_Sequence(ExecuteContext=@RigUnit_SequenceAggregate_1.B)
   RigVMFunctionReturnNode(A=@RigUnit_SequenceAggregate.A, B=@RigUnit_SequenceAggregate_1.A, C=@RigUnit_SequenceAggregate_2.A, D=@RigUnit_SequenceAggregate_2.B)
  ]}
 ContainedGraph {nodes [
   n1 = RigVMFunctionEntryNode()
   RigUnit_SequenceAggregate_1 = &RigVMFunction_Sequence(ExecuteContext=@n1.ExecuteContext)
   RigUnit_SequenceAggregate = RigVMAggregateNode(ExecuteContext=@RigUnit_SequenceAggregate_1.A)
   Greater = &RigVMFunction_MathDoubleGreater(B=0, A=@n1.Mirror Plane.X) {template="Greater::Execute(in A,in B,out Result)"}
   BranchNode_4 = &RigVMFunction_ControlFlowBranch(Condition=@Greater.Result, ExecuteContext=@RigUnit_SequenceAggregate.A)
   n2 = &RigVMFunction_MathDoubleGreater(B=0, A=@n1.Mirror Plane.Y) {template="Greater::Execute(in A,in B,out Result)"}
   BranchNode_1_1 = &RigVMFunction_ControlFlowBranch(ExecuteContext=@BranchNode_4.Completed, Condition=@n2.Result)
   n3 = &RigVMFunction_MathDoubleGreater(B=0, A=@n1.Mirror Plane.Z) {template="Greater::Execute(in A,in B,out Result)"}
   BranchNode_2_1 = &RigVMFunction_ControlFlowBranch(ExecuteContext=@BranchNode_1_1.Completed, Condition=@n3.Result)
   Cast = &RigVMFunction_MathBoolToFloat(Value=@Greater.Result)
   Dot__ = &RigVMFunction_MathVectorDot(
    B=(X=1,Y=1,Z=0)
    B.X=@Cast.Result
    B.Y=&RigVMFunction_MathBoolToFloat(Value=@n2.Result).Result
    B.Z=&RigVMFunction_MathBoolToFloat(Value=@n3.Result).Result
    A=RigVMVariableNode() {variable=vector pos}.Value
   ) {template="Dot,|::Execute(in A,in B,out Result)"}
   Less = &RigVMFunction_MathDoubleLess(B=0, A=@Dot__.Result) {template="Less::Execute(in A,in B,out Result)"}
   RigVMVariableNode(ExecuteContext=@BranchNode_2_1.Completed, Value=@Less.Result) {variable=is on right}
   n4 = RigVMVariableNode() {variable=vector pos}
   RigVMVariableNode(
    Value=(X=0,Y=0,Z=0)
    ExecuteContext=@BranchNode_2_1.True
    Value.Z=&RigUnit_GetTransform(bInitial=true, Item=@n1.Item).Transform.Translation.Z
    Value.X=@n4.Value.X
    Value.Y=@n4.Value.Y
   ) {variable=vector pos}
   Dot___1_1_1_1_1_2 = &RigVMFunction_MathVectorDot(
    B=(X=1,Y=1,Z=0)
    B.Y=&RigVMFunction_MathBoolToFloat(Value=@n2.Result).Result
    B.Z=&RigVMFunction_MathBoolToFloat(Value=@n3.Result).Result
    B.X=&RigVMFunction_MathBoolToFloat(Value=@Greater.Result).Result
    A=RigVMVariableNode() {variable=vector pos}.Value
   ) {template="Dot,|::Execute(in A,in B,out Result)"}
   > RigVMVariableNode(Value=&RigVMFunction_MathDoubleLess(B=0, A=@Dot___1_1_1_1_1_2.Result) {template="Less::Execute(in A,in B,out Result)"}.Result) {variable=is on right}
   n5 = RigVMVariableNode() {variable=vector pos}
   RigVMVariableNode(
    Value=(X=0,Y=0,Z=0)
    ExecuteContext=@BranchNode_1_1.True
    Value.Y=&RigUnit_GetTransform(bInitial=true, Item=@n1.Item).Transform.Translation.Y
    Value.X=@n5.Value.X
    Value.Z=@n5.Value.Z
   ) {variable=vector pos}
   Dot___1_1_1_1_1_1 = &RigVMFunction_MathVectorDot(
    B=(X=1,Y=1,Z=0)
    B.Y=&RigVMFunction_MathBoolToFloat(Value=@n2.Result).Result
    B.Z=&RigVMFunction_MathBoolToFloat(Value=@n3.Result).Result
    B.X=&RigVMFunction_MathBoolToFloat(Value=@Greater.Result).Result
    A=RigVMVariableNode() {variable=vector pos}.Value
   ) {template="Dot,|::Execute(in A,in B,out Result)"}
   > RigVMVariableNode(Value=&RigVMFunction_MathDoubleLess(B=0, A=@Dot___1_1_1_1_1_1.Result) {template="Less::Execute(in A,in B,out Result)"}.Result) {variable=is on right}
   n6 = RigVMVariableNode() {variable=vector pos}
   RigVMVariableNode(
    Value=(X=0,Y=0,Z=0)
    ExecuteContext=@BranchNode_4.True
    Value.X=&RigUnit_GetTransform(bInitial=true, Item=@n1.Item).Transform.Translation.X
    Value.Y=@n6.Value.Y
    Value.Z=@n6.Value.Z
   ) {variable=vector pos}
   Dot___1_1_1_1_1 = &RigVMFunction_MathVectorDot(
    B=(X=1,Y=1,Z=0)
    B.Y=&RigVMFunction_MathBoolToFloat(Value=@n2.Result).Result
    B.Z=&RigVMFunction_MathBoolToFloat(Value=@n3.Result).Result
    B.X=&RigVMFunction_MathBoolToFloat(Value=@Greater.Result).Result
    A=RigVMVariableNode() {variable=vector pos}.Value
   ) {template="Dot,|::Execute(in A,in B,out Result)"}
   > RigVMVariableNode(Value=&RigVMFunction_MathDoubleLess(B=0, A=@Dot___1_1_1_1_1.Result) {template="Less::Execute(in A,in B,out Result)"}.Result) {variable=is on right}
   n7 = &RigVMFunction_MathDoubleLess(B=0, A=@n1.Mirror Plane.X) {template="Less::Execute(in A,in B,out Result)"}
   BranchNode_4_1 = &RigVMFunction_ControlFlowBranch(Condition=@n7.Result, ExecuteContext=@RigUnit_SequenceAggregate.D)
   n8 = RigVMVariableNode() {variable=vector pos}
   RigVMVariableNode(
    Value=(X=0,Y=0,Z=0)
    ExecuteContext=@BranchNode_4_1.True
    Value.X=&RigUnit_GetTransform(bInitial=true, Item=@n1.Item).Transform.Translation.X
    Value.Y=@n8.Value.Y
    Value.Z=@n8.Value.Z
   ) {variable=vector pos}
   n9 = &RigVMFunction_MathDoubleLess(B=0, A=@n1.Mirror Plane.Y) {template="Less::Execute(in A,in B,out Result)"}
   n10 = &RigVMFunction_MathDoubleLess(B=0, A=@n1.Mirror Plane.Z) {template="Less::Execute(in A,in B,out Result)"}
   Dot___1_1_1_1 = &RigVMFunction_MathVectorDot(
    B=(X=1,Y=1,Z=0)
    B.Y=&RigVMFunction_MathBoolToFloat(Value=@n9.Result).Result
    B.Z=&RigVMFunction_MathBoolToFloat(Value=@n10.Result).Result
    B.X=&RigVMFunction_MathBoolToFloat(Value=@n7.Result).Result
    A=RigVMVariableNode() {variable=vector pos}.Value
   ) {template="Dot,|::Execute(in A,in B,out Result)"}
   > RigVMVariableNode(Value=&RigVMFunction_MathDoubleGreater(B=0, A=@Dot___1_1_1_1.Result) {template="Greater::Execute(in A,in B,out Result)"}.Result) {variable=is on right}
   BranchNode_1_1_1 = &RigVMFunction_ControlFlowBranch(Condition=@n9.Result, ExecuteContext=@RigUnit_SequenceAggregate.E)
   n11 = RigVMVariableNode() {variable=vector pos}
   RigVMVariableNode(
    Value=(X=0,Y=0,Z=0)
    ExecuteContext=@BranchNode_1_1_1.True
    Value.Y=&RigUnit_GetTransform(bInitial=true, Item=@n1.Item).Transform.Translation.Y
    Value.X=@n11.Value.X
    Value.Z=@n11.Value.Z
   ) {variable=vector pos}
   Dot___1_1 = &RigVMFunction_MathVectorDot(
    B=(X=1,Y=1,Z=0)
    B.Y=&RigVMFunction_MathBoolToFloat(Value=@n9.Result).Result
    B.Z=&RigVMFunction_MathBoolToFloat(Value=@n10.Result).Result
    B.X=&RigVMFunction_MathBoolToFloat(Value=@n7.Result).Result
    A=RigVMVariableNode() {variable=vector pos}.Value
   ) {template="Dot,|::Execute(in A,in B,out Result)"}
   > RigVMVariableNode(Value=&RigVMFunction_MathDoubleGreater(B=0, A=@Dot___1_1.Result) {template="Greater::Execute(in A,in B,out Result)"}.Result) {variable=is on right}
   BranchNode_2_1_1 = &RigVMFunction_ControlFlowBranch(Condition=@n10.Result, ExecuteContext=@RigUnit_SequenceAggregate.F)
   n12 = RigVMVariableNode() {variable=vector pos}
   RigVMVariableNode(
    Value=(X=0,Y=0,Z=0)
    ExecuteContext=@BranchNode_2_1_1.True
    Value.Z=&RigUnit_GetTransform(bInitial=true, Item=@n1.Item).Transform.Translation.Z
    Value.X=@n12.Value.X
    Value.Y=@n12.Value.Y
   ) {variable=vector pos}
   Dot___1_1_1 = &RigVMFunction_MathVectorDot(
    B=(X=1,Y=1,Z=0)
    B.Y=&RigVMFunction_MathBoolToFloat(Value=@n9.Result).Result
    B.Z=&RigVMFunction_MathBoolToFloat(Value=@n10.Result).Result
    B.X=&RigVMFunction_MathBoolToFloat(Value=@n7.Result).Result
    A=RigVMVariableNode() {variable=vector pos}.Value
   ) {template="Dot,|::Execute(in A,in B,out Result)"}
   > RigVMVariableNode(Value=&RigVMFunction_MathDoubleGreater(B=0, A=@Dot___1_1_1.Result) {template="Greater::Execute(in A,in B,out Result)"}.Result) {variable=is on right}
   RigVMFunctionReturnNode(ExecuteContext=@RigUnit_SequenceAggregate_1.B, Right Side=RigVMVariableNode() {variable=is on right}.Value)
   &RigVMFunction_MathDoubleLess(B=0, A=@Dot___1_1.Result) {template="Less::Execute(in A,in B,out Result)"}
   &RigVMFunction_MathDoubleLess(B=0, A=@Dot___1_1_1.Result) {template="Less::Execute(in A,in B,out Result)"}
   &RigVMFunction_MathDoubleLess(B=0, A=@Dot___1_1_1_1.Result) {template="Less::Execute(in A,in B,out Result)"}
  ]}
 ContainedGraph {nodes [
   n1 = RigVMFunctionEntryNode()
   RigUnit_SequenceAggregate = &RigVMFunction_Sequence(ExecuteContext=@n1.ExecuteContext)
   RigUnit_SequenceAggregate_1 = &RigVMFunction_Sequence(ExecuteContext=@RigUnit_SequenceAggregate.B)
   RigUnit_SequenceAggregate_2 = &RigVMFunction_Sequence(ExecuteContext=@RigUnit_SequenceAggregate_1.B)
   RigUnit_SequenceAggregate_3 = &RigVMFunction_Sequence(ExecuteContext=@RigUnit_SequenceAggregate_2.B)
   RigUnit_SequenceAggregate_4 = &RigVMFunction_Sequence(ExecuteContext=@RigUnit_SequenceAggregate_3.B)
   RigUnit_SequenceAggregate_5 = &RigVMFunction_Sequence(ExecuteContext=@RigUnit_SequenceAggregate_4.B)
   RigVMFunctionReturnNode(
    A=@RigUnit_SequenceAggregate.A
    B=@RigUnit_SequenceAggregate_1.A
    C=@RigUnit_SequenceAggregate_2.A
    D=@RigUnit_SequenceAggregate_3.A
    E=@RigUnit_SequenceAggregate_4.A
    F=@RigUnit_SequenceAggregate_5.A
    G=@RigUnit_SequenceAggregate_5.B
   )
  ]}
 New Function_1_ContainedGraph {nodes [
   n1 = RigVMFunctionEntryNode()
   n2 = &RigVMFunction_Sequence(ExecuteContext=@n1.ExecuteContext)
   RigVMFunction_Sequence = RigVMAggregateNode(ExecuteContext=@n2.A)
   n3 = /Script/ControlRig.RigUnit_ResolveConnector(SkipSocket=true, Connector=RigVMVariableNode() {variable=Start Connector}.Value) {is=struct RigVMUnitNode}
   n4 = /Script/ControlRig.RigUnit_HierarchyGetChainItemArray(
    bIncludeEnd=True
    bIncludeStart=True
    bReverse=false
    Start=@n3.Result
    End=&RigUnit_ResolveConnector(SkipSocket=true, Connector=RigVMVariableNode() {variable=End Connector}.Value).Result
   ) {is=struct RigVMUnitNode}
   Set_Module_Metadata = RigVMDispatchNode(Name=Bones, NameSpace=Self, Value=@n4.Chain, ExecuteContext=@RigVMFunction_Sequence.A) {template="DISPATCH_RigDispatch_SetModuleMetadata(in Name,in NameSpace,in Value,out Success)"}
   > n5 = RigVMDispatchNode(Array=@n4.Chain) {template="DISPATCH_RigVMDispatch_ArrayIterator(in Array,out Element,out Index,out Count,out Ratio)"}
   > /Script/ControlRig.RigUnit_SetMetadataTag(NameSpace=Self, Tag=Bones, Item=@n5.Element) {is=struct RigVMUnitNode}
   Is Right Side v2 = RigVMFunctionReferenceNode(ExecuteContext=@n5.Completed, Mirror Plane=@n1.Module Settings.MirrorAxis_31_#1, Item=@n3.Result)
   n6 = &RigVMFunction_ControlFlowBranch(ExecuteContext=@"Is Right Side v2".ExecuteContext, Condition=@"Is Right Side v2".Right Side)
   Get_Module_Metadata_2 = RigVMDispatchNode(Default=(), Name=Bones, NameSpace=Self) {template="DISPATCH_RigDispatch_GetModuleMetadata(in Name,in NameSpace,in Default,out Value,out Found)"}
   n7 = RigVMDispatchNode(Index=0, Array=@Get_Module_Metadata_2.Value) {template="DISPATCH_RigVMDispatch_ArrayGetAtIndex(in Array,in Index,out Element)"}
   Multiply = &RigVMFunction_MathDoubleMul(B=-1, A=RigVMVariableNode() {variable=Axis Pos or Neg}.Value)
   Create Bone Chain = RigVMFunctionReferenceNode(
    Debug=true
    Debug Scale=15
    Debug Thickness=0.5
    Enable Orient Bones=true
    Kind=Direction
    Optional Suffix=rig
    Primary Axis=(X=1,Y=0,Z=0)
    Secondary Axis=(X=0,Y=1,Z=0)
    Name=@n7.Element.Name
    Transforms=/Script/ControlRig.RigUnit_GetTransformItemArray(Space=GlobalSpace, bInitial=False, Items=@Get_Module_Metadata_2.Value) {is=struct RigVMUnitNode}.Transforms
    Parent=/Script/ControlRig.RigUnit_HierarchyGetParent(bDefaultParent=True, Child=@n7.Element) {is=struct RigVMUnitNode}.Parent
    Primary Axis.X=@Multiply.Result
    Secondary Axis.Y=@Multiply.Result
    ExecuteContext=@n6.Completed
   )
   > Set_Module_Metadata_1 = RigVMDispatchNode(Name=Rig Bones, NameSpace=Self, Value=@"Create Bone Chain".Bones) {template="DISPATCH_RigDispatch_SetModuleMetadata(in Name,in NameSpace,in Value,out Success)"}
   RigVMVariableNode(Value=-1, ExecuteContext=@n6.False) {variable=Axis Pos or Neg}
   RigVMVariableNode(Value=1, ExecuteContext=@n6.True) {variable=Axis Pos or Neg}
   Get_Module_Metadata_3 = RigVMDispatchNode(Default=(), Name=Rig Bones, NameSpace=Self) {template="DISPATCH_RigDispatch_GetModuleMetadata(in Name,in NameSpace,in Default,out Value,out Found)"}
   DISPATCH_RigVMDispatch_ArrayIterator = RigVMDispatchNode(Array=@Get_Module_Metadata_3.Value, ExecuteContext=@RigVMFunction_Sequence.B) {template="DISPATCH_RigVMDispatch_ArrayIterator(in Array,out Element,out Index,out Count,out Ratio)"}
   n8 = RigVMDispatchNode(B=0, A=@DISPATCH_RigVMDispatch_ArrayIterator.Index) {template="DISPATCH_RigVMDispatch_CoreEquals(in A,in B,out Result)"}
   RigVMFunction_ControlFlowBranch = &RigVMFunction_ControlFlowBranch(ExecuteContext=@DISPATCH_RigVMDispatch_ArrayIterator.ExecuteContext, Condition=@n8.Result)
   n9 = RigVMDispatchNode(
    Index=@DISPATCH_RigVMDispatch_ArrayIterator.Index
    Array=/Script/ControlRig.RigUnit_FindItemsWithMetadataTag(NameSpace=Self, Tag=Bones) {is=struct RigVMUnitNode}.Items
   ) {template="DISPATCH_RigVMDispatch_ArrayGetAtIndex(in Array,in Index,out Element)"}
   Concat = /Script/RigVM.RigVMFunction_NameConcat(B=_ctrl, A=&RigUnit_HierarchyGetParent(bDefaultParent=True, Child=@n9.Element).Parent.Name) {is=struct RigVMUnitNode, template="Concat::Execute(in A,in B,out Result)"}
   Item = /Script/ControlRig.RigUnit_Item(Item=(Type=Control,Name="None"), Item.Name=@Concat.Result) {is=struct RigVMUnitNode}
   n10 = RigVMVariableNode() {variable=Parent Control Connector}
   Control Stack at Item_1 = RigVMFunctionReferenceNode(
    Add Bottom Null=False
    Add Secondary Control=False
    Control Suffix=_ctrl
    Null Suffix=_null
    Secondary Settings="(DisplayName=\"None\",Shape=(bVisible=True,Name=\"Default\",Color=(R=1,G=0,B=0,A=1),Transform=(Rotation=(X=0,Y=0,Z=0,W=1),Translation=(X=0,Y=0,Z=0),Scale3D=(X=1,Y=1,Z=1))),Proxy=(bIsProxy=False,ShapeVisibility=BasedOnSelection),InitialSpace=LocalSpace,bUsePref ...(truncated)"
    Tag Controls=true
    Zero Controls=true
    Name=@n9.Element.Name
    Item=@DISPATCH_RigVMDispatch_ArrayIterator.Element
    Parent=RigVMDispatchNode(Condition=@n8.Result, False=@Item.Item, True=&RigUnit_ResolveConnector(SkipSocket=true, Connector=@n10.Value).Result) {template="DISPATCH_RigVMDispatch_If(in Condition,in True,in False,out Result)"}.Result
    ExecuteContext=@RigVMFunction_ControlFlowBranch.Completed
    Add Null=RigVMVariableNode() {variable=Add Null Above}.Value
    Settings=RigVMVariableNode() {variable=Shape Settings}.Value
   )
   > RigVMDispatchNode(Name=FK Parent, NameSpace=Self, Value=@"Control Stack at Item_1".Control) {template="DISPATCH_RigDispatch_SetModuleMetadata(in Name,in NameSpace,in Value,out Success)"}
   > Is Right Side = RigVMFunctionReferenceNode(Item=@"Control Stack at Item_1".Control, Mirror Plane=@n1.Module Settings.MirrorAxis_31_#1)
   RigVMVariableNode(ExecuteContext=@RigVMFunction_ControlFlowBranch.False) {variable=Add Null Above}
   RigVMVariableNode(Value=true, ExecuteContext=@RigVMFunction_ControlFlowBranch.True) {variable=Add Null Above}
   Get_Module_Metadata_1_1 = RigVMDispatchNode(Default=(), Name=Bones, NameSpace=Self) {template="DISPATCH_RigDispatch_GetModuleMetadata(in Name,in NameSpace,in Default,out Value,out Found)"}
   Get_Module_Metadata_3_1 = RigVMDispatchNode(Default=(), Name=Rig Bones, NameSpace=Self) {template="DISPATCH_RigDispatch_GetModuleMetadata(in Name,in NameSpace,in Default,out Value,out Found)"}
   Set Closest Bone = RigVMFunctionReferenceNode(Bones=@Get_Module_Metadata_1_1.Value, Items=@Get_Module_Metadata_3_1.Value, ExecuteContext=@RigVMFunction_Sequence.C)
   Get_Module_Metadata_1_1_1 = RigVMDispatchNode(Default=(), Name=Bones, NameSpace=Self) {template="DISPATCH_RigDispatch_GetModuleMetadata(in Name,in NameSpace,in Default,out Value,out Found)"}
   > Set Closest Bone_1 = RigVMFunctionReferenceNode(Items=&RigUnit_FindItemsWithMetadataTag(NameSpace=Self, Tag=Controls).Items, Bones=@Get_Module_Metadata_1_1_1.Value)
   Finalize Controls = RigVMCollapseNode(Module Settings=@n1.Module Settings, ExecuteContext=@RigVMFunction_Sequence.D)
   RigVMFunctionReturnNode(ExecuteContext=@n2.B)
   &RigUnit_FindItemsWithMetadataTag(NameSpace=Self, Tag=Bones)
   Get_Module_Metadata = RigVMDispatchNode(Default=(Type=Bone,Name="None"), Name=Clavicle Bone, NameSpace=Self) {template="DISPATCH_RigDispatch_GetModuleMetadata(in Name,in NameSpace,in Default,out Value,out Found)"}
   Get_Module_Metadata_1 = RigVMDispatchNode(Default=(Type=Bone,Name="None"), Name=Shoulder Bone, NameSpace=Self) {template="DISPATCH_RigDispatch_GetModuleMetadata(in Name,in NameSpace,in Default,out Value,out Found)"}
   n11 = /Script/ControlRig.RigUnit_GetShapeTransform(Control=@"Control Stack at Item_1".Control.Name) {is=struct RigVMUnitNode}
   Scale = /Script/RigVM.RigVMFunction_MathQuaternionScaleV2(Factor=-1, Value=@n11.Transform.Rotation) {is=struct RigVMUnitNode, template="Scale::Execute(in Value,in Factor,out Result)"}
   n12 = RigVMDispatchNode(
    True=(Rotation=(X=0,Y=0,Z=0,W=1),Translation=(X=0,Y=0,Z=0),Scale3D=(X=1,Y=1,Z=1))
    False=@n11.Transform
    True.Scale3D=@n11.Transform.Scale3D
    Condition=@"Is Right Side".Right Side
    True.Rotation=@Scale.Result
    True.Translation=/Script/RigVM.RigVMFunction_MathVectorScale(Factor=-1, Value=@n11.Transform.Translation) {is=struct RigVMUnitNode, template="Scale::Execute(in Value,in Factor,out Result)"}.Result
   ) {template="DISPATCH_RigVMDispatch_If(in Condition,in True,in False,out Result)"}
   n13 = /Script/ControlRig.RigUnit_HierarchyGetShapeSettings(Item=@"Control Stack at Item_1".Control) {is=struct RigVMUnitNode}
   /Script/ControlRig.RigUnit_HierarchySetShapeSettings(
    ExecutePin=()
    Settings=(bVisible=True,Name="Arrow_Solid",Color=(R=1,G=0,B=0,A=1),Transform=(Rotation=(X=0,Y=0,Z=-0.7071,W=0.7071),Translation=(X=0,Y=0,Z=0),Scale3D=(X=1,Y=1,Z=1)))
    Item=@"Control Stack at Item_1".Control
    Settings.Transform.Rotation=@n12.Result.Rotation
    Settings.Transform.Scale3D=@n12.Result.Scale3D
    Settings.Transform.Translation.X=@n12.Result.Translation.X
    Settings.Transform.Translation.Z=@n12.Result.Translation.Z
    Settings.Color=@n13.Settings.Color
    Settings.Name=@n13.Settings.Name
    Settings.bVisible=@n13.Settings.bVisible
    Settings.Transform.Translation.Y=&RigVMFunction_MathDoubleMul(B=-2, A=RigVMVariableNode() {variable=Axis Pos or Neg}.Value).Result
   ) {is=struct RigVMUnitNode}
  ]}
 New Function_1_ContainedGraph {nodes [
   n1 = RigVMFunctionEntryNode()
   > RigVMFunction_Sequence = RigVMAggregateNode()
   n2 = &RigUnit_FindItemsWithMetadataTag(NameSpace=Self, Tag=Controls)
   n3 = RigVMDispatchNode(Array=@n2.Items, ExecuteContext=@RigVMFunction_Sequence.A) {template="DISPATCH_RigVMDispatch_ArrayIterator(in Array,out Element,out Index,out Count,out Ratio)"}
   DISPATCH_RigDispatch_GetMetadata = RigVMDispatchNode(Default=(Type=Bone,Name="None"), Name=Item, NameSpace=Self, Item=@n3.Element) {template="DISPATCH_RigDispatch_GetMetadata(in Item,in Name,in NameSpace,in Default,out Value,out Found)"}
   > Parent_0_1 = RigVMFunctionReferenceNode(Child=@DISPATCH_RigDispatch_GetMetadata.Value, New Parent=@n3.Element)
   n4 = RigVMDispatchNode(Index=0, Array=@n2.Items) {template="DISPATCH_RigVMDispatch_ArrayGetAtIndex(in Array,in Index,out Element)"}
   Parent_0 = RigVMFunctionReferenceNode(
    Child=RigVMDispatchNode(Default=(Type=Bone,Name="None"), Name=Null, NameSpace=Self, Item=@n4.Element) {template="DISPATCH_RigDispatch_GetMetadata(in Item,in Name,in NameSpace,in Default,out Value,out Found)"}.Value
    ExecuteContext=@n3.Completed
    New Parent=&RigUnit_HierarchyGetParent(
     bDefaultParent=True
     Child=RigVMDispatchNode(Default=(Type=Bone,Name="None"), Name=Item, NameSpace=Self, Item=@n4.Element) {template="DISPATCH_RigDispatch_GetMetadata(in Item,in Name,in NameSpace,in Default,out Value,out Found)"}.Value
    ).Parent
   )
   n5 = &RigVMFunction_ControlFlowBranch(Condition=@n1.Attach Skeleton, ExecuteContext=@RigVMFunction_Sequence.B)
   Get_Module_Metadata_3_1 = RigVMDispatchNode(Default=(), Name=Bones, NameSpace=Self) {template="DISPATCH_RigDispatch_GetModuleMetadata(in Name,in NameSpace,in Default,out Value,out Found)"}
   For_Each = RigVMDispatchNode(Array=@Get_Module_Metadata_3_1.Value, ExecuteContext=@n5.True) {template="DISPATCH_RigVMDispatch_ArrayIterator(in Array,out Element,out Index,out Count,out Ratio)"}
   Get_Module_Metadata_3 = RigVMDispatchNode(Default=(), Name=Rig Bones, NameSpace=Self) {template="DISPATCH_RigDispatch_GetModuleMetadata(in Name,in NameSpace,in Default,out Value,out Found)"}
   > Parent = RigVMFunctionReferenceNode(
    New Parent=RigVMDispatchNode(Index=@For_Each.Index, Array=@Get_Module_Metadata_3.Value) {template="DISPATCH_RigVMDispatch_ArrayGetAtIndex(in Array,in Index,out Element)"}.Element
    Child=@For_Each.Element
   )
   RigVMFunctionReturnNode(ExecuteContext=@RigVMFunction_Sequence.C)
   n6 = /Script/ControlRig.RigUnit_IsInteracting() {is=struct RigVMUnitNode}
   RigVMFunction_ControlFlowBranch = &RigVMFunction_ControlFlowBranch(ExecuteContext=(), Condition=@n6.bIsRotating)
   For_Each_1 = RigVMDispatchNode(Array=@n6.Items, ExecuteContext=@RigVMFunction_ControlFlowBranch.True) {template="DISPATCH_RigVMDispatch_ArrayIterator(in Array,out Element,out Index,out Count,out Ratio)"}
   Get_Module_Metadata_4 = RigVMDispatchNode(Default=(Type=Bone,Name="None"), Name=Proxy Control, NameSpace=Self) {template="DISPATCH_RigDispatch_GetModuleMetadata(in Name,in NameSpace,in Default,out Value,out Found)"}
   n7 = &RigVMFunction_ControlFlowBranch(
    Condition=RigVMDispatchNode(A=@For_Each_1.Element, B=@Get_Module_Metadata_4.Value) {template="DISPATCH_RigVMDispatch_CoreEquals(in A,in B,out Result)"}.Result
    ExecuteContext=@For_Each_1.ExecuteContext
   )
   n8 = &RigUnit_FindItemsWithMetadataTag(NameSpace=Self, Tag=Controls)
   Get_Module_Metadata_4_1_1_1_1 = RigVMDispatchNode(Default=(Type=Bone,Name="None"), Name=Proxy Control, NameSpace=Self) {template="DISPATCH_RigDispatch_GetModuleMetadata(in Name,in NameSpace,in Default,out Value,out Found)"}
   &RigUnit_RotationConstraintLocalSpaceOffset(
    AdvancedSettings=(InterpolationType=Shortest,RotationOrderForFilter=ZYX)
    Filter=(bX=True,bY=True,bZ=True)
    Parents=((Item=(Type=Bone,Name="None"),Weight=0.1),(Item=(Type=Bone,Name="None"),Weight=0.9))
    Weight=1
    bMaintainOffset=true
    Child=RigVMDispatchNode(Index=0, Array=@n8.Items) {template="DISPATCH_RigVMDispatch_ArrayGetAtIndex(in Array,in Index,out Element)"}.Element
    Parents.0.Item=@Get_Module_Metadata_4_1_1_1_1.Value
    Parents.1.Item=&RigUnit_HierarchyGetParent(bDefaultParent=True, Child=@Get_Module_Metadata_4_1_1_1_1.Value).Parent
    ExecutePin=@n7.True
   )
   > For_Each_2_1 = RigVMDispatchNode(Array=@n8.Items) {template="DISPATCH_RigVMDispatch_ArrayIterator(in Array,out Element,out Index,out Count,out Ratio)"}
   Greater = /Script/RigVM.RigVMFunction_MathIntGreater(B=0, A=@For_Each_2_1.Index) {is=struct RigVMUnitNode, template="Greater::Execute(in A,in B,out Result)"}
   n9 = &RigVMFunction_ControlFlowBranch(ExecuteContext=@For_Each_2_1.ExecuteContext, Condition=@Greater.Result)
   Get_Module_Metadata_4_1_1_1 = RigVMDispatchNode(Default=(Type=Bone,Name="None"), Name=Proxy Control, NameSpace=Self) {template="DISPATCH_RigDispatch_GetModuleMetadata(in Name,in NameSpace,in Default,out Value,out Found)"}
   Subtract = &RigVMFunction_MathDoubleSub(A=1, B=@For_Each_2_1.Ratio)
   &RigUnit_RotationConstraintLocalSpaceOffset(
    AdvancedSettings=(InterpolationType=Shortest,RotationOrderForFilter=ZYX)
    Filter=(bX=True,bY=True,bZ=True)
    Parents=((Item=(Type=Bone,Name="None"),Weight=1),(Item=(Type=Bone,Name="None"),Weight=1))
    Weight=1
    bMaintainOffset=true
    Child=@For_Each_2_1.Element
    Parents.0.Weight=@For_Each_2_1.Ratio
    Parents.0.Item=@Get_Module_Metadata_4_1_1_1.Value
    Parents.1.Item=&RigUnit_HierarchyGetParent(bDefaultParent=True, Child=@Get_Module_Metadata_4_1_1_1.Value).Parent
    ExecutePin=@n9.True
    Parents.1.Weight=@Subtract.Result
   )
   Get_Module_Metadata = RigVMDispatchNode(Default=(Type=Bone,Name="None"), Name=Clavicle Bone, NameSpace=Self) {template="DISPATCH_RigDispatch_GetModuleMetadata(in Name,in NameSpace,in Default,out Value,out Found)"}
   Get_Module_Metadata_2 = RigVMDispatchNode(Default=(Type=Bone,Name="None"), Name=Clavicle Rig Bone, NameSpace=Self) {template="DISPATCH_RigDispatch_GetModuleMetadata(in Name,in NameSpace,in Default,out Value,out Found)"}
   &RigUnit_SetTransform(
    ExecutePin=()
    Weight=1
    bInitial=False
    Item=@Get_Module_Metadata.Value
    Value=/Script/ControlRig.RigUnit_ProjectTransformToNewParent(
     bChildInitial=True
     bNewParentInitial=False
     bOldParentInitial=True
     Child=@Get_Module_Metadata.Value
     NewParent=@Get_Module_Metadata_2.Value
     OldParent=@Get_Module_Metadata_2.Value
    ) {is=struct RigVMUnitNode}.Transform
   )
   Get_Module_Metadata_1 = RigVMDispatchNode(Default=(Type=Bone,Name="None"), Name=Shoulder Bone, NameSpace=Self) {template="DISPATCH_RigDispatch_GetModuleMetadata(in Name,in NameSpace,in Default,out Value,out Found)"}
   Get_Module_Metadata_1_1 = RigVMDispatchNode(Default=(Type=Bone,Name="None"), Name=Shoulder Rig Bone, NameSpace=Self) {template="DISPATCH_RigDispatch_GetModuleMetadata(in Name,in NameSpace,in Default,out Value,out Found)"}
   > &RigUnit_SetTransform(
    Weight=1
    bInitial=False
    Item=@Get_Module_Metadata_1.Value
    Value=&RigUnit_ProjectTransformToNewParent(
     bChildInitial=True
     bNewParentInitial=False
     bOldParentInitial=True
     Child=@Get_Module_Metadata_1.Value
     NewParent=@Get_Module_Metadata_1_1.Value
     OldParent=@Get_Module_Metadata_1_1.Value
    ).Transform
   )
   Add = &RigVMFunction_MathDoubleAdd(B=1, A=@For_Each_2_1.Ratio)
   &RigVMFunction_MathDoubleAdd(B=1, A=@For_Each_2_1.Ratio)
   &RigVMFunction_MathDoubleMul(A=1, B=2)
  ]}
 New Function_ContainedGraph {nodes [
   n1 = RigVMFunctionEntryNode()
   n2 = &RigUnit_FindItemsWithMetadataTag(NameSpace=Self, Tag=Controls)
   > Scale Module Controls = RigVMFunctionReferenceNode(
    Global Scale=RigVMDispatchNode(Default=1, Name=Control Size, NameSpace=Root) {template="DISPATCH_RigDispatch_GetModuleMetadata(in Name,in NameSpace,in Default,out Value,out Found)"}.Value
    Local Scale=@n1.Module Settings.ControlSize_12_#2
    Controls=@n2.Items
   )
   > Color Module Controls = RigVMFunctionReferenceNode(
    CenterColor=@n1.Module Settings.CenterColor_9_#3
    LeftColor=@n1.Module Settings.LeftColor_3_#4
    Override=@n1.Module Settings.OverrideColor_48_#5
    RightColor=@n1.Module Settings.RightColor_5_#6
    Controls=@n2.Items
   )
   > RigVMFunctionReturnNode()
  ]}
 New Function_ContainedGraph {nodes [
   n1 = RigVMFunctionEntryNode()
   > n2 = RigVMDispatchNode(Array=@n1.Items) {template="DISPATCH_RigVMDispatch_ArrayIterator(in Array,out Element,out Index,out Count,out Ratio)"}
   > FindClosestBone = RigVMFunctionReferenceNode(Initial=true, Item=@n2.Element, Bones=@n1.Bones)
   > Set_Metadata = RigVMDispatchNode(Name=Closest Bone, NameSpace=Self, Item=@n2.Element, Value=@FindClosestBone.Value) {template="DISPATCH_RigDispatch_SetMetadata(in Item,in Name,in NameSpace,in Value,out Success)"}
   RigVMFunctionReturnNode(ExecuteContext=@n2.Completed)
  ]}
 RigVMFunctionLibrary {nodes [
   CONSTR Finger = RigVMCollapseNode(
    End Connector=(Type=Connector,Name="End")
    ExecuteContext=()
    Module Settings="(OverrideColor_48_#5=False,LeftColor_3_#4=(R=0,G=0.0231,B=0.4115,A=1),RightColor_5_#6=(R=1,G=0,B=0,A=1),CenterColor_9_#3=(R=1,G=1,B=0,A= ...(truncated)"
    Parent Control Connector=(Type=Connector,Name="Parent Control")
    Proxy Ctrl Connector=(Type=Connector,Name="None")
    Shape Settings="(DisplayName=\"None\",bSelectable=True,InitialSpace=LocalSpace,bUsePreferredRotationOrder=False,PreferredRotationOrder=YZX,Limits=(LimitTranslationX=(bMinimum=False,bMaximum=False),LimitTranslationY=(bMinimum=False,bMaximum=False),LimitTranslationZ=(bMinimum ...(truncated)"
    Start Connector=(Type=Connector,Name="Start")
   )
   FWD Finger = RigVMCollapseNode(Attach Skeleton=true, ExecuteContext=())
   Is Right Side = RigVMCollapseNode(ExecuteContext=(), Item=(Type=Bone), Mirror Plane=(X=1,Y=1,Z=0))
   Is Right Side v2 = RigVMCollapseNode(ExecuteContext=(), Item=(Type=Bone), Mirror Plane=(X=1,Y=1,Z=0))
   Set Closest Bone = RigVMCollapseNode(Bones=(), ExecuteContext=(), Items=())
  ]}
 RigVMFunction_Sequence_1_1_ContainedGraph {nodes [
   n1 = RigVMFunctionEntryNode()
   n2 = &RigVMFunction_Sequence(ExecuteContext=@n1.ExecuteContext)
   n3 = &RigVMFunction_Sequence(ExecuteContext=@n2.B)
   RigVMFunctionReturnNode(A=@n2.A, B=@n3.A, C=@n3.B)
  ]}
 RigVMFunction_Sequence_1_ContainedGraph {nodes [
   n1 = RigVMFunctionEntryNode()
   RigVMFunction_Sequence = &RigVMFunction_Sequence(ExecuteContext=@n1.ExecuteContext)
   n2 = &RigVMFunction_Sequence(ExecuteContext=@RigVMFunction_Sequence.B)
   n3 = &RigVMFunction_Sequence(ExecuteContext=@n2.B)
   RigVMFunctionReturnNode(A=@RigVMFunction_Sequence.A, B=@n2.A, C=@n3.A, D=@n3.B)
  ]}
 RigVMFunction_Sequence_1_ContainedGraph {nodes [
   n1 = RigVMFunctionEntryNode()
   RigVMFunction_Sequence = &RigVMFunction_Sequence(ExecuteContext=@n1.ExecuteContext)
   n2 = &RigVMFunction_Sequence(ExecuteContext=@RigVMFunction_Sequence.B)
   n3 = &RigVMFunction_Sequence(ExecuteContext=@n2.B)
   RigVMFunctionReturnNode(A=@RigVMFunction_Sequence.A, B=@n2.A, C=@n3.A, D=@n3.B)
  ]}
 RigVMFunction_Sequence_1_ContainedGraph {nodes [
   n1 = RigVMFunctionEntryNode()
   RigVMFunction_Sequence = &RigVMFunction_Sequence(ExecuteContext=@n1.ExecuteContext)
   n2 = &RigVMFunction_Sequence(ExecuteContext=@RigVMFunction_Sequence.B)
   RigVMFunctionReturnNode(A=@RigVMFunction_Sequence.A, B=@n2.A, C=@n2.B)
  ]}
 RigVMModel {nodes [
   n1 = &RigUnit_BeginExecution()
   FWD Finger = RigVMFunctionReferenceNode(Attach Skeleton=true, ExecuteContext=@n1.ExecutePin)
  ]}
 RigVMModel Backwards Solve Graph {nodes [
   n1 = /Script/ControlRig.RigUnit_InverseExecution() {is=struct RigVMUnitNode}
   RigVMFunction_Sequence = RigVMAggregateNode(ExecuteContext=@n1.ExecutePin)
   n2 = RigVMDispatchNode(Index=0, Array=&RigUnit_FindItemsWithMetadataTag(NameSpace=Self, Tag=Controls).Items) {template="DISPATCH_RigVMDispatch_ArrayGetAtIndex(in Array,in Index,out Element)"}
   Get_Metadata_1_1 = RigVMDispatchNode(Default=(Type=Bone,Name="None"), Name=Closest Bone, NameSpace=Self, Item=@n2.Element) {template="DISPATCH_RigDispatch_GetMetadata(in Item,in Name,in NameSpace,in Default,out Value,out Found)"}
   Parent_1_1 = RigVMFunctionReferenceNode(
    New Parent=@Get_Metadata_1_1.Value
    Child=&RigUnit_HierarchyGetParent(bDefaultParent=True, Child=@n2.Element).Parent
    ExecuteContext=@RigVMFunction_Sequence.A
   )
   n3 = RigVMDispatchNode(Array=&RigUnit_FindItemsWithMetadataTag(NameSpace=Self, Tag=Controls).Items, ExecuteContext=@RigVMFunction_Sequence.B) {template="DISPATCH_RigVMDispatch_ArrayIterator(in Array,out Element,out Index,out Count,out Ratio)"}
   Get_Metadata_1 = RigVMDispatchNode(Default=(Type=Bone,Name="None"), Name=Closest Bone, NameSpace=Self, Item=@n3.Element) {template="DISPATCH_RigDispatch_GetMetadata(in Item,in Name,in NameSpace,in Default,out Value,out Found)"}
   > Parent_1 = RigVMFunctionReferenceNode(Child=@n3.Element, New Parent=@Get_Metadata_1.Value)
   Get_Module_Metadata = RigVMDispatchNode(Default=(), Name=Rig Bones, NameSpace=Self) {template="DISPATCH_RigDispatch_GetModuleMetadata(in Name,in NameSpace,in Default,out Value,out Found)"}
   n4 = RigVMDispatchNode(Array=@Get_Module_Metadata.Value, ExecuteContext=@RigVMFunction_Sequence.C) {template="DISPATCH_RigVMDispatch_ArrayIterator(in Array,out Element,out Index,out Count,out Ratio)"}
   Get_Metadata = RigVMDispatchNode(Default=(Type=Bone,Name="None"), Name=Closest Bone, NameSpace=Self, Item=@n4.Element) {template="DISPATCH_RigDispatch_GetMetadata(in Item,in Name,in NameSpace,in Default,out Value,out Found)"}
   > Parent_2 = RigVMFunctionReferenceNode(Child=@n4.Element, New Parent=@Get_Metadata.Value)
  ]}
 RigVMModel Connector Event Graph {nodes [
   n1 = /Script/ControlRig.RigUnit_ConnectorExecution() {is=struct RigVMUnitNode}
   RigVMFunction_Sequence_1 = RigVMAggregateNode(ExecuteContext=@n1.ExecuteContext)
   n2 = /Script/ControlRig.RigUnit_GetCandidates() {is=struct RigVMUnitNode}
   n3 = &RigVMFunction_ControlFlowBranch(
    Condition=RigVMDispatchNode(B=(Type=Connector,Name="End Bone"), A=@n2.Connector) {template="DISPATCH_RigVMDispatch_CoreEquals(in A,in B,out Result)"}.Result
    ExecuteContext=@RigVMFunction_Sequence_1.A
   )
   RigVMDispatchNode(Array=@n2.Candidates, ExecuteContext=@n3.True) {template="DISPATCH_RigVMDispatch_ArrayIterator(in Array,out Element,out Index,out Count,out Ratio)"}
   n4 = /Script/ControlRig.RigUnit_CollectionChildrenArray(
    TypeToSearch=Bone
    bDefaultChildren=True
    bIncludeParent=true
    bRecursive=False
    Parent=&RigUnit_ResolveConnector(Connector=(Type=Connector,Name="Finger Primary"), SkipSocket=true).Result
   ) {is=struct RigVMUnitNode}
   > /Script/ControlRig.RigUnit_SetDefaultMatch(Default=RigVMDispatchNode(Index=-1, Array=@n4.Items) {template="DISPATCH_RigVMDispatch_ArrayGetAtIndex(in Array,in Index,out Element)"}.Element) {is=struct RigVMUnitNode, template="Set Default Match::Execute(in Default)"}
   n5 = &RigUnit_GetCandidates()
   n6 = &RigVMFunction_ControlFlowBranch(
    Condition=RigVMDispatchNode(B=(Type=Connector,Name="Start Bone"), A=@n5.Connector) {template="DISPATCH_RigVMDispatch_CoreEquals(in A,in B,out Result)"}.Result
    ExecuteContext=@RigVMFunction_Sequence_1.B
   )
   n7 = RigVMDispatchNode(Array=@n5.Candidates, ExecuteContext=@n6.True) {template="DISPATCH_RigVMDispatch_ArrayIterator(in Array,out Element,out Index,out Count,out Ratio)"}
   n8 = /Script/ControlRig.RigUnit_HierarchyGetParentsItemArray(
    bDefaultParent=True
    bIncludeChild=true
    bReverse=False
    Child=&RigUnit_ResolveConnector(Connector=(Type=Connector,Name="Finger Primary"), SkipSocket=true).Result
   ) {is=struct RigVMUnitNode}
   > n9 = RigVMDispatchNode(Array=@n8.Parents) {template="DISPATCH_RigVMDispatch_ArrayIterator(in Array,out Element,out Index,out Count,out Ratio)"}
   DISPATCH_RigDispatch_GetModuleMetadata = RigVMDispatchNode(Default=(Type=Bone,Name="None"), Name=Hand Bone, NameSpace=Parent) {template="DISPATCH_RigDispatch_GetModuleMetadata(in Name,in NameSpace,in Default,out Value,out Found)"}
   RigVMFunction_ControlFlowBranch = &RigVMFunction_ControlFlowBranch(Condition=@DISPATCH_RigDispatch_GetModuleMetadata.Found, ExecuteContext=@n9.ExecuteContext)
   n10 = &RigVMFunction_ControlFlowBranch(
    Condition=RigVMDispatchNode(A=(Type=None,Name="None"), B=(Type=None,Name="None"), A.Name=@DISPATCH_RigDispatch_GetModuleMetadata.Value.Name, B.Name=@n9.Element.Name) {template="DISPATCH_RigVMDispatch_CoreEquals(in A,in B,out Result)"}.Result
    ExecuteContext=@RigVMFunction_ControlFlowBranch.True
   )
   Subtract = /Script/RigVM.RigVMFunction_MathIntSub(B=1, A=@n9.Index) {is=struct RigVMUnitNode, template="Subtract::Execute(in A,in B,out Result)"}
   RigVMVariableNode(
    Value=RigVMDispatchNode(Array=@n8.Parents, Index=@Subtract.Result) {template="DISPATCH_RigVMDispatch_ArrayGetAtIndex(in Array,in Index,out Element)"}.Element
    ExecuteContext=@n10.True
   ) {variable=Start Connector Default Match}
   DISPATCH_RigVMDispatch_CoreEquals = RigVMDispatchNode(B=(Type=None,Name="None"), A=RigVMVariableNode() {variable=Start Connector Default Match}.Value) {template="DISPATCH_RigVMDispatch_CoreEquals(in A,in B,out Result)"}
   n11 = &RigVMFunction_ControlFlowBranch(ExecuteContext=@n9.Completed, Condition=@DISPATCH_RigVMDispatch_CoreEquals.Result)
   DISPATCH_RigVMDispatch_ArrayGetAtIndex = RigVMDispatchNode(Index=1, Array=@n8.Parents) {template="DISPATCH_RigVMDispatch_ArrayGetAtIndex(in Array,in Index,out Element)"}
   RigVMVariableNode(Value=@DISPATCH_RigVMDispatch_ArrayGetAtIndex.Element, ExecuteContext=@n11.True) {variable=Start Connector Default Match}
   &RigUnit_SetDefaultMatch(ExecutePin=@n7.Completed, Default=RigVMVariableNode() {variable=Start Connector Default Match}.Value) {template="Set Default Match::Execute(in Default)"}
   Connect to Module Metadata_1 = RigVMFunctionReferenceNode(
    Connector=(Type=Connector,Name="Parent Control")
    Module Metadata Name=FK Parent
    Module Metadata NameSpace=Parent
    ExecuteContext=@RigVMFunction_Sequence_1.C
   )
   DISPATCH_RigVMDispatch_ArrayGetNum = RigVMDispatchNode(Array=@n4.Items) {template="DISPATCH_RigVMDispatch_ArrayGetNum(in Array,out Num)"}
   RigVMDispatchNode(B=0, A=@DISPATCH_RigVMDispatch_ArrayGetNum.Num) {template="DISPATCH_RigVMDispatch_CoreEquals(in A,in B,out Result)"}
  ]}
]
hierarchy [
 Finger Primary {type=Connector, loc=<0 0 0>, rot=<0 0 0 1>}
 Root {type=Socket, loc=<0 0 0>, rot=<0 0 0 1>}
 Start Bone {type=Connector, loc=<0 0 0>, rot=<0 0 0 1>}
 End Bone {type=Connector, loc=<0 0 0>, rot=<0 0 0 1>}
 Parent Control {type=Connector, loc=<0 0 0>, rot=<0 0 0 1>}
]
variables [
 Control Settings {
  type=FRigUnit_HierarchyAddControlTransform_Settings
  default="(InitialSpace=LocalSpace,bUsePreferredRotationOrder=False,PreferredRotationOrder=YZX,Limits=(LimitTranslationX=(bMinimum=False,bMaximum=False),LimitTranslationY=(bMinimum=False,bMaximum=False),LimitTranslationZ=(bMinimum=False,bMaximum=False),LimitPitch=(b ...(truncated)"
 }
 Module Settings {
  type=FUserDefinedStruct_#7
  default="(OverrideColor_48_#5=False,LeftColor_3_#4=(R=0,G=0.0231,B=0.4115,A=1),RightColor_5_#6=(R=1,G=0,B=0,A=1),CenterColor_9_#3=(R=1,G=1,B=0,A= ...(truncated)"
 }
 Override Control Settings {type=bool, default=False}
 Parent Control {type=FRigElementKey, default=(Type=Bone,Name="")}
 Start Connector Default Match {type=FRigElementKey, default=(Type=None,Name="")}
]
