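# Modular Control Rig module that builds an FK/IK rig for a four-bone hind leg (thigh, upper knee, lower knee, foot).
# Construction detects side from the mirror axis, creates hamhock helper chains, thigh controls and a lock null, and sockets childless bones below the foot.
# A two-bone limb function rigs upper knee to foot, with its FK reparented under the thigh's.
# Helper graphs supply two-bone IK with stretch and exponential softness.
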ControlRigBlueprint /ControlRigModules/Modules56/HindLeg
graphs [
 Concat_4_1_ContainedGraph {nodes [
   n1 = RigVMFunctionEntryNode()
   n2 = /Script/RigVM.RigVMFunction_NameConcat(A=@n1.A, B=@n1.B) {is=struct RigVMUnitNode, template="Concat::Execute(in A,in B,out Result)"}
   RigVMFunctionReturnNode(Result=&RigVMFunction_NameConcat(A=&RigVMFunction_NameConcat(A=@n2.Result, B=@n1.C) {template="Concat::Execute(in A,in B,out Result)"}.Result, B=@n1.D) {template="Concat::Execute(in A,in B,out Result)"}.Result)
  ]}
 ConstructionGraph {nodes [
   Format Name_1_1 = RigVMFunctionReferenceNode(
    name=hamhockLock_
    suffix=None
    side=RigVMVariableNode() {variable=SideName}.Value
    prefix=RigVMVariableNode() {variable=ModulePrefixName}.Value
   )
   n1 = /Script/ControlRig.RigUnit_ResolveConnector(Connector=(Type=Connector,Name="HindLeg Primary"), SkipSocket=true) {is=struct RigVMUnitNode}
   n2 = &RigUnit_GetTransform(bInitial=true, Item=@n1.Result)
   n3 = /Script/ControlRig.RigUnit_HierarchyGetParent(bDefaultParent=True, Child=@n1.Result) {is=struct RigVMUnitNode}
   Create Bone Chain_1_1_1 = RigVMFunctionReferenceNode(
    Debug=true
    Debug Scale=15
    Debug Thickness=0.5
    Enable Orient Bones=true
    ExecuteContext=()
    Kind=Direction
    Optional Suffix=rig
    Primary Axis=(X=1,Y=0,Z=0)
    Secondary Axis=(X=0,Y=-1,Z=0)
    Transforms=((Rotation=(X=0,Y=0,Z=0,W=1),Translation=(X=0,Y=0,Z=0),Scale3D=(X=1,Y=1,Z=1)),(Rotation=(X=0,Y=0,Z=0,W=1),Translation=(X=0,Y=0,Z=0),Scale3D=(X=1,Y=1,Z=1)))
    Name=@"Format Name_1_1".Result
    Transforms.1=@n2.Transform
    Transforms.0=&RigUnit_GetTransform(bInitial=true, Item=&RigUnit_ResolveConnector(Connector=(Type=Connector,Name="Foot Bone"), SkipSocket=true).Result).Transform
    Parent=@n3.Parent
   )
   > RigVMVariableNode(Value=@"Create Bone Chain_1_1_1".Bones) {variable=LockingBones}
   Format Name_0 = RigVMFunctionReferenceNode(
    name=hamhockLock_
    suffix=_aim_null
    prefix=RigVMVariableNode() {variable=ModulePrefixName}.Value
    side=RigVMVariableNode() {variable=SideName}.Value
   )
   > n4 = &RigUnit_HierarchyAddNull(Space=GlobalSpace, Name=@"Format Name_0".Result, Transform=@n2.Transform, Parent=@n3.Parent)
   > RigVMVariableNode(Value=@n4.Item) {variable=lock null}
   RigUnit_PrepareForExecution = &RigUnit_PrepareForExecution()
   RigVMFunction_Sequence = RigVMAggregateNode(ExecuteContext=@RigUnit_PrepareForExecution.ExecutePin)
   RigVMVariableNode(
    ExecuteContext=@RigVMFunction_Sequence.A
    Value=RigVMVariableNode() {variable=Module Settings}.Value.Prefix_20_#1
   ) {variable=ModulePrefixName}
   Get_Module_Metadata_3 = RigVMDispatchNode(Default=(X=0,Y=0,Z=0), Name=Mirror Axis, NameSpace=Root) {template="DISPATCH_RigDispatch_GetModuleMetadata(in Name,in NameSpace,in Default,out Value,out Found)"}
   Is Right Side v2 = RigVMFunctionReferenceNode(
    Mirror Plane=@Get_Module_Metadata_3.Value
    Item=&RigUnit_ResolveConnector(Connector=(Type=Connector,Name="HindLeg Primary"), SkipSocket=true).Result
    ExecuteContext=@RigVMFunction_Sequence.B
   )
   > RigVMVariableNode(Value=@"Is Right Side v2".Right Side) {variable=RightSide}
   > RigVMVariableNode(Value=RigVMDispatchNode(False=l, True=r, Condition=@"Is Right Side v2".Right Side) {template="DISPATCH_RigVMDispatch_If(in Condition,in True,in False,out Result)"}.Result) {variable=SideName}
   n5 = &RigUnit_ResolveConnector(Connector=(Type=Connector,Name="Thigh"), SkipSocket=False)
   Control Stack at Item = RigVMFunctionReferenceNode(
    Add Bottom Null=False
    Add Null=true
    Add Secondary Control=False
    Control Suffix=_ctrl
    Null Suffix=_null
    Secondary Settings="(Shape=(bVisible=True,Name=\"Default\",Color=(R=1,G=0,B=0,A=1),Transform=(Rotation=(X=0,Y=0,Z=0,W=1),Translation=(X=0,Y=0,Z=0),Scale3D=(X=1,Y=1,Z=1))),Proxy=(bIsProxy=False,ShapeVisibility=BasedOnSelection),DisplayName=\"None\",InitialSpace=LocalSpace,bUsePref ...(truncated)"
    Tag Controls=true
    Zero Controls=true
    Name=&RigVMFunction_NameConcat(B=_fk, A=@n5.Result.Name) {template="Concat::Execute(in A,in B,out Result)"}.Result
    Item=@n5.Result
    Parent=&RigUnit_ResolveConnector(Connector=(Type=Connector,Name="Parent Control"), SkipSocket=False).Result
    ExecuteContext=@RigVMFunction_Sequence.C
    Settings=RigVMVariableNode() {variable=FK Control Shape}.Value
   )
   > Set_Module_Metadata = RigVMDispatchNode(Name=Thigh FK Control, NameSpace=Self, Value=@"Control Stack at Item".Control) {template="DISPATCH_RigDispatch_SetModuleMetadata(in Name,in NameSpace,in Value,out Success)"}
   > RigVMVariableNode(Value=@"Control Stack at Item".Control) {variable=ThighFKControl}
   Format Name = RigVMFunctionReferenceNode(name=hamhock_, suffix=None, side=RigVMVariableNode() {variable=SideName}.Value, prefix=RigVMVariableNode() {variable=ModulePrefixName}.Value)
   n6 = &RigUnit_ResolveConnector(Connector=(Type=Connector,Name="Foot Bone"), SkipSocket=False)
   n7 = &RigUnit_ResolveConnector(Connector=(Type=Connector,Name="Upper Knee"), SkipSocket=False)
   n8 = &RigUnit_ResolveConnector(Connector=(Type=Connector,Name="Thigh"), SkipSocket=False)
   n9 = /Script/ControlRig.RigUnit_GetTransformItemArray(
    Items=((Type=None,Name="None"),(Type=None,Name="None"),(Type=None,Name="None"))
    Space=GlobalSpace
    bInitial=true
    Items.2=@n6.Result
    Items.1=@n7.Result
    Items.0=@n8.Result
   ) {is=struct RigVMUnitNode}
   n10 = &RigUnit_HierarchyGetParent(bDefaultParent=True, Child=@n8.Result)
   Create Bone Chain_1 = RigVMFunctionReferenceNode(
    Debug=true
    Debug Scale=15
    Debug Thickness=0.5
    Enable Orient Bones=true
    Kind=Location
    Optional Suffix=rig
    Primary Axis=(X=1,Y=0,Z=0)
    Secondary Axis=(X=0,Y=1,Z=0)
    Name=@"Format Name".Result
    Transforms=@n9.Transforms
    Parent=@n10.Parent
    ExecuteContext=@RigVMFunction_Sequence.D
   )
   > Set_Module_Metadata_1 = RigVMDispatchNode(Name=Phantom Rig Bones, NameSpace=Self, Value=@"Create Bone Chain_1".Bones) {template="DISPATCH_RigDispatch_SetModuleMetadata(in Name,in NameSpace,in Value,out Success)"}
   > RigVMVariableNode(Value=@"Create Bone Chain_1".Bones) {variable=PhantomRigBones}
   Format Name_1 = RigVMFunctionReferenceNode(
    name=hamhockThigh_
    suffix=None
    side=RigVMVariableNode() {variable=SideName}.Value
    prefix=RigVMVariableNode() {variable=ModulePrefixName}.Value
   )
   > Create Bone Chain_1_1 = RigVMFunctionReferenceNode(
    Debug=true
    Debug Scale=15
    Debug Thickness=0.5
    Enable Orient Bones=true
    Kind=Direction
    Optional Suffix=rig
    Primary Axis=(X=1,Y=0,Z=0)
    Secondary Axis=(X=0,Y=-1,Z=0)
    Name=@"Format Name_1".Result
    Transforms=@n9.Transforms
    Parent=@n10.Parent
   )
   > Set_Module_Metadata_1_2 = RigVMDispatchNode(Name=Thigh Phantom Bones, NameSpace=Self, Value=@"Create Bone Chain_1_1".Bones) {template="DISPATCH_RigDispatch_SetModuleMetadata(in Name,in NameSpace,in Value,out Success)"}
   > RigVMVariableNode(Value=@"Create Bone Chain_1_1".Bones) {variable=ThighPhantomRigBones}
   Make_Array = RigVMDispatchNode(
    Values=((Type=None,Name="None"),((),Type=None,Name="None"),((),Type=None,Name="None"),((),Type=None,Name="None"))
    Values.3=@n6.Result
    Values.2=&RigUnit_ResolveConnector(Connector=(Type=Connector,Name="Lower Knee"), SkipSocket=False).Result
    Values.1=@n7.Result
    Values.0=@n8.Result
   ) {template="DISPATCH_RigVMDispatch_ArrayMake(in Values,out Array)"}
   > Set_Module_Metadata_1_2_1 = RigVMDispatchNode(Name=Bones, NameSpace=Self, Value=@Make_Array.Array) {template="DISPATCH_RigDispatch_SetModuleMetadata(in Name,in NameSpace,in Value,out Success)"}
   > RigVMVariableNode(Value=@Make_Array.Array) {variable=AllBones}
   n11 = RigVMVariableNode() {variable=PhantomRigBones}
   ComputePoleVector = RigVMFunctionReferenceNode(
    Draw Transform=false
    Initial=true
    OffsetFactor=2
    Origin Scale=(X=0,Y=0,Z=0)
    Bone A=RigVMDispatchNode(Index=0, Array=@n11.Value) {template="DISPATCH_RigVMDispatch_ArrayGetAtIndex(in Array,in Index,out Element)"}.Element
    Bone B=RigVMDispatchNode(Index=1, Array=@n11.Value) {template="DISPATCH_RigVMDispatch_ArrayGetAtIndex(in Array,in Index,out Element)"}.Element
    Bone C=RigVMDispatchNode(Index=2, Array=@n11.Value) {template="DISPATCH_RigVMDispatch_ArrayGetAtIndex(in Array,in Index,out Element)"}.Element
    ExecuteContext=@RigVMFunction_Sequence.E
   )
   n12 = &RigUnit_GetTransform(
    bInitial=False
    Item=RigVMDispatchNode(Index=1, Array=RigVMVariableNode() {variable=PhantomRigBones}.Value) {template="DISPATCH_RigVMDispatch_ArrayGetAtIndex(in Array,in Index,out Element)"}.Element
   )
   > Control Stack at Item_1 = RigVMFunctionReferenceNode(
    Add Bottom Null=False
    Add Null=true
    Add Secondary Control=False
    Control Suffix=_ctrl
    Global Position=(Rotation=(X=0,Y=0,Z=0,W=1),Translation=(X=0,Y=0,Z=0),Scale3D=(X=1,Y=1,Z=1))
    Name=thigh
    Null Suffix=_null
    Orient to World=false
    Secondary Settings="(Shape=(bVisible=True,Name=\"Default\",Color=(R=1,G=0,B=0,A=1),Transform=(Rotation=(X=0,Y=0,Z=0,W=1),Translation=(X=0,Y=0,Z=0),Scale3D=(X=1,Y=1,Z=1))),Proxy=(bIsProxy=False,ShapeVisibility=BasedOnSelection),DisplayName=\"None\",InitialSpace=LocalSpace,bUsePref ...(truncated)"
    Tag Controls=true
    Zero Controls=true
    Global Position.Rotation=@ComputePoleVector.Transform.Rotation
    Global Position.Scale3D=@n12.Transform.Scale3D
    Global Position.Translation=@n12.Transform.Translation
    Parent=&RigUnit_ResolveConnector(Connector=(Type=Connector,Name="Parent Control"), SkipSocket=False).Result
    Settings=RigVMVariableNode() {variable=IK Thigh Shape}.Value
   )
   > Set_Module_Metadata_1_1 = RigVMDispatchNode(Name=Thigh Control, NameSpace=Self, Value=@"Control Stack at Item_1".Control) {template="DISPATCH_RigDispatch_SetModuleMetadata(in Name,in NameSpace,in Value,out Success)"}
   > RigVMVariableNode(Value=@"Control Stack at Item_1".Control) {variable=ThighControl}
   Get_Module_Metadata = RigVMDispatchNode(Default=(Type=Control,Name="None"), Name=Body Control, NameSpace=Root) {template="DISPATCH_RigDispatch_GetModuleMetadata(in Name,in NameSpace,in Default,out Value,out Found)"}
   Get_Module_Metadata_1_1 = RigVMDispatchNode(Default=(Type=Control,Name="None"), Name=Hips Control, NameSpace=Parent) {template="DISPATCH_RigDispatch_GetModuleMetadata(in Name,in NameSpace,in Default,out Value,out Found)"}
   CONSTR TwoBoneLimb = RigVMFunctionReferenceNode(
    Basename=hindleg
    Enable Keep Aligned=false
    End Bone Connector=(Type=Connector,Name="Foot Bone")
    IK Control Basename=foot
    IK Control Spaces=((Type=None,Name="None"),(Type=None,Name="None"))
    Local Default=False
    Middle Bone Connector=(Type=Connector,Name="Lower Knee")
    Offset Elbow Bone=(X=0,Y=0,Z=0)
    Offset Elbow Bone Space=LocalSpace
    Offset IK Ctrl Transform=(Rotation=(X=0,Y=0,Z=0,W=1),Translation=(X=0,Y=0,Z=0),Scale3D=(X=1,Y=1,Z=1))
    PV Control Spaces=((Type=None,Name="None"),(Type=None,Name="None"))
    Parent Control Connector=(Type=Connector,Name="Parent Control")
    Pole Vector Parent Connector=(Type=Connector,Name="Pole Vector Parent")
    Settings Control Connector=(Type=Connector,Name="Settings Control")
    Start Bone Connector=(Type=Connector,Name="Upper Knee")
    Wrist Control Settings="(DisplayName=\"None\",InitialSpace=LocalSpace,bUsePreferredRotationOrder=False,PreferredRotationOrder=YZX,Limits=(LimitTranslationX=(bMinimum=False,bMaximum=False),LimitTranslationY=(bMinimum=False,bMaximum=False),LimitTranslationZ=(bMinimum=False,bMaximum=F ...(truncated)"
    IK Control Spaces.0=@Get_Module_Metadata.Value
    PV Control Spaces.0=@Get_Module_Metadata.Value
    IK Control Spaces.1=@Get_Module_Metadata_1_1.Value
    PV Control Spaces.1=@Get_Module_Metadata_1_1.Value
    ExecuteContext=@RigVMFunction_Sequence.F
    Module Settings=RigVMVariableNode() {variable=Module Settings}.Value
    IK Control Settings=RigVMVariableNode() {variable=IK Foot Shape}.Value
    Enable Softness=RigVMVariableNode() {variable=Enable Softness}.Value
    FK IK Default=RigVMVariableNode() {variable=FK IK Default}.Value
    FK Control Settings=RigVMVariableNode() {variable=FK Control Shape}.Value
    Enable PV Follow=RigVMVariableNode() {variable=Enable PV Follow - version1}.Value
    PV Control Settings=RigVMVariableNode() {variable=Pole Vector Shape}.Value
    Orient to World=RigVMVariableNode() {variable=Orient IK to World}.Value
    PV Offset Factor=RigVMVariableNode() {variable=Pole Vector Position}.Value
    Offset PV Control=RigVMVariableNode() {variable=IK Control Offset}.Value
    Secondary Axis=RigVMVariableNode() {variable=Secondary Axis}.Value
    Primary Axis=RigVMVariableNode() {variable=Primary Axis}.Value
   )
   Get_Module_Metadata_1 = RigVMDispatchNode(Default=(Type=Bone,Name="None"), Name=Effector Null, NameSpace=Self) {template="DISPATCH_RigDispatch_GetModuleMetadata(in Name,in NameSpace,in Default,out Value,out Found)"}
   > RigVMVariableNode(Value=@Get_Module_Metadata_1.Value) {variable=Effector}
   n13 = &RigUnit_ResolveConnector(Connector=(Type=Connector,Name="Upper Knee"), SkipSocket=true)
   Get_Module_Metadata_2_1 = RigVMDispatchNode(Default=(Type=Bone,Name="None"), Name=Thigh FK Control, NameSpace=Self) {template="DISPATCH_RigDispatch_GetModuleMetadata(in Name,in NameSpace,in Default,out Value,out Found)"}
   > /Script/ControlRig.RigUnit_SetDefaultParent(
    Parent=(Type=Null,Name="None")
    Parent.Name=&RigVMFunction_NameConcat(B=_fk, A=@n13.Result.Name) {template="Concat::Execute(in A,in B,out Result)"}.Result
    Child=&RigUnit_HierarchyGetParent(bDefaultParent=True, Child=@Get_Module_Metadata_2_1.Value).Parent
   ) {is=struct RigVMUnitNode}
   Get_Module_Metadata_2 = RigVMDispatchNode(Default=(Type=Bone,Name="None"), Name=Thigh FK Control, NameSpace=Self) {template="DISPATCH_RigDispatch_GetModuleMetadata(in Name,in NameSpace,in Default,out Value,out Found)"}
   > &RigUnit_SetDefaultParent(
    Child=(Type=Control,Name="None")
    Child.Name=&RigVMFunction_NameConcat(B=_fk_ctrl, A=@n13.Result.Name) {template="Concat::Execute(in A,in B,out Result)"}.Result
    Parent=@Get_Module_Metadata_2.Value
   )
   > RigVMVariableNode(Value=RigVMDispatchNode(Default=(), Name=Rig Bones, NameSpace=Self) {template="DISPATCH_RigDispatch_GetModuleMetadata(in Name,in NameSpace,in Default,out Value,out Found)"}.Value) {variable=RigBones}
   Get_Module_Metadata_1_1_1 = RigVMDispatchNode(Default=(Type=Bone,Name="None"), Name=IK Control, NameSpace=Self) {template="DISPATCH_RigDispatch_GetModuleMetadata(in Name,in NameSpace,in Default,out Value,out Found)"}
   > n14 = RigVMVariableNode(Value=@Get_Module_Metadata_1_1_1.Value) {variable=IKControl}
   n15 = RigVMVariableNode() {variable=Enable PV Follow - version1}
   n16 = &RigVMFunction_ControlFlowBranch(
    Condition=&RigVMFunction_MathBoolAnd(B=&RigVMFunction_MathBoolNot(Value=@n15.Value).Result, A=RigVMVariableNode() {variable=Enable PV Follow - version2}.Value).Result
    ExecuteContext=@n14.ExecuteContext
   )
   n17 = &RigUnit_HierarchyAddNull(
    Name=NewNull
    Space=LocalSpace
    Transform=&RigUnit_GetTransform(bInitial=False, Item=RigVMVariableNode() {variable=IKControl}.Value).Transform
    Parent=&RigUnit_ResolveConnector(Connector=(Type=Connector,Name="Parent Control"), SkipSocket=False).Result
    ExecutePin=@n16.True
   )
   > RigVMVariableNode(Value=@n17.Item) {variable=PVControlRotateNull}
   n18 = RigVMDispatchNode(Index=1, Array=/Script/ControlRig.RigUnit_FindItemsWithMetadataTag(NameSpace=Self, Tag=Controls IK) {is=struct RigVMUnitNode}.Items) {template="DISPATCH_RigVMDispatch_ArrayGetAtIndex(in Array,in Index,out Element)"}
   > &RigUnit_SetDefaultParent(Child=@n18.Element, Parent=@n17.Item)
   Get_Metadata = RigVMDispatchNode(Default=(Type=Bone,Name="None"), Name=Null, NameSpace=Self, Item=@n18.Element) {template="DISPATCH_RigDispatch_GetMetadata(in Item,in Name,in NameSpace,in Default,out Value,out Found)"}
   > &RigUnit_SetDefaultParent(Parent=@Get_Metadata.Value, Child=RigVMVariableNode() {variable=PVControlRotateNull}.Value)
   For_Each = RigVMDispatchNode(
    Array=/Script/ControlRig.RigUnit_CollectionChildrenArray(
     TypeToSearch=Bone
     bDefaultChildren=True
     bIncludeParent=False
     bRecursive=true
     Parent=&RigUnit_ResolveConnector(Connector=(Type=Connector,Name="Foot Bone"), SkipSocket=true).Result
    ) {is=struct RigVMUnitNode}.Items
    ExecuteContext=@RigVMFunction_Sequence.G
   ) {template="DISPATCH_RigVMDispatch_ArrayIterator(in Array,out Element,out Index,out Count,out Ratio)"}
   Num = RigVMDispatchNode(Array=&RigUnit_CollectionChildrenArray(TypeToSearch=Bone, bDefaultChildren=True, bIncludeParent=False, bRecursive=False, Parent=@For_Each.Element).Items) {template="DISPATCH_RigVMDispatch_ArrayGetNum(in Array,out Num)"}
   RigVMFunction_ControlFlowBranch = &RigVMFunction_ControlFlowBranch(
    Condition=RigVMDispatchNode(B=0, A=@Num.Num) {template="DISPATCH_RigVMDispatch_CoreEquals(in A,in B,out Result)"}.Result
    ExecuteContext=@For_Each.ExecuteContext
   )
   Concat = &RigVMFunction_NameConcat(B=_socket, A=@For_Each.Element.Name) {template="Concat::Execute(in A,in B,out Result)"}
   /Script/ControlRig.RigUnit_HierarchyAddSocket(
    Color=(R=1,G=1,B=1,A=1)
    Description=Place Finger here . . .
    Space=LocalSpace
    Transform=(Rotation=(X=0,Y=0,Z=0,W=1),Translation=(X=0,Y=0,Z=0),Scale3D=(X=1,Y=1,Z=1))
    Name=@Concat.Result
    Parent=@For_Each.Element
    ExecutePin=@RigVMFunction_ControlFlowBranch.True
   ) {is=struct RigVMUnitNode}
   RigVMVariableNode() {variable=Enable PV Follow - version2}
  ]}
 ContainedGraph {nodes [
   n1 = RigVMFunctionEntryNode()
   > /Script/ControlRig.RigUnit_TwoBoneIKSimplePerItem(
    DebugSettings=(Scale=10,WorldOffset=(Rotation=(X=0,Y=0,Z=0,W=1),Translation=(X=0,Y=0,Z=0),Scale3D=(X=1,Y=1,Z=1)))
    PoleVectorKind=Location
    PoleVectorSpace=(Type=Bone)
    SecondaryAxisWeight=1
    Weight=1
    bPropagateToChildren=True
    ItemA=@n1.Bone A
    ItemB=@n1.Bone B
    EffectorItem=@n1.Bone C
    ItemALength=@n1.Item ALength
    ItemBLength=@n1.Item BLength
    PrimaryAxis=@n1.Primary Axis
    SecondaryAxis=@n1.Secondary Axis
    Effector=/Script/ControlRig.RigUnit_ProjectTransformToNewParent(bChildInitial=True, bOldParentInitial=True, Child=@n1.Bone C, NewParent=@n1.End Ctrl, OldParent=@n1.End Ctrl) {is=struct RigVMUnitNode}.Transform
    PoleVector=&RigUnit_GetTransform(bInitial=false, Item=@n1.Pole Vector Node).Transform.Translation
    StretchMaximumRatio=@n1.Stretch Maximum Ratio
    StretchStartRatio=@n1.Stretch Start Ratio
    bEnableStretch=@n1.Enable Stretch
   ) {is=struct RigVMUnitNode}
   > DebugLineItemSpace = /Script/RigVM.RigVMFunction_DebugLineNoSpace(
    Color=(R=0.2,G=0.2,B=0.2,A=1)
    DebugDrawSettings=(DepthPriority=SDPG_Foreground,Lifetime=-1)
    WorldOffset=(Rotation=(X=0,Y=0,Z=0,W=1),Translation=(X=0,Y=0,Z=0),Scale3D=(X=1,Y=1,Z=1))
    bEnabled=True
    Thickness=@n1.Line Thickness
    B=&RigUnit_GetTransform(Item=@n1.Bone B).Transform.Translation
    A=&RigUnit_GetTransform(bInitial=false, Item=@n1.Pole Vector Node).Transform.Translation
   ) {is=struct RigVMUnitNode}
   > RigVMFunctionReturnNode()
   n2 = &RigUnit_GetTransform(bInitial=true, Item=@n1.Bone B)
   /Script/RigVM.RigVMFunction_MathVectorScale(
    Factor=&RigVMFunction_MathVectorDistance(A=&RigUnit_GetTransform(bInitial=true, Item=@n1.Bone A).Transform.Translation, B=@n2.Transform.Translation).Result
    Value=@n1.scale
   ) {is=struct RigVMUnitNode, template="Scale::Execute(in Value,in Factor,out Result)"}
   &RigVMFunction_MathVectorScale(
    Factor=&RigVMFunction_MathVectorDistance(A=@n2.Transform.Translation, B=&RigUnit_GetTransform(bInitial=true, Item=@n1.Bone C).Transform.Translation).Result
    Value=@n1.scale
   ) {template="Scale::Execute(in Value,in Factor,out Result)"}
  ]}
 ContainedGraph {nodes [
   n1 = RigVMFunctionEntryNode()
   RigUnit_SequenceAggregate = &RigVMFunction_Sequence(ExecuteContext=@n1.ExecuteContext)
   n2 = &RigVMFunction_ControlFlowBranch(Condition=@n1.Active, ExecuteContext=@RigUnit_SequenceAggregate.A)
   DebugTransformMutableItemSpace = /Script/RigVM.RigVMFunction_DebugTransformMutableNoSpace(
    Color=(R=1,G=0,B=0,A=1)
    DebugDrawSettings=(DepthPriority=SDPG_Foreground,Lifetime=-1)
    Mode=Axes
    WorldOffset=(Rotation=(X=0,Y=0,Z=0,W=1),Translation=(X=0,Y=0,Z=0),Scale3D=(X=1,Y=1,Z=1))
    bEnabled=True
    ExecutePin=@n2.True
    Scale=@n1.Scale
    Thickness=@n1.Thickness
    Transform=&RigUnit_GetTransform(Item=@n1.Item).Transform
   ) {is=struct RigVMUnitNode}
   RigVMFunctionReturnNode(ExecuteContext=@RigUnit_SequenceAggregate.B)
  ]}
 ContainedGraph {nodes [
   RigVMFunctionEntryNode()
   n1 = RigVMVariableNode() {variable=Bones}
   DISPATCH_RigVMDispatch_ArrayGetAtIndex = RigVMDispatchNode(Index=0, Array=@n1.Value) {template="DISPATCH_RigVMDispatch_ArrayGetAtIndex(in Array,in Index,out Element)"}
   > ComputePoleVector = RigVMFunctionReferenceNode(
    Draw Transform=false
    Initial=false
    OffsetFactor=2
    Origin Scale=(X=0,Y=0,Z=0)
    Bone A=@DISPATCH_RigVMDispatch_ArrayGetAtIndex.Element
    Bone B=RigVMDispatchNode(Index=1, Array=@n1.Value) {template="DISPATCH_RigVMDispatch_ArrayGetAtIndex(in Array,in Index,out Element)"}.Element
    Bone C=RigVMDispatchNode(Index=2, Array=@n1.Value) {template="DISPATCH_RigVMDispatch_ArrayGetAtIndex(in Array,in Index,out Element)"}.Element
   )
   n2 = RigVMVariableNode() {variable=Pole Vector Control}
   > &RigUnit_SetTransform(Weight=1, Value=@ComputePoleVector.Transform, Item=@n2.Value)
   n3 = RigVMDispatchNode(Index=2, Array=RigVMVariableNode() {variable=Bones}.Value) {template="DISPATCH_RigVMDispatch_ArrayGetAtIndex(in Array,in Index,out Element)"}
   n4 = RigVMVariableNode() {variable=IK Control}
   > &RigUnit_SetTransform(
    Weight=1
    Value=&RigUnit_ProjectTransformToNewParent(bChildInitial=True, bOldParentInitial=True, NewParent=@n3.Element, OldParent=@n3.Element, Child=@n4.Value).Transform
    Item=@n4.Value
   )
   > RigVMFunctionReturnNode()
   n5 = RigVMDispatchNode(Index=0, Array=RigVMVariableNode() {variable=Bones}.Value) {template="DISPATCH_RigVMDispatch_ArrayGetAtIndex(in Array,in Index,out Element)"}
   &RigUnit_ProjectTransformToNewParent(bChildInitial=True, bOldParentInitial=True, NewParent=@n5.Element, OldParent=@n5.Element, Child=@n2.Value)
  ]}
 ContainedGraph {nodes [
   n1 = RigVMFunctionEntryNode()
   > n2 = RigVMDispatchNode(Array=@n1.Children) {template="DISPATCH_RigVMDispatch_ArrayIterator(in Array,out Element,out Index,out Count,out Ratio)"}
   n3 = RigVMDispatchNode(Index=@n2.Index, Array=@n1.Parents) {template="DISPATCH_RigVMDispatch_ArrayGetAtIndex(in Array,in Index,out Element)"}
   > &RigUnit_SetTransform(
    Weight=1
    Item=@n2.Element
    bPropagateToChildren=@n1.Propagate to Children
    Value=&RigUnit_ProjectTransformToNewParent(bChildInitial=True, bOldParentInitial=True, NewParent=@n3.Element, OldParent=@n3.Element, Child=@n2.Element).Transform
   )
   RigVMFunctionReturnNode(ExecuteContext=@n2.Completed)
  ]}
 ContainedGraph {nodes [
   n1 = RigVMFunctionEntryNode()
   RigUnit_SequenceAggregate = &RigVMFunction_Sequence(ExecuteContext=@n1.ExecuteContext)
   RigVMVariableNode(Value=1000000000, ExecuteContext=@RigUnit_SequenceAggregate.A) {variable=closestDistanceSoFar}
   > n2 = RigVMDispatchNode(Array=@n1.Bones) {template="DISPATCH_RigVMDispatch_ArrayIterator(in Array,out Element,out Index,out Count,out Ratio)"}
   Item = /Script/ControlRig.RigUnit_Item(Item=@n1.Item) {is=struct RigVMUnitNode}
   n3 = &RigVMFunction_MathVectorDistance(
    A=&RigUnit_GetTransform(bInitial=@n1.Initial, Item=@n2.Element).Transform.Translation
    B=&RigUnit_GetTransform(bInitial=@n1.Initial, Item=@Item.Item).Transform.Translation
   )
   n4 = &RigVMFunction_ControlFlowBranch(
    ExecuteContext=@n2.ExecuteContext
    Condition=/Script/RigVM.RigVMFunction_MathFloatLessEqual(A=@n3.Result, B=RigVMVariableNode() {variable=closestDistanceSoFar}.Value) {is=struct RigVMUnitNode, template="LessEqual::Execute(in A,in B,out Result)"}.Result
   )
   RigVMVariableNode(ExecuteContext=@n4.True, Value=@n3.Result) {variable=closestDistanceSoFar}
   > RigVMVariableNode(Value=@n2.Element) {variable=closestBone}
   RigVMFunctionReturnNode(ExecuteContext=@RigUnit_SequenceAggregate.B, Value=RigVMVariableNode() {variable=closestBone}.Value)
  ]}
 ContainedGraph {nodes [
   n1 = RigVMFunctionEntryNode()
   n2 = RigVMVariableNode() {variable=Softness}
   n3 = /Script/RigVM.RigVMFunction_MathFloatAbs(Value=&RigUnit_GetTransform(
     Space=LocalSpace
     bInitial=true
     Item=RigVMDispatchNode(Index=1, Array=@n1.Bones) {template="DISPATCH_RigVMDispatch_ArrayGetAtIndex(in Array,in Index,out Element)"}.Element
    ).Transform.Translation.X) {is=struct RigVMUnitNode, template="Absolute::Execute(in Value,out Result)"}
   n4 = &RigVMFunction_MathFloatAbs(Value=&RigUnit_GetTransform(
     Space=LocalSpace
     bInitial=true
     Item=RigVMDispatchNode(Index=2, Array=@n1.Bones) {template="DISPATCH_RigVMDispatch_ArrayGetAtIndex(in Array,in Index,out Element)"}.Element
    ).Transform.Translation.X) {template="Absolute::Execute(in Value,out Result)"}
   n5 = &RigVMFunction_MathFloatAdd(A=@n3.Result, B=@n4.Result) {template="Add::Execute(in A,in B,out Result)"}
   n6 = &RigVMFunction_MathFloatMul(
    B=0.02
    A=&RigVMFunction_MathFloatMul(A=@n5.Result, B=RigVMVariableNode() {variable=Softness}.Value) {template="Multiply::Execute(in A,in B,out Result)"}.Result
   ) {template="Multiply::Execute(in A,in B,out Result)"}
   n7 = /Script/RigVM.RigVMFunction_MathFloatSub(A=@n5.Result, B=@n6.Result) {is=struct RigVMUnitNode, template="Subtract::Execute(in A,in B,out Result)"}
   n8 = &RigVMFunction_MathVectorDistance(
    B=&RigUnit_GetTransform(Item=@n1.Effector).Transform.Translation
    A=&RigUnit_GetTransform(Item=RigVMDispatchNode(Index=0, Array=@n1.Bones) {template="DISPATCH_RigVMDispatch_ArrayGetAtIndex(in Array,in Index,out Element)"}.Element).Transform.Translation
   )
   Exponential Softness = RigVMCollapseNode(Softness=@n6.Result, DA=@n7.Result, Distance=@n8.Result, Stretch Distance=@n8.Result)
   n9 = RigVMDispatchNode(
    False=@n1.Maximum
    Condition=&RigVMFunction_MathBoolAnd(
     A=/Script/RigVM.RigVMFunction_MathFloatGreater(B=0, A=@n2.Value) {is=struct RigVMUnitNode, template="Greater::Execute(in A,in B,out Result)"}.Result
     B=&RigVMFunction_MathFloatGreater(B=@n7.Result, A=@n8.Result) {template="Greater::Execute(in A,in B,out Result)"}.Result
    ).Result
    True=/Script/RigVM.RigVMFunction_MathFloatMax(B=1, A=@"Exponential Softness".Result) {is=struct RigVMUnitNode, template="Maximum::Execute(in A,in B,out Result)"}.Result
   ) {template="DISPATCH_RigVMDispatch_If(in Condition,in True,in False,out Result)"}
   RigVMFunctionReturnNode(
    Result=0
    Length A=&RigVMFunction_MathFloatMul(B=@n9.Result, A=@n3.Result) {template="Multiply::Execute(in A,in B,out Result)"}.Result
    Length B=&RigVMFunction_MathFloatMul(B=@n9.Result, A=@n4.Result) {template="Multiply::Execute(in A,in B,out Result)"}.Result
   )
  ]}
 ContainedGraph {nodes [
   n1 = RigVMFunctionEntryNode()
   n2 = /Script/RigVM.RigVMFunction_MathFloatNegate(Value=/Script/RigVM.RigVMFunction_MathFloatDiv(
     B=@n1.Softness
     A=&RigVMFunction_MathFloatSub(B=@n1.DA, A=@n1.Distance) {template="Subtract::Execute(in A,in B,out Result)"}.Result
    ) {is=struct RigVMUnitNode, template="Divide::Execute(in A,in B,out Result)"}.Result) {is=struct RigVMUnitNode, template="Negate::Execute(in Value,out Result)"}
   n3 = &RigVMFunction_MathFloatMul(
    B=@n1.Softness
    A=&RigVMFunction_MathFloatSub(
     A=1
     B=/Script/RigVM.RigVMFunction_MathFloatExponential(Value=@n2.Result) {is=struct RigVMUnitNode, template="Exponential::Execute(in Value,out Result)"}.Result
    ) {template="Subtract::Execute(in A,in B,out Result)"}.Result
   ) {template="Multiply::Execute(in A,in B,out Result)"}
   RigVMFunctionReturnNode(Result=&RigVMFunction_MathFloatDiv(
     A=@n1.Stretch Distance
     B=&RigVMFunction_MathFloatAdd(B=@n1.DA, A=@n3.Result) {template="Add::Execute(in A,in B,out Result)"}.Result
    ) {template="Divide::Execute(in A,in B,out Result)"}.Result)
  ]}
 ContainedGraph {nodes [
   n1 = RigVMFunctionEntryNode()
   n2 = RigVMVariableNode() {variable=Softness}
   n3 = &RigVMFunction_MathFloatAbs(Value=&RigUnit_GetTransform(
     Space=LocalSpace
     bInitial=true
     Item=RigVMDispatchNode(Index=1, Array=@n1.Bones) {template="DISPATCH_RigVMDispatch_ArrayGetAtIndex(in Array,in Index,out Element)"}.Element
    ).Transform.Translation.X) {template="Absolute::Execute(in Value,out Result)"}
   n4 = &RigVMFunction_MathFloatAbs(Value=&RigUnit_GetTransform(
     Space=LocalSpace
     bInitial=true
     Item=RigVMDispatchNode(Index=2, Array=@n1.Bones) {template="DISPATCH_RigVMDispatch_ArrayGetAtIndex(in Array,in Index,out Element)"}.Element
    ).Transform.Translation.X) {template="Absolute::Execute(in Value,out Result)"}
   n5 = &RigVMFunction_MathFloatAdd(A=@n3.Result, B=@n4.Result) {template="Add::Execute(in A,in B,out Result)"}
   n6 = &RigVMFunction_MathFloatMul(
    B=0.02
    A=&RigVMFunction_MathFloatMul(A=@n5.Result, B=RigVMVariableNode() {variable=Softness}.Value) {template="Multiply::Execute(in A,in B,out Result)"}.Result
   ) {template="Multiply::Execute(in A,in B,out Result)"}
   n7 = &RigVMFunction_MathFloatSub(A=@n5.Result, B=@n6.Result) {template="Subtract::Execute(in A,in B,out Result)"}
   n8 = &RigVMFunction_MathVectorDistance(
    B=&RigUnit_GetTransform(Item=@n1.Effector).Transform.Translation
    A=&RigUnit_GetTransform(Item=RigVMDispatchNode(Index=0, Array=@n1.Bones) {template="DISPATCH_RigVMDispatch_ArrayGetAtIndex(in Array,in Index,out Element)"}.Element).Transform.Translation
   )
   Exponential Softness = RigVMCollapseNode(Softness=@n6.Result, DA=@n7.Result, Distance=@n8.Result, Stretch Distance=@n8.Result)
   n9 = RigVMDispatchNode(
    False=@n1.Maximum
    Condition=&RigVMFunction_MathBoolAnd(
     A=&RigVMFunction_MathFloatGreater(B=0, A=@n2.Value) {template="Greater::Execute(in A,in B,out Result)"}.Result
     B=&RigVMFunction_MathFloatGreater(B=@n7.Result, A=@n8.Result) {template="Greater::Execute(in A,in B,out Result)"}.Result
    ).Result
    True=&RigVMFunction_MathFloatMax(B=1, A=@"Exponential Softness".Result) {template="Maximum::Execute(in A,in B,out Result)"}.Result
   ) {template="DISPATCH_RigVMDispatch_If(in Condition,in True,in False,out Result)"}
   RigVMFunctionReturnNode(
    Result=0
    Length A=&RigVMFunction_MathFloatMul(B=@n9.Result, A=@n3.Result) {template="Multiply::Execute(in A,in B,out Result)"}.Result
    Length B=&RigVMFunction_MathFloatMul(B=@n9.Result, A=@n4.Result) {template="Multiply::Execute(in A,in B,out Result)"}.Result
   )
  ]}
 ContainedGraph {nodes [
   n1 = RigVMFunctionEntryNode()
   n2 = &RigVMFunction_MathFloatNegate(Value=&RigVMFunction_MathFloatDiv(
     B=@n1.Softness
     A=&RigVMFunction_MathFloatSub(B=@n1.DA, A=@n1.Distance) {template="Subtract::Execute(in A,in B,out Result)"}.Result
    ) {template="Divide::Execute(in A,in B,out Result)"}.Result) {template="Negate::Execute(in Value,out Result)"}
   n3 = &RigVMFunction_MathFloatMul(
    B=@n1.Softness
    A=&RigVMFunction_MathFloatSub(A=1, B=&RigVMFunction_MathFloatExponential(Value=@n2.Result) {template="Exponential::Execute(in Value,out Result)"}.Result) {template="Subtract::Execute(in A,in B,out Result)"}.Result
   ) {template="Multiply::Execute(in A,in B,out Result)"}
   RigVMFunctionReturnNode(Result=&RigVMFunction_MathFloatDiv(
     A=@n1.Stretch Distance
     B=&RigVMFunction_MathFloatAdd(B=@n1.DA, A=@n3.Result) {template="Add::Execute(in A,in B,out Result)"}.Result
    ) {template="Divide::Execute(in A,in B,out Result)"}.Result)
  ]}
 ContainedGraph {nodes [
   n1 = RigVMFunctionEntryNode()
   RigUnit_SequenceAggregate_1 = &RigVMFunction_Sequence(ExecuteContext=@n1.ExecuteContext)
   RigUnit_SequenceAggregate = RigVMAggregateNode(ExecuteContext=@RigUnit_SequenceAggregate_1.A)
   Greater = /Script/RigVM.RigVMFunction_MathDoubleGreater(B=0, A=@n1.Mirror Plane.X) {is=struct RigVMUnitNode, template="Greater::Execute(in A,in B,out Result)"}
   BranchNode_4 = &RigVMFunction_ControlFlowBranch(Condition=@Greater.Result, ExecuteContext=@RigUnit_SequenceAggregate.A)
   n2 = &RigVMFunction_MathDoubleGreater(B=0, A=@n1.Mirror Plane.Y) {template="Greater::Execute(in A,in B,out Result)"}
   BranchNode_1_1 = &RigVMFunction_ControlFlowBranch(ExecuteContext=@BranchNode_4.Completed, Condition=@n2.Result)
   n3 = &RigVMFunction_MathDoubleGreater(B=0, A=@n1.Mirror Plane.Z) {template="Greater::Execute(in A,in B,out Result)"}
   BranchNode_2_1 = &RigVMFunction_ControlFlowBranch(ExecuteContext=@BranchNode_1_1.Completed, Condition=@n3.Result)
   Cast = &RigVMFunction_MathBoolToFloat(Value=@Greater.Result)
   Dot__ = /Script/RigVM.RigVMFunction_MathVectorDot(
    B=(X=1,Y=1,Z=0)
    B.X=@Cast.Result
    B.Y=&RigVMFunction_MathBoolToFloat(Value=@n2.Result).Result
    B.Z=&RigVMFunction_MathBoolToFloat(Value=@n3.Result).Result
    A=RigVMVariableNode() {variable=vector pos}.Value
   ) {is=struct RigVMUnitNode, template="Dot,|::Execute(in A,in B,out Result)"}
   Less = /Script/RigVM.RigVMFunction_MathDoubleLess(B=0, A=@Dot__.Result) {is=struct RigVMUnitNode, template="Less::Execute(in A,in B,out Result)"}
   RigVMVariableNode(ExecuteContext=@BranchNode_2_1.Completed, Value=@Less.Result) {variable=is on right}
   n4 = RigVMVariableNode() {variable=vector pos}
   RigVMVariableNode(
    Value=(X=0,Y=0,Z=0)
    ExecuteContext=@BranchNode_2_1.True
    Value.Z=&RigUnit_GetTransform(bInitial=true, Item=@n1.Item).Transform.Translation.Z
    Value.X=@n4.Value.X
    Value.Y=@n4.Value.Y
   ) {variable=vector pos}
   Dot___1_1_1_1_1_2 = &RigVMFunction_MathVectorDot(
    B=(X=1,Y=1,Z=0)
    B.Y=&RigVMFunction_MathBoolToFloat(Value=@n2.Result).Result
    B.Z=&RigVMFunction_MathBoolToFloat(Value=@n3.Result).Result
    B.X=&RigVMFunction_MathBoolToFloat(Value=@Greater.Result).Result
    A=RigVMVariableNode() {variable=vector pos}.Value
   ) {template="Dot,|::Execute(in A,in B,out Result)"}
   > RigVMVariableNode(Value=&RigVMFunction_MathDoubleLess(B=0, A=@Dot___1_1_1_1_1_2.Result) {template="Less::Execute(in A,in B,out Result)"}.Result) {variable=is on right}
   n5 = RigVMVariableNode() {variable=vector pos}
   RigVMVariableNode(
    Value=(X=0,Y=0,Z=0)
    ExecuteContext=@BranchNode_1_1.True
    Value.Y=&RigUnit_GetTransform(bInitial=true, Item=@n1.Item).Transform.Translation.Y
    Value.X=@n5.Value.X
    Value.Z=@n5.Value.Z
   ) {variable=vector pos}
   Dot___1_1_1_1_1_1 = &RigVMFunction_MathVectorDot(
    B=(X=1,Y=1,Z=0)
    B.Y=&RigVMFunction_MathBoolToFloat(Value=@n2.Result).Result
    B.Z=&RigVMFunction_MathBoolToFloat(Value=@n3.Result).Result
    B.X=&RigVMFunction_MathBoolToFloat(Value=@Greater.Result).Result
    A=RigVMVariableNode() {variable=vector pos}.Value
   ) {template="Dot,|::Execute(in A,in B,out Result)"}
   > RigVMVariableNode(Value=&RigVMFunction_MathDoubleLess(B=0, A=@Dot___1_1_1_1_1_1.Result) {template="Less::Execute(in A,in B,out Result)"}.Result) {variable=is on right}
   n6 = RigVMVariableNode() {variable=vector pos}
   RigVMVariableNode(
    Value=(X=0,Y=0,Z=0)
    ExecuteContext=@BranchNode_4.True
    Value.X=&RigUnit_GetTransform(bInitial=true, Item=@n1.Item).Transform.Translation.X
    Value.Y=@n6.Value.Y
    Value.Z=@n6.Value.Z
   ) {variable=vector pos}
   Dot___1_1_1_1_1 = &RigVMFunction_MathVectorDot(
    B=(X=1,Y=1,Z=0)
    B.Y=&RigVMFunction_MathBoolToFloat(Value=@n2.Result).Result
    B.Z=&RigVMFunction_MathBoolToFloat(Value=@n3.Result).Result
    B.X=&RigVMFunction_MathBoolToFloat(Value=@Greater.Result).Result
    A=RigVMVariableNode() {variable=vector pos}.Value
   ) {template="Dot,|::Execute(in A,in B,out Result)"}
   > RigVMVariableNode(Value=&RigVMFunction_MathDoubleLess(B=0, A=@Dot___1_1_1_1_1.Result) {template="Less::Execute(in A,in B,out Result)"}.Result) {variable=is on right}
   n7 = &RigVMFunction_MathDoubleLess(B=0, A=@n1.Mirror Plane.X) {template="Less::Execute(in A,in B,out Result)"}
   BranchNode_4_1 = &RigVMFunction_ControlFlowBranch(Condition=@n7.Result, ExecuteContext=@RigUnit_SequenceAggregate.D)
   n8 = RigVMVariableNode() {variable=vector pos}
   RigVMVariableNode(
    Value=(X=0,Y=0,Z=0)
    ExecuteContext=@BranchNode_4_1.True
    Value.X=&RigUnit_GetTransform(bInitial=true, Item=@n1.Item).Transform.Translation.X
    Value.Y=@n8.Value.Y
    Value.Z=@n8.Value.Z
   ) {variable=vector pos}
   n9 = &RigVMFunction_MathDoubleLess(B=0, A=@n1.Mirror Plane.Y) {template="Less::Execute(in A,in B,out Result)"}
   n10 = &RigVMFunction_MathDoubleLess(B=0, A=@n1.Mirror Plane.Z) {template="Less::Execute(in A,in B,out Result)"}
   Dot___1_1_1_1 = &RigVMFunction_MathVectorDot(
    B=(X=1,Y=1,Z=0)
    B.Y=&RigVMFunction_MathBoolToFloat(Value=@n9.Result).Result
    B.Z=&RigVMFunction_MathBoolToFloat(Value=@n10.Result).Result
    B.X=&RigVMFunction_MathBoolToFloat(Value=@n7.Result).Result
    A=RigVMVariableNode() {variable=vector pos}.Value
   ) {template="Dot,|::Execute(in A,in B,out Result)"}
   > RigVMVariableNode(Value=&RigVMFunction_MathDoubleGreater(B=0, A=@Dot___1_1_1_1.Result) {template="Greater::Execute(in A,in B,out Result)"}.Result) {variable=is on right}
   BranchNode_1_1_1 = &RigVMFunction_ControlFlowBranch(Condition=@n9.Result, ExecuteContext=@RigUnit_SequenceAggregate.E)
   n11 = RigVMVariableNode() {variable=vector pos}
   RigVMVariableNode(
    Value=(X=0,Y=0,Z=0)
    ExecuteContext=@BranchNode_1_1_1.True
    Value.Y=&RigUnit_GetTransform(bInitial=true, Item=@n1.Item).Transform.Translation.Y
    Value.X=@n11.Value.X
    Value.Z=@n11.Value.Z
   ) {variable=vector pos}
   Dot___1_1 = &RigVMFunction_MathVectorDot(
    B=(X=1,Y=1,Z=0)
    B.Y=&RigVMFunction_MathBoolToFloat(Value=@n9.Result).Result
    B.Z=&RigVMFunction_MathBoolToFloat(Value=@n10.Result).Result
    B.X=&RigVMFunction_MathBoolToFloat(Value=@n7.Result).Result
    A=RigVMVariableNode() {variable=vector pos}.Value
   ) {template="Dot,|::Execute(in A,in B,out Result)"}
   > RigVMVariableNode(Value=&RigVMFunction_MathDoubleGreater(B=0, A=@Dot___1_1.Result) {template="Greater::Execute(in A,in B,out Result)"}.Result) {variable=is on right}
   BranchNode_2_1_1 = &RigVMFunction_ControlFlowBranch(Condition=@n10.Result, ExecuteContext=@RigUnit_SequenceAggregate.F)
   n12 = RigVMVariableNode() {variable=vector pos}
   RigVMVariableNode(
    Value=(X=0,Y=0,Z=0)
    ExecuteContext=@BranchNode_2_1_1.True
    Value.Z=&RigUnit_GetTransform(bInitial=true, Item=@n1.Item).Transform.Translation.Z
    Value.X=@n12.Value.X
    Value.Y=@n12.Value.Y
   ) {variable=vector pos}
   Dot___1_1_1 = &RigVMFunction_MathVectorDot(
    B=(X=1,Y=1,Z=0)
    B.Y=&RigVMFunction_MathBoolToFloat(Value=@n9.Result).Result
    B.Z=&RigVMFunction_MathBoolToFloat(Value=@n10.Result).Result
    B.X=&RigVMFunction_MathBoolToFloat(Value=@n7.Result).Result
    A=RigVMVariableNode() {variable=vector pos}.Value
   ) {template="Dot,|::Execute(in A,in B,out Result)"}
   > RigVMVariableNode(Value=&RigVMFunction_MathDoubleGreater(B=0, A=@Dot___1_1_1.Result) {template="Greater::Execute(in A,in B,out Result)"}.Result) {variable=is on right}
   RigVMFunctionReturnNode(ExecuteContext=@RigUnit_SequenceAggregate_1.B, Right Side=RigVMVariableNode() {variable=is on right}.Value)
   &RigVMFunction_MathDoubleLess(B=0, A=@Dot___1_1.Result) {template="Less::Execute(in A,in B,out Result)"}
   &RigVMFunction_MathDoubleLess(B=0, A=@Dot___1_1_1.Result) {template="Less::Execute(in A,in B,out Result)"}
   &RigVMFunction_MathDoubleLess(B=0, A=@Dot___1_1_1_1.Result) {template="Less::Execute(in A,in B,out Result)"}
  ]}
 ContainedGraph {nodes [
   n1 = RigVMFunctionEntryNode()
   RigUnit_SequenceAggregate = &RigVMFunction_Sequence(ExecuteContext=@n1.ExecuteContext)
   RigUnit_SequenceAggregate_1 = &RigVMFunction_Sequence(ExecuteContext=@RigUnit_SequenceAggregate.B)
   RigUnit_SequenceAggregate_2 = &RigVMFunction_Sequence(ExecuteContext=@RigUnit_SequenceAggregate_1.B)
   RigUnit_SequenceAggregate_3 = &RigVMFunction_Sequence(ExecuteContext=@RigUnit_SequenceAggregate_2.B)
   RigUnit_SequenceAggregate_4 = &RigVMFunction_Sequence(ExecuteContext=@RigUnit_SequenceAggregate_3.B)
   RigUnit_SequenceAggregate_5 = &RigVMFunction_Sequence(ExecuteContext=@RigUnit_SequenceAggregate_4.B)
   RigVMFunctionReturnNode(
    A=@RigUnit_SequenceAggregate.A
    B=@RigUnit_SequenceAggregate_1.A
    C=@RigUnit_SequenceAggregate_2.A
    D=@RigUnit_SequenceAggregate_3.A
    E=@RigUnit_SequenceAggregate_4.A
    F=@RigUnit_SequenceAggregate_5.A
    G=@RigUnit_SequenceAggregate_5.B
   )
  ]}
 Find_ContainedGraph {nodes [
   n1 = RigVMFunctionEntryNode()
   Contains = /Script/RigVM.RigVMFunction_StringContains(Name=@n1.Element, Search=@n1.Separator) {is=struct RigVMUnitNode, template="Contains::Execute(in Name,in Search,out Result)"}
   To_String = RigVMDispatchNode(Value=@n1.Value) {template="DISPATCH_RigDispatch_ToString(in Value,out Result)"}
   RigVMFunction_StringSplit = /Script/RigVM.RigVMFunction_StringSplit(Separator=@n1.Separator, Value=@To_String.Result) {is=struct RigVMUnitNode}
   Num = RigVMDispatchNode(Array=@RigVMFunction_StringSplit.Result) {template="DISPATCH_RigVMDispatch_ArrayGetNum(in Array,out Num)"}
   > RigVMFunctionReturnNode(Success=RigVMDispatchNode(
     True=@Contains.Result
     Condition=RigVMDispatchNode(B=0, A=@Num.Num) {template="DISPATCH_RigVMDispatch_CoreEquals(in A,in B,out Result)"}.Result
     False=RigVMDispatchNode(Element=@n1.Element, Array=@RigVMFunction_StringSplit.Result) {template="DISPATCH_RigVMDispatch_ArrayFind(in Array,in Element,out Index,out Success)"}.Success
    ) {template="DISPATCH_RigVMDispatch_If(in Condition,in True,in False,out Result)"}.Result)
  ]}
 New Function_ContainedGraph {nodes [
   RigVMFunctionEntryNode()
   > RigVMFunctionReturnNode()
  ]}
 New Function_ContainedGraph {nodes [
   RigVMFunctionEntryNode()
   > RigVMFunctionReturnNode()
  ]}
 New Function_ContainedGraph {nodes [
   DrawAxis = RigVMFunctionReferenceNode(
    ExecuteContext=()
    Scale=15
    Thickness=2
    Item=RigVMDispatchNode(Default=(Type=Bone,Name="None"), Name=Effector Ori, NameSpace=Self) {template="DISPATCH_RigDispatch_GetModuleMetadata(in Name,in NameSpace,in Default,out Value,out Found)"}.Value
    Active=RigVMVariableNode() {variable=Debug}.Value
   )
   > DrawAxis_1 = RigVMFunctionReferenceNode(
    Scale=15
    Thickness=2
    Item=RigVMDispatchNode(
     Index=0
     Array=RigVMDispatchNode(Default=(), Name=Aim Follow Bones, NameSpace=Self) {template="DISPATCH_RigDispatch_GetModuleMetadata(in Name,in NameSpace,in Default,out Value,out Found)"}.Value
    ) {template="DISPATCH_RigVMDispatch_ArrayGetAtIndex(in Array,in Index,out Element)"}.Element
    Active=RigVMVariableNode() {variable=Debug}.Value
   )
   n1 = RigVMFunctionEntryNode()
   > RigVMFunction_Sequence_1 = RigVMAggregateNode()
   RigVMVariableNode(Value=@n1.On Right Side, ExecuteContext=@RigVMFunction_Sequence_1.A) {variable=Is on Right Side}
   n2 = &RigVMFunction_ControlFlowBranch(ExecuteContext=@RigVMFunction_Sequence_1.B, Condition=RigVMVariableNode() {variable=FK_or_IK}.Value)
   RigVMFunction_Sequence = RigVMAggregateNode(ExecuteContext=@n2.False)
   n3 = RigVMDispatchNode(Index=0, Array=RigVMVariableNode() {variable=FK Controls}.Value) {template="DISPATCH_RigVMDispatch_ArrayGetAtIndex(in Array,in Index,out Element)"}
   n4 = &RigUnit_HierarchyGetParent(bDefaultParent=True, Child=@n3.Element)
   Parent = RigVMFunctionReferenceNode(
    New Parent=&RigUnit_HierarchyGetParent(
     bDefaultParent=True
     Child=RigVMDispatchNode(Default=(Type=Bone,Name="None"), Name=Item, NameSpace=Self, Item=@n3.Element) {template="DISPATCH_RigDispatch_GetMetadata(in Item,in Name,in NameSpace,in Default,out Value,out Found)"}.Value
    ).Parent
    Child=@n4.Parent
    ExecuteContext=@RigVMFunction_Sequence.A
   )
   RigVMFunction_ControlFlowBranch = &RigVMFunction_ControlFlowBranch(Condition=@n1.local, ExecuteContext=@Parent.ExecuteContext)
   Set Transform_1 = &RigUnit_SetRotation(
    Space=GlobalSpace
    Weight=1
    bInitial=False
    Item=@n4.Parent
    Value=&RigUnit_ProjectTransformToNewParent(
     bChildInitial=True
     bNewParentInitial=False
     bOldParentInitial=True
     Child=@n4.Parent
     NewParent=RigVMVariableNode() {variable=FK Arm Rotation Space}.Value
     OldParent=RigVMVariableNode() {variable=FK Arm Rotation Space}.Value
    ).Transform.Rotation
    ExecutePin=@RigVMFunction_ControlFlowBranch.False
   )
   n5 = RigVMDispatchNode(ExecuteContext=@RigVMFunction_Sequence.B, Array=RigVMVariableNode() {variable=FK Controls}.Value) {template="DISPATCH_RigVMDispatch_ArrayIterator(in Array,out Element,out Index,out Count,out Ratio)"}
   DISPATCH_RigDispatch_GetMetadata = RigVMDispatchNode(Default=(Type=Bone,Name="None"), Name=Item, NameSpace=Self, Item=@n5.Element) {template="DISPATCH_RigDispatch_GetMetadata(in Item,in Name,in NameSpace,in Default,out Value,out Found)"}
   > Parent_1 = RigVMFunctionReferenceNode(Child=@DISPATCH_RigDispatch_GetMetadata.Value, New Parent=@n5.Element)
   > /Script/ControlRig.RigUnit_SetControlVisibility(bVisible=true, Item=@n5.Element) {is=struct RigVMUnitNode, template="SetControlVisibility::Execute(in Item,in Pattern,in bVisible)"}
   Match IK onto FK = RigVMFunctionReferenceNode(
    IK Control=RigVMDispatchNode(Index=0, Array=RigVMVariableNode() {variable=IK Controls}.Value) {template="DISPATCH_RigVMDispatch_ArrayGetAtIndex(in Array,in Index,out Element)"}.Element
    Pole Vector Control=RigVMDispatchNode(Index=1, Array=RigVMVariableNode() {variable=IK Controls}.Value) {template="DISPATCH_RigVMDispatch_ArrayGetAtIndex(in Array,in Index,out Element)"}.Element
    ExecuteContext=@RigVMFunction_Sequence.C
    Bones=RigVMVariableNode() {variable=Bones}.Value
   )
   n6 = RigVMDispatchNode(ExecuteContext=@RigVMFunction_Sequence.D, Array=RigVMVariableNode() {variable=IK Controls}.Value) {template="DISPATCH_RigVMDispatch_ArrayIterator(in Array,out Element,out Index,out Count,out Ratio)"}
   > SetControlVisibility = &RigUnit_SetControlVisibility(bVisible=false, Item=@n6.Element) {template="SetControlVisibility::Execute(in Item,in Pattern,in bVisible)"}
   DISPATCH_RigDispatch_GetModuleMetadata = RigVMDispatchNode(Default=(Type=Bone,Name="None"), Name=Effector Null, NameSpace=Self) {template="DISPATCH_RigDispatch_GetModuleMetadata(in Name,in NameSpace,in Default,out Value,out Found)"}
   Softness = RigVMCollapseNode(
    Maximum=1
    Effector=@DISPATCH_RigDispatch_GetModuleMetadata.Value
    Bones=RigVMVariableNode() {variable=Bones}.Value
    Softness=RigVMVariableNode() {variable=Softness}.Value
   )
   n7 = RigVMVariableNode() {variable=Is on Right Side}
   n8 = RigVMVariableNode() {variable=Auto Determine Axis}
   IKTwoBone = RigVMFunctionReferenceNode(
    Line Thickness=0.25
    Stretch Maximum Ratio=2
    Stretch Start Ratio=1
    debug=false
    scale=(X=0,Y=0,Z=0)
    End Ctrl=@DISPATCH_RigDispatch_GetModuleMetadata.Value
    Bone B=RigVMDispatchNode(Index=1, Array=RigVMVariableNode() {variable=Bones}.Value) {template="DISPATCH_RigVMDispatch_ArrayGetAtIndex(in Array,in Index,out Element)"}.Element
    Bone C=RigVMDispatchNode(Index=2, Array=RigVMVariableNode() {variable=Bones}.Value) {template="DISPATCH_RigVMDispatch_ArrayGetAtIndex(in Array,in Index,out Element)"}.Element
    Bone A=RigVMDispatchNode(Index=0, Array=RigVMVariableNode() {variable=Bones}.Value) {template="DISPATCH_RigVMDispatch_ArrayGetAtIndex(in Array,in Index,out Element)"}.Element
    Pole Vector Node=RigVMDispatchNode(Index=1, Array=RigVMVariableNode() {variable=IK Controls}.Value) {template="DISPATCH_RigVMDispatch_ArrayGetAtIndex(in Array,in Index,out Element)"}.Element
    Item ALength=RigVMDispatchNode(False=0, True=@Softness.Length A, Condition=RigVMVariableNode() {variable=Enable Softness}.Value) {template="DISPATCH_RigVMDispatch_If(in Condition,in True,in False,out Result)"}.Result
    Item BLength=RigVMDispatchNode(False=0, True=@Softness.Length B, Condition=RigVMVariableNode() {variable=Enable Softness}.Value) {template="DISPATCH_RigVMDispatch_If(in Condition,in True,in False,out Result)"}.Result
    Primary Axis=RigVMDispatchNode(
     True=(X=0,Y=0,Z=0)
     True.X=RigVMDispatchNode(False=1, True=-1, Condition=@n7.Value) {template="DISPATCH_RigVMDispatch_If(in Condition,in True,in False,out Result)"}.Result
     False=RigVMVariableNode() {variable=Primary Axis}.Value
     Condition=@n8.Value
    ) {template="DISPATCH_RigVMDispatch_If(in Condition,in True,in False,out Result)"}.Result
    Secondary Axis=RigVMDispatchNode(
     True=(X=0,Y=0,Z=0)
     True.Y=RigVMDispatchNode(False=-1, True=1, Condition=@n7.Value) {template="DISPATCH_RigVMDispatch_If(in Condition,in True,in False,out Result)"}.Result
     False=RigVMVariableNode() {variable=Secondary Axis}.Value
     Condition=@n8.Value
    ) {template="DISPATCH_RigVMDispatch_If(in Condition,in True,in False,out Result)"}.Result
    ExecuteContext=@n2.True
    Enable Stretch=RigVMVariableNode() {variable=Enable Stretch}.Value
   )
   > Project to New Parent = RigVMFunctionReferenceNode(Propagate to Children=True, Parents=RigVMVariableNode() {variable=Bones}.Value, Children=RigVMVariableNode() {variable=FK Controls}.Value)
   > n9 = RigVMDispatchNode(Array=RigVMVariableNode() {variable=IK Controls}.Value) {template="DISPATCH_RigVMDispatch_ArrayIterator(in Array,out Element,out Index,out Count,out Ratio)"}
   > &RigUnit_SetControlVisibility(bVisible=true, Item=@n9.Element) {template="SetControlVisibility::Execute(in Item,in Pattern,in bVisible)"}
   n10 = RigVMDispatchNode(ExecuteContext=@n9.Completed, Array=RigVMVariableNode() {variable=FK Controls}.Value) {template="DISPATCH_RigVMDispatch_ArrayIterator(in Array,out Element,out Index,out Count,out Ratio)"}
   > &RigUnit_SetControlVisibility(bVisible=false, Item=@n10.Element) {template="SetControlVisibility::Execute(in Item,in Pattern,in bVisible)"}
   RigVMFunction_Sequence_2 = RigVMAggregateNode(ExecuteContext=@RigVMFunction_Sequence_1.C)
   n11 = &RigVMFunction_ControlFlowBranch(ExecuteContext=@RigVMFunction_Sequence_2.A, Condition=RigVMVariableNode() {variable=Attach Skeleton}.Value)
   DISPATCH_RigVMDispatch_ArrayIterator = RigVMDispatchNode(Array=@n1.Bones, ExecuteContext=@n11.True) {template="DISPATCH_RigVMDispatch_ArrayIterator(in Array,out Element,out Index,out Count,out Ratio)"}
   > FindClosestBone_0 = RigVMFunctionReferenceNode(Initial=true, Item=@DISPATCH_RigVMDispatch_ArrayIterator.Element, Bones=@n1.Driven Bones)
   > &RigUnit_SetTransform(
    Weight=1
    bInitial=False
    Item=@FindClosestBone_0.Value
    Value=&RigUnit_ProjectTransformToNewParent(
     bChildInitial=True
     bNewParentInitial=False
     bOldParentInitial=True
     NewParent=@DISPATCH_RigVMDispatch_ArrayIterator.Element
     OldParent=@DISPATCH_RigVMDispatch_ArrayIterator.Element
     Child=@FindClosestBone_0.Value
    ).Transform
   )
   n12 = &RigVMFunction_ControlFlowBranch(ExecuteContext=@RigVMFunction_Sequence_2.B, Condition=RigVMVariableNode() {variable=Enable PV Follow}.Value)
   n13 = RigVMDispatchNode(Default=(Type=Bone,Name="None"), Name=Effector Null, NameSpace=Self) {template="DISPATCH_RigDispatch_GetModuleMetadata(in Name,in NameSpace,in Default,out Value,out Found)"}
   n14 = RigVMDispatchNode(Default=(Type=Bone,Name="None"), Name=Effector Ori, NameSpace=Self) {template="DISPATCH_RigDispatch_GetModuleMetadata(in Name,in NameSpace,in Default,out Value,out Found)"}
   /Script/ControlRig.RigUnit_PositionConstraintLocalSpaceOffset(
    Filter=(bX=True,bY=True,bZ=True)
    Parents=((Item=(Type=Bone,Name="None"),Weight=1))
    Weight=1
    bMaintainOffset=True
    Parents.0.Item=@n13.Value
    Child=@n14.Value
    ExecutePin=@n12.True
   ) {is=struct RigVMUnitNode}
   > &RigUnit_RotationConstraintLocalSpaceOffset(
    AdvancedSettings=(InterpolationType=Average,RotationOrderForFilter=XZY)
    Parents=((Item=(Type=Bone,Name="None"),Weight=1))
    Weight=1
    bMaintainOffset=false
    Parents.0.Item=@n13.Value
    Child=@n14.Value
    Filter=@n1.PV Follow Twist Lock
   )
   n15 = RigVMDispatchNode(
    Index=0
    Array=RigVMDispatchNode(Default=(), Name=Aim Follow Bones, NameSpace=Self) {template="DISPATCH_RigDispatch_GetModuleMetadata(in Name,in NameSpace,in Default,out Value,out Found)"}.Value
   ) {template="DISPATCH_RigVMDispatch_ArrayGetAtIndex(in Array,in Index,out Element)"}
   n16 = RigVMDispatchNode(True=-1, Condition=RigVMVariableNode() {variable=Is on Right Side}.Value, False=RigVMVariableNode() {variable=Mirror Threshold}.Value) {template="DISPATCH_RigVMDispatch_If(in Condition,in True,in False,out Result)"}
   > &RigUnit_AimConstraintLocalSpaceOffset(
    AdvancedSettings=(DebugSettings=(bEnabled=false,Scale=10,WorldOffset=(Rotation=(X=0,Y=0,Z=0,W=1),Translation=(X=0,Y=0,Z=0),Scale3D=(X=1,Y=1,Z=1))),RotationOrderForFilter=XZY)
    Filter=(bX=True,bY=True,bZ=True)
    Parents=((Item=(Type=Bone,Name="None"),Weight=1))
    Weight=1
    WorldUp=(Target=(X=0,Y=0,Z=1),Kind=Location,Space=(Type=None,Name="None"))
    bMaintainOffset=true
    Child=@n15.Element
    Parents.0.Item=@n14.Value
    WorldUp.Space=@n14.Value
    AimAxis=/Script/RigVM.RigVMFunction_MathVectorMul(B=(X=1,Y=1,Z=1), B.X=@n16.Result, B.Y=@n16.Result, B.Z=@n16.Result, A=RigVMVariableNode() {variable=PV Follow Aim Axis}.Value) {is=struct RigVMUnitNode, template="Multiply::Execute(in A,in B,out Result)"}.Result
    UpAxis=&RigVMFunction_MathVectorMul(B=(X=1,Y=1,Z=1), B.X=@n16.Result, B.Y=@n16.Result, B.Z=@n16.Result, A=RigVMVariableNode() {variable=PV Follow Up Axis}.Value) {template="Multiply::Execute(in A,in B,out Result)"}.Result
    WorldUp.Target=&RigVMFunction_MathVectorMul(B=(X=1,Y=1,Z=1), B.X=@n16.Result, B.Y=@n16.Result, B.Z=@n16.Result, A=RigVMVariableNode() {variable=PV Follow Aim Direction}.Value) {template="Multiply::Execute(in A,in B,out Result)"}.Result
   )
   n17 = RigVMVariableNode() {variable=IK Controls}
   > Parent_2 = RigVMFunctionReferenceNode(
    New Parent=@n15.Element
    Child=&RigUnit_HierarchyGetParent(
     bDefaultParent=True
     Child=RigVMDispatchNode(Index=1, Array=@n17.Value) {template="DISPATCH_RigVMDispatch_ArrayGetAtIndex(in Array,in Index,out Element)"}.Element
    ).Parent
   )
   n18 = &RigVMFunction_ControlFlowBranch(ExecuteContext=@RigVMFunction_Sequence_2.C, Condition=RigVMVariableNode() {variable=Keep Aligned}.Value)
   Get_Module_Metadata_2_1 = RigVMDispatchNode(Default=(Type=Bone,Name="None"), Name=Aligned Null, NameSpace=Self) {template="DISPATCH_RigDispatch_GetModuleMetadata(in Name,in NameSpace,in Default,out Value,out Found)"}
   &RigUnit_SetControlVisibility(bVisible=false, Item=@Get_Module_Metadata_2_1.Value, ExecutePin=@n18.False) {template="SetControlVisibility::Execute(in Item,in Pattern,in bVisible)"}
   Get_Module_Metadata_1 = RigVMDispatchNode(Default=(), Name=Rig Bones, NameSpace=Self) {template="DISPATCH_RigDispatch_GetModuleMetadata(in Name,in NameSpace,in Default,out Value,out Found)"}
   Get_Module_Metadata = RigVMDispatchNode(Default=(Type=Bone,Name="None"), Name=Aligned Null, NameSpace=Self) {template="DISPATCH_RigDispatch_GetModuleMetadata(in Name,in NameSpace,in Default,out Value,out Found)"}
   &RigUnit_ParentConstraint(
    AdvancedSettings=(InterpolationType=Average,RotationOrderForFilter=XZY)
    Parents=((Item=(Type=Bone,Name="None"),Weight=1))
    Weight=1
    Parents.0.Item=RigVMDispatchNode(Index=1, Array=@Get_Module_Metadata_1.Value) {template="DISPATCH_RigVMDispatch_ArrayGetAtIndex(in Array,in Index,out Element)"}.Element
    Child=&RigUnit_HierarchyGetParent(bDefaultParent=True, Child=@Get_Module_Metadata.Value).Parent
    ExecutePin=@n18.True
   )
   > DrawAxis_2 = RigVMFunctionReferenceNode(Active=false, Scale=15, Thickness=2, Item=@Get_Module_Metadata.Value)
   Get_Module_Metadata_2 = RigVMDispatchNode(Default=(Type=Bone,Name="None"), Name=Aligned Null, NameSpace=Self) {template="DISPATCH_RigDispatch_GetModuleMetadata(in Name,in NameSpace,in Default,out Value,out Found)"}
   > &RigUnit_RotationConstraintLocalSpaceOffset(
    AdvancedSettings=(InterpolationType=Average,RotationOrderForFilter=XZY)
    Filter=(bX=True,bY=True,bZ=True)
    Parents=((Item=(Type=Bone,Name="None"),Weight=1))
    Weight=1
    Child=RigVMDispatchNode(Index=2, Array=&RigUnit_FindItemsWithMetadataTag(NameSpace=Self, Tag=Bones).Items) {template="DISPATCH_RigVMDispatch_ArrayGetAtIndex(in Array,in Index,out Element)"}.Element
    Parents.0.Item=@Get_Module_Metadata_2.Value
   )
   > &RigUnit_SetControlVisibility(bVisible=True, Item=@Get_Module_Metadata_2.Value) {template="SetControlVisibility::Execute(in Item,in Pattern,in bVisible)"}
   RigVMFunctionReturnNode(ExecuteContext=@RigVMFunction_Sequence_1.E)
   Get_Module_Metadata_1_1 = RigVMDispatchNode(Default=(), Name=Rig Bones, NameSpace=Self) {template="DISPATCH_RigDispatch_GetModuleMetadata(in Name,in NameSpace,in Default,out Value,out Found)"}
   RigVMDispatchNode(Index=2, Array=@Get_Module_Metadata_1_1.Value) {template="DISPATCH_RigVMDispatch_ArrayGetAtIndex(in Array,in Index,out Element)"}
   Multiply = &RigVMFunction_MathDoubleMul(
    A=RigVMDispatchNode(False=1, True=-1, Condition=RigVMVariableNode() {variable=Is on Right Side}.Value) {template="DISPATCH_RigVMDispatch_If(in Condition,in True,in False,out Result)"}.Result
    B=RigVMVariableNode() {variable=Mirror Threshold}.Value
   )
   n19 = RigVMDispatchNode(Index=1, Array=RigVMVariableNode() {variable=Bones}.Value) {template="DISPATCH_RigVMDispatch_ArrayGetAtIndex(in Array,in Index,out Element)"}
   n20 = RigVMDispatchNode(Index=0, Array=RigVMVariableNode() {variable=Bones}.Value) {template="DISPATCH_RigVMDispatch_ArrayGetAtIndex(in Array,in Index,out Element)"}
   n21 = /Script/RigVM.RigVMFunction_MathVectorUnit(Value=/Script/RigVM.RigVMFunction_MathVectorSub(
     A=&RigUnit_GetTransform(Space=LocalSpace, bInitial=true, Item=@n19.Element).Transform.Translation
     B=&RigUnit_GetTransform(Space=LocalSpace, bInitial=true, Item=@n20.Element).Transform.Translation
    ) {is=struct RigVMUnitNode, template="Subtract::Execute(in A,in B,out Result)"}.Result) {is=struct RigVMUnitNode, template="Unit::Execute(in Value,out Result)"}
   DISPATCH_RigVMDispatch_If = RigVMDispatchNode(
    True=(X=0,Y=0,Z=0)
    True.Y=&RigVMFunction_MathDoubleMul(A=@Multiply.Result, B=@n21.Result.X).Result
    Condition=RigVMVariableNode() {variable=Auto Determine Axis}.Value
    False=RigVMVariableNode() {variable=Secondary Axis}.Value
   ) {template="DISPATCH_RigVMDispatch_If(in Condition,in True,in False,out Result)"}
   RigVMDispatchNode(
    True=(X=0,Y=0,Z=0)
    True.X=@n21.Result.X
    Condition=RigVMVariableNode() {variable=Auto Determine Axis}.Value
    False=RigVMVariableNode() {variable=Primary Axis}.Value
   ) {template="DISPATCH_RigVMDispatch_If(in Condition,in True,in False,out Result)"}
   &RigVMFunction_MathDoubleMul(A=@n21.Result.X, B=RigVMVariableNode() {variable=Mirror Threshold}.Value)
   &RigVMFunction_MathDoubleSub(A=@Multiply.Result, B=@n21.Result.X)
  ]}
 New Function_ContainedGraph {nodes [
   n1 = RigVMFunctionEntryNode()
   DISPATCH_RigDispatch_GetModuleMetadata = RigVMDispatchNode(Default=None, NameSpace=Self, Name=@n1.Name) {template="DISPATCH_RigDispatch_GetModuleMetadata(in Name,in NameSpace,in Default,out Value,out Found)"}
   GetAnimationChannel = /Script/ControlRig.RigUnit_GetBoolAnimationChannel(
    bInitial=False
    Channel=@DISPATCH_RigDispatch_GetModuleMetadata.Value
    Control=&RigUnit_ResolveConnector(SkipSocket=true, Connector=@n1.Connector).Result.Name
   ) {is=struct RigVMUnitNode, template="GetAnimationChannel::Execute(out Value,in Control,in Channel,in bInitial)"}
   > RigVMFunctionReturnNode(Value=@GetAnimationChannel.Value)
  ]}
 New Function_ContainedGraph {nodes [
   RigVMFunctionEntryNode()
   > RigVMFunction_Sequence = RigVMAggregateNode()
   Get_Module_Metadata_1 = RigVMDispatchNode(Default=(Type=Bone,Name="None"), Name=IK Control, NameSpace=Self) {template="DISPATCH_RigDispatch_GetModuleMetadata(in Name,in NameSpace,in Default,out Value,out Found)"}
   Get_Module_Metadata_2 = RigVMDispatchNode(Default=(), Name=Phantom Rig Bones, NameSpace=Self) {template="DISPATCH_RigDispatch_GetModuleMetadata(in Name,in NameSpace,in Default,out Value,out Found)"}
   /Script/ControlRig.RigUnit_CCDIKItemArray(
    BaseRotationLimit=30
    MaxIterations=10
    Precision=1
    RotationLimits=()
    Weight=1
    bPropagateToChildren=True
    bStartFromTail=false
    EffectorTransform=&RigUnit_GetTransform(bInitial=False, Item=@Get_Module_Metadata_1.Value).Transform
    Items=@Get_Module_Metadata_2.Value
    ExecutePin=@RigVMFunction_Sequence.A
   ) {is=struct RigVMUnitNode}
   Get_Module_Metadata_1_1 = RigVMDispatchNode(Default=(Type=Bone,Name="None"), Name=IK Control, NameSpace=Self) {template="DISPATCH_RigDispatch_GetModuleMetadata(in Name,in NameSpace,in Default,out Value,out Found)"}
   Get_Module_Metadata_2_1 = RigVMDispatchNode(Default=(), Name=Thigh Phantom Bones, NameSpace=Self) {template="DISPATCH_RigDispatch_GetModuleMetadata(in Name,in NameSpace,in Default,out Value,out Found)"}
   > &RigUnit_CCDIKItemArray(
    BaseRotationLimit=30
    MaxIterations=10
    Precision=1
    RotationLimits=()
    Weight=1
    bPropagateToChildren=True
    bStartFromTail=false
    EffectorTransform=&RigUnit_GetTransform(bInitial=False, Item=@Get_Module_Metadata_1_1.Value).Transform
    Items=@Get_Module_Metadata_2_1.Value
   )
   Get_Module_Metadata_3 = RigVMDispatchNode(Default=(Type=Bone,Name="None"), Name=Thigh Control, NameSpace=Self) {template="DISPATCH_RigDispatch_GetModuleMetadata(in Name,in NameSpace,in Default,out Value,out Found)"}
   Parent = RigVMFunctionReferenceNode(
    New Parent=RigVMDispatchNode(
     Index=0
     Array=RigVMDispatchNode(Default=(), Name=Thigh Phantom Bones, NameSpace=Self) {template="DISPATCH_RigDispatch_GetModuleMetadata(in Name,in NameSpace,in Default,out Value,out Found)"}.Value
    ) {template="DISPATCH_RigVMDispatch_ArrayGetAtIndex(in Array,in Index,out Element)"}.Element
    Child=&RigUnit_HierarchyGetParent(bDefaultParent=True, Child=@Get_Module_Metadata_3.Value).Parent
    ExecuteContext=@RigVMFunction_Sequence.B
   )
   n1 = RigVMDispatchNode(
    Index=0
    Array=RigVMDispatchNode(Default=(), Name=Phantom Rig Bones, NameSpace=Self) {template="DISPATCH_RigDispatch_GetModuleMetadata(in Name,in NameSpace,in Default,out Value,out Found)"}.Value
   ) {template="DISPATCH_RigVMDispatch_ArrayGetAtIndex(in Array,in Index,out Element)"}
   n2 = &RigUnit_ResolveConnector(Connector=(Type=Connector,Name="HindLeg Primary"), SkipSocket=true)
   n3 = RigVMVariableNode() {variable=Secondary Axis}
   > &RigUnit_AimConstraintLocalSpaceOffset(
    Filter=(bX=True,bY=True,bZ=True)
    Parents=((Item=(Type=Bone,Name="None"),Weight=1))
    Weight=1
    WorldUp=(Target=(X=0,Y=0,Z=1),Kind=Location,Space=(Type=None,Name="None"))
    Child=@n1.Element
    Parents.0.Item=@Get_Module_Metadata_3.Value
    WorldUp.Space=@n2.Result
    AimAxis=RigVMVariableNode() {variable=Primary Axis}.Value
    UpAxis=@n3.Value
   )
   n4 = RigVMVariableNode() {variable=Debug}
   > DrawAxis = RigVMFunctionReferenceNode(Scale=15, Thickness=2, Item=@n1.Element, Active=@n4.Value)
   Multiply = &RigVMFunction_MathVectorMul(B=(X=25,Y=25,Z=25), A=@n3.Value) {template="Multiply::Execute(in A,in B,out Result)"}
   RigVMFunction_MathTransformTransformVector = /Script/RigVM.RigVMFunction_MathTransformTransformVector(Transform=&RigUnit_GetTransform(bInitial=false, Item=@n2.Result).Transform, Location=@Multiply.Result) {is=struct RigVMUnitNode}
   > RigVMFunction_DebugTransformMutableNoSpace = &RigVMFunction_DebugTransformMutableNoSpace(
    Color=(R=1,G=0,B=0,A=1)
    DebugDrawSettings=(DepthPriority=SDPG_Foreground,Lifetime=-1)
    Mode=Axes
    Scale=15
    Thickness=2
    Transform=(Rotation=(X=0,Y=0,Z=0,W=1),Translation=(X=0,Y=0,Z=0),Scale3D=(X=1,Y=1,Z=1))
    WorldOffset=(Rotation=(X=0,Y=0,Z=0,W=1),Translation=(X=0,Y=0,Z=0),Scale3D=(X=1,Y=1,Z=1))
    Transform.Translation=@RigVMFunction_MathTransformTransformVector.Result
    bEnabled=@n4.Value
   )
   Get_Module_Metadata_2_2 = RigVMDispatchNode(Default=(), Name=Phantom Rig Bones, NameSpace=Self) {template="DISPATCH_RigDispatch_GetModuleMetadata(in Name,in NameSpace,in Default,out Value,out Found)"}
   Get_Module_Metadata_1_1_1 = RigVMDispatchNode(Default=(Type=Bone,Name="None"), Name=IK Control, NameSpace=Self) {template="DISPATCH_RigDispatch_GetModuleMetadata(in Name,in NameSpace,in Default,out Value,out Found)"}
   > n5 = &RigUnit_PositionConstraintLocalSpaceOffset(
    Filter=(bX=True,bY=True,bZ=True)
    Parents=((Item=(Type=Bone,Name="None"),Weight=1))
    Weight=1
    bMaintainOffset=true
    Child=RigVMDispatchNode(Index=-1, Array=@Get_Module_Metadata_2_2.Value) {template="DISPATCH_RigVMDispatch_ArrayGetAtIndex(in Array,in Index,out Element)"}.Element
    Parents.0.Item=@Get_Module_Metadata_1_1_1.Value
   )
   Get_Module_Metadata_1_1_1_1_1 = RigVMDispatchNode(Default=(Type=Bone,Name="None"), Name=IK Control, NameSpace=Self) {template="DISPATCH_RigDispatch_GetModuleMetadata(in Name,in NameSpace,in Default,out Value,out Found)"}
   GetAnimationChannel = &RigUnit_GetBoolAnimationChannel(Channel=Stretch, bInitial=False, Control=@Get_Module_Metadata_1_1_1_1_1.Value.Name) {template="GetAnimationChannel::Execute(out Value,in Control,in Channel,in bInitial)"}
   RigVMFunction_ControlFlowBranch = &RigVMFunction_ControlFlowBranch(Condition=@GetAnimationChannel.Value, ExecuteContext=@n5.ExecutePin)
   Get_Module_Metadata_2_2_2 = RigVMDispatchNode(Default=(), Name=Phantom Rig Bones, NameSpace=Self) {template="DISPATCH_RigDispatch_GetModuleMetadata(in Name,in NameSpace,in Default,out Value,out Found)"}
   Get_Module_Metadata_1_1_1_1 = RigVMDispatchNode(Default=(Type=Bone,Name="None"), Name=Thigh Control, NameSpace=Self) {template="DISPATCH_RigDispatch_GetModuleMetadata(in Name,in NameSpace,in Default,out Value,out Found)"}
   &RigUnit_PositionConstraintLocalSpaceOffset(
    Filter=(bX=True,bY=True,bZ=True)
    Parents=((Item=(Type=Bone,Name="None"),Weight=1))
    Weight=1
    bMaintainOffset=true
    Child=RigVMDispatchNode(Index=1, Array=@Get_Module_Metadata_2_2_2.Value) {template="DISPATCH_RigVMDispatch_ArrayGetAtIndex(in Array,in Index,out Element)"}.Element
    Parents.0.Item=@Get_Module_Metadata_1_1_1_1.Value
    ExecutePin=@RigVMFunction_ControlFlowBranch.True
   )
   Get_Module_Metadata_2_1_1 = RigVMDispatchNode(Default=(), Name=Phantom Rig Bones, NameSpace=Self) {template="DISPATCH_RigDispatch_GetModuleMetadata(in Name,in NameSpace,in Default,out Value,out Found)"}
   Parent_1 = RigVMFunctionReferenceNode(
    New Parent=RigVMDispatchNode(Index=0, Array=@Get_Module_Metadata_2_1_1.Value) {template="DISPATCH_RigVMDispatch_ArrayGetAtIndex(in Array,in Index,out Element)"}.Element
    Child=RigVMDispatchNode(
     Index=0
     Array=RigVMDispatchNode(Default=(), Name=Bones, NameSpace=Self) {template="DISPATCH_RigDispatch_GetModuleMetadata(in Name,in NameSpace,in Default,out Value,out Found)"}.Value
    ) {template="DISPATCH_RigVMDispatch_ArrayGetAtIndex(in Array,in Index,out Element)"}.Element
    ExecuteContext=@RigVMFunction_Sequence.C
   )
   Get_Module_Metadata_2_2_1 = RigVMDispatchNode(Default=(), Name=Phantom Rig Bones, NameSpace=Self) {template="DISPATCH_RigDispatch_GetModuleMetadata(in Name,in NameSpace,in Default,out Value,out Found)"}
   Parent_1_1_2 = RigVMFunctionReferenceNode(
    Child=RigVMDispatchNode(
     Index=0
     Array=RigVMDispatchNode(Default=(), Name=Rig Bones, NameSpace=Self) {template="DISPATCH_RigDispatch_GetModuleMetadata(in Name,in NameSpace,in Default,out Value,out Found)"}.Value
    ) {template="DISPATCH_RigVMDispatch_ArrayGetAtIndex(in Array,in Index,out Element)"}.Element
    New Parent=RigVMDispatchNode(Index=1, Array=@Get_Module_Metadata_2_2_1.Value) {template="DISPATCH_RigVMDispatch_ArrayGetAtIndex(in Array,in Index,out Element)"}.Element
    ExecuteContext=@RigVMFunction_Sequence.D
   )
   Get_Module_Metadata_1_2 = RigVMDispatchNode(Default=(Type=Bone,Name="None"), Name=IK Control, NameSpace=Self) {template="DISPATCH_RigDispatch_GetModuleMetadata(in Name,in NameSpace,in Default,out Value,out Found)"}
   FWD TwoBoneLimb Quad = RigVMFunctionReferenceNode(
    Attach Skeleton=true
    Debug=True
    Enable PV Follow=true
    FK Arm Rotation Space=(Type=Bone,Name="root")
    FK Space=true
    FK_or_IK=true
    Keep Aligned=false
    Mirror Threshold=1
    PV Follow Aim Axis=(X=0,Y=1,Z=0)
    PV Follow Aim Direction=(X=0,Y=1,Z=0)
    PV Follow Twist Lock=(bX=false,bY=false,bZ=true)
    PV Follow Up Axis=(X=1,Y=0,Z=0)
    Softness=1
    local=true
    Bones=RigVMDispatchNode(Default=(), Name=Rig Bones, NameSpace=Self) {template="DISPATCH_RigDispatch_GetModuleMetadata(in Name,in NameSpace,in Default,out Value,out Found)"}.Value
    Driven Bones=&RigUnit_FindItemsWithMetadataTag(NameSpace=Self, Tag=Bones).Items
    FK Controls=&RigUnit_FindItemsWithMetadataTag(NameSpace=Self, Tag=Controls FK).Items
    IK Controls=&RigUnit_FindItemsWithMetadataTag(NameSpace=Self, Tag=Controls IK).Items
    End Ctrl=@Get_Module_Metadata_1_2.Value
    ExecuteContext=@RigVMFunction_Sequence.E
    Auto Determine Axis=RigVMVariableNode() {variable=Auto Determine Axis}.Value
    Enable Softness=RigVMVariableNode() {variable=Enable Softness}.Value
    Enable Stretch=RigVMVariableNode() {variable=Enable Stretch}.Value
    Secondary Axis=RigVMVariableNode() {variable=Secondary Axis}.Value
    Primary Axis=RigVMVariableNode() {variable=Primary Axis}.Value
   )
   RigVMFunctionReturnNode(ExecuteContext=@RigVMFunction_Sequence.F)
  ]}
 New Function_ContainedGraph {nodes [
   n1 = RigVMFunctionEntryNode()
   > n2 = RigVMDispatchNode(Array=@n1.IK Controls) {template="DISPATCH_RigVMDispatch_ArrayIterator(in Array,out Element,out Index,out Count,out Ratio)"}
   FWD Get FK or IK_1 = RigVMFunctionReferenceNode(Connector=(Type=Connector,Name="Settings Control"), Name=fk_ik_switch)
   > &RigUnit_SetControlVisibility(Item=@n2.Element, bVisible=@"FWD Get FK or IK_1".Value) {template="SetControlVisibility::Execute(in Item,in Pattern,in bVisible)"}
   n3 = RigVMDispatchNode(ExecuteContext=@n2.Completed, Array=@n1.FK Controls) {template="DISPATCH_RigVMDispatch_ArrayIterator(in Array,out Element,out Index,out Count,out Ratio)"}
   FWD Get FK or IK_1_1 = RigVMFunctionReferenceNode(Connector=(Type=Connector,Name="Settings Control"), Name=fk_ik_switch)
   > &RigUnit_SetControlVisibility(Item=@n3.Element, bVisible=&RigVMFunction_MathBoolNot(Value=@"FWD Get FK or IK_1_1".Value).Result) {template="SetControlVisibility::Execute(in Item,in Pattern,in bVisible)"}
   RigVMFunctionReturnNode(ExecuteContext=@n3.Completed)
  ]}
 New Function_ContainedGraph {nodes [
   n1 = RigVMFunctionEntryNode()
   > RigVMFunction_Sequence_1 = RigVMAggregateNode()
   RigVMVariableNode(Value=@n1.On Right Side, ExecuteContext=@RigVMFunction_Sequence_1.A) {variable=Is on Right Side}
   n2 = &RigVMFunction_ControlFlowBranch(ExecuteContext=@RigVMFunction_Sequence_1.B, Condition=RigVMVariableNode() {variable=FK_or_IK}.Value)
   RigVMFunction_Sequence = RigVMAggregateNode(ExecuteContext=@n2.False)
   n3 = &RigVMFunction_ControlFlowBranch(Condition=@n1.FK Space, ExecuteContext=@RigVMFunction_Sequence.A)
   RigVMFunction_ControlFlowBranch = &RigVMFunction_ControlFlowBranch(Condition=@n1.local, ExecuteContext=@n3.Completed)
   n4 = RigVMDispatchNode(Index=0, Array=RigVMVariableNode() {variable=FK Controls}.Value) {template="DISPATCH_RigVMDispatch_ArrayGetAtIndex(in Array,in Index,out Element)"}
   n5 = &RigUnit_HierarchyGetParent(bDefaultParent=True, Child=@n4.Element)
   n6 = RigVMVariableNode() {variable=FK Arm Rotation Space}
   Set Transform_1 = &RigUnit_SetRotation(
    Space=GlobalSpace
    Weight=1
    bInitial=False
    Item=@n5.Parent
    Value=&RigUnit_ProjectTransformToNewParent(bChildInitial=True, bNewParentInitial=False, bOldParentInitial=True, Child=@n5.Parent, NewParent=@n6.Value, OldParent=@n6.Value).Transform.Rotation
    ExecutePin=@RigVMFunction_ControlFlowBranch.False
   )
   Parent = RigVMFunctionReferenceNode(
    New Parent=&RigUnit_HierarchyGetParent(
     bDefaultParent=True
     Child=RigVMDispatchNode(Default=(Type=Bone,Name="None"), Name=Item, NameSpace=Self, Item=@n4.Element) {template="DISPATCH_RigDispatch_GetMetadata(in Item,in Name,in NameSpace,in Default,out Value,out Found)"}.Value
    ).Parent
    Child=@n5.Parent
    ExecuteContext=@n3.True
   )
   n7 = RigVMDispatchNode(ExecuteContext=@RigVMFunction_Sequence.B, Array=RigVMVariableNode() {variable=FK Controls}.Value) {template="DISPATCH_RigVMDispatch_ArrayIterator(in Array,out Element,out Index,out Count,out Ratio)"}
   DISPATCH_RigDispatch_GetMetadata = RigVMDispatchNode(Default=(Type=Bone,Name="None"), Name=Item, NameSpace=Self, Item=@n7.Element) {template="DISPATCH_RigDispatch_GetMetadata(in Item,in Name,in NameSpace,in Default,out Value,out Found)"}
   > Parent_1 = RigVMFunctionReferenceNode(Child=@DISPATCH_RigDispatch_GetMetadata.Value, New Parent=@n7.Element)
   > &RigUnit_SetControlVisibility(bVisible=true, Item=@n7.Element) {template="SetControlVisibility::Execute(in Item,in Pattern,in bVisible)"}
   n8 = RigVMVariableNode() {variable=IK Controls}
   Match IK onto FK = RigVMFunctionReferenceNode(
    IK Control=RigVMDispatchNode(Index=0, Array=@n8.Value) {template="DISPATCH_RigVMDispatch_ArrayGetAtIndex(in Array,in Index,out Element)"}.Element
    Pole Vector Control=RigVMDispatchNode(Index=1, Array=@n8.Value) {template="DISPATCH_RigVMDispatch_ArrayGetAtIndex(in Array,in Index,out Element)"}.Element
    ExecuteContext=@RigVMFunction_Sequence.C
    Bones=RigVMVariableNode() {variable=Bones}.Value
   )
   n9 = RigVMDispatchNode(ExecuteContext=@RigVMFunction_Sequence.D, Array=@n8.Value) {template="DISPATCH_RigVMDispatch_ArrayIterator(in Array,out Element,out Index,out Count,out Ratio)"}
   > SetControlVisibility = &RigUnit_SetControlVisibility(bVisible=false, Item=@n9.Element) {template="SetControlVisibility::Execute(in Item,in Pattern,in bVisible)"}
   DISPATCH_RigDispatch_GetModuleMetadata = RigVMDispatchNode(Default=(Type=Bone,Name="None"), Name=Effector Null, NameSpace=Self) {template="DISPATCH_RigDispatch_GetModuleMetadata(in Name,in NameSpace,in Default,out Value,out Found)"}
   n10 = RigVMVariableNode() {variable=Bones}
   Softness = RigVMCollapseNode(
    Maximum=1
    Effector=@DISPATCH_RigDispatch_GetModuleMetadata.Value
    Softness=RigVMVariableNode() {variable=Softness}.Value
    Bones=RigVMVariableNode() {variable=Bones}.Value
   )
   n11 = RigVMVariableNode() {variable=Enable Softness}
   n12 = RigVMVariableNode() {variable=Is on Right Side}
   IKTwoBone = RigVMFunctionReferenceNode(
    Line Thickness=0.25
    Primary Axis=(X=1,Y=0,Z=0)
    Secondary Axis=(X=0,Y=1,Z=0)
    Stretch Maximum Ratio=2
    Stretch Start Ratio=1
    debug=false
    scale=(X=0,Y=0,Z=0)
    End Ctrl=@DISPATCH_RigDispatch_GetModuleMetadata.Value
    Bone B=RigVMDispatchNode(Index=1, Array=@n10.Value) {template="DISPATCH_RigVMDispatch_ArrayGetAtIndex(in Array,in Index,out Element)"}.Element
    Bone C=RigVMDispatchNode(Index=2, Array=@n10.Value) {template="DISPATCH_RigVMDispatch_ArrayGetAtIndex(in Array,in Index,out Element)"}.Element
    Bone A=RigVMDispatchNode(Index=0, Array=@n10.Value) {template="DISPATCH_RigVMDispatch_ArrayGetAtIndex(in Array,in Index,out Element)"}.Element
    Pole Vector Node=RigVMDispatchNode(Index=1, Array=RigVMVariableNode() {variable=IK Controls}.Value) {template="DISPATCH_RigVMDispatch_ArrayGetAtIndex(in Array,in Index,out Element)"}.Element
    Item ALength=RigVMDispatchNode(False=0, True=@Softness.Length A, Condition=@n11.Value) {template="DISPATCH_RigVMDispatch_If(in Condition,in True,in False,out Result)"}.Result
    Item BLength=RigVMDispatchNode(False=0, True=@Softness.Length B, Condition=@n11.Value) {template="DISPATCH_RigVMDispatch_If(in Condition,in True,in False,out Result)"}.Result
    Primary Axis.X=RigVMDispatchNode(False=1, True=-1, Condition=@n12.Value) {template="DISPATCH_RigVMDispatch_If(in Condition,in True,in False,out Result)"}.Result
    Secondary Axis.Y=RigVMDispatchNode(False=-1, True=1, Condition=@n12.Value) {template="DISPATCH_RigVMDispatch_If(in Condition,in True,in False,out Result)"}.Result
    ExecuteContext=@n2.True
    Enable Stretch=RigVMVariableNode() {variable=Enable Stretch}.Value
   )
   > Project to New Parent = RigVMFunctionReferenceNode(Propagate to Children=True, Children=RigVMVariableNode() {variable=FK Controls}.Value, Parents=RigVMVariableNode() {variable=Bones}.Value)
   > n13 = RigVMDispatchNode(Array=RigVMVariableNode() {variable=IK Controls}.Value) {template="DISPATCH_RigVMDispatch_ArrayIterator(in Array,out Element,out Index,out Count,out Ratio)"}
   > &RigUnit_SetControlVisibility(bVisible=true, Item=@n13.Element) {template="SetControlVisibility::Execute(in Item,in Pattern,in bVisible)"}
   n14 = RigVMDispatchNode(ExecuteContext=@n13.Completed, Array=RigVMVariableNode() {variable=FK Controls}.Value) {template="DISPATCH_RigVMDispatch_ArrayIterator(in Array,out Element,out Index,out Count,out Ratio)"}
   > &RigUnit_SetControlVisibility(bVisible=false, Item=@n14.Element) {template="SetControlVisibility::Execute(in Item,in Pattern,in bVisible)"}
   RigVMFunction_Sequence_2 = RigVMAggregateNode(ExecuteContext=@RigVMFunction_Sequence_1.C)
   n15 = &RigVMFunction_ControlFlowBranch(ExecuteContext=@RigVMFunction_Sequence_2.A, Condition=RigVMVariableNode() {variable=Attach Skeleton}.Value)
   DISPATCH_RigVMDispatch_ArrayIterator = RigVMDispatchNode(Array=@n1.Bones, ExecuteContext=@n15.True) {template="DISPATCH_RigVMDispatch_ArrayIterator(in Array,out Element,out Index,out Count,out Ratio)"}
   > FindClosestBone_0 = RigVMFunctionReferenceNode(Initial=true, Item=@DISPATCH_RigVMDispatch_ArrayIterator.Element, Bones=@n1.Driven Bones)
   > &RigUnit_SetTransform(
    Weight=1
    bInitial=False
    Item=@FindClosestBone_0.Value
    Value=&RigUnit_ProjectTransformToNewParent(
     bChildInitial=True
     bNewParentInitial=False
     bOldParentInitial=True
     NewParent=@DISPATCH_RigVMDispatch_ArrayIterator.Element
     OldParent=@DISPATCH_RigVMDispatch_ArrayIterator.Element
     Child=@FindClosestBone_0.Value
    ).Transform
   )
   n16 = &RigVMFunction_ControlFlowBranch(ExecuteContext=@RigVMFunction_Sequence_2.B, Condition=RigVMVariableNode() {variable=Enable PV Follow}.Value)
   n17 = RigVMDispatchNode(Default=(Type=Bone,Name="None"), Name=Effector Null, NameSpace=Self) {template="DISPATCH_RigDispatch_GetModuleMetadata(in Name,in NameSpace,in Default,out Value,out Found)"}
   n18 = RigVMDispatchNode(Default=(Type=Bone,Name="None"), Name=Effector Ori, NameSpace=Self) {template="DISPATCH_RigDispatch_GetModuleMetadata(in Name,in NameSpace,in Default,out Value,out Found)"}
   &RigUnit_PositionConstraintLocalSpaceOffset(
    Filter=(bX=True,bY=True,bZ=True)
    Parents=((Item=(Type=Bone,Name="None"),Weight=1))
    Weight=1
    bMaintainOffset=True
    Parents.0.Item=@n17.Value
    Child=@n18.Value
    ExecutePin=@n16.True
   )
   > &RigUnit_RotationConstraintLocalSpaceOffset(
    AdvancedSettings=(InterpolationType=Average,RotationOrderForFilter=XZY)
    Parents=((Item=(Type=Bone,Name="None"),Weight=1))
    Weight=1
    bMaintainOffset=false
    Parents.0.Item=@n17.Value
    Child=@n18.Value
    Filter=@n1.PV Follow Twist Lock
   )
   n19 = RigVMDispatchNode(
    Index=0
    Array=RigVMDispatchNode(Default=(), Name=Aim Follow Bones, NameSpace=Self) {template="DISPATCH_RigDispatch_GetModuleMetadata(in Name,in NameSpace,in Default,out Value,out Found)"}.Value
   ) {template="DISPATCH_RigVMDispatch_ArrayGetAtIndex(in Array,in Index,out Element)"}
   n20 = RigVMDispatchNode(True=-1, Condition=RigVMVariableNode() {variable=Is on Right Side}.Value, False=RigVMVariableNode() {variable=Mirror Threshold}.Value) {template="DISPATCH_RigVMDispatch_If(in Condition,in True,in False,out Result)"}
   > &RigUnit_AimConstraintLocalSpaceOffset(
    AdvancedSettings=(DebugSettings=(bEnabled=false,Scale=10,WorldOffset=(Rotation=(X=0,Y=0,Z=0,W=1),Translation=(X=0,Y=0,Z=0),Scale3D=(X=1,Y=1,Z=1))),RotationOrderForFilter=XZY)
    Filter=(bX=True,bY=True,bZ=True)
    Parents=((Item=(Type=Bone,Name="None"),Weight=1))
    Weight=1
    WorldUp=(Target=(X=0,Y=0,Z=1),Kind=Location,Space=(Type=None,Name="None"))
    bMaintainOffset=true
    Child=@n19.Element
    Parents.0.Item=@n18.Value
    WorldUp.Space=@n18.Value
    AimAxis=&RigVMFunction_MathVectorMul(B=(X=1,Y=1,Z=1), B.X=@n20.Result, B.Y=@n20.Result, B.Z=@n20.Result, A=RigVMVariableNode() {variable=PV Follow Aim Axis}.Value) {template="Multiply::Execute(in A,in B,out Result)"}.Result
    UpAxis=&RigVMFunction_MathVectorMul(B=(X=1,Y=1,Z=1), B.X=@n20.Result, B.Y=@n20.Result, B.Z=@n20.Result, A=RigVMVariableNode() {variable=PV Follow Up Axis}.Value) {template="Multiply::Execute(in A,in B,out Result)"}.Result
    WorldUp.Target=&RigVMFunction_MathVectorMul(B=(X=1,Y=1,Z=1), B.X=@n20.Result, B.Y=@n20.Result, B.Z=@n20.Result, A=RigVMVariableNode() {variable=PV Follow Aim Direction}.Value) {template="Multiply::Execute(in A,in B,out Result)"}.Result
   )
   n21 = RigVMVariableNode() {variable=IK Controls}
   > Parent_2 = RigVMFunctionReferenceNode(
    New Parent=@n19.Element
    Child=&RigUnit_HierarchyGetParent(
     bDefaultParent=True
     Child=RigVMDispatchNode(Index=1, Array=@n21.Value) {template="DISPATCH_RigVMDispatch_ArrayGetAtIndex(in Array,in Index,out Element)"}.Element
    ).Parent
   )
   n22 = &RigVMFunction_ControlFlowBranch(ExecuteContext=@RigVMFunction_Sequence_2.C, Condition=RigVMVariableNode() {variable=Keep Aligned}.Value)
   Get_Module_Metadata_2_1 = RigVMDispatchNode(Default=(Type=Bone,Name="None"), Name=Aligned Null, NameSpace=Self) {template="DISPATCH_RigDispatch_GetModuleMetadata(in Name,in NameSpace,in Default,out Value,out Found)"}
   &RigUnit_SetControlVisibility(bVisible=false, Item=@Get_Module_Metadata_2_1.Value, ExecutePin=@n22.False) {template="SetControlVisibility::Execute(in Item,in Pattern,in bVisible)"}
   Get_Module_Metadata_1 = RigVMDispatchNode(Default=(), Name=Rig Bones, NameSpace=Self) {template="DISPATCH_RigDispatch_GetModuleMetadata(in Name,in NameSpace,in Default,out Value,out Found)"}
   Get_Module_Metadata = RigVMDispatchNode(Default=(Type=Bone,Name="None"), Name=Aligned Null, NameSpace=Self) {template="DISPATCH_RigDispatch_GetModuleMetadata(in Name,in NameSpace,in Default,out Value,out Found)"}
   &RigUnit_ParentConstraint(
    AdvancedSettings=(InterpolationType=Average,RotationOrderForFilter=XZY)
    Parents=((Item=(Type=Bone,Name="None"),Weight=1))
    Weight=1
    Parents.0.Item=RigVMDispatchNode(Index=1, Array=@Get_Module_Metadata_1.Value) {template="DISPATCH_RigVMDispatch_ArrayGetAtIndex(in Array,in Index,out Element)"}.Element
    Child=&RigUnit_HierarchyGetParent(bDefaultParent=True, Child=@Get_Module_Metadata.Value).Parent
    ExecutePin=@n22.True
   )
   Get_Module_Metadata_2 = RigVMDispatchNode(Default=(Type=Bone,Name="None"), Name=Aligned Null, NameSpace=Self) {template="DISPATCH_RigDispatch_GetModuleMetadata(in Name,in NameSpace,in Default,out Value,out Found)"}
   > &RigUnit_RotationConstraintLocalSpaceOffset(
    AdvancedSettings=(InterpolationType=Average,RotationOrderForFilter=XZY)
    Filter=(bX=True,bY=True,bZ=True)
    Parents=((Item=(Type=Bone,Name="None"),Weight=1))
    Weight=1
    Child=RigVMDispatchNode(Index=2, Array=&RigUnit_FindItemsWithMetadataTag(NameSpace=Self, Tag=Bones).Items) {template="DISPATCH_RigVMDispatch_ArrayGetAtIndex(in Array,in Index,out Element)"}.Element
    Parents.0.Item=@Get_Module_Metadata_2.Value
   )
   > &RigUnit_SetControlVisibility(bVisible=True, Item=@Get_Module_Metadata_2.Value) {template="SetControlVisibility::Execute(in Item,in Pattern,in bVisible)"}
   RigVMFunctionReturnNode(ExecuteContext=@RigVMFunction_Sequence_1.E)
   Get_Module_Metadata_1_1 = RigVMDispatchNode(Default=(), Name=Rig Bones, NameSpace=Self) {template="DISPATCH_RigDispatch_GetModuleMetadata(in Name,in NameSpace,in Default,out Value,out Found)"}
   RigVMDispatchNode(Index=2, Array=@Get_Module_Metadata_1_1.Value) {template="DISPATCH_RigVMDispatch_ArrayGetAtIndex(in Array,in Index,out Element)"}
   Multiply = &RigVMFunction_MathDoubleMul(
    A=RigVMDispatchNode(False=1, True=-1, Condition=RigVMVariableNode() {variable=Is on Right Side}.Value) {template="DISPATCH_RigVMDispatch_If(in Condition,in True,in False,out Result)"}.Result
    B=RigVMVariableNode() {variable=Mirror Threshold}.Value
   )
   n23 = RigVMDispatchNode(Index=1, Array=RigVMVariableNode() {variable=Bones}.Value) {template="DISPATCH_RigVMDispatch_ArrayGetAtIndex(in Array,in Index,out Element)"}
   n24 = RigVMDispatchNode(Index=0, Array=RigVMVariableNode() {variable=Bones}.Value) {template="DISPATCH_RigVMDispatch_ArrayGetAtIndex(in Array,in Index,out Element)"}
   n25 = &RigVMFunction_MathVectorUnit(Value=&RigVMFunction_MathVectorSub(
     A=&RigUnit_GetTransform(Space=LocalSpace, bInitial=true, Item=@n23.Element).Transform.Translation
     B=&RigUnit_GetTransform(Space=LocalSpace, bInitial=true, Item=@n24.Element).Transform.Translation
    ) {template="Subtract::Execute(in A,in B,out Result)"}.Result) {template="Unit::Execute(in Value,out Result)"}
   DISPATCH_RigVMDispatch_If = RigVMDispatchNode(
    True=(X=0,Y=0,Z=0)
    True.Y=&RigVMFunction_MathDoubleMul(A=@Multiply.Result, B=@n25.Result.X).Result
    Condition=RigVMVariableNode() {variable=Auto Determine Axis}.Value
    False=RigVMVariableNode() {variable=Secondary Axis}.Value
   ) {template="DISPATCH_RigVMDispatch_If(in Condition,in True,in False,out Result)"}
   RigVMDispatchNode(
    True=(X=0,Y=0,Z=0)
    True.X=@n25.Result.X
    Condition=RigVMVariableNode() {variable=Auto Determine Axis}.Value
    False=RigVMVariableNode() {variable=Primary Axis}.Value
   ) {template="DISPATCH_RigVMDispatch_If(in Condition,in True,in False,out Result)"}
   &RigVMFunction_MathDoubleMul(A=@n25.Result.X, B=RigVMVariableNode() {variable=Mirror Threshold}.Value)
   &RigVMFunction_MathDoubleSub(A=@Multiply.Result, B=@n25.Result.X)
  ]}
 New Function_ContainedGraph {nodes [
   n1 = RigVMFunctionEntryNode()
   Get_Module_Metadata_4 = RigVMDispatchNode(Default=(X=0,Y=0,Z=0), Name=Mirror Axis, NameSpace=Root) {template="DISPATCH_RigDispatch_GetModuleMetadata(in Name,in NameSpace,in Default,out Value,out Found)"}
   n2 = &RigVMFunction_MathVectorMul(A=&RigUnit_GetTransform(bInitial=False, Item=@n1.Item).Transform.Translation, B=@Get_Module_Metadata_4.Value) {template="Multiply::Execute(in A,in B,out Result)"}
   RigVMFunction_MathBoolOr = RigVMAggregateNode(
    B=&RigVMFunction_MathDoubleGreater(B=0, A=@n2.Result.Y) {template="Greater::Execute(in A,in B,out Result)"}.Result
    C=&RigVMFunction_MathDoubleGreater(B=0, A=@n2.Result.Z) {template="Greater::Execute(in A,in B,out Result)"}.Result
    A=&RigVMFunction_MathDoubleGreater(B=0, A=@n2.Result.X) {template="Greater::Execute(in A,in B,out Result)"}.Result
   )
   RigVMFunction_MathBoolNot = &RigVMFunction_MathBoolNot(Value=@RigVMFunction_MathBoolOr.Result)
   Less = &RigVMFunction_MathDoubleLess(B=0, A=@Get_Module_Metadata_4.Value.X) {template="Less::Execute(in A,in B,out Result)"}
   RigVMFunction_MathBoolOr_1 = RigVMAggregateNode(
    A=@Less.Result
    B=&RigVMFunction_MathDoubleLess(B=0, A=@Get_Module_Metadata_4.Value.Y) {template="Less::Execute(in A,in B,out Result)"}.Result
    C=&RigVMFunction_MathDoubleLess(B=0, A=@Get_Module_Metadata_4.Value.Z) {template="Less::Execute(in A,in B,out Result)"}.Result
   )
   > RigVMFunctionReturnNode(Result=RigVMDispatchNode(True=@RigVMFunction_MathBoolNot.Result, False=@RigVMFunction_MathBoolOr.Result, Condition=@RigVMFunction_MathBoolOr_1.Result) {template="DISPATCH_RigVMDispatch_If(in Condition,in True,in False,out Result)"}.Result)
  ]}
 New Function_ContainedGraph {nodes [
   n1 = RigVMFunctionEntryNode()
   Concat_4 = RigVMAggregateNode(B=@n1.name, A=@n1.prefix, C=@n1.side, D=@n1.suffix)
   > RigVMFunctionReturnNode(Result=@Concat_4.Result)
  ]}
 RigVMFunctionLibrary {nodes [
   CONSTR HindLeg = RigVMCollapseNode(ExecuteContext=())
   DrawAxis = RigVMCollapseNode(Active=true, ExecuteContext=(), Item=(Type=Bone), Scale=15, Thickness=2)
   FWD FK or IK Control Vis = RigVMCollapseNode(ExecuteContext=(), FK Controls=(), IK Controls=())
   FWD Get FK or IK = RigVMCollapseNode(Connector=(Type=Connector,Name="Settings Control"), Name=fk_ik_switch)
   FWD HindLeg = RigVMCollapseNode(ExecuteContext=())
   FWD IK HindLeg = RigVMCollapseNode(ExecuteContext=(), FK_or_IK=true)
   FWD TwoBoneLimb = RigVMCollapseNode(
    Attach Skeleton=true
    Auto Determine Axis=False
    Bones=()
    Debug=True
    Driven Bones=()
    Enable PV Follow=true
    Enable Softness=False
    Enable Stretch=false
    End Ctrl=(Type=Control,Name="None")
    ExecuteContext=()
    FK Arm Rotation Space=(Type=Bone,Name="root")
    FK Controls=()
    FK Space=true
    FK_or_IK=False
    IK Controls=()
    Keep Aligned=False
    Mirror Threshold=1
    PV Follow Aim Axis=(X=1,Y=0,Z=0)
    PV Follow Aim Direction=(X=0,Y=0,Z=1)
    PV Follow Twist Lock=(bX=True,bY=False,bZ=False)
    PV Follow Up Axis=(X=0,Y=0,Z=1)
    Primary Axis=(X=1,Y=0,Z=0)
    Secondary Axis=(X=0,Y=-1,Z=0)
    Softness=1
   )
   FWD TwoBoneLimb Quad = RigVMCollapseNode(
    Attach Skeleton=true
    Auto Determine Axis=False
    Bones=()
    Debug=True
    Driven Bones=()
    Enable PV Follow=true
    Enable Softness=False
    Enable Stretch=false
    End Ctrl=(Type=Control,Name="None")
    ExecuteContext=()
    FK Arm Rotation Space=(Type=Bone,Name="root")
    FK Controls=()
    FK_or_IK=False
    IK Controls=()
    Keep Aligned=False
    Mirror Threshold=1
    PV Follow Aim Axis=(X=1,Y=0,Z=0)
    PV Follow Aim Direction=(X=0,Y=0,Z=1)
    PV Follow Twist Lock=(bX=True,bY=False,bZ=False)
    PV Follow Up Axis=(X=0,Y=0,Z=1)
    Primary Axis=(X=1,Y=0,Z=0)
    Secondary Axis=(X=0,Y=-1,Z=0)
    Softness=1
   )
   Find = RigVMCollapseNode(Element=foot, Value=None)
   FindClosestBone = RigVMCollapseNode(Bones=(), ExecuteContext=(), Initial=true, Item=())
   Format Name = RigVMCollapseNode(name=arm, prefix=None, side=_fk_ik_switch, suffix=_fk_ik_switch)
   Get Side Metadata = RigVMCollapseNode(Item=(Type=Bone,Name="None"))
   IKTwoBone = RigVMCollapseNode(
    Bone A=(Type=Bone,Name="None")
    Bone B=(Type=Bone,Name="None")
    Bone C=(Type=Bone,Name="None")
    Enable Stretch=false
    End Ctrl=(Type=Control,Name="None")
    ExecuteContext=()
    Line Thickness=0.25
    Pole Vector Node=(Type=Control,Name="None")
    Primary Axis=(X=1,Y=0,Z=0)
    Secondary Axis=(X=0,Y=-1,Z=0)
    Stretch Maximum Ratio=1.25
    Stretch Start Ratio=0.75
    debug=true
    scale=(X=0,Y=0,Z=0)
   )
   Is Right Side v2 = RigVMCollapseNode(ExecuteContext=(), Item=(Type=Bone), Mirror Plane=(X=1,Y=1,Z=0))
   Match 2 Bones IK onto FK = RigVMCollapseNode(Bones=(), ExecuteContext=(), IK Control=(Type=Control,Name="hand_l_ik_ctrl"), Pole Vector Control=(Type=Control,Name="arm_l_pv_ik_ctrl"))
   Project to New Parents = RigVMCollapseNode(Children=(), ExecuteContext=(), Parents=(), Propagate to Children=True)
  ]}
 RigVMFunction_MathBoolOr_1_ContainedGraph {nodes [
   n1 = RigVMFunctionEntryNode()
   RigVMFunction_MathBoolOr = &RigVMFunction_MathBoolOr(A=@n1.A, B=@n1.B)
   RigVMFunctionReturnNode(Result=&RigVMFunction_MathBoolOr(B=@n1.C, A=@RigVMFunction_MathBoolOr.Result).Result)
  ]}
 RigVMFunction_MathBoolOr_1_ContainedGraph {nodes [
   n1 = RigVMFunctionEntryNode()
   RigVMFunction_MathBoolOr = &RigVMFunction_MathBoolOr(A=@n1.A, B=@n1.B)
   RigVMFunctionReturnNode(Result=&RigVMFunction_MathBoolOr(B=@n1.C, A=@RigVMFunction_MathBoolOr.Result).Result)
  ]}
 RigVMFunction_Sequence_1_1_ContainedGraph {nodes [
   n1 = RigVMFunctionEntryNode()
   n2 = &RigVMFunction_Sequence(ExecuteContext=@n1.ExecuteContext)
   n3 = &RigVMFunction_Sequence(ExecuteContext=@n2.B)
   n4 = &RigVMFunction_Sequence(ExecuteContext=@n3.B)
   n5 = &RigVMFunction_Sequence(ExecuteContext=@n4.B)
   RigVMFunctionReturnNode(A=@n2.A, B=@n3.A, C=@n4.A, D=@n5.A, E=@n5.B)
  ]}
 RigVMFunction_Sequence_1_1_ContainedGraph {nodes [
   n1 = RigVMFunctionEntryNode()
   n2 = &RigVMFunction_Sequence(ExecuteContext=@n1.ExecuteContext)
   n3 = &RigVMFunction_Sequence(ExecuteContext=@n2.B)
   n4 = &RigVMFunction_Sequence(ExecuteContext=@n3.B)
   n5 = &RigVMFunction_Sequence(ExecuteContext=@n4.B)
   RigVMFunctionReturnNode(A=@n2.A, B=@n3.A, C=@n4.A, D=@n5.A, E=@n5.B)
  ]}
 RigVMFunction_Sequence_1_ContainedGraph {nodes [
   n1 = RigVMFunctionEntryNode()
   RigVMFunction_Sequence = &RigVMFunction_Sequence(ExecuteContext=@n1.ExecuteContext)
   n2 = &RigVMFunction_Sequence(ExecuteContext=@RigVMFunction_Sequence.B)
   n3 = &RigVMFunction_Sequence(ExecuteContext=@n2.B)
   n4 = &RigVMFunction_Sequence(ExecuteContext=@n3.B)
   n5 = &RigVMFunction_Sequence(ExecuteContext=@n4.B)
   n6 = &RigVMFunction_Sequence(ExecuteContext=@n5.B)
   RigVMFunctionReturnNode(A=@RigVMFunction_Sequence.A, B=@n2.A, C=@n3.A, D=@n4.A, E=@n5.A, F=@n6.A, G=@n6.B)
  ]}
 RigVMFunction_Sequence_1_ContainedGraph {nodes [
   n1 = RigVMFunctionEntryNode()
   RigVMFunction_Sequence = &RigVMFunction_Sequence(ExecuteContext=@n1.ExecuteContext)
   n2 = &RigVMFunction_Sequence(ExecuteContext=@RigVMFunction_Sequence.B)
   n3 = &RigVMFunction_Sequence(ExecuteContext=@n2.B)
   n4 = &RigVMFunction_Sequence(ExecuteContext=@n3.B)
   RigVMFunctionReturnNode(A=@RigVMFunction_Sequence.A, B=@n2.A, C=@n3.A, D=@n4.A, E=@n4.B)
  ]}
 RigVMFunction_Sequence_1_ContainedGraph {nodes [
   n1 = RigVMFunctionEntryNode()
   RigVMFunction_Sequence = &RigVMFunction_Sequence(ExecuteContext=@n1.ExecuteContext)
   n2 = &RigVMFunction_Sequence(ExecuteContext=@RigVMFunction_Sequence.B)
   n3 = &RigVMFunction_Sequence(ExecuteContext=@n2.B)
   n4 = &RigVMFunction_Sequence(ExecuteContext=@n3.B)
   n5 = &RigVMFunction_Sequence(ExecuteContext=@n4.B)
   RigVMFunctionReturnNode(A=@RigVMFunction_Sequence.A, B=@n2.A, C=@n3.A, D=@n4.A, E=@n5.A, F=@n5.B)
  ]}
 RigVMFunction_Sequence_2_1_ContainedGraph {nodes [
   n1 = RigVMFunctionEntryNode()
   n2 = &RigVMFunction_Sequence(ExecuteContext=@n1.ExecuteContext)
   n3 = &RigVMFunction_Sequence(ExecuteContext=@n2.B)
   n4 = &RigVMFunction_Sequence(ExecuteContext=@n3.B)
   n5 = &RigVMFunction_Sequence(ExecuteContext=@n4.B)
   n6 = &RigVMFunction_Sequence(ExecuteContext=@n5.B)
   n7 = &RigVMFunction_Sequence(ExecuteContext=@n6.B)
   RigVMFunctionReturnNode(A=@n2.A, B=@n3.A, C=@n4.A, D=@n5.A, E=@n6.A, F=@n7.A, G=@n7.B)
  ]}
 RigVMFunction_Sequence_2_1_ContainedGraph {nodes [
   n1 = RigVMFunctionEntryNode()
   n2 = &RigVMFunction_Sequence(ExecuteContext=@n1.ExecuteContext)
   n3 = &RigVMFunction_Sequence(ExecuteContext=@n2.B)
   n4 = &RigVMFunction_Sequence(ExecuteContext=@n3.B)
   RigVMFunctionReturnNode(A=@n2.A, B=@n3.A, C=@n4.A, D=@n4.B)
  ]}
 RigVMFunction_Sequence_2_1_ContainedGraph {nodes [
   n1 = RigVMFunctionEntryNode()
   n2 = &RigVMFunction_Sequence(ExecuteContext=@n1.ExecuteContext)
   n3 = &RigVMFunction_Sequence(ExecuteContext=@n2.B)
   n4 = &RigVMFunction_Sequence(ExecuteContext=@n3.B)
   RigVMFunctionReturnNode(A=@n2.A, B=@n3.A, C=@n4.A, D=@n4.B)
  ]}
 RigVMFunction_Sequence_2_ContainedGraph {nodes [
   n1 = RigVMFunctionEntryNode()
   RigVMFunction_Sequence = &RigVMFunction_Sequence(ExecuteContext=@n1.ExecuteContext)
   n2 = &RigVMFunction_Sequence(ExecuteContext=@RigVMFunction_Sequence.B)
   n3 = &RigVMFunction_Sequence(ExecuteContext=@n2.B)
   n4 = &RigVMFunction_Sequence(ExecuteContext=@n3.B)
   RigVMFunctionReturnNode(A=@RigVMFunction_Sequence.A, B=@n2.A, C=@n3.A, D=@n4.A, E=@n4.B)
  ]}
 RigVMFunction_Sequence_2_ContainedGraph {nodes [
   n1 = RigVMFunctionEntryNode()
   RigVMFunction_Sequence = &RigVMFunction_Sequence(ExecuteContext=@n1.ExecuteContext)
   n2 = &RigVMFunction_Sequence(ExecuteContext=@RigVMFunction_Sequence.B)
   n3 = &RigVMFunction_Sequence(ExecuteContext=@n2.B)
   n4 = &RigVMFunction_Sequence(ExecuteContext=@n3.B)
   RigVMFunctionReturnNode(A=@RigVMFunction_Sequence.A, B=@n2.A, C=@n3.A, D=@n4.A, E=@n4.B)
  ]}
 RigVMModel {nodes [
   &RigUnit_ParentConstraint(
    AdvancedSettings=(InterpolationType=Average,RotationOrderForFilter=XZY)
    ExecutePin=()
    Parents=((Item=(Type=Bone,Name="None"),Weight=1))
    Weight=1
    Child=RigVMDispatchNode(Index=0, Array=RigVMVariableNode() {variable=LockingBones}.Value) {template="DISPATCH_RigVMDispatch_ArrayGetAtIndex(in Array,in Index,out Element)"}.Element
    Parents.0.Item=RigVMVariableNode() {variable=IKControl}.Value
   )
   > &RigUnit_AimConstraintLocalSpaceOffset(
    Filter=(bX=True,bY=True,bZ=True)
    Parents=((Item=(Type=Bone,Name="None"),Weight=1))
    Weight=1
    WorldUp=(Target=(X=0,Y=0,Z=1),Kind=Direction,Space=(Type=None,Name="None"))
    Child=RigVMDispatchNode(Index=0, Array=RigVMVariableNode() {variable=LockingBones}.Value) {template="DISPATCH_RigVMDispatch_ArrayGetAtIndex(in Array,in Index,out Element)"}.Element
    Parents.0.Item=RigVMVariableNode() {variable=lock null}.Value
   )
   n1 = RigVMDispatchNode(Index=1, Array=RigVMVariableNode() {variable=LockingBones}.Value) {template="DISPATCH_RigVMDispatch_ArrayGetAtIndex(in Array,in Index,out Element)"}
   > &RigUnit_ParentConstraint(
    AdvancedSettings=(InterpolationType=Average,RotationOrderForFilter=XZY)
    Parents=((Item=(Type=Bone,Name="None"),Weight=1))
    Weight=1
    Parents.0.Item=@n1.Element
    Child=&RigUnit_ResolveConnector(Connector=(Type=Connector,Name="Thigh"), SkipSocket=true).Result
   )
   &RigUnit_ParentConstraint(
    AdvancedSettings=(InterpolationType=Average,RotationOrderForFilter=XZY)
    ExecutePin=()
    Parents=((Item=(Type=Bone,Name="None"),Weight=1))
    Weight=1
    Parents.0.Item=@n1.Element
    Child=RigVMDispatchNode(Index=0, Array=RigVMVariableNode() {variable=PhantomRigBones}.Value) {template="DISPATCH_RigVMDispatch_ArrayGetAtIndex(in Array,in Index,out Element)"}.Element
   )
   > &RigUnit_ParentConstraint(
    AdvancedSettings=(InterpolationType=Average,RotationOrderForFilter=XZY)
    Parents=((Item=(Type=Bone,Name="None"),Weight=1))
    Weight=1
    Parents.0.Item=@n1.Element
    Child=RigVMDispatchNode(Index=0, Array=RigVMVariableNode() {variable=ThighPhantomRigBones}.Value) {template="DISPATCH_RigVMDispatch_ArrayGetAtIndex(in Array,in Index,out Element)"}.Element
   )
   RigUnit_BeginExecution = &RigUnit_BeginExecution()
   RigVMFunction_Sequence_2 = RigVMAggregateNode(ExecuteContext=@RigUnit_BeginExecution.ExecutePin)
   FWD Get FK or IK_1_2 = RigVMFunctionReferenceNode(Connector=(Type=Connector,Name="Settings Control"), Name=fk_ik_switch)
   n2 = &RigVMFunction_ControlFlowBranch(Condition=@"FWD Get FK or IK_1_2".Value, ExecuteContext=@RigVMFunction_Sequence_2.A)
   &RigUnit_CCDIKItemArray(
    BaseRotationLimit=30
    MaxIterations=10
    Precision=1
    RotationLimits=()
    Weight=1
    bPropagateToChildren=True
    bStartFromTail=false
    EffectorTransform=&RigUnit_GetTransform(bInitial=False, Item=RigVMVariableNode() {variable=IKControl}.Value).Transform
    ExecutePin=@n2.True
    Items=RigVMVariableNode() {variable=PhantomRigBones}.Value
   )
   > &RigUnit_CCDIKItemArray(
    BaseRotationLimit=30
    MaxIterations=10
    Precision=1
    RotationLimits=()
    Weight=1
    bPropagateToChildren=True
    bStartFromTail=false
    EffectorTransform=&RigUnit_GetTransform(bInitial=False, Item=RigVMVariableNode() {variable=IKControl}.Value).Transform
    Items=RigVMVariableNode() {variable=ThighPhantomRigBones}.Value
   )
   FWD Get FK or IK_1 = RigVMFunctionReferenceNode(Connector=(Type=Connector,Name="Settings Control"), Name=fk_ik_switch)
   n3 = &RigVMFunction_ControlFlowBranch(Condition=@"FWD Get FK or IK_1".Value, ExecuteContext=@RigVMFunction_Sequence_2.B)
   Get_Module_Metadata_3 = RigVMDispatchNode(Default=(Type=Bone,Name="None"), Name=Thigh Control, NameSpace=Self) {template="DISPATCH_RigDispatch_GetModuleMetadata(in Name,in NameSpace,in Default,out Value,out Found)"}
   Parent = RigVMFunctionReferenceNode(
    New Parent=RigVMDispatchNode(Index=0, Array=RigVMVariableNode() {variable=ThighPhantomRigBones}.Value) {template="DISPATCH_RigVMDispatch_ArrayGetAtIndex(in Array,in Index,out Element)"}.Element
    Child=&RigUnit_HierarchyGetParent(bDefaultParent=True, Child=@Get_Module_Metadata_3.Value).Parent
    ExecuteContext=@n3.True
   )
   n4 = RigVMDispatchNode(Index=0, Array=RigVMVariableNode() {variable=PhantomRigBones}.Value) {template="DISPATCH_RigVMDispatch_ArrayGetAtIndex(in Array,in Index,out Element)"}
   n5 = &RigUnit_ResolveConnector(Connector=(Type=Connector,Name="HindLeg Primary"), SkipSocket=true)
   n6 = RigVMVariableNode() {variable=Secondary Axis}
   > &RigUnit_AimConstraintLocalSpaceOffset(
    Filter=(bX=True,bY=True,bZ=True)
    Parents=((Item=(Type=Bone,Name="None"),Weight=1))
    Weight=1
    WorldUp=(Target=(X=0,Y=0,Z=1),Kind=Location,Space=(Type=None,Name="None"))
    Child=@n4.Element
    Parents.0.Item=@Get_Module_Metadata_3.Value
    WorldUp.Space=@n5.Result
    AimAxis=RigVMVariableNode() {variable=Primary Axis}.Value
    UpAxis=@n6.Value
   )
   n7 = RigVMVariableNode() {variable=Debug}
   > DrawAxis = RigVMFunctionReferenceNode(Scale=15, Thickness=2, Item=@n4.Element, Active=@n7.Value)
   Multiply = &RigVMFunction_MathVectorMul(B=(X=25,Y=25,Z=25), A=@n6.Value) {template="Multiply::Execute(in A,in B,out Result)"}
   RigVMFunction_MathTransformTransformVector = &RigVMFunction_MathTransformTransformVector(Transform=&RigUnit_GetTransform(bInitial=false, Item=@n5.Result).Transform, Location=@Multiply.Result)
   > RigVMFunction_DebugTransformMutableNoSpace = &RigVMFunction_DebugTransformMutableNoSpace(
    Color=(R=1,G=0,B=0,A=1)
    DebugDrawSettings=(DepthPriority=SDPG_Foreground,Lifetime=-1)
    Mode=Axes
    Scale=15
    Thickness=2
    Transform=(Rotation=(X=0,Y=0,Z=0,W=1),Translation=(X=0,Y=0,Z=0),Scale3D=(X=1,Y=1,Z=1))
    WorldOffset=(Rotation=(X=0,Y=0,Z=0,W=1),Translation=(X=0,Y=0,Z=0),Scale3D=(X=1,Y=1,Z=1))
    Transform.Translation=@RigVMFunction_MathTransformTransformVector.Result
    bEnabled=@n7.Value
   )
   > n8 = &RigUnit_PositionConstraintLocalSpaceOffset(
    Filter=(bX=True,bY=True,bZ=True)
    Parents=((Item=(Type=Bone,Name="None"),Weight=1))
    Weight=1
    bMaintainOffset=true
    Child=RigVMDispatchNode(Index=-1, Array=RigVMVariableNode() {variable=PhantomRigBones}.Value) {template="DISPATCH_RigVMDispatch_ArrayGetAtIndex(in Array,in Index,out Element)"}.Element
    Parents.0.Item=RigVMVariableNode() {variable=IKControl}.Value
   )
   GetAnimationChannel = &RigUnit_GetBoolAnimationChannel(Channel=Stretch, bInitial=False, Control=RigVMVariableNode() {variable=IKControl}.Value.Name) {template="GetAnimationChannel::Execute(out Value,in Control,in Channel,in bInitial)"}
   RigVMFunction_ControlFlowBranch = &RigVMFunction_ControlFlowBranch(Condition=@GetAnimationChannel.Value, ExecuteContext=@n8.ExecutePin)
   &RigUnit_PositionConstraintLocalSpaceOffset(
    Filter=(bX=True,bY=True,bZ=True)
    Parents=((Item=(Type=Bone,Name="None"),Weight=1))
    Weight=1
    bMaintainOffset=true
    Child=RigVMDispatchNode(Index=1, Array=RigVMVariableNode() {variable=PhantomRigBones}.Value) {template="DISPATCH_RigVMDispatch_ArrayGetAtIndex(in Array,in Index,out Element)"}.Element
    ExecutePin=@RigVMFunction_ControlFlowBranch.True
    Parents.0.Item=RigVMVariableNode() {variable=ThighControl}.Value
   )
   Parent_1 = RigVMFunctionReferenceNode(
    New Parent=RigVMDispatchNode(Index=0, Array=RigVMVariableNode() {variable=PhantomRigBones}.Value) {template="DISPATCH_RigVMDispatch_ArrayGetAtIndex(in Array,in Index,out Element)"}.Element
    Child=RigVMDispatchNode(Index=0, Array=RigVMVariableNode() {variable=AllBones}.Value) {template="DISPATCH_RigVMDispatch_ArrayGetAtIndex(in Array,in Index,out Element)"}.Element
    ExecuteContext=@RigVMFunction_Sequence_2.C
   )
   FWD Get FK or IK_1_1 = RigVMFunctionReferenceNode(Connector=(Type=Connector,Name="Settings Control"), Name=fk_ik_switch)
   n9 = &RigVMFunction_ControlFlowBranch(Condition=@"FWD Get FK or IK_1_1".Value, ExecuteContext=@RigVMFunction_Sequence_2.D)
   Parent_1_1_2 = RigVMFunctionReferenceNode(
    Child=RigVMDispatchNode(Index=0, Array=RigVMVariableNode() {variable=RigBones}.Value) {template="DISPATCH_RigVMDispatch_ArrayGetAtIndex(in Array,in Index,out Element)"}.Element
    New Parent=RigVMDispatchNode(Index=1, Array=RigVMVariableNode() {variable=PhantomRigBones}.Value) {template="DISPATCH_RigVMDispatch_ArrayGetAtIndex(in Array,in Index,out Element)"}.Element
    ExecuteContext=@n9.True
   )
   Get_Module_Metadata_2_3 = RigVMDispatchNode(Default=(), Name=Main Controls, NameSpace=Self) {template="DISPATCH_RigDispatch_GetModuleMetadata(in Name,in NameSpace,in Default,out Value,out Found)"}
   Clone = RigVMDispatchNode(Array=@Get_Module_Metadata_2_3.Value) {template="DISPATCH_RigVMDispatch_ArrayClone(in Array,out Clone)"}
   Get_Module_Metadata = RigVMDispatchNode(Default=(), Name=IK Controls, NameSpace=Self) {template="DISPATCH_RigDispatch_GetModuleMetadata(in Name,in NameSpace,in Default,out Value,out Found)"}
   n10 = RigVMDispatchNode(Array=@Clone.Clone, Other=@Get_Module_Metadata.Value, ExecuteContext=@RigVMFunction_Sequence_2.G) {template="DISPATCH_RigVMDispatch_ArrayAppend(io Array,in Other)"}
   Get_Module_Metadata_4_1 = RigVMDispatchNode(Default=(Type=Bone,Name="None"), Name=Thigh Control, NameSpace=Self) {template="DISPATCH_RigDispatch_GetModuleMetadata(in Name,in NameSpace,in Default,out Value,out Found)"}
   > n11 = RigVMDispatchNode(Other=((Type=None,Name="None")), Array=@n10.Array, Other.0=@Get_Module_Metadata_4_1.Value) {template="DISPATCH_RigVMDispatch_ArrayAppend(io Array,in Other)"}
   Get_Module_Metadata_4 = RigVMDispatchNode(Default=(Type=Bone,Name="None"), Name=Thigh FK Control, NameSpace=Self) {template="DISPATCH_RigDispatch_GetModuleMetadata(in Name,in NameSpace,in Default,out Value,out Found)"}
   > FWD FK or IK Control Vis = RigVMFunctionReferenceNode(FK Controls=((Type=None,Name="None")), IK Controls=@n11.Array, FK Controls.0=@Get_Module_Metadata_4.Value)
   FWD Get FK or IK = RigVMFunctionReferenceNode(Connector=(Type=Connector,Name="Settings Control"), Name=fk_ik_switch)
   FWD Get FK or IK_2 = RigVMFunctionReferenceNode(Connector=(Type=Connector,Name="Settings Control"), Name=fk_ik_switch)
   > FWD TwoBoneLimb Quad = RigVMFunctionReferenceNode(
    Attach Skeleton=true
    Debug=True
    FK Arm Rotation Space=(Type=Bone,Name="root")
    Keep Aligned=false
    Mirror Threshold=1
    PV Follow Aim Axis=(X=0,Y=1,Z=0)
    PV Follow Aim Direction=(X=0,Y=1,Z=0)
    PV Follow Twist Lock=(bX=false,bY=false,bZ=true)
    PV Follow Up Axis=(X=1,Y=0,Z=0)
    Softness=1
    local=true
    FK_or_IK=@"FWD Get FK or IK".Value
    FK Space=@"FWD Get FK or IK_2".Value
    Driven Bones=&RigUnit_FindItemsWithMetadataTag(NameSpace=Self, Tag=Bones).Items
    FK Controls=&RigUnit_FindItemsWithMetadataTag(NameSpace=Self, Tag=Controls FK).Items
    IK Controls=&RigUnit_FindItemsWithMetadataTag(NameSpace=Self, Tag=Controls IK).Items
    On Right Side=RigVMVariableNode() {variable=RightSide}.Value
    End Ctrl=RigVMVariableNode() {variable=IKControl}.Value
    Bones=RigVMVariableNode() {variable=RigBones}.Value
    Auto Determine Axis=RigVMVariableNode() {variable=Auto Determine Axis}.Value
    Enable PV Follow=RigVMVariableNode() {variable=Enable PV Follow - version1}.Value
    Enable Softness=RigVMVariableNode() {variable=Enable Softness}.Value
    Enable Stretch=RigVMVariableNode() {variable=Enable Stretch}.Value
    Secondary Axis=RigVMVariableNode() {variable=Secondary Axis}.Value
    Primary Axis=RigVMVariableNode() {variable=Primary Axis}.Value
   )
   n12 = RigVMVariableNode() {variable=Enable PV Follow - version2}
   n13 = &RigVMFunction_ControlFlowBranch(ExecuteContext=@"FWD TwoBoneLimb Quad".ExecuteContext, Condition=@n12.Value)
   &RigUnit_RotationConstraintLocalSpaceOffset(
    AdvancedSettings=(InterpolationType=Average,RotationOrderForFilter=ZXY)
    Filter=(bX=false,bY=false,bZ=True)
    Parents=((Item=(Type=Bone,Name="None"),Weight=1))
    Weight=1
    ExecutePin=@n13.True
    Child=RigVMVariableNode() {variable=PVControlRotateNull}.Value
    Parents.0.Item=RigVMVariableNode() {variable=IKControl}.Value
   )
   Get_Metadata = RigVMDispatchNode(
    Default=(Type=Bone,Name="None")
    Name=Null
    NameSpace=Self
    Item=RigVMDispatchNode(Index=1, Array=&RigUnit_FindItemsWithMetadataTag(NameSpace=Self, Tag=Controls IK).Items) {template="DISPATCH_RigVMDispatch_ArrayGetAtIndex(in Array,in Index,out Element)"}.Element
   ) {template="DISPATCH_RigDispatch_GetMetadata(in Item,in Name,in NameSpace,in Default,out Value,out Found)"}
   > Parent_2 = RigVMFunctionReferenceNode(
    New Parent=RigVMDispatchNode(Index=1, Array=RigVMVariableNode() {variable=ThighPhantomRigBones}.Value) {template="DISPATCH_RigVMDispatch_ArrayGetAtIndex(in Array,in Index,out Element)"}.Element
    Child=@Get_Metadata.Value
   )
   RigVMFunction_MathBoolNot = &RigVMFunction_MathBoolNot(Value=RigVMVariableNode() {variable=Enable PV Follow - version1}.Value)
   RigVMFunction_MathBoolAnd = &RigVMFunction_MathBoolAnd(B=@RigVMFunction_MathBoolNot.Result, A=@n12.Value)
  ]}
 RigVMModel Backwards Solve Graph {nodes [
   n1 = /Script/ControlRig.RigUnit_InverseExecution() {is=struct RigVMUnitNode}
   Get_Module_Metadata = RigVMDispatchNode(Default=(), Name=Bones, NameSpace=Self) {template="DISPATCH_RigDispatch_GetModuleMetadata(in Name,in NameSpace,in Default,out Value,out Found)"}
   Get_Module_Metadata_3_1 = RigVMDispatchNode(Default=(Type=Bone,Name="None"), Name=IK Control, NameSpace=Self) {template="DISPATCH_RigDispatch_GetModuleMetadata(in Name,in NameSpace,in Default,out Value,out Found)"}
   Parent_0 = RigVMFunctionReferenceNode(
    New Parent=RigVMDispatchNode(Index=-1, Array=@Get_Module_Metadata.Value) {template="DISPATCH_RigVMDispatch_ArrayGetAtIndex(in Array,in Index,out Element)"}.Element
    Child=@Get_Module_Metadata_3_1.Value
    ExecuteContext=@n1.ExecutePin
   )
   Get_Module_Metadata_1 = RigVMDispatchNode(Default=(Type=Bone,Name="None"), Name=PV Control, NameSpace=Self) {template="DISPATCH_RigDispatch_GetModuleMetadata(in Name,in NameSpace,in Default,out Value,out Found)"}
   > Parent = RigVMFunctionReferenceNode(
    New Parent=RigVMDispatchNode(Index=2, Array=@Get_Module_Metadata.Value) {template="DISPATCH_RigVMDispatch_ArrayGetAtIndex(in Array,in Index,out Element)"}.Element
    Child=@Get_Module_Metadata_1.Value
   )
   Get_Module_Metadata_3 = RigVMDispatchNode(Default=(Type=Bone,Name="None"), Name=Thigh Control, NameSpace=Self) {template="DISPATCH_RigDispatch_GetModuleMetadata(in Name,in NameSpace,in Default,out Value,out Found)"}
   Parent_0_1 = RigVMFunctionReferenceNode(
    ExecuteContext=()
    New Parent=RigVMDispatchNode(Index=0, Array=@Get_Module_Metadata.Value) {template="DISPATCH_RigVMDispatch_ArrayGetAtIndex(in Array,in Index,out Element)"}.Element
    Child=@Get_Module_Metadata_3.Value
   )
  ]}
 RigVMModel Connector Event Graph {nodes [
   n1 = /Script/ControlRig.RigUnit_ConnectorExecution() {is=struct RigVMUnitNode}
   RigVMFunction_Sequence = RigVMAggregateNode(ExecuteContext=@n1.ExecuteContext)
   Connect to Module Metadata = RigVMFunctionReferenceNode(
    Connector=(Type=Connector,Name="Parent Control")
    Module Metadata Name=Parent Control
    Module Metadata NameSpace=Root
    ExecuteContext=@RigVMFunction_Sequence.A
   )
   > Connect to Module Metadata_1 = RigVMFunctionReferenceNode(Connector=(Type=Connector,Name="Settings Control"), Module Metadata Name=Settings Control, Module Metadata NameSpace=Root)
   > Connect to Module Metadata_2 = RigVMFunctionReferenceNode(Connector=(Type=Connector,Name="Pole Vector Parent"), Module Metadata Name=Parent Control, Module Metadata NameSpace=Root)
   n2 = /Script/ControlRig.RigUnit_GetCandidates() {is=struct RigVMUnitNode}
   n3 = &RigVMFunction_ControlFlowBranch(
    Condition=RigVMDispatchNode(B=(Type=Connector,Name="Foot Bone"), A=@n2.Connector) {template="DISPATCH_RigVMDispatch_CoreEquals(in A,in B,out Result)"}.Result
    ExecuteContext=@RigVMFunction_Sequence.B
   )
   RigVMDispatchNode(ExecuteContext=@n3.True, Array=@n2.Candidates) {template="DISPATCH_RigVMDispatch_ArrayIterator(in Array,out Element,out Index,out Count,out Ratio)"}
   n4 = &RigUnit_ResolveConnector(Connector=(Type=Connector,Name="HindLeg Primary"), SkipSocket=true)
   > /Script/ControlRig.RigUnit_SetDefaultMatch(Default=RigVMDispatchNode(
     Index=2
     Array=&RigUnit_CollectionChildrenArray(TypeToSearch=Bone, bDefaultChildren=True, bIncludeParent=False, bRecursive=true, Parent=@n4.Result).Items
    ) {template="DISPATCH_RigVMDispatch_ArrayGetAtIndex(in Array,in Index,out Element)"}.Element) {is=struct RigVMUnitNode, template="Set Default Match::Execute(in Default)"}
   n5 = &RigVMFunction_ControlFlowBranch(
    Condition=RigVMDispatchNode(B=(Type=Connector,Name="Lower Knee"), A=&RigUnit_GetCandidates().Connector) {template="DISPATCH_RigVMDispatch_CoreEquals(in A,in B,out Result)"}.Result
    ExecuteContext=@RigVMFunction_Sequence.C
   )
   &RigUnit_SetDefaultMatch(
    Default=&RigUnit_HierarchyGetParent(bDefaultParent=True, Child=&RigUnit_ResolveConnector(Connector=(Type=Connector,Name="Foot Bone"), SkipSocket=true).Result).Parent
    ExecutePin=@n5.True
   ) {template="Set Default Match::Execute(in Default)"}
   n6 = &RigVMFunction_ControlFlowBranch(
    Condition=RigVMDispatchNode(B=(Type=Connector,Name="Upper Knee"), A=&RigUnit_GetCandidates().Connector) {template="DISPATCH_RigVMDispatch_CoreEquals(in A,in B,out Result)"}.Result
    ExecuteContext=@RigVMFunction_Sequence.D
   )
   n7 = &RigUnit_ResolveConnector(Connector=(Type=Connector,Name="HindLeg Primary"), SkipSocket=true)
   &RigUnit_SetDefaultMatch(
    Default=RigVMDispatchNode(
     Index=0
     Array=&RigUnit_CollectionChildrenArray(TypeToSearch=Bone, bDefaultChildren=True, bIncludeParent=False, bRecursive=False, Parent=@n7.Result).Items
    ) {template="DISPATCH_RigVMDispatch_ArrayGetAtIndex(in Array,in Index,out Element)"}.Element
    ExecutePin=@n6.True
   ) {template="Set Default Match::Execute(in Default)"}
   n8 = &RigVMFunction_ControlFlowBranch(
    Condition=RigVMDispatchNode(B=(Type=Connector,Name="Thigh"), A=&RigUnit_GetCandidates().Connector) {template="DISPATCH_RigVMDispatch_CoreEquals(in A,in B,out Result)"}.Result
    ExecuteContext=@RigVMFunction_Sequence.E
   )
   &RigUnit_SetDefaultMatch(
    Default=&RigUnit_ResolveConnector(Connector=(Type=Connector,Name="HindLeg Primary"), SkipSocket=true).Result
    ExecutePin=@n8.True
   ) {template="Set Default Match::Execute(in Default)"}
   n9 = &RigUnit_GetCandidates()
   n10 = &RigVMFunction_ControlFlowBranch(
    ExecuteContext=()
    Condition=RigVMDispatchNode(B=(Type=Connector,Name="Foot Bone"), A=@n9.Connector) {template="DISPATCH_RigVMDispatch_CoreEquals(in A,in B,out Result)"}.Result
   )
   RigVMDispatchNode(ExecuteContext=@n10.True, Array=@n9.Candidates) {template="DISPATCH_RigVMDispatch_ArrayIterator(in Array,out Element,out Index,out Count,out Ratio)"}
   > For_Each = RigVMDispatchNode(Array=&RigUnit_CollectionChildrenArray(
     TypeToSearch=Bone
     bDefaultChildren=True
     bIncludeParent=False
     bRecursive=true
     Parent=&RigUnit_ResolveConnector(Connector=(Type=Connector,Name="HindLeg Primary"), SkipSocket=true).Result
    ).Items) {template="DISPATCH_RigVMDispatch_ArrayIterator(in Array,out Element,out Index,out Count,out Ratio)"}
   Find = RigVMFunctionReferenceNode(Element=foot, Value=@For_Each.Element.Name)
   RigVMFunction_ControlFlowBranch = &RigVMFunction_ControlFlowBranch(Condition=@Find.Success, ExecuteContext=@For_Each.ExecuteContext)
   &RigUnit_SetDefaultMatch(Default=@For_Each.Element, ExecutePin=@RigVMFunction_ControlFlowBranch.True) {template="Set Default Match::Execute(in Default)"}
  ]}
]
hierarchy [
 HindLeg Primary {type=Connector, loc=<0 0 0>, rot=<0 0 0 1>}
 Root {type=Socket, loc=<0 0 0>, rot=<0 0 0 1>}
 Upper Knee {type=Connector, loc=<0 0 0>, rot=<0 0 0 1>}
 Foot Bone {type=Connector, loc=<0 0 0>, rot=<0 0 0 1>}
 Parent Control {type=Connector, loc=<0 0 0>, rot=<0 0 0 1>}
 Settings Control {type=Connector, loc=<0 0 0>, rot=<0 0 0 1>}
 Pole Vector Parent {type=Connector, loc=<0 0 0>, rot=<0 0 0 1>}
 Lower Knee {type=Connector, loc=<0 0 0>, rot=<0 0 0 1>}
 Thigh {type=Connector, loc=<0 0 0>, rot=<0 0 0 1>}
]
variables [
 AllBones {type="TArray<FRigElementKey>"}
 Auto Determine Axis {type=bool, default=True}
 Debug {type=bool, default=False}
 Effector {type=FRigElementKey, default=(Type=Bone,Name="")}
 Enable PV Follow - version1 {type=bool, default=False}
 Enable PV Follow - version2 {type=bool, default=True}
 Enable Softness {type=bool, default=False}
 Enable Stretch {type=bool, default=True}
 FK Control Shape {
  type=FRigUnit_HierarchyAddControlTransform_Settings
  default="(InitialSpace=LocalSpace,bUsePreferredRotationOrder=False,PreferredRotationOrder=YZX,Limits=(LimitTranslationX=(bMinimum=False,bMaximum=False),LimitTranslationY=(bMinimum=False,bMaximum=False),LimitTranslationZ=(bMinimum=False,bMaximum=False),LimitPitch=(b ...(truncated)"
 }
 FK IK Default {type=bool, default=True}
 IK Control Offset {type=FTransform, default=(Rotation=(X=0,Y=0,Z=0,W=1),Translation=(X=0,Y=0,Z=0),Scale3D=(X=1,Y=1,Z=1))}
 IK Foot Shape {
  type=FRigUnit_HierarchyAddControlTransform_Settings
  default="(InitialSpace=LocalSpace,bUsePreferredRotationOrder=False,PreferredRotationOrder=YZX,Limits=(LimitTranslationX=(bMinimum=False,bMaximum=False),LimitTranslationY=(bMinimum=False,bMaximum=False),LimitTranslationZ=(bMinimum=False,bMaximum=False),LimitPitch=(b ...(truncated)"
 }
 IK Thigh Shape {
  type=FRigUnit_HierarchyAddControlTransform_Settings
  default="(InitialSpace=LocalSpace,bUsePreferredRotationOrder=False,PreferredRotationOrder=ZYX,Limits=(LimitTranslationX=(bMinimum=False,bMaximum=False),LimitTranslationY=(bMinimum=False,bMaximum=False),LimitTranslationZ=(bMinimum=False,bMaximum=False),LimitPitch=(b ...(truncated)"
 }
 IKControl {type=FRigElementKey, default=(Type=Bone,Name="")}
 Knee Bone Offset {type=FVector, default=(X=0,Y=0,Z=0)}
 Knee Bone Offset Global Space {type=bool, default=False}
 LockingBones {type="TArray<FRigElementKey>"}
 Module Settings {
  type=FUserDefinedStruct_#2
  default="(OverrideColor_48_#3=False,LeftColor_3_#4=(R=0,G=0.0231,B=0.4115,A=1),RightColor_5_#5=(R=1,G=0,B=0,A=1),CenterColor_9_#6=(R=1,G=1,B=0,A= ...(truncated)"
 }
 ModulePrefixName {type=FName, default=None}
 Orient IK to World {type=bool, default=True}
 PVControlRotateNull {type=FRigElementKey, default=(Type=Bone,Name="")}
 PhantomRigBones {type="TArray<FRigElementKey>"}
 Pole Vector Position {type=float, default=5}
 Pole Vector Shape {
  type=FRigUnit_HierarchyAddControlTransform_Settings
  default="(InitialSpace=LocalSpace,bUsePreferredRotationOrder=False,PreferredRotationOrder=YZX,Limits=(LimitTranslationX=(bMinimum=False,bMaximum=False),LimitTranslationY=(bMinimum=False,bMaximum=False),LimitTranslationZ=(bMinimum=False,bMaximum=False),LimitPitch=(b ...(truncated)"
 }
 Primary Axis {type=FVector, default=(X=1,Y=0,Z=0)}
 RigBones {type="TArray<FRigElementKey>"}
 RightSide {type=bool, default=False}
 Secondary Axis {type=FVector, default=(X=0,Y=1,Z=0)}
 SideName {type=FName, default=None}
 ThighControl {type=FRigElementKey, default=(Type=None,Name="")}
 ThighFKControl {type=FRigElementKey, default=(Type=None,Name="")}
 ThighPhantomRigBones {type="TArray<FRigElementKey>"}
 lock null {type=FRigElementKey, default=(Type=Bone,Name="")}
]
