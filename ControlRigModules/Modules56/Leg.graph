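# Modular Control Rig leg module that builds an FK/IK two-bone limb from thigh, knee and foot connectors.
# Construction reads root and parent metadata for control spaces, knee offset and pole-vector distance, which otherwise derives from leg length.
# It then adds an IK effector null, a foot-module socket on a toe or foot bone, and a twist rig.
# The effector null and foot control are published as module metadata.
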ControlRigBlueprint /ControlRigModules/Modules56/Leg
graphs [
 CollapseNode_1_ContainedGraph {nodes [
   RigVMFunctionEntryNode()
   Add = /Script/RigVM.RigVMFunction_MathVectorAdd(A=RigVMVariableNode() {variable=Knee Bone Offset Metadata}.Value, B=RigVMVariableNode() {variable=Knee Bone Offset}.Value) {is=struct RigVMUnitNode, template="Add::Execute(in A,in B,out Result)"}
   n1 = RigVMVariableNode() {variable=Knee Bone Offset Global Space}
   n2 = RigVMVariableNode() {variable=knee Bone Offset GlobalSpace Metadata}
   n3 = RigVMVariableNode() {variable=Body Ctrl}
   n4 = RigVMVariableNode() {variable=Hips Ctrl}
   > CONSTR TwoBoneLimb Leg = RigVMFunctionReferenceNode(
    Basename=leg
    Enable Keep Aligned=false
    End Bone Connector=(Type=Connector,Name="Foot Bone")
    IK Control Basename=foot
    IK Control Spaces=((Type=None,Name="None"),(Type=None,Name="None"))
    Local Default=false
    Middle Bone Connector=(Type=Connector,Name="Knee Bone")
    Offset PV Control=(Rotation=(X=0,Y=0,Z=0,W=1),Translation=(X=0,Y=0,Z=0),Scale3D=(X=1,Y=1,Z=1))
    PV Control Spaces=((Type=None,Name="None"),(Type=None,Name="None"))
    Parent Control Connector=(Type=Connector,Name="Parent Control")
    Pole Vector Parent Connector=(Type=Connector,Name="Pole Vector Parent")
    Primary Axis=(X=1,Y=0,Z=0)
    Secondary Axis=(X=0,Y=-1,Z=0)
    Settings Control Connector=(Type=Connector,Name="Settings Control")
    Start Bone Connector=(Type=Connector,Name="Thigh Bone")
    Wrist Control Settings="(DisplayName=\"None\",InitialSpace=LocalSpace,bUsePreferredRotationOrder=False,PreferredRotationOrder=YZX,Limits=(LimitTranslationX=(bMinimum=False,bMaximum=False),LimitTranslationY=(bMinimum=False,bMaximum=False),LimitTranslationZ=(bMinimum=False,bMaximum=F ...(truncated)"
    Offset Elbow Bone=@Add.Result
    PV Offset Factor=RigVMCollapseNode(Mult Value=RigVMVariableNode() {variable=Pole Vector Position}.Value).Result
    Offset Elbow Bone Space=RigVMDispatchNode(True=GlobalSpace, Condition=&RigVMFunction_MathBoolOr(B=@n1.Value, A=@n2.Value).Result) {template="DISPATCH_RigVMDispatch_If(in Condition,in True,in False,out Result)"}.Result
    FK IK Default=RigVMVariableNode() {variable=FK IK Default}.Value
    Enable Softness=RigVMVariableNode() {variable=Enable Softness}.Value
    Enable PV Follow=RigVMVariableNode() {variable=Enable PV Follow}.Value
    Orient to World=RigVMVariableNode() {variable=Orient IK to World}.Value
    Module Settings=RigVMVariableNode() {variable=Module Settings}.Value
    IK Control Settings=RigVMVariableNode() {variable=IK Shape Settings}.Value
    PV Control Settings=RigVMVariableNode() {variable=PV Shape Settings}.Value
    FK Control Settings=RigVMVariableNode() {variable=FK Shape Settings}.Value
    Offset IK Ctrl Transform=RigVMVariableNode() {variable=IK Control Offset}.Value
    IK Control Spaces.0=@n3.Value
    PV Control Spaces.0=@n3.Value
    IK Control Spaces.1=@n4.Value
    PV Control Spaces.1=@n4.Value
   )
   > RigVMVariableNode(Value=@"CONSTR TwoBoneLimb Leg".Current Side Name) {variable="Current Side "}
   > RigVMFunctionReturnNode(Template Bones=@"CONSTR TwoBoneLimb Leg".Template Bones)
  ]}
 CollapseNode_ContainedGraph {nodes [
   n1 = /Script/ControlRig.RigUnit_ResolveConnector(Connector=(Type=Connector,Name="Foot Bone"), SkipSocket=true) {is=struct RigVMUnitNode}
   n2 = &RigUnit_ResolveConnector(Connector=(Type=Connector,Name="Thigh Bone"), SkipSocket=true)
   n3 = &RigVMFunction_MathDoubleDiv(
    B=3
    A=&RigVMFunction_MathVectorDistance(
     B=&RigUnit_GetTransform(Space=LocalSpace, bInitial=true, Item=@n1.Result).Transform.Translation
     A=&RigUnit_GetTransform(Space=LocalSpace, bInitial=true, Item=@n2.Result).Transform.Translation
    ).Result
   )
   n4 = RigVMVariableNode() {variable=Leg PV Position}
   n5 = RigVMVariableNode() {variable=Leg PV Position Metadata Found}
   RigVMFunctionReturnNode(Result=&RigVMFunction_MathDoubleMul(
     A=RigVMFunctionEntryNode().Mult Value
     B=RigVMDispatchNode(False=@n3.Result, True=@n4.Value, Condition=@n5.Value) {template="DISPATCH_RigVMDispatch_If(in Condition,in True,in False,out Result)"}.Result
    ).Result)
  ]}
 CollapseNode_ContainedGraph {nodes [
   n1 = RigVMFunctionEntryNode()
   Format Name = RigVMCollapseNode(name=foot, prefix=None, suffix=_ik_effector, side=RigVMVariableNode() {variable="Current Side "}.Value)
   n2 = &RigUnit_HierarchyAddNull(
    Space=LocalSpace
    Transform=(Rotation=(X=0,Y=0,Z=0,W=1),Translation=(X=0,Y=0,Z=0),Scale3D=(X=1,Y=1,Z=1))
    Parent=RigVMDispatchNode(Index=0, Array=/Script/ControlRig.RigUnit_FindItemsWithMetadataTag(NameSpace=Self, Tag=Controls IK) {is=struct RigVMUnitNode}.Items) {template="DISPATCH_RigVMDispatch_ArrayGetAtIndex(in Array,in Index,out Element)"}.Element
    ExecutePin=@n1.ExecutePin
    Name=@"Format Name".Result
   )
   > n3 = RigVMVariableNode(Value=@n2.Item) {variable=Effector Null}
   RigVMFunctionReturnNode(ExecutePin=@n3.ExecuteContext)
  ]}
 CollapseNode_ContainedGraph {nodes [
   RigVMFunctionEntryNode()
   > RigVMDispatchNode(Name=Effector Null, NameSpace=Self, Value=RigVMVariableNode() {variable=Effector Null}.Value) {template="DISPATCH_RigDispatch_SetModuleMetadata(in Name,in NameSpace,in Value,out Success)"}
   > RigVMDispatchNode(
    Name=Foot Control
    NameSpace=Self
    Value=RigVMDispatchNode(Index=0, Array=&RigUnit_FindItemsWithMetadataTag(NameSpace=Self, Tag=Controls IK).Items) {template="DISPATCH_RigVMDispatch_ArrayGetAtIndex(in Array,in Index,out Element)"}.Element
   ) {template="DISPATCH_RigDispatch_SetModuleMetadata(in Name,in NameSpace,in Value,out Success)"}
   > RigVMDispatchNode(
    Name=Parent Control
    NameSpace=Self
    Value=RigVMDispatchNode(Index=0, Array=&RigUnit_FindItemsWithMetadataTag(NameSpace=Self, Tag=Controls IK).Items) {template="DISPATCH_RigVMDispatch_ArrayGetAtIndex(in Array,in Index,out Element)"}.Element
   ) {template="DISPATCH_RigDispatch_SetModuleMetadata(in Name,in NameSpace,in Value,out Success)"}
   > RigVMFunctionReturnNode()
  ]}
 CollapseNode_ContainedGraph {nodes [
   RigVMFunctionEntryNode()
   Get_Module_Metadata_3_3 = RigVMDispatchNode(Default=(Type=Bone,Name="None"), Name=toe_l, NameSpace=Root) {template="DISPATCH_RigDispatch_GetModuleMetadata(in Name,in NameSpace,in Default,out Value,out Found)"}
   > RigVMVariableNode(Value=@Get_Module_Metadata_3_3.Found) {variable=Toe L Metadata Found}
   > RigVMVariableNode(Value=@Get_Module_Metadata_3_3.Value) {variable=Toe L}
   Get_Module_Metadata_3_2_2 = RigVMDispatchNode(Default=(Type=Bone,Name="None"), Name=toe_r, NameSpace=Root) {template="DISPATCH_RigDispatch_GetModuleMetadata(in Name,in NameSpace,in Default,out Value,out Found)"}
   > RigVMVariableNode(Value=@Get_Module_Metadata_3_2_2.Value) {variable=Toe R}
   > RigVMVariableNode(Value=@Get_Module_Metadata_3_2_2.Found) {variable=Toe R Metadata Found}
   Get_Module_Metadata_3 = RigVMDispatchNode(Default=(Type=Bone,Name="None"), Name=foot_l, NameSpace=Root) {template="DISPATCH_RigDispatch_GetModuleMetadata(in Name,in NameSpace,in Default,out Value,out Found)"}
   > RigVMVariableNode(Value=@Get_Module_Metadata_3.Value) {variable=Foot L}
   > RigVMVariableNode(Value=@Get_Module_Metadata_3.Found) {variable=Foot L Metadata Found}
   Get_Module_Metadata_3_1 = RigVMDispatchNode(Default=(Type=Bone,Name="None"), Name=foot_r, NameSpace=Root) {template="DISPATCH_RigDispatch_GetModuleMetadata(in Name,in NameSpace,in Default,out Value,out Found)"}
   > RigVMVariableNode(Value=@Get_Module_Metadata_3_1.Value) {variable=Foot R}
   > RigVMVariableNode(Value=@Get_Module_Metadata_3_1.Found) {variable=Foot R Metadata Found}
   Get_Module_Metadata_2_1 = RigVMDispatchNode(Default=50, Name=Leg PV Position, NameSpace=Root) {template="DISPATCH_RigDispatch_GetModuleMetadata(in Name,in NameSpace,in Default,out Value,out Found)"}
   > RigVMVariableNode(Value=@Get_Module_Metadata_2_1.Value) {variable=Leg PV Position}
   > RigVMVariableNode(Value=@Get_Module_Metadata_2_1.Found) {variable=Leg PV Position Metadata Found}
   Get_Module_Metadata = RigVMDispatchNode(Default=(Type=Control,Name="None"), Name=Body Control, NameSpace=Root) {template="DISPATCH_RigDispatch_GetModuleMetadata(in Name,in NameSpace,in Default,out Value,out Found)"}
   > RigVMVariableNode(Value=@Get_Module_Metadata.Value) {variable=Body Ctrl}
   Get_Module_Metadata_1 = RigVMDispatchNode(Default=(Type=Control,Name="None"), Name=Hips Control, NameSpace=Parent) {template="DISPATCH_RigDispatch_GetModuleMetadata(in Name,in NameSpace,in Default,out Value,out Found)"}
   > RigVMVariableNode(Value=@Get_Module_Metadata_1.Value) {variable=Hips Ctrl}
   Get_Module_Metadata_4 = RigVMDispatchNode(Default=(X=0,Y=0,Z=0), Name=knee_bone_offset, NameSpace=Root) {template="DISPATCH_RigDispatch_GetModuleMetadata(in Name,in NameSpace,in Default,out Value,out Found)"}
   > RigVMVariableNode(Value=@Get_Module_Metadata_4.Value) {variable=Knee Bone Offset Metadata}
   Get_Module_Metadata_4_1 = RigVMDispatchNode(Default=False, Name=knee_bone_offset_globalSpace, NameSpace=Root) {template="DISPATCH_RigDispatch_GetModuleMetadata(in Name,in NameSpace,in Default,out Value,out Found)"}
   > RigVMVariableNode(Value=@Get_Module_Metadata_4_1.Value) {variable=knee Bone Offset GlobalSpace Metadata}
   > RigVMFunctionReturnNode()
  ]}
 CollapseNode_ContainedGraph {nodes [
   n1 = RigVMFunctionEntryNode()
   n2 = &RigVMFunction_ControlFlowBranch(
    ExecuteContext=@n1.ExecuteContext_1
    Condition=&RigVMFunction_MathBoolOr(A=RigVMVariableNode() {variable=Toe L Metadata Found}.Value, B=RigVMVariableNode() {variable=Toe R Metadata Found}.Value).Result
   )
   n3 = /Script/ControlRig.RigUnit_CollectionChildrenArray(
    TypeToSearch=Bone
    bDefaultChildren=True
    bIncludeParent=False
    bRecursive=False
    Parent=&RigUnit_ResolveConnector(Connector=(Type=Connector,Name="Foot Bone"), SkipSocket=true).Result
   ) {is=struct RigVMUnitNode}
   RigVMFunction_ControlFlowBranch = &RigVMFunction_ControlFlowBranch(
    Condition=RigVMDispatchNode(Array=@n3.Items, Element=RigVMVariableNode() {variable=Foot L}.Value) {template="DISPATCH_RigVMDispatch_ArrayFind(in Array,in Element,out Index,out Success)"}.Success
    ExecuteContext=@n2.False
   )
   n4 = &RigVMFunction_ControlFlowBranch(
    Condition=RigVMDispatchNode(Array=@n3.Items, Element=RigVMVariableNode() {variable=Foot R}.Value) {template="DISPATCH_RigVMDispatch_ArrayFind(in Array,in Element,out Index,out Success)"}.Success
    ExecuteContext=@RigVMFunction_ControlFlowBranch.Completed
   )
   RigVMFunction_MathBoolOr = &RigVMFunction_MathBoolOr(A=RigVMVariableNode() {variable=Foot L Metadata Found}.Value, B=RigVMVariableNode() {variable=Foot R Metadata Found}.Value)
   RigVMFunction_MathBoolNot = &RigVMFunction_MathBoolNot(Value=@RigVMFunction_MathBoolOr.Result)
   n5 = &RigVMFunction_ControlFlowBranch(ExecuteContext=@n4.Completed, Condition=@RigVMFunction_MathBoolNot.Result)
   Format Name_1 = RigVMCollapseNode(name=foot, prefix=None, suffix=_socket, side=RigVMVariableNode() {variable="Current Side "}.Value)
   /Script/ControlRig.RigUnit_HierarchyAddSocket(
    Color=(R=1,G=1,B=1,A=1)
    Description="Place the foot module here . . . "
    Space=LocalSpace
    Transform=(Rotation=(X=0,Y=0,Z=0,W=1),Translation=(X=0,Y=0,Z=0),Scale3D=(X=1,Y=1,Z=1))
    Name=@"Format Name_1".Result
    ExecutePin=@n5.Completed
    Parent=RigVMVariableNode() {variable=Foot Socket Target}.Value
   ) {is=struct RigVMUnitNode}
   RigVMVariableNode(
    ExecuteContext=@n5.True
    Value=RigVMDispatchNode(Index=0, Array=@n3.Items) {template="DISPATCH_RigVMDispatch_ArrayGetAtIndex(in Array,in Index,out Element)"}.Element
   ) {variable=Foot Socket Target}
   RigVMVariableNode(ExecuteContext=@n4.True, Value=RigVMVariableNode() {variable=Foot R}.Value) {variable=Foot Socket Target}
   RigVMVariableNode(ExecuteContext=@RigVMFunction_ControlFlowBranch.True, Value=RigVMVariableNode() {variable=Foot L}.Value) {variable=Foot Socket Target}
   n6 = &RigUnit_ResolveConnector(Connector=(Type=Connector,Name="Leg Primary"), SkipSocket=true)
   Get Side from Metadata = RigVMFunctionReferenceNode(Item=&RigUnit_ResolveConnector(Connector=(Type=Connector,Name="Leg Primary"), SkipSocket=true).Result)
   n7 = RigVMDispatchNode(
    Condition=@"Get Side from Metadata".Result
    False=RigVMVariableNode() {variable=Toe L}.Value
    True=RigVMVariableNode() {variable=Toe R}.Value
   ) {template="DISPATCH_RigVMDispatch_If(in Condition,in True,in False,out Result)"}
   n8 = &RigVMFunction_ControlFlowBranch(
    Condition=RigVMDispatchNode(
     Array=&RigUnit_CollectionChildrenArray(TypeToSearch=Bone, bDefaultChildren=True, bIncludeParent=true, bRecursive=true, Parent=@n6.Result).Items
     Element=@n7.Result
    ) {template="DISPATCH_RigVMDispatch_ArrayFind(in Array,in Element,out Index,out Success)"}.Success
    ExecuteContext=@n2.True
   )
   &RigUnit_HierarchyAddSocket(
    Color=(R=1,G=1,B=1,A=1)
    Description="Place the foot module here . . . "
    Space=LocalSpace
    Transform=(Rotation=(X=0,Y=0,Z=0,W=1),Translation=(X=0,Y=0,Z=0),Scale3D=(X=1,Y=1,Z=1))
    Name=/Script/RigVM.RigVMFunction_NameConcat(B=_socket, A=@n7.Result.Name) {is=struct RigVMUnitNode, template="Concat::Execute(in A,in B,out Result)"}.Result
    Parent=@n7.Result
    ExecutePin=@n8.True
   )
   RigVMFunctionReturnNode(Execute=@n2.Completed)
  ]}
 CollapseNode_ContainedGraph {nodes [
   n1 = &RigUnit_ResolveConnector(Connector=(Type=Connector,Name="Foot Bone"), SkipSocket=true)
   n2 = &RigUnit_ResolveConnector(Connector=(Type=Connector,Name="Thigh Bone"), SkipSocket=true)
   n3 = &RigVMFunction_MathDoubleDiv(
    B=3
    A=&RigVMFunction_MathVectorDistance(
     B=&RigUnit_GetTransform(Space=LocalSpace, bInitial=true, Item=@n1.Result).Transform.Translation
     A=&RigUnit_GetTransform(Space=LocalSpace, bInitial=true, Item=@n2.Result).Transform.Translation
    ).Result
   )
   Get_Module_Metadata_2_1 = RigVMDispatchNode(Default=50, Name=Leg PV Position, NameSpace=Root) {template="DISPATCH_RigDispatch_GetModuleMetadata(in Name,in NameSpace,in Default,out Value,out Found)"}
   RigVMFunctionReturnNode(Result=&RigVMFunction_MathDoubleMul(
     A=RigVMFunctionEntryNode().Mult Value
     B=RigVMDispatchNode(False=@n3.Result, Condition=@Get_Module_Metadata_2_1.Found, True=@Get_Module_Metadata_2_1.Value) {template="DISPATCH_RigVMDispatch_If(in Condition,in True,in False,out Result)"}.Result
    ).Result)
  ]}
 Concat_3_1_ContainedGraph {nodes [
   n1 = RigVMFunctionEntryNode()
   RigVMFunctionReturnNode(Result=&RigVMFunction_NameConcat(A=&RigVMFunction_NameConcat(A=@n1.A, B=@n1.B) {template="Concat::Execute(in A,in B,out Result)"}.Result, B=@n1.C) {template="Concat::Execute(in A,in B,out Result)"}.Result)
  ]}
 Concat_4_1_ContainedGraph {nodes [
   n1 = RigVMFunctionEntryNode()
   n2 = &RigVMFunction_NameConcat(A=@n1.A, B=@n1.B) {template="Concat::Execute(in A,in B,out Result)"}
   RigVMFunctionReturnNode(Result=&RigVMFunction_NameConcat(A=&RigVMFunction_NameConcat(A=@n2.Result, B=@n1.C) {template="Concat::Execute(in A,in B,out Result)"}.Result, B=@n1.D) {template="Concat::Execute(in A,in B,out Result)"}.Result)
  ]}
 Concat_4_1_ContainedGraph {nodes [
   n1 = RigVMFunctionEntryNode()
   n2 = &RigVMFunction_NameConcat(A=@n1.A, B=@n1.B) {template="Concat::Execute(in A,in B,out Result)"}
   RigVMFunctionReturnNode(Result=&RigVMFunction_NameConcat(A=&RigVMFunction_NameConcat(A=@n2.Result, B=@n1.C) {template="Concat::Execute(in A,in B,out Result)"}.Result, B=@n1.D) {template="Concat::Execute(in A,in B,out Result)"}.Result)
  ]}
 Concat_4_1_ContainedGraph {nodes [
   n1 = RigVMFunctionEntryNode()
   n2 = &RigVMFunction_NameConcat(A=@n1.A, B=@n1.B) {template="Concat::Execute(in A,in B,out Result)"}
   RigVMFunctionReturnNode(Result=&RigVMFunction_NameConcat(A=&RigVMFunction_NameConcat(A=@n2.Result, B=@n1.C) {template="Concat::Execute(in A,in B,out Result)"}.Result, B=@n1.D) {template="Concat::Execute(in A,in B,out Result)"}.Result)
  ]}
 Concat_4_1_ContainedGraph {nodes [
   n1 = RigVMFunctionEntryNode()
   n2 = &RigVMFunction_NameConcat(A=@n1.A, B=@n1.B) {template="Concat::Execute(in A,in B,out Result)"}
   RigVMFunctionReturnNode(Result=&RigVMFunction_NameConcat(A=&RigVMFunction_NameConcat(A=@n2.Result, B=@n1.C) {template="Concat::Execute(in A,in B,out Result)"}.Result, B=@n1.D) {template="Concat::Execute(in A,in B,out Result)"}.Result)
  ]}
 Concat_4_1_ContainedGraph {nodes [
   n1 = RigVMFunctionEntryNode()
   n2 = &RigVMFunction_NameConcat(A=@n1.A, B=@n1.B) {template="Concat::Execute(in A,in B,out Result)"}
   RigVMFunctionReturnNode(Result=&RigVMFunction_NameConcat(A=&RigVMFunction_NameConcat(A=@n2.Result, B=@n1.C) {template="Concat::Execute(in A,in B,out Result)"}.Result, B=@n1.D) {template="Concat::Execute(in A,in B,out Result)"}.Result)
  ]}
 Concat_5_1_1_ContainedGraph {nodes [
   n1 = RigVMFunctionEntryNode()
   RigVMFunctionReturnNode(Result=&RigVMFunction_NameConcat(A=&RigVMFunction_NameConcat(A=@n1.A, B=@n1.B) {template="Concat::Execute(in A,in B,out Result)"}.Result, B=@n1.C) {template="Concat::Execute(in A,in B,out Result)"}.Result)
  ]}
 ConstructionGraph {nodes [
   n1 = &RigUnit_PrepareForExecution()
   RigVMFunction_Sequence_1 = RigVMAggregateNode(ExecuteContext=@n1.ExecutePin)
   Get Incoming Metadata = RigVMCollapseNode(ExecuteContext=@RigVMFunction_Sequence_1.A)
   Create TwoBoneLimb from Arm = RigVMCollapseNode(ExecuteContext=@RigVMFunction_Sequence_1.B)
   Spawn Effector Null = RigVMCollapseNode(ExecutePin=@"Create TwoBoneLimb from Arm".ExecuteContext)
   Spawn Socket = RigVMCollapseNode(ExecuteContext_1=@"Spawn Effector Null".ExecutePin)
   CONSTR Basic Twist Rig = RigVMFunctionReferenceNode(ExecuteContext=@"Spawn Socket".Execute, Template Bones=@"Create TwoBoneLimb from Arm".Template Bones)
   Set Outgoing Metadata = RigVMCollapseNode(ExecuteContext=@RigVMFunction_Sequence_1.C)
   RigVMFunction_Sequence = &RigVMFunction_Sequence(ExecuteContext=())
  ]}
 ContainedGraph {nodes [
   n1 = RigVMFunctionEntryNode()
   RigUnit_SequenceAggregate_1 = &RigVMFunction_Sequence(ExecuteContext=@n1.ExecuteContext)
   RigUnit_SequenceAggregate = RigVMAggregateNode(ExecuteContext=@RigUnit_SequenceAggregate_1.A)
   BranchNode_4 = &RigVMFunction_ControlFlowBranch(Condition=@n1.X Plane, ExecuteContext=@RigUnit_SequenceAggregate.A)
   n2 = RigVMVariableNode() {variable=vector pos}
   RigVMVariableNode(
    Value=(X=0,Y=0,Z=0)
    ExecuteContext=@BranchNode_4.True
    Value.X=&RigUnit_GetTransform(bInitial=true, Item=@n1.Item).Transform.Translation.X
    Value.Y=@n2.Value.Y
    Value.Z=@n2.Value.Z
   ) {variable=vector pos}
   BranchNode_1_1 = &RigVMFunction_ControlFlowBranch(Condition=@n1.Y Plane, ExecuteContext=@RigUnit_SequenceAggregate.B)
   n3 = RigVMVariableNode() {variable=vector pos}
   RigVMVariableNode(
    Value=(X=0,Y=0,Z=0)
    ExecuteContext=@BranchNode_1_1.True
    Value.Y=&RigUnit_GetTransform(bInitial=true, Item=@n1.Item).Transform.Translation.Y
    Value.X=@n3.Value.X
    Value.Z=@n3.Value.Z
   ) {variable=vector pos}
   BranchNode_2_1 = &RigVMFunction_ControlFlowBranch(Condition=@n1.Z Plane, ExecuteContext=@RigUnit_SequenceAggregate.C)
   n4 = RigVMVariableNode() {variable=vector pos}
   RigVMVariableNode(
    Value=(X=0,Y=0,Z=0)
    ExecuteContext=@BranchNode_2_1.True
    Value.Z=&RigUnit_GetTransform(bInitial=true, Item=@n1.Item).Transform.Translation.Z
    Value.X=@n4.Value.X
    Value.Y=@n4.Value.Y
   ) {variable=vector pos}
   Cast = &RigVMFunction_MathBoolToFloat(Value=@n1.X Plane)
   Dot__ = /Script/RigVM.RigVMFunction_MathVectorDot(
    B=(X=1,Y=1,Z=0)
    B.X=@Cast.Result
    B.Y=&RigVMFunction_MathBoolToFloat(Value=@n1.Y Plane).Result
    B.Z=&RigVMFunction_MathBoolToFloat(Value=@n1.Z Plane).Result
    A=RigVMVariableNode() {variable=vector pos}.Value
   ) {is=struct RigVMUnitNode, template="Dot,|::Execute(in A,in B,out Result)"}
   Less = /Script/RigVM.RigVMFunction_MathDoubleLess(B=0, A=@Dot__.Result) {is=struct RigVMUnitNode, template="Less::Execute(in A,in B,out Result)"}
   RigVMVariableNode(Value=@Less.Result, ExecuteContext=@RigUnit_SequenceAggregate.D) {variable=is on right}
   RigVMFunctionReturnNode(ExecuteContext=@RigUnit_SequenceAggregate_1.B, Right Side=RigVMVariableNode() {variable=is on right}.Value)
  ]}
 ContainedGraph {nodes [
   n1 = RigVMFunctionEntryNode()
   RigUnit_SequenceAggregate = &RigVMFunction_Sequence(ExecuteContext=@n1.ExecuteContext)
   RigUnit_SequenceAggregate_1 = &RigVMFunction_Sequence(ExecuteContext=@RigUnit_SequenceAggregate.B)
   RigUnit_SequenceAggregate_2 = &RigVMFunction_Sequence(ExecuteContext=@RigUnit_SequenceAggregate_1.B)
   RigVMFunctionReturnNode(A=@RigUnit_SequenceAggregate.A, B=@RigUnit_SequenceAggregate_1.A, C=@RigUnit_SequenceAggregate_2.A, D=@RigUnit_SequenceAggregate_2.B)
  ]}
 ContainedGraph {nodes [
   n1 = RigVMFunctionEntryNode()
   RigUnit_SequenceAggregate_1 = &RigVMFunction_Sequence(ExecuteContext=@n1.ExecuteContext)
   RigUnit_SequenceAggregate = RigVMAggregateNode(ExecuteContext=@RigUnit_SequenceAggregate_1.A)
   Greater = /Script/RigVM.RigVMFunction_MathDoubleGreater(B=0, A=@n1.Mirror Plane.X) {is=struct RigVMUnitNode, template="Greater::Execute(in A,in B,out Result)"}
   BranchNode_4 = &RigVMFunction_ControlFlowBranch(Condition=@Greater.Result, ExecuteContext=@RigUnit_SequenceAggregate.A)
   n2 = RigVMVariableNode() {variable=vector pos}
   RigVMVariableNode(
    Value=(X=0,Y=0,Z=0)
    ExecuteContext=@BranchNode_4.True
    Value.X=&RigUnit_GetTransform(bInitial=true, Item=@n1.Item).Transform.Translation.X
    Value.Y=@n2.Value.Y
    Value.Z=@n2.Value.Z
   ) {variable=vector pos}
   n3 = &RigVMFunction_MathDoubleGreater(B=0, A=@n1.Mirror Plane.Y) {template="Greater::Execute(in A,in B,out Result)"}
   BranchNode_1_1 = &RigVMFunction_ControlFlowBranch(Condition=@n3.Result, ExecuteContext=@RigUnit_SequenceAggregate.B)
   n4 = RigVMVariableNode() {variable=vector pos}
   RigVMVariableNode(
    Value=(X=0,Y=0,Z=0)
    ExecuteContext=@BranchNode_1_1.True
    Value.Y=&RigUnit_GetTransform(bInitial=true, Item=@n1.Item).Transform.Translation.Y
    Value.X=@n4.Value.X
    Value.Z=@n4.Value.Z
   ) {variable=vector pos}
   n5 = &RigVMFunction_MathDoubleGreater(B=0, A=@n1.Mirror Plane.Z) {template="Greater::Execute(in A,in B,out Result)"}
   BranchNode_2_1 = &RigVMFunction_ControlFlowBranch(Condition=@n5.Result, ExecuteContext=@RigUnit_SequenceAggregate.C)
   n6 = RigVMVariableNode() {variable=vector pos}
   RigVMVariableNode(
    Value=(X=0,Y=0,Z=0)
    ExecuteContext=@BranchNode_2_1.True
    Value.Z=&RigUnit_GetTransform(bInitial=true, Item=@n1.Item).Transform.Translation.Z
    Value.X=@n6.Value.X
    Value.Y=@n6.Value.Y
   ) {variable=vector pos}
   Cast = &RigVMFunction_MathBoolToFloat(Value=@Greater.Result)
   Dot__ = &RigVMFunction_MathVectorDot(
    B=(X=1,Y=1,Z=0)
    B.X=@Cast.Result
    B.Y=&RigVMFunction_MathBoolToFloat(Value=@n3.Result).Result
    B.Z=&RigVMFunction_MathBoolToFloat(Value=@n5.Result).Result
    A=RigVMVariableNode() {variable=vector pos}.Value
   ) {template="Dot,|::Execute(in A,in B,out Result)"}
   Less = &RigVMFunction_MathDoubleLess(B=0, A=@Dot__.Result) {template="Less::Execute(in A,in B,out Result)"}
   RigVMVariableNode(Value=@Less.Result, ExecuteContext=@RigUnit_SequenceAggregate.D) {variable=is on right}
   RigVMFunctionReturnNode(ExecuteContext=@RigUnit_SequenceAggregate_1.B, Right Side=RigVMVariableNode() {variable=is on right}.Value)
  ]}
 ContainedGraph {nodes [
   n1 = RigVMFunctionEntryNode()
   RigUnit_SequenceAggregate = &RigVMFunction_Sequence(ExecuteContext=@n1.ExecuteContext)
   RigUnit_SequenceAggregate_1 = &RigVMFunction_Sequence(ExecuteContext=@RigUnit_SequenceAggregate.B)
   RigUnit_SequenceAggregate_2 = &RigVMFunction_Sequence(ExecuteContext=@RigUnit_SequenceAggregate_1.B)
   RigVMFunctionReturnNode(A=@RigUnit_SequenceAggregate.A, B=@RigUnit_SequenceAggregate_1.A, C=@RigUnit_SequenceAggregate_2.A, D=@RigUnit_SequenceAggregate_2.B)
  ]}
 ContainedGraph {nodes [
   n1 = RigVMFunctionEntryNode()
   RigUnit_SequenceAggregate_1 = &RigVMFunction_Sequence(ExecuteContext=@n1.ExecuteContext)
   RigUnit_SequenceAggregate = RigVMAggregateNode(ExecuteContext=@RigUnit_SequenceAggregate_1.A)
   Greater = &RigVMFunction_MathDoubleGreater(B=0, A=@n1.Mirror Plane.X) {template="Greater::Execute(in A,in B,out Result)"}
   BranchNode_4 = &RigVMFunction_ControlFlowBranch(Condition=@Greater.Result, ExecuteContext=@RigUnit_SequenceAggregate.A)
   n2 = &RigVMFunction_MathDoubleGreater(B=0, A=@n1.Mirror Plane.Y) {template="Greater::Execute(in A,in B,out Result)"}
   BranchNode_1_1 = &RigVMFunction_ControlFlowBranch(ExecuteContext=@BranchNode_4.Completed, Condition=@n2.Result)
   n3 = &RigVMFunction_MathDoubleGreater(B=0, A=@n1.Mirror Plane.Z) {template="Greater::Execute(in A,in B,out Result)"}
   BranchNode_2_1 = &RigVMFunction_ControlFlowBranch(ExecuteContext=@BranchNode_1_1.Completed, Condition=@n3.Result)
   Cast = &RigVMFunction_MathBoolToFloat(Value=@Greater.Result)
   Dot__ = &RigVMFunction_MathVectorDot(
    B=(X=1,Y=1,Z=0)
    B.X=@Cast.Result
    B.Y=&RigVMFunction_MathBoolToFloat(Value=@n2.Result).Result
    B.Z=&RigVMFunction_MathBoolToFloat(Value=@n3.Result).Result
    A=RigVMVariableNode() {variable=vector pos}.Value
   ) {template="Dot,|::Execute(in A,in B,out Result)"}
   Less = &RigVMFunction_MathDoubleLess(B=0, A=@Dot__.Result) {template="Less::Execute(in A,in B,out Result)"}
   RigVMVariableNode(ExecuteContext=@BranchNode_2_1.Completed, Value=@Less.Result) {variable=is on right}
   n4 = RigVMVariableNode() {variable=vector pos}
   RigVMVariableNode(
    Value=(X=0,Y=0,Z=0)
    ExecuteContext=@BranchNode_2_1.True
    Value.Z=&RigUnit_GetTransform(bInitial=true, Item=@n1.Item).Transform.Translation.Z
    Value.X=@n4.Value.X
    Value.Y=@n4.Value.Y
   ) {variable=vector pos}
   Dot___1_1_1_1_1_2 = &RigVMFunction_MathVectorDot(
    B=(X=1,Y=1,Z=0)
    B.Y=&RigVMFunction_MathBoolToFloat(Value=@n2.Result).Result
    B.Z=&RigVMFunction_MathBoolToFloat(Value=@n3.Result).Result
    B.X=&RigVMFunction_MathBoolToFloat(Value=@Greater.Result).Result
    A=RigVMVariableNode() {variable=vector pos}.Value
   ) {template="Dot,|::Execute(in A,in B,out Result)"}
   > RigVMVariableNode(Value=&RigVMFunction_MathDoubleLess(B=0, A=@Dot___1_1_1_1_1_2.Result) {template="Less::Execute(in A,in B,out Result)"}.Result) {variable=is on right}
   n5 = RigVMVariableNode() {variable=vector pos}
   RigVMVariableNode(
    Value=(X=0,Y=0,Z=0)
    ExecuteContext=@BranchNode_1_1.True
    Value.Y=&RigUnit_GetTransform(bInitial=true, Item=@n1.Item).Transform.Translation.Y
    Value.X=@n5.Value.X
    Value.Z=@n5.Value.Z
   ) {variable=vector pos}
   Dot___1_1_1_1_1_1 = &RigVMFunction_MathVectorDot(
    B=(X=1,Y=1,Z=0)
    B.Y=&RigVMFunction_MathBoolToFloat(Value=@n2.Result).Result
    B.Z=&RigVMFunction_MathBoolToFloat(Value=@n3.Result).Result
    B.X=&RigVMFunction_MathBoolToFloat(Value=@Greater.Result).Result
    A=RigVMVariableNode() {variable=vector pos}.Value
   ) {template="Dot,|::Execute(in A,in B,out Result)"}
   > RigVMVariableNode(Value=&RigVMFunction_MathDoubleLess(B=0, A=@Dot___1_1_1_1_1_1.Result) {template="Less::Execute(in A,in B,out Result)"}.Result) {variable=is on right}
   n6 = RigVMVariableNode() {variable=vector pos}
   RigVMVariableNode(
    Value=(X=0,Y=0,Z=0)
    ExecuteContext=@BranchNode_4.True
    Value.X=&RigUnit_GetTransform(bInitial=true, Item=@n1.Item).Transform.Translation.X
    Value.Y=@n6.Value.Y
    Value.Z=@n6.Value.Z
   ) {variable=vector pos}
   Dot___1_1_1_1_1 = &RigVMFunction_MathVectorDot(
    B=(X=1,Y=1,Z=0)
    B.Y=&RigVMFunction_MathBoolToFloat(Value=@n2.Result).Result
    B.Z=&RigVMFunction_MathBoolToFloat(Value=@n3.Result).Result
    B.X=&RigVMFunction_MathBoolToFloat(Value=@Greater.Result).Result
    A=RigVMVariableNode() {variable=vector pos}.Value
   ) {template="Dot,|::Execute(in A,in B,out Result)"}
   > RigVMVariableNode(Value=&RigVMFunction_MathDoubleLess(B=0, A=@Dot___1_1_1_1_1.Result) {template="Less::Execute(in A,in B,out Result)"}.Result) {variable=is on right}
   n7 = &RigVMFunction_MathDoubleLess(B=0, A=@n1.Mirror Plane.X) {template="Less::Execute(in A,in B,out Result)"}
   BranchNode_4_1 = &RigVMFunction_ControlFlowBranch(Condition=@n7.Result, ExecuteContext=@RigUnit_SequenceAggregate.D)
   n8 = RigVMVariableNode() {variable=vector pos}
   RigVMVariableNode(
    Value=(X=0,Y=0,Z=0)
    ExecuteContext=@BranchNode_4_1.True
    Value.X=&RigUnit_GetTransform(bInitial=true, Item=@n1.Item).Transform.Translation.X
    Value.Y=@n8.Value.Y
    Value.Z=@n8.Value.Z
   ) {variable=vector pos}
   n9 = &RigVMFunction_MathDoubleLess(B=0, A=@n1.Mirror Plane.Y) {template="Less::Execute(in A,in B,out Result)"}
   n10 = &RigVMFunction_MathDoubleLess(B=0, A=@n1.Mirror Plane.Z) {template="Less::Execute(in A,in B,out Result)"}
   Dot___1_1_1_1 = &RigVMFunction_MathVectorDot(
    B=(X=1,Y=1,Z=0)
    B.Y=&RigVMFunction_MathBoolToFloat(Value=@n9.Result).Result
    B.Z=&RigVMFunction_MathBoolToFloat(Value=@n10.Result).Result
    B.X=&RigVMFunction_MathBoolToFloat(Value=@n7.Result).Result
    A=RigVMVariableNode() {variable=vector pos}.Value
   ) {template="Dot,|::Execute(in A,in B,out Result)"}
   > RigVMVariableNode(Value=&RigVMFunction_MathDoubleGreater(B=0, A=@Dot___1_1_1_1.Result) {template="Greater::Execute(in A,in B,out Result)"}.Result) {variable=is on right}
   BranchNode_1_1_1 = &RigVMFunction_ControlFlowBranch(Condition=@n9.Result, ExecuteContext=@RigUnit_SequenceAggregate.E)
   n11 = RigVMVariableNode() {variable=vector pos}
   RigVMVariableNode(
    Value=(X=0,Y=0,Z=0)
    ExecuteContext=@BranchNode_1_1_1.True
    Value.Y=&RigUnit_GetTransform(bInitial=true, Item=@n1.Item).Transform.Translation.Y
    Value.X=@n11.Value.X
    Value.Z=@n11.Value.Z
   ) {variable=vector pos}
   Dot___1_1 = &RigVMFunction_MathVectorDot(
    B=(X=1,Y=1,Z=0)
    B.Y=&RigVMFunction_MathBoolToFloat(Value=@n9.Result).Result
    B.Z=&RigVMFunction_MathBoolToFloat(Value=@n10.Result).Result
    B.X=&RigVMFunction_MathBoolToFloat(Value=@n7.Result).Result
    A=RigVMVariableNode() {variable=vector pos}.Value
   ) {template="Dot,|::Execute(in A,in B,out Result)"}
   > RigVMVariableNode(Value=&RigVMFunction_MathDoubleGreater(B=0, A=@Dot___1_1.Result) {template="Greater::Execute(in A,in B,out Result)"}.Result) {variable=is on right}
   BranchNode_2_1_1 = &RigVMFunction_ControlFlowBranch(Condition=@n10.Result, ExecuteContext=@RigUnit_SequenceAggregate.F)
   n12 = RigVMVariableNode() {variable=vector pos}
   RigVMVariableNode(
    Value=(X=0,Y=0,Z=0)
    ExecuteContext=@BranchNode_2_1_1.True
    Value.Z=&RigUnit_GetTransform(bInitial=true, Item=@n1.Item).Transform.Translation.Z
    Value.X=@n12.Value.X
    Value.Y=@n12.Value.Y
   ) {variable=vector pos}
   Dot___1_1_1 = &RigVMFunction_MathVectorDot(
    B=(X=1,Y=1,Z=0)
    B.Y=&RigVMFunction_MathBoolToFloat(Value=@n9.Result).Result
    B.Z=&RigVMFunction_MathBoolToFloat(Value=@n10.Result).Result
    B.X=&RigVMFunction_MathBoolToFloat(Value=@n7.Result).Result
    A=RigVMVariableNode() {variable=vector pos}.Value
   ) {template="Dot,|::Execute(in A,in B,out Result)"}
   > RigVMVariableNode(Value=&RigVMFunction_MathDoubleGreater(B=0, A=@Dot___1_1_1.Result) {template="Greater::Execute(in A,in B,out Result)"}.Result) {variable=is on right}
   RigVMFunctionReturnNode(ExecuteContext=@RigUnit_SequenceAggregate_1.B, Right Side=RigVMVariableNode() {variable=is on right}.Value)
   &RigVMFunction_MathDoubleLess(B=0, A=@Dot___1_1.Result) {template="Less::Execute(in A,in B,out Result)"}
   &RigVMFunction_MathDoubleLess(B=0, A=@Dot___1_1_1.Result) {template="Less::Execute(in A,in B,out Result)"}
   &RigVMFunction_MathDoubleLess(B=0, A=@Dot___1_1_1_1.Result) {template="Less::Execute(in A,in B,out Result)"}
  ]}
 ContainedGraph {nodes [
   n1 = RigVMFunctionEntryNode()
   RigUnit_SequenceAggregate = &RigVMFunction_Sequence(ExecuteContext=@n1.ExecuteContext)
   RigUnit_SequenceAggregate_1 = &RigVMFunction_Sequence(ExecuteContext=@RigUnit_SequenceAggregate.B)
   RigUnit_SequenceAggregate_2 = &RigVMFunction_Sequence(ExecuteContext=@RigUnit_SequenceAggregate_1.B)
   RigUnit_SequenceAggregate_3 = &RigVMFunction_Sequence(ExecuteContext=@RigUnit_SequenceAggregate_2.B)
   RigUnit_SequenceAggregate_4 = &RigVMFunction_Sequence(ExecuteContext=@RigUnit_SequenceAggregate_3.B)
   RigUnit_SequenceAggregate_5 = &RigVMFunction_Sequence(ExecuteContext=@RigUnit_SequenceAggregate_4.B)
   RigVMFunctionReturnNode(
    A=@RigUnit_SequenceAggregate.A
    B=@RigUnit_SequenceAggregate_1.A
    C=@RigUnit_SequenceAggregate_2.A
    D=@RigUnit_SequenceAggregate_3.A
    E=@RigUnit_SequenceAggregate_4.A
    F=@RigUnit_SequenceAggregate_5.A
    G=@RigUnit_SequenceAggregate_5.B
   )
  ]}
 Create Twist _ContainedGraph {nodes [
   n1 = RigVMFunctionEntryNode()
   > For_Each = RigVMDispatchNode(Array=/Script/ControlRig.RigUnit_ResolveArrayConnector(Connector=(Type=Connector,Name="Twist Bones"), SkipSocket=true) {is=struct RigVMUnitNode}.Result) {template="DISPATCH_RigVMDispatch_ArrayIterator(in Array,out Element,out Index,out Count,out Ratio)"}
   > FindClosestBone_1 = RigVMFunctionReferenceNode(Initial=true, Bones=@n1.Template Bones, Item=@For_Each.Element)
   Clone = RigVMDispatchNode(Array=@n1.Template Bones) {template="DISPATCH_RigVMDispatch_ArrayClone(in Array,out Clone)"}
   > n2 = RigVMDispatchNode(
    Array=@Clone.Clone
    Index=RigVMDispatchNode(Array=@Clone.Clone, Element=@FindClosestBone_1.Value) {template="DISPATCH_RigVMDispatch_ArrayFind(in Array,in Element,out Index,out Success)"}.Index
   ) {template="DISPATCH_RigVMDispatch_ArrayRemove(io Array,in Index)"}
   > FindClosestBone_0 = RigVMFunctionReferenceNode(Initial=true, Item=@For_Each.Element, Bones=@n2.Array)
   RigVMFunction_MathVectorDistance = &RigVMFunction_MathVectorDistance(
    A=&RigUnit_GetTransform(bInitial=False, Item=@FindClosestBone_1.Value).Transform.Translation
    B=&RigUnit_GetTransform(bInitial=False, Item=@For_Each.Element).Transform.Translation
   )
   n3 = &RigVMFunction_MathVectorDistance(
    A=&RigUnit_GetTransform(bInitial=False, Item=@FindClosestBone_1.Value).Transform.Translation
    B=&RigUnit_GetTransform(bInitial=False, Item=@FindClosestBone_0.Value).Transform.Translation
   )
   Divide = &RigVMFunction_MathDoubleDiv(A=@RigVMFunction_MathVectorDistance.Result, B=@n3.Result)
   > Set_Metadata = RigVMDispatchNode(Name=Twist Amount A, NameSpace=Self, Value=@Divide.Result, Item=@For_Each.Element) {template="DISPATCH_RigDispatch_SetMetadata(in Item,in Name,in NameSpace,in Value,out Success)"}
   n4 = &RigUnit_GetTransform(bInitial=False, Item=@For_Each.Element)
   n5 = &RigUnit_GetTransform(bInitial=False, Item=@FindClosestBone_0.Value)
   > Set_Metadata_1 = RigVMDispatchNode(
    Name=Twist Amount B
    NameSpace=Self
    Value=&RigVMFunction_MathDoubleDiv(A=&RigVMFunction_MathVectorDistance(A=@n4.Transform.Translation, B=@n5.Transform.Translation).Result, B=@n3.Result).Result
    Item=@For_Each.Element
   ) {template="DISPATCH_RigDispatch_SetMetadata(in Item,in Name,in NameSpace,in Value,out Success)"}
   > Set_Metadata_2 = RigVMDispatchNode(Name=Twist Item A, NameSpace=Self, Value=@FindClosestBone_1.Value, Item=@For_Each.Element) {template="DISPATCH_RigDispatch_SetMetadata(in Item,in Name,in NameSpace,in Value,out Success)"}
   > Set_Metadata_3 = RigVMDispatchNode(Name=Twist Item B, NameSpace=Self, Value=@FindClosestBone_0.Value, Item=@For_Each.Element) {template="DISPATCH_RigDispatch_SetMetadata(in Item,in Name,in NameSpace,in Value,out Success)"}
   RigVMFunctionReturnNode(ExecuteContext=@For_Each.Completed)
  ]}
 New Function_ContainedGraph {nodes [
   n1 = RigVMFunctionEntryNode()
   Concat_4 = RigVMAggregateNode(B=@n1.name, A=@n1.prefix, C=@n1.side, D=@n1.suffix)
   RigVMFunctionReturnNode(Result=@Concat_4.Result)
  ]}
 New Function_ContainedGraph {nodes [
   n1 = RigVMFunctionEntryNode()
   Concat_4 = RigVMAggregateNode(B=@n1.name, A=@n1.prefix, C=@n1.side, D=@n1.suffix)
   RigVMFunctionReturnNode(Result=@Concat_4.Result)
  ]}
 New Function_ContainedGraph {nodes [
   n1 = RigVMFunctionEntryNode()
   n2 = /Script/ControlRig.RigUnit_AimBoneMath(
    DebugSettings=(bEnabled=False,Scale=10,WorldOffset=(Rotation=(X=0,Y=0,Z=0,W=1),Translation=(X=0,Y=0,Z=0),Scale3D=(X=1,Y=1,Z=1)))
    Primary=(Weight=1,Axis=(X=1,Y=0,Z=0),Target=(X=0,Y=0,Z=0),Kind=Location,Space=(Type=Control,Name="None"))
    Secondary=(Weight=1,Axis=(X=0,Y=-1,Z=0),Target=(X=0,Y=0,Z=1),Kind=Direction,Space=(Type=Control,Name="None"))
    Weight=1
    Primary.Axis=@n1.Axis
    Secondary.Axis=@n1.Secondary Axis
    Secondary.Target=@n1.Secondary Direction
    Secondary.Space=@n1.Target item
    InputTransform=&RigUnit_GetTransform(bInitial=false, Item=@n1.Item).Transform
    Primary.Target=&RigUnit_GetTransform(bInitial=False, Item=@n1.Target item).Transform.Translation
   ) {is=struct RigVMUnitNode}
   > &RigUnit_SetTransform(Weight=1, bInitial=False, Value=@n2.Result, Item=@n1.Item)
   > RigVMFunction_DebugTransformMutableNoSpace = /Script/RigVM.RigVMFunction_DebugTransformMutableNoSpace(
    Color=(R=1,G=0,B=0,A=1)
    DebugDrawSettings=(DepthPriority=SDPG_Foreground,Lifetime=-1)
    Mode=Axes
    Scale=50
    Thickness=3
    WorldOffset=(Rotation=(X=0,Y=0,Z=0,W=1),Translation=(X=0,Y=0,Z=0),Scale3D=(X=1,Y=1,Z=1))
    Transform=@n2.Result
    bEnabled=@n1.Debug
   ) {is=struct RigVMUnitNode}
   > RigVMFunctionReturnNode(Result=@n2.Result)
  ]}
 New Function_ContainedGraph {nodes [
   n1 = RigVMFunctionEntryNode()
   n2 = &RigUnit_AimBoneMath(
    DebugSettings=(bEnabled=False,Scale=10,WorldOffset=(Rotation=(X=0,Y=0,Z=0,W=1),Translation=(X=0,Y=0,Z=0),Scale3D=(X=1,Y=1,Z=1)))
    Primary=(Weight=0.5,Axis=(X=1,Y=0,Z=0),Target=(X=0,Y=0,Z=0),Kind=Location,Space=(Type=Control,Name="None"))
    Secondary=(Weight=0.5,Axis=(X=0,Y=-1,Z=0),Target=(X=0,Y=0,Z=1),Kind=Direction,Space=(Type=Control,Name="None"))
    Weight=1
    Primary.Axis=@n1.Axis
    Secondary.Axis=@n1.Secondary Axis
    Secondary.Target=@n1.Secondary Direction
    Secondary.Space=@n1.Target item
    InputTransform=&RigUnit_GetTransform(bInitial=false, Item=@n1.Item).Transform
    Primary.Target=&RigUnit_GetTransform(bInitial=False, Item=@n1.Target item).Transform.Translation
   )
   > &RigUnit_SetTransform(Weight=1, bInitial=False, Value=@n2.Result, Item=@n1.Item)
   > RigVMFunction_DebugTransformMutableNoSpace = &RigVMFunction_DebugTransformMutableNoSpace(
    Color=(R=1,G=0,B=0,A=1)
    DebugDrawSettings=(DepthPriority=SDPG_Foreground,Lifetime=-1)
    Mode=Axes
    Scale=50
    Thickness=3
    WorldOffset=(Rotation=(X=0,Y=0,Z=0,W=1),Translation=(X=0,Y=0,Z=0),Scale3D=(X=1,Y=1,Z=1))
    Transform=@n2.Result
    bEnabled=@n1.Debug
   )
   > RigVMFunctionReturnNode()
  ]}
 New Function_ContainedGraph {nodes [
   RigVMFunctionEntryNode()
   > RigVMFunction_Sequence = RigVMAggregateNode()
   DISPATCH_RigVMDispatch_ArrayIterator = RigVMDispatchNode(Array=&RigUnit_FindItemsWithMetadataTag(NameSpace=Self, Tag=Controls).Items, ExecuteContext=@RigVMFunction_Sequence.A) {template="DISPATCH_RigVMDispatch_ArrayIterator(in Array,out Element,out Index,out Count,out Ratio)"}
   DISPATCH_RigDispatch_GetMetadata = RigVMDispatchNode(Default=(Type=Bone,Name="None"), Name=Item, NameSpace=Self, Item=@DISPATCH_RigVMDispatch_ArrayIterator.Element) {template="DISPATCH_RigDispatch_GetMetadata(in Item,in Name,in NameSpace,in Default,out Value,out Found)"}
   RigVMFunction_ControlFlowBranch = &RigVMFunction_ControlFlowBranch(Condition=@DISPATCH_RigDispatch_GetMetadata.Found, ExecuteContext=@DISPATCH_RigVMDispatch_ArrayIterator.ExecuteContext)
   &RigUnit_SetTransform(
    Weight=1
    bInitial=False
    Item=@DISPATCH_RigVMDispatch_ArrayIterator.Element
    Value=/Script/ControlRig.RigUnit_ProjectTransformToNewParent(
     bChildInitial=True
     bNewParentInitial=False
     bOldParentInitial=True
     NewParent=@DISPATCH_RigDispatch_GetMetadata.Value
     OldParent=@DISPATCH_RigDispatch_GetMetadata.Value
     Child=@DISPATCH_RigVMDispatch_ArrayIterator.Element
    ) {is=struct RigVMUnitNode}.Transform
    ExecutePin=@RigVMFunction_ControlFlowBranch.True
   )
   n1 = RigVMDispatchNode(Array=&RigUnit_FindItemsWithMetadataTag(NameSpace=Self, Tag=PV Controls).Items, ExecuteContext=@RigVMFunction_Sequence.B) {template="DISPATCH_RigVMDispatch_ArrayIterator(in Array,out Element,out Index,out Count,out Ratio)"}
   > FindClosestBone = RigVMFunctionReferenceNode(Initial=true, Item=@n1.Element, Bones=&RigUnit_FindItemsWithMetadataTag(NameSpace=Self, Tag=Bones).Items)
   > &RigUnit_SetTransform(
    Weight=1
    bInitial=False
    Item=@n1.Element
    Value=&RigUnit_ProjectTransformToNewParent(
     bChildInitial=True
     bNewParentInitial=False
     bOldParentInitial=True
     Child=@n1.Element
     NewParent=@FindClosestBone.Value
     OldParent=@FindClosestBone.Value
    ).Transform
   )
   RigVMFunctionReturnNode(ExecuteContext=@RigVMFunction_Sequence.C)
  ]}
 New Function_ContainedGraph {nodes [
   n1 = RigVMFunctionEntryNode()
   n2 = /Script/RigVM.RigVMFunction_MathVectorSub(
    A=&RigUnit_GetTransform(Space=LocalSpace, bInitial=true, Item=@n1.End Bone).Transform.Translation
    B=&RigUnit_GetTransform(Space=LocalSpace, bInitial=true, Item=@n1.Start Bone).Transform.Translation
   ) {is=struct RigVMUnitNode, template="Subtract::Execute(in A,in B,out Result)"}
   DISPATCH_RigVMDispatch_If = RigVMDispatchNode(
    True=@n1.Left Side Suffix
    False=@n1.Right Side Suffix
    Condition=&RigVMFunction_MathDoubleGreater(
     B=0
     A=/Script/RigVM.RigVMFunction_MathVectorUnit(Value=@n2.Result) {is=struct RigVMUnitNode, template="Unit::Execute(in Value,out Result)"}.Result.X
    ) {template="Greater::Execute(in A,in B,out Result)"}.Result
   ) {template="DISPATCH_RigVMDispatch_If(in Condition,in True,in False,out Result)"}
   > RigVMVariableNode(Value=@DISPATCH_RigVMDispatch_If.Result) {variable=side}
   > RigVMFunctionReturnNode(Result=@DISPATCH_RigVMDispatch_If.Result)
  ]}
 New Function_ContainedGraph {nodes [
   RigVMFunctionEntryNode()
   > RigVMFunction_Sequence = RigVMAggregateNode()
   Get_Module_Metadata_4 = RigVMDispatchNode(Default=(X=0,Y=0,Z=0), Name=knee_bone_offset, NameSpace=Root) {template="DISPATCH_RigDispatch_GetModuleMetadata(in Name,in NameSpace,in Default,out Value,out Found)"}
   Add = &RigVMFunction_MathVectorAdd(A=@Get_Module_Metadata_4.Value, B=RigVMVariableNode() {variable=Knee Bone Offset}.Value) {template="Add::Execute(in A,in B,out Result)"}
   Get_Module_Metadata = RigVMDispatchNode(Default=(Type=Control,Name="None"), Name=Body Control, NameSpace=Root) {template="DISPATCH_RigDispatch_GetModuleMetadata(in Name,in NameSpace,in Default,out Value,out Found)"}
   Get_Module_Metadata_1 = RigVMDispatchNode(Default=(Type=Control,Name="None"), Name=Hips Control, NameSpace=Parent) {template="DISPATCH_RigDispatch_GetModuleMetadata(in Name,in NameSpace,in Default,out Value,out Found)"}
   Get_Module_Metadata_4_1 = RigVMDispatchNode(Default=False, Name=knee_bone_offset_globalSpace, NameSpace=Root) {template="DISPATCH_RigDispatch_GetModuleMetadata(in Name,in NameSpace,in Default,out Value,out Found)"}
   n1 = RigVMVariableNode() {variable=Knee Bone Offset Global Space}
   CONSTR TwoBoneLimb Leg = RigVMFunctionReferenceNode(
    Basename=leg
    Enable Keep Aligned=false
    End Bone Connector=(Type=Connector,Name="Foot Bone")
    IK Control Basename=foot
    IK Control Spaces=((Type=None,Name="None"),(Type=None,Name="None"))
    Local Default=false
    Middle Bone Connector=(Type=Connector,Name="Knee Bone")
    Offset PV Control=(Rotation=(X=0,Y=0,Z=0,W=1),Translation=(X=0,Y=0,Z=0),Scale3D=(X=1,Y=1,Z=1))
    PV Control Spaces=((Type=None,Name="None"),(Type=None,Name="None"))
    Parent Control Connector=(Type=Connector,Name="Parent Control")
    Pole Vector Parent Connector=(Type=Connector,Name="Pole Vector Parent")
    Primary Axis=(X=1,Y=0,Z=0)
    Secondary Axis=(X=0,Y=-1,Z=0)
    Settings Control Connector=(Type=Connector,Name="Settings Control")
    Start Bone Connector=(Type=Connector,Name="Thigh Bone")
    Wrist Control Settings="(DisplayName=\"None\",InitialSpace=LocalSpace,bUsePreferredRotationOrder=False,PreferredRotationOrder=YZX,Limits=(LimitTranslationX=(bMinimum=False,bMaximum=False),LimitTranslationY=(bMinimum=False,bMaximum=False),LimitTranslationZ=(bMinimum=False,bMaximum=F ...(truncated)"
    Offset Elbow Bone=@Add.Result
    PV Offset Factor=RigVMCollapseNode(Mult Value=RigVMVariableNode() {variable=Pole Vector Position}.Value).Result
    IK Control Spaces.0=@Get_Module_Metadata.Value
    PV Control Spaces.0=@Get_Module_Metadata.Value
    IK Control Spaces.1=@Get_Module_Metadata_1.Value
    PV Control Spaces.1=@Get_Module_Metadata_1.Value
    Offset Elbow Bone Space=RigVMDispatchNode(True=GlobalSpace, Condition=&RigVMFunction_MathBoolOr(A=@Get_Module_Metadata_4_1.Value, B=@n1.Value).Result) {template="DISPATCH_RigVMDispatch_If(in Condition,in True,in False,out Result)"}.Result
    ExecuteContext=@RigVMFunction_Sequence.B
    FK IK Default=RigVMVariableNode() {variable=FK IK Default}.Value
    Enable Softness=RigVMVariableNode() {variable=Enable Softness}.Value
    Enable PV Follow=RigVMVariableNode() {variable=Enable PV Follow}.Value
    Module Settings=RigVMVariableNode() {variable=Module Settings}.Value
    IK Control Settings=RigVMVariableNode() {variable=IK Shape Settings}.Value
    PV Control Settings=RigVMVariableNode() {variable=PV Shape Settings}.Value
    FK Control Settings=RigVMVariableNode() {variable=FK Shape Settings}.Value
    Offset IK Ctrl Transform=RigVMVariableNode() {variable=IK Control Offset}.Value
    Orient to World=RigVMVariableNode() {variable=Orient IK to World}.Value
   )
   n2 = RigVMDispatchNode(Index=0, Array=&RigUnit_FindItemsWithMetadataTag(NameSpace=Self, Tag=Controls IK).Items) {template="DISPATCH_RigVMDispatch_ArrayGetAtIndex(in Array,in Index,out Element)"}
   Format Name = RigVMCollapseNode(name=foot, prefix=None, suffix=_ik_effector, side=@"CONSTR TwoBoneLimb Leg".Current Side Name)
   > n3 = &RigUnit_HierarchyAddNull(
    Space=LocalSpace
    Transform=(Rotation=(X=0,Y=0,Z=0,W=1),Translation=(X=0,Y=0,Z=0),Scale3D=(X=1,Y=1,Z=1))
    Parent=@n2.Element
    Name=@"Format Name".Result
   )
   > RigVMDispatchNode(Name=Effector Null, NameSpace=Self, Value=@n3.Item) {template="DISPATCH_RigDispatch_SetModuleMetadata(in Name,in NameSpace,in Value,out Success)"}
   > RigVMDispatchNode(Name=Foot Control, NameSpace=Self, Value=@n2.Element) {template="DISPATCH_RigDispatch_SetModuleMetadata(in Name,in NameSpace,in Value,out Success)"}
   > n4 = RigVMDispatchNode(Name=Parent Control, NameSpace=Self, Value=@n2.Element) {template="DISPATCH_RigDispatch_SetModuleMetadata(in Name,in NameSpace,in Value,out Success)"}
   Get_Module_Metadata_3_2 = RigVMDispatchNode(Default=(Type=Bone,Name="None"), Name=toe_l, NameSpace=Root) {template="DISPATCH_RigDispatch_GetModuleMetadata(in Name,in NameSpace,in Default,out Value,out Found)"}
   Get_Module_Metadata_3_2_1 = RigVMDispatchNode(Default=(Type=Bone,Name="None"), Name=toe_r, NameSpace=Root) {template="DISPATCH_RigDispatch_GetModuleMetadata(in Name,in NameSpace,in Default,out Value,out Found)"}
   n5 = &RigVMFunction_ControlFlowBranch(
    ExecuteContext=@n4.ExecuteContext
    Condition=&RigVMFunction_MathBoolOr(A=@Get_Module_Metadata_3_2.Found, B=@Get_Module_Metadata_3_2_1.Found).Result
   )
   n6 = &RigUnit_CollectionChildrenArray(
    TypeToSearch=Bone
    bDefaultChildren=True
    bIncludeParent=False
    bRecursive=False
    Parent=&RigUnit_ResolveConnector(Connector=(Type=Connector,Name="Foot Bone"), SkipSocket=true).Result
   )
   Get_Module_Metadata_3 = RigVMDispatchNode(Default=(Type=Bone,Name="None"), Name=foot_l, NameSpace=Root) {template="DISPATCH_RigDispatch_GetModuleMetadata(in Name,in NameSpace,in Default,out Value,out Found)"}
   RigVMFunction_ControlFlowBranch = &RigVMFunction_ControlFlowBranch(
    Condition=RigVMDispatchNode(Array=@n6.Items, Element=@Get_Module_Metadata_3.Value) {template="DISPATCH_RigVMDispatch_ArrayFind(in Array,in Element,out Index,out Success)"}.Success
    ExecuteContext=@n5.False
   )
   Get_Module_Metadata_3_1 = RigVMDispatchNode(Default=(Type=Bone,Name="None"), Name=foot_r, NameSpace=Root) {template="DISPATCH_RigDispatch_GetModuleMetadata(in Name,in NameSpace,in Default,out Value,out Found)"}
   n7 = &RigVMFunction_ControlFlowBranch(
    Condition=RigVMDispatchNode(Array=@n6.Items, Element=@Get_Module_Metadata_3_1.Value) {template="DISPATCH_RigVMDispatch_ArrayFind(in Array,in Element,out Index,out Success)"}.Success
    ExecuteContext=@RigVMFunction_ControlFlowBranch.Completed
   )
   RigVMFunction_MathBoolOr = &RigVMFunction_MathBoolOr(A=@Get_Module_Metadata_3.Found, B=@Get_Module_Metadata_3_1.Found)
   RigVMFunction_MathBoolNot = &RigVMFunction_MathBoolNot(Value=@RigVMFunction_MathBoolOr.Result)
   n8 = &RigVMFunction_ControlFlowBranch(ExecuteContext=@n7.Completed, Condition=@RigVMFunction_MathBoolNot.Result)
   Format Name_1 = RigVMCollapseNode(name=foot, prefix=None, suffix=_socket, side=@"CONSTR TwoBoneLimb Leg".Current Side Name)
   &RigUnit_HierarchyAddSocket(
    Color=(R=1,G=1,B=1,A=1)
    Description="Place the foot module here . . . "
    Space=LocalSpace
    Transform=(Rotation=(X=0,Y=0,Z=0,W=1),Translation=(X=0,Y=0,Z=0),Scale3D=(X=1,Y=1,Z=1))
    Name=@"Format Name_1".Result
    ExecutePin=@n8.Completed
    Parent=RigVMVariableNode() {variable=foot_socket}.Value
   )
   RigVMVariableNode(
    Value=RigVMDispatchNode(Index=0, Array=@n6.Items) {template="DISPATCH_RigVMDispatch_ArrayGetAtIndex(in Array,in Index,out Element)"}.Element
    ExecuteContext=@n8.True
   ) {variable=foot_socket}
   RigVMVariableNode(Value=@Get_Module_Metadata_3_1.Value, ExecuteContext=@n7.True) {variable=foot_socket}
   RigVMVariableNode(Value=@Get_Module_Metadata_3.Value, ExecuteContext=@RigVMFunction_ControlFlowBranch.True) {variable=foot_socket}
   n9 = &RigUnit_ResolveConnector(Connector=(Type=Connector,Name="Leg Primary"), SkipSocket=true)
   Get Side from Metadata = RigVMFunctionReferenceNode(Item=&RigUnit_ResolveConnector(Connector=(Type=Connector,Name="Leg Primary"), SkipSocket=true).Result)
   n10 = RigVMDispatchNode(Condition=@"Get Side from Metadata".Result, False=@Get_Module_Metadata_3_2.Value, True=@Get_Module_Metadata_3_2_1.Value) {template="DISPATCH_RigVMDispatch_If(in Condition,in True,in False,out Result)"}
   n11 = &RigVMFunction_ControlFlowBranch(
    Condition=RigVMDispatchNode(
     Array=&RigUnit_CollectionChildrenArray(TypeToSearch=Bone, bDefaultChildren=True, bIncludeParent=true, bRecursive=true, Parent=@n9.Result).Items
     Element=@n10.Result
    ) {template="DISPATCH_RigVMDispatch_ArrayFind(in Array,in Element,out Index,out Success)"}.Success
    ExecuteContext=@n5.True
   )
   &RigUnit_HierarchyAddSocket(
    Color=(R=1,G=1,B=1,A=1)
    Description="Place the foot module here . . . "
    Space=LocalSpace
    Transform=(Rotation=(X=0,Y=0,Z=0,W=1),Translation=(X=0,Y=0,Z=0),Scale3D=(X=1,Y=1,Z=1))
    Name=&RigVMFunction_NameConcat(B=_socket, A=@n10.Result.Name) {template="Concat::Execute(in A,in B,out Result)"}.Result
    Parent=@n10.Result
    ExecutePin=@n11.True
   )
   RigVMFunctionReturnNode(ExecuteContext=@RigVMFunction_Sequence.C)
  ]}
 New Function_ContainedGraph {nodes [
   n1 = RigVMFunctionEntryNode()
   Concat_4 = RigVMAggregateNode(B=@n1.name, A=@n1.prefix, C=@n1.side, D=@n1.suffix)
   RigVMFunctionReturnNode(Result=@Concat_4.Result)
  ]}
 New Function_ContainedGraph {nodes [
   n1 = RigVMFunctionEntryNode()
   Concat_4 = RigVMAggregateNode(B=@n1.name, A=@n1.prefix, C=@n1.side, D=@n1.suffix)
   RigVMFunctionReturnNode(Result=@Concat_4.Result)
  ]}
 New Function_ContainedGraph {nodes [
   RigVMFunctionEntryNode()
   > RigVMFunction_Sequence = RigVMAggregateNode()
   Connect to Module Metadata = RigVMFunctionReferenceNode(
    Connector=(Type=Connector,Name="Parent Control")
    Module Metadata Name=Parent Control
    Module Metadata NameSpace=Root
    ExecuteContext=@RigVMFunction_Sequence.B
   )
   > Connect to Module Metadata_1 = RigVMFunctionReferenceNode(Connector=(Type=Connector,Name="Settings Control"), Module Metadata Name=Settings Control, Module Metadata NameSpace=Root)
   > Connect to Module Metadata_2 = RigVMFunctionReferenceNode(Connector=(Type=Connector,Name="Pole Vector Parent"), Module Metadata Name=Parent Control, Module Metadata NameSpace=Root)
   Connect to Module Metadata_1_1_1 = RigVMFunctionReferenceNode(
    Module Metadata NameSpace=Root
    Module Metadata Name=RigVMDispatchNode(False=foot_l, True=foot_r, Condition=RigVMVariableNode() {variable=Right Side}.Value) {template="DISPATCH_RigVMDispatch_If(in Condition,in True,in False,out Result)"}.Result
    Connector=/Script/ControlRig.RigUnit_Item(Item=(Type=Connector,Name="Foot Bone")) {is=struct RigVMUnitNode}.Item
    ExecuteContext=@RigVMFunction_Sequence.C
   )
   n1 = &RigVMFunction_ControlFlowBranch(ExecuteContext=@"Connect to Module Metadata_1_1_1".ExecuteContext, Condition=@"Connect to Module Metadata_1_1_1".Found)
   n2 = /Script/ControlRig.RigUnit_GetCandidates() {is=struct RigVMUnitNode}
   n3 = &RigVMFunction_ControlFlowBranch(
    Condition=RigVMDispatchNode(B=@n2.Connector, A=&RigUnit_Item(Item=(Type=Connector,Name="Foot Bone")).Item) {template="DISPATCH_RigVMDispatch_CoreEquals(in A,in B,out Result)"}.Result
    ExecuteContext=@n1.False
   )
   /Script/ControlRig.RigUnit_DiscardMatches(
    Excluded=/Script/ControlRig.RigUnit_CollectionNameSearchArray(PartialName=corrective, TypeToSearch=Bone) {is=struct RigVMUnitNode}.Items
    ExecutePin=@n3.True
   ) {is=struct RigVMUnitNode}
   > n4 = RigVMDispatchNode(Array=@n2.Candidates) {template="DISPATCH_RigVMDispatch_ArrayIterator(in Array,out Element,out Index,out Count,out Ratio)"}
   > Find first in Hierarchy = RigVMFunctionReferenceNode(
    Children=true
    Ignore=()
    Search for=("foot")
    Type to Search=Bone
    Item=&RigUnit_ResolveConnector(Connector=(Type=Connector,Name="Leg Primary"), SkipSocket=true).Result
   )
   n5 = &RigVMFunction_ControlFlowBranch(
    Condition=RigVMDispatchNode(B=@n4.Element, A=@"Find first in Hierarchy".Item Found) {template="DISPATCH_RigVMDispatch_CoreEquals(in A,in B,out Result)"}.Result
    ExecuteContext=@"Find first in Hierarchy".ExecuteContext
   )
   /Script/ControlRig.RigUnit_SetDefaultMatch(Default=@n4.Element, ExecutePin=@n5.True) {is=struct RigVMUnitNode, template="Set Default Match::Execute(in Default)"}
   Connect to Module Metadata_1_1_1_1 = RigVMFunctionReferenceNode(
    Module Metadata NameSpace=Root
    Module Metadata Name=RigVMDispatchNode(False=thigh_l, True=thigh_r, Condition=RigVMVariableNode() {variable=Right Side}.Value) {template="DISPATCH_RigVMDispatch_If(in Condition,in True,in False,out Result)"}.Result
    Connector=&RigUnit_Item(Item=(Type=Connector,Name="Thigh Bone")).Item
    ExecuteContext=@RigVMFunction_Sequence.D
   )
   n6 = &RigVMFunction_ControlFlowBranch(ExecuteContext=@"Connect to Module Metadata_1_1_1_1".ExecuteContext, Condition=@"Connect to Module Metadata_1_1_1_1".Found)
   n7 = &RigUnit_GetCandidates()
   n8 = &RigVMFunction_ControlFlowBranch(
    Condition=RigVMDispatchNode(B=@n7.Connector, A=&RigUnit_Item(Item=(Type=Connector,Name="Thigh Bone")).Item) {template="DISPATCH_RigVMDispatch_CoreEquals(in A,in B,out Result)"}.Result
    ExecuteContext=@n6.False
   )
   n9 = RigVMDispatchNode(Array=@n7.Candidates, ExecuteContext=@n8.True) {template="DISPATCH_RigVMDispatch_ArrayIterator(in Array,out Element,out Index,out Count,out Ratio)"}
   n10 = &RigVMFunction_ControlFlowBranch(
    ExecuteContext=@n9.ExecuteContext
    Condition=RigVMDispatchNode(B=@n9.Element, A=&RigUnit_ResolveConnector(Connector=(Type=Connector,Name="Leg Primary"), SkipSocket=true).Result) {template="DISPATCH_RigVMDispatch_CoreEquals(in A,in B,out Result)"}.Result
   )
   &RigUnit_SetDefaultMatch(Default=@n9.Element, ExecutePin=@n10.True) {template="Set Default Match::Execute(in Default)"}
   n11 = &RigUnit_Item(Item=(Type=Connector,Name="Knee Bone"))
   Connect to Module Metadata_1_1_1_2 = RigVMFunctionReferenceNode(
    Module Metadata NameSpace=Root
    Module Metadata Name=RigVMDispatchNode(False=knee_l, True=knee_r, Condition=RigVMVariableNode() {variable=Right Side}.Value) {template="DISPATCH_RigVMDispatch_If(in Condition,in True,in False,out Result)"}.Result
    Connector=@n11.Item
    ExecuteContext=@RigVMFunction_Sequence.E
   )
   n12 = &RigVMFunction_ControlFlowBranch(ExecuteContext=@"Connect to Module Metadata_1_1_1_2".ExecuteContext, Condition=@"Connect to Module Metadata_1_1_1_2".Found)
   n13 = &RigUnit_GetCandidates()
   n14 = &RigVMFunction_ControlFlowBranch(
    Condition=RigVMDispatchNode(B=@n13.Connector, A=@n11.Item) {template="DISPATCH_RigVMDispatch_CoreEquals(in A,in B,out Result)"}.Result
    ExecuteContext=@n12.False
   )
   n15 = RigVMDispatchNode(Array=@n13.Candidates, ExecuteContext=@n14.True) {template="DISPATCH_RigVMDispatch_ArrayIterator(in Array,out Element,out Index,out Count,out Ratio)"}
   n16 = &RigUnit_ResolveConnector(SkipSocket=true, Connector=&RigUnit_Item(Item=(Type=Connector,Name="Foot Bone")).Item)
   n17 = &RigVMFunction_ControlFlowBranch(
    ExecuteContext=@n15.ExecuteContext
    Condition=RigVMDispatchNode(B=@n15.Element, A=/Script/ControlRig.RigUnit_HierarchyGetParent(bDefaultParent=True, Child=@n16.Result) {is=struct RigVMUnitNode}.Parent) {template="DISPATCH_RigVMDispatch_CoreEquals(in A,in B,out Result)"}.Result
   )
   &RigUnit_SetDefaultMatch(Default=@n15.Element, ExecutePin=@n17.True) {template="Set Default Match::Execute(in Default)"}
   Connect to Module Metadata_1_1_1_1_1 = RigVMFunctionReferenceNode(
    Module Metadata Name=pelvis
    Module Metadata NameSpace=Root
    Connector=&RigUnit_Item(Item=(Type=Connector,Name="Pelvis Bone")).Item
    ExecuteContext=@RigVMFunction_Sequence.F
   )
   n18 = &RigVMFunction_ControlFlowBranch(ExecuteContext=@"Connect to Module Metadata_1_1_1_1_1".ExecuteContext, Condition=@"Connect to Module Metadata_1_1_1_1_1".Found)
   n19 = &RigUnit_GetCandidates()
   n20 = &RigVMFunction_ControlFlowBranch(
    Condition=RigVMDispatchNode(B=@n19.Connector, A=&RigUnit_Item(Item=(Type=Connector,Name="Pelvis Bone")).Item) {template="DISPATCH_RigVMDispatch_CoreEquals(in A,in B,out Result)"}.Result
    ExecuteContext=@n18.False
   )
   n21 = RigVMDispatchNode(Array=@n19.Candidates, ExecuteContext=@n20.True) {template="DISPATCH_RigVMDispatch_ArrayIterator(in Array,out Element,out Index,out Count,out Ratio)"}
   n22 = &RigUnit_ResolveConnector(Connector=(Type=Connector,Name="Leg Primary"), SkipSocket=true)
   n23 = &RigVMFunction_ControlFlowBranch(
    ExecuteContext=@n21.ExecuteContext
    Condition=RigVMDispatchNode(B=@n21.Element, A=&RigUnit_HierarchyGetParent(bDefaultParent=True, Child=@n22.Result).Parent) {template="DISPATCH_RigVMDispatch_CoreEquals(in A,in B,out Result)"}.Result
   )
   &RigUnit_SetDefaultMatch(Default=@n21.Element, ExecutePin=@n23.True) {template="Set Default Match::Execute(in Default)"}
   RigVMFunctionReturnNode(ExecuteContext=@RigVMFunction_Sequence.G)
   Get_Module_Metadata_3 = RigVMDispatchNode(Default=(X=0,Y=0,Z=0), Name=Mirror Axis, NameSpace=Root) {template="DISPATCH_RigDispatch_GetModuleMetadata(in Name,in NameSpace,in Default,out Value,out Found)"}
   Is Right Side v2 = RigVMFunctionReferenceNode(
    Mirror Plane=@Get_Module_Metadata_3.Value
    Item=&RigUnit_ResolveConnector(Connector=(Type=Connector,Name="Leg Primary"), SkipSocket=true).Result
    ExecuteContext=@RigVMFunction_Sequence.A
   )
   > RigVMVariableNode(Value=@"Is Right Side v2".Right Side) {variable=Right Side}
   Connect to Module Metadata_2_1 = RigVMFunctionReferenceNode(Connector=(Type=Connector,Name="Foot Bone"), ExecuteContext=(), Module Metadata Name=Parent Control, Module Metadata NameSpace=Root)
  ]}
 New Function_ContainedGraph {nodes [
   n1 = RigVMFunctionEntryNode()
   Get_Module_Metadata_4 = RigVMDispatchNode(Default=(X=1,Y=0,Z=0), Name=Mirror Axis, NameSpace=Root) {template="DISPATCH_RigDispatch_GetModuleMetadata(in Name,in NameSpace,in Default,out Value,out Found)"}
   Not_Equals_1 = RigVMDispatchNode(B=0, A=@Get_Module_Metadata_4.Value.Y) {template="DISPATCH_RigVMDispatch_CoreNotEquals(in A,in B,out Result)"}
   Not_Equals_2 = RigVMDispatchNode(B=0, A=@Get_Module_Metadata_4.Value.Z) {template="DISPATCH_RigVMDispatch_CoreNotEquals(in A,in B,out Result)"}
   Not_Equals = RigVMDispatchNode(B=0, A=@Get_Module_Metadata_4.Value.X) {template="DISPATCH_RigVMDispatch_CoreNotEquals(in A,in B,out Result)"}
   Dot___1 = &RigVMFunction_MathVectorDot(
    A=(X=0,Y=0,Z=0)
    B=(X=1,Y=1,Z=0)
    B.Y=&RigVMFunction_MathBoolToFloat(Value=@Not_Equals_1.Result).Result
    B.Z=&RigVMFunction_MathBoolToFloat(Value=@Not_Equals_2.Result).Result
    B.X=&RigVMFunction_MathBoolToFloat(Value=@Not_Equals.Result).Result
    A.X=RigVMDispatchNode(False=0, True=&RigUnit_GetTransform(bInitial=true, Item=@n1.Item).Transform.Translation.X, Condition=@Not_Equals.Result) {template="DISPATCH_RigVMDispatch_If(in Condition,in True,in False,out Result)"}.Result
    A.Y=RigVMDispatchNode(False=0, True=&RigUnit_GetTransform(bInitial=true, Item=@n1.Item).Transform.Translation.Y, Condition=@Not_Equals_1.Result) {template="DISPATCH_RigVMDispatch_If(in Condition,in True,in False,out Result)"}.Result
    A.Z=RigVMDispatchNode(False=0, True=&RigUnit_GetTransform(bInitial=true, Item=@n1.Item).Transform.Translation.Z, Condition=@Not_Equals_2.Result) {template="DISPATCH_RigVMDispatch_If(in Condition,in True,in False,out Result)"}.Result
   ) {template="Dot,|::Execute(in A,in B,out Result)"}
   Greater = &RigVMFunction_MathDoubleGreater(B=0, A=@Get_Module_Metadata_4.Value.X) {template="Greater::Execute(in A,in B,out Result)"}
   RigVMFunction_MathBoolOr = RigVMAggregateNode(
    A=@Greater.Result
    B=&RigVMFunction_MathDoubleGreater(B=0, A=@Get_Module_Metadata_4.Value.Y) {template="Greater::Execute(in A,in B,out Result)"}.Result
    C=&RigVMFunction_MathDoubleGreater(B=0, A=@Get_Module_Metadata_4.Value.Z) {template="Greater::Execute(in A,in B,out Result)"}.Result
   )
   RigVMFunction_MathBoolNot = &RigVMFunction_MathBoolNot(Value=RigVMDispatchNode(
     True=&RigVMFunction_MathDoubleGreater(B=0, A=@Dot___1.Result) {template="Greater::Execute(in A,in B,out Result)"}.Result
     False=&RigVMFunction_MathDoubleLess(B=0, A=@Dot___1.Result) {template="Less::Execute(in A,in B,out Result)"}.Result
     Condition=@RigVMFunction_MathBoolOr.Result
    ) {template="DISPATCH_RigVMDispatch_If(in Condition,in True,in False,out Result)"}.Result)
   > RigVMFunctionReturnNode(Result=@RigVMFunction_MathBoolNot.Result)
  ]}
 New Function_ContainedGraph {nodes [
   n1 = RigVMFunctionEntryNode()
   > n2 = RigVMDispatchNode(Array=@n1.Items) {template="DISPATCH_RigVMDispatch_ArrayIterator(in Array,out Element,out Index,out Count,out Ratio)"}
   > FindClosestBone = RigVMFunctionReferenceNode(Initial=true, Item=@n2.Element, Bones=@n1.Bones)
   > Set_Metadata = RigVMDispatchNode(Name=Closest Bone, NameSpace=Self, Item=@n2.Element, Value=@FindClosestBone.Value) {template="DISPATCH_RigDispatch_SetMetadata(in Item,in Name,in NameSpace,in Value,out Success)"}
   RigVMFunctionReturnNode(ExecuteContext=@n2.Completed)
  ]}
 New Function_ContainedGraph {nodes [
   n1 = RigVMFunctionEntryNode()
   Get_Module_Metadata_4 = RigVMDispatchNode(Default=(X=0,Y=0,Z=0), Name=Mirror Axis, NameSpace=Root) {template="DISPATCH_RigDispatch_GetModuleMetadata(in Name,in NameSpace,in Default,out Value,out Found)"}
   n2 = /Script/RigVM.RigVMFunction_MathVectorMul(A=&RigUnit_GetTransform(bInitial=False, Item=@n1.Item).Transform.Translation, B=@Get_Module_Metadata_4.Value) {is=struct RigVMUnitNode, template="Multiply::Execute(in A,in B,out Result)"}
   RigVMFunction_MathBoolOr = RigVMAggregateNode(
    B=&RigVMFunction_MathDoubleGreater(B=0, A=@n2.Result.Y) {template="Greater::Execute(in A,in B,out Result)"}.Result
    C=&RigVMFunction_MathDoubleGreater(B=0, A=@n2.Result.Z) {template="Greater::Execute(in A,in B,out Result)"}.Result
    A=&RigVMFunction_MathDoubleGreater(B=0, A=@n2.Result.X) {template="Greater::Execute(in A,in B,out Result)"}.Result
   )
   RigVMFunction_MathBoolNot = &RigVMFunction_MathBoolNot(Value=@RigVMFunction_MathBoolOr.Result)
   Less = &RigVMFunction_MathDoubleLess(B=0, A=@Get_Module_Metadata_4.Value.X) {template="Less::Execute(in A,in B,out Result)"}
   RigVMFunction_MathBoolOr_1 = RigVMAggregateNode(
    A=@Less.Result
    B=&RigVMFunction_MathDoubleLess(B=0, A=@Get_Module_Metadata_4.Value.Y) {template="Less::Execute(in A,in B,out Result)"}.Result
    C=&RigVMFunction_MathDoubleLess(B=0, A=@Get_Module_Metadata_4.Value.Z) {template="Less::Execute(in A,in B,out Result)"}.Result
   )
   > RigVMFunctionReturnNode(Result=RigVMDispatchNode(True=@RigVMFunction_MathBoolNot.Result, False=@RigVMFunction_MathBoolOr.Result, Condition=@RigVMFunction_MathBoolOr_1.Result) {template="DISPATCH_RigVMDispatch_If(in Condition,in True,in False,out Result)"}.Result)
  ]}
 New Function_ContainedGraph {nodes [
   n1 = RigVMFunctionEntryNode()
   > RigVMVariableNode(Value=@n1.Module Settings) {variable=Module Settings Constr}
   > RigVMFunction_Sequence_1 = RigVMAggregateNode()
   RigVMFunction_Sequence = RigVMAggregateNode(ExecuteContext=@RigVMFunction_Sequence_1.A)
   /Script/ControlRig.RigUnit_SetMetadataTag(
    NameSpace=Self
    Tag=Bones
    Item=&RigUnit_ResolveConnector(SkipSocket=true, Connector=RigVMVariableNode() {variable=End Bone Connector}.Value).Result
    ExecutePin=@RigVMFunction_Sequence.B
   ) {is=struct RigVMUnitNode}
   > &RigUnit_SetMetadataTag(
    NameSpace=Self
    Tag=Bones
    Item=&RigUnit_ResolveConnector(SkipSocket=true, Connector=RigVMVariableNode() {variable=Middle Bone Connector}.Value).Result
   )
   > &RigUnit_SetMetadataTag(
    NameSpace=Self
    Tag=Bones
    Item=&RigUnit_ResolveConnector(SkipSocket=true, Connector=RigVMVariableNode() {variable=Start Bone Connector}.Value).Result
   )
   Get_Module_Metadata_3 = RigVMDispatchNode(Default=(X=0,Y=0,Z=0), Name=Mirror Axis, NameSpace=Root) {template="DISPATCH_RigDispatch_GetModuleMetadata(in Name,in NameSpace,in Default,out Value,out Found)"}
   > Is Right Side v2 = RigVMFunctionReferenceNode(
    Mirror Plane=@Get_Module_Metadata_3.Value
    Item=&RigUnit_ResolveConnector(Connector=(Type=Connector,Name="Foot Bone"), SkipSocket=true).Result
   )
   RigVMFunction_MathBoolNot = &RigVMFunction_MathBoolNot(Value=@"Is Right Side v2".Right Side)
   n2 = RigVMVariableNode() {variable=Module Settings Constr}
   > RigVMVariableNode(Value=RigVMDispatchNode(
     Condition=@RigVMFunction_MathBoolNot.Result
     True=@n2.Value.LeftSideSuffix_23_#1
     False=@n2.Value.RightSideSuffix_24_#2
    ) {template="DISPATCH_RigVMDispatch_If(in Condition,in True,in False,out Result)"}.Result) {variable=Current Side}
   > Set_Module_Metadata = RigVMDispatchNode(Name=Is Right Side, NameSpace=Self, Value=@"Is Right Side v2".Right Side) {template="DISPATCH_RigDispatch_SetModuleMetadata(in Name,in NameSpace,in Value,out Success)"}
   Concat = &RigVMFunction_NameConcat(B=RigVMVariableNode() {variable=Current Side}.Value, A=RigVMVariableNode() {variable=Basename}.Value) {template="Concat::Execute(in A,in B,out Result)"}
   n3 = &RigUnit_FindItemsWithMetadataTag(NameSpace=Self, Tag=Bones)
   n4 = RigVMDispatchNode(Condition=False, False=(X=1,Y=1,Z=1), True=(X=-1,Y=-1,Z=-1)) {template="DISPATCH_RigVMDispatch_If(in Condition,in True,in False,out Result)"}
   Multiply = &RigVMFunction_MathVectorMul(B=@n4.Result, A=RigVMVariableNode() {variable=Primary Axis}.Value) {template="Multiply::Execute(in A,in B,out Result)"}
   n5 = &RigVMFunction_MathVectorMul(B=@n4.Result, A=RigVMVariableNode() {variable=Secondary Axis}.Value) {template="Multiply::Execute(in A,in B,out Result)"}
   > Create Bone Chain = RigVMFunctionReferenceNode(
    Debug=false
    Debug Scale=15
    Debug Thickness=0.5
    Enable Orient Bones=true
    Kind=Location
    Optional Suffix=rig
    Name=@Concat.Result
    Transforms=/Script/ControlRig.RigUnit_GetTransformItemArray(Space=GlobalSpace, bInitial=False, Items=@n3.Items) {is=struct RigVMUnitNode}.Transforms
    Parent=&RigUnit_HierarchyGetParent(
     bDefaultParent=True
     Child=RigVMDispatchNode(Index=0, Array=@n3.Items) {template="DISPATCH_RigVMDispatch_ArrayGetAtIndex(in Array,in Index,out Element)"}.Element
    ).Parent
    Primary Axis=@Multiply.Result
    Secondary Axis=@n5.Result
   )
   n6 = RigVMDispatchNode(Index=1, Array=@"Create Bone Chain".Bones) {template="DISPATCH_RigVMDispatch_ArrayGetAtIndex(in Array,in Index,out Element)"}
   Add = &RigVMFunction_MathVectorAdd(
    A=&RigUnit_GetTransform(bInitial=true, Item=@n6.Element, Space=RigVMVariableNode() {variable=Offset Elbow Bone Space}.Value).Transform.Translation
    B=RigVMVariableNode() {variable=Offset Elbow Bone}.Value
   ) {template="Add::Execute(in A,in B,out Result)"}
   > Set Transform_2 = /Script/ControlRig.RigUnit_SetTranslation(
    Weight=1
    bInitial=true
    bPropagateToChildren=false
    Value=@Add.Result
    Item=@n6.Element
    Space=RigVMVariableNode() {variable=Offset Elbow Bone Space}.Value
   ) {is=struct RigVMUnitNode, template="Set Transform::Execute(in Item,in Space,in bInitial,in Value,in Weight,in bPropagateToChildren)"}
   > Orient Bones_0 = RigVMFunctionReferenceNode(Debug=false, Kind=Location, Bones=@"Create Bone Chain".Bones, Primary Axis=@Multiply.Result, Secondary Axis=@n5.Result)
   > RigVMDispatchNode(Name=Rig Bones, NameSpace=Self, Value=@"Create Bone Chain".Bones) {template="DISPATCH_RigDispatch_SetModuleMetadata(in Name,in NameSpace,in Value,out Success)"}
   > /Script/ControlRig.RigUnit_SetTransformItemArray(
    Space=GlobalSpace
    Weight=1
    bInitial=true
    bPropagateToChildren=True
    Transforms=&RigUnit_GetTransformItemArray(Space=GlobalSpace, bInitial=true, Items=@"Create Bone Chain".Bones).Transforms
    Items=@"Create Bone Chain".Bones
   ) {is=struct RigVMUnitNode}
   n7 = &RigUnit_ResolveConnector(SkipSocket=true, Connector=RigVMVariableNode() {variable=Start Bone Connector}.Value)
   > Create Bone Chain_1 = RigVMFunctionReferenceNode(
    Debug=true
    Debug Scale=15
    Debug Thickness=0.5
    Enable Orient Bones=true
    Kind=Location
    Optional Suffix=aimFollow
    Transforms=((Rotation=(X=0,Y=0,Z=0,W=1),Translation=(X=0,Y=0,Z=0),Scale3D=(X=1,Y=1,Z=1)))
    Transforms.0=&RigUnit_GetTransform(bInitial=true, Item=@n7.Result).Transform
    Parent=&RigUnit_HierarchyGetParent(bDefaultParent=True, Child=@n7.Result).Parent
    Name=RigVMVariableNode() {variable=Basename}.Value
    Primary Axis=RigVMVariableNode() {variable=Primary Axis}.Value
    Secondary Axis=RigVMVariableNode() {variable=Secondary Axis}.Value
   )
   > RigVMDispatchNode(Name=Aim Follow Bones, NameSpace=Self, Value=@"Create Bone Chain_1".Bones) {template="DISPATCH_RigDispatch_SetModuleMetadata(in Name,in NameSpace,in Value,out Success)"}
   RigVMFunction_Sequence_2 = RigVMAggregateNode(ExecuteContext=@RigVMFunction_Sequence.D)
   RigVMVariableNode(
    ExecuteContext=@RigVMFunction_Sequence_2.A
    Value=RigVMVariableNode() {variable=Module Settings}.Value.Prefix_20_#3
   ) {variable=name prefix}
   Format Name_1 = RigVMFunctionReferenceNode(
    suffix=None
    prefix=RigVMVariableNode() {variable=name prefix}.Value
    name=RigVMVariableNode() {variable=IK Control Basename}.Value
    side=RigVMVariableNode() {variable=Current Side}.Value
   )
   Control Stack at Item = RigVMFunctionReferenceNode(
    Add Bottom Null=False
    Add Null=true
    Add Secondary Control=False
    Control Suffix=_ctrl
    Null Suffix=_null
    Secondary Settings="(DisplayName=\"None\",Shape=(bVisible=True,Name=\"Default\",Color=(R=1,G=0,B=0,A=1),Transform=(Rotation=(X=0,Y=0,Z=0,W=1),Translation=(X=0,Y=0,Z=0),Scale3D=(X=1,Y=1,Z=1))),Proxy=(bIsProxy=False,ShapeVisibility=BasedOnSelection),InitialSpace=LocalSpace,bUsePref ...(truncated)"
    Tag Controls=true
    Zero Controls=true
    Name=@"Format Name_1".Result
    Parent=&RigUnit_ResolveConnector(SkipSocket=true, Connector=RigVMVariableNode() {variable=Parent Control Connector}.Value).Result
    Item=&RigUnit_ResolveConnector(SkipSocket=true, Connector=RigVMVariableNode() {variable=End Bone Connector}.Value).Result
    ExecuteContext=@RigVMFunction_Sequence_2.C
    Settings=RigVMVariableNode() {variable=IK Control Settings}.Value
    Orient to World=RigVMVariableNode() {variable=Orient to World}.Value
   )
   > Add Null Below = RigVMFunctionReferenceNode(
    New Suffix=_effector
    Replace Suffix=_null
    Transform=(Rotation=(X=0,Y=0,Z=0,W=1),Translation=(X=0,Y=0,Z=0),Scale3D=(X=1,Y=1,Z=1))
    Use Transform=False
    Item=@"Control Stack at Item".Control
   )
   > Add Null Below_1 = RigVMFunctionReferenceNode(
    New Suffix=_oriNull
    Replace Suffix=_oriNull
    Transform=(Rotation=(X=0,Y=0,Z=0,W=1),Translation=(X=0,Y=0,Z=0),Scale3D=(X=1,Y=1,Z=1))
    Use Transform=false
    Item=@"Control Stack at Item".Control
   )
   n8 = &RigVMFunction_ControlFlowBranch(ExecuteContext=@"Add Null Below_1".ExecuteContext, Condition=RigVMVariableNode() {variable=Enable Keep Aligned}.Value)
   RigVMDispatchNode(Name=Effector Ori, NameSpace=Self, Value=@"Add Null Below_1".Null, ExecuteContext=@n8.Completed) {template="DISPATCH_RigDispatch_SetModuleMetadata(in Name,in NameSpace,in Value,out Success)"}
   > /Script/ControlRig.RigUnit_SetDefaultParent(Child=@"Add Null Below_1".Null, Parent=@"Control Stack at Item".Null) {is=struct RigVMUnitNode}
   > &RigUnit_SetDefaultParent(Child=(Type=Control,Name="None"), Parent=@"Control Stack at Item".Null)
   > &RigUnit_SetMetadataTag(NameSpace=Self, Tag=Controls IK, Item=@"Control Stack at Item".Control)
   > RigVMDispatchNode(Name=IK Control, NameSpace=Self, Value=@"Control Stack at Item".Control) {template="DISPATCH_RigDispatch_SetModuleMetadata(in Name,in NameSpace,in Value,out Success)"}
   > RigVMDispatchNode(Name=Effector Null, NameSpace=Self, Value=@"Add Null Below".Null) {template="DISPATCH_RigDispatch_SetModuleMetadata(in Name,in NameSpace,in Value,out Success)"}
   > &RigUnit_SetTransform(
    Weight=1
    bInitial=true
    Value=&RigUnit_GetTransform(bInitial=true, Item=@"Control Stack at Item".Control).Transform
    Item=@"Control Stack at Item".Control
   )
   Format Name_1_1 = RigVMFunctionReferenceNode(
    suffix=_fk_ik_switch
    name=RigVMVariableNode() {variable=Basename}.Value
    prefix=RigVMVariableNode() {variable=name prefix}.Value
    side=RigVMVariableNode() {variable=Current Side}.Value
   )
   > SpawnAnimationChannel = /Script/ControlRig.RigUnit_HierarchyAddAnimationChannelBool(
    LimitsEnabled=()
    MaximumValue=True
    MinimumValue=False
    Name=@"Format Name_1_1".Result
    Parent=&RigUnit_ResolveConnector(SkipSocket=true, Connector=RigVMVariableNode() {variable=Settings Control Connector}.Value).Result
    InitialValue=RigVMVariableNode() {variable=FK IK Default}.Value
   ) {
    is=struct RigVMUnitNode
    template="SpawnAnimationChannel::Execute(in InitialValue,in MinimumValue,in MaximumValue,in LimitsEnabled,in Parent,in Name,out Item)"
   }
   > DISPATCH_RigDispatch_SetModuleMetadata = RigVMDispatchNode(Name=fk_ik_switch, NameSpace=Self, Value=@SpawnAnimationChannel.Item.Name) {template="DISPATCH_RigDispatch_SetModuleMetadata(in Name,in NameSpace,in Value,out Success)"}
   n9 = &RigVMFunction_ControlFlowBranch(ExecuteContext=@DISPATCH_RigDispatch_SetModuleMetadata.ExecuteContext, Condition=RigVMVariableNode() {variable=Enable Softness}.Value)
   n10 = &RigUnit_HierarchyAddAnimationChannelBool(
    InitialValue=false
    LimitsEnabled=()
    MaximumValue=True
    MinimumValue=False
    Name=Stretch
    Parent=RigVMDispatchNode(Index=0, Array=&RigUnit_FindItemsWithMetadataTag(NameSpace=Self, Tag=Controls IK).Items) {template="DISPATCH_RigVMDispatch_ArrayGetAtIndex(in Array,in Index,out Element)"}.Element
    ExecutePin=@n9.Completed
   ) {template="SpawnAnimationChannel::Execute(in InitialValue,in MinimumValue,in MaximumValue,in LimitsEnabled,in Parent,in Name,out Item)"}
   n11 = &RigVMFunction_ControlFlowBranch(ExecuteContext=@n10.ExecutePin, Condition=RigVMVariableNode() {variable=Enable PV Follow}.Value)
   RigVMFunction_ControlFlowBranch = &RigVMFunction_ControlFlowBranch(ExecuteContext=@n11.Completed, Condition=RigVMVariableNode() {variable=Enable Keep Aligned}.Value)
   &RigUnit_HierarchyAddAnimationChannelBool(
    InitialValue=false
    LimitsEnabled=()
    MaximumValue=True
    MinimumValue=False
    Name=Keep Aligned
    Parent=RigVMDispatchNode(Index=0, Array=&RigUnit_FindItemsWithMetadataTag(NameSpace=Self, Tag=Controls IK).Items) {template="DISPATCH_RigVMDispatch_ArrayGetAtIndex(in Array,in Index,out Element)"}.Element
    ExecutePin=@RigVMFunction_ControlFlowBranch.True
   ) {template="SpawnAnimationChannel::Execute(in InitialValue,in MinimumValue,in MaximumValue,in LimitsEnabled,in Parent,in Name,out Item)"}
   &RigUnit_HierarchyAddAnimationChannelBool(
    InitialValue=true
    LimitsEnabled=()
    MaximumValue=True
    MinimumValue=False
    Name=PV Follow
    Parent=RigVMDispatchNode(Index=0, Array=&RigUnit_FindItemsWithMetadataTag(NameSpace=Self, Tag=Controls IK).Items) {template="DISPATCH_RigVMDispatch_ArrayGetAtIndex(in Array,in Index,out Element)"}.Element
    ExecutePin=@n11.True
   ) {template="SpawnAnimationChannel::Execute(in InitialValue,in MinimumValue,in MaximumValue,in LimitsEnabled,in Parent,in Name,out Item)"}
   /Script/ControlRig.RigUnit_HierarchyAddAnimationChannelFloat(
    InitialValue=0
    LimitsEnabled=(Enabled=(bMinimum=True,bMaximum=True))
    MaximumValue=2
    MinimumValue=0
    Name=Softness
    Parent=RigVMDispatchNode(Index=0, Array=&RigUnit_FindItemsWithMetadataTag(NameSpace=Self, Tag=Controls IK).Items) {template="DISPATCH_RigVMDispatch_ArrayGetAtIndex(in Array,in Index,out Element)"}.Element
    ExecutePin=@n9.True
   ) {
    is=struct RigVMUnitNode
    template="SpawnAnimationChannel::Execute(in InitialValue,in MinimumValue,in MaximumValue,in LimitsEnabled,in Parent,in Name,out Item)"
   }
   Format Name_1_2 = RigVMFunctionReferenceNode(
    suffix=_ankle
    prefix=RigVMVariableNode() {variable=name prefix}.Value
    name=RigVMVariableNode() {variable=IK Control Basename}.Value
    side=RigVMVariableNode() {variable=Current Side}.Value
   )
   Control Stack at Item_2 = RigVMFunctionReferenceNode(
    Add Bottom Null=False
    Add Null=true
    Add Secondary Control=False
    Control Suffix=_ctrl
    Null Suffix=_null
    Secondary Settings="(DisplayName=\"None\",Shape=(bVisible=True,Name=\"Default\",Color=(R=1,G=0,B=0,A=1),Transform=(Rotation=(X=0,Y=0,Z=0,W=1),Translation=(X=0,Y=0,Z=0),Scale3D=(X=1,Y=1,Z=1))),Proxy=(bIsProxy=False,ShapeVisibility=BasedOnSelection),InitialSpace=LocalSpace,bUsePref ...(truncated)"
    Tag Controls=true
    Zero Controls=true
    Item=@"Control Stack at Item".Control
    Parent=@"Control Stack at Item".Null
    Settings=@n1.Wrist Control Settings
    Name=@"Format Name_1_2".Result
    ExecuteContext=@n8.True
   )
   > RigVMDispatchNode(Name=Aligned Null, NameSpace=Self, Value=@"Control Stack at Item_2".Control) {template="DISPATCH_RigDispatch_SetModuleMetadata(in Name,in NameSpace,in Value,out Success)"}
   For_Each_1_1 = RigVMDispatchNode(ExecuteContext=@RigVMFunction_Sequence_2.F, Array=RigVMVariableNode() {variable=IK Control Spaces}.Value) {template="DISPATCH_RigVMDispatch_ArrayIterator(in Array,out Element,out Index,out Count,out Ratio)"}
   Get_Module_Metadata = RigVMDispatchNode(Default=(Type=Bone,Name="None"), Name=IK Control, NameSpace=Self) {template="DISPATCH_RigDispatch_GetModuleMetadata(in Name,in NameSpace,in Default,out Value,out Found)"}
   > /Script/ControlRig.RigUnit_AddParent(DisplayLabel=None, Parent=@For_Each_1_1.Element, Child=@Get_Module_Metadata.Value) {is=struct RigVMUnitNode}
   For_Each_1 = RigVMDispatchNode(ExecuteContext=@For_Each_1_1.Completed, Array=RigVMVariableNode() {variable=PV Control Spaces}.Value) {template="DISPATCH_RigVMDispatch_ArrayIterator(in Array,out Element,out Index,out Count,out Ratio)"}
   Get_Module_Metadata_1 = RigVMDispatchNode(Default=(Type=Bone,Name="None"), Name=PV Control, NameSpace=Self) {template="DISPATCH_RigDispatch_GetModuleMetadata(in Name,in NameSpace,in Default,out Value,out Found)"}
   > &RigUnit_AddParent(DisplayLabel=None, Parent=@For_Each_1.Element, Child=@Get_Module_Metadata_1.Value)
   n12 = RigVMDispatchNode(
    Array=RigVMDispatchNode(Default=(), Name=Rig Bones, NameSpace=Self) {template="DISPATCH_RigDispatch_GetModuleMetadata(in Name,in NameSpace,in Default,out Value,out Found)"}.Value
    ExecuteContext=@RigVMFunction_Sequence_2.E
   ) {template="DISPATCH_RigVMDispatch_ArrayIterator(in Array,out Element,out Index,out Count,out Ratio)"}
   n13 = RigVMDispatchNode(Index=@n12.Index, Array=&RigUnit_FindItemsWithMetadataTag(NameSpace=Self, Tag=Bones).Items) {template="DISPATCH_RigVMDispatch_ArrayGetAtIndex(in Array,in Index,out Element)"}
   DISPATCH_RigVMDispatch_If = RigVMDispatchNode(
    False=False
    True=true
    Condition=RigVMDispatchNode(B=0, A=@n12.Index) {template="DISPATCH_RigVMDispatch_CoreEquals(in A,in B,out Result)"}.Result
   ) {template="DISPATCH_RigVMDispatch_If(in Condition,in True,in False,out Result)"}
   Concat_3 = RigVMAggregateNode(B=_fk, C=_ctrl, A=&RigUnit_HierarchyGetParent(bDefaultParent=True, Child=@n13.Element).Parent.Name)
   > Control Stack at Item_1 = RigVMFunctionReferenceNode(
    Add Bottom Null=False
    Add Secondary Control=False
    Control Suffix=_ctrl
    Secondary Settings="(DisplayName=\"None\",Shape=(bVisible=True,Name=\"Default\",Color=(R=1,G=0,B=0,A=1),Transform=(Rotation=(X=0,Y=0,Z=0,W=1),Translation=(X=0,Y=0,Z=0),Scale3D=(X=1,Y=1,Z=1))),Proxy=(bIsProxy=False,ShapeVisibility=BasedOnSelection),InitialSpace=LocalSpace,bUsePref ...(truncated)"
    Tag Controls=true
    Zero Controls=true
    Name=&RigVMFunction_NameConcat(B=_fk, A=@n13.Element.Name) {template="Concat::Execute(in A,in B,out Result)"}.Result
    Item=@n12.Element
    Add Null=@DISPATCH_RigVMDispatch_If.Result
    Parent=RigVMDispatchNode(
     False=(Type=Control,Name="None")
     False.Name=@Concat_3.Result
     Condition=RigVMDispatchNode(B=0, A=@n12.Index) {template="DISPATCH_RigVMDispatch_CoreEquals(in A,in B,out Result)"}.Result
     True=&RigUnit_ResolveConnector(SkipSocket=true, Connector=@n1.Parent Control Connector).Result
    ) {template="DISPATCH_RigVMDispatch_If(in Condition,in True,in False,out Result)"}.Result
    Settings=RigVMVariableNode() {variable=FK Control Settings}.Value
   )
   > &RigUnit_SetTransform(
    Weight=1
    bInitial=true
    Item=@"Control Stack at Item_1".Control
    Value=&RigUnit_GetTransform(bInitial=true, Item=@"Control Stack at Item_1".Control).Transform
   )
   > &RigUnit_SetMetadataTag(NameSpace=Self, Tag=Controls FK, Item=@"Control Stack at Item_1".Control)
   > RigVMDispatchNode(Name=FK Parent, NameSpace=Self, Value=@"Control Stack at Item_1".Control) {template="DISPATCH_RigDispatch_SetModuleMetadata(in Name,in NameSpace,in Value,out Success)"}
   &RigUnit_HierarchyAddAnimationChannelBool(
    LimitsEnabled=()
    MaximumValue=True
    MinimumValue=False
    Name=local
    Parent=RigVMDispatchNode(Index=0, Array=&RigUnit_FindItemsWithMetadataTag(NameSpace=Self, Tag=Controls FK).Items) {template="DISPATCH_RigVMDispatch_ArrayGetAtIndex(in Array,in Index,out Element)"}.Element
    InitialValue=RigVMVariableNode() {variable=Local Default}.Value
    ExecutePin=@n12.Completed
   ) {template="SpawnAnimationChannel::Execute(in InitialValue,in MinimumValue,in MaximumValue,in LimitsEnabled,in Parent,in Name,out Item)"}
   DISPATCH_RigDispatch_GetModuleMetadata = RigVMDispatchNode(Default=(), Name=Rig Bones, NameSpace=Self) {template="DISPATCH_RigDispatch_GetModuleMetadata(in Name,in NameSpace,in Default,out Value,out Found)"}
   DISPATCH_RigVMDispatch_ArrayGetAtIndex = RigVMDispatchNode(Index=0, Array=@DISPATCH_RigDispatch_GetModuleMetadata.Value) {template="DISPATCH_RigVMDispatch_ArrayGetAtIndex(in Array,in Index,out Element)"}
   ComputePoleVector_1 = RigVMFunctionReferenceNode(
    Draw Transform=false
    Initial=true
    Origin Scale=(X=0,Y=0,Z=0)
    Bone A=@DISPATCH_RigVMDispatch_ArrayGetAtIndex.Element
    Bone B=RigVMDispatchNode(Index=1, Array=@DISPATCH_RigDispatch_GetModuleMetadata.Value) {template="DISPATCH_RigVMDispatch_ArrayGetAtIndex(in Array,in Index,out Element)"}.Element
    Bone C=RigVMDispatchNode(Index=2, Array=@DISPATCH_RigDispatch_GetModuleMetadata.Value) {template="DISPATCH_RigVMDispatch_ArrayGetAtIndex(in Array,in Index,out Element)"}.Element
    OffsetFactor=RigVMVariableNode() {variable=PV Offset Factor}.Value
    ExecuteContext=@RigVMFunction_Sequence_2.D
   )
   Format Name = RigVMFunctionReferenceNode(
    suffix=_pv_ik
    prefix=RigVMVariableNode() {variable=name prefix}.Value
    name=RigVMVariableNode() {variable=Basename}.Value
    side=RigVMVariableNode() {variable=Current Side}.Value
   )
   > Control Stack at Position = RigVMFunctionReferenceNode(
    Add Bottom Null=False
    Add Null=true
    Add Secondary Control=False
    Control Suffix=_ctrl
    Null Suffix=_null
    Secondary Settings="(DisplayName=\"None\",Shape=(bVisible=True,Name=\"Default\",Color=(R=1,G=0,B=0,A=1),Transform=(Rotation=(X=0,Y=0,Z=0,W=1),Translation=(X=0,Y=0,Z=0),Scale3D=(X=1,Y=1,Z=1))),Proxy=(bIsProxy=False,ShapeVisibility=BasedOnSelection),InitialSpace=LocalSpace,bUsePref ...(truncated)"
    Tag Controls=true
    Zero Controls=true
    Global Position=@ComputePoleVector_1.Transform
    Name=@"Format Name".Result
    Settings=RigVMVariableNode() {variable=PV Control Settings}.Value
    Parent=RigVMVariableNode() {variable=Pole Vector Parent Connector}.Value
   )
   > &RigUnit_SetMetadataTag(NameSpace=Self, Tag=Controls IK, Item=@"Control Stack at Position".Control)
   > &RigUnit_SetMetadataTag(NameSpace=Self, Tag=PV Controls, Item=@"Control Stack at Position".Control)
   > RigVMDispatchNode(Name=PV Control, NameSpace=Self, Value=@"Control Stack at Position".Control) {template="DISPATCH_RigDispatch_SetModuleMetadata(in Name,in NameSpace,in Value,out Success)"}
   Finalize Controls = RigVMCollapseNode(ExecuteContext=@RigVMFunction_Sequence.E, Module Settings=RigVMVariableNode() {variable=Module Settings}.Value)
   Set Closest Bone = RigVMFunctionReferenceNode(
    Bones=&RigUnit_FindItemsWithMetadataTag(NameSpace=Self, Tag=Bones).Items
    Items=&RigUnit_FindItemsWithMetadataTag(NameSpace=Self, Tag=PV Controls).Items
    ExecuteContext=@RigVMFunction_Sequence_1.B
   )
   RigVMFunctionReturnNode(ExecuteContext=@RigVMFunction_Sequence_1.C, Current Side Name=RigVMVariableNode() {variable=Current Side}.Value)
   Int to Name_1 = /Script/RigVM.RigVMFunction_MathIntToName(Number=0, PaddedSize=2) {is=struct RigVMUnitNode, template="Int to Name::Execute(in Number,in PaddedSize,out Result)"}
   Concat_5_1 = RigVMAggregateNode(A=spine_, C=_ctrl, B=@"Int to Name_1".Result)
   RigVMDispatchNode(
    False=(Type=Control,Name="None")
    True=(Type=None,Name="None")
    False.Name=@Concat_5_1.Result
    Condition=RigVMDispatchNode(A=0, B=0) {template="DISPATCH_RigVMDispatch_CoreEquals(in A,in B,out Result)"}.Result
   ) {template="DISPATCH_RigVMDispatch_If(in Condition,in True,in False,out Result)"}
   &RigUnit_FindItemsWithMetadataTag(NameSpace=Self, Tag=Bones)
   RigVMRerouteNode(Value=@"Control Stack at Position".Control)
  ]}
 New Function_ContainedGraph {nodes [
   n1 = RigVMFunctionEntryNode()
   n2 = &RigUnit_FindItemsWithMetadataTag(NameSpace=Self, Tag=Controls)
   > Scale Module Controls = RigVMFunctionReferenceNode(
    Global Scale=RigVMDispatchNode(Default=1, Name=Control Size, NameSpace=Root) {template="DISPATCH_RigDispatch_GetModuleMetadata(in Name,in NameSpace,in Default,out Value,out Found)"}.Value
    Local Scale=@n1.Module Settings.ControlSize_12_#4
    Controls=@n2.Items
   )
   > Color Module Controls = RigVMFunctionReferenceNode(
    CenterColor=@n1.Module Settings.CenterColor_9_#5
    LeftColor=@n1.Module Settings.LeftColor_3_#6
    Override=@n1.Module Settings.OverrideColor_48_#7
    RightColor=@n1.Module Settings.RightColor_5_#8
    Controls=@n2.Items
   )
   > RigVMFunctionReturnNode()
   Get_Module_Metadata_4 = RigVMDispatchNode(Default=(X=1,Y=0,Z=0), Name=Mirror Axis, NameSpace=Root) {template="DISPATCH_RigDispatch_GetModuleMetadata(in Name,in NameSpace,in Default,out Value,out Found)"}
   Greater = &RigVMFunction_MathDoubleGreater(B=0, A=@Get_Module_Metadata_4.Value.X) {template="Greater::Execute(in A,in B,out Result)"}
   RigVMDispatchNode(False=False, True=true, Condition=@Greater.Result) {template="DISPATCH_RigVMDispatch_If(in Condition,in True,in False,out Result)"}
   RigVMDispatchNode(
    False=False
    True=true
    Condition=&RigVMFunction_MathDoubleGreater(B=0, A=@Get_Module_Metadata_4.Value.Y) {template="Greater::Execute(in A,in B,out Result)"}.Result
   ) {template="DISPATCH_RigVMDispatch_If(in Condition,in True,in False,out Result)"}
   RigVMDispatchNode(
    False=False
    True=true
    Condition=&RigVMFunction_MathDoubleGreater(B=0, A=@Get_Module_Metadata_4.Value.Z) {template="Greater::Execute(in A,in B,out Result)"}.Result
   ) {template="DISPATCH_RigVMDispatch_If(in Condition,in True,in False,out Result)"}
   Less = &RigVMFunction_MathDoubleLess(B=0, A=@Get_Module_Metadata_4.Value.X) {template="Less::Execute(in A,in B,out Result)"}
   RigVMFunction_MathBoolOr = RigVMAggregateNode(
    A=@Less.Result
    B=&RigVMFunction_MathDoubleLess(B=0, A=@Get_Module_Metadata_4.Value.Y) {template="Less::Execute(in A,in B,out Result)"}.Result
    C=&RigVMFunction_MathDoubleLess(B=0, A=@Get_Module_Metadata_4.Value.Z) {template="Less::Execute(in A,in B,out Result)"}.Result
   )
  ]}
 New Function_ContainedGraph {nodes [
   n1 = RigVMFunctionEntryNode()
   Concat_4 = RigVMAggregateNode(B=@n1.name, A=@n1.prefix, C=@n1.side, D=@n1.suffix)
   > RigVMFunctionReturnNode(Result=@Concat_4.Result)
  ]}
 New Function_ContainedGraph {nodes [
   n1 = RigVMFunctionEntryNode()
   > &RigUnit_SetTransform(
    Weight=1
    bInitial=False
    Item=@n1.Child
    Value=&RigUnit_ProjectTransformToNewParent(bChildInitial=True, bNewParentInitial=False, bOldParentInitial=True, Child=@n1.Child, NewParent=@n1.New Parent, OldParent=@n1.New Parent).Transform
   )
   > RigVMFunctionReturnNode()
  ]}
 New Function_ContainedGraph {nodes [
   RigVMFunctionEntryNode()
   > RigVMFunction_Sequence = RigVMAggregateNode()
   Get_Module_Metadata = RigVMDispatchNode(Default=(Type=Bone,Name="None"), Name=IK Control, NameSpace=Self) {template="DISPATCH_RigDispatch_GetModuleMetadata(in Name,in NameSpace,in Default,out Value,out Found)"}
   SetAnimationChannel = /Script/ControlRig.RigUnit_SetBoolAnimationChannel(Channel=PV Follow, Value=False, bInitial=False, Control=@Get_Module_Metadata.Value.Name, ExecutePin=@RigVMFunction_Sequence.A) {is=struct RigVMUnitNode, template="SetAnimationChannel::Execute(in Value,in Control,in Channel,in bInitial)"}
   DISPATCH_RigVMDispatch_ArrayIterator = RigVMDispatchNode(Array=&RigUnit_FindItemsWithMetadataTag(NameSpace=Self, Tag=Controls).Items, ExecuteContext=@RigVMFunction_Sequence.B) {template="DISPATCH_RigVMDispatch_ArrayIterator(in Array,out Element,out Index,out Count,out Ratio)"}
   DISPATCH_RigDispatch_GetMetadata = RigVMDispatchNode(Default=(Type=Bone,Name="None"), Name=Item, NameSpace=Self, Item=@DISPATCH_RigVMDispatch_ArrayIterator.Element) {template="DISPATCH_RigDispatch_GetMetadata(in Item,in Name,in NameSpace,in Default,out Value,out Found)"}
   RigVMFunction_ControlFlowBranch = &RigVMFunction_ControlFlowBranch(Condition=@DISPATCH_RigDispatch_GetMetadata.Found, ExecuteContext=@DISPATCH_RigVMDispatch_ArrayIterator.ExecuteContext)
   &RigUnit_SetTransform(
    Weight=1
    bInitial=False
    Item=@DISPATCH_RigVMDispatch_ArrayIterator.Element
    Value=&RigUnit_ProjectTransformToNewParent(
     bChildInitial=True
     bNewParentInitial=False
     bOldParentInitial=True
     NewParent=@DISPATCH_RigDispatch_GetMetadata.Value
     OldParent=@DISPATCH_RigDispatch_GetMetadata.Value
     Child=@DISPATCH_RigVMDispatch_ArrayIterator.Element
    ).Transform
    ExecutePin=@RigVMFunction_ControlFlowBranch.True
   )
   n1 = RigVMDispatchNode(Array=&RigUnit_FindItemsWithMetadataTag(NameSpace=Self, Tag=PV Controls).Items, ExecuteContext=@RigVMFunction_Sequence.C) {template="DISPATCH_RigVMDispatch_ArrayIterator(in Array,out Element,out Index,out Count,out Ratio)"}
   Get_Metadata = RigVMDispatchNode(Default=(Type=Bone,Name="None"), Name=Closest Bone, NameSpace=Self, Item=@n1.Element) {template="DISPATCH_RigDispatch_GetMetadata(in Item,in Name,in NameSpace,in Default,out Value,out Found)"}
   > Parent_3 = RigVMFunctionReferenceNode(Child=@n1.Element, New Parent=@Get_Metadata.Value)
   RigVMFunctionReturnNode(ExecuteContext=@RigVMFunction_Sequence.D)
   n2 = &RigUnit_FindItemsWithMetadataTag(NameSpace=Self, Tag=Controls FK)
   Parent = RigVMFunctionReferenceNode(
    Child=RigVMDispatchNode(Index=0, Array=@n2.Items) {template="DISPATCH_RigVMDispatch_ArrayGetAtIndex(in Array,in Index,out Element)"}.Element
    New Parent=&RigUnit_ResolveConnector(Connector=(Type=Connector,Name="Shoulder Bone"), SkipSocket=False).Result
    ExecuteContext=@RigVMFunction_Sequence.E
   )
   > Parent_1 = RigVMFunctionReferenceNode(
    Child=RigVMDispatchNode(Index=1, Array=@n2.Items) {template="DISPATCH_RigVMDispatch_ArrayGetAtIndex(in Array,in Index,out Element)"}.Element
    New Parent=&RigUnit_ResolveConnector(Connector=(Type=Connector,Name="Elbow Bone"), SkipSocket=False).Result
   )
   > Parent_2 = RigVMFunctionReferenceNode(
    Child=RigVMDispatchNode(Index=2, Array=@n2.Items) {template="DISPATCH_RigVMDispatch_ArrayGetAtIndex(in Array,in Index,out Element)"}.Element
    New Parent=&RigUnit_ResolveConnector(Connector=(Type=Connector,Name="Hand Bone"), SkipSocket=False).Result
   )
   > For_Each = RigVMDispatchNode(Array=&RigUnit_FindItemsWithMetadataTag(NameSpace=Self, Tag=PV Controls).Items) {template="DISPATCH_RigVMDispatch_ArrayIterator(in Array,out Element,out Index,out Count,out Ratio)"}
   Get_Metadata_1 = RigVMDispatchNode(Default=(Type=Bone,Name="None"), Name=Closest Bone, NameSpace=Self, Item=@For_Each.Element) {template="DISPATCH_RigDispatch_GetMetadata(in Item,in Name,in NameSpace,in Default,out Value,out Found)"}
   > &RigUnit_SetTransform(
    Weight=1
    bInitial=False
    Item=@For_Each.Element
    Value=&RigUnit_GetTransform(bInitial=False, Item=@Get_Metadata_1.Value).Transform
   )
  ]}
 New Function_ContainedGraph {nodes [
   n1 = RigVMFunctionEntryNode()
   > RigVMFunction_Sequence = RigVMAggregateNode()
   n2 = &RigUnit_ResolveArrayConnector(Connector=(Type=Connector,Name="Twist Bones"), SkipSocket=true)
   Num = RigVMDispatchNode(Array=@n2.Result) {template="DISPATCH_RigVMDispatch_ArrayGetNum(in Array,out Num)"}
   Greater = /Script/RigVM.RigVMFunction_MathIntGreater(B=0, A=@Num.Num) {is=struct RigVMUnitNode, template="Greater::Execute(in A,in B,out Result)"}
   n3 = &RigVMFunction_ControlFlowBranch(Condition=@Greater.Result, ExecuteContext=@RigVMFunction_Sequence.A)
   For_Each = RigVMDispatchNode(Array=@n2.Result, ExecuteContext=@n3.True) {template="DISPATCH_RigVMDispatch_ArrayIterator(in Array,out Element,out Index,out Count,out Ratio)"}
   Get_Metadata = RigVMDispatchNode(Default=(Type=Bone,Name="None"), Name=Twist Item A, NameSpace=Self, Item=@For_Each.Element) {template="DISPATCH_RigDispatch_GetMetadata(in Item,in Name,in NameSpace,in Default,out Value,out Found)"}
   Get_Metadata_1 = RigVMDispatchNode(Default=(Type=Bone,Name="None"), Name=Twist Item B, NameSpace=Self, Item=@For_Each.Element) {template="DISPATCH_RigDispatch_GetMetadata(in Item,in Name,in NameSpace,in Default,out Value,out Found)"}
   Get_Metadata_2 = RigVMDispatchNode(Default=0, Name=Twist Amount A, NameSpace=Self, Item=@For_Each.Element) {template="DISPATCH_RigDispatch_GetMetadata(in Item,in Name,in NameSpace,in Default,out Value,out Found)"}
   Get_Metadata_3 = RigVMDispatchNode(Default=0, Name=Twist Amount B, NameSpace=Self, Item=@For_Each.Element) {template="DISPATCH_RigDispatch_GetMetadata(in Item,in Name,in NameSpace,in Default,out Value,out Found)"}
   > &RigUnit_RotationConstraintLocalSpaceOffset(
    AdvancedSettings=(InterpolationType=Shortest,RotationOrderForFilter=ZXY)
    Parents=((Item=(Type=Bone,Name="None"),Weight=1),(Item=(Type=Bone,Name="None"),Weight=1))
    Weight=1
    bMaintainOffset=true
    Filter=@n1.Filter
    Child=@For_Each.Element
    Parents.0.Item=@Get_Metadata.Value
    Parents.1.Item=@Get_Metadata_1.Value
    Parents.1.Weight=@Get_Metadata_2.Value
    Parents.0.Weight=@Get_Metadata_3.Value
   )
   > /Script/ControlRig.RigUnit_PositionConstraintLocalSpaceOffset(
    Filter=(bX=True,bY=True,bZ=True)
    Parents=((Item=(Type=Bone,Name="None"),Weight=1),(Item=(Type=Bone,Name="None"),Weight=1))
    Weight=1
    bMaintainOffset=True
    Child=@For_Each.Element
    Parents.0.Item=@Get_Metadata.Value
    Parents.1.Item=@Get_Metadata_1.Value
    Parents.1.Weight=@Get_Metadata_2.Value
    Parents.0.Weight=@Get_Metadata_3.Value
   ) {is=struct RigVMUnitNode}
   > Debug Item = RigVMFunctionReferenceNode(Scale=15, Thickness=2, Active=@n1.Active, Item=@For_Each.Element)
   RigVMFunctionReturnNode(ExecuteContext=@RigVMFunction_Sequence.C)
  ]}
 RigVMFunctionLibrary {nodes [
   CONNECT Leg = RigVMCollapseNode(ExecuteContext=())
   CONSTR Basic Twist Rig = RigVMCollapseNode(ExecuteContext=(), Template Bones=())
   CONSTR Leg = RigVMCollapseNode(ExecuteContext=())
   CONSTR TwoBoneLimb Leg = RigVMCollapseNode(
    Basename=arm
    Enable Keep Aligned=False
    Enable PV Follow=False
    Enable Softness=False
    End Bone Connector=(Type=Connector,Name="Hand")
    ExecuteContext=()
    FK Control Settings="(DisplayName=\"None\",bSelectable=True,InitialSpace=LocalSpace,bUsePreferredRotationOrder=False,PreferredRotationOrder=YZX,Limits=(LimitTranslationX=(bMinimum=False,bMaximum=False),LimitTranslationY=(bMinimum=False,bMaximum=False),LimitTranslationZ=(bMinimum ...(truncated)"
    FK IK Default=false
    IK Control Basename=hand
    IK Control Settings="(DisplayName=\"None\",bSelectable=True,InitialSpace=LocalSpace,bUsePreferredRotationOrder=False,PreferredRotationOrder=YZX,Limits=(LimitTranslationX=(bMinimum=False,bMaximum=False),LimitTranslationY=(bMinimum=False,bMaximum=False),LimitTranslationZ=(bMinimum ...(truncated)"
    IK Control Spaces=()
    Local Default=False
    Middle Bone Connector=(Type=None,Name="None")
    Module Settings="(OverrideColor_48_#7=False,LeftColor_3_#6=(R=0,G=0.0231,B=0.4115,A=1),RightColor_5_#8=(R=1,G=0,B=0,A=1),CenterColor_9_#5=(R=1,G=1,B=0,A= ...(truncated)"
    Offset Elbow Bone=(X=0,Y=0,Z=0)
    Offset Elbow Bone Space=LocalSpace
    Offset IK Ctrl Transform=(Rotation=(X=0,Y=0,Z=0,W=1),Translation=(X=0,Y=0,Z=0),Scale3D=(X=1,Y=1,Z=1))
    Offset PV Control=(Rotation=(X=0,Y=0,Z=0,W=1),Translation=(X=0,Y=0,Z=0),Scale3D=(X=1,Y=1,Z=1))
    PV Control Settings="(DisplayName=\"None\",bSelectable=True,InitialSpace=LocalSpace,bUsePreferredRotationOrder=False,PreferredRotationOrder=YZX,Limits=(LimitTranslationX=(bMinimum=False,bMaximum=False),LimitTranslationY=(bMinimum=False,bMaximum=False),LimitTranslationZ=(bMinimum ...(truncated)"
    PV Control Spaces=()
    PV Offset Factor=3
    Parent Control Connector=(Type=Connector,Name="Parent Control")
    Pole Vector Parent Connector=(Type=Connector,Name="None")
    Primary Axis=(X=1,Y=0,Z=0)
    Secondary Axis=(X=0,Y=0,Z=-1)
    Settings Control Connector=(Type=Connector,Name="Settings Control")
    Start Bone Connector=(Type=Connector,Name="Shoulder")
    Wrist Control Settings="(DisplayName=\"None\",bSelectable=True,InitialSpace=LocalSpace,bUsePreferredRotationOrder=False,PreferredRotationOrder=YZX,Limits=(LimitTranslationX=(bMinimum=False,bMaximum=False),LimitTranslationY=(bMinimum=False,bMaximum=False),LimitTranslationZ=(bMinimum ...(truncated)"
   )
   Compute Pole Vector Offset for Twist = RigVMCollapseNode(
    Axis=(X=1,Y=0,Z=0)
    Debug=True
    ExecuteContext=()
    Initial=true
    Item=(Type=Null,Name="None")
    Secondary Axis=(X=0,Y=-1,Z=0)
    Secondary Direction=(X=0,Y=0,Z=1)
    Target item=(Type=Control,Name="None")
   )
   Compute Pole Vector Offset for Twist_1 = RigVMCollapseNode(
    Axis=(X=1,Y=0,Z=0)
    Debug=True
    ExecuteContext=()
    Initial=true
    Item=(Type=Null,Name="None")
    Secondary Axis=(X=0,Y=-1,Z=0)
    Secondary Direction=(X=0,Y=0,Z=1)
    Target item=(Type=Control,Name="None")
   )
   FWD Basic Twist Rig = RigVMCollapseNode(Active=true, ExecuteContext=(), Filter=(bX=True,bY=True,bZ=True))
   Format Name = RigVMCollapseNode(name=arm, prefix=None, side=_fk_ik_switch, suffix=_fk_ik_switch)
   Get Right Side DEPRECATED = RigVMCollapseNode(Item=(Type=None,Name="None"))
   Get Side = RigVMCollapseNode(End Bone=(Type=Bone,Name="None"), ExecuteContext=(), Left Side Suffix=None, Right Side Suffix=None, Start Bone=(Type=Bone,Name="None"))
   Get Side from Metadata = RigVMCollapseNode(Item=(Type=Bone,Name="None"))
   INV Leg = RigVMCollapseNode(ExecuteContext=())
   INV TwoBoneLimb = RigVMCollapseNode(ExecuteContext=())
   Is Right Side = RigVMCollapseNode(ExecuteContext=(), Item=(Type=Bone))
   Is Right Side v2 = RigVMCollapseNode(ExecuteContext=(), Item=(Type=Bone), Mirror Plane=(X=1,Y=1,Z=0))
   Is Right Side wVector = RigVMCollapseNode(ExecuteContext=(), Item=(Type=Bone), Mirror Plane=(X=1,Y=1,Z=0))
   Parent = RigVMCollapseNode(Child=(Type=Bone,Name="None"), ExecuteContext=(), New Parent=(Type=Bone,Name="None"))
   Set Closest Bone = RigVMCollapseNode(Bones=(), ExecuteContext=(), Items=())
  ]}
 RigVMFunction_MathBoolOr_1_ContainedGraph {nodes [
   n1 = RigVMFunctionEntryNode()
   RigVMFunction_MathBoolOr = &RigVMFunction_MathBoolOr(A=@n1.A, B=@n1.B)
   RigVMFunctionReturnNode(Result=&RigVMFunction_MathBoolOr(B=@n1.C, A=@RigVMFunction_MathBoolOr.Result).Result)
  ]}
 RigVMFunction_MathBoolOr_1_ContainedGraph {nodes [
   n1 = RigVMFunctionEntryNode()
   RigVMFunction_MathBoolOr = &RigVMFunction_MathBoolOr(A=@n1.A, B=@n1.B)
   RigVMFunctionReturnNode(Result=&RigVMFunction_MathBoolOr(B=@n1.C, A=@RigVMFunction_MathBoolOr.Result).Result)
  ]}
 RigVMFunction_MathBoolOr_1_ContainedGraph {nodes [
   n1 = RigVMFunctionEntryNode()
   RigVMFunction_MathBoolOr = &RigVMFunction_MathBoolOr(A=@n1.A, B=@n1.B)
   RigVMFunctionReturnNode(Result=&RigVMFunction_MathBoolOr(B=@n1.C, A=@RigVMFunction_MathBoolOr.Result).Result)
  ]}
 RigVMFunction_MathBoolOr_1_ContainedGraph {nodes [
   n1 = RigVMFunctionEntryNode()
   RigVMFunction_MathBoolOr = &RigVMFunction_MathBoolOr(A=@n1.A, B=@n1.B)
   RigVMFunctionReturnNode(Result=&RigVMFunction_MathBoolOr(B=@n1.C, A=@RigVMFunction_MathBoolOr.Result).Result)
  ]}
 RigVMFunction_Sequence_1_1_ContainedGraph {nodes [
   n1 = RigVMFunctionEntryNode()
   n2 = &RigVMFunction_Sequence(ExecuteContext=@n1.ExecuteContext)
   n3 = &RigVMFunction_Sequence(ExecuteContext=@n2.B)
   RigVMFunctionReturnNode(A=@n2.A, B=@n3.A, C=@n3.B)
  ]}
 RigVMFunction_Sequence_1_ContainedGraph {nodes [
   n1 = RigVMFunctionEntryNode()
   RigVMFunction_Sequence = &RigVMFunction_Sequence(ExecuteContext=@n1.ExecuteContext)
   n2 = &RigVMFunction_Sequence(ExecuteContext=@RigVMFunction_Sequence.B)
   RigVMFunctionReturnNode(A=@RigVMFunction_Sequence.A, B=@n2.A, C=@n2.B)
  ]}
 RigVMFunction_Sequence_1_ContainedGraph {nodes [
   n1 = RigVMFunctionEntryNode()
   RigVMFunction_Sequence = &RigVMFunction_Sequence(ExecuteContext=@n1.ExecuteContext)
   n2 = &RigVMFunction_Sequence(ExecuteContext=@RigVMFunction_Sequence.B)
   RigVMFunctionReturnNode(A=@RigVMFunction_Sequence.A, B=@n2.A, C=@n2.B)
  ]}
 RigVMFunction_Sequence_1_ContainedGraph {nodes [
   n1 = RigVMFunctionEntryNode()
   RigVMFunction_Sequence = &RigVMFunction_Sequence(ExecuteContext=@n1.ExecuteContext)
   n2 = &RigVMFunction_Sequence(ExecuteContext=@RigVMFunction_Sequence.B)
   RigVMFunctionReturnNode(A=@RigVMFunction_Sequence.A, B=@n2.A, C=@n2.B)
  ]}
 RigVMFunction_Sequence_1_ContainedGraph {nodes [
   n1 = RigVMFunctionEntryNode()
   RigVMFunction_Sequence = &RigVMFunction_Sequence(ExecuteContext=@n1.ExecuteContext)
   n2 = &RigVMFunction_Sequence(ExecuteContext=@RigVMFunction_Sequence.B)
   n3 = &RigVMFunction_Sequence(ExecuteContext=@n2.B)
   n4 = &RigVMFunction_Sequence(ExecuteContext=@n3.B)
   n5 = &RigVMFunction_Sequence(ExecuteContext=@n4.B)
   n6 = &RigVMFunction_Sequence(ExecuteContext=@n5.B)
   RigVMFunctionReturnNode(A=@RigVMFunction_Sequence.A, B=@n2.A, C=@n3.A, D=@n4.A, E=@n5.A, F=@n6.A, G=@n6.B)
  ]}
 RigVMFunction_Sequence_1_ContainedGraph {nodes [
   n1 = RigVMFunctionEntryNode()
   RigVMFunction_Sequence = &RigVMFunction_Sequence(ExecuteContext=@n1.ExecuteContext)
   n2 = &RigVMFunction_Sequence(ExecuteContext=@RigVMFunction_Sequence.B)
   n3 = &RigVMFunction_Sequence(ExecuteContext=@n2.B)
   n4 = &RigVMFunction_Sequence(ExecuteContext=@n3.B)
   n5 = &RigVMFunction_Sequence(ExecuteContext=@n4.B)
   RigVMFunctionReturnNode(A=@RigVMFunction_Sequence.A, B=@n2.A, C=@n3.A, D=@n4.A, E=@n5.A, F=@n5.B)
  ]}
 RigVMFunction_Sequence_1_ContainedGraph {nodes [
   n1 = RigVMFunctionEntryNode()
   RigVMFunction_Sequence = &RigVMFunction_Sequence(ExecuteContext=@n1.ExecuteContext)
   n2 = &RigVMFunction_Sequence(ExecuteContext=@RigVMFunction_Sequence.B)
   n3 = &RigVMFunction_Sequence(ExecuteContext=@n2.B)
   n4 = &RigVMFunction_Sequence(ExecuteContext=@n3.B)
   RigVMFunctionReturnNode(A=@RigVMFunction_Sequence.A, B=@n2.A, C=@n3.A, D=@n4.A, E=@n4.B)
  ]}
 RigVMFunction_Sequence_1_ContainedGraph {nodes [
   n1 = RigVMFunctionEntryNode()
   RigVMFunction_Sequence = &RigVMFunction_Sequence(ExecuteContext=@n1.ExecuteContext)
   n2 = &RigVMFunction_Sequence(ExecuteContext=@RigVMFunction_Sequence.B)
   RigVMFunctionReturnNode(A=@RigVMFunction_Sequence.A, B=@n2.A, C=@n2.B)
  ]}
 RigVMFunction_Sequence_2_1_ContainedGraph {nodes [
   n1 = RigVMFunctionEntryNode()
   n2 = &RigVMFunction_Sequence(ExecuteContext=@n1.ExecuteContext)
   n3 = &RigVMFunction_Sequence(ExecuteContext=@n2.B)
   n4 = &RigVMFunction_Sequence(ExecuteContext=@n3.B)
   n5 = &RigVMFunction_Sequence(ExecuteContext=@n4.B)
   n6 = &RigVMFunction_Sequence(ExecuteContext=@n5.B)
   RigVMFunctionReturnNode(A=@n2.A, B=@n3.A, C=@n4.A, D=@n5.A, E=@n6.A, F=@n6.B)
  ]}
 RigVMFunction_Sequence_2_ContainedGraph {nodes [
   n1 = RigVMFunctionEntryNode()
   n2 = &RigVMFunction_Sequence(ExecuteContext=@n1.ExecuteContext)
   n3 = &RigVMFunction_Sequence(ExecuteContext=@n2.B)
   RigVMFunctionReturnNode(A=@n2.A, B=@n3.A, C=@n3.B)
  ]}
 RigVMModel {nodes [
   n1 = &RigUnit_BeginExecution()
   RigVMFunction_Sequence = RigVMAggregateNode(ExecuteContext=@n1.ExecutePin)
   Get_Module_Metadata_3_1 = RigVMDispatchNode(Default=true, Name=Enable Softness, NameSpace=Root) {template="DISPATCH_RigDispatch_GetModuleMetadata(in Name,in NameSpace,in Default,out Value,out Found)"}
   RigVMFunction_ControlFlowBranch = &RigVMFunction_ControlFlowBranch(Condition=@Get_Module_Metadata_3_1.Found, ExecuteContext=@RigVMFunction_Sequence.A)
   RigVMVariableNode(Value=@Get_Module_Metadata_3_1.Value, ExecuteContext=@RigVMFunction_ControlFlowBranch.True) {variable=Enable Softness}
   DISPATCH_RigDispatch_GetModuleMetadata = RigVMDispatchNode(Default=None, Name=fk_ik_switch, NameSpace=Self) {template="DISPATCH_RigDispatch_GetModuleMetadata(in Name,in NameSpace,in Default,out Value,out Found)"}
   GetAnimationChannel = /Script/ControlRig.RigUnit_GetBoolAnimationChannel(
    bInitial=false
    Channel=@DISPATCH_RigDispatch_GetModuleMetadata.Value
    Control=&RigUnit_ResolveConnector(Connector=(Type=Connector,Name="Settings Control"), SkipSocket=true).Result.Name
   ) {is=struct RigVMUnitNode, template="GetAnimationChannel::Execute(out Value,in Control,in Channel,in bInitial)"}
   n2 = &RigUnit_FindItemsWithMetadataTag(NameSpace=Self, Tag=Controls IK)
   n3 = &RigUnit_FindItemsWithMetadataTag(NameSpace=Self, Tag=Controls IK)
   n4 = &RigUnit_FindItemsWithMetadataTag(NameSpace=Self, Tag=Controls IK)
   n5 = RigVMDispatchNode(Index=0, Array=&RigUnit_FindItemsWithMetadataTag(NameSpace=Self, Tag=Controls IK).Items) {template="DISPATCH_RigVMDispatch_ArrayGetAtIndex(in Array,in Index,out Element)"}
   DISPATCH_RigVMDispatch_ArrayGetAtIndex = RigVMDispatchNode(Index=0, Array=&RigUnit_FindItemsWithMetadataTag(NameSpace=Self, Tag=Controls FK).Items) {template="DISPATCH_RigVMDispatch_ArrayGetAtIndex(in Array,in Index,out Element)"}
   RigUnit_GetBoolAnimationChannel = &RigUnit_GetBoolAnimationChannel(Channel=local, bInitial=False, Control=@DISPATCH_RigVMDispatch_ArrayGetAtIndex.Element.Name) {template="GetAnimationChannel::Execute(out Value,in Control,in Channel,in bInitial)"}
   FWD TwoBoneLimb = RigVMFunctionReferenceNode(
    Attach Skeleton=true
    Debug=true
    End Ctrl=(Type=Control,Name="None")
    FK Arm Rotation Space=(Type=Bone,Name="root")
    FK Space=true
    Mirror Threshold=1
    PV Follow Aim Axis=(X=1,Y=0,Z=0)
    PV Follow Aim Direction=(X=0,Y=1,Z=0)
    PV Follow Twist Lock=(bX=false,bY=false,bZ=true)
    PV Follow Up Axis=(X=0,Y=1,Z=0)
    Bones=RigVMDispatchNode(Default=(), Name=Rig Bones, NameSpace=Self) {template="DISPATCH_RigDispatch_GetModuleMetadata(in Name,in NameSpace,in Default,out Value,out Found)"}.Value
    Driven Bones=&RigUnit_FindItemsWithMetadataTag(NameSpace=Self, Tag=Bones).Items
    FK Controls=&RigUnit_FindItemsWithMetadataTag(NameSpace=Self, Tag=Controls FK).Items
    IK Controls=&RigUnit_FindItemsWithMetadataTag(NameSpace=Self, Tag=Controls IK).Items
    FK_or_IK=@GetAnimationChannel.Value
    Softness=/Script/ControlRig.RigUnit_GetFloatAnimationChannel(
     Channel=Softness
     bInitial=False
     Control=RigVMDispatchNode(Index=0, Array=@n2.Items) {template="DISPATCH_RigVMDispatch_ArrayGetAtIndex(in Array,in Index,out Element)"}.Element.Name
    ) {is=struct RigVMUnitNode, template="GetAnimationChannel::Execute(out Value,in Control,in Channel,in bInitial)"}.Value
    Enable Stretch=&RigUnit_GetBoolAnimationChannel(
     Channel=Stretch
     bInitial=False
     Control=RigVMDispatchNode(Index=0, Array=@n3.Items) {template="DISPATCH_RigVMDispatch_ArrayGetAtIndex(in Array,in Index,out Element)"}.Element.Name
    ) {template="GetAnimationChannel::Execute(out Value,in Control,in Channel,in bInitial)"}.Value
    Keep Aligned=&RigUnit_GetBoolAnimationChannel(
     Channel=Keep Aligned
     bInitial=False
     Control=RigVMDispatchNode(Index=0, Array=@n4.Items) {template="DISPATCH_RigVMDispatch_ArrayGetAtIndex(in Array,in Index,out Element)"}.Element.Name
    ) {template="GetAnimationChannel::Execute(out Value,in Control,in Channel,in bInitial)"}.Value
    Enable PV Follow=RigVMDispatchNode(
     False=False
     True=&RigUnit_GetBoolAnimationChannel(Channel=PV Follow, bInitial=False, Control=@n5.Element.Name) {template="GetAnimationChannel::Execute(out Value,in Control,in Channel,in bInitial)"}.Value
     Condition=RigVMVariableNode() {variable=Enable PV Follow}.Value
    ) {template="DISPATCH_RigVMDispatch_If(in Condition,in True,in False,out Result)"}.Result
    local=@RigUnit_GetBoolAnimationChannel.Value
    ExecuteContext=@RigVMFunction_Sequence.B
    Enable Softness=RigVMVariableNode() {variable=Enable Softness}.Value
    Auto Determine Axis=RigVMVariableNode() {variable=Auto Determine Axis}.Value
    Primary Axis=RigVMVariableNode() {variable=Primary Axis}.Value
    Secondary Axis=RigVMVariableNode() {variable=Secondary Axis}.Value
   )
   > DISPATCH_RigDispatch_SetModuleMetadata = RigVMDispatchNode(Name=Current FK IK Switch, NameSpace=Self, Value=@GetAnimationChannel.Value) {template="DISPATCH_RigDispatch_SetModuleMetadata(in Name,in NameSpace,in Value,out Success)"}
   > FWD Basic Twist Rig = RigVMFunctionReferenceNode(Filter=RigVMVariableNode() {variable=Twist Bone Axis}.Value, Active=RigVMVariableNode() {variable=Debug Twist Rigs}.Value)
  ]}
 RigVMModel Backwards Solve Graph {nodes [
   n1 = /Script/ControlRig.RigUnit_InverseExecution() {is=struct RigVMUnitNode}
   INV TwoBoneLimb_1 = RigVMFunctionReferenceNode(ExecuteContext=@n1.ExecutePin)
  ]}
 RigVMModel Connector Event Graph {nodes [
   n1 = /Script/ControlRig.RigUnit_ConnectorExecution() {is=struct RigVMUnitNode}
   CONNECT Leg = RigVMFunctionReferenceNode(ExecuteContext=@n1.ExecuteContext)
  ]}
]
hierarchy [
 Leg Primary {type=Connector, loc=<0 0 0>, rot=<0 0 0 1>}
 Thigh Bone {type=Connector, loc=<0 0 0>, rot=<0 0 0 1>}
 Pelvis Bone {type=Connector, loc=<0 0 0>, rot=<0 0 0 1>}
 Parent Control {type=Connector, loc=<0 0 0>, rot=<0 0 0 1>}
 Root {type=Socket, loc=<0 0 0>, rot=<0 0 0 1>}
 Foot Bone {type=Connector, loc=<0 0 0>, rot=<0 0 0 1>}
 Settings Control {type=Connector, loc=<0 0 0>, rot=<0 0 0 1>}
 Pole Vector Parent {type=Connector, loc=<0 0 0>, rot=<0 0 0 1>}
 Knee Bone {type=Connector, loc=<0 0 0>, rot=<0 0 0 1>}
 Twist Bones {type=Connector, loc=<0 0 0>, rot=<0 0 0 1>}
]
variables [
 Ankle Bone {type=FRigElementKey, default=(Type=Bone,Name="")}
 Auto Determine Axis {type=bool, default=True}
 Body Ctrl {type=FRigElementKey, default=(Type=Bone,Name="")}
 "Current Side " {type=FName, default=None}
 Debug Twist Rigs {type=bool, default=False}
 Effector Null {type=FRigElementKey, default=(Type=Bone,Name="")}
 Enable PV Follow {type=bool, default=True}
 Enable Softness {type=bool, default=False}
 FK IK Default {type=bool, default=True}
 FK Shape Settings {
  type=FRigUnit_HierarchyAddControlTransform_Settings
  default="(InitialSpace=LocalSpace,bUsePreferredRotationOrder=False,PreferredRotationOrder=YZX,Limits=(LimitTranslationX=(bMinimum=False,bMaximum=False),LimitTranslationY=(bMinimum=False,bMaximum=False),LimitTranslationZ=(bMinimum=False,bMaximum=False),LimitPitch=(b ...(truncated)"
 }
 Foot Control {type=FRigElementKey, default=(Type=None,Name="")}
 Foot L {type=FRigElementKey, default=(Type=Bone,Name="")}
 Foot L Metadata Found {type=bool, default=False}
 Foot R {type=FRigElementKey, default=(Type=Bone,Name="")}
 Foot R Metadata Found {type=bool, default=False}
 Foot Socket Target {type=FRigElementKey, default=(Type=Bone,Name="")}
 Get Side_Result {type=FName, default=None}
 Hips Ctrl {type=FRigElementKey, default=(Type=Bone,Name="")}
 IK Control Offset {type=FTransform, default=(Rotation=(X=0,Y=0,Z=0,W=1),Translation=(X=0,Y=0,Z=0),Scale3D=(X=1,Y=1,Z=1))}
 IK Null {type=FRigElementKey, default=(Type=Bone,Name="")}
 IK Shape Settings {
  type=FRigUnit_HierarchyAddControlTransform_Settings
  default="(InitialSpace=LocalSpace,bUsePreferredRotationOrder=False,PreferredRotationOrder=YZX,Limits=(LimitTranslationX=(bMinimum=False,bMaximum=False),LimitTranslationY=(bMinimum=False,bMaximum=False),LimitTranslationZ=(bMinimum=False,bMaximum=False),LimitPitch=(b ...(truncated)"
 }
 Knee Bone Offset {type=FVector, default=(X=0,Y=0,Z=0)}
 Knee Bone Offset Global Space {type=bool, default=False}
 Knee Bone Offset Metadata {type=FVector, default=(X=0,Y=0,Z=0)}
 Leg PV Position {type=float, default=0}
 Leg PV Position Metadata Found {type=bool, default=False}
 Module Settings {
  type=FUserDefinedStruct_#9
  default="(OverrideColor_48_#7=False,LeftColor_3_#6=(R=0,G=0.0231,B=0.4115,A=1),RightColor_5_#8=(R=1,G=0,B=0,A=1),CenterColor_9_#5=(R=1,G=1,B=0,A= ...(truncated)"
 }
 Orient IK to World {type=bool, default=True}
 PV Ctrl {type=FRigElementKey, default=(Type=None,Name="")}
 PV Shape Settings {
  type=FRigUnit_HierarchyAddControlTransform_Settings
  default="(InitialSpace=LocalSpace,bUsePreferredRotationOrder=False,PreferredRotationOrder=YZX,Limits=(LimitTranslationX=(bMinimum=False,bMaximum=False),LimitTranslationY=(bMinimum=False,bMaximum=False),LimitTranslationZ=(bMinimum=False,bMaximum=False),LimitPitch=(b ...(truncated)"
 }
 Pole Vector Position {type=float, default=1}
 Primary Axis {type=FVector, default=(X=-1,Y=0,Z=0)}
 RigVMFunction_MathVectorClampLength_Result {type=FVector, default=(X=0,Y=0,Z=0)}
 RightSideThreshold {type=double, default=0}
 Secondary Axis {type=FVector, default=(X=0,Y=-1,Z=0)}
 Subtract_1_Result {type=double, default=0}
 Thigh Bone {type=FRigElementKey, default=(Type=Bone,Name="")}
 Toe L {type=FRigElementKey, default=(Type=Bone,Name="")}
 Toe L Metadata Found {type=bool, default=False}
 Toe R {type=FRigElementKey, default=(Type=Bone,Name="")}
 Toe R Metadata Found {type=bool, default=False}
 Twist Bone Axis {type=FFilterOptionPerAxis, default=(bX=True,bY=False,bZ=False)}
 effector {type=FRigElementKey, default=(Type=Bone,Name="")}
 follow_null {type=FRigElementKey, default=(Type=Bone,Name="")}
 foot ik effector null {type=FRigElementKey, default=(Type=Bone,Name="")}
 foot_control {type=FRigElementKey, default=(Type=None,Name="")}
 knee Bone Offset GlobalSpace Metadata {type=bool, default=False}
 prefix {type=FName, default=None}
 pv_control {type=FRigElementKey, default=(Type=None,Name="")}
]
