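# Modular Control Rig shoulder module that rigs a clavicle and shoulder bone pair and adds hand sockets for arm modules.
# Construction resolves bones from connectors, builds a side-mirrored rig bone chain, then creates, scales and colors a clavicle control and optional translation proxy.
# Solve functions project bones from the rig chain.
# With "Aim at Proxy" on, translating the proxy aims the clavicle control, while rotating that control carries the proxy.
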ControlRigBlueprint /ControlRigModules/Modules56/Shoulder
graphs [
 CollapseNode_ContainedGraph {nodes [
   RigVMFunctionEntryNode()
   > RigVMFunction_Sequence = RigVMAggregateNode()
   n1 = /Script/ControlRig.RigUnit_ResolveConnector(Connector=(Type=Connector,Name="Shoulder Primary"), SkipSocket=true) {is=struct RigVMUnitNode}
   DISPATCH_RigDispatch_SetModuleMetadata = RigVMDispatchNode(Name=Shoulder Bone, NameSpace=Self, Value=@n1.Result, ExecuteContext=@RigVMFunction_Sequence.A) {template="DISPATCH_RigDispatch_SetModuleMetadata(in Name,in NameSpace,in Value,out Success)"}
   > RigVMDispatchNode(
    Name=Clavicle Bone
    NameSpace=Self
    Value=/Script/ControlRig.RigUnit_HierarchyGetParent(bDefaultParent=True, Child=@n1.Result) {is=struct RigVMUnitNode}.Parent
   ) {template="DISPATCH_RigDispatch_SetModuleMetadata(in Name,in NameSpace,in Value,out Success)"}
   Get_Module_Metadata = RigVMDispatchNode(Default=(Type=Bone,Name="None"), Name=Clavicle Bone, NameSpace=Self) {template="DISPATCH_RigDispatch_GetModuleMetadata(in Name,in NameSpace,in Default,out Value,out Found)"}
   Is Right Side v2 = RigVMFunctionReferenceNode(Mirror Plane=(X=1,Y=0,Z=0), ExecuteContext=@RigVMFunction_Sequence.B, Item=@Get_Module_Metadata.Value)
   Get_Module_Metadata_1 = RigVMDispatchNode(Default=(Type=Bone,Name="None"), Name=Shoulder Bone, NameSpace=Self) {template="DISPATCH_RigDispatch_GetModuleMetadata(in Name,in NameSpace,in Default,out Value,out Found)"}
   n2 = RigVMDispatchNode(False=(X=1,Y=1,Z=1), True=(X=-1,Y=-1,Z=-1), Condition=@"Is Right Side v2".Right Side) {template="DISPATCH_RigVMDispatch_If(in Condition,in True,in False,out Result)"}
   n3 = RigVMVariableNode() {variable=Module Settings}
   > Create Bone Chain = RigVMFunctionReferenceNode(
    Debug=true
    Debug Scale=15
    Debug Thickness=0.5
    Enable Orient Bones=true
    Kind=Direction
    Optional Suffix=rig
    Transforms=/Script/ControlRig.RigUnit_GetTransformItemArray(
     Items=((Type=None,Name="None"),(Type=None,Name="None"))
     Space=GlobalSpace
     bInitial=False
     Items.1=@Get_Module_Metadata_1.Value
     Items.0=@Get_Module_Metadata.Value
    ) {is=struct RigVMUnitNode}.Transforms
    Name=@Get_Module_Metadata.Value.Name
    Parent=&RigUnit_HierarchyGetParent(bDefaultParent=True, Child=@Get_Module_Metadata.Value).Parent
    Primary Axis=/Script/RigVM.RigVMFunction_MathVectorMul(B=@n2.Result, A=@n3.Value.PrimaryBoneAxis_32_#1) {is=struct RigVMUnitNode, template="Multiply::Execute(in A,in B,out Result)"}.Result
    Secondary Axis=&RigVMFunction_MathVectorMul(B=@n2.Result, A=@n3.Value.SecondaryAxis_51_#2) {template="Multiply::Execute(in A,in B,out Result)"}.Result
   )
   > RigVMDispatchNode(
    Name=Clavicle Rig Bone
    NameSpace=Self
    Value=RigVMDispatchNode(Index=0, Array=@"Create Bone Chain".Bones) {template="DISPATCH_RigVMDispatch_ArrayGetAtIndex(in Array,in Index,out Element)"}.Element
   ) {template="DISPATCH_RigDispatch_SetModuleMetadata(in Name,in NameSpace,in Value,out Success)"}
   > RigVMDispatchNode(
    Name=Shoulder Rig Bone
    NameSpace=Self
    Value=RigVMDispatchNode(Index=1, Array=@"Create Bone Chain".Bones) {template="DISPATCH_RigVMDispatch_ArrayGetAtIndex(in Array,in Index,out Element)"}.Element
   ) {template="DISPATCH_RigDispatch_SetModuleMetadata(in Name,in NameSpace,in Value,out Success)"}
   DISPATCH_RigDispatch_GetModuleMetadata = RigVMDispatchNode(Default=(Type=Bone,Name="None"), Name=Clavicle Rig Bone, NameSpace=Self) {template="DISPATCH_RigDispatch_GetModuleMetadata(in Name,in NameSpace,in Default,out Value,out Found)"}
   Control Stack at Item_1 = RigVMFunctionReferenceNode(
    Add Bottom Null=False
    Add Null=true
    Add Secondary Control=False
    Control Suffix=_ctrl
    Null Suffix=_null
    Orient to World=false
    Secondary Settings="(DisplayName=\"None\",Shape=(bVisible=True,Name=\"Default\",Color=(R=1,G=0,B=0,A=1),Transform=(Rotation=(X=0,Y=0,Z=0,W=1),Translation=(X=0,Y=0,Z=0),Scale3D=(X=1,Y=1,Z=1))),Proxy=(bIsProxy=False,ShapeVisibility=BasedOnSelection),InitialSpace=LocalSpace,bUsePref ...(truncated)"
    Tag Controls=true
    Zero Controls=true
    Item=@DISPATCH_RigDispatch_GetModuleMetadata.Value
    Name=RigVMDispatchNode(Default=(Type=Bone,Name="None"), Name=Clavicle Bone, NameSpace=Self) {template="DISPATCH_RigDispatch_GetModuleMetadata(in Name,in NameSpace,in Default,out Value,out Found)"}.Value.Name
    Parent=&RigUnit_ResolveConnector(Connector=(Type=Connector,Name="Parent Control"), SkipSocket=true).Result
    ExecuteContext=@RigVMFunction_Sequence.C
    Settings=RigVMVariableNode() {variable=Clavicle Control Settings}.Value
   )
   > RigVMDispatchNode(Name=FK Parent, NameSpace=Self, Value=@"Control Stack at Item_1".Control) {template="DISPATCH_RigDispatch_SetModuleMetadata(in Name,in NameSpace,in Value,out Success)"}
   > RigVMDispatchNode(Name=Shoulder Rotation Control, NameSpace=Self, Value=@"Control Stack at Item_1".Control) {template="DISPATCH_RigDispatch_SetModuleMetadata(in Name,in NameSpace,in Value,out Success)"}
   > RigVMVariableNode(Value=@"Control Stack at Item_1".Control) {variable=Shoulder Ctrl}
   n4 = /Script/ControlRig.RigUnit_HierarchyGetShapeSettings(Item=@"Control Stack at Item_1".Control) {is=struct RigVMUnitNode}
   n5 = &RigUnit_GetRelativeTransformForItem(
    bChildInitial=False
    bParentInitial=False
    Child=@"Control Stack at Item_1".Control
    Parent=RigVMDispatchNode(Default=(Type=Bone,Name="None"), Name=Shoulder Rig Bone, NameSpace=Self) {template="DISPATCH_RigDispatch_GetModuleMetadata(in Name,in NameSpace,in Default,out Value,out Found)"}.Value
   )
   n6 = &RigUnit_GetTransform(Space=LocalSpace, bInitial=False, Item=@"Control Stack at Item_1".Control)
   > n7 = /Script/ControlRig.RigUnit_HierarchySetShapeSettings(
    Settings=(bVisible=True,Name="Default",Color=(R=1,G=0,B=0,A=1),Transform=(Rotation=(X=0,Y=0,Z=0,W=1),Translation=(X=0,Y=0,Z=0),Scale3D=(X=1,Y=1,Z=1)))
    Item=@"Control Stack at Item_1".Control
    Settings.Color=@n4.Settings.Color
    Settings.Name=@n4.Settings.Name
    Settings.Transform.Rotation=@n4.Settings.Transform.Rotation
    Settings.Transform.Scale3D=@n4.Settings.Transform.Scale3D
    Settings.bVisible=@n4.Settings.bVisible
    Settings.Transform.Translation=/Script/RigVM.RigVMFunction_MathTransformLerp(
     B=(Rotation=(X=0,Y=0,Z=0,W=1),Translation=(X=0,Y=0,Z=0),Scale3D=(X=1,Y=1,Z=1))
     T=0.5
     A=&RigVMFunction_MathTransformMakeRelative(Parent=@n5.RelativeTransform, Global=@n6.Transform).Local
    ) {is=struct RigVMUnitNode, template="Interpolate::Execute(in A,in B,in T,out Result)"}.Result.Translation
   ) {is=struct RigVMUnitNode}
   n8 = &RigVMFunction_ControlFlowBranch(ExecuteContext=@n7.ExecutePin, Condition=RigVMVariableNode() {variable=Enable Translation}.Value)
   Concat = /Script/RigVM.RigVMFunction_NameConcat(
    B=_proxy
    A=RigVMDispatchNode(Default=(Type=Bone,Name="None"), Name=Clavicle Bone, NameSpace=Self) {template="DISPATCH_RigDispatch_GetModuleMetadata(in Name,in NameSpace,in Default,out Value,out Found)"}.Value.Name
   ) {is=struct RigVMUnitNode, template="Concat::Execute(in A,in B,out Result)"}
   Get_Module_Metadata_1_1_1 = RigVMDispatchNode(Default=(Type=Bone,Name="None"), Name=Shoulder Rig Bone, NameSpace=Self) {template="DISPATCH_RigDispatch_GetModuleMetadata(in Name,in NameSpace,in Default,out Value,out Found)"}
   Control Stack at Item_1_1 = RigVMFunctionReferenceNode(
    Add Bottom Null=False
    Add Null=true
    Add Secondary Control=False
    Control Suffix=_ctrl
    Null Suffix=_null
    Orient to World=true
    Secondary Settings="(DisplayName=\"None\",Shape=(bVisible=True,Name=\"Default\",Color=(R=1,G=0,B=0,A=1),Transform=(Rotation=(X=0,Y=0,Z=0,W=1),Translation=(X=0,Y=0,Z=0),Scale3D=(X=1,Y=1,Z=1))),Proxy=(bIsProxy=False,ShapeVisibility=BasedOnSelection),InitialSpace=LocalSpace,bUsePref ...(truncated)"
    Settings="(DisplayName=\"None\",Shape=(bVisible=True,Name=\"Box_Thick\",Color=(R=0,G=1,B=0,A=1),Transform=(Rotation=(X=0,Y=0,Z=0,W=1),Translation=(X=0,Y=0,Z=0),Scale3D=(X=1.2,Y=1.2,Z=1.2))),Proxy=(bIsProxy=true,DrivenControls=((Type=None,Name=\"None\")),ShapeVisibility=Ba ...(truncated)"
    Tag Controls=true
    Zero Controls=true
    Name=@Concat.Result
    Item=@Get_Module_Metadata_1_1_1.Value
    Parent=&RigUnit_ResolveConnector(Connector=(Type=Connector,Name="Parent Control"), SkipSocket=true).Result
    ExecuteContext=@n8.True
    Settings.Proxy.DrivenControls.0=RigVMVariableNode() {variable=Shoulder Ctrl}.Value
   )
   > RigVMVariableNode(Value=@"Control Stack at Item_1_1".Control) {variable=Shoulder Proxy Ctrl}
   > RigVMDispatchNode(Name=Shoulder Translation Proxy, NameSpace=Self, Value=@"Control Stack at Item_1_1".Control) {template="DISPATCH_RigDispatch_SetModuleMetadata(in Name,in NameSpace,in Value,out Success)"}
   > SpawnAnimationChannel = /Script/ControlRig.RigUnit_HierarchyAddAnimationChannelBool(
    InitialValue=true
    LimitsEnabled=()
    MaximumValue=True
    MinimumValue=False
    Name=Aim at Proxy
    Parent=RigVMVariableNode() {variable=Shoulder Ctrl}.Value
   ) {
    is=struct RigVMUnitNode
    template="SpawnAnimationChannel::Execute(in InitialValue,in MinimumValue,in MaximumValue,in LimitsEnabled,in Parent,in Name,out Item)"
   }
   RigVMFunctionReturnNode(ExecuteContext=@RigVMFunction_Sequence.D)
   /Script/RigVM.RigVMFunction_MathVectorAdd(B=(X=0,Y=0,Z=10), A=@n5.RelativeTransform.Translation) {is=struct RigVMUnitNode, template="Add::Execute(in A,in B,out Result)"}
  ]}
 CollapseNode_ContainedGraph {nodes [
   RigVMFunctionEntryNode()
   > RigVMVariableNode() {variable=Found}
   > RigVMFunction_Sequence = RigVMAggregateNode()
   n1 = RigVMDispatchNode(Default=(Type=Bone,Name="None"), Name=Shoulder Bone, NameSpace=Self) {template="DISPATCH_RigDispatch_GetModuleMetadata(in Name,in NameSpace,in Default,out Value,out Found)"}
   n2 = &RigVMFunction_ControlFlowBranch(
    Condition=RigVMDispatchNode(
     Array=/Script/ControlRig.RigUnit_CollectionChildrenArray(TypeToSearch=Bone, bDefaultChildren=True, bIncludeParent=False, bRecursive=true, Parent=@n1.Value) {is=struct RigVMUnitNode}.Items
     Element=RigVMDispatchNode(Default=(Type=Bone,Name="None"), Name=hand_l, NameSpace=Root) {template="DISPATCH_RigDispatch_GetModuleMetadata(in Name,in NameSpace,in Default,out Value,out Found)"}.Value
    ) {template="DISPATCH_RigVMDispatch_ArrayFind(in Array,in Element,out Index,out Success)"}.Success
    ExecuteContext=@RigVMFunction_Sequence.A
   )
   Create Socket at Metadata_2 = RigVMFunctionReferenceNode(
    Color=(R=1,G=1,B=1,A=1)
    Description=Place the arm module here . . .
    Extra Transform=(Rotation=(X=0,Y=0,Z=0,W=1),Translation=(X=0,Y=0,Z=0),Scale3D=(X=1,Y=1,Z=1))
    Metadata Name=hand_l
    Name=hand_l_socket
    Space=LocalSpace
    ExecuteContext=@n2.True
   )
   > RigVMVariableNode(Value=true) {variable=Found}
   n3 = RigVMDispatchNode(Default=(Type=Bone,Name="None"), Name=Shoulder Bone, NameSpace=Self) {template="DISPATCH_RigDispatch_GetModuleMetadata(in Name,in NameSpace,in Default,out Value,out Found)"}
   n4 = &RigVMFunction_ControlFlowBranch(
    Condition=RigVMDispatchNode(
     Array=&RigUnit_CollectionChildrenArray(TypeToSearch=Bone, bDefaultChildren=True, bIncludeParent=False, bRecursive=true, Parent=@n3.Value).Items
     Element=RigVMDispatchNode(Default=(Type=Bone,Name="None"), Name=hand_r, NameSpace=Root) {template="DISPATCH_RigDispatch_GetModuleMetadata(in Name,in NameSpace,in Default,out Value,out Found)"}.Value
    ) {template="DISPATCH_RigVMDispatch_ArrayFind(in Array,in Element,out Index,out Success)"}.Success
    ExecuteContext=@RigVMFunction_Sequence.B
   )
   Create Socket at Metadata_2_1 = RigVMFunctionReferenceNode(
    Color=(R=1,G=1,B=1,A=1)
    Description=Place the arm module here . . .
    Extra Transform=(Rotation=(X=0,Y=0,Z=0,W=1),Translation=(X=0,Y=0,Z=0),Scale3D=(X=1,Y=1,Z=1))
    Metadata Name=hand_r
    Name=hand_r_socket
    Space=LocalSpace
    ExecuteContext=@n4.True
   )
   > RigVMVariableNode(Value=true) {variable=Found}
   RigVMFunction_MathBoolNot = &RigVMFunction_MathBoolNot(Value=RigVMVariableNode() {variable=Found}.Value)
   n5 = &RigVMFunction_ControlFlowBranch(Condition=@RigVMFunction_MathBoolNot.Result, ExecuteContext=@RigVMFunction_Sequence.C)
   Find first in Hierarchy_0 = RigVMFunctionReferenceNode(
    Children=true
    Ignore=()
    Search for=("hand")
    Type to Search=Bone
    Item=RigVMDispatchNode(Default=(Type=Bone,Name="None"), Name=Shoulder Bone, NameSpace=Self) {template="DISPATCH_RigDispatch_GetModuleMetadata(in Name,in NameSpace,in Default,out Value,out Found)"}.Value
    ExecuteContext=@n5.True
   )
   > /Script/ControlRig.RigUnit_HierarchyAddSocket(
    Color=(R=1,G=1,B=1,A=1)
    Description=Place the arm module here . . .
    Parent=(Type=Bone,Name="hand_l")
    Space=LocalSpace
    Transform=(Rotation=(X=0,Y=0,Z=0,W=1),Translation=(X=0,Y=0,Z=0),Scale3D=(X=1,Y=1,Z=1))
    Name=&RigVMFunction_NameConcat(B=_socket, A=@"Find first in Hierarchy_0".Item Found.Name) {template="Concat::Execute(in A,in B,out Result)"}.Result
   ) {is=struct RigVMUnitNode}
   RigVMFunctionReturnNode(ExecuteContext=@RigVMFunction_Sequence.D)
  ]}
 ConstructionGraph {nodes [
   RigUnit_PrepareForExecution = &RigUnit_PrepareForExecution()
   Create Shoulder Controls = RigVMCollapseNode(ExecuteContext=@RigUnit_PrepareForExecution.ExecutePin)
   > Finalize Controls = RigVMCollapseNode(Module Settings=RigVMVariableNode() {variable=Module Settings}.Value)
   > Create Arm Sockets = RigVMCollapseNode()
  ]}
 ContainedGraph {nodes [
   n1 = RigVMFunctionEntryNode()
   RigUnit_SequenceAggregate_1 = &RigVMFunction_Sequence(ExecuteContext=@n1.ExecuteContext)
   RigUnit_SequenceAggregate = RigVMAggregateNode(ExecuteContext=@RigUnit_SequenceAggregate_1.A)
   Greater = /Script/RigVM.RigVMFunction_MathDoubleGreater(B=0, A=@n1.Mirror Plane.X) {is=struct RigVMUnitNode, template="Greater::Execute(in A,in B,out Result)"}
   BranchNode_4 = &RigVMFunction_ControlFlowBranch(Condition=@Greater.Result, ExecuteContext=@RigUnit_SequenceAggregate.A)
   n2 = RigVMVariableNode() {variable=vector pos}
   RigVMVariableNode(
    Value=(X=0,Y=0,Z=0)
    ExecuteContext=@BranchNode_4.True
    Value.X=&RigUnit_GetTransform(bInitial=true, Item=@n1.Item).Transform.Translation.X
    Value.Y=@n2.Value.Y
    Value.Z=@n2.Value.Z
   ) {variable=vector pos}
   n3 = &RigVMFunction_MathDoubleGreater(B=0, A=@n1.Mirror Plane.Y) {template="Greater::Execute(in A,in B,out Result)"}
   BranchNode_1_1 = &RigVMFunction_ControlFlowBranch(Condition=@n3.Result, ExecuteContext=@RigUnit_SequenceAggregate.B)
   n4 = RigVMVariableNode() {variable=vector pos}
   RigVMVariableNode(
    Value=(X=0,Y=0,Z=0)
    ExecuteContext=@BranchNode_1_1.True
    Value.Y=&RigUnit_GetTransform(bInitial=true, Item=@n1.Item).Transform.Translation.Y
    Value.X=@n4.Value.X
    Value.Z=@n4.Value.Z
   ) {variable=vector pos}
   n5 = &RigVMFunction_MathDoubleGreater(B=0, A=@n1.Mirror Plane.Z) {template="Greater::Execute(in A,in B,out Result)"}
   BranchNode_2_1 = &RigVMFunction_ControlFlowBranch(Condition=@n5.Result, ExecuteContext=@RigUnit_SequenceAggregate.C)
   n6 = RigVMVariableNode() {variable=vector pos}
   RigVMVariableNode(
    Value=(X=0,Y=0,Z=0)
    ExecuteContext=@BranchNode_2_1.True
    Value.Z=&RigUnit_GetTransform(bInitial=true, Item=@n1.Item).Transform.Translation.Z
    Value.X=@n6.Value.X
    Value.Y=@n6.Value.Y
   ) {variable=vector pos}
   Cast = &RigVMFunction_MathBoolToFloat(Value=@Greater.Result)
   Dot__ = /Script/RigVM.RigVMFunction_MathVectorDot(
    B=(X=1,Y=1,Z=0)
    B.X=@Cast.Result
    B.Y=&RigVMFunction_MathBoolToFloat(Value=@n3.Result).Result
    B.Z=&RigVMFunction_MathBoolToFloat(Value=@n5.Result).Result
    A=RigVMVariableNode() {variable=vector pos}.Value
   ) {is=struct RigVMUnitNode, template="Dot,|::Execute(in A,in B,out Result)"}
   Less = /Script/RigVM.RigVMFunction_MathDoubleLess(B=0, A=@Dot__.Result) {is=struct RigVMUnitNode, template="Less::Execute(in A,in B,out Result)"}
   RigVMVariableNode(Value=@Less.Result, ExecuteContext=@RigUnit_SequenceAggregate.D) {variable=is on right}
   RigVMFunctionReturnNode(ExecuteContext=@RigUnit_SequenceAggregate_1.B, Right Side=RigVMVariableNode() {variable=is on right}.Value)
  ]}
 ContainedGraph {nodes [
   n1 = RigVMFunctionEntryNode()
   RigUnit_SequenceAggregate = &RigVMFunction_Sequence(ExecuteContext=@n1.ExecuteContext)
   RigUnit_SequenceAggregate_1 = &RigVMFunction_Sequence(ExecuteContext=@RigUnit_SequenceAggregate.B)
   RigUnit_SequenceAggregate_2 = &RigVMFunction_Sequence(ExecuteContext=@RigUnit_SequenceAggregate_1.B)
   RigVMFunctionReturnNode(A=@RigUnit_SequenceAggregate.A, B=@RigUnit_SequenceAggregate_1.A, C=@RigUnit_SequenceAggregate_2.A, D=@RigUnit_SequenceAggregate_2.B)
  ]}
 ContainedGraph {nodes [
   n1 = RigVMFunctionEntryNode()
   RigUnit_SequenceAggregate = &RigVMFunction_Sequence(ExecuteContext=@n1.ExecuteContext)
   RigVMVariableNode(Value=1000000000, ExecuteContext=@RigUnit_SequenceAggregate.A) {variable=closestDistanceSoFar}
   > n2 = RigVMDispatchNode(Array=@n1.Bones) {template="DISPATCH_RigVMDispatch_ArrayIterator(in Array,out Element,out Index,out Count,out Ratio)"}
   Item = /Script/ControlRig.RigUnit_Item(Item=@n1.Item) {is=struct RigVMUnitNode}
   n3 = &RigVMFunction_MathVectorDistance(
    A=&RigUnit_GetTransform(bInitial=@n1.Initial, Item=@n2.Element).Transform.Translation
    B=&RigUnit_GetTransform(bInitial=@n1.Initial, Item=@Item.Item).Transform.Translation
   )
   n4 = &RigVMFunction_ControlFlowBranch(
    ExecuteContext=@n2.ExecuteContext
    Condition=/Script/RigVM.RigVMFunction_MathFloatLessEqual(A=@n3.Result, B=RigVMVariableNode() {variable=closestDistanceSoFar}.Value) {is=struct RigVMUnitNode, template="LessEqual::Execute(in A,in B,out Result)"}.Result
   )
   RigVMVariableNode(ExecuteContext=@n4.True, Value=@n3.Result) {variable=closestDistanceSoFar}
   > RigVMVariableNode(Value=@n2.Element) {variable=closestBone}
   RigVMFunctionReturnNode(ExecuteContext=@RigUnit_SequenceAggregate.B, Value=RigVMVariableNode() {variable=closestBone}.Value)
  ]}
 New Function_ContainedGraph {nodes [
   n1 = RigVMFunctionEntryNode()
   n2 = /Script/ControlRig.RigUnit_FindItemsWithMetadataTag(NameSpace=Self, Tag=Controls) {is=struct RigVMUnitNode}
   > Scale Module Controls = RigVMFunctionReferenceNode(
    Global Scale=RigVMDispatchNode(Default=1, Name=Control Size, NameSpace=Root) {template="DISPATCH_RigDispatch_GetModuleMetadata(in Name,in NameSpace,in Default,out Value,out Found)"}.Value
    Local Scale=@n1.Module Settings.ControlSize_12_#3
    Controls=@n2.Items
   )
   > Color Module Controls = RigVMFunctionReferenceNode(
    CenterColor=@n1.Module Settings.CenterColor_9_#4
    LeftColor=@n1.Module Settings.LeftColor_3_#5
    Override=@n1.Module Settings.OverrideColor_48_#6
    RightColor=@n1.Module Settings.RightColor_5_#7
    Controls=@n2.Items
   )
   > RigVMFunctionReturnNode()
  ]}
 New Function_ContainedGraph {nodes [
   RigVMFunctionEntryNode()
   > RigVMFunction_Sequence = RigVMAggregateNode()
   DISPATCH_RigVMDispatch_ArrayIterator = RigVMDispatchNode(Array=&RigUnit_FindItemsWithMetadataTag(NameSpace=Self, Tag=Controls).Items, ExecuteContext=@RigVMFunction_Sequence.A) {template="DISPATCH_RigVMDispatch_ArrayIterator(in Array,out Element,out Index,out Count,out Ratio)"}
   DISPATCH_RigDispatch_GetMetadata = RigVMDispatchNode(Default=(Type=Bone,Name="None"), Name=Item, NameSpace=Self, Item=@DISPATCH_RigVMDispatch_ArrayIterator.Element) {template="DISPATCH_RigDispatch_GetMetadata(in Item,in Name,in NameSpace,in Default,out Value,out Found)"}
   RigVMFunction_ControlFlowBranch = &RigVMFunction_ControlFlowBranch(Condition=@DISPATCH_RigDispatch_GetMetadata.Found, ExecuteContext=@DISPATCH_RigVMDispatch_ArrayIterator.ExecuteContext)
   &RigUnit_SetTransform(
    Weight=1
    bInitial=False
    Item=@DISPATCH_RigVMDispatch_ArrayIterator.Element
    Value=/Script/ControlRig.RigUnit_ProjectTransformToNewParent(
     bChildInitial=True
     bNewParentInitial=False
     bOldParentInitial=True
     NewParent=@DISPATCH_RigDispatch_GetMetadata.Value
     OldParent=@DISPATCH_RigDispatch_GetMetadata.Value
     Child=@DISPATCH_RigVMDispatch_ArrayIterator.Element
    ) {is=struct RigVMUnitNode}.Transform
    ExecutePin=@RigVMFunction_ControlFlowBranch.True
   )
   > Set Transform_1 = /Script/ControlRig.RigUnit_SetTranslation(
    Space=LocalSpace
    Value=(X=0,Y=0,Z=0)
    Weight=1
    bInitial=False
    bPropagateToChildren=True
    Item=@DISPATCH_RigVMDispatch_ArrayIterator.Element
   ) {is=struct RigVMUnitNode, template="Set Transform::Execute(in Item,in Space,in bInitial,in Value,in Weight,in bPropagateToChildren)"}
   RigVMFunctionReturnNode(ExecuteContext=@RigVMFunction_Sequence.C)
  ]}
 New Function_ContainedGraph {nodes [
   n1 = RigVMFunctionEntryNode()
   n2 = &RigVMFunction_Sequence(ExecuteContext=@n1.ExecuteContext)
   RigVMFunction_Sequence = RigVMAggregateNode(ExecuteContext=@n2.A)
   DISPATCH_RigDispatch_SetModuleMetadata = RigVMDispatchNode(Name=Shoulder Bone, NameSpace=Self, ExecuteContext=@RigVMFunction_Sequence.B, Value=RigVMVariableNode() {variable=Shoulder Connector}.Value) {template="DISPATCH_RigDispatch_SetModuleMetadata(in Name,in NameSpace,in Value,out Success)"}
   > RigVMDispatchNode(
    Name=Clavicle Bone
    NameSpace=Self
    Value=&RigUnit_HierarchyGetParent(bDefaultParent=True, Child=RigVMVariableNode() {variable=Shoulder Connector}.Value).Parent
   ) {template="DISPATCH_RigDispatch_SetModuleMetadata(in Name,in NameSpace,in Value,out Success)"}
   Get_Module_Metadata = RigVMDispatchNode(Default=(Type=Bone,Name="None"), Name=Clavicle Bone, NameSpace=Self) {template="DISPATCH_RigDispatch_GetModuleMetadata(in Name,in NameSpace,in Default,out Value,out Found)"}
   > Is Right Side v2 = RigVMFunctionReferenceNode(Mirror Plane=(X=1,Y=0,Z=0), Item=@Get_Module_Metadata.Value)
   Get_Module_Metadata_1 = RigVMDispatchNode(Default=(Type=Bone,Name="None"), Name=Shoulder Bone, NameSpace=Self) {template="DISPATCH_RigDispatch_GetModuleMetadata(in Name,in NameSpace,in Default,out Value,out Found)"}
   n3 = RigVMDispatchNode(False=(X=1,Y=1,Z=1), True=(X=-1,Y=-1,Z=-1), Condition=@"Is Right Side v2".Right Side) {template="DISPATCH_RigVMDispatch_If(in Condition,in True,in False,out Result)"}
   > Create Bone Chain = RigVMFunctionReferenceNode(
    Debug=true
    Debug Scale=15
    Debug Thickness=0.5
    Enable Orient Bones=true
    Kind=Direction
    Optional Suffix=rig
    Transforms=&RigUnit_GetTransformItemArray(
     Items=((Type=None,Name="None"),(Type=None,Name="None"))
     Space=GlobalSpace
     bInitial=False
     Items.1=@Get_Module_Metadata_1.Value
     Items.0=@Get_Module_Metadata.Value
    ).Transforms
    Name=@Get_Module_Metadata.Value.Name
    Parent=&RigUnit_HierarchyGetParent(bDefaultParent=True, Child=@Get_Module_Metadata.Value).Parent
    Primary Axis=&RigVMFunction_MathVectorMul(B=@n3.Result, A=RigVMVariableNode() {variable=Primary Axis}.Value) {template="Multiply::Execute(in A,in B,out Result)"}.Result
    Secondary Axis=&RigVMFunction_MathVectorMul(B=@n3.Result, A=RigVMVariableNode() {variable=Secondary Axis}.Value) {template="Multiply::Execute(in A,in B,out Result)"}.Result
   )
   > RigVMDispatchNode(
    Name=Clavicle Rig Bone
    NameSpace=Self
    Value=RigVMDispatchNode(Index=0, Array=@"Create Bone Chain".Bones) {template="DISPATCH_RigVMDispatch_ArrayGetAtIndex(in Array,in Index,out Element)"}.Element
   ) {template="DISPATCH_RigDispatch_SetModuleMetadata(in Name,in NameSpace,in Value,out Success)"}
   > RigVMDispatchNode(
    Name=Shoulder Rig Bone
    NameSpace=Self
    Value=RigVMDispatchNode(Index=1, Array=@"Create Bone Chain".Bones) {template="DISPATCH_RigVMDispatch_ArrayGetAtIndex(in Array,in Index,out Element)"}.Element
   ) {template="DISPATCH_RigDispatch_SetModuleMetadata(in Name,in NameSpace,in Value,out Success)"}
   DISPATCH_RigDispatch_GetModuleMetadata = RigVMDispatchNode(Default=(Type=Bone,Name="None"), Name=Clavicle Rig Bone, NameSpace=Self) {template="DISPATCH_RigDispatch_GetModuleMetadata(in Name,in NameSpace,in Default,out Value,out Found)"}
   Control Stack at Item_1 = RigVMFunctionReferenceNode(
    Add Bottom Null=False
    Add Null=true
    Add Secondary Control=False
    Control Suffix=_ctrl
    Null Suffix=_null
    Orient to World=false
    Secondary Settings="(DisplayName=\"None\",Shape=(bVisible=True,Name=\"Default\",Color=(R=1,G=0,B=0,A=1),Transform=(Rotation=(X=0,Y=0,Z=0,W=1),Translation=(X=0,Y=0,Z=0),Scale3D=(X=1,Y=1,Z=1))),Proxy=(bIsProxy=False,ShapeVisibility=BasedOnSelection),InitialSpace=LocalSpace,bUsePref ...(truncated)"
    Tag Controls=true
    Zero Controls=true
    Item=@DISPATCH_RigDispatch_GetModuleMetadata.Value
    Name=RigVMDispatchNode(Default=(Type=Bone,Name="None"), Name=Clavicle Bone, NameSpace=Self) {template="DISPATCH_RigDispatch_GetModuleMetadata(in Name,in NameSpace,in Default,out Value,out Found)"}.Value.Name
    ExecuteContext=@RigVMFunction_Sequence.C
    Settings=RigVMVariableNode() {variable=Settings}.Value
    Parent=RigVMVariableNode() {variable=Parent Control Connector}.Value
   )
   > RigVMDispatchNode(Name=FK Parent, NameSpace=Self, Value=@"Control Stack at Item_1".Control) {template="DISPATCH_RigDispatch_SetModuleMetadata(in Name,in NameSpace,in Value,out Success)"}
   > RigVMDispatchNode(Name=Shoulder Rotation Control, NameSpace=Self, Value=@"Control Stack at Item_1".Control) {template="DISPATCH_RigDispatch_SetModuleMetadata(in Name,in NameSpace,in Value,out Success)"}
   n4 = &RigUnit_HierarchyGetShapeSettings(Item=@"Control Stack at Item_1".Control)
   n5 = &RigUnit_GetRelativeTransformForItem(
    bChildInitial=False
    bParentInitial=False
    Child=@"Control Stack at Item_1".Control
    Parent=RigVMDispatchNode(Default=(Type=Bone,Name="None"), Name=Shoulder Rig Bone, NameSpace=Self) {template="DISPATCH_RigDispatch_GetModuleMetadata(in Name,in NameSpace,in Default,out Value,out Found)"}.Value
   )
   n6 = &RigUnit_GetTransform(Space=LocalSpace, bInitial=False, Item=@"Control Stack at Item_1".Control)
   > n7 = &RigUnit_HierarchySetShapeSettings(
    Settings=(bVisible=True,Name="Default",Color=(R=1,G=0,B=0,A=1),Transform=(Rotation=(X=0,Y=0,Z=0,W=1),Translation=(X=0,Y=0,Z=0),Scale3D=(X=1,Y=1,Z=1)))
    Item=@"Control Stack at Item_1".Control
    Settings.Color=@n4.Settings.Color
    Settings.Name=@n4.Settings.Name
    Settings.Transform.Rotation=@n4.Settings.Transform.Rotation
    Settings.Transform.Scale3D=@n4.Settings.Transform.Scale3D
    Settings.bVisible=@n4.Settings.bVisible
    Settings.Transform.Translation=&RigVMFunction_MathTransformLerp(
     B=(Rotation=(X=0,Y=0,Z=0,W=1),Translation=(X=0,Y=0,Z=0),Scale3D=(X=1,Y=1,Z=1))
     T=0.5
     A=&RigVMFunction_MathTransformMakeRelative(Parent=@n5.RelativeTransform, Global=@n6.Transform).Local
    ) {template="Interpolate::Execute(in A,in B,in T,out Result)"}.Result.Translation
   )
   n8 = &RigVMFunction_ControlFlowBranch(ExecuteContext=@n7.ExecutePin, Condition=RigVMVariableNode() {variable=Enable Translation}.Value)
   Concat = &RigVMFunction_NameConcat(
    B=_proxy
    A=RigVMDispatchNode(Default=(Type=Bone,Name="None"), Name=Clavicle Bone, NameSpace=Self) {template="DISPATCH_RigDispatch_GetModuleMetadata(in Name,in NameSpace,in Default,out Value,out Found)"}.Value.Name
   ) {template="Concat::Execute(in A,in B,out Result)"}
   Get_Module_Metadata_1_1_1 = RigVMDispatchNode(Default=(Type=Bone,Name="None"), Name=Shoulder Rig Bone, NameSpace=Self) {template="DISPATCH_RigDispatch_GetModuleMetadata(in Name,in NameSpace,in Default,out Value,out Found)"}
   Control Stack at Item_1_1 = RigVMFunctionReferenceNode(
    Add Bottom Null=False
    Add Null=true
    Add Secondary Control=False
    Control Suffix=_ctrl
    Null Suffix=_null
    Orient to World=true
    Secondary Settings="(DisplayName=\"None\",Shape=(bVisible=True,Name=\"Default\",Color=(R=1,G=0,B=0,A=1),Transform=(Rotation=(X=0,Y=0,Z=0,W=1),Translation=(X=0,Y=0,Z=0),Scale3D=(X=1,Y=1,Z=1))),Proxy=(bIsProxy=False,ShapeVisibility=BasedOnSelection),InitialSpace=LocalSpace,bUsePref ...(truncated)"
    Settings="(DisplayName=\"None\",Shape=(bVisible=True,Name=\"Box_Thick\",Color=(R=0,G=1,B=0,A=1),Transform=(Rotation=(X=0,Y=0,Z=0,W=1),Translation=(X=0,Y=0,Z=0),Scale3D=(X=1.2,Y=1.2,Z=1.2))),Proxy=(bIsProxy=true,DrivenControls=((Type=None,Name=\"None\")),ShapeVisibility=Ba ...(truncated)"
    Tag Controls=true
    Zero Controls=true
    Name=@Concat.Result
    Settings.Proxy.DrivenControls.0=RigVMDispatchNode(Default=(Type=Bone,Name="None"), Name=Shoulder Rotation Control, NameSpace=Self) {template="DISPATCH_RigDispatch_GetModuleMetadata(in Name,in NameSpace,in Default,out Value,out Found)"}.Value
    Item=@Get_Module_Metadata_1_1_1.Value
    ExecuteContext=@n8.True
    Parent=RigVMVariableNode() {variable=Parent Control Connector}.Value
   )
   > RigVMDispatchNode(Name=Shoulder Translation Proxy, NameSpace=Self, Value=@"Control Stack at Item_1_1".Control) {template="DISPATCH_RigDispatch_SetModuleMetadata(in Name,in NameSpace,in Value,out Success)"}
   Get_Module_Metadata_2 = RigVMDispatchNode(Default=(Type=Bone,Name="None"), Name=Shoulder Rotation Control, NameSpace=Self) {template="DISPATCH_RigDispatch_GetModuleMetadata(in Name,in NameSpace,in Default,out Value,out Found)"}
   > SpawnAnimationChannel = &RigUnit_HierarchyAddAnimationChannelBool(InitialValue=true, LimitsEnabled=(), MaximumValue=True, MinimumValue=False, Name=Aim at Proxy, Parent=@Get_Module_Metadata_2.Value) {template="SpawnAnimationChannel::Execute(in InitialValue,in MinimumValue,in MaximumValue,in LimitsEnabled,in Parent,in Name,out Item)"}
   Finalize Controls = RigVMCollapseNode(ExecuteContext=@RigVMFunction_Sequence.E, Module Settings=RigVMVariableNode() {variable=Module Settings}.Value)
   RigVMFunctionReturnNode(ExecuteContext=@n2.B)
   &RigVMFunction_MathVectorAdd(B=(X=0,Y=0,Z=10), A=@n5.RelativeTransform.Translation) {template="Add::Execute(in A,in B,out Result)"}
   &RigUnit_FindItemsWithMetadataTag(NameSpace=Self, Tag=Bones)
  ]}
 New Function_ContainedGraph {nodes [
   n1 = RigVMFunctionEntryNode()
   n2 = &RigUnit_FindItemsWithMetadataTag(NameSpace=Self, Tag=Controls)
   > Scale Module Controls = RigVMFunctionReferenceNode(
    Global Scale=RigVMDispatchNode(Default=1, Name=Control Size, NameSpace=Root) {template="DISPATCH_RigDispatch_GetModuleMetadata(in Name,in NameSpace,in Default,out Value,out Found)"}.Value
    Local Scale=@n1.Module Settings.ControlSize_12_#3
    Controls=@n2.Items
   )
   > Color Module Controls = RigVMFunctionReferenceNode(
    CenterColor=@n1.Module Settings.CenterColor_9_#4
    LeftColor=@n1.Module Settings.LeftColor_3_#5
    Override=@n1.Module Settings.OverrideColor_48_#6
    RightColor=@n1.Module Settings.RightColor_5_#7
    Controls=@n2.Items
   )
   > RigVMFunctionReturnNode()
  ]}
 New Function_ContainedGraph {nodes [
   n1 = RigVMFunctionEntryNode()
   > RigVMFunction_Sequence = RigVMAggregateNode()
   Parent = RigVMFunctionReferenceNode(
    ExecuteContext=@RigVMFunction_Sequence.A
    Child=RigVMVariableNode() {variable=Clavicle Bone}.Value
    New Parent=RigVMVariableNode() {variable=Clavicle Control}.Value
   )
   DISPATCH_RigDispatch_GetMetadata = RigVMDispatchNode(Default=(Type=Bone,Name="None"), Name=Null, NameSpace=Self, Item=RigVMVariableNode() {variable=Clavicle Control}.Value) {template="DISPATCH_RigDispatch_GetMetadata(in Item,in Name,in NameSpace,in Default,out Value,out Found)"}
   n2 = RigVMVariableNode() {variable=Shoulder Bone}
   > Parent_1 = RigVMFunctionReferenceNode(
    Child=@DISPATCH_RigDispatch_GetMetadata.Value
    New Parent=&RigUnit_HierarchyGetParent(bDefaultParent=True, Child=&RigUnit_HierarchyGetParent(bDefaultParent=True, Child=@n2.Value).Parent).Parent
   )
   n3 = &RigVMFunction_ControlFlowBranch(Condition=@n1.Attach Skeleton, ExecuteContext=@RigVMFunction_Sequence.B)
   Get_Module_Metadata = RigVMDispatchNode(Default=(Type=Bone,Name="None"), Name=Clavicle Bone, NameSpace=Self) {template="DISPATCH_RigDispatch_GetModuleMetadata(in Name,in NameSpace,in Default,out Value,out Found)"}
   Get_Module_Metadata_2 = RigVMDispatchNode(Default=(Type=Bone,Name="None"), Name=Clavicle Rig Bone, NameSpace=Self) {template="DISPATCH_RigDispatch_GetModuleMetadata(in Name,in NameSpace,in Default,out Value,out Found)"}
   &RigUnit_SetTransform(
    Weight=1
    bInitial=False
    Item=@Get_Module_Metadata.Value
    Value=&RigUnit_ProjectTransformToNewParent(
     bChildInitial=True
     bNewParentInitial=False
     bOldParentInitial=True
     Child=@Get_Module_Metadata.Value
     NewParent=@Get_Module_Metadata_2.Value
     OldParent=@Get_Module_Metadata_2.Value
    ).Transform
    ExecutePin=@n3.True
   )
   Get_Module_Metadata_1 = RigVMDispatchNode(Default=(Type=Bone,Name="None"), Name=Shoulder Bone, NameSpace=Self) {template="DISPATCH_RigDispatch_GetModuleMetadata(in Name,in NameSpace,in Default,out Value,out Found)"}
   Get_Module_Metadata_1_1 = RigVMDispatchNode(Default=(Type=Bone,Name="None"), Name=Shoulder Rig Bone, NameSpace=Self) {template="DISPATCH_RigDispatch_GetModuleMetadata(in Name,in NameSpace,in Default,out Value,out Found)"}
   > &RigUnit_SetTransform(
    Weight=1
    bInitial=False
    Item=@Get_Module_Metadata_1.Value
    Value=&RigUnit_ProjectTransformToNewParent(
     bChildInitial=True
     bNewParentInitial=False
     bOldParentInitial=True
     Child=@Get_Module_Metadata_1.Value
     NewParent=@Get_Module_Metadata_1_1.Value
     OldParent=@Get_Module_Metadata_1_1.Value
    ).Transform
   )
   n4 = &RigVMFunction_ControlFlowBranch(ExecuteContext=@RigVMFunction_Sequence.C, Condition=RigVMVariableNode() {variable=Enable Translation}.Value)
   Get_Module_Metadata_4 = RigVMDispatchNode(Default=(Type=Bone,Name="None"), Name=Shoulder Rotation Control, NameSpace=Self) {template="DISPATCH_RigDispatch_GetModuleMetadata(in Name,in NameSpace,in Default,out Value,out Found)"}
   GetAnimationChannel = /Script/ControlRig.RigUnit_GetBoolAnimationChannel(Channel=Aim at Proxy, bInitial=False, Control=@Get_Module_Metadata_4.Value.Name) {is=struct RigVMUnitNode, template="GetAnimationChannel::Execute(out Value,in Control,in Channel,in bInitial)"}
   n5 = &RigVMFunction_ControlFlowBranch(Condition=@GetAnimationChannel.Value, ExecuteContext=@n4.True)
   n6 = /Script/ControlRig.RigUnit_IsInteracting() {is=struct RigVMUnitNode}
   RigVMFunction_ControlFlowBranch = &RigVMFunction_ControlFlowBranch(Condition=@n6.bIsTranslating, ExecuteContext=@n5.True)
   n7 = &RigUnit_IsInteracting()
   n8 = &RigVMFunction_ControlFlowBranch(Condition=@n7.bIsRotating, ExecuteContext=@RigVMFunction_ControlFlowBranch.Completed)
   Get_Module_Metadata_3_1_2_1 = RigVMDispatchNode(Default=(Type=Bone,Name="None"), Name=Shoulder Translation Proxy, NameSpace=Self) {template="DISPATCH_RigDispatch_GetModuleMetadata(in Name,in NameSpace,in Default,out Value,out Found)"}
   n9 = RigVMVariableNode() {variable=Shoulder Bone}
   &RigUnit_ParentConstraint(
    AdvancedSettings=(InterpolationType=Average,RotationOrderForFilter=XZY)
    Parents=((Item=(Type=Bone,Name="None"),Weight=1))
    Weight=1
    Child=&RigUnit_HierarchyGetParent(bDefaultParent=True, Child=@Get_Module_Metadata_3_1_2_1.Value).Parent
    Parents.0.Item=&RigUnit_HierarchyGetParent(bDefaultParent=True, Child=&RigUnit_HierarchyGetParent(bDefaultParent=True, Child=@n9.Value).Parent).Parent
    ExecutePin=@n8.Completed
   )
   For_Each_1 = RigVMDispatchNode(Array=@n7.Items, ExecuteContext=@n8.True) {template="DISPATCH_RigVMDispatch_ArrayIterator(in Array,out Element,out Index,out Count,out Ratio)"}
   Get_Module_Metadata_3_2 = RigVMDispatchNode(Default=(Type=Bone,Name="None"), Name=Shoulder Rotation Control, NameSpace=Self) {template="DISPATCH_RigDispatch_GetModuleMetadata(in Name,in NameSpace,in Default,out Value,out Found)"}
   n10 = &RigVMFunction_ControlFlowBranch(
    Condition=RigVMDispatchNode(A=@For_Each_1.Element, B=@Get_Module_Metadata_3_2.Value) {template="DISPATCH_RigVMDispatch_CoreEquals(in A,in B,out Result)"}.Result
    ExecuteContext=@For_Each_1.ExecuteContext
   )
   Get_Module_Metadata_3_1_1_1 = RigVMDispatchNode(Default=(Type=Bone,Name="None"), Name=Shoulder Rotation Control, NameSpace=Self) {template="DISPATCH_RigDispatch_GetModuleMetadata(in Name,in NameSpace,in Default,out Value,out Found)"}
   Get_Module_Metadata_3_1_2 = RigVMDispatchNode(Default=(Type=Bone,Name="None"), Name=Shoulder Translation Proxy, NameSpace=Self) {template="DISPATCH_RigDispatch_GetModuleMetadata(in Name,in NameSpace,in Default,out Value,out Found)"}
   &RigUnit_ParentConstraint(
    AdvancedSettings=(InterpolationType=Average,RotationOrderForFilter=XZY)
    Parents=((Item=(Type=Bone,Name="None"),Weight=1))
    Weight=1
    Parents.0.Item=@Get_Module_Metadata_3_1_1_1.Value
    Child=@Get_Module_Metadata_3_1_2.Value
    ExecutePin=@n10.True
   )
   For_Each = RigVMDispatchNode(Array=@n6.Items, ExecuteContext=@RigVMFunction_ControlFlowBranch.True) {template="DISPATCH_RigVMDispatch_ArrayIterator(in Array,out Element,out Index,out Count,out Ratio)"}
   Get_Module_Metadata_3 = RigVMDispatchNode(Default=(Type=Bone,Name="None"), Name=Shoulder Translation Proxy, NameSpace=Self) {template="DISPATCH_RigDispatch_GetModuleMetadata(in Name,in NameSpace,in Default,out Value,out Found)"}
   n11 = &RigVMFunction_ControlFlowBranch(
    Condition=RigVMDispatchNode(A=@For_Each.Element, B=@Get_Module_Metadata_3.Value) {template="DISPATCH_RigVMDispatch_CoreEquals(in A,in B,out Result)"}.Result
    ExecuteContext=@For_Each.ExecuteContext
   )
   Get_Module_Metadata_3_1 = RigVMDispatchNode(Default=(Type=Bone,Name="None"), Name=Shoulder Translation Proxy, NameSpace=Self) {template="DISPATCH_RigDispatch_GetModuleMetadata(in Name,in NameSpace,in Default,out Value,out Found)"}
   Get_Module_Metadata_3_1_1 = RigVMDispatchNode(Default=(Type=Bone,Name="None"), Name=Shoulder Rotation Control, NameSpace=Self) {template="DISPATCH_RigDispatch_GetModuleMetadata(in Name,in NameSpace,in Default,out Value,out Found)"}
   &RigUnit_AimConstraintLocalSpaceOffset(
    Filter=(bX=True,bY=True,bZ=True)
    Parents=((Item=(Type=Bone,Name="None"),Weight=1))
    Weight=1
    WorldUp=(Target=(X=0,Y=0,Z=1),Kind=Direction,Space=(Type=None,Name="None"))
    Parents.0.Item=@Get_Module_Metadata_3_1.Value
    Child=@Get_Module_Metadata_3_1_1.Value
    ExecutePin=@n11.True
   )
   RigVMFunctionReturnNode(ExecuteContext=@RigVMFunction_Sequence.E)
  ]}
 New Function_ContainedGraph {nodes [
   RigVMFunctionEntryNode()
   > RigVMVariableNode() {variable=Found}
   > RigVMFunction_Sequence = RigVMAggregateNode()
   n1 = RigVMDispatchNode(Default=(Type=Bone,Name="None"), Name=Shoulder Bone, NameSpace=Self) {template="DISPATCH_RigDispatch_GetModuleMetadata(in Name,in NameSpace,in Default,out Value,out Found)"}
   n2 = &RigVMFunction_ControlFlowBranch(
    Condition=RigVMDispatchNode(
     Array=&RigUnit_CollectionChildrenArray(TypeToSearch=Bone, bDefaultChildren=True, bIncludeParent=False, bRecursive=true, Parent=@n1.Value).Items
     Element=RigVMDispatchNode(Default=(Type=Bone,Name="None"), Name=hand_l, NameSpace=Root) {template="DISPATCH_RigDispatch_GetModuleMetadata(in Name,in NameSpace,in Default,out Value,out Found)"}.Value
    ) {template="DISPATCH_RigVMDispatch_ArrayFind(in Array,in Element,out Index,out Success)"}.Success
    ExecuteContext=@RigVMFunction_Sequence.A
   )
   Create Socket at Metadata_2 = RigVMFunctionReferenceNode(
    Color=(R=1,G=1,B=1,A=1)
    Description=Place the arm module here . . .
    Extra Transform=(Rotation=(X=0,Y=0,Z=0,W=1),Translation=(X=0,Y=0,Z=0),Scale3D=(X=1,Y=1,Z=1))
    Metadata Name=hand_l
    Name=hand_l_socket
    Space=LocalSpace
    ExecuteContext=@n2.True
   )
   > RigVMVariableNode(Value=true) {variable=Found}
   n3 = RigVMDispatchNode(Default=(Type=Bone,Name="None"), Name=Shoulder Bone, NameSpace=Self) {template="DISPATCH_RigDispatch_GetModuleMetadata(in Name,in NameSpace,in Default,out Value,out Found)"}
   n4 = &RigVMFunction_ControlFlowBranch(
    Condition=RigVMDispatchNode(
     Array=&RigUnit_CollectionChildrenArray(TypeToSearch=Bone, bDefaultChildren=True, bIncludeParent=False, bRecursive=true, Parent=@n3.Value).Items
     Element=RigVMDispatchNode(Default=(Type=Bone,Name="None"), Name=hand_r, NameSpace=Root) {template="DISPATCH_RigDispatch_GetModuleMetadata(in Name,in NameSpace,in Default,out Value,out Found)"}.Value
    ) {template="DISPATCH_RigVMDispatch_ArrayFind(in Array,in Element,out Index,out Success)"}.Success
    ExecuteContext=@RigVMFunction_Sequence.B
   )
   Create Socket at Metadata_2_1 = RigVMFunctionReferenceNode(
    Color=(R=1,G=1,B=1,A=1)
    Description=Place the arm module here . . .
    Extra Transform=(Rotation=(X=0,Y=0,Z=0,W=1),Translation=(X=0,Y=0,Z=0),Scale3D=(X=1,Y=1,Z=1))
    Metadata Name=hand_r
    Name=hand_r_socket
    Space=LocalSpace
    ExecuteContext=@n4.True
   )
   > RigVMVariableNode(Value=true) {variable=Found}
   RigVMFunction_MathBoolNot = &RigVMFunction_MathBoolNot(Value=RigVMVariableNode() {variable=Found}.Value)
   n5 = &RigVMFunction_ControlFlowBranch(Condition=@RigVMFunction_MathBoolNot.Result, ExecuteContext=@RigVMFunction_Sequence.C)
   Find first in Hierarchy_0 = RigVMFunctionReferenceNode(
    Children=true
    Ignore=()
    Search for=("hand")
    Type to Search=Bone
    Item=RigVMDispatchNode(Default=(Type=Bone,Name="None"), Name=Shoulder Bone, NameSpace=Self) {template="DISPATCH_RigDispatch_GetModuleMetadata(in Name,in NameSpace,in Default,out Value,out Found)"}.Value
    ExecuteContext=@n5.True
   )
   > &RigUnit_HierarchyAddSocket(
    Color=(R=1,G=1,B=1,A=1)
    Description=Place the arm module here . . .
    Parent=(Type=Bone,Name="hand_l")
    Space=LocalSpace
    Transform=(Rotation=(X=0,Y=0,Z=0,W=1),Translation=(X=0,Y=0,Z=0),Scale3D=(X=1,Y=1,Z=1))
    Name=&RigVMFunction_NameConcat(B=_socket, A=@"Find first in Hierarchy_0".Item Found.Name) {template="Concat::Execute(in A,in B,out Result)"}.Result
   )
   RigVMFunctionReturnNode(ExecuteContext=@RigVMFunction_Sequence.D)
  ]}
 New Function_ContainedGraph {nodes [
   n1 = RigVMFunctionEntryNode()
   Get_Module_Metadata = RigVMDispatchNode(Default=(X=0,Y=0,Z=0), Name=Mirror Axis, NameSpace=Root) {template="DISPATCH_RigDispatch_GetModuleMetadata(in Name,in NameSpace,in Default,out Value,out Found)"}
   Multiply = &RigVMFunction_MathVectorMul(A=&RigUnit_GetTransform(bInitial=False, Item=@n1.Item).Transform.Translation, B=@Get_Module_Metadata.Value) {template="Multiply::Execute(in A,in B,out Result)"}
   RigVMFunction_MathBoolOr = RigVMAggregateNode(
    B=&RigVMFunction_MathDoubleGreater(B=0, A=@Multiply.Result.Y) {template="Greater::Execute(in A,in B,out Result)"}.Result
    C=&RigVMFunction_MathDoubleGreater(B=0, A=@Multiply.Result.Z) {template="Greater::Execute(in A,in B,out Result)"}.Result
    A=&RigVMFunction_MathDoubleGreater(B=0, A=@Multiply.Result.X) {template="Greater::Execute(in A,in B,out Result)"}.Result
   )
   RigVMFunction_MathBoolNot = &RigVMFunction_MathBoolNot(Value=@RigVMFunction_MathBoolOr.Result)
   Less = &RigVMFunction_MathDoubleLess(B=0, A=@Get_Module_Metadata.Value.X) {template="Less::Execute(in A,in B,out Result)"}
   RigVMFunction_MathBoolOr_1 = RigVMAggregateNode(
    A=@Less.Result
    B=&RigVMFunction_MathDoubleLess(B=0, A=@Get_Module_Metadata.Value.Y) {template="Less::Execute(in A,in B,out Result)"}.Result
    C=&RigVMFunction_MathDoubleLess(B=0, A=@Get_Module_Metadata.Value.Z) {template="Less::Execute(in A,in B,out Result)"}.Result
   )
   > RigVMFunctionReturnNode(Result=RigVMDispatchNode(True=@RigVMFunction_MathBoolNot.Result, False=@RigVMFunction_MathBoolOr.Result, Condition=@RigVMFunction_MathBoolOr_1.Result) {template="DISPATCH_RigVMDispatch_If(in Condition,in True,in False,out Result)"}.Result)
  ]}
 New Function_ContainedGraph {nodes [
   n1 = RigVMFunctionEntryNode()
   Less = &RigVMFunction_MathDoubleLess(B=0, A=@n1.Value.X) {template="Less::Execute(in A,in B,out Result)"}
   n2 = &RigVMFunction_MathDoubleLess(B=0, A=@n1.Value.Y) {template="Less::Execute(in A,in B,out Result)"}
   n3 = &RigVMFunction_MathDoubleLess(B=0, A=@n1.Value.Z) {template="Less::Execute(in A,in B,out Result)"}
   RigVMFunction_MathBoolOr_3 = RigVMAggregateNode(A=@Less.Result, B=@n2.Result, C=@n3.Result)
   Greater = &RigVMFunction_MathDoubleGreater(B=0, A=@n1.Value.X) {template="Greater::Execute(in A,in B,out Result)"}
   RigVMFunction_MathBoolOr = &RigVMFunction_MathBoolOr(A=@Greater.Result, B=@Less.Result)
   > RigVMFunctionReturnNode(
    Is Negative=RigVMDispatchNode(False=false, True=true, Condition=@RigVMFunction_MathBoolOr_3.Result) {template="DISPATCH_RigVMDispatch_If(in Condition,in True,in False,out Result)"}.Result
    X=@RigVMFunction_MathBoolOr.Result
    Y=&RigVMFunction_MathBoolOr(A=&RigVMFunction_MathDoubleGreater(B=0, A=@n1.Value.Y) {template="Greater::Execute(in A,in B,out Result)"}.Result, B=@n2.Result).Result
    Z=&RigVMFunction_MathBoolOr(A=&RigVMFunction_MathDoubleGreater(B=0, A=@n1.Value.Z) {template="Greater::Execute(in A,in B,out Result)"}.Result, B=@n3.Result).Result
   )
  ]}
 RigVMFunctionLibrary {nodes [
   Axis to Bool = RigVMCollapseNode(Value=(X=0,Y=0,Z=0))
   CONSTR Shoulder = RigVMCollapseNode(
    ExecuteContext=()
    Mirror Plane=(X=1,Y=1,Z=0)
    Parent Control Connector=(Type=None,Name="None")
    Primary Axis=(X=1,Y=1,Z=1)
    Secondary Axis=(X=0,Y=0,Z=1)
    Settings="(DisplayName=\"None\",bSelectable=True,InitialSpace=LocalSpace,bUsePreferredRotationOrder=False,PreferredRotationOrder=YZX,Limits=(LimitTranslationX=(bMinimum=False,bMaximum=False),LimitTranslationY=(bMinimum=False,bMaximum=False),LimitTranslationZ=(bMinimum ...(truncated)"
    Shoulder Connector=(Type=Bone,Name="None")
   )
   Create Shoulder Sockets = RigVMCollapseNode(ExecuteContext=())
   FWD Shoulder = RigVMCollapseNode(
    Attach Skeleton=true
    Clavicle Bone=(Type=Bone,Name="None")
    Clavicle Control=(Type=Bone,Name="None")
    ExecuteContext=()
    Shoulder Bone=(Type=Bone,Name="None")
   )
   FindClosestBone = RigVMCollapseNode(Bones=(), ExecuteContext=(), Initial=true, Item=())
   Get Side from Metadata = RigVMCollapseNode(ExecuteContext=(), Item=(Type=None,Name="None"))
   INV Shoulder = RigVMCollapseNode(ExecuteContext=())
   Is Right Side v2 = RigVMCollapseNode(ExecuteContext=(), Item=(Type=Bone), Mirror Plane=(X=1,Y=1,Z=0))
  ]}
 RigVMFunction_MathBoolOr_1_ContainedGraph {nodes [
   n1 = RigVMFunctionEntryNode()
   RigVMFunction_MathBoolOr = &RigVMFunction_MathBoolOr(A=@n1.A, B=@n1.B)
   RigVMFunctionReturnNode(Result=&RigVMFunction_MathBoolOr(B=@n1.C, A=@RigVMFunction_MathBoolOr.Result).Result)
  ]}
 RigVMFunction_MathBoolOr_1_ContainedGraph {nodes [
   n1 = RigVMFunctionEntryNode()
   RigVMFunction_MathBoolOr = &RigVMFunction_MathBoolOr(A=@n1.A, B=@n1.B)
   RigVMFunctionReturnNode(Result=&RigVMFunction_MathBoolOr(B=@n1.C, A=@RigVMFunction_MathBoolOr.Result).Result)
  ]}
 RigVMFunction_MathBoolOr_3_1_ContainedGraph {nodes [
   n1 = RigVMFunctionEntryNode()
   RigVMFunctionReturnNode(Result=&RigVMFunction_MathBoolOr(B=@n1.C, A=&RigVMFunction_MathBoolOr(A=@n1.A, B=@n1.B).Result).Result)
  ]}
 RigVMFunction_Sequence_1_ContainedGraph {nodes [
   n1 = RigVMFunctionEntryNode()
   RigVMFunction_Sequence = &RigVMFunction_Sequence(ExecuteContext=@n1.ExecuteContext)
   n2 = &RigVMFunction_Sequence(ExecuteContext=@RigVMFunction_Sequence.B)
   n3 = &RigVMFunction_Sequence(ExecuteContext=@n2.B)
   RigVMFunctionReturnNode(A=@RigVMFunction_Sequence.A, B=@n2.A, C=@n3.A, D=@n3.B)
  ]}
 RigVMFunction_Sequence_1_ContainedGraph {nodes [
   n1 = RigVMFunctionEntryNode()
   RigVMFunction_Sequence = &RigVMFunction_Sequence(ExecuteContext=@n1.ExecuteContext)
   n2 = &RigVMFunction_Sequence(ExecuteContext=@RigVMFunction_Sequence.B)
   n3 = &RigVMFunction_Sequence(ExecuteContext=@n2.B)
   RigVMFunctionReturnNode(A=@RigVMFunction_Sequence.A, B=@n2.A, C=@n3.A, D=@n3.B)
  ]}
 RigVMFunction_Sequence_1_ContainedGraph {nodes [
   n1 = RigVMFunctionEntryNode()
   RigVMFunction_Sequence = &RigVMFunction_Sequence(ExecuteContext=@n1.ExecuteContext)
   n2 = &RigVMFunction_Sequence(ExecuteContext=@RigVMFunction_Sequence.B)
   n3 = &RigVMFunction_Sequence(ExecuteContext=@n2.B)
   RigVMFunctionReturnNode(A=@RigVMFunction_Sequence.A, B=@n2.A, C=@n3.A, D=@n3.B)
  ]}
 RigVMFunction_Sequence_1_ContainedGraph {nodes [
   n1 = RigVMFunctionEntryNode()
   RigVMFunction_Sequence = &RigVMFunction_Sequence(ExecuteContext=@n1.ExecuteContext)
   n2 = &RigVMFunction_Sequence(ExecuteContext=@RigVMFunction_Sequence.B)
   RigVMFunctionReturnNode(A=@RigVMFunction_Sequence.A, B=@n2.A, C=@n2.B)
  ]}
 RigVMFunction_Sequence_1_ContainedGraph {nodes [
   n1 = RigVMFunctionEntryNode()
   RigVMFunction_Sequence = &RigVMFunction_Sequence(ExecuteContext=@n1.ExecuteContext)
   n2 = &RigVMFunction_Sequence(ExecuteContext=@RigVMFunction_Sequence.B)
   n3 = &RigVMFunction_Sequence(ExecuteContext=@n2.B)
   n4 = &RigVMFunction_Sequence(ExecuteContext=@n3.B)
   n5 = &RigVMFunction_Sequence(ExecuteContext=@n4.B)
   RigVMFunctionReturnNode(A=@RigVMFunction_Sequence.A, B=@n2.A, C=@n3.A, D=@n4.A, E=@n5.A, F=@n5.B)
  ]}
 RigVMFunction_Sequence_1_ContainedGraph {nodes [
   n1 = RigVMFunctionEntryNode()
   RigVMFunction_Sequence = &RigVMFunction_Sequence(ExecuteContext=@n1.ExecuteContext)
   n2 = &RigVMFunction_Sequence(ExecuteContext=@RigVMFunction_Sequence.B)
   n3 = &RigVMFunction_Sequence(ExecuteContext=@n2.B)
   n4 = &RigVMFunction_Sequence(ExecuteContext=@n3.B)
   RigVMFunctionReturnNode(A=@RigVMFunction_Sequence.A, B=@n2.A, C=@n3.A, D=@n4.A, E=@n4.B)
  ]}
 RigVMFunction_Sequence_1_ContainedGraph {nodes [
   n1 = RigVMFunctionEntryNode()
   RigVMFunction_Sequence = &RigVMFunction_Sequence(ExecuteContext=@n1.ExecuteContext)
   n2 = &RigVMFunction_Sequence(ExecuteContext=@RigVMFunction_Sequence.B)
   n3 = &RigVMFunction_Sequence(ExecuteContext=@n2.B)
   RigVMFunctionReturnNode(A=@RigVMFunction_Sequence.A, B=@n2.A, C=@n3.A, D=@n3.B)
  ]}
 RigVMModel {nodes [
   RigUnit_BeginExecution = &RigUnit_BeginExecution()
   DISPATCH_RigDispatch_GetModuleMetadata = RigVMDispatchNode(Default=(Type=Bone,Name="None"), Name=Clavicle Rig Bone, NameSpace=Self) {template="DISPATCH_RigDispatch_GetModuleMetadata(in Name,in NameSpace,in Default,out Value,out Found)"}
   DISPATCH_RigVMDispatch_ArrayGetAtIndex = RigVMDispatchNode(Index=0, Array=&RigUnit_FindItemsWithMetadataTag(NameSpace=Self, Tag=Controls).Items) {template="DISPATCH_RigVMDispatch_ArrayGetAtIndex(in Array,in Index,out Element)"}
   FWD Shoulder = RigVMFunctionReferenceNode(
    Attach Skeleton=true
    Clavicle Bone=@DISPATCH_RigDispatch_GetModuleMetadata.Value
    Shoulder Bone=RigVMDispatchNode(Default=(Type=Bone,Name="None"), Name=Shoulder Rig Bone, NameSpace=Self) {template="DISPATCH_RigDispatch_GetModuleMetadata(in Name,in NameSpace,in Default,out Value,out Found)"}.Value
    Clavicle Control=@DISPATCH_RigVMDispatch_ArrayGetAtIndex.Element
    ExecuteContext=@RigUnit_BeginExecution.ExecutePin
   )
  ]}
 RigVMModel Backwards Solve Graph {nodes [
   n1 = /Script/ControlRig.RigUnit_InverseExecution() {is=struct RigVMUnitNode}
   RigVMFunction_Sequence = RigVMAggregateNode(ExecuteContext=@n1.ExecutePin)
   Get_Module_Metadata_4 = RigVMDispatchNode(Default=(Type=Bone,Name="None"), Name=Shoulder Rotation Control, NameSpace=Self) {template="DISPATCH_RigDispatch_GetModuleMetadata(in Name,in NameSpace,in Default,out Value,out Found)"}
   SetAnimationChannel = /Script/ControlRig.RigUnit_SetBoolAnimationChannel(Channel=Aim at Proxy, Value=true, bInitial=False, Control=@Get_Module_Metadata_4.Value.Name, ExecutePin=@RigVMFunction_Sequence.A) {is=struct RigVMUnitNode, template="SetAnimationChannel::Execute(in Value,in Control,in Channel,in bInitial)"}
   Get_Module_Metadata_1 = RigVMDispatchNode(Default=(Type=Bone,Name="None"), Name=Clavicle Rig Bone, NameSpace=Self) {template="DISPATCH_RigDispatch_GetModuleMetadata(in Name,in NameSpace,in Default,out Value,out Found)"}
   Get_Module_Metadata_1_1 = RigVMDispatchNode(Default=(Type=Bone,Name="None"), Name=Clavicle Bone, NameSpace=Self) {template="DISPATCH_RigDispatch_GetModuleMetadata(in Name,in NameSpace,in Default,out Value,out Found)"}
   Parent = RigVMFunctionReferenceNode(Child=@Get_Module_Metadata_1.Value, New Parent=@Get_Module_Metadata_1_1.Value, ExecuteContext=@RigVMFunction_Sequence.B)
   Get_Module_Metadata = RigVMDispatchNode(Default=(Type=Bone,Name="None"), Name=Shoulder Rig Bone, NameSpace=Self) {template="DISPATCH_RigDispatch_GetModuleMetadata(in Name,in NameSpace,in Default,out Value,out Found)"}
   Get_Module_Metadata_2 = RigVMDispatchNode(Default=(Type=Bone,Name="None"), Name=Shoulder Bone, NameSpace=Self) {template="DISPATCH_RigDispatch_GetModuleMetadata(in Name,in NameSpace,in Default,out Value,out Found)"}
   > Parent_2 = RigVMFunctionReferenceNode(Child=@Get_Module_Metadata.Value, New Parent=@Get_Module_Metadata_2.Value)
   RigVMFunction_ControlFlowBranch = &RigVMFunction_ControlFlowBranch(ExecuteContext=@RigVMFunction_Sequence.C, Condition=RigVMVariableNode() {variable=Enable Translation}.Value)
   Parent_1 = RigVMFunctionReferenceNode(
    New Parent=&RigUnit_HierarchyGetParent(
     bDefaultParent=True
     Child=&RigUnit_ResolveConnector(Connector=(Type=Connector,Name="Shoulder Primary"), SkipSocket=true).Result
    ).Parent
    Child=&RigUnit_HierarchyGetParent(bDefaultParent=True, Child=RigVMVariableNode() {variable=Shoulder Ctrl}.Value).Parent
    ExecuteContext=@RigVMFunction_ControlFlowBranch.Completed
   )
   Parent_1_1 = RigVMFunctionReferenceNode(
    Child=&RigUnit_HierarchyGetParent(bDefaultParent=True, Child=RigVMVariableNode() {variable=Shoulder Proxy Ctrl}.Value).Parent
    New Parent=&RigUnit_ResolveConnector(Connector=(Type=Connector,Name="Shoulder Primary"), SkipSocket=true).Result
    ExecuteContext=@RigVMFunction_ControlFlowBranch.True
   )
  ]}
 RigVMModel Connector Event Graph {nodes [
   n1 = /Script/ControlRig.RigUnit_ConnectorExecution() {is=struct RigVMUnitNode}
   Connect to Module Metadata_1 = RigVMFunctionReferenceNode(
    Connector=(Type=Connector,Name="Parent Control")
    Module Metadata Name=Chest Space Null
    Module Metadata NameSpace=Root
    ExecuteContext=@n1.ExecuteContext
   )
   > Connect to Module Metadata_1_1 = RigVMFunctionReferenceNode(Connector=(Type=Connector,Name="Chest Bone"), Module Metadata Name=Chest Bone, Module Metadata NameSpace=Root)
  ]}
]
hierarchy [
 Root {type=Socket, loc=<0 0 0>, rot=<0 0 0 1>}
 Shoulder Primary {type=Connector, loc=<0 0 0>, rot=<0 0 0 1>}
 Parent Control {type=Connector, loc=<0 0 0>, rot=<0 0 0 1>}
 Chest Bone {type=Connector, loc=<0 0 0>, rot=<0 0 0 1>}
]
variables [
 Clavicle Control Settings {
  type=FRigUnit_HierarchyAddControlTransform_Settings
  default="(InitialSpace=LocalSpace,bUsePreferredRotationOrder=False,PreferredRotationOrder=YZX,Limits=(LimitTranslationX=(bMinimum=False,bMaximum=False),LimitTranslationY=(bMinimum=False,bMaximum=False),LimitTranslationZ=(bMinimum=False,bMaximum=False),LimitPitch=(b ...(truncated)"
 }
 Enable Translation {type=bool, default=False}
 Found {type=bool, default=False}
 Module Settings {
  type=FUserDefinedStruct_#8
  default="(OverrideColor_48_#6=False,LeftColor_3_#5=(R=0,G=0.0231,B=0.4115,A=1),RightColor_5_#7=(R=1,G=0,B=0,A=1),CenterColor_9_#4=(R=1,G=1,B=0,A= ...(truncated)"
 }
 Shoulder Ctrl {type=FRigElementKey, default=(Type=None,Name="")}
 Shoulder Primary {type=FRigElementKey, default=(Type=Connector,Name="")}
 Shoulder Proxy Ctrl {type=FRigElementKey, default=(Type=None,Name="")}
]
