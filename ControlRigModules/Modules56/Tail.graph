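# Modular Control Rig tail module that builds a tail rig with spline-based IK controls, FK controls and an IK/FK switch.
# Construction creates rig bones, IK and FK controls and the switch, then scales and colors controls from Module Settings.
# Bones fit to a control-driven spline, optionally stretched, with an offset secondary spline as aim target for twist.
# Dynamics toggle, strength, damping and blend values are read from animation channels found through module metadata.
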ControlRigBlueprint /ControlRigModules/Modules56/Tail
graphs [
 CollapseNode_ContainedGraph {nodes [
   RigVMFunctionEntryNode()
   DISPATCH_RigDispatch_GetModuleMetadata = RigVMDispatchNode(Default=(Type=Bone,Name="None"), Name=Dynamics Channel, NameSpace=Self) {template="DISPATCH_RigDispatch_GetModuleMetadata(in Name,in NameSpace,in Default,out Value,out Found)"}
   n1 = RigVMDispatchNode(Default=(Type=Bone,Name="None"), Name=Channel Parent, NameSpace=Self) {template="DISPATCH_RigDispatch_GetModuleMetadata(in Name,in NameSpace,in Default,out Value,out Found)"}
   GetAnimationChannel = /Script/ControlRig.RigUnit_GetBoolAnimationChannel(bInitial=False, Channel=@DISPATCH_RigDispatch_GetModuleMetadata.Value.Name, Control=@n1.Value.Name) {is=struct RigVMUnitNode, template="GetAnimationChannel::Execute(out Value,in Control,in Channel,in bInitial)"}
   RigVMFunctionReturnNode(
    Value=@GetAnimationChannel.Value
    Value_1=/Script/ControlRig.RigUnit_GetFloatAnimationChannel(
     bInitial=False
     Channel=RigVMDispatchNode(Default=(Type=Bone,Name="None"), Name=Strength Channel, NameSpace=Self) {template="DISPATCH_RigDispatch_GetModuleMetadata(in Name,in NameSpace,in Default,out Value,out Found)"}.Value.Name
     Control=@n1.Value.Name
    ) {is=struct RigVMUnitNode, template="GetAnimationChannel::Execute(out Value,in Control,in Channel,in bInitial)"}.Value
    Value_2=&RigUnit_GetFloatAnimationChannel(
     bInitial=False
     Channel=RigVMDispatchNode(Default=(Type=Bone,Name="None"), Name=Damping Channel, NameSpace=Self) {template="DISPATCH_RigDispatch_GetModuleMetadata(in Name,in NameSpace,in Default,out Value,out Found)"}.Value.Name
     Control=@n1.Value.Name
    ) {template="GetAnimationChannel::Execute(out Value,in Control,in Channel,in bInitial)"}.Value
    Value_3=&RigUnit_GetFloatAnimationChannel(
     bInitial=False
     Channel=RigVMDispatchNode(Default=(Type=Bone,Name="None"), Name=Blend Start Channel, NameSpace=Self) {template="DISPATCH_RigDispatch_GetModuleMetadata(in Name,in NameSpace,in Default,out Value,out Found)"}.Value.Name
     Control=@n1.Value.Name
    ) {template="GetAnimationChannel::Execute(out Value,in Control,in Channel,in bInitial)"}.Value
    Value_4=&RigUnit_GetFloatAnimationChannel(
     bInitial=False
     Channel=RigVMDispatchNode(Default=(Type=Bone,Name="None"), Name=Blend End Channel, NameSpace=Self) {template="DISPATCH_RigDispatch_GetModuleMetadata(in Name,in NameSpace,in Default,out Value,out Found)"}.Value.Name
     Control=@n1.Value.Name
    ) {template="GetAnimationChannel::Execute(out Value,in Control,in Channel,in bInitial)"}.Value
    Value_5=&RigUnit_GetFloatAnimationChannel(
     bInitial=False
     Channel=RigVMDispatchNode(Default=(Type=Bone,Name="None"), Name=Blend Twist Channel, NameSpace=Self) {template="DISPATCH_RigDispatch_GetModuleMetadata(in Name,in NameSpace,in Default,out Value,out Found)"}.Value.Name
     Control=@n1.Value.Name
    ) {template="GetAnimationChannel::Execute(out Value,in Control,in Channel,in bInitial)"}.Value
   )
  ]}
 CollapseNode_ContainedGraph {nodes [
   n1 = RigVMFunctionEntryNode()
   RigVMFunction_StringSplit = /Script/RigVM.RigVMFunction_StringSplit(Separator=_, Value=RigVMDispatchNode(Value=@n1.Value) {template="DISPATCH_RigDispatch_ToString(in Value,out Result)"}.Result) {is=struct RigVMUnitNode}
   > RigVMFunctionReturnNode(Result=RigVMDispatchNode(String=RigVMDispatchNode(Index=0, Array=@RigVMFunction_StringSplit.Result) {template="DISPATCH_RigVMDispatch_ArrayGetAtIndex(in Array,in Index,out Element)"}.Element) {template="DISPATCH_RigDispatch_FromString(in String,out Result)"}.Result)
  ]}
 Concat_1_1_ContainedGraph {nodes [
   n1 = RigVMFunctionEntryNode()
   n2 = /Script/RigVM.RigVMFunction_NameConcat(A=@n1.A, B=@n1.B) {is=struct RigVMUnitNode, template="Concat::Execute(in A,in B,out Result)"}
   RigVMFunctionReturnNode(Result=&RigVMFunction_NameConcat(A=&RigVMFunction_NameConcat(A=@n2.Result, B=@n1.C) {template="Concat::Execute(in A,in B,out Result)"}.Result, B=@n1.D) {template="Concat::Execute(in A,in B,out Result)"}.Result)
  ]}
 Concat_1_1_ContainedGraph {nodes [
   n1 = RigVMFunctionEntryNode()
   n2 = &RigVMFunction_NameConcat(A=@n1.A, B=@n1.B) {template="Concat::Execute(in A,in B,out Result)"}
   RigVMFunctionReturnNode(Result=&RigVMFunction_NameConcat(A=&RigVMFunction_NameConcat(A=@n2.Result, B=@n1.C) {template="Concat::Execute(in A,in B,out Result)"}.Result, B=@n1.D) {template="Concat::Execute(in A,in B,out Result)"}.Result)
  ]}
 Concat_1_1_ContainedGraph {nodes [
   n1 = RigVMFunctionEntryNode()
   n2 = &RigVMFunction_NameConcat(A=@n1.A, B=@n1.B) {template="Concat::Execute(in A,in B,out Result)"}
   RigVMFunctionReturnNode(Result=&RigVMFunction_NameConcat(A=&RigVMFunction_NameConcat(A=@n2.Result, B=@n1.C) {template="Concat::Execute(in A,in B,out Result)"}.Result, B=@n1.D) {template="Concat::Execute(in A,in B,out Result)"}.Result)
  ]}
 Concat_4_1_ContainedGraph {nodes [
   n1 = RigVMFunctionEntryNode()
   n2 = &RigVMFunction_NameConcat(A=@n1.A, B=@n1.B) {template="Concat::Execute(in A,in B,out Result)"}
   RigVMFunctionReturnNode(Result=&RigVMFunction_NameConcat(A=&RigVMFunction_NameConcat(A=@n2.Result, B=@n1.C) {template="Concat::Execute(in A,in B,out Result)"}.Result, B=@n1.D) {template="Concat::Execute(in A,in B,out Result)"}.Result)
  ]}
 Concat_4_1_ContainedGraph {nodes [
   n1 = RigVMFunctionEntryNode()
   n2 = &RigVMFunction_NameConcat(A=@n1.A, B=@n1.B) {template="Concat::Execute(in A,in B,out Result)"}
   RigVMFunctionReturnNode(Result=&RigVMFunction_NameConcat(A=&RigVMFunction_NameConcat(A=@n2.Result, B=@n1.C) {template="Concat::Execute(in A,in B,out Result)"}.Result, B=@n1.D) {template="Concat::Execute(in A,in B,out Result)"}.Result)
  ]}
 ConstructionGraph {nodes [
   RigUnit_PrepareForExecution = &RigUnit_PrepareForExecution()
   Create Tail Rig Bones = RigVMFunctionReferenceNode(ExecuteContext=@RigUnit_PrepareForExecution.ExecutePin, Enable Orient Bones=RigVMVariableNode() {variable=Re-Orient Controls}.Value)
   n1 = RigVMVariableNode() {variable=Module Settings}
   > Create IK Tail = RigVMFunctionReferenceNode(
    Axis=(X=-1,Y=0,Z=0)
    Dynamics=RigVMVariableNode() {variable=Dynamics}.Value
    Create Control Per Bone=RigVMVariableNode() {variable=Create Control per Bone}.Value
    Up Axis=RigVMVariableNode() {variable=Up Axis}.Value
    Parent Controls to each other=RigVMVariableNode() {variable=Main Controls as FK}.Value
    Local IK Control Settings=RigVMVariableNode() {variable=Local IK Control Settings}.Value
    Debug=RigVMVariableNode() {variable=Debug}.Value
    Dynamics Strength=RigVMVariableNode() {variable=Default Strength}.Value
    Suffix=@n1.Value.ControlSuffix_38_#1
    Prefix=@n1.Value.Prefix_20_#2
    Primary Axis=@n1.Value.PrimaryBoneAxis_32_#3
    Dynamics Damping=RigVMVariableNode() {variable=Default Damping}.Value
    Number of Controls=RigVMVariableNode() {variable=Number of Controls}.Value
    Enable Orient Bones=RigVMVariableNode() {variable=Re-Orient Controls}.Value
    Dynamics Blend Start=RigVMVariableNode() {variable=Default Blend Start}.Value
    Dynamics Blend End=RigVMVariableNode() {variable=Default Blend End}.Value
    Main Control Settings=RigVMVariableNode() {variable=IK Control Settings}.Value
    Number of Main Controls=RigVMVariableNode() {variable=Number of Main Controls}.Value
    Dynamics Blend Twist=RigVMVariableNode() {variable=Default Blend Twist}.Value
    Secondary Spline Direction=RigVMVariableNode() {variable=Secondary Spline Direction}.Value
   )
   > Create FK IK Switch = RigVMFunctionReferenceNode(name=tail, prefix=None, side=None, Initial Value=RigVMVariableNode() {variable=Default IK or FK}.Value)
   Get_Module_Metadata = RigVMDispatchNode(Default=(), Name=Rig Bones, NameSpace=Self) {template="DISPATCH_RigDispatch_GetModuleMetadata(in Name,in NameSpace,in Default,out Value,out Found)"}
   > Create FK Tail Controls = RigVMFunctionReferenceNode(
    Start=RigVMDispatchNode(Index=0, Array=@Get_Module_Metadata.Value) {template="DISPATCH_RigVMDispatch_ArrayGetAtIndex(in Array,in Index,out Element)"}.Element
    End=RigVMDispatchNode(Index=-1, Array=@Get_Module_Metadata.Value) {template="DISPATCH_RigVMDispatch_ArrayGetAtIndex(in Array,in Index,out Element)"}.Element
    Settings=RigVMVariableNode() {variable=FK Control Settings}.Value
   )
   > Set Controls Scale and Color = RigVMFunctionReferenceNode(Module Settings=RigVMVariableNode() {variable=Module Settings}.Value)
  ]}
 ContainedGraph {nodes [
   n1 = RigVMFunctionEntryNode()
   RigUnit_SequenceAggregate_1 = &RigVMFunction_Sequence(ExecuteContext=@n1.ExecuteContext)
   RigUnit_SequenceAggregate = RigVMAggregateNode(ExecuteContext=@RigUnit_SequenceAggregate_1.A)
   BranchNode_4 = &RigVMFunction_ControlFlowBranch(Condition=@n1.X Plane, ExecuteContext=@RigUnit_SequenceAggregate.A)
   n2 = RigVMVariableNode() {variable=vector pos}
   RigVMVariableNode(
    Value=(X=0,Y=0,Z=0)
    ExecuteContext=@BranchNode_4.True
    Value.X=&RigUnit_GetTransform(bInitial=true, Item=@n1.Item).Transform.Translation.X
    Value.Y=@n2.Value.Y
    Value.Z=@n2.Value.Z
   ) {variable=vector pos}
   BranchNode_1_1 = &RigVMFunction_ControlFlowBranch(Condition=@n1.Y Plane, ExecuteContext=@RigUnit_SequenceAggregate.B)
   n3 = RigVMVariableNode() {variable=vector pos}
   RigVMVariableNode(
    Value=(X=0,Y=0,Z=0)
    ExecuteContext=@BranchNode_1_1.True
    Value.Y=&RigUnit_GetTransform(bInitial=true, Item=@n1.Item).Transform.Translation.Y
    Value.X=@n3.Value.X
    Value.Z=@n3.Value.Z
   ) {variable=vector pos}
   BranchNode_2_1 = &RigVMFunction_ControlFlowBranch(Condition=@n1.Z Plane, ExecuteContext=@RigUnit_SequenceAggregate.C)
   n4 = RigVMVariableNode() {variable=vector pos}
   RigVMVariableNode(
    Value=(X=0,Y=0,Z=0)
    ExecuteContext=@BranchNode_2_1.True
    Value.Z=&RigUnit_GetTransform(bInitial=true, Item=@n1.Item).Transform.Translation.Z
    Value.X=@n4.Value.X
    Value.Y=@n4.Value.Y
   ) {variable=vector pos}
   Cast = &RigVMFunction_MathBoolToFloat(Value=@n1.X Plane)
   Dot__ = /Script/RigVM.RigVMFunction_MathVectorDot(
    B=(X=1,Y=1,Z=0)
    B.X=@Cast.Result
    B.Y=&RigVMFunction_MathBoolToFloat(Value=@n1.Y Plane).Result
    B.Z=&RigVMFunction_MathBoolToFloat(Value=@n1.Z Plane).Result
    A=RigVMVariableNode() {variable=vector pos}.Value
   ) {is=struct RigVMUnitNode, template="Dot,|::Execute(in A,in B,out Result)"}
   Less = /Script/RigVM.RigVMFunction_MathDoubleLess(B=0, A=@Dot__.Result) {is=struct RigVMUnitNode, template="Less::Execute(in A,in B,out Result)"}
   RigVMVariableNode(Value=@Less.Result, ExecuteContext=@RigUnit_SequenceAggregate.D) {variable=is on right}
   RigVMFunctionReturnNode(ExecuteContext=@RigUnit_SequenceAggregate_1.B, Right Side=RigVMVariableNode() {variable=is on right}.Value)
  ]}
 ContainedGraph {nodes [
   n1 = RigVMFunctionEntryNode()
   RigUnit_SequenceAggregate = &RigVMFunction_Sequence(ExecuteContext=@n1.ExecuteContext)
   RigUnit_SequenceAggregate_1 = &RigVMFunction_Sequence(ExecuteContext=@RigUnit_SequenceAggregate.B)
   RigUnit_SequenceAggregate_2 = &RigVMFunction_Sequence(ExecuteContext=@RigUnit_SequenceAggregate_1.B)
   RigVMFunctionReturnNode(A=@RigUnit_SequenceAggregate.A, B=@RigUnit_SequenceAggregate_1.A, C=@RigUnit_SequenceAggregate_2.A, D=@RigUnit_SequenceAggregate_2.B)
  ]}
 ContainedGraph {nodes [
   BranchNode_RigVMUnitNode = &RigVMFunction_ControlFlowBranch(ExecuteContext=(), Condition=RigVMVariableNode() {variable=debug}.Value)
   /Script/ControlRigSpline.RigUnit_DrawControlRigSpline(
    Color=(R=1,G=0,B=0,A=1)
    Detail=16
    Thickness=0.75
    ExecutePin=@BranchNode_RigVMUnitNode.True
    Spline=RigVMVariableNode() {variable=template_spline}.Value
   ) {is=struct RigVMUnitNode}
   n1 = RigVMFunctionEntryNode()
   RigVMFunction_Sequence = &RigVMFunction_Sequence(ExecuteContext=@n1.ExecuteContext)
   n2 = RigVMVariableNode() {variable=Controls}
   BranchNode_4_RigVMUnitNode = &RigVMFunction_ControlFlowBranch(
    Condition=/Script/RigVM.RigVMFunction_MathIntLess(B=4, A=RigVMDispatchNode(Array=@n2.Value) {template="DISPATCH_RigVMDispatch_ArrayGetNum(in Array,out Num)"}.Num) {is=struct RigVMUnitNode, template="Less::Execute(in A,in B,out Result)"}.Result
    ExecuteContext=@RigVMFunction_Sequence.B
   )
   RigVMFunctionReturnNode(ExecuteContext=@BranchNode_4_RigVMUnitNode.Completed, Output Spline=RigVMVariableNode() {variable=local_spline}.Value)
   RigUnit_SequenceExecution_1 = RigVMAggregateNode(ExecuteContext=@BranchNode_4_RigVMUnitNode.False)
   RigVMVariableNode(Value=@n1.Template Spline, ExecuteContext=@RigUnit_SequenceExecution_1.A) {variable=template_spline}
   n3 = /Script/ControlRig.RigUnit_GetTransformItemArray(Space=GlobalSpace, bInitial=False, Items=RigVMVariableNode() {variable=Controls}.Value) {is=struct RigVMUnitNode}
   > n4 = /Script/ControlRigSpline.RigUnit_SetSplineTransforms(Spline=@n1.Template Spline, Transforms=@n3.Transforms) {is=struct RigVMUnitNode}
   > RigVMVariableNode(Value=@n4.Spline) {variable=local_spline}
   > RigVMVariableNode(Value=@n3.Transforms) {variable=template_transforms}
   > RigUnit_SequenceExecution = RigVMAggregateNode()
   n5 = RigVMVariableNode() {variable=template_transforms}
   RigVMDispatchNode(
    Num=RigVMDispatchNode(Array=@n5.Value) {template="DISPATCH_RigVMDispatch_ArrayGetNum(in Array,out Num)"}.Num
    ExecuteContext=@RigUnit_SequenceExecution.A
    Array=RigVMVariableNode() {variable=offset_positions}.Value
   ) {template="DISPATCH_RigVMDispatch_ArraySetNum(io Array,in Num)"}
   > n6 = RigVMDispatchNode(Array=@n5.Value) {template="DISPATCH_RigVMDispatch_ArrayIterator(in Array,out Element,out Index,out Count,out Ratio)"}
   > RigVMDispatchNode(
    Index=@n6.Index
    Element=/Script/RigVM.RigVMFunction_MathTransformTransformVector(Transform=@n6.Element, Location=RigVMVariableNode() {variable=Secondary Spline Direction}.Value) {is=struct RigVMUnitNode}.Result
    Array=RigVMVariableNode() {variable=offset_positions}.Value
   ) {template="DISPATCH_RigVMDispatch_ArraySetAtIndex(io Array,in Index,in Element)"}
   RigUnit_SequenceExecution_4 = RigVMAggregateNode(ExecuteContext=@RigUnit_SequenceExecution.C)
   Branch_RigVMUnitNode = &RigVMFunction_ControlFlowBranch(ExecuteContext=@RigUnit_SequenceExecution_4.A, Condition=RigVMVariableNode() {variable=Stretch}.Value)
   n7 = RigVMVariableNode() {variable=template_spline}
   n8 = RigVMVariableNode() {variable=Primary Axis}
   n9 = RigVMVariableNode() {variable=Bones}
   /Script/ControlRigSpline.RigUnit_FitChainToSplineCurveItemArray(
    Alignment=Front
    DebugSettings=(Scale=1,CurveColor=(R=1,G=1,B=0,A=1),SegmentsColor=(R=1,G=0,B=0,A=1),WorldOffset=(Rotation=(X=0,Y=0,Z=0,W=1),Translation=(X=0,Y=0,Z=0),Scale3D=(X=1,Y=1,Z=1)))
    Maximum=1
    PoleVectorPosition=(X=0,Y=0,Z=0)
    Rotations=()
    SamplingPrecision=12
    SecondaryAxis=(X=0,Y=0,Z=0)
    Weight=1
    bPropagateToChildren=True
    ExecutePin=@Branch_RigVMUnitNode.False
    Spline=@n7.Value
    PrimaryAxis=@n8.Value
    Items=@n9.Value
   ) {is=struct RigVMUnitNode}
   &RigUnit_FitChainToSplineCurveItemArray(
    Alignment=Stretched
    DebugSettings=(Scale=1,CurveColor=(R=1,G=1,B=0,A=1),SegmentsColor=(R=1,G=0,B=0,A=1),WorldOffset=(Rotation=(X=0,Y=0,Z=0,W=1),Translation=(X=0,Y=0,Z=0),Scale3D=(X=1,Y=1,Z=1)))
    Maximum=1
    PoleVectorPosition=(X=0,Y=0,Z=0)
    Rotations=()
    SamplingPrecision=12
    SecondaryAxis=(X=0,Y=0,Z=0)
    Weight=1
    bPropagateToChildren=True
    ExecutePin=@Branch_RigVMUnitNode.True
    Spline=@n7.Value
    PrimaryAxis=@n8.Value
    Items=@n9.Value
   )
   RigUnit_SequenceExecution_2 = RigVMAggregateNode(ExecuteContext=@RigUnit_SequenceExecution_4.B)
   BranchNode_2_RigVMUnitNode = &RigVMFunction_ControlFlowBranch(ExecuteContext=@RigUnit_SequenceExecution_2.A, Condition=RigVMVariableNode() {variable=debug}.Value)
   &RigUnit_DrawControlRigSpline(
    Color=(R=1,G=0,B=0,A=1)
    Detail=16
    Thickness=1
    ExecutePin=@BranchNode_2_RigVMUnitNode.True
    Spline=RigVMVariableNode() {variable=local_spline}.Value
   )
   n10 = /Script/ControlRigSpline.RigUnit_SetSplinePoints(
    ExecutePin=@RigUnit_SequenceExecution_2.B
    Points=RigVMVariableNode() {variable=offset_positions}.Value
    Spline=RigVMVariableNode() {variable=local_spline}.Value
   ) {is=struct RigVMUnitNode}
   > RigVMVariableNode(Value=@n10.Spline) {variable=local_spline}
   n11 = RigVMVariableNode() {variable=Bones}
   DISPATCH_RigVMDispatch_ArrayGetNum = RigVMDispatchNode(Array=@n11.Value) {template="DISPATCH_RigVMDispatch_ArrayGetNum(in Array,out Num)"}
   > DISPATCH_RigVMDispatch_ArraySetNum = RigVMDispatchNode(Num=@DISPATCH_RigVMDispatch_ArrayGetNum.Num, Array=RigVMVariableNode() {variable=aim_transforms}.Value) {template="DISPATCH_RigVMDispatch_ArraySetNum(io Array,in Num)"}
   > n12 = RigVMDispatchNode(Array=@n11.Value) {template="DISPATCH_RigVMDispatch_ArrayIterator(in Array,out Element,out Index,out Count,out Ratio)"}
   n13 = &RigUnit_GetTransform(Item=@n12.Element)
   n14 = RigVMVariableNode() {variable=local_spline}
   > DISPATCH_RigVMDispatch_ArraySetAtIndex = RigVMDispatchNode(
    Index=@n12.Index
    Element=/Script/ControlRigSpline.RigUnit_PositionFromControlRigSpline(
     U=/Script/ControlRigSpline.RigUnit_ClosestParameterFromControlRigSpline(Position=@n13.Transform.Translation, Spline=@n14.Value) {is=struct RigVMUnitNode}.U
     Spline=@n14.Value
    ) {is=struct RigVMUnitNode}.Position
    Array=RigVMVariableNode() {variable=aim_transforms}.Value
   ) {template="DISPATCH_RigVMDispatch_ArraySetAtIndex(io Array,in Index,in Element)"}
   n15 = RigVMDispatchNode(ExecuteContext=@RigUnit_SequenceExecution.D, Array=RigVMVariableNode() {variable=Bones}.Value) {template="DISPATCH_RigVMDispatch_ArrayIterator(in Array,out Element,out Index,out Count,out Ratio)"}
   DISPATCH_RigVMDispatch_ArrayGetAtIndex = RigVMDispatchNode(Index=@n15.Index, Array=RigVMVariableNode() {variable=aim_transforms}.Value) {template="DISPATCH_RigVMDispatch_ArrayGetAtIndex(in Array,in Index,out Element)"}
   n16 = /Script/ControlRig.RigUnit_AimBoneMath(
    DebugSettings=(Scale=10,WorldOffset=(Rotation=(X=0,Y=0,Z=0,W=1),Translation=(X=0,Y=0,Z=0),Scale3D=(X=1,Y=1,Z=1)))
    Primary=(Weight=1,Axis=(X=0,Y=1,Z=0),Target=(X=1,Y=0,Z=0),Kind=Location,Space=(Type=Bone))
    Secondary=(Weight=0,Axis=(X=0,Y=0,Z=0),Target=(X=0,Y=0,Z=0),Kind=Location,Space=(Type=Bone))
    Weight=1
    Primary.Target=@DISPATCH_RigVMDispatch_ArrayGetAtIndex.Element
    InputTransform=&RigUnit_GetTransform(Item=@n15.Element).Transform
    Primary.Axis=RigVMVariableNode() {variable=Up Axis}.Value
   ) {is=struct RigVMUnitNode}
   MathIntEquals = RigVMDispatchNode(
    A=@n15.Index
    B=/Script/RigVM.RigVMFunction_MathIntSub(B=1, A=@n15.Count) {is=struct RigVMUnitNode, template="Subtract::Execute(in A,in B,out Result)"}.Result
   ) {template="DISPATCH_RigVMDispatch_CoreEquals(in A,in B,out Result)"}
   > n17 = &RigUnit_SetTransform(Weight=1, bInitial=False, Value=@n16.Result, Item=@n15.Element, bPropagateToChildren=@MathIntEquals.Result)
   BranchNode_3_RigVMUnitNode = &RigVMFunction_ControlFlowBranch(ExecuteContext=@n17.ExecutePin, Condition=RigVMVariableNode() {variable=debug}.Value)
   /Script/ControlRig.RigUnit_DebugTransformMutableItemSpace(
    Color=(R=1,G=0,B=0,A=1)
    DebugDrawSettings=(DepthPriority=SDPG_Foreground,Lifetime=-1)
    Mode=Axes
    Scale=10
    Space=(Type=Bone)
    Thickness=0.15
    WorldOffset=(Rotation=(X=0,Y=0,Z=0,W=1),Translation=(X=0,Y=0,Z=0),Scale3D=(X=1,Y=1,Z=1))
    bEnabled=True
    Transform=@n16.Result
    ExecutePin=@BranchNode_3_RigVMUnitNode.True
   ) {is=struct RigVMUnitNode}
   Branch_1_RigVMUnitNode = &RigVMFunction_ControlFlowBranch(ExecuteContext=@RigUnit_SequenceExecution_1.B, Condition=RigVMVariableNode() {variable=Stretch}.Value)
   n18 = RigVMDispatchNode(Index=-1, Array=RigVMVariableNode() {variable=Controls}.Value) {template="DISPATCH_RigVMDispatch_ArrayGetAtIndex(in Array,in Index,out Element)"}
   n19 = RigVMDispatchNode(Index=-1, Array=RigVMVariableNode() {variable=Bones}.Value) {template="DISPATCH_RigVMDispatch_ArrayGetAtIndex(in Array,in Index,out Element)"}
   &RigUnit_ParentConstraint(
    AdvancedSettings=(RotationOrderForFilter=XZY)
    Filter=(TranslationFilter=(bX=false,bY=false,bZ=false),RotationFilter=(bX=True,bY=True,bZ=True),ScaleFilter=(bX=True,bY=True,bZ=True))
    Parents=((Item=(Type=Bone),Weight=1))
    Weight=1
    Parents.0.Item=@n18.Element
    Child=@n19.Element
    ExecutePin=@Branch_1_RigVMUnitNode.False
   )
   &RigUnit_ParentConstraint(
    AdvancedSettings=(RotationOrderForFilter=XZY)
    Parents=((Item=(Type=Bone),Weight=1))
    Weight=1
    Parents.0.Item=@n18.Element
    Child=@n19.Element
    ExecutePin=@Branch_1_RigVMUnitNode.True
   )
   RigVMVariableNode() {variable=Spline Mode}
   RigVMVariableNode() {variable=Samples Per Segment}
  ]}
 ContainedGraph {nodes [
   n1 = RigVMFunctionEntryNode()
   RigUnit_SequenceExecution_2 = &RigVMFunction_Sequence(ExecuteContext=@n1.ExecuteContext)
   RigUnit_SequenceExecution_2_1 = &RigVMFunction_Sequence(ExecuteContext=@RigUnit_SequenceExecution_2.B)
   RigUnit_SequenceExecution_2_2 = &RigVMFunction_Sequence(ExecuteContext=@RigUnit_SequenceExecution_2_1.B)
   RigVMFunctionReturnNode(A=@RigUnit_SequenceExecution_2.A, B=@RigUnit_SequenceExecution_2_1.A, C=@RigUnit_SequenceExecution_2_2.A, D=@RigUnit_SequenceExecution_2_2.B)
  ]}
 ContainedGraph {nodes [
   n1 = RigVMFunctionEntryNode()
   RigUnit_SequenceExecution_4 = &RigVMFunction_Sequence(ExecuteContext=@n1.ExecuteContext)
   RigUnit_SequenceExecution_4_1 = &RigVMFunction_Sequence(ExecuteContext=@RigUnit_SequenceExecution_4.B)
   RigUnit_SequenceExecution_4_2 = &RigVMFunction_Sequence(ExecuteContext=@RigUnit_SequenceExecution_4_1.B)
   RigVMFunctionReturnNode(A=@RigUnit_SequenceExecution_4.A, B=@RigUnit_SequenceExecution_4_1.A, C=@RigUnit_SequenceExecution_4_2.A, D=@RigUnit_SequenceExecution_4_2.B)
  ]}
 ContainedGraph {nodes [
   n1 = RigVMFunctionEntryNode()
   RigUnit_SequenceExecution = &RigVMFunction_Sequence(ExecuteContext=@n1.ExecuteContext)
   RigUnit_SequenceExecution_1 = &RigVMFunction_Sequence(ExecuteContext=@RigUnit_SequenceExecution.B)
   RigUnit_SequenceExecution_2 = &RigVMFunction_Sequence(ExecuteContext=@RigUnit_SequenceExecution_1.B)
   RigVMFunctionReturnNode(A=@RigUnit_SequenceExecution.A, B=@RigUnit_SequenceExecution_1.A, C=@RigUnit_SequenceExecution_2.A, D=@RigUnit_SequenceExecution_2.B)
  ]}
 ContainedGraph {nodes [
   n1 = RigVMFunctionEntryNode()
   RigUnit_SequenceExecution_1 = &RigVMFunction_Sequence(ExecuteContext=@n1.ExecuteContext)
   RigUnit_SequenceExecution_1_1 = &RigVMFunction_Sequence(ExecuteContext=@RigUnit_SequenceExecution_1.B)
   RigUnit_SequenceExecution_1_2 = &RigVMFunction_Sequence(ExecuteContext=@RigUnit_SequenceExecution_1_1.B)
   RigVMFunctionReturnNode(A=@RigUnit_SequenceExecution_1.A, B=@RigUnit_SequenceExecution_1_1.A, C=@RigUnit_SequenceExecution_1_2.A, D=@RigUnit_SequenceExecution_1_2.B)
  ]}
 ContainedGraph {nodes [
   n1 = RigVMFunctionEntryNode()
   n2 = RigVMVariableNode() {variable=Controls}
   BranchNode_4_RigVMUnitNode = &RigVMFunction_ControlFlowBranch(
    ExecuteContext=@n1.ExecuteContext
    Condition=&RigVMFunction_MathIntLess(B=4, A=RigVMDispatchNode(Array=@n2.Value) {template="DISPATCH_RigVMDispatch_ArrayGetNum(in Array,out Num)"}.Num) {template="Less::Execute(in A,in B,out Result)"}.Result
   )
   RigVMFunctionReturnNode(
    ExecuteContext=@BranchNode_4_RigVMUnitNode.Completed
    Local Spline=RigVMVariableNode() {variable=local_spline}.Value
    Template Spline=RigVMVariableNode() {variable=template_spline}.Value
   )
   RigUnit_SequenceExecution_1 = RigVMAggregateNode(ExecuteContext=@BranchNode_4_RigVMUnitNode.False)
   RigVMVariableNode(
    Value=&RigUnit_GetTransformItemArray(Space=GlobalSpace, bInitial=true, Items=@n1.Bones).Transforms
    ExecuteContext=@RigUnit_SequenceExecution_1.A
   ) {variable=Original Positions}
   n3 = &RigUnit_GetTransformItemArray(Space=GlobalSpace, bInitial=False, Items=RigVMVariableNode() {variable=Controls}.Value)
   RigVMFunction_MathTransformArrayToSRT = /Script/RigVM.RigVMFunction_MathTransformArrayToSRT(Transforms=@n3.Transforms) {is=struct RigVMUnitNode}
   RigVMVariableNode(
    Value=/Script/ControlRigSpline.RigUnit_ControlRigSplineFromPoints(
     bClosed=False
     Points=@RigVMFunction_MathTransformArrayToSRT.Translations
     SplineMode=RigVMVariableNode() {variable=Spline Mode}.Value
     SamplesPerSegment=RigVMVariableNode() {variable=Samples Per Segment}.Value
    ) {is=struct RigVMUnitNode}.Spline
    ExecuteContext=@RigUnit_SequenceExecution_1.B
   ) {variable=template_spline}
   > RigVMVariableNode(Value=@n3.Transforms) {variable=template_transforms}
   > RigUnit_SequenceExecution = RigVMAggregateNode()
   n4 = RigVMVariableNode() {variable=template_transforms}
   RigVMDispatchNode(
    Num=RigVMDispatchNode(Array=@n4.Value) {template="DISPATCH_RigVMDispatch_ArrayGetNum(in Array,out Num)"}.Num
    ExecuteContext=@RigUnit_SequenceExecution.A
    Array=RigVMVariableNode() {variable=offset_positions}.Value
   ) {template="DISPATCH_RigVMDispatch_ArraySetNum(io Array,in Num)"}
   > n5 = RigVMDispatchNode(Array=@n4.Value) {template="DISPATCH_RigVMDispatch_ArrayIterator(in Array,out Element,out Index,out Count,out Ratio)"}
   > RigVMDispatchNode(
    Index=@n5.Index
    Element=&RigVMFunction_MathTransformTransformVector(Transform=@n5.Element, Location=RigVMVariableNode() {variable=Secondary Spline Direction}.Value).Result
    Array=RigVMVariableNode() {variable=offset_positions}.Value
   ) {template="DISPATCH_RigVMDispatch_ArraySetAtIndex(io Array,in Index,in Element)"}
   RigUnit_SequenceExecution_4 = RigVMAggregateNode(ExecuteContext=@RigUnit_SequenceExecution.C)
   Branch_RigVMUnitNode = &RigVMFunction_ControlFlowBranch(ExecuteContext=@RigUnit_SequenceExecution_4.A, Condition=RigVMVariableNode() {variable=Stretch}.Value)
   n6 = RigVMVariableNode() {variable=template_spline}
   n7 = RigVMVariableNode() {variable=Primary Axis}
   n8 = RigVMVariableNode() {variable=Bones}
   &RigUnit_FitChainToSplineCurveItemArray(
    Alignment=Front
    DebugSettings=(Scale=1,CurveColor=(R=1,G=1,B=0,A=1),SegmentsColor=(R=1,G=0,B=0,A=1),WorldOffset=(Rotation=(X=0,Y=0,Z=0,W=1),Translation=(X=0,Y=0,Z=0),Scale3D=(X=1,Y=1,Z=1)))
    Maximum=1
    PoleVectorPosition=(X=0,Y=0,Z=0)
    Rotations=()
    SamplingPrecision=12
    SecondaryAxis=(X=0,Y=0,Z=0)
    Weight=1
    bPropagateToChildren=True
    ExecutePin=@Branch_RigVMUnitNode.False
    Spline=@n6.Value
    PrimaryAxis=@n7.Value
    Items=@n8.Value
   )
   &RigUnit_FitChainToSplineCurveItemArray(
    Alignment=Stretched
    DebugSettings=(Scale=1,CurveColor=(R=1,G=1,B=0,A=1),SegmentsColor=(R=1,G=0,B=0,A=1),WorldOffset=(Rotation=(X=0,Y=0,Z=0,W=1),Translation=(X=0,Y=0,Z=0),Scale3D=(X=1,Y=1,Z=1)))
    Maximum=1
    PoleVectorPosition=(X=0,Y=0,Z=0)
    Rotations=()
    SamplingPrecision=12
    SecondaryAxis=(X=0,Y=0,Z=0)
    Weight=1
    bPropagateToChildren=True
    ExecutePin=@Branch_RigVMUnitNode.True
    Spline=@n6.Value
    PrimaryAxis=@n7.Value
    Items=@n8.Value
   )
   RigUnit_SequenceExecution_2 = RigVMAggregateNode(ExecuteContext=@RigUnit_SequenceExecution_4.B)
   n9 = RigVMVariableNode(
    Value=&RigUnit_ControlRigSplineFromPoints(
     Stretch=0
     bClosed=False
     SplineMode=RigVMVariableNode() {variable=Spline Mode}.Value
     SamplesPerSegment=RigVMVariableNode() {variable=Samples Per Segment}.Value
     Points=RigVMVariableNode() {variable=offset_positions}.Value
    ).Spline
    ExecuteContext=@RigUnit_SequenceExecution_2.A
   ) {variable=local_spline}
   BranchNode_2_RigVMUnitNode = &RigVMFunction_ControlFlowBranch(ExecuteContext=@n9.ExecuteContext, Condition=RigVMVariableNode() {variable=debug}.Value)
   &RigUnit_DrawControlRigSpline(
    Color=(R=1,G=0,B=0,A=1)
    Detail=16
    Thickness=1
    ExecutePin=@BranchNode_2_RigVMUnitNode.True
    Spline=RigVMVariableNode() {variable=local_spline}.Value
   )
   n10 = RigVMVariableNode() {variable=Bones}
   DISPATCH_RigVMDispatch_ArrayGetNum = RigVMDispatchNode(Array=@n10.Value) {template="DISPATCH_RigVMDispatch_ArrayGetNum(in Array,out Num)"}
   DISPATCH_RigVMDispatch_ArraySetNum = RigVMDispatchNode(
    Num=@DISPATCH_RigVMDispatch_ArrayGetNum.Num
    ExecuteContext=@RigUnit_SequenceExecution_2.B
    Array=RigVMVariableNode() {variable=aim_transforms}.Value
   ) {template="DISPATCH_RigVMDispatch_ArraySetNum(io Array,in Num)"}
   > n11 = RigVMDispatchNode(Array=@n10.Value) {template="DISPATCH_RigVMDispatch_ArrayIterator(in Array,out Element,out Index,out Count,out Ratio)"}
   n12 = &RigUnit_GetTransform(bInitial=false, Item=@n11.Element)
   n13 = RigVMVariableNode() {variable=local_spline}
   > DISPATCH_RigVMDispatch_ArraySetAtIndex = RigVMDispatchNode(
    Index=@n11.Index
    Element=&RigUnit_PositionFromControlRigSpline(
     U=&RigUnit_ClosestParameterFromControlRigSpline(Position=@n12.Transform.Translation, Spline=@n13.Value).U
     Spline=@n13.Value
    ).Position
    Array=RigVMVariableNode() {variable=aim_transforms}.Value
   ) {template="DISPATCH_RigVMDispatch_ArraySetAtIndex(io Array,in Index,in Element)"}
   n14 = RigVMDispatchNode(ExecuteContext=@RigUnit_SequenceExecution.D, Array=RigVMVariableNode() {variable=Bones}.Value) {template="DISPATCH_RigVMDispatch_ArrayIterator(in Array,out Element,out Index,out Count,out Ratio)"}
   DISPATCH_RigVMDispatch_ArrayGetAtIndex = RigVMDispatchNode(Index=@n14.Index, Array=RigVMVariableNode() {variable=aim_transforms}.Value) {template="DISPATCH_RigVMDispatch_ArrayGetAtIndex(in Array,in Index,out Element)"}
   n15 = &RigUnit_AimBoneMath(
    DebugSettings=(Scale=10,WorldOffset=(Rotation=(X=0,Y=0,Z=0,W=1),Translation=(X=0,Y=0,Z=0),Scale3D=(X=1,Y=1,Z=1)))
    Primary=(Weight=1,Axis=(X=0,Y=1,Z=0),Target=(X=1,Y=0,Z=0),Kind=Location,Space=(Type=Bone))
    Secondary=(Weight=0,Axis=(X=0,Y=0,Z=0),Target=(X=0,Y=0,Z=0),Kind=Location,Space=(Type=Bone))
    Weight=1
    Primary.Target=@DISPATCH_RigVMDispatch_ArrayGetAtIndex.Element
    InputTransform=&RigUnit_GetTransform(Item=@n14.Element).Transform
    Primary.Axis=RigVMVariableNode() {variable=Up Axis}.Value
   )
   MathIntEquals = RigVMDispatchNode(A=@n14.Index, B=&RigVMFunction_MathIntSub(B=1, A=@n14.Count) {template="Subtract::Execute(in A,in B,out Result)"}.Result) {template="DISPATCH_RigVMDispatch_CoreEquals(in A,in B,out Result)"}
   > n16 = &RigUnit_SetTransform(Weight=1, bInitial=true, Value=@n15.Result, Item=@n14.Element, bPropagateToChildren=@MathIntEquals.Result)
   BranchNode_3_RigVMUnitNode = &RigVMFunction_ControlFlowBranch(ExecuteContext=@n16.ExecutePin, Condition=RigVMVariableNode() {variable=debug}.Value)
   &RigUnit_DebugTransformMutableItemSpace(
    Color=(R=1,G=0,B=0,A=1)
    DebugDrawSettings=(DepthPriority=SDPG_Foreground,Lifetime=-1)
    Mode=Axes
    Scale=10
    Space=(Type=Bone)
    Thickness=0.15
    WorldOffset=(Rotation=(X=0,Y=0,Z=0,W=1),Translation=(X=0,Y=0,Z=0),Scale3D=(X=1,Y=1,Z=1))
    bEnabled=True
    Transform=@n15.Result
    ExecutePin=@BranchNode_3_RigVMUnitNode.True
   )
   Branch_1_RigVMUnitNode = &RigVMFunction_ControlFlowBranch(ExecuteContext=@RigUnit_SequenceExecution_1.C, Condition=RigVMVariableNode() {variable=Stretch}.Value)
   n17 = RigVMDispatchNode(Index=-1, Array=RigVMVariableNode() {variable=Controls}.Value) {template="DISPATCH_RigVMDispatch_ArrayGetAtIndex(in Array,in Index,out Element)"}
   n18 = RigVMDispatchNode(Index=-1, Array=RigVMVariableNode() {variable=Bones}.Value) {template="DISPATCH_RigVMDispatch_ArrayGetAtIndex(in Array,in Index,out Element)"}
   &RigUnit_ParentConstraint(
    AdvancedSettings=(RotationOrderForFilter=XZY)
    Filter=(TranslationFilter=(bX=false,bY=false,bZ=false),RotationFilter=(bX=True,bY=True,bZ=True),ScaleFilter=(bX=True,bY=True,bZ=True))
    Parents=((Item=(Type=Bone),Weight=1))
    Weight=1
    Parents.0.Item=@n17.Element
    Child=@n18.Element
    ExecutePin=@Branch_1_RigVMUnitNode.False
   )
   &RigUnit_ParentConstraint(
    AdvancedSettings=(RotationOrderForFilter=XZY)
    Parents=((Item=(Type=Bone),Weight=1))
    Weight=1
    Parents.0.Item=@n17.Element
    Child=@n18.Element
    ExecutePin=@Branch_1_RigVMUnitNode.True
   )
   /Script/ControlRig.RigUnit_SetTransformItemArray(
    ExecutePin=()
    Space=GlobalSpace
    Weight=1
    bInitial=true
    bPropagateToChildren=True
    Items=@n1.Bones
    Transforms=RigVMVariableNode() {variable=Original Positions}.Value
   ) {is=struct RigVMUnitNode}
  ]}
 ContainedGraph {nodes [
   n1 = RigVMFunctionEntryNode()
   RigUnit_SequenceExecution_2 = &RigVMFunction_Sequence(ExecuteContext=@n1.ExecuteContext)
   RigUnit_SequenceExecution_2_1 = &RigVMFunction_Sequence(ExecuteContext=@RigUnit_SequenceExecution_2.B)
   RigUnit_SequenceExecution_2_2 = &RigVMFunction_Sequence(ExecuteContext=@RigUnit_SequenceExecution_2_1.B)
   RigVMFunctionReturnNode(A=@RigUnit_SequenceExecution_2.A, B=@RigUnit_SequenceExecution_2_1.A, C=@RigUnit_SequenceExecution_2_2.A, D=@RigUnit_SequenceExecution_2_2.B)
  ]}
 ContainedGraph {nodes [
   n1 = RigVMFunctionEntryNode()
   RigUnit_SequenceExecution_4 = &RigVMFunction_Sequence(ExecuteContext=@n1.ExecuteContext)
   RigUnit_SequenceExecution_4_1 = &RigVMFunction_Sequence(ExecuteContext=@RigUnit_SequenceExecution_4.B)
   RigUnit_SequenceExecution_4_2 = &RigVMFunction_Sequence(ExecuteContext=@RigUnit_SequenceExecution_4_1.B)
   RigVMFunctionReturnNode(A=@RigUnit_SequenceExecution_4.A, B=@RigUnit_SequenceExecution_4_1.A, C=@RigUnit_SequenceExecution_4_2.A, D=@RigUnit_SequenceExecution_4_2.B)
  ]}
 ContainedGraph {nodes [
   n1 = RigVMFunctionEntryNode()
   RigUnit_SequenceExecution = &RigVMFunction_Sequence(ExecuteContext=@n1.ExecuteContext)
   RigUnit_SequenceExecution_1 = &RigVMFunction_Sequence(ExecuteContext=@RigUnit_SequenceExecution.B)
   RigUnit_SequenceExecution_2 = &RigVMFunction_Sequence(ExecuteContext=@RigUnit_SequenceExecution_1.B)
   RigVMFunctionReturnNode(A=@RigUnit_SequenceExecution.A, B=@RigUnit_SequenceExecution_1.A, C=@RigUnit_SequenceExecution_2.A, D=@RigUnit_SequenceExecution_2.B)
  ]}
 ContainedGraph {nodes [
   n1 = RigVMFunctionEntryNode()
   RigUnit_SequenceExecution_1 = &RigVMFunction_Sequence(ExecuteContext=@n1.ExecuteContext)
   RigUnit_SequenceExecution_1_1 = &RigVMFunction_Sequence(ExecuteContext=@RigUnit_SequenceExecution_1.B)
   RigUnit_SequenceExecution_1_2 = &RigVMFunction_Sequence(ExecuteContext=@RigUnit_SequenceExecution_1_1.B)
   RigVMFunctionReturnNode(A=@RigUnit_SequenceExecution_1.A, B=@RigUnit_SequenceExecution_1_1.A, C=@RigUnit_SequenceExecution_1_2.A, D=@RigUnit_SequenceExecution_1_2.B)
  ]}
 ContainedGraph {nodes [
   n1 = RigVMFunctionEntryNode()
   RigUnit_SequenceAggregate = &RigVMFunction_Sequence(ExecuteContext=@n1.ExecuteContext)
   RigVMVariableNode(Value=1000000000, ExecuteContext=@RigUnit_SequenceAggregate.A) {variable=closestDistanceSoFar}
   > n2 = RigVMDispatchNode(Array=@n1.Bones) {template="DISPATCH_RigVMDispatch_ArrayIterator(in Array,out Element,out Index,out Count,out Ratio)"}
   Item = /Script/ControlRig.RigUnit_Item(Item=@n1.Item) {is=struct RigVMUnitNode}
   n3 = &RigVMFunction_MathVectorDistance(
    A=&RigUnit_GetTransform(bInitial=@n1.Initial, Item=@n2.Element).Transform.Translation
    B=&RigUnit_GetTransform(bInitial=@n1.Initial, Item=@Item.Item).Transform.Translation
   )
   n4 = &RigVMFunction_ControlFlowBranch(
    ExecuteContext=@n2.ExecuteContext
    Condition=/Script/RigVM.RigVMFunction_MathFloatLessEqual(A=@n3.Result, B=RigVMVariableNode() {variable=closestDistanceSoFar}.Value) {is=struct RigVMUnitNode, template="LessEqual::Execute(in A,in B,out Result)"}.Result
   )
   RigVMVariableNode(ExecuteContext=@n4.True, Value=@n3.Result) {variable=closestDistanceSoFar}
   > RigVMVariableNode(Value=@n2.Element) {variable=closestBone}
   RigVMFunctionReturnNode(ExecuteContext=@RigUnit_SequenceAggregate.B, Value=RigVMVariableNode() {variable=closestBone}.Value)
  ]}
 ContainedGraph {nodes [
   n1 = RigVMFunctionEntryNode()
   RigUnit_SequenceAggregate = &RigVMFunction_Sequence(ExecuteContext=@n1.ExecuteContext)
   RigVMVariableNode(Value=1000000000, ExecuteContext=@RigUnit_SequenceAggregate.A) {variable=closestDistanceSoFar}
   > n2 = RigVMDispatchNode(Array=@n1.Bones) {template="DISPATCH_RigVMDispatch_ArrayIterator(in Array,out Element,out Index,out Count,out Ratio)"}
   n3 = &RigVMFunction_MathVectorDistance(
    A=&RigUnit_GetTransform(bInitial=@n1.Initial, Item=@n2.Element).Transform.Translation
    B=&RigUnit_GetTransform(bInitial=@n1.Initial, Item=@n1.Item).Transform.Translation
   )
   n4 = &RigVMFunction_ControlFlowBranch(
    ExecuteContext=@n2.ExecuteContext
    Condition=&RigVMFunction_MathFloatLessEqual(A=@n3.Result, B=RigVMVariableNode() {variable=closestDistanceSoFar}.Value) {template="LessEqual::Execute(in A,in B,out Result)"}.Result
   )
   RigVMVariableNode(ExecuteContext=@n4.True, Value=@n3.Result) {variable=closestDistanceSoFar}
   > RigVMVariableNode(Value=@n2.Element) {variable=closestBone}
   RigVMFunctionReturnNode(
    ExecuteContext=@RigUnit_SequenceAggregate.B
    Value=RigVMVariableNode() {variable=closestBone}.Value
    Distance=RigVMVariableNode() {variable=closestDistanceSoFar}.Value
   )
   Item = &RigUnit_Item(Item=@n1.Item)
  ]}
 ContainedGraph {nodes [
   RigVMFunctionEntryNode()
   n1 = RigVMVariableNode() {variable=Pole Vector Control}
   > &RigUnit_SetTransform(Value=(Rotation=(X=0,Y=0,Z=0,W=1),Translation=(X=0,Y=0,Z=0),Scale3D=(X=1,Y=1,Z=1)), Weight=1, Item=@n1.Value)
   n2 = RigVMDispatchNode(Index=2, Array=RigVMVariableNode() {variable=Bones}.Value) {template="DISPATCH_RigVMDispatch_ArrayGetAtIndex(in Array,in Index,out Element)"}
   n3 = RigVMVariableNode() {variable=IK Control}
   > &RigUnit_SetTransform(
    Weight=1
    Value=/Script/ControlRig.RigUnit_ProjectTransformToNewParent(bChildInitial=True, bOldParentInitial=True, NewParent=@n2.Element, OldParent=@n2.Element, Child=@n3.Value) {is=struct RigVMUnitNode}.Transform
    Item=@n3.Value
   )
   > RigVMFunctionReturnNode()
   n4 = RigVMVariableNode() {variable=Bones}
   DISPATCH_RigVMDispatch_ArrayGetAtIndex = RigVMDispatchNode(Index=0, Array=@n4.Value) {template="DISPATCH_RigVMDispatch_ArrayGetAtIndex(in Array,in Index,out Element)"}
   RigVMDispatchNode(Index=1, Array=@n4.Value) {template="DISPATCH_RigVMDispatch_ArrayGetAtIndex(in Array,in Index,out Element)"}
   RigVMDispatchNode(Index=2, Array=@n4.Value) {template="DISPATCH_RigVMDispatch_ArrayGetAtIndex(in Array,in Index,out Element)"}
   n5 = RigVMDispatchNode(Index=0, Array=RigVMVariableNode() {variable=Bones}.Value) {template="DISPATCH_RigVMDispatch_ArrayGetAtIndex(in Array,in Index,out Element)"}
   &RigUnit_ProjectTransformToNewParent(bChildInitial=True, bOldParentInitial=True, NewParent=@n5.Element, OldParent=@n5.Element, Child=@n1.Value)
  ]}
 Finalize Controls_1_ContainedGraph {nodes [
   n1 = RigVMFunctionEntryNode()
   n2 = /Script/ControlRig.RigUnit_FindItemsWithMetadataTag(NameSpace=Self, Tag=Controls) {is=struct RigVMUnitNode}
   > Scale Module Controls = RigVMFunctionReferenceNode(
    Global Scale=RigVMDispatchNode(Default=1, Name=Control Size, NameSpace=Root) {template="DISPATCH_RigDispatch_GetModuleMetadata(in Name,in NameSpace,in Default,out Value,out Found)"}.Value
    Local Scale=@n1.Module Settings.ControlSize_12_#4
    Controls=@n2.Items
   )
   > Color Module Controls = RigVMFunctionReferenceNode(
    CenterColor=@n1.Module Settings.CenterColor_9_#5
    LeftColor=@n1.Module Settings.LeftColor_3_#6
    Override=@n1.Module Settings.OverrideColor_48_#7
    RightColor=@n1.Module Settings.RightColor_5_#8
    Controls=@n2.Items
   )
   > RigVMFunctionReturnNode()
  ]}
 New Function_1_ContainedGraph {nodes [
   n1 = RigVMFunctionEntryNode()
   RigVMFunction_Sequence = &RigVMFunction_Sequence(ExecuteContext=@n1.ExecuteContext)
   n2 = /Script/ControlRig.RigUnit_GetCandidates() {is=struct RigVMUnitNode}
   n3 = &RigVMFunction_ControlFlowBranch(
    Condition=RigVMDispatchNode(B=@n1.Connector, A=@n2.Connector) {template="DISPATCH_RigVMDispatch_CoreEquals(in A,in B,out Result)"}.Result
    ExecuteContext=@RigVMFunction_Sequence.A
   )
   RigVMDispatchNode(Array=@n2.Candidates, ExecuteContext=@n3.True) {template="DISPATCH_RigVMDispatch_ArrayIterator(in Array,out Element,out Index,out Count,out Ratio)"}
   n4 = /Script/ControlRig.RigUnit_CollectionChildrenArray(
    TypeToSearch=Bone
    bDefaultChildren=True
    bIncludeParent=true
    bRecursive=true
    Parent=/Script/ControlRig.RigUnit_ResolveConnector(SkipSocket=true, Connector=@n1.Primary Connector) {is=struct RigVMUnitNode}.Result
   ) {is=struct RigVMUnitNode}
   > Clamp Array = RigVMFunctionReferenceNode(Array=@n4.Items, Max=@n1.Clamp)
   > /Script/ControlRig.RigUnit_SetDefaultMatch(Default=RigVMDispatchNode(Index=-1, Array=@n4.Items) {template="DISPATCH_RigVMDispatch_ArrayGetAtIndex(in Array,in Index,out Element)"}.Element) {is=struct RigVMUnitNode, template="Set Default Match::Execute(in Default)"}
   RigVMFunctionReturnNode(ExecuteContext=@RigVMFunction_Sequence.B)
  ]}
 New Function_ContainedGraph {nodes [
   n1 = RigVMFunctionEntryNode()
   n2 = &RigUnit_FindItemsWithMetadataTag(NameSpace=Self, Tag=Controls)
   > Scale Module Controls = RigVMFunctionReferenceNode(
    Global Scale=RigVMDispatchNode(Default=1, Name=Control Size, NameSpace=Root) {template="DISPATCH_RigDispatch_GetModuleMetadata(in Name,in NameSpace,in Default,out Value,out Found)"}.Value
    Local Scale=@n1.Module Settings.ControlSize_12_#4
    Controls=@n2.Items
   )
   > Color Module Controls = RigVMFunctionReferenceNode(
    CenterColor=@n1.Module Settings.CenterColor_9_#5
    LeftColor=@n1.Module Settings.LeftColor_3_#6
    Override=@n1.Module Settings.OverrideColor_48_#7
    RightColor=@n1.Module Settings.RightColor_5_#8
    Controls=@n2.Items
   )
   > RigVMFunctionReturnNode()
  ]}
 New Function_ContainedGraph {nodes [
   n1 = &RigUnit_ResolveConnector(Connector=(Type=Connector,Name="Start Bone"), SkipSocket=False)
   RigVMDispatchNode(ExecuteContext=(), Name=Start Bone, NameSpace=Self, Value=@n1.Result) {template="DISPATCH_RigDispatch_SetModuleMetadata(in Name,in NameSpace,in Value,out Success)"}
   n2 = &RigUnit_ResolveConnector(Connector=(Type=Connector,Name="End Bone"), SkipSocket=False)
   > DISPATCH_RigDispatch_SetModuleMetadata = RigVMDispatchNode(Name=End Bone, NameSpace=Self, Value=@n2.Result) {template="DISPATCH_RigDispatch_SetModuleMetadata(in Name,in NameSpace,in Value,out Success)"}
   > RigVMDispatchNode(
    Name=Bones
    NameSpace=Self
    Value=/Script/ControlRig.RigUnit_HierarchyGetChainItemArray(bIncludeEnd=True, bIncludeStart=True, bReverse=false, End=@n2.Result, Start=@n1.Result) {is=struct RigVMUnitNode}.Chain
   ) {template="DISPATCH_RigDispatch_SetModuleMetadata(in Name,in NameSpace,in Value,out Success)"}
   Get_Module_Metadata = RigVMDispatchNode(Default=(), Name=Bones, NameSpace=Self) {template="DISPATCH_RigDispatch_GetModuleMetadata(in Name,in NameSpace,in Default,out Value,out Found)"}
   n3 = RigVMVariableNode() {variable=Module Settings}
   > Create Bone Chain = RigVMFunctionReferenceNode(
    Debug=true
    Debug Scale=15
    Debug Thickness=0.5
    Kind=Location
    Optional Suffix=rig
    Transforms=&RigUnit_GetTransformItemArray(Space=GlobalSpace, bInitial=False, Items=@Get_Module_Metadata.Value).Transforms
    Parent=/Script/ControlRig.RigUnit_HierarchyGetParent(
     bDefaultParent=True
     Child=RigVMDispatchNode(Index=0, Array=@Get_Module_Metadata.Value) {template="DISPATCH_RigVMDispatch_ArrayGetAtIndex(in Array,in Index,out Element)"}.Element
    ) {is=struct RigVMUnitNode}.Parent
    Name=@n3.Value.Prefix_20_#2
    Primary Axis=@n3.Value.PrimaryBoneAxis_32_#3
    Secondary Axis=@n3.Value.SecondaryAxis_51_#9
    Enable Orient Bones=RigVMVariableNode() {variable=Enable Orient Bones}.Value
   )
   > RigVMDispatchNode(Name=Rig Bones, NameSpace=Self, Value=@"Create Bone Chain".Bones) {template="DISPATCH_RigDispatch_SetModuleMetadata(in Name,in NameSpace,in Value,out Success)"}
   n4 = RigVMFunctionEntryNode()
   RigVMFunction_Sequence = &RigVMFunction_Sequence(ExecuteContext=@n4.ExecuteContext)
   RigVMVariableNode(Value=@n4.Axis, ExecuteContext=@RigVMFunction_Sequence.A) {variable=Bone Axis}
   > RigVMFunction_Sequence_1 = RigVMAggregateNode()
   n5 = &RigVMFunction_ControlFlowBranch(ExecuteContext=@RigVMFunction_Sequence_1.B, Condition=RigVMVariableNode() {variable=Create Control Per Bone}.Value)
   RigVMFunction_Sequence_2 = RigVMAggregateNode(ExecuteContext=@n5.False)
   LessEqual = /Script/RigVM.RigVMFunction_MathIntLessEqual(
    B=3
    A=RigVMDispatchNode(Array=RigVMDispatchNode(Default=(), Name=Bones, NameSpace=Self) {template="DISPATCH_RigDispatch_GetModuleMetadata(in Name,in NameSpace,in Default,out Value,out Found)"}.Value) {template="DISPATCH_RigVMDispatch_ArrayGetNum(in Array,out Num)"}.Num
   ) {is=struct RigVMUnitNode, template="LessEqual::Execute(in A,in B,out Result)"}
   n6 = &RigVMFunction_ControlFlowBranch(Condition=@LessEqual.Result, ExecuteContext=@RigVMFunction_Sequence_2.A)
   SplineFromItems_1 = RigVMFunctionReferenceNode(Samples Per Segment=16, Spline Mode=Hermite, ExecuteContext=@n6.Completed, Items=RigVMVariableNode() {variable=Creation Spline Bones}.Value)
   > n7 = RigVMVariableNode(Value=@SplineFromItems_1.Spline) {variable=Creation Spline}
   n8 = &RigVMFunction_ControlFlowBranch(Condition=RigVMVariableNode() {variable=Debug}.Value, ExecuteContext=@n7.ExecuteContext)
   &RigUnit_DrawControlRigSpline(Color=(R=1,G=0,B=0,A=1), Detail=16, Thickness=1, ExecutePin=@n8.True, Spline=@SplineFromItems_1.Spline)
   RigVMVariableNode(
    Value=RigVMDispatchNode(Default=(), Name=Bones, NameSpace=Self) {template="DISPATCH_RigDispatch_GetModuleMetadata(in Name,in NameSpace,in Default,out Value,out Found)"}.Value
    ExecuteContext=@n6.False
   ) {variable=Creation Spline Bones}
   n9 = RigVMDispatchNode(Default=(), Name=Bones, NameSpace=Self) {template="DISPATCH_RigDispatch_GetModuleMetadata(in Name,in NameSpace,in Default,out Value,out Found)"}
   Num = RigVMDispatchNode(Array=@n9.Value) {template="DISPATCH_RigVMDispatch_ArrayGetNum(in Array,out Num)"}
   n10 = &RigVMFunction_ControlFlowBranch(
    Condition=RigVMDispatchNode(B=3, A=@Num.Num) {template="DISPATCH_RigVMDispatch_CoreEquals(in A,in B,out Result)"}.Result
    ExecuteContext=@n6.True
   )
   n11 = &RigVMFunction_ControlFlowBranch(
    Condition=RigVMDispatchNode(B=2, A=@Num.Num) {template="DISPATCH_RigVMDispatch_CoreEquals(in A,in B,out Result)"}.Result
    ExecuteContext=@n10.Completed
   )
   n12 = &RigVMFunction_ControlFlowBranch(
    Condition=RigVMDispatchNode(B=1, A=@Num.Num) {template="DISPATCH_RigVMDispatch_CoreEquals(in A,in B,out Result)"}.Result
    ExecuteContext=@n11.Completed
   )
   n13 = RigVMDispatchNode(Index=0, Array=@n9.Value) {template="DISPATCH_RigVMDispatch_ArrayGetAtIndex(in Array,in Index,out Element)"}
   RigVMVariableNode(
    Value=((Type=None,Name="None"),(Type=None,Name="None"),(Type=None,Name="None"),(Type=None,Name="None"))
    Value.0=@n13.Element
    Value.1=@n13.Element
    Value.2=@n13.Element
    Value.3=@n13.Element
    ExecuteContext=@n12.True
   ) {variable=Creation Spline Bones}
   n14 = RigVMDispatchNode(Index=1, Array=@n9.Value) {template="DISPATCH_RigVMDispatch_ArrayGetAtIndex(in Array,in Index,out Element)"}
   RigVMVariableNode(
    Value=((Type=None,Name="None"),(Type=None,Name="None"),(Type=None,Name="None"),(Type=None,Name="None"))
    Value.0=@n13.Element
    Value.1=@n13.Element
    Value.2=@n13.Element
    Value.3=@n14.Element
    ExecuteContext=@n11.True
   ) {variable=Creation Spline Bones}
   RigVMVariableNode(
    Value=((Type=None,Name="None"),(Type=None,Name="None"),(Type=None,Name="None"),(Type=None,Name="None"))
    Value.0=@n13.Element
    Value.1=@n13.Element
    Value.2=@n14.Element
    Value.3=RigVMDispatchNode(Index=2, Array=@n9.Value) {template="DISPATCH_RigVMDispatch_ArrayGetAtIndex(in Array,in Index,out Element)"}.Element
    ExecuteContext=@n10.True
   ) {variable=Creation Spline Bones}
   RigVMDispatchNode(ExecuteContext=@RigVMFunction_Sequence_2.B, Array=RigVMVariableNode() {variable=IK Controls}.Value) {template="DISPATCH_RigVMDispatch_ArrayReset(io Array)"}
   > RigVMDispatchNode(Array=RigVMVariableNode() {variable=U Values}.Value) {template="DISPATCH_RigVMDispatch_ArrayReset(io Array)"}
   > RigVMFunction_ForLoopCount = /Script/RigVM.RigVMFunction_ForLoopCount(Count=RigVMVariableNode() {variable=Number of Controls}.Value) {is=struct RigVMUnitNode}
   Auto Get Prefix = RigVMFunctionReferenceNode(Value=RigVMDispatchNode(
     Index=0
     Array=RigVMDispatchNode(Default=(), Name=Bones, NameSpace=Self) {template="DISPATCH_RigDispatch_GetModuleMetadata(in Name,in NameSpace,in Default,out Value,out Found)"}.Value
    ) {template="DISPATCH_RigVMDispatch_ArrayGetAtIndex(in Array,in Index,out Element)"}.Element.Name)
   n15 = RigVMVariableNode() {variable=Prefix}
   Add = /Script/RigVM.RigVMFunction_MathIntAdd(B=1, A=@RigVMFunction_ForLoopCount.Index) {is=struct RigVMUnitNode, template="Add::Execute(in A,in B,out Result)"}
   Int to Name = /Script/RigVM.RigVMFunction_MathIntToName(PaddedSize=2, Number=@Add.Result) {is=struct RigVMUnitNode, template="Int to Name::Execute(in Number,in PaddedSize,out Result)"}
   Concat_1 = RigVMAggregateNode(
    B=_
    A=RigVMDispatchNode(
     True=@"Auto Get Prefix".Result
     Condition=RigVMDispatchNode(B=None, A=@n15.Value) {template="DISPATCH_RigVMDispatch_CoreEquals(in A,in B,out Result)"}.Result
     False=RigVMVariableNode() {variable=Prefix}.Value
    ) {template="DISPATCH_RigVMDispatch_If(in Condition,in True,in False,out Result)"}.Result
    D=@"Int to Name".Result
   )
   > Control Stack at Position_1 = RigVMFunctionReferenceNode(
    Add Bottom Null=False
    Add Null=true
    Add Secondary Control=False
    Null Suffix=_null
    Secondary Settings="(DisplayName=\"None\",Shape=(bVisible=True,Name=\"Default\",Color=(R=1,G=0,B=0,A=1),Transform=(Rotation=(X=0,Y=0,Z=0,W=1),Translation=(X=0,Y=0,Z=0),Scale3D=(X=1,Y=1,Z=1))),Proxy=(bIsProxy=False,ShapeVisibility=BasedOnSelection),InitialSpace=LocalSpace,bUsePref ...(truncated)"
    Tag Controls=true
    Zero Controls=true
    Name=@Concat_1.Result
    Settings=@n4.Local IK Control Settings
    Parent=&RigUnit_ResolveConnector(Connector=(Type=Connector,Name="Parent Control"), SkipSocket=true).Result
    Global Position=/Script/ControlRigSpline.RigUnit_TransformFromControlRigSpline2(
     PrimaryAxis=(X=1,Y=0,Z=0)
     SecondaryAxis=(X=0,Y=0,Z=0)
     U=@RigVMFunction_ForLoopCount.Ratio
     Spline=RigVMVariableNode() {variable=Creation Spline}.Value
    ) {is=struct RigVMUnitNode}.Transform
    Control Suffix=RigVMVariableNode() {variable=Suffix}.Value
   )
   > RigVMDispatchNode(Name=FK Parent, NameSpace=Self, Value=@"Control Stack at Position_1".Control) {template="DISPATCH_RigDispatch_SetModuleMetadata(in Name,in NameSpace,in Value,out Success)"}
   > RigVMDispatchNode(Element=@RigVMFunction_ForLoopCount.Ratio, Array=RigVMVariableNode() {variable=U Values}.Value) {template="DISPATCH_RigVMDispatch_ArrayAdd(io Array,in Element,out Index)"}
   > RigVMDispatchNode(Element=@"Control Stack at Position_1".Control, Array=RigVMVariableNode() {variable=IK Controls}.Value) {template="DISPATCH_RigVMDispatch_ArrayAdd(io Array,in Element,out Index)"}
   RigVMDispatchNode(Name=U Values, NameSpace=Self, ExecuteContext=@RigVMFunction_ForLoopCount.Completed, Value=RigVMVariableNode() {variable=U Values}.Value) {template="DISPATCH_RigDispatch_SetModuleMetadata(in Name,in NameSpace,in Value,out Success)"}
   > Set_Module_Metadata = RigVMDispatchNode(Name=IK Controls, NameSpace=Self, Value=RigVMVariableNode() {variable=IK Controls}.Value) {template="DISPATCH_RigDispatch_SetModuleMetadata(in Name,in NameSpace,in Value,out Success)"}
   > RigVMDispatchNode(Array=RigVMVariableNode() {variable=Main Controls}.Value) {template="DISPATCH_RigVMDispatch_ArrayReset(io Array)"}
   > n16 = &RigVMFunction_ForLoopCount(Count=@n4.Number of Main Controls)
   Auto Get Prefix_1 = RigVMFunctionReferenceNode(Value=RigVMDispatchNode(
     Index=0
     Array=RigVMDispatchNode(Default=(), Name=Bones, NameSpace=Self) {template="DISPATCH_RigDispatch_GetModuleMetadata(in Name,in NameSpace,in Default,out Value,out Found)"}.Value
    ) {template="DISPATCH_RigVMDispatch_ArrayGetAtIndex(in Array,in Index,out Element)"}.Element.Name)
   n17 = RigVMVariableNode() {variable=Prefix}
   Int to Name_2 = &RigVMFunction_MathIntToName(PaddedSize=2, Number=&RigVMFunction_MathIntAdd(B=1, A=@n16.Index) {template="Add::Execute(in A,in B,out Result)"}.Result) {template="Int to Name::Execute(in Number,in PaddedSize,out Result)"}
   Concat_1_2 = RigVMAggregateNode(
    B=_Main_
    C=None
    A=RigVMDispatchNode(
     True=@"Auto Get Prefix_1".Result
     Condition=RigVMDispatchNode(B=None, A=@n17.Value) {template="DISPATCH_RigVMDispatch_CoreEquals(in A,in B,out Result)"}.Result
     False=RigVMVariableNode() {variable=Prefix}.Value
    ) {template="DISPATCH_RigVMDispatch_If(in Condition,in True,in False,out Result)"}.Result
    D=@"Int to Name_2".Result
   )
   > Control Stack at Position_1_1 = RigVMFunctionReferenceNode(
    Add Bottom Null=False
    Add Null=false
    Add Secondary Control=False
    Null Suffix=_null
    Secondary Settings="(DisplayName=\"None\",Shape=(bVisible=True,Name=\"Default\",Color=(R=1,G=0,B=0,A=1),Transform=(Rotation=(X=0,Y=0,Z=0,W=1),Translation=(X=0,Y=0,Z=0),Scale3D=(X=1,Y=1,Z=1))),Proxy=(bIsProxy=False,ShapeVisibility=BasedOnSelection),InitialSpace=LocalSpace,bUsePref ...(truncated)"
    Tag Controls=true
    Zero Controls=true
    Name=@Concat_1_2.Result
    Parent=&RigUnit_ResolveConnector(Connector=(Type=Connector,Name="Parent Control"), SkipSocket=true).Result
    Global Position=&RigUnit_TransformFromControlRigSpline2(PrimaryAxis=(X=1,Y=0,Z=0), SecondaryAxis=(X=0,Y=0,Z=0), U=@n16.Ratio, Spline=RigVMVariableNode() {variable=Creation Spline}.Value).Transform
    Control Suffix=RigVMVariableNode() {variable=Suffix}.Value
    Settings=RigVMVariableNode() {variable=Main Control Settings}.Value
   )
   > RigVMDispatchNode(Element=@"Control Stack at Position_1_1".Control, Array=RigVMVariableNode() {variable=Main Controls}.Value) {template="DISPATCH_RigVMDispatch_ArrayAdd(io Array,in Element,out Index)"}
   RigVMDispatchNode(Name=Main Controls, NameSpace=Self, ExecuteContext=@n16.Completed, Value=RigVMVariableNode() {variable=Main Controls}.Value) {template="DISPATCH_RigDispatch_SetModuleMetadata(in Name,in NameSpace,in Value,out Success)"}
   RigVMFunction_ControlFlowBranch = &RigVMFunction_ControlFlowBranch(ExecuteContext=@RigVMFunction_Sequence_2.C, Condition=RigVMVariableNode() {variable=Parent Controls to each other}.Value)
   Get_Module_Metadata_1 = RigVMDispatchNode(Default=(), Name=Main Controls, NameSpace=Self) {template="DISPATCH_RigDispatch_GetModuleMetadata(in Name,in NameSpace,in Default,out Value,out Found)"}
   DISPATCH_RigVMDispatch_ArrayIterator = RigVMDispatchNode(Array=@Get_Module_Metadata_1.Value, ExecuteContext=@RigVMFunction_ControlFlowBranch.True) {template="DISPATCH_RigVMDispatch_ArrayIterator(in Array,out Element,out Index,out Count,out Ratio)"}
   Subtract = &RigVMFunction_MathIntSub(B=1, A=@DISPATCH_RigVMDispatch_ArrayIterator.Count) {template="Subtract::Execute(in A,in B,out Result)"}
   DISPATCH_RigVMDispatch_CoreEquals = RigVMDispatchNode(A=@DISPATCH_RigVMDispatch_ArrayIterator.Index, B=@Subtract.Result) {template="DISPATCH_RigVMDispatch_CoreEquals(in A,in B,out Result)"}
   RigVMFunction_MathBoolNot = &RigVMFunction_MathBoolNot(Value=@DISPATCH_RigVMDispatch_CoreEquals.Result)
   n18 = &RigVMFunction_ControlFlowBranch(ExecuteContext=@DISPATCH_RigVMDispatch_ArrayIterator.ExecuteContext, Condition=@RigVMFunction_MathBoolNot.Result)
   DISPATCH_RigVMDispatch_ArrayGetAtIndex = RigVMDispatchNode(
    Index=&RigVMFunction_MathIntAdd(B=1, A=@DISPATCH_RigVMDispatch_ArrayIterator.Index) {template="Add::Execute(in A,in B,out Result)"}.Result
    Array=@Get_Module_Metadata_1.Value
   ) {template="DISPATCH_RigVMDispatch_ArrayGetAtIndex(in Array,in Index,out Element)"}
   /Script/ControlRig.RigUnit_SetDefaultParent(Child=@DISPATCH_RigVMDispatch_ArrayGetAtIndex.Element, Parent=@DISPATCH_RigVMDispatch_ArrayIterator.Element, ExecutePin=@n18.True) {is=struct RigVMUnitNode}
   n19 = RigVMDispatchNode(
    Array=RigVMDispatchNode(Default=(), Name=Rig Bones, NameSpace=Self) {template="DISPATCH_RigDispatch_GetModuleMetadata(in Name,in NameSpace,in Default,out Value,out Found)"}.Value
    ExecuteContext=@n5.True
   ) {template="DISPATCH_RigVMDispatch_ArrayIterator(in Array,out Element,out Index,out Count,out Ratio)"}
   n20 = RigVMDispatchNode(B=0, A=@n19.Index) {template="DISPATCH_RigVMDispatch_CoreEquals(in A,in B,out Result)"}
   n21 = &RigVMFunction_ControlFlowBranch(ExecuteContext=@n19.ExecuteContext, Condition=@n20.Result)
   n22 = RigVMDispatchNode(
    Array=RigVMDispatchNode(Default=(), Name=Bones, NameSpace=Self) {template="DISPATCH_RigDispatch_GetModuleMetadata(in Name,in NameSpace,in Default,out Value,out Found)"}.Value
    Index=@n19.Index
   ) {template="DISPATCH_RigVMDispatch_ArrayGetAtIndex(in Array,in Index,out Element)"}
   Concat = &RigVMFunction_NameConcat(B=_ctrl, A=&RigUnit_HierarchyGetParent(bDefaultParent=True, Child=@n22.Element).Parent.Name) {template="Concat::Execute(in A,in B,out Result)"}
   Item = &RigUnit_Item(Item=(Type=Control,Name="None"), Item.Name=@Concat.Result)
   RigVMFunction_MathBoolOr = &RigVMFunction_MathBoolOr(A=RigVMVariableNode() {variable=Add Null Above}.Value, B=RigVMVariableNode() {variable=Add Null}.Value)
   Control Stack at Item = RigVMFunctionReferenceNode(
    Add Bottom Null=False
    Add Secondary Control=False
    Null Suffix=_null
    Secondary Settings="(DisplayName=\"None\",Shape=(bVisible=True,Name=\"Default\",Color=(R=1,G=0,B=0,A=1),Transform=(Rotation=(X=0,Y=0,Z=0,W=1),Translation=(X=0,Y=0,Z=0),Scale3D=(X=1,Y=1,Z=1))),Proxy=(bIsProxy=False,ShapeVisibility=BasedOnSelection),InitialSpace=LocalSpace,bUsePref ...(truncated)"
    Settings="(DisplayName=\"None\",Shape=(bVisible=True,Name=\"Default\",Color=(R=1,G=0,B=0,A=1),Transform=(Rotation=(X=0,Y=0,Z=0,W=1),Translation=(X=0,Y=0,Z=0),Scale3D=(X=1,Y=1,Z=1))),Proxy=(bIsProxy=False,ShapeVisibility=BasedOnSelection),InitialSpace=LocalSpace,bUsePref ...(truncated)"
    Tag Controls=true
    Zero Controls=true
    Name=@n22.Element.Name
    Item=@n19.Element
    Parent=RigVMDispatchNode(
     Condition=@n20.Result
     False=@Item.Item
     True=&RigUnit_ResolveConnector(Connector=(Type=Connector,Name="Parent Control"), SkipSocket=true).Result
    ) {template="DISPATCH_RigVMDispatch_If(in Condition,in True,in False,out Result)"}.Result
    ExecuteContext=@n21.Completed
    Add Null=@RigVMFunction_MathBoolOr.Result
    Control Suffix=RigVMVariableNode() {variable=Suffix}.Value
   )
   > RigVMDispatchNode(Name=FK Parent, NameSpace=Self, Value=@"Control Stack at Item".Control) {template="DISPATCH_RigDispatch_SetModuleMetadata(in Name,in NameSpace,in Value,out Success)"}
   RigVMVariableNode(Value=false, ExecuteContext=@n21.False) {variable=Add Null Above}
   RigVMVariableNode(Value=true, ExecuteContext=@n21.True) {variable=Add Null Above}
   n23 = &RigVMFunction_ControlFlowBranch(ExecuteContext=@RigVMFunction_Sequence_1.C, Condition=RigVMVariableNode() {variable=Dynamics}.Value)
   n24 = &RigVMFunction_Sequence(ExecuteContext=@n23.True)
   n25 = RigVMDispatchNode(Default=(Type=Bone,Name="None"), Name=Settings Control, NameSpace=Root) {template="DISPATCH_RigDispatch_GetModuleMetadata(in Name,in NameSpace,in Default,out Value,out Found)"}
   DISPATCH_RigVMDispatch_If = RigVMDispatchNode(
    Condition=@n25.Found
    True=@n25.Value
    False=RigVMDispatchNode(Index=0, Array=&RigUnit_FindItemsWithMetadataTag(NameSpace=Self, Tag=Controls).Items) {template="DISPATCH_RigVMDispatch_ArrayGetAtIndex(in Array,in Index,out Element)"}.Element
   ) {template="DISPATCH_RigVMDispatch_If(in Condition,in True,in False,out Result)"}
   RigVMDispatchNode(Name=Channel Parent, NameSpace=Self, Value=@DISPATCH_RigVMDispatch_If.Result, ExecuteContext=@n24.B) {template="DISPATCH_RigDispatch_SetModuleMetadata(in Name,in NameSpace,in Value,out Success)"}
   Format Name = RigVMFunctionReferenceNode(name=" Dynamics", side=None, suffix=None, prefix=RigVMVariableNode() {variable=Prefix}.Value)
   > SpawnAnimationChannel = /Script/ControlRig.RigUnit_HierarchyAddAnimationChannelBool(
    InitialValue=true
    LimitsEnabled=()
    MaximumValue=True
    MinimumValue=False
    Parent=@DISPATCH_RigVMDispatch_If.Result
    Name=@"Format Name".Result
   ) {
    is=struct RigVMUnitNode
    template="SpawnAnimationChannel::Execute(in InitialValue,in MinimumValue,in MaximumValue,in LimitsEnabled,in Parent,in Name,out Item)"
   }
   > RigVMDispatchNode(Name=Dynamics Channel, NameSpace=Self, Value=(Type=Control,Name="None"), Value.Name=@"Format Name".Result) {template="DISPATCH_RigDispatch_SetModuleMetadata(in Name,in NameSpace,in Value,out Success)"}
   Format Name_1 = RigVMFunctionReferenceNode(name=" Strength", side=None, suffix=None, prefix=RigVMVariableNode() {variable=Prefix}.Value)
   > /Script/ControlRig.RigUnit_HierarchyAddAnimationChannelFloat(
    LimitsEnabled=(Enabled=(bMinimum=True,bMaximum=True))
    MaximumValue=999
    MinimumValue=0
    Parent=@DISPATCH_RigVMDispatch_If.Result
    Name=@"Format Name_1".Result
    InitialValue=RigVMVariableNode() {variable=Dynamics Strength}.Value
   ) {
    is=struct RigVMUnitNode
    template="SpawnAnimationChannel::Execute(in InitialValue,in MinimumValue,in MaximumValue,in LimitsEnabled,in Parent,in Name,out Item)"
   }
   > RigVMDispatchNode(Name=Strength Channel, NameSpace=Self, Value=(Type=Control,Name="None"), Value.Name=@"Format Name_1".Result) {template="DISPATCH_RigDispatch_SetModuleMetadata(in Name,in NameSpace,in Value,out Success)"}
   Format Name_1_1 = RigVMFunctionReferenceNode(name=" Damping", side=None, suffix=None, prefix=RigVMVariableNode() {variable=Prefix}.Value)
   > &RigUnit_HierarchyAddAnimationChannelFloat(
    LimitsEnabled=(Enabled=(bMinimum=True,bMaximum=True))
    MaximumValue=999
    MinimumValue=0
    Parent=@DISPATCH_RigVMDispatch_If.Result
    Name=@"Format Name_1_1".Result
    InitialValue=RigVMVariableNode() {variable=Dynamics Damping}.Value
   ) {template="SpawnAnimationChannel::Execute(in InitialValue,in MinimumValue,in MaximumValue,in LimitsEnabled,in Parent,in Name,out Item)"}
   > RigVMDispatchNode(Name=Damping Channel, NameSpace=Self, Value=(Type=Control,Name="None"), Value.Name=@"Format Name_1_1".Result) {template="DISPATCH_RigDispatch_SetModuleMetadata(in Name,in NameSpace,in Value,out Success)"}
   Format Name_1_2 = RigVMFunctionReferenceNode(name=" Blend Start", side=None, suffix=None, prefix=RigVMVariableNode() {variable=Prefix}.Value)
   > &RigUnit_HierarchyAddAnimationChannelFloat(
    LimitsEnabled=(Enabled=(bMinimum=True,bMaximum=True))
    MaximumValue=999
    MinimumValue=0
    Parent=@DISPATCH_RigVMDispatch_If.Result
    Name=@"Format Name_1_2".Result
    InitialValue=RigVMVariableNode() {variable=Dynamics Blend Start}.Value
   ) {template="SpawnAnimationChannel::Execute(in InitialValue,in MinimumValue,in MaximumValue,in LimitsEnabled,in Parent,in Name,out Item)"}
   > RigVMDispatchNode(Name=Blend Start Channel, NameSpace=Self, Value=(Type=Bone,Name="None"), Value.Name=@"Format Name_1_2".Result) {template="DISPATCH_RigDispatch_SetModuleMetadata(in Name,in NameSpace,in Value,out Success)"}
   Format Name_1_2_1 = RigVMFunctionReferenceNode(name=" Blend End", side=None, suffix=None, prefix=RigVMVariableNode() {variable=Prefix}.Value)
   > &RigUnit_HierarchyAddAnimationChannelFloat(
    LimitsEnabled=(Enabled=(bMinimum=True,bMaximum=True))
    MaximumValue=999
    MinimumValue=0
    Parent=@DISPATCH_RigVMDispatch_If.Result
    Name=@"Format Name_1_2_1".Result
    InitialValue=RigVMVariableNode() {variable=Dynamics Blend End}.Value
   ) {template="SpawnAnimationChannel::Execute(in InitialValue,in MinimumValue,in MaximumValue,in LimitsEnabled,in Parent,in Name,out Item)"}
   > RigVMDispatchNode(Name=Blend End Channel, NameSpace=Self, Value=(Type=Bone,Name="None"), Value.Name=@"Format Name_1_2_1".Result) {template="DISPATCH_RigDispatch_SetModuleMetadata(in Name,in NameSpace,in Value,out Success)"}
   Format Name_1_2_1_1 = RigVMFunctionReferenceNode(name=" Blend Twist", side=None, suffix=None, prefix=RigVMVariableNode() {variable=Prefix}.Value)
   > &RigUnit_HierarchyAddAnimationChannelFloat(
    LimitsEnabled=(Enabled=(bMinimum=True,bMaximum=True))
    MaximumValue=999
    MinimumValue=0
    Parent=@DISPATCH_RigVMDispatch_If.Result
    Name=@"Format Name_1_2_1_1".Result
    InitialValue=RigVMVariableNode() {variable=Dynamics Blend Twist}.Value
   ) {template="SpawnAnimationChannel::Execute(in InitialValue,in MinimumValue,in MaximumValue,in LimitsEnabled,in Parent,in Name,out Item)"}
   > RigVMDispatchNode(Name=Blend Twist Channel, NameSpace=Self, Value=(Type=Bone,Name="None"), Value.Name=@"Format Name_1_2_1_1".Result) {template="DISPATCH_RigDispatch_SetModuleMetadata(in Name,in NameSpace,in Value,out Success)"}
   Create Dual Spline = RigVMFunctionReferenceNode(
    Samples Per Segment=16
    Spline Mode=BSpline
    Stretch=true
    debug=true
    Controls=RigVMDispatchNode(Default=(), Name=IK Controls, NameSpace=Self) {template="DISPATCH_RigDispatch_GetModuleMetadata(in Name,in NameSpace,in Default,out Value,out Found)"}.Value
    Bones=RigVMDispatchNode(Default=(), Name=Rig Bones, NameSpace=Self) {template="DISPATCH_RigDispatch_GetModuleMetadata(in Name,in NameSpace,in Default,out Value,out Found)"}.Value
    Primary Axis=@n4.Primary Axis
    Secondary Spline Direction=@n4.Secondary Spline Direction
    Up Axis=@n4.Up Axis
    ExecuteContext=@RigVMFunction_Sequence_1.D
   )
   > RigVMVariableNode(Value=@"Create Dual Spline".Local Spline) {variable=Local Spline}
   > RigVMVariableNode(Value=@"Create Dual Spline".Template Spline) {variable=Template Spline}
   > For_Each = RigVMDispatchNode(Array=RigVMDispatchNode(Default=(), Name=Rig Bones, NameSpace=Self) {template="DISPATCH_RigDispatch_GetModuleMetadata(in Name,in NameSpace,in Default,out Value,out Found)"}.Value) {template="DISPATCH_RigVMDispatch_ArrayIterator(in Array,out Element,out Index,out Count,out Ratio)"}
   > &RigUnit_SetTransform(
    Weight=1
    bInitial=true
    Item=@For_Each.Element
    Value=&RigUnit_GetTransform(bInitial=False, Item=@For_Each.Element).Transform
   )
   RigVMFunctionReturnNode(ExecuteContext=@RigVMFunction_Sequence.B)
   DISPATCH_RigDispatch_GetModuleMetadata = RigVMDispatchNode(Default=(), Name=Bones, NameSpace=Self) {template="DISPATCH_RigDispatch_GetModuleMetadata(in Name,in NameSpace,in Default,out Value,out Found)"}
   SplineFromItems = RigVMFunctionReferenceNode(ExecuteContext=(), Samples Per Segment=16, Spline Mode=Hermite, Items=@DISPATCH_RigDispatch_GetModuleMetadata.Value)
   > n26 = RigVMVariableNode(Value=@SplineFromItems.Spline) {variable=Creation Spline}
   n27 = &RigVMFunction_ControlFlowBranch(ExecuteContext=@n26.ExecuteContext, Condition=RigVMVariableNode() {variable=Debug}.Value)
   &RigUnit_DrawControlRigSpline(Color=(R=1,G=0,B=0,A=1), Detail=16, Thickness=1, ExecutePin=@n27.True, Spline=@SplineFromItems.Spline)
   Int to Name_1 = &RigVMFunction_MathIntToName(PaddedSize=0, Number=&RigVMFunction_MathIntAdd(B=1, A=@n19.Index) {template="Add::Execute(in A,in B,out Result)"}.Result) {template="Int to Name::Execute(in Number,in PaddedSize,out Result)"}
   Concat_1_1 = RigVMAggregateNode(B=_, C=_, D=@"Int to Name_1".Result, A=RigVMVariableNode() {variable=Prefix}.Value)
   Get_Module_Metadata_2 = RigVMDispatchNode(Default=(), Name=IK Controls, NameSpace=Self) {template="DISPATCH_RigDispatch_GetModuleMetadata(in Name,in NameSpace,in Default,out Value,out Found)"}
  ]}
 New Function_ContainedGraph {nodes [
   n1 = RigVMFunctionEntryNode()
   RigVMFunction_Sequence = &RigVMFunction_Sequence(ExecuteContext=@n1.ExecuteContext)
   RigVMFunction_Sequence_1 = RigVMAggregateNode(ExecuteContext=@RigVMFunction_Sequence.A)
   RigVMFunction_ControlFlowBranch = &RigVMFunction_ControlFlowBranch(ExecuteContext=@RigVMFunction_Sequence_1.A, Condition=RigVMVariableNode() {variable=Use SplineIK}.Value)
   n2 = RigVMDispatchNode(
    Array=RigVMDispatchNode(Default=(), Name=IK Controls, NameSpace=Self) {template="DISPATCH_RigDispatch_GetModuleMetadata(in Name,in NameSpace,in Default,out Value,out Found)"}.Value
    ExecuteContext=@RigVMFunction_ControlFlowBranch.False
   ) {template="DISPATCH_RigVMDispatch_ArrayIterator(in Array,out Element,out Index,out Count,out Ratio)"}
   DISPATCH_RigDispatch_GetMetadata = RigVMDispatchNode(Default=(Type=Bone,Name="None"), Name=Item, NameSpace=Self, Item=@n2.Element) {template="DISPATCH_RigDispatch_GetMetadata(in Item,in Name,in NameSpace,in Default,out Value,out Found)"}
   > &RigUnit_ParentConstraint(
    AdvancedSettings=(InterpolationType=Average,RotationOrderForFilter=XZY)
    Parents=((Item=(Type=Bone,Name="None"),Weight=1))
    Weight=1
    Child=@DISPATCH_RigDispatch_GetMetadata.Value
    Parents.0.Item=@n2.Element
   )
   > &RigUnit_ParentConstraint(
    AdvancedSettings=(InterpolationType=Average,RotationOrderForFilter=XZY)
    Parents=((Item=(Type=Bone,Name="None"),Weight=1))
    Weight=1
    Child=RigVMDispatchNode(Default=(Type=Bone,Name="None"), Name=Dyn Bone, NameSpace=Self, Item=@DISPATCH_RigDispatch_GetMetadata.Value) {template="DISPATCH_RigDispatch_GetMetadata(in Item,in Name,in NameSpace,in Default,out Value,out Found)"}.Value
    Parents.0.Item=@n2.Element
   )
   n3 = &RigVMFunction_ControlFlowBranch(ExecuteContext=@n2.Completed, Condition=RigVMVariableNode() {variable=Create Control per Bone}.Value)
   DISPATCH_RigVMDispatch_ArrayIterator = RigVMDispatchNode(
    Array=RigVMDispatchNode(Default=(), Name=IK Controls, NameSpace=Self) {template="DISPATCH_RigDispatch_GetModuleMetadata(in Name,in NameSpace,in Default,out Value,out Found)"}.Value
    ExecuteContext=@n3.False
   ) {template="DISPATCH_RigVMDispatch_ArrayIterator(in Array,out Element,out Index,out Count,out Ratio)"}
   > FindClosestBone Local = RigVMFunctionReferenceNode(
    Initial=true
    Bones=RigVMDispatchNode(Default=(), Name=Bones, NameSpace=Self) {template="DISPATCH_RigDispatch_GetModuleMetadata(in Name,in NameSpace,in Default,out Value,out Found)"}.Value
    Item=@DISPATCH_RigVMDispatch_ArrayIterator.Element
   )
   > Parent = RigVMFunctionReferenceNode(New Parent=@DISPATCH_RigVMDispatch_ArrayIterator.Element, Child=@"FindClosestBone Local".Value)
   Dual SplineIK = RigVMFunctionReferenceNode(
    Samples Per Segment=16
    Bones=RigVMDispatchNode(Default=(), Name=Rig Bones, NameSpace=Self) {template="DISPATCH_RigDispatch_GetModuleMetadata(in Name,in NameSpace,in Default,out Value,out Found)"}.Value
    Controls=RigVMDispatchNode(Default=(), Name=IK Controls, NameSpace=Self) {template="DISPATCH_RigDispatch_GetModuleMetadata(in Name,in NameSpace,in Default,out Value,out Found)"}.Value
    Primary Axis=@n1.Primary Axis
    Secondary Spline Direction=@n1.Secondary Spline Direction
    Spline Mode=@n1.Spline Mode
    Stretch=@n1.Stretch
    Up Axis=@n1.Up Axis
    debug=@n1.debug
    ExecuteContext=@RigVMFunction_ControlFlowBranch.True
    Aim Spline=RigVMVariableNode() {variable=Local Spline}.Value
    Template Spline=RigVMVariableNode() {variable=Template Spline}.Value
   )
   Get_Module_Metadata_2_1 = RigVMDispatchNode(Default=(), Name=IK Controls, NameSpace=Self) {template="DISPATCH_RigDispatch_GetModuleMetadata(in Name,in NameSpace,in Default,out Value,out Found)"}
   > For_Each = RigVMDispatchNode(Array=@Get_Module_Metadata_2_1.Value) {template="DISPATCH_RigVMDispatch_ArrayIterator(in Array,out Element,out Index,out Count,out Ratio)"}
   Get_Module_Metadata_2 = RigVMDispatchNode(Default=(), Name=Main Controls, NameSpace=Self) {template="DISPATCH_RigDispatch_GetModuleMetadata(in Name,in NameSpace,in Default,out Value,out Found)"}
   > Find Closest_0 = RigVMFunctionReferenceNode(Initial=true, Item=@For_Each.Element, Bones=@Get_Module_Metadata_2.Value)
   Clone = RigVMDispatchNode(Array=@Get_Module_Metadata_2.Value) {template="DISPATCH_RigVMDispatch_ArrayClone(in Array,out Clone)"}
   > n4 = RigVMDispatchNode(
    Array=@Clone.Clone
    Index=RigVMDispatchNode(Array=@Clone.Clone, Element=@"Find Closest_0".Value) {template="DISPATCH_RigVMDispatch_ArrayFind(in Array,in Element,out Index,out Success)"}.Index
   ) {template="DISPATCH_RigVMDispatch_ArrayRemove(io Array,in Index)"}
   > Find Closest_0_1 = RigVMFunctionReferenceNode(Initial=true, Item=@For_Each.Element, Bones=@n4.Array)
   n5 = &RigUnit_HierarchyGetParent(bDefaultParent=True, Child=@For_Each.Element)
   IsNearlyZero = /Script/RigVM.RigVMFunction_MathDoubleIsNearlyZero(Tolerance=0, Value=@"Find Closest_0".Distance) {is=struct RigVMUnitNode, template="IsNearlyZero::Execute(in Value,in Tolerance,out Result)"}
   Interpolate = /Script/RigVM.RigVMFunction_MathTransformLerp(
    T=RigVMDispatchNode(False=0.5, True=1, Condition=@IsNearlyZero.Result) {template="DISPATCH_RigVMDispatch_If(in Condition,in True,in False,out Result)"}.Result
    B=&RigUnit_ProjectTransformToNewParent(
     bChildInitial=True
     bNewParentInitial=False
     bOldParentInitial=True
     NewParent=@"Find Closest_0".Value
     OldParent=@"Find Closest_0".Value
     Child=@n5.Parent
    ).Transform
    A=&RigUnit_ProjectTransformToNewParent(
     bChildInitial=True
     bNewParentInitial=False
     bOldParentInitial=True
     NewParent=@"Find Closest_0_1".Value
     OldParent=@"Find Closest_0_1".Value
     Child=@n5.Parent
    ).Transform
   ) {is=struct RigVMUnitNode, template="Interpolate::Execute(in A,in B,in T,out Result)"}
   > &RigUnit_SetTransform(Weight=1, bInitial=False, Item=@n5.Parent, Value=@Interpolate.Result)
   n6 = &RigVMFunction_ControlFlowBranch(ExecuteContext=@RigVMFunction_Sequence_1.B, Condition=RigVMVariableNode() {variable=Dynamics}.Value)
   Local Chain Dynamics = RigVMFunctionReferenceNode(
    Bones=RigVMDispatchNode(Default=(), Name=Rig Bones, NameSpace=Self) {template="DISPATCH_RigDispatch_GetModuleMetadata(in Name,in NameSpace,in Default,out Value,out Found)"}.Value
    Blend End=@n1.Blend End
    Blend Start=@n1.Blend Start
    Blend Twist=@n1.Blend Twist
    Damping=@n1.Damping
    Strength=@n1.Strength
    ExecuteContext=@n6.True
    Draw Debug=RigVMVariableNode() {variable=debug}.Value
   )
   RigVMFunctionReturnNode(ExecuteContext=@RigVMFunction_Sequence.B)
   &RigUnit_FindItemsWithMetadataTag(NameSpace=Self, Tag=Controls)
  ]}
 New Function_ContainedGraph {nodes [
   RigVMFunctionEntryNode()
   > RigVMFunction_Sequence = RigVMAggregateNode()
   DISPATCH_RigVMDispatch_ArrayIterator = RigVMDispatchNode(
    Array=RigVMDispatchNode(Default=(), Name=IK Controls, NameSpace=Self) {template="DISPATCH_RigDispatch_GetModuleMetadata(in Name,in NameSpace,in Default,out Value,out Found)"}.Value
    ExecuteContext=@RigVMFunction_Sequence.A
   ) {template="DISPATCH_RigVMDispatch_ArrayIterator(in Array,out Element,out Index,out Count,out Ratio)"}
   DISPATCH_RigDispatch_GetMetadata = RigVMDispatchNode(Default=(Type=Bone,Name="None"), Name=Item, NameSpace=Self, Item=@DISPATCH_RigVMDispatch_ArrayIterator.Element) {template="DISPATCH_RigDispatch_GetMetadata(in Item,in Name,in NameSpace,in Default,out Value,out Found)"}
   RigVMFunction_ControlFlowBranch = &RigVMFunction_ControlFlowBranch(Condition=@DISPATCH_RigDispatch_GetMetadata.Found, ExecuteContext=@DISPATCH_RigVMDispatch_ArrayIterator.ExecuteContext)
   &RigUnit_SetTransform(
    Weight=1
    bInitial=False
    Item=@DISPATCH_RigVMDispatch_ArrayIterator.Element
    Value=&RigUnit_ProjectTransformToNewParent(
     bChildInitial=True
     bNewParentInitial=False
     bOldParentInitial=True
     NewParent=@DISPATCH_RigDispatch_GetMetadata.Value
     OldParent=@DISPATCH_RigDispatch_GetMetadata.Value
     Child=@DISPATCH_RigVMDispatch_ArrayIterator.Element
    ).Transform
    ExecutePin=@RigVMFunction_ControlFlowBranch.True
   )
   RigVMFunctionReturnNode(ExecuteContext=@RigVMFunction_Sequence.C)
   &RigUnit_FindItemsWithMetadataTag(NameSpace=Self, Tag=Controls)
  ]}
 New Function_ContainedGraph {nodes [
   n1 = RigVMFunctionEntryNode()
   Concat_4 = RigVMAggregateNode(B=@n1.name, A=@n1.prefix, C=@n1.side, D=@n1.suffix)
   > RigVMFunctionReturnNode(Result=@Concat_4.Result)
  ]}
 New Function_ContainedGraph {nodes [
   n1 = RigVMFunctionEntryNode()
   > For_Each_1 = RigVMDispatchNode(Array=@n1.Items) {template="DISPATCH_RigVMDispatch_ArrayIterator(in Array,out Element,out Index,out Count,out Ratio)"}
   n2 = &RigVMFunction_ControlFlowBranch(
    Condition=RigVMDispatchNode(
     A=@For_Each_1.Index
     B=&RigVMFunction_MathIntSub(B=1, A=@For_Each_1.Count) {template="Subtract::Execute(in A,in B,out Result)"}.Result
    ) {template="DISPATCH_RigVMDispatch_CoreEquals(in A,in B,out Result)"}.Result
    ExecuteContext=@For_Each_1.ExecuteContext
   )
   n3 = &RigUnit_GetTransform(bInitial=False, Item=@For_Each_1.Element)
   n4 = &RigUnit_GetTransform(
    bInitial=False
    Item=RigVMDispatchNode(
     Index=&RigVMFunction_MathIntAdd(B=1, A=@For_Each_1.Index) {template="Add::Execute(in A,in B,out Result)"}.Result
     Array=@n1.Items
    ) {template="DISPATCH_RigVMDispatch_ArrayGetAtIndex(in Array,in Index,out Element)"}.Element
   )
   RigVMFunction_DebugLineNoSpace = /Script/RigVM.RigVMFunction_DebugLineNoSpace(
    DebugDrawSettings=(DepthPriority=SDPG_Foreground,Lifetime=-1)
    Color=@n1.Color
    bEnabled=@n1.Enabled
    Thickness=@n1.Thickness
    WorldOffset=@n1.World Offset
    A=@n3.Transform.Translation
    B=@n4.Transform.Translation
    ExecutePin=@n2.False
   ) {is=struct RigVMUnitNode}
   RigVMFunctionReturnNode(ExecuteContext=@For_Each_1.Completed)
   Add = /Script/RigVM.RigVMFunction_MathVectorAdd(B=(X=0,Y=-30,Z=0), A=@n3.Transform.Translation) {is=struct RigVMUnitNode, template="Add::Execute(in A,in B,out Result)"}
   &RigVMFunction_MathVectorAdd(B=(X=0,Y=-30,Z=0), A=@n4.Transform.Translation) {template="Add::Execute(in A,in B,out Result)"}
  ]}
 New Function_ContainedGraph {nodes [
   n1 = RigVMFunctionEntryNode()
   Concat_4 = RigVMAggregateNode(B=@n1.name, A=@n1.prefix, C=@n1.side, D=@n1.suffix)
   > RigVMFunctionReturnNode(Result=@Concat_4.Result)
  ]}
 New Function_ContainedGraph {nodes [
   n1 = RigVMFunctionEntryNode()
   > RigVMDispatchNode(Array=RigVMVariableNode() {variable=Tail FK Controls}.Value) {template="DISPATCH_RigVMDispatch_ArrayReset(io Array)"}
   > n2 = RigVMDispatchNode(Array=&RigUnit_HierarchyGetChainItemArray(bIncludeEnd=True, bIncludeStart=True, bReverse=False, End=@n1.End, Start=@n1.Start).Chain) {template="DISPATCH_RigVMDispatch_ArrayIterator(in Array,out Element,out Index,out Count,out Ratio)"}
   n3 = &RigUnit_HierarchyGetParent(bDefaultParent=True, Child=@n2.Element)
   > Control Stack at Item_1 = RigVMFunctionReferenceNode(
    Add Bottom Null=False
    Add Null=true
    Add Secondary Control=False
    Control Suffix=_ctrl
    Null Suffix=_null
    Secondary Settings="(DisplayName=\"None\",Shape=(bVisible=True,Name=\"Default\",Color=(R=1,G=0,B=0,A=1),Transform=(Rotation=(X=0,Y=0,Z=0,W=1),Translation=(X=0,Y=0,Z=0),Scale3D=(X=1,Y=1,Z=1))),Proxy=(bIsProxy=False,ShapeVisibility=BasedOnSelection),InitialSpace=LocalSpace,bSelecta ...(truncated)"
    Tag Controls=true
    Zero Controls=true
    Name=&RigVMFunction_NameConcat(B=_fk, A=@n2.Element.Name) {template="Concat::Execute(in A,in B,out Result)"}.Result
    Item=@n2.Element
    Settings=@n1.Settings
    Parent=RigVMDispatchNode(
     False=(Type=Control,Name="None")
     False.Name=&RigVMFunction_NameConcat(B=_fk_ctrl, A=@n3.Parent.Name) {template="Concat::Execute(in A,in B,out Result)"}.Result
     Condition=RigVMDispatchNode(B=0, A=@n2.Index) {template="DISPATCH_RigVMDispatch_CoreEquals(in A,in B,out Result)"}.Result
     True=&RigUnit_ResolveConnector(Connector=(Type=Connector,Name="Parent Control"), SkipSocket=true).Result
    ) {template="DISPATCH_RigVMDispatch_If(in Condition,in True,in False,out Result)"}.Result
   )
   > &RigUnit_SetTransform(
    Weight=1
    bInitial=true
    Item=@"Control Stack at Item_1".Control
    Value=&RigUnit_GetTransform(bInitial=true, Item=@"Control Stack at Item_1".Control).Transform
   )
   > RigVMDispatchNode(Element=@"Control Stack at Item_1".Control, Array=RigVMVariableNode() {variable=Tail FK Controls}.Value) {template="DISPATCH_RigVMDispatch_ArrayAdd(io Array,in Element,out Index)"}
   Set_Module_Metadata = RigVMDispatchNode(Name=FK Controls, NameSpace=Self, ExecuteContext=@n2.Completed, Value=RigVMVariableNode() {variable=Tail FK Controls}.Value) {template="DISPATCH_RigDispatch_SetModuleMetadata(in Name,in NameSpace,in Value,out Success)"}
   > RigVMFunctionReturnNode(FK Controls=RigVMVariableNode() {variable=Tail FK Controls}.Value)
  ]}
 New Function_ContainedGraph {nodes [
   n1 = RigVMFunctionEntryNode()
   Format Name_2 = RigVMFunctionReferenceNode(suffix=_fk_ik_switch, name=@n1.name, prefix=@n1.prefix, side=@n1.side)
   > n2 = &RigUnit_HierarchyAddAnimationChannelBool(
    LimitsEnabled=()
    MaximumValue=True
    MinimumValue=False
    InitialValue=@n1.Initial Value
    Name=@"Format Name_2".Result
    Parent=&RigUnit_ResolveConnector(Connector=(Type=Connector,Name="Settings Control"), SkipSocket=true).Result
   ) {template="SpawnAnimationChannel::Execute(in InitialValue,in MinimumValue,in MaximumValue,in LimitsEnabled,in Parent,in Name,out Item)"}
   > RigVMDispatchNode(Name=fk_ik_switch, NameSpace=Self, Value=@n2.Item.Name) {template="DISPATCH_RigDispatch_SetModuleMetadata(in Name,in NameSpace,in Value,out Success)"}
   > RigVMFunctionReturnNode()
  ]}
 New Function_ContainedGraph {nodes [
   RigVMFunctionEntryNode()
   Get_Module_Metadata = RigVMDispatchNode(Default=(), Name=FK Controls, NameSpace=Self) {template="DISPATCH_RigDispatch_GetModuleMetadata(in Name,in NameSpace,in Default,out Value,out Found)"}
   > n1 = RigVMDispatchNode(Array=@Get_Module_Metadata.Value) {template="DISPATCH_RigVMDispatch_ArrayIterator(in Array,out Element,out Index,out Count,out Ratio)"}
   > Parent_3_1_1_1_1_1 = RigVMFunctionReferenceNode(
    Child=RigVMDispatchNode(Default=(Type=Bone,Name="None"), Name=Item, NameSpace=Self, Item=@n1.Element) {template="DISPATCH_RigDispatch_GetMetadata(in Item,in Name,in NameSpace,in Default,out Value,out Found)"}.Value
    New Parent=@n1.Element
   )
   RigVMFunctionReturnNode(ExecuteContext=@n1.Completed)
  ]}
 New Function_ContainedGraph {nodes [
   n1 = RigVMFunctionEntryNode()
   n2 = &RigUnit_ResolveConnector(Connector=(Type=Connector,Name="Start Bone"), SkipSocket=False)
   > RigVMDispatchNode(Name=Start Bone, NameSpace=Self, Value=@n2.Result) {template="DISPATCH_RigDispatch_SetModuleMetadata(in Name,in NameSpace,in Value,out Success)"}
   n3 = &RigUnit_ResolveConnector(Connector=(Type=Connector,Name="End Bone"), SkipSocket=False)
   > DISPATCH_RigDispatch_SetModuleMetadata = RigVMDispatchNode(Name=End Bone, NameSpace=Self, Value=@n3.Result) {template="DISPATCH_RigDispatch_SetModuleMetadata(in Name,in NameSpace,in Value,out Success)"}
   > RigVMDispatchNode(
    Name=Bones
    NameSpace=Self
    Value=&RigUnit_HierarchyGetChainItemArray(bIncludeEnd=True, bIncludeStart=True, bReverse=false, End=@n3.Result, Start=@n2.Result).Chain
   ) {template="DISPATCH_RigDispatch_SetModuleMetadata(in Name,in NameSpace,in Value,out Success)"}
   Get_Module_Metadata = RigVMDispatchNode(Default=(), Name=Bones, NameSpace=Self) {template="DISPATCH_RigDispatch_GetModuleMetadata(in Name,in NameSpace,in Default,out Value,out Found)"}
   n4 = RigVMVariableNode() {variable=Module Settings}
   > Create Bone Chain = RigVMFunctionReferenceNode(
    Debug=true
    Debug Scale=15
    Debug Thickness=0.5
    Kind=Location
    Optional Suffix=rig
    Enable Orient Bones=@n1.Enable Orient Bones
    Transforms=&RigUnit_GetTransformItemArray(Space=GlobalSpace, bInitial=False, Items=@Get_Module_Metadata.Value).Transforms
    Parent=&RigUnit_HierarchyGetParent(
     bDefaultParent=True
     Child=RigVMDispatchNode(Index=0, Array=@Get_Module_Metadata.Value) {template="DISPATCH_RigVMDispatch_ArrayGetAtIndex(in Array,in Index,out Element)"}.Element
    ).Parent
    Name=@n4.Value.Prefix_20_#2
    Primary Axis=@n4.Value.PrimaryBoneAxis_32_#3
    Secondary Axis=@n4.Value.SecondaryAxis_51_#9
   )
   > RigVMDispatchNode(Name=Rig Bones, NameSpace=Self, Value=@"Create Bone Chain".Bones) {template="DISPATCH_RigDispatch_SetModuleMetadata(in Name,in NameSpace,in Value,out Success)"}
   > RigVMFunctionReturnNode()
  ]}
 New Function_ContainedGraph {nodes [
   n1 = RigVMFunctionEntryNode()
   n2 = &RigVMFunction_ControlFlowBranch(Condition=@n1.Attach, ExecuteContext=@n1.ExecuteContext)
   RigVMFunctionReturnNode(ExecuteContext=@n2.Completed)
   Get_Module_Metadata = RigVMDispatchNode(Default=(), Name=Rig Bones, NameSpace=Self) {template="DISPATCH_RigDispatch_GetModuleMetadata(in Name,in NameSpace,in Default,out Value,out Found)"}
   n3 = RigVMDispatchNode(Array=@Get_Module_Metadata.Value, ExecuteContext=@n2.True) {template="DISPATCH_RigVMDispatch_ArrayIterator(in Array,out Element,out Index,out Count,out Ratio)"}
   Get_Module_Metadata_1 = RigVMDispatchNode(Default=(), Name=Bones, NameSpace=Self) {template="DISPATCH_RigDispatch_GetModuleMetadata(in Name,in NameSpace,in Default,out Value,out Found)"}
   > Parent_1 = RigVMFunctionReferenceNode(
    Child=RigVMDispatchNode(Index=@n3.Index, Array=@Get_Module_Metadata_1.Value) {template="DISPATCH_RigVMDispatch_ArrayGetAtIndex(in Array,in Index,out Element)"}.Element
    New Parent=@n3.Element
   )
  ]}
 New Function_ContainedGraph {nodes [
   n1 = RigVMFunctionEntryNode()
   DISPATCH_RigDispatch_GetModuleMetadata = RigVMDispatchNode(Default=None, NameSpace=Self, Name=@n1.Name) {template="DISPATCH_RigDispatch_GetModuleMetadata(in Name,in NameSpace,in Default,out Value,out Found)"}
   GetAnimationChannel = &RigUnit_GetBoolAnimationChannel(
    bInitial=False
    Channel=@DISPATCH_RigDispatch_GetModuleMetadata.Value
    Control=&RigUnit_ResolveConnector(SkipSocket=true, Connector=@n1.Connector).Result.Name
   ) {template="GetAnimationChannel::Execute(out Value,in Control,in Channel,in bInitial)"}
   > RigVMFunctionReturnNode(Value=@GetAnimationChannel.Value)
  ]}
 New Function_ContainedGraph {nodes [
   n1 = RigVMFunctionEntryNode()
   > n2 = RigVMDispatchNode(Array=@n1.IK Controls) {template="DISPATCH_RigVMDispatch_ArrayIterator(in Array,out Element,out Index,out Count,out Ratio)"}
   FWD Get FK or IK_1 = RigVMFunctionReferenceNode(Connector=(Type=Connector,Name="Settings Control"), Name=fk_ik_switch)
   > /Script/ControlRig.RigUnit_SetControlVisibility(Item=@n2.Element, bVisible=@"FWD Get FK or IK_1".Value) {is=struct RigVMUnitNode, template="SetControlVisibility::Execute(in Item,in Pattern,in bVisible)"}
   n3 = RigVMDispatchNode(ExecuteContext=@n2.Completed, Array=@n1.FK Controls) {template="DISPATCH_RigVMDispatch_ArrayIterator(in Array,out Element,out Index,out Count,out Ratio)"}
   FWD Get FK or IK_1_1 = RigVMFunctionReferenceNode(Connector=(Type=Connector,Name="Settings Control"), Name=fk_ik_switch)
   > &RigUnit_SetControlVisibility(Item=@n3.Element, bVisible=&RigVMFunction_MathBoolNot(Value=@"FWD Get FK or IK_1_1".Value).Result) {template="SetControlVisibility::Execute(in Item,in Pattern,in bVisible)"}
   RigVMFunctionReturnNode(ExecuteContext=@n3.Completed)
  ]}
 New Function_ContainedGraph {nodes [
   n1 = RigVMFunctionEntryNode()
   DISPATCH_RigDispatch_GetModuleMetadata = RigVMDispatchNode(Default=None, Name=fk_ik_switch, NameSpace=Self) {template="DISPATCH_RigDispatch_GetModuleMetadata(in Name,in NameSpace,in Default,out Value,out Found)"}
   > SetAnimationChannel = /Script/ControlRig.RigUnit_SetBoolAnimationChannel(
    bInitial=False
    Channel=@DISPATCH_RigDispatch_GetModuleMetadata.Value
    Value=@n1.IK
    Control=&RigUnit_ResolveConnector(Connector=(Type=Connector,Name="Settings Control"), SkipSocket=true).Result.Name
   ) {is=struct RigVMUnitNode, template="SetAnimationChannel::Execute(in Value,in Control,in Channel,in bInitial)"}
   > RigVMFunctionReturnNode()
  ]}
 RigVMFunctionLibrary {nodes [
   Auto Get Prefix = RigVMCollapseNode(Value=None)
   Connect to Leaf Bones = RigVMCollapseNode(Clamp=0, Connector=(Type=Connector,Name="End"), ExecuteContext=(), Primary Connector=(Type=Connector,Name="Finger"))
   Create Dual Spline = RigVMCollapseNode(
    Bones=()
    Controls=()
    ExecuteContext=()
    Primary Axis=(X=1,Y=0,Z=0)
    Samples Per Segment=16
    Secondary Spline Direction=(X=0,Y=5,Z=0)
    Spline Mode=BSpline
    Up Axis=(X=0,Y=1,Z=0)
   )
   Create FK IK Switch = RigVMCollapseNode(name=tail, ExecuteContext=(), Initial Value=false, prefix=None, side=None)
   Create FK Tail Controls = RigVMCollapseNode(
    End=(Type=Bone,Name="Tail15")
    ExecuteContext=()
    Settings="(DisplayName=\"None\",bSelectable=True,InitialSpace=LocalSpace,bUsePreferredRotationOrder=False,PreferredRotationOrder=YZX,Limits=(LimitTranslationX=(bMinimum=False,bMaximum=False),LimitTranslationY=(bMinimum=False,bMaximum=False),LimitTranslationZ=(bMinimum ...(truncated)"
    Start=(Type=Bone,Name="Tail")
   )
   Create IK Tail = RigVMCollapseNode(
    Axis=(X=1,Y=0,Z=0)
    Create Control Per Bone=False
    Debug=False
    Dynamics Blend End=1
    Dynamics Blend Start=1
    Dynamics Blend Twist=1
    Dynamics Damping=0.5
    Dynamics Strength=4
    Enable Orient Bones=true
    ExecuteContext=()
    Local IK Control Settings="(DisplayName=\"None\",bSelectable=True,InitialSpace=LocalSpace,bUsePreferredRotationOrder=False,PreferredRotationOrder=YZX,Limits=(LimitTranslationX=(bMinimum=False,bMaximum=False),LimitTranslationY=(bMinimum=False,bMaximum=False),LimitTranslationZ=(bMinimum ...(truncated)"
    Main Control Settings="(DisplayName=\"None\",bSelectable=True,InitialSpace=LocalSpace,bUsePreferredRotationOrder=False,PreferredRotationOrder=YZX,Limits=(LimitTranslationX=(bMinimum=False,bMaximum=False),LimitTranslationY=(bMinimum=False,bMaximum=False),LimitTranslationZ=(bMinimum ...(truncated)"
    Number of Controls=4
    Number of Main Controls=3
    Parent Controls to each other=False
    Primary Axis=(X=1,Y=0,Z=0)
    Secondary Spline Direction=(X=0,Y=5,Z=0)
    Up Axis=(X=0,Y=1,Z=0)
   )
   Create Tail Rig Bones = RigVMCollapseNode(Enable Orient Bones=true, ExecuteContext=())
   Draw Line Between Items = RigVMCollapseNode(
    Color=(R=1,G=1,B=0,A=1)
    Enabled=True
    ExecuteContext=()
    Items=()
    Thickness=0.15
    World Offset=(Rotation=(X=0,Y=0,Z=0,W=1),Translation=(X=0,Y=0,Z=0),Scale3D=(X=1,Y=1,Z=1))
   )
   Dual SplineIK = RigVMCollapseNode(
    Aim Spline=()
    Bones=()
    Controls=()
    ExecuteContext=()
    Primary Axis=(X=1,Y=0,Z=0)
    Samples Per Segment=16
    Secondary Spline Direction=(X=0,Y=5,Z=0)
    Spline Mode=BSpline
    Template Spline=()
    Up Axis=(X=0,Y=1,Z=0)
   )
   FWD Attach Bones to Rig Bones = RigVMCollapseNode(Attach=true, ExecuteContext=())
   FWD FK Tail = RigVMCollapseNode(ExecuteContext=())
   FWD FK or IK Control Vis = RigVMCollapseNode(ExecuteContext=(), FK Controls=(), IK Controls=())
   FWD Get FK or IK = RigVMCollapseNode(Connector=(Type=Connector,Name="Settings Control"), Name=fk_ik_switch)
   FWD IK Tail = RigVMCollapseNode(
    Attach Skeleton=true
    Blend End=1
    Blend Start=1
    Blend Twist=1
    Damping=0.5
    Dynamics=False
    ExecuteContext=()
    Primary Axis=(X=1,Y=0,Z=0)
    Secondary Spline Direction=(X=0,Y=5,Z=0)
    Spline Mode=BSpline
    Strength=4
    Stretch=true
    Up Axis=(X=0,Y=1,Z=0)
    Use SplineIK=False
   )
   FWD Set FK or IK = RigVMCollapseNode(ExecuteContext=(), IK=False)
   Finalize Controls = RigVMCollapseNode(
    ExecuteContext=()
    Module Settings="(OverrideColor_48_#7=False,LeftColor_3_#6=(R=0,G=0.0231,B=0.4115,A=1),RightColor_5_#8=(R=1,G=0,B=0,A=1),CenterColor_9_#5=(R=1,G=1,B=0,A= ...(truncated)"
   )
   Find Closest = RigVMCollapseNode(Bones=(), ExecuteContext=(), Initial=true, Item=())
   FindClosestBone Local = RigVMCollapseNode(Bones=(), ExecuteContext=(), Initial=true, Item=())
   Format Name = RigVMCollapseNode(name=arm, prefix=None, side=_fk_ik_switch, suffix=_fk_ik_switch)
   Format Name_1 = RigVMCollapseNode(name=arm, prefix=None, side=_fk_ik_switch, suffix=_fk_ik_switch)
   INV Tail = RigVMCollapseNode(ExecuteContext=())
   Is Right Side = RigVMCollapseNode(ExecuteContext=(), Item=(Type=Bone))
   Match 2 Bones IK onto FK = RigVMCollapseNode(Bones=(), ExecuteContext=(), IK Control=(Type=Control,Name="hand_l_ik_ctrl"), Pole Vector Control=(Type=Control,Name="arm_l_pv_ik_ctrl"))
   Set Controls Scale and Color = RigVMCollapseNode(
    ExecuteContext=()
    Module Settings="(OverrideColor_48_#7=False,LeftColor_3_#6=(R=0,G=0.0231,B=0.4115,A=1),RightColor_5_#8=(R=1,G=0,B=0,A=1),CenterColor_9_#5=(R=1,G=1,B=0,A= ...(truncated)"
   )
  ]}
 RigVMFunction_Sequence_1_1_ContainedGraph {nodes [
   n1 = RigVMFunctionEntryNode()
   n2 = &RigVMFunction_Sequence(ExecuteContext=@n1.ExecuteContext)
   n3 = &RigVMFunction_Sequence(ExecuteContext=@n2.B)
   n4 = &RigVMFunction_Sequence(ExecuteContext=@n3.B)
   n5 = &RigVMFunction_Sequence(ExecuteContext=@n4.B)
   n6 = &RigVMFunction_Sequence(ExecuteContext=@n5.B)
   RigVMFunctionReturnNode(A=@n2.A, B=@n3.A, C=@n4.A, D=@n5.A, E=@n6.A, F=@n6.B)
  ]}
 RigVMFunction_Sequence_1_1_ContainedGraph {nodes [
   n1 = RigVMFunctionEntryNode()
   n2 = &RigVMFunction_Sequence(ExecuteContext=@n1.ExecuteContext)
   n3 = &RigVMFunction_Sequence(ExecuteContext=@n2.B)
   n4 = &RigVMFunction_Sequence(ExecuteContext=@n3.B)
   RigVMFunctionReturnNode(A=@n2.A, B=@n3.A, C=@n4.A, D=@n4.B)
  ]}
 RigVMFunction_Sequence_1_ContainedGraph {nodes [
   n1 = RigVMFunctionEntryNode()
   RigVMFunction_Sequence = &RigVMFunction_Sequence(ExecuteContext=@n1.ExecuteContext)
   n2 = &RigVMFunction_Sequence(ExecuteContext=@RigVMFunction_Sequence.B)
   n3 = &RigVMFunction_Sequence(ExecuteContext=@n2.B)
   RigVMFunctionReturnNode(A=@RigVMFunction_Sequence.A, B=@n2.A, C=@n3.A, D=@n3.B)
  ]}
 RigVMFunction_Sequence_1_ContainedGraph {nodes [
   n1 = RigVMFunctionEntryNode()
   RigVMFunction_Sequence = &RigVMFunction_Sequence(ExecuteContext=@n1.ExecuteContext)
   n2 = &RigVMFunction_Sequence(ExecuteContext=@RigVMFunction_Sequence.B)
   RigVMFunctionReturnNode(A=@RigVMFunction_Sequence.A, B=@n2.A, C=@n2.B)
  ]}
 RigVMFunction_Sequence_2_1_ContainedGraph {nodes [
   n1 = RigVMFunctionEntryNode()
   n2 = &RigVMFunction_Sequence(ExecuteContext=@n1.ExecuteContext)
   n3 = &RigVMFunction_Sequence(ExecuteContext=@n2.B)
   RigVMFunctionReturnNode(A=@n2.A, B=@n3.A, C=@n3.B)
  ]}
 RigVMModel {nodes [
   RigUnit_BeginExecution = &RigUnit_BeginExecution()
   FWD Get FK or IK = RigVMFunctionReferenceNode(Connector=(Type=Connector,Name="Settings Control"), Name=fk_ik_switch)
   n1 = &RigVMFunction_ControlFlowBranch(Condition=@"FWD Get FK or IK".Value, ExecuteContext=@RigUnit_BeginExecution.ExecutePin)
   FWD Attach Bones to Rig Bones = RigVMFunctionReferenceNode(Attach=true, ExecuteContext=@n1.Completed)
   Get_Module_Metadata_2 = RigVMDispatchNode(Default=(), Name=Main Controls, NameSpace=Self) {template="DISPATCH_RigDispatch_GetModuleMetadata(in Name,in NameSpace,in Default,out Value,out Found)"}
   Clone = RigVMDispatchNode(Array=@Get_Module_Metadata_2.Value) {template="DISPATCH_RigVMDispatch_ArrayClone(in Array,out Clone)"}
   Get_Module_Metadata = RigVMDispatchNode(Default=(), Name=IK Controls, NameSpace=Self) {template="DISPATCH_RigDispatch_GetModuleMetadata(in Name,in NameSpace,in Default,out Value,out Found)"}
   > n2 = RigVMDispatchNode(Array=@Clone.Clone, Other=@Get_Module_Metadata.Value) {template="DISPATCH_RigVMDispatch_ArrayAppend(io Array,in Other)"}
   Get_Module_Metadata_1 = RigVMDispatchNode(Default=(), Name=FK Controls, NameSpace=Self) {template="DISPATCH_RigDispatch_GetModuleMetadata(in Name,in NameSpace,in Default,out Value,out Found)"}
   > FWD FK or IK Control Vis = RigVMFunctionReferenceNode(IK Controls=@n2.Array, FK Controls=@Get_Module_Metadata_1.Value)
   FWD FK Tail = RigVMFunctionReferenceNode(ExecuteContext=@n1.False)
   RigVMFunction_ControlFlowBranch = &RigVMFunction_ControlFlowBranch(ExecuteContext=@"FWD FK Tail".ExecuteContext, Condition=RigVMVariableNode() {variable=Debug}.Value)
   &RigUnit_DrawControlRigSpline(
    Color=(R=1,G=0,B=0,A=1)
    Detail=16
    Thickness=1
    ExecutePin=@RigVMFunction_ControlFlowBranch.True
    Spline=RigVMVariableNode() {variable=Local Spline}.Value
   )
   Get Dynamics Channels = RigVMCollapseNode()
   FWD IK Tail = RigVMFunctionReferenceNode(
    Attach Skeleton=true
    Create Control per Bone=false
    Dynamics=@"Get Dynamics Channels".Value
    Strength=@"Get Dynamics Channels".Value_1
    Damping=@"Get Dynamics Channels".Value_2
    Blend Start=@"Get Dynamics Channels".Value_3
    Blend End=@"Get Dynamics Channels".Value_4
    Blend Twist=@"Get Dynamics Channels".Value_5
    ExecuteContext=@n1.True
    Primary Axis=RigVMVariableNode() {variable=Module Settings}.Value.PrimaryBoneAxis_32_#3
    Use SplineIK=RigVMVariableNode() {variable=Solve with SplineIK}.Value
    Secondary Spline Direction=RigVMVariableNode() {variable=Secondary Spline Direction}.Value
    debug=RigVMVariableNode() {variable=Debug}.Value
    Spline Mode=RigVMVariableNode() {variable=Spline Mode}.Value
    Stretch=RigVMVariableNode() {variable=Stretch}.Value
    Up Axis=RigVMVariableNode() {variable=Up Axis}.Value
   )
   RigVMVariableNode() {variable=Template Spline}
  ]}
 RigVMModel Backwards Solve Graph {nodes [
   n1 = /Script/ControlRig.RigUnit_InverseExecution() {is=struct RigVMUnitNode}
   FWD Set FK or IK = RigVMFunctionReferenceNode(IK=false, ExecuteContext=@n1.ExecutePin)
   > DISPATCH_RigVMDispatch_ArrayIterator = RigVMDispatchNode(Array=RigVMDispatchNode(Default=(), Name=FK Controls, NameSpace=Self) {template="DISPATCH_RigDispatch_GetModuleMetadata(in Name,in NameSpace,in Default,out Value,out Found)"}.Value) {template="DISPATCH_RigVMDispatch_ArrayIterator(in Array,out Element,out Index,out Count,out Ratio)"}
   n2 = RigVMDispatchNode(
    Array=RigVMDispatchNode(Default=(), Name=Bones, NameSpace=Self) {template="DISPATCH_RigDispatch_GetModuleMetadata(in Name,in NameSpace,in Default,out Value,out Found)"}.Value
    Index=@DISPATCH_RigVMDispatch_ArrayIterator.Index
   ) {template="DISPATCH_RigVMDispatch_ArrayGetAtIndex(in Array,in Index,out Element)"}
   > &RigUnit_SetTransform(
    Weight=1
    bInitial=False
    Item=@DISPATCH_RigVMDispatch_ArrayIterator.Element
    Value=&RigUnit_ProjectTransformToNewParent(
     bChildInitial=True
     bNewParentInitial=False
     bOldParentInitial=True
     NewParent=@n2.Element
     OldParent=@n2.Element
     Child=@DISPATCH_RigVMDispatch_ArrayIterator.Element
    ).Transform
   )
  ]}
 RigVMModel Connector Event Graph {nodes [
   n1 = /Script/ControlRig.RigUnit_ConnectorExecution() {is=struct RigVMUnitNode}
   RigVMFunction_Sequence = RigVMAggregateNode(ExecuteContext=@n1.ExecuteContext)
   n2 = &RigUnit_GetCandidates()
   n3 = &RigVMFunction_ControlFlowBranch(
    Condition=RigVMDispatchNode(B=(Type=Connector,Name="Start Bone"), A=@n2.Connector) {template="DISPATCH_RigVMDispatch_CoreEquals(in A,in B,out Result)"}.Result
    ExecuteContext=@RigVMFunction_Sequence.A
   )
   n4 = RigVMDispatchNode(ExecuteContext=@n3.True, Array=@n2.Candidates) {template="DISPATCH_RigVMDispatch_ArrayIterator(in Array,out Element,out Index,out Count,out Ratio)"}
   n5 = &RigUnit_ResolveConnector(Connector=(Type=Connector,Name="Tail Primary"), SkipSocket=true)
   n6 = &RigUnit_CollectionChildrenArray(TypeToSearch=Bone, bDefaultChildren=True, bIncludeParent=False, bRecursive=False, Parent=@n5.Result)
   n7 = &RigVMFunction_ControlFlowBranch(
    ExecuteContext=@n4.ExecuteContext
    Condition=RigVMDispatchNode(B=0, A=RigVMDispatchNode(Array=@n6.Items) {template="DISPATCH_RigVMDispatch_ArrayGetNum(in Array,out Num)"}.Num) {template="DISPATCH_RigVMDispatch_CoreEquals(in A,in B,out Result)"}.Result
   )
   &RigUnit_SetDefaultMatch(Default=@n5.Result, ExecutePin=@n7.False) {template="Set Default Match::Execute(in Default)"}
   &RigUnit_SetDefaultMatch(Default=&RigUnit_HierarchyGetParent(bDefaultParent=True, Child=@n5.Result).Parent, ExecutePin=@n7.True) {template="Set Default Match::Execute(in Default)"}
   > DISPATCH_RigDispatch_SetModuleMetadata = RigVMDispatchNode(Name=Connected at End, NameSpace=Self, Value=true) {template="DISPATCH_RigDispatch_SetModuleMetadata(in Name,in NameSpace,in Value,out Success)"}
   n8 = &RigUnit_GetCandidates()
   n9 = &RigVMFunction_ControlFlowBranch(
    Condition=RigVMDispatchNode(B=(Type=Connector,Name="End Bone"), A=@n8.Connector) {template="DISPATCH_RigVMDispatch_CoreEquals(in A,in B,out Result)"}.Result
    ExecuteContext=@RigVMFunction_Sequence.B
   )
   n10 = RigVMDispatchNode(ExecuteContext=@n9.True, Array=@n8.Candidates) {template="DISPATCH_RigVMDispatch_ArrayIterator(in Array,out Element,out Index,out Count,out Ratio)"}
   DISPATCH_RigDispatch_GetModuleMetadata = RigVMDispatchNode(Default=False, Name=Connected at End, NameSpace=Self) {template="DISPATCH_RigDispatch_GetModuleMetadata(in Name,in NameSpace,in Default,out Value,out Found)"}
   n11 = &RigVMFunction_ControlFlowBranch(Condition=@DISPATCH_RigDispatch_GetModuleMetadata.Found, ExecuteContext=@n10.ExecuteContext)
   n12 = &RigVMFunction_ControlFlowBranch(Condition=@DISPATCH_RigDispatch_GetModuleMetadata.Value, ExecuteContext=@n11.True)
   RigVMFunction_ControlFlowBranch = &RigVMFunction_ControlFlowBranch(
    Condition=RigVMDispatchNode(A=@n10.Element, B=&RigUnit_ResolveConnector(Connector=(Type=Connector,Name="Tail Primary"), SkipSocket=true).Result) {template="DISPATCH_RigVMDispatch_CoreEquals(in A,in B,out Result)"}.Result
    ExecuteContext=@n12.True
   )
   &RigUnit_SetDefaultMatch(Default=@n10.Element, ExecutePin=@RigVMFunction_ControlFlowBranch.True) {template="Set Default Match::Execute(in Default)"}
   n13 = RigVMDispatchNode(Default=False, Name=Connected at End, NameSpace=Self) {template="DISPATCH_RigDispatch_GetModuleMetadata(in Name,in NameSpace,in Default,out Value,out Found)"}
   n14 = &RigVMFunction_ControlFlowBranch(Condition=@n13.Found, ExecuteContext=@n10.Completed)
   n15 = &RigVMFunction_ControlFlowBranch(Condition=@n13.Value, ExecuteContext=@n14.False)
   Connect to Leaf Bones = RigVMFunctionReferenceNode(Clamp=4, Connector=(Type=Connector,Name="End Bone"), Primary Connector=(Type=Connector,Name="Tail Primary"), ExecuteContext=@n15.False)
   Connect to Module Metadata_1 = RigVMFunctionReferenceNode(
    Connector=(Type=Connector,Name="Parent Control")
    Module Metadata Name=Hips Control
    Module Metadata NameSpace=Parent
    ExecuteContext=@RigVMFunction_Sequence.D
   )
   > Connect to Module Metadata_1_2 = RigVMFunctionReferenceNode(Connector=(Type=Connector,Name="Settings Control"), Module Metadata Name=Settings Control, Module Metadata NameSpace=Root)
  ]}
]
hierarchy [
 Tail Primary {type=Connector, loc=<0 0 0>, rot=<0 0 0 1>}
 Root {type=Socket, loc=<0 0 0>, rot=<0 0 0 1>}
 Start Bone {type=Connector, loc=<0 0 0>, rot=<0 0 0 1>}
 End Bone {type=Connector, loc=<0 0 0>, rot=<0 0 0 1>}
 Parent Control {type=Connector, loc=<0 0 0>, rot=<0 0 0 1>}
 Settings Control {type=Connector, loc=<0 0 0>, rot=<0 0 0 1>}
]
variables [
 Add Null {type=bool, default=False}
 Control Stack at Position_1_Settings {
  type=FRigUnit_HierarchyAddControlTransform_Settings
  default="(InitialSpace=LocalSpace,bUsePreferredRotationOrder=False,PreferredRotationOrder=YZX,Limits=(LimitTranslationX=(bMinimum=False,bMaximum=False),LimitTranslationY=(bMinimum=False,bMaximum=False),LimitTranslationZ=(bMinimum=False,bMaximum=False),LimitPitch=(b ...(truncated)"
 }
 Create Control per Bone {type=bool, default=False}
 Creation Spline Bones {type="TArray<FRigElementKey>"}
 Debug {type=bool, default=False}
 Default Blend End {type=double, default=1}
 Default Blend Start {type=double, default=0.5}
 Default Blend Twist {type=double, default=1}
 Default Damping {type=double, default=0.8}
 Default IK or FK {type=bool, default=True}
 Default Strength {type=double, default=4}
 Dynamics {type=bool, default=False}
 FK Control Settings {
  type=FRigUnit_HierarchyAddControlTransform_Settings
  default="(InitialSpace=LocalSpace,bUsePreferredRotationOrder=False,PreferredRotationOrder=YZX,Limits=(LimitTranslationX=(bMinimum=False,bMaximum=False),LimitTranslationY=(bMinimum=False,bMaximum=False),LimitTranslationZ=(bMinimum=False,bMaximum=False),LimitPitch=(b ...(truncated)"
 }
 IK Control Settings {
  type=FRigUnit_HierarchyAddControlTransform_Settings
  default="(InitialSpace=LocalSpace,bUsePreferredRotationOrder=False,PreferredRotationOrder=YZX,Limits=(LimitTranslationX=(bMinimum=False,bMaximum=False),LimitTranslationY=(bMinimum=False,bMaximum=False),LimitTranslationZ=(bMinimum=False,bMaximum=False),LimitPitch=(b ...(truncated)"
 }
 Local IK Control Settings {
  type=FRigUnit_HierarchyAddControlTransform_Settings
  default="(InitialSpace=LocalSpace,bUsePreferredRotationOrder=False,PreferredRotationOrder=YZX,Limits=(LimitTranslationX=(bMinimum=False,bMaximum=False),LimitTranslationY=(bMinimum=False,bMaximum=False),LimitTranslationZ=(bMinimum=False,bMaximum=False),LimitPitch=(b ...(truncated)"
 }
 Local Spline {type=FControlRigSpline, default=()}
 Main Controls as FK {type=bool, default=True}
 Module Settings {
  type=FUserDefinedStruct_#10
  default="(OverrideColor_48_#7=False,LeftColor_3_#6=(R=0,G=0.0231,B=0.4115,A=1),RightColor_5_#8=(R=1,G=0,B=0,A=1),CenterColor_9_#5=(R=1,G=1,B=0,A= ...(truncated)"
 }
 Name {type=FName, default=None}
 Number of Controls {type=int32, default=5}
 Number of Main Controls {type=int32, default=4}
 Prefix {type=FName, default=FOo}
 Primary Axis {type=FVector, default=(X=1,Y=0,Z=0)}
 Re-Orient Controls {type=bool, default=True}
 Secondary Spline Direction {type=FVector, default=(X=0,Y=5,Z=0)}
 Solve with SplineIK {type=bool, default=True}
 Spline Mode {type=ESplineType, default=BSpline}
 Stretch {type=bool, default=True}
 Suffix {type=FName, default=_ctrl}
 Template Spline {type=FControlRigSpline, default=()}
 Up Axis {type=FVector, default=(X=0,Y=1,Z=0)}
]
